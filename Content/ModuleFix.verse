
Mining_System<public> := module:
    Code<public> := module:
        PlayerData<public> := module:
        MiningObject<public> := module:
        HPBar<public> := module:
        Pet<public> := module:
        Skill<public> := module:
            Textures<public> := module:
        Profile<public> := module:
        Weapon<public> := module:
        GachasSystem<public> := module:
        Utility<public> := module:
        BossFlow<public> := module:
        
VerseWidget<public> := module:
VerseDevice<public> := module:
    QuestSystem<public> := module: