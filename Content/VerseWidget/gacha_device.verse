
using { /Fortnite.com/Devices }
using { /Verse.org/Simulation }
using { /Verse.org/Colors }

using { Mining_System.Code.PlayerData }
using { Mining_System.Code.Pet }
using { Mining_System.Code.GachasSystem } 

gacha_device<public> := class(creative_device):

    var WidgetPerPlayer : [player]?widget_gacha = map{}
    
    #XP付与
    @editable XP_GotPet : accolades_device = accolades_device{}
    @editable XP_GotSuperPet : accolades_device = accolades_device{}

    SaveDataManager<public> : save_data_manager = save_data_manager{}

    @editable
    PetManager<public> : pet_manager = pet_manager{}

    @editable
    QuestTrigger : trigger_device = trigger_device{}

    @editable
    YokaiAudio : audio_player_device = audio_player_device{}

    @editable
    ResourceAudio : audio_player_device = audio_player_device{}

    @editable
    FailureSFX : audio_player_device = audio_player_device{}

    @editable
    Yokai1_HUD : hud_message_device = hud_message_device{}
    @editable
    Yokai2_HUD : hud_message_device = hud_message_device{}
    @editable
    Yokai3_HUD : hud_message_device = hud_message_device{}
    @editable
    Yokai4_HUD : hud_message_device = hud_message_device{}
    @editable
    Yokai5_HUD : hud_message_device = hud_message_device{}
    @editable
    Yokai11_HUD : hud_message_device = hud_message_device{}
    @editable
    Yokai12_HUD : hud_message_device = hud_message_device{}
    @editable
    Yokai13_HUD : hud_message_device = hud_message_device{}
    @editable
    Yokai14_HUD : hud_message_device = hud_message_device{}
    @editable
    Yokai15_HUD : hud_message_device = hud_message_device{}
    @editable
    Stone1_HUD : hud_message_device = hud_message_device{}
    @editable
    Stone2_HUD : hud_message_device = hud_message_device{}
    @editable
    Stone3_HUD : hud_message_device = hud_message_device{}

    @editable
    DebugMode : logic = false

    var CurrentEgg : [player]int = map{}
    var CurrentReward : [player]gacha_system = map{}
    var CurrentResult : [player]int = map{}

    (Value:int).ToCostEggType<public>()<transacts>:cost_egg_type=
        case(Value):
            0 => cost_egg_type.Common
            1 => cost_egg_type.Uncommon
            2 => cost_egg_type.Rare
            3 => cost_egg_type.Epic
            4 => cost_egg_type.Legendary
            _ => cost_egg_type.Common

    (CostEggType:cost_egg_type).ToValue<public>()<transacts>:int=
        case(CostEggType):
            cost_egg_type.Common => 0
            cost_egg_type.Uncommon => 1
            cost_egg_type.Rare => 2
            cost_egg_type.Epic => 3
            cost_egg_type.Legendary => 4

    # ステージごとのGachaSystemを引数に入れる。
    SetGacha<public>(Player:player, GachaSystem:gacha_system):void=

        if(not CurrentEgg[Player], set CurrentEgg[Player] = 0){}
        if(set CurrentReward[Player] = GachaSystem){}

        NewUI := widget_gacha{Player := Player}
        NewUI.Init()
        NewUI.AddInteractUI()
        if:
            set WidgetPerPlayer[Player] = option{NewUI}

        UpdateUI(Player)
        spawn{AwaitSelectLeft(NewUI)}
        spawn{AwaitSelectRight(NewUI)}
        spawn{AwaitOpenGacha(NewUI)}
        spawn{AwaitAutoGacha(NewUI)}
        

    SetCostEggType(Player:player, EggType:cost_egg_type):void=
        if(set CurrentEgg[Player] = EggType.ToValue()){}

    UpdateUI<public>(Player:player):void=
        UpdateInventory(Player)
        UpdateRewards(Player)
        UpdateCost(Player)
        if(Result := CurrentResult[Player]):
            UpdateResultAsReward(Player, Result)
        else:
            UpdateResultAsReward(Player, -1)

    UpdateInventory<public>(Player:player):void=
        SilverOreCount := SaveDataManager.GetResourceStone(Player)
        GoldOreCount := SaveDataManager.GetResourcePowerupStone(Player)
        SummonStoneCommon := SaveDataManager.GetResourceSummonStoneCommon(Player)
        SummonStoneUncommon := SaveDataManager.GetResourceSummonStoneUncommon(Player)
        SummonStoneRare := SaveDataManager.GetResourceSummonStoneRare(Player)
        SummonStoneEpic := SaveDataManager.GetResourceSummonStoneEpic(Player)
        SummonStoneLegendary := SaveDataManager.GetResourceSummonStoneLegendary(Player)

        if(Widget := WidgetPerPlayer[Player]?):
            Widget.InvAValue.SetText(StringToMessage(BigNumber_FormatDisplay(SilverOreCount)))
            Widget.InvBValue.SetText(StringToMessage(BigNumber_FormatDisplay(GoldOreCount)))
            Widget.InvCValue.SetText(StringToMessage(BigNumber_FormatDisplay(SummonStoneCommon)))
            Widget.InvDValue.SetText(StringToMessage(BigNumber_FormatDisplay(SummonStoneUncommon)))
            Widget.InvEValue.SetText(StringToMessage(BigNumber_FormatDisplay(SummonStoneRare)))
            Widget.InvFValue.SetText(StringToMessage(BigNumber_FormatDisplay(SummonStoneEpic)))
            Widget.InvGValue.SetText(StringToMessage(BigNumber_FormatDisplay(SummonStoneLegendary)))

    UpdateRewards<public>(Player:player):void=
        if(CostEggType := CurrentEgg[Player], Reward := CurrentReward[Player], Widget := WidgetPerPlayer[Player]?):
            TotalPercent := Reward.GetPercentbyEgg(CostEggType.ToCostEggType())
            LowStonePercentCount := Reward.GetGachaResource(0)
            LowStonePercentPercent := TotalPercent(1) * 100.0

            MidStonePercentCount := Reward.GetGachaResource(1)
            MidStonePercentPercent := TotalPercent(2) * 100.0
            
            HighStonePercentCount := Reward.GetGachaResource(2)
            HighStonePercentPercent := TotalPercent(3) * 100.0

            YokaiPercent := 1.0 - (TotalPercent(1) + TotalPercent(2) + TotalPercent(3))

            RewardIDs := Reward.GetRewards()
            for(Index -> ID : RewardIDs):
                Print("RewardID[{Index}]: {ID}")
            for:
                ButtonIndex -> ID : RewardIDs
                PetIcon := PetManager.PetIcon[(ID, 0)]
                PetStatus := PetManager.GetPetStatusByIndex[ID]
            do:
                Print("UpdateRewards - ButtonIndex: {ButtonIndex}, ID: {ID}, PetName: {PetStatus.PetName}")

                # IDズレてるから本物のIDを設定してる
                var RealID : int = 0
                if(ID = 1):
                    set RealID = 0
                if(ID = 2):
                    set RealID = 3
                if(ID = 3):
                    set RealID = 1
                if(ID = 4):
                    set RealID = 2
                if(ID = 5):
                    set RealID = 4

                if(ID = 11):
                    set RealID = 5
                if(ID = 12):
                    set RealID = 8
                if(ID = 13):
                    set RealID = 6
                if(ID = 14):
                    set RealID = 7
                if(ID = 15):
                    set RealID = 9
            
                if(ButtonIndex = 0):
                    RewardPercent := Lerp(0.0, YokaiPercent, Reward.RewardYokaiAPercent/100.0) * 100.0
                    Widget.RewardAIcon.SetImage(PetIcon)
                    Widget.RewardAName.SetText(StringToMessage(PetStatus.PetName))
                    Widget.RewardAPercent.SetText(StringToMessage("{RewardPercent.ToPercent()}%"))
            
                    # PetStatusInfoを取得して名前とレベルを表示
                    if(PetStatusInfo := PetManager.GetPetStatus[Player, RealID]):
                        Print("GetPetStatus - ButtonIndex: {ButtonIndex}, ID: {ID}, PetName: {PetStatusInfo(0).PetName}, Level: {PetStatusInfo(1).Level}",?Duration := 10.0)
                        
                        # レベルのチェック
                        PetLevel := PetStatusInfo(1).Level
                        if(PetLevel > 0):
                            Widget.RewardALvValue.SetText(StringToMessage("{PetLevel}"))
                        else:
                            Print("PetLevel is 0 or negative, setting to 1")
                            Widget.RewardALvValue.SetText(StringToMessage("1"))
                    else:
                        Print("PetStatusInfo not found for ID: {ID}")
                        Widget.RewardALvValue.SetText(StringToMessage("0"))
            
                else if(ButtonIndex = 1):
                    RewardPercent := Lerp(0.0, YokaiPercent, Reward.RewardYokaiBPercent/100.0) * 100.0
                    Widget.RewardBIcon.SetImage(PetIcon)
                    Widget.RewardBName.SetText(StringToMessage(PetStatus.PetName))
                    Widget.RewardBPercent.SetText(StringToMessage("{RewardPercent.ToPercent()}%"))
            
                    # デバッグログを追加
                    if(PetStatusInfo := PetManager.GetPetStatus[Player, RealID]):
                        Print("PetStatusInfo - ID: {ID}, PetName: {PetStatusInfo(0).PetName}, Level: {PetStatusInfo(1).Level}",?Duration := 10.0)
                        PetLevel := PetStatusInfo(1).Level
                        if(PetLevel > 0):
                            Widget.RewardBLvValue.SetText(StringToMessage("{PetLevel}"))
                        else:
                            Widget.RewardBLvValue.SetText(StringToMessage("1"))
                    else:
                        Print("PetStatusInfo not found for ID: {ID}")
                        Widget.RewardBLvValue.SetText(StringToMessage("0"))
            
                else if(ButtonIndex = 2):
                    RewardPercent := Lerp(0.0, YokaiPercent, Reward.RewardYokaiARarePercent/100.0) * 100.0
                    Widget.RewardCIcon.SetImage(PetIcon)
                    Widget.RewardCName.SetText(StringToMessage(PetStatus.PetName))
                    Widget.RewardCPercent.SetText(StringToMessage("{RewardPercent.ToPercent()}%"))
            
                    # デバッグログを追加
                    if(PetStatusInfo := PetManager.GetPetStatus[Player, RealID]):
                        Print("PetStatusInfo - ID: {ID}, PetName: {PetStatusInfo(0).PetName}, Level: {PetStatusInfo(1).Level}",?Duration := 10.0)
                        PetLevel := PetStatusInfo(1).Level
                        if(PetLevel > 0):
                            Widget.RewardCLvValue.SetText(StringToMessage("{PetLevel}"))
                        else:
                            Widget.RewardCLvValue.SetText(StringToMessage("1"))
                    else:
                        Print("PetStatusInfo not found for ID: {ID}")
                        Widget.RewardCLvValue.SetText(StringToMessage("0"))
            
                else if(ButtonIndex = 3):
                    RewardPercent := Lerp(0.0, YokaiPercent, Reward.RewardYokaiBRarePercent/100.0) * 100.0
                    Widget.RewardDIcon.SetImage(PetIcon)
                    Widget.RewardDName.SetText(StringToMessage(PetStatus.PetName))
                    Widget.RewardDPercent.SetText(StringToMessage("{RewardPercent.ToPercent()}%"))
            
                    # デバッグログを追加
                    if(PetStatusInfo := PetManager.GetPetStatus[Player, RealID]):
                        Print("PetStatusInfo - ID: {ID}, PetName: {PetStatusInfo(0).PetName}, Level: {PetStatusInfo(1).Level}",?Duration := 10.0)
                        PetLevel := PetStatusInfo(1).Level
                        if(PetLevel > 0):
                            Widget.RewardDLvValue.SetText(StringToMessage("{PetLevel}"))
                        else:
                            Widget.RewardDLvValue.SetText(StringToMessage("1"))
                    else:
                        Print("PetStatusInfo not found for ID: {ID}")
                        Widget.RewardDLvValue.SetText(StringToMessage("0"))
            

            Widget.RewardECount.SetText(StringToMessage("x{LowStonePercentCount}"))
            Widget.RewardEPercent.SetText(StringToMessage("{LowStonePercentPercent.ToPercent()}%"))
            Widget.RewardFCount.SetText(StringToMessage("x{MidStonePercentCount}"))
            Widget.RewardFPercent.SetText(StringToMessage("{MidStonePercentPercent.ToPercent()}%"))
            Widget.RewardGCount.SetText(StringToMessage("x{HighStonePercentCount}"))
            Widget.RewardGPercent.SetText(StringToMessage("{HighStonePercentPercent.ToPercent()}%"))

    UpdateResultAsReward(Player:player, ID:int):void=
        if:
            Reward := CurrentReward[Player]
            PetID := PetManager.GetPetID[Player, ID]
            PetIcon := PetManager.PetIcon[PetID]
            PetStatus := PetManager.GetPetStatus[Player, ID] #レベルと名前, アップグレード回数
            PetGather := PetManager.GetPetGatherRate[Player, ID]
            PetStrength := PetManager.GetPetStrength[Player, ID]
            PetCriticalRate := PetManager.GetPetCriticalRate[Player, ID]
            PetCriticalDamage := PetManager.GetPetCriticalDamage[Player, ID]
            PetLastHit := PetManager.GetPetLastHit[Player, ID]
            Widget := WidgetPerPlayer[Player]?
        then:
            Widget.StatIcon.SetImage(PetIcon)
            Widget.StatName.SetText(StringToMessage(PetStatus(0).PetName))
            Widget.StatAValue.SetText(StringToMessage("{BigNumber_FormatDisplay(PetGather)}%"))
            Widget.StatBValue.SetText(StringToMessage("{BigNumber_FormatDisplay(PetStrength)}%"))
            Widget.StatCValue.SetText(StringToMessage("{BigNumber_FormatDisplay(PetCriticalRate)}%"))
            Widget.StatDValue.SetText(StringToMessage("{BigNumber_FormatDisplay(PetCriticalDamage)}%"))
            Widget.StatEValue.SetText(StringToMessage("{BigNumber_FormatDisplay(PetLastHit)}%"))
        else if(Widget := WidgetPerPlayer[Player]?):
            Widget.StatIcon.SetImage(Boss_Textures.Pet_1)
            Widget.StatName.SetText(StringToMessage("???"))
            Widget.StatAValue.SetText(StringToMessage("???"))
            Widget.StatBValue.SetText(StringToMessage("???"))
            Widget.StatCValue.SetText(StringToMessage("???"))
            Widget.StatDValue.SetText(StringToMessage("???"))
            Widget.StatEValue.SetText(StringToMessage("???"))

    UpdateCost<public>(Player:player):void=
        if:
            CostEggType := CurrentEgg[Player]
            Reward := CurrentReward[Player]
            Widget := WidgetPerPlayer[Player]?
            # ステージに基づいたコストを取得
            CostEgg := Reward.GetCostPerEgg[CostEggType.ToCostEggType()]
            CostStone := Reward.GetCostPerStone[CostEggType.ToCostEggType()]
        then:
            if:
                EggIcon := Reward.EggIcon[CostEggType.ToCostEggType()]
            then:
                Widget.CostAIcon.SetImage(EggIcon)
                Widget.CostAValue.SetText(StringToMessage("{BigNumber_FormatDisplay(CostEgg)}"))
                Widget.CostBValue.SetText(StringToMessage("{BigNumber_FormatDisplay(CostStone)}"))

            PlayerStone := SaveDataManager.GetResourceStone(Player)
            PlayerEgg  := case(CostEggType.ToCostEggType()):
                cost_egg_type.Common => SaveDataManager.GetResourceSummonStoneCommon(Player)
                cost_egg_type.Uncommon => SaveDataManager.GetResourceSummonStoneUncommon(Player)
                cost_egg_type.Rare => SaveDataManager.GetResourceSummonStoneRare(Player)
                cost_egg_type.Epic => SaveDataManager.GetResourceSummonStoneEpic(Player)
                cost_egg_type.Legendary => SaveDataManager.GetResourceSummonStoneLegendary(Player)

            if(BigNumber_GreaterThanOrEquals(PlayerStone, CostStone)?):
                Widget.CostBValue.SetTextColor(NamedColors.White)
            else:
                Widget.CostBValue.SetTextColor(NamedColors.Red)

            if(BigNumber_GreaterThanOrEquals(PlayerEgg, CostEgg)?):
                Widget.CostAValue.SetTextColor(NamedColors.White)
            else:
                Widget.CostAValue.SetTextColor(NamedColors.Red)

    AwaitSelectLeft<public>(Widget:widget_gacha)<suspends>:void=
        race:
            Widget.CloseEvent.Await()
            loop:
                Widget.SelectLeft.OnClick().Await()
                if(set CurrentEgg[Widget.Player] -= 1, not Widget.IsAuto?):
                    if(CurrentEgg[Widget.Player] < 0, set CurrentEgg[Widget.Player] = 4){}
                UpdateUI(Widget.Player)

    AwaitSelectRight<public>(Widget:widget_gacha)<suspends>:void=
        race:
            Widget.CloseEvent.Await()
            loop:
                Widget.SelectRight.OnClick().Await()
                if(set CurrentEgg[Widget.Player] += 1, not Widget.IsAuto?):
                    if(CurrentEgg[Widget.Player] > 4, set CurrentEgg[Widget.Player] = 0){}
                UpdateUI(Widget.Player)

    # ガチャを引いたとき、プレイヤーに報酬を与えるコード
    # リソースが足りない場合はガチャが引けない。
    RollGacha(Widget:widget_gacha):void=
        if(Egg := CurrentEgg[Widget.Player], Reward := CurrentReward[Widget.Player]):
            Result := Reward.Roll(Widget.Player, Egg.ToCostEggType())
            Print("AfterRoll - Result ID: {Result(1)}")

            if(Result(0) = true):

                ResourceAudio.Play(Widget.Player)

                # ここにリソースを引くコード
                if:
                    CostEgg := Reward.GetCostPerEgg[Egg.ToCostEggType()]
                    CostStone := Reward.GetCostPerStone[Egg.ToCostEggType()]
                then:
                    SaveDataManager.SubtractResourceStone(Widget.Player, CostStone)

                    Print("CostEgg : {Mining_System.Code.PlayerData.BigNumber_FormatDisplay(CostEgg)}", ?Duration := 10.0)
                    Print("CostStone : {Mining_System.Code.PlayerData.BigNumber_FormatDisplay(CostStone)}", ?Duration := 10.0)

                    if(Egg = 0):
                        SaveDataManager.SubtractResourceSummonStoneCommon(Widget.Player, CostEgg)
                        Print("Commonで引きました", ?Duration := 10.0)
                    if(Egg = 1):
                        SaveDataManager.SubtractResourceSummonStoneUncommon(Widget.Player, CostEgg)
                        Print("Uncommonで引きました", ?Duration := 10.0)
                    if(Egg = 2):
                        SaveDataManager.SubtractResourceSummonStoneRare(Widget.Player, CostEgg)
                        Print("Rareで引きました", ?Duration := 10.0)
                    if(Egg = 3):
                        SaveDataManager.SubtractResourceSummonStoneEpic(Widget.Player, CostEgg)
                        Print("Epicで引きました", ?Duration := 10.0)
                    if(Egg = 4):
                        SaveDataManager.SubtractResourceSummonStoneLegendary(Widget.Player, CostEgg)
                        Print("Legendaryで引きました", ?Duration := 10.0)

            else:
                # ここはリソースが足りなくてガチャが失敗した時に実行される
                FailureSFX.Play(Widget.Player)

            if(Result(1) > 0):
                # Quest
                QuestTrigger.Trigger(Widget.Player)
                YokaiAudio.Play(Widget.Player)
                if(set CurrentResult[Widget.Player] = Result(1)){}

                # ガチャの結果に応じてHUDを表示する
                if(Result(1) = 1):
                    XP_GotPet.Award(Widget.Player)
                    Yokai1_HUD.Show(Widget.Player)
                else if(Result(1) = 2):
                    XP_GotPet.Award(Widget.Player)
                    Yokai2_HUD.Show(Widget.Player)
                else if(Result(1) = 3):
                    XP_GotPet.Award(Widget.Player)
                    Yokai3_HUD.Show(Widget.Player)
                else if(Result(1) = 4):
                    XP_GotPet.Award(Widget.Player)
                    Yokai4_HUD.Show(Widget.Player)
                else if(Result(1) = 5):
                    XP_GotPet.Award(Widget.Player)
                    Yokai5_HUD.Show(Widget.Player)
                else if(Result(1) = 11):
                    XP_GotSuperPet.Award(Widget.Player)
                    Yokai11_HUD.Show(Widget.Player)
                else if(Result(1) = 12):
                    XP_GotSuperPet.Award(Widget.Player)
                    Yokai12_HUD.Show(Widget.Player)
                else if(Result(1) = 13):
                    XP_GotSuperPet.Award(Widget.Player)
                    Yokai13_HUD.Show(Widget.Player)
                else if(Result(1) = 14):
                    XP_GotSuperPet.Award(Widget.Player)
                    Yokai14_HUD.Show(Widget.Player)
                else if(Result(1) = 15):
                    XP_GotSuperPet.Award(Widget.Player)
                    Yokai15_HUD.Show(Widget.Player)

        UpdateUI(Widget.Player)

    AwaitOpenGacha<public>(Widget:widget_gacha)<suspends>:void=
        race:
            Widget.CloseEvent.Await()
            loop:
                Widget.SelectActionOpen.OnClick().Await()
                if(not Widget.IsAuto?):
                    RollGacha(Widget)

    AwaitAutoGacha<public>(Widget:widget_gacha)<suspends>:void=
        race:
            Widget.CloseEvent.Await()
            loop:
                Widget.SelectActionAuto.OnClick().Await()
                if(Egg := CurrentEgg[Widget.Player], Reward := CurrentReward[Widget.Player]):
                    set Widget.IsAuto = true
                    defer:
                        set Widget.IsAuto = false
                    race:
                        loop:
                            if(DebugMode = true):
                                Sleep(0.01)
                            else:
                                Sleep(0.25)
                            RollGacha(Widget)
                        Widget.SelectActionAuto.OnClick().Await()

    (Number:float).ToPercent<public>():string=
        ZeroPoint := GetZero(Number)
        ToTruncatedString(Number, ZeroPoint)

    ToTruncatedString<public>(Number:float, Decimals:int):[]char=
        var Str:[]char = ToString(Number)
        if:
            DotIndex := Str.Find['.']
            StopIndex := if (Decimals > 0) then Min(DotIndex+Decimals+1,Str.Length) else DotIndex
            set Str = Str.Slice[0, StopIndex] 
        Str

    GetZero<public>(Number:float):int=
        var Str:[]char = ToString(Number)
        var Index : int = 0
        if:
            DotIndex := Str.Find['.']
        then:
            set Index = DotIndex+1
            for(Val := DotIndex..Str.Length, Element := Str[Val]):
                if(Element <> '0'):
                    set Index = Val - DotIndex
        Index
