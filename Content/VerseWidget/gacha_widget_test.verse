using { /Fortnite.com/UI }
using { /UnrealEngine.com/Temporary/UI }
using { /Verse.org/Colors }
using { /UnrealEngine.com/Temporary/SpatialMath }
using { /Verse.org/Assets }

using {icon_set_2}
using {Boss_Textures}

widget_gacha<public> := class(widget_base):

    SelectLeft<public> : button_quiet = button_quiet{}
    SelectRight<public> : button_quiet = button_quiet{}

    SelectActionOpen<public> : button_quiet = button_quiet{}
    SelectActionAuto<public> : button_quiet = button_quiet{}

    SelectOpen<public> : button_quiet = button_quiet{}
    # SelectEvolve<public> : button_quiet = button_quiet{}
    # SelectEvolve2<public> : button_quiet = button_quiet{}
    # SelectEvolve3<public> : button_quiet = button_quiet{}
    # SelectEvolve4<public> : button_quiet = button_quiet{}

    var IsAuto<public> : logic = false

    Init<override>():void=
        (super:)Init()

        SelectLeft.SetText(StringToMessage("               <<               "))
        SelectRight.SetText(StringToMessage("               >>               "))
        SelectActionOpen.SetText(StringToMessage("                         Open                         "))
        SelectActionAuto.SetText(StringToMessage("  Auto  "))
        SelectOpen.SetText(StringToMessage("Open"))
        # SelectEvolve.SetText(StringToMessage("Evolve1"))
        # SelectEvolve2.SetText(StringToMessage("Evolve2"))
        # SelectEvolve3.SetText(StringToMessage("Evolve3"))
        # SelectEvolve4.SetText(StringToMessage("Evolve4"))

        RewardEName.SetText(StringToMessage("Upgrade Item"))
        RewardFName.SetText(StringToMessage("Upgrade Item"))
        RewardGName.SetText(StringToMessage("Upgrade Item"))
        RewardCLvTitle.SetText(StringToMessage("Lv."))
        RewardBLvTitle.SetText(StringToMessage("Lv."))
        RewardALvTitle.SetText(StringToMessage("Lv."))
        RewardDLvTitle.SetText(StringToMessage("Lv."))
        BasicBG.Title.SetText(StringToMessage("Yo-kai"))

    Background<public> : texture_block = texture_block: # Background
        DefaultImage := bg
        DefaultTint := color{R := 1.0, G := 1.0, B := 1.0}
        DefaultDesiredSize := vector2{X := 0.0,Y := 0.0}
    
    Title<public> : text_block = text_block: # Title
        DefaultTextColor := color{R := 1.0, G := 1.0, B := 1.0}
        DefaultTextOpacity := 1.0
        DefaultShadowOffset := option{vector2{X := 0.0,Y := 0.0}}
        DefaultShadowColor := color{R := 1.0, G := 1.0, B := 1.0}
        DefaultShadowOpacity := 1.0
    
    RewardFBG<public> : button_quiet = button_quiet{}
    
    RewardABG<public> : button_quiet = button_quiet{}
    
    RewardBBG<public> : button_quiet = button_quiet{}
    
    RewardDBG<public> : button_quiet = button_quiet{}
    
    RewardEBG<public> : button_quiet = button_quiet{}
    
    RewardCBG<public> : button_quiet = button_quiet{}
    
    RewardGBG<public> : button_quiet = button_quiet{}

    RewardImage<public> : texture_block = texture_block:
        DefaultImage := IC_Power
        DefaultTint := color{R := 1.0, G := 1.0, B := 1.0}
        DefaultDesiredSize := vector2{X := 50.0,Y := 50.0}
    
    InvTitle<public> : text_block = text_block: # InvTitle
        DefaultTextColor := color{R := 0.178899, G := 0.285579, B := 0.661458}
        DefaultTextOpacity := 1.000000
        DefaultShadowOffset := option{vector2{X := 0.0,Y := 0.0}}
        DefaultShadowColor := color{R := 1.0, G := 1.0, B := 1.0}
        DefaultShadowOpacity := 1.0
    
    InvAIcon<public> : texture_block = texture_block: # InvAIcon
        DefaultImage := SilverOre
        DefaultTint := color{R := 1.0, G := 1.0, B := 1.0}
        DefaultDesiredSize := vector2{X := 50.0,Y := 50.0}
    
    InvAValue<public> : text_block = text_block: # InvAValue
        DefaultTextColor := color{R := 1.0, G := 1.0, B := 1.0}
        DefaultTextOpacity := 1.0
        DefaultShadowOffset := option{vector2{X := 0.0,Y := 0.0}}
        DefaultShadowColor := color{R := 1.0, G := 1.0, B := 1.0}
        DefaultShadowOpacity := 1.0
    
    InvBIcon<public> : texture_block = texture_block: # InvBIcon
        DefaultImage := IC_Power
        DefaultTint := color{R := 1.0, G := 1.0, B := 1.0}
        DefaultDesiredSize := vector2{X := 50.0,Y := 50.0}
    
    InvBValue<public> : text_block = text_block: # InvBValue
        DefaultTextColor := color{R := 1.0, G := 1.0, B := 1.0}
        DefaultTextOpacity := 1.0
        DefaultShadowOffset := option{vector2{X := 0.0,Y := 0.0}}
        DefaultShadowColor := color{R := 1.0, G := 1.0, B := 1.0}
        DefaultShadowOpacity := 1.0
    
    InvCIcon<public> : texture_block = texture_block: # InvCIcon
        DefaultImage := SS_1
        DefaultTint := color{R := 1.0, G := 1.0, B := 1.0}
        DefaultDesiredSize := vector2{X := 50.0,Y := 50.0}
    
    InvCValue<public> : text_block = text_block: # InvCValue
        DefaultTextColor := color{R := 1.0, G := 1.0, B := 1.0}
        DefaultTextOpacity := 1.0
        DefaultShadowOffset := option{vector2{X := 0.0,Y := 0.0}}
        DefaultShadowColor := color{R := 1.0, G := 1.0, B := 1.0}
        DefaultShadowOpacity := 1.0
    
    InvDIcon<public> : texture_block = texture_block: # InvDIcon
        DefaultImage := SS_2
        DefaultTint := color{R := 1.0, G := 1.0, B := 1.0}
        DefaultDesiredSize := vector2{X := 50.0,Y := 50.0}
    
    InvDValue<public> : text_block = text_block: # InvDValue
        DefaultTextColor := color{R := 1.0, G := 1.0, B := 1.0}
        DefaultTextOpacity := 1.0
        DefaultShadowOffset := option{vector2{X := 0.0,Y := 0.0}}
        DefaultShadowColor := color{R := 1.0, G := 1.0, B := 1.0}
        DefaultShadowOpacity := 1.0
    
    InvEIcon<public> : texture_block = texture_block: # InvEIcon
        DefaultImage := SS_3
        DefaultTint := color{R := 1.0, G := 1.0, B := 1.0}
        DefaultDesiredSize := vector2{X := 50.0,Y := 50.0}
    
    InvEValue<public> : text_block = text_block: # InvEValue
        DefaultTextColor := color{R := 1.0, G := 1.0, B := 1.0}
        DefaultTextOpacity := 1.0
        DefaultShadowOffset := option{vector2{X := 0.0,Y := 0.0}}
        DefaultShadowColor := color{R := 1.0, G := 1.0, B := 1.0}
        DefaultShadowOpacity := 1.0
    
    InvFIcon<public> : texture_block = texture_block: # InvFIcon
        DefaultImage := SS_4
        DefaultTint := color{R := 1.0, G := 1.0, B := 1.0}
        DefaultDesiredSize := vector2{X := 50.0,Y := 50.0}
    
    InvFValue<public> : text_block = text_block: # InvFValue
        DefaultTextColor := color{R := 1.0, G := 1.0, B := 1.0}
        DefaultTextOpacity := 1.0
        DefaultShadowOffset := option{vector2{X := 0.0,Y := 0.0}}
        DefaultShadowColor := color{R := 1.0, G := 1.0, B := 1.0}
        DefaultShadowOpacity := 1.0
    
    InvGIcon<public> : texture_block = texture_block: # InvGIcon
        DefaultImage := SS_5
        DefaultTint := color{R := 1.0, G := 1.0, B := 1.0}
        DefaultDesiredSize := vector2{X := 50.0,Y := 50.0}
    
    InvGValue<public> : text_block = text_block: # InvGValue
        DefaultTextColor := color{R := 1.0, G := 1.0, B := 1.0}
        DefaultTextOpacity := 1.0
        DefaultShadowOffset := option{vector2{X := 0.0,Y := 0.0}}
        DefaultShadowColor := color{R := 1.0, G := 1.0, B := 1.0}
        DefaultShadowOpacity := 1.0
    
    RewardAIcon<public> : texture_block = texture_block: # RewardAIcon
        DefaultImage := koppar
        DefaultTint := color{R := 1.0, G := 1.0, B := 1.0}
        DefaultDesiredSize := vector2{X := 64.0,Y := 64.0}
    
    RewardALvTitle<public> : text_block = text_block: # RewardALvTitle
        DefaultTextColor := color{R := 1.000000, G := 0.361098, B := 0.000000}
        DefaultTextOpacity := 1.000000
        DefaultShadowOffset := option{vector2{X := 0.0,Y := 0.0}}
        DefaultShadowColor := color{R := 1.0, G := 1.0, B := 1.0}
        DefaultShadowOpacity := 1.0
    
    RewardALvValue<public> : text_block = text_block: # RewardALvValue
        DefaultTextColor := color{R := 1.0, G := 1.0, B := 1.0}
        DefaultTextOpacity := 1.0
        DefaultShadowOffset := option{vector2{X := 0.0,Y := 0.0}}
        DefaultShadowColor := color{R := 1.0, G := 1.0, B := 1.0}
        DefaultShadowOpacity := 1.0
    
    RewardAName<public> : text_block = text_block: # RewardAName
        DefaultTextColor := color{R := 1.0, G := 1.0, B := 1.0}
        DefaultTextOpacity := 1.0
        DefaultShadowOffset := option{vector2{X := 0.0,Y := 0.0}}
        DefaultShadowColor := color{R := 1.0, G := 1.0, B := 1.0}
        DefaultShadowOpacity := 1.0
    
    RewardAPercent<public> : text_block = text_block: # RewardAPercent
        DefaultTextColor := color{R := 1.0, G := 1.0, B := 1.0}
        DefaultTextOpacity := 1.0
        DefaultShadowOffset := option{vector2{X := 0.0,Y := 0.0}}
        DefaultShadowColor := color{R := 1.0, G := 1.0, B := 1.0}
        DefaultShadowOpacity := 1.0
    
    RewardBIcon<public> : texture_block = texture_block: # RewardBIcon
        DefaultImage := kachinta
        DefaultTint := color{R := 1.0, G := 1.0, B := 1.0}
        DefaultDesiredSize := vector2{X := 64.0,Y := 64.0}
    
    RewardBLvTitle<public> : text_block = text_block: # RewardBLvTitle
        DefaultTextColor := color{R := 1.000000, G := 0.361098, B := 0.000000}
        DefaultTextOpacity := 1.000000
        DefaultShadowOffset := option{vector2{X := 0.0,Y := 0.0}}
        DefaultShadowColor := color{R := 1.0, G := 1.0, B := 1.0}
        DefaultShadowOpacity := 1.0
    
    RewardBLvValue<public> : text_block = text_block: # RewardBLvValue
        DefaultTextColor := color{R := 1.0, G := 1.0, B := 1.0}
        DefaultTextOpacity := 1.0
        DefaultShadowOffset := option{vector2{X := 0.0,Y := 0.0}}
        DefaultShadowColor := color{R := 1.0, G := 1.0, B := 1.0}
        DefaultShadowOpacity := 1.0
    
    RewardBName<public> : text_block = text_block: # RewardBName
        DefaultTextColor := color{R := 1.0, G := 1.0, B := 1.0}
        DefaultTextOpacity := 1.0
        DefaultShadowOffset := option{vector2{X := 0.0,Y := 0.0}}
        DefaultShadowColor := color{R := 1.0, G := 1.0, B := 1.0}
        DefaultShadowOpacity := 1.0
    
    RewardBPercent<public> : text_block = text_block: # RewardBPercent
        DefaultTextColor := color{R := 1.0, G := 1.0, B := 1.0}
        DefaultTextOpacity := 1.0
        DefaultShadowOffset := option{vector2{X := 0.0,Y := 0.0}}
        DefaultShadowColor := color{R := 1.0, G := 1.0, B := 1.0}
        DefaultShadowOpacity := 1.0
    
    RewardCIcon<public> : texture_block = texture_block: # RewardCIcon
        DefaultImage := Pet_1
        DefaultTint := color{R := 1.0, G := 1.0, B := 1.0}
        DefaultDesiredSize := vector2{X := 64.0,Y := 64.0}
    
    RewardCLvTitle<public> : text_block = text_block: # RewardCLvTitle
        DefaultTextColor := color{R := 1.000000, G := 0.361098, B := 0.000000}
        DefaultTextOpacity := 1.000000
        DefaultShadowOffset := option{vector2{X := 0.0,Y := 0.0}}
        DefaultShadowColor := color{R := 1.0, G := 1.0, B := 1.0}
        DefaultShadowOpacity := 1.0
    
    RewardCLvValue<public> : text_block = text_block: # RewardCLvValue
        DefaultTextColor := color{R := 1.0, G := 1.0, B := 1.0}
        DefaultTextOpacity := 1.0
        DefaultShadowOffset := option{vector2{X := 0.0,Y := 0.0}}
        DefaultShadowColor := color{R := 1.0, G := 1.0, B := 1.0}
        DefaultShadowOpacity := 1.0
    
    RewardCName<public> : text_block = text_block: # RewardCName
        DefaultTextColor := color{R := 1.0, G := 1.0, B := 1.0}
        DefaultTextOpacity := 1.0
        DefaultShadowOffset := option{vector2{X := 0.0,Y := 0.0}}
        DefaultShadowColor := color{R := 1.0, G := 1.0, B := 1.0}
        DefaultShadowOpacity := 1.0
    
    RewardCPercent<public> : text_block = text_block: # RewardCPercent
        DefaultTextColor := color{R := 1.0, G := 1.0, B := 1.0}
        DefaultTextOpacity := 1.0
        DefaultShadowOffset := option{vector2{X := 0.0,Y := 0.0}}
        DefaultShadowColor := color{R := 1.0, G := 1.0, B := 1.0}
        DefaultShadowOpacity := 1.0
    
    RewardDIcon<public> : texture_block = texture_block: # RewardDIcon
        DefaultImage := _kachinta
        DefaultTint := color{R := 1.0, G := 1.0, B := 1.0}
        DefaultDesiredSize := vector2{X := 64.0,Y := 64.0}
    
    RewardDLvTitle<public> : text_block = text_block: # RewardDLvTitle
        DefaultTextColor := color{R := 1.000000, G := 0.361098, B := 0.000000}
        DefaultTextOpacity := 1.000000
        DefaultShadowOffset := option{vector2{X := 0.0,Y := 0.0}}
        DefaultShadowColor := color{R := 1.0, G := 1.0, B := 1.0}
        DefaultShadowOpacity := 1.0
    
    RewardDLvValue<public> : text_block = text_block: # RewardDLvValue
        DefaultTextColor := color{R := 1.0, G := 1.0, B := 1.0}
        DefaultTextOpacity := 1.0
        DefaultShadowOffset := option{vector2{X := 0.0,Y := 0.0}}
        DefaultShadowColor := color{R := 1.0, G := 1.0, B := 1.0}
        DefaultShadowOpacity := 1.0
    
    RewardDName<public> : text_block = text_block: # RewardDName
        DefaultTextColor := color{R := 1.0, G := 1.0, B := 1.0}
        DefaultTextOpacity := 1.0
        DefaultShadowOffset := option{vector2{X := 0.0,Y := 0.0}}
        DefaultShadowColor := color{R := 1.0, G := 1.0, B := 1.0}
        DefaultShadowOpacity := 1.0
    
    RewardDPercent<public> : text_block = text_block: # RewardDPercent
        DefaultTextColor := color{R := 1.0, G := 1.0, B := 1.0}
        DefaultTextOpacity := 1.0
        DefaultShadowOffset := option{vector2{X := 0.0,Y := 0.0}}
        DefaultShadowColor := color{R := 1.0, G := 1.0, B := 1.0}
        DefaultShadowOpacity := 1.0
    
    RewardEIcon<public> : texture_block = texture_block: # RewardEIcon
        DefaultImage := IC_Power
        DefaultTint := color{R := 1.0, G := 1.0, B := 1.0}
        DefaultDesiredSize := vector2{X := 64.0,Y := 64.0}
    
    RewardECount<public> : text_block = text_block: # RewardECount
        DefaultTextColor := color{R := 1.0, G := 1.0, B := 1.0}
        DefaultTextOpacity := 1.0
        DefaultShadowOffset := option{vector2{X := 0.0,Y := 0.0}}
        DefaultShadowColor := color{R := 1.0, G := 1.0, B := 1.0}
        DefaultShadowOpacity := 1.0
    
    RewardEName<public> : text_block = text_block: # RewardEName
        DefaultTextColor := color{R := 1.0, G := 1.0, B := 1.0}
        DefaultTextOpacity := 1.0
        DefaultShadowOffset := option{vector2{X := 0.0,Y := 0.0}}
        DefaultShadowColor := color{R := 1.0, G := 1.0, B := 1.0}
        DefaultShadowOpacity := 1.0
    
    RewardEPercent<public> : text_block = text_block: # RewardEPercent
        DefaultTextColor := color{R := 1.0, G := 1.0, B := 1.0}
        DefaultTextOpacity := 1.0
        DefaultShadowOffset := option{vector2{X := 0.0,Y := 0.0}}
        DefaultShadowColor := color{R := 1.0, G := 1.0, B := 1.0}
        DefaultShadowOpacity := 1.0
    
    RewardFIcon<public> : texture_block = texture_block: # RewardFIcon
        DefaultImage := IC_Power
        DefaultTint := color{R := 1.0, G := 1.0, B := 1.0}
        DefaultDesiredSize := vector2{X := 64.0,Y := 64.0}
    
    RewardFCount<public> : text_block = text_block: # RewardFCount
        DefaultTextColor := color{R := 1.0, G := 1.0, B := 1.0}
        DefaultTextOpacity := 1.0
        DefaultShadowOffset := option{vector2{X := 0.0,Y := 0.0}}
        DefaultShadowColor := color{R := 1.0, G := 1.0, B := 1.0}
        DefaultShadowOpacity := 1.0
    
    RewardFName<public> : text_block = text_block: # RewardFName
        DefaultTextColor := color{R := 1.0, G := 1.0, B := 1.0}
        DefaultTextOpacity := 1.0
        DefaultShadowOffset := option{vector2{X := 0.0,Y := 0.0}}
        DefaultShadowColor := color{R := 1.0, G := 1.0, B := 1.0}
        DefaultShadowOpacity := 1.0
    
    RewardFPercent<public> : text_block = text_block: # RewardFPercent
        DefaultTextColor := color{R := 1.0, G := 1.0, B := 1.0}
        DefaultTextOpacity := 1.0
        DefaultShadowOffset := option{vector2{X := 0.0,Y := 0.0}}
        DefaultShadowColor := color{R := 1.0, G := 1.0, B := 1.0}
        DefaultShadowOpacity := 1.0
    
    RewardGIcon<public> : texture_block = texture_block: # RewardGIcon
        DefaultImage := IC_Power
        DefaultTint := color{R := 1.0, G := 1.0, B := 1.0}
        DefaultDesiredSize := vector2{X := 64.0,Y := 64.0}
    
    RewardGCount<public> : text_block = text_block: # RewardGCount
        DefaultTextColor := color{R := 1.0, G := 1.0, B := 1.0}
        DefaultTextOpacity := 1.0
        DefaultShadowOffset := option{vector2{X := 0.0,Y := 0.0}}
        DefaultShadowColor := color{R := 1.0, G := 1.0, B := 1.0}
        DefaultShadowOpacity := 1.0
    
    RewardGName<public> : text_block = text_block: # RewardGName
        DefaultTextColor := color{R := 1.0, G := 1.0, B := 1.0}
        DefaultTextOpacity := 1.0
        DefaultShadowOffset := option{vector2{X := 0.0,Y := 0.0}}
        DefaultShadowColor := color{R := 1.0, G := 1.0, B := 1.0}
        DefaultShadowOpacity := 1.0
    
    RewardGPercent<public> : text_block = text_block: # RewardGPercent
        DefaultTextColor := color{R := 1.0, G := 1.0, B := 1.0}
        DefaultTextOpacity := 1.0
        DefaultShadowOffset := option{vector2{X := 0.0,Y := 0.0}}
        DefaultShadowColor := color{R := 1.0, G := 1.0, B := 1.0}
        DefaultShadowOpacity := 1.0
    
    StatIcon<public> : texture_block = texture_block: # StatIcon
        DefaultImage := koppar
        DefaultTint := color{R := 1.0, G := 1.0, B := 1.0}
        DefaultDesiredSize := vector2{X := 132.0,Y := 130.0}
    
    StatRankIcon<public> : texture_block = texture_block: # StatRankIcon
        DefaultImage := IC_Star
        DefaultTint := color{R := 1.0, G := 1.0, B := 1.0}
        DefaultDesiredSize := vector2{X := 0.0,Y := 0.0}
    
    StatName<public> : text_block = text_block: # StatName
        DefaultTextColor := color{R := 1.0, G := 1.0, B := 1.0}
        DefaultTextOpacity := 1.0
        DefaultShadowOffset := option{vector2{X := 0.0,Y := 0.0}}
        DefaultShadowColor := color{R := 1.0, G := 1.0, B := 1.0}
        DefaultShadowOpacity := 1.0
    
    StatLvTitle<public> : text_block = text_block: # StatLvTitle
        DefaultTextColor := color{R := 1.000000, G := 0.361099, B := 0.000000}
        DefaultTextOpacity := 1.000000
        DefaultShadowOffset := option{vector2{X := 0.0,Y := 0.0}}
        DefaultShadowColor := color{R := 1.0, G := 1.0, B := 1.0}
        DefaultShadowOpacity := 1.0
    
    StatLvValue<public> : text_block = text_block: # StatLvValue
        DefaultTextColor := color{R := 1.0, G := 1.0, B := 1.0}
        DefaultTextOpacity := 1.0
        DefaultShadowOffset := option{vector2{X := 0.0,Y := 0.0}}
        DefaultShadowColor := color{R := 1.0, G := 1.0, B := 1.0}
        DefaultShadowOpacity := 1.0
    
    StatTitle<public> : text_block = text_block: # StatTitle
        DefaultTextColor := color{R := 1.0, G := 1.0, B := 1.0}
        DefaultTextOpacity := 1.0
        DefaultShadowOffset := option{vector2{X := 0.0,Y := 0.0}}
        DefaultShadowColor := color{R := 1.0, G := 1.0, B := 1.0}
        DefaultShadowOpacity := 1.0
    
    StatAIcon<public> : texture_block = texture_block: # StatAIcon
        DefaultImage := Stats_PetGather
        DefaultTint := color{R := 1.0, G := 1.0, B := 1.0}
        DefaultDesiredSize := vector2{X := 45.763123,Y := 32.0}
    
    StatATitle<public> : text_block = text_block: # StatATitle
        DefaultTextColor := color{R := 0.223877, G := 0.325219, B := 0.682292}
        DefaultTextOpacity := 1.000000
        DefaultShadowOffset := option{vector2{X := 0.0,Y := 0.0}}
        DefaultShadowColor := color{R := 1.0, G := 1.0, B := 1.0}
        DefaultShadowOpacity := 1.0
    
    StatAValue<public> : text_block = text_block: # StatAValue
        DefaultTextColor := color{R := 0.385010, G := 0.890625, B := 0.496330}
        DefaultTextOpacity := 1.000000
        DefaultShadowOffset := option{vector2{X := 0.0,Y := 0.0}}
        DefaultShadowColor := color{R := 1.0, G := 1.0, B := 1.0}
        DefaultShadowOpacity := 1.0
    
    StatBIcon<public> : texture_block = texture_block: # StatBIcon
        DefaultImage := Stats_Str
        DefaultTint := color{R := 1.0, G := 1.0, B := 1.0}
        DefaultDesiredSize := vector2{X := 39.17606,Y := 32.0}
    
    StatBTitle<public> : text_block = text_block: # StatBTitle
        DefaultTextColor := color{R := 0.223877, G := 0.325219, B := 0.682292}
        DefaultTextOpacity := 1.000000
        DefaultShadowOffset := option{vector2{X := 0.0,Y := 0.0}}
        DefaultShadowColor := color{R := 1.0, G := 1.0, B := 1.0}
        DefaultShadowOpacity := 1.0
    
    StatBValue<public> : text_block = text_block: # StatBValue
        DefaultTextColor := color{R := 0.385010, G := 0.890625, B := 0.496330}
        DefaultTextOpacity := 1.000000
        DefaultShadowOffset := option{vector2{X := 0.0,Y := 0.0}}
        DefaultShadowColor := color{R := 1.0, G := 1.0, B := 1.0}
        DefaultShadowOpacity := 1.0
    
    StatCIcon<public> : texture_block = texture_block: # StatCIcon
        DefaultImage := Stats_CriRate
        DefaultTint := color{R := 1.0, G := 1.0, B := 1.0}
        DefaultDesiredSize := vector2{X := 42.493393,Y := 32.0}
    
    StatCTitle<public> : text_block = text_block: # StatCTitle
        DefaultTextColor := color{R := 0.223877, G := 0.325219, B := 0.682292}
        DefaultTextOpacity := 1.000000
        DefaultShadowOffset := option{vector2{X := 0.0,Y := 0.0}}
        DefaultShadowColor := color{R := 1.0, G := 1.0, B := 1.0}
        DefaultShadowOpacity := 1.0
    
    StatCValue<public> : text_block = text_block: # StatCValue
        DefaultTextColor := color{R := 0.385010, G := 0.890625, B := 0.496330}
        DefaultTextOpacity := 1.000000
        DefaultShadowOffset := option{vector2{X := 0.0,Y := 0.0}}
        DefaultShadowColor := color{R := 1.0, G := 1.0, B := 1.0}
        DefaultShadowOpacity := 1.0
    
    StatDIcon<public> : texture_block = texture_block: # StatDIcon
        DefaultImage := Stats_CriDamage
        DefaultTint := color{R := 1.0, G := 1.0, B := 1.0}
        DefaultDesiredSize := vector2{X := 42.046722,Y := 32.0}
    
    StatDTitle<public> : text_block = text_block: # StatDTitle
        DefaultTextColor := color{R := 0.223877, G := 0.325219, B := 0.682292}
        DefaultTextOpacity := 1.000000
        DefaultShadowOffset := option{vector2{X := 0.0,Y := 0.0}}
        DefaultShadowColor := color{R := 1.0, G := 1.0, B := 1.0}
        DefaultShadowOpacity := 1.0
    
    StatDValue<public> : text_block = text_block: # StatDValue
        DefaultTextColor := color{R := 0.385010, G := 0.890625, B := 0.496330}
        DefaultTextOpacity := 1.000000
        DefaultShadowOffset := option{vector2{X := 0.0,Y := 0.0}}
        DefaultShadowColor := color{R := 1.0, G := 1.0, B := 1.0}
        DefaultShadowOpacity := 1.0
    
    StatEIcon<public> : texture_block = texture_block: # StatEIcon
        DefaultImage := IC_LastHit
        DefaultTint := color{R := 1.0, G := 1.0, B := 1.0}
        DefaultDesiredSize := vector2{X := 44.496067,Y := 32.0}
    
    StatETitle<public> : text_block = text_block: # StatETitle
        DefaultTextColor := color{R := 0.223877, G := 0.325219, B := 0.682292}
        DefaultTextOpacity := 1.000000
        DefaultShadowOffset := option{vector2{X := 0.0,Y := 0.0}}
        DefaultShadowColor := color{R := 1.0, G := 1.0, B := 1.0}
        DefaultShadowOpacity := 1.0
    
    StatEValue<public> : text_block = text_block: # StatEValue
        DefaultTextColor := color{R := 0.385010, G := 0.890625, B := 0.496330}
        DefaultTextOpacity := 1.000000
        DefaultShadowOffset := option{vector2{X := 0.0,Y := 0.0}}
        DefaultShadowColor := color{R := 1.0, G := 1.0, B := 1.0}
        DefaultShadowOpacity := 1.0
    
    CostTitle<public> : text_block = text_block: # CostTitle
        DefaultTextColor := color{R := 0.223877, G := 0.325219, B := 0.682292}
        DefaultTextOpacity := 1.000000
        DefaultShadowOffset := option{vector2{X := 0.0,Y := 0.0}}
        DefaultShadowColor := color{R := 1.0, G := 1.0, B := 1.0}
        DefaultShadowOpacity := 1.0
    
    CostAIcon<public> : texture_block = texture_block: # CostAIcon
        DefaultImage := SS_1
        DefaultTint := color{R := 1.0, G := 1.0, B := 1.0}
        DefaultDesiredSize := vector2{X := 40.0,Y := 32.0}
    
    CostAValue<public> : text_block = text_block: # CostAValue
        DefaultTextColor := color{R := 1.0, G := 1.0, B := 1.0}
        DefaultTextOpacity := 1.0
        DefaultShadowOffset := option{vector2{X := 0.0,Y := 0.0}}
        DefaultShadowColor := color{R := 1.0, G := 1.0, B := 1.0}
        DefaultShadowOpacity := 1.0
    
    CostBIcon<public> : texture_block = texture_block: # CostBIcon
        DefaultImage := SilverOre
        DefaultTint := color{R := 1.0, G := 1.0, B := 1.0}
        DefaultDesiredSize := vector2{X := 40.0,Y := 32.0}
    
    CostBValue<public> : text_block = text_block: # CostBValue
        DefaultTextColor := color{R := 1.0, G := 1.0, B := 1.0}
        DefaultTextOpacity := 1.0
        DefaultShadowOffset := option{vector2{X := 0.0,Y := 0.0}}
        DefaultShadowColor := color{R := 1.0, G := 1.0, B := 1.0}
        DefaultShadowOpacity := 1.0
    
    CreateCanvas<override>():canvas=
        canvas: # CanvasPanel_10
            Slots := array:
                canvas_slot: # CanvasPanelSlot_1
                    Anchors := anchors:
                        Minimum := vector2{X := 0.0,Y := 0.0}
                        Maximum := vector2{X := 1.0,Y := 1.0}
                    Offsets := margin:
                        Top := 0.0
                        Right := 1.921875
                        Bottom := 1.081055
                        Left := 0.0
                    SizeToContent := true
                    Alignment := vector2{X := 0.0,Y := 0.0}
                    
                    Widget := Background
                canvas_slot: # CanvasPanelSlot_3
                    Anchors := anchors:
                        Minimum := vector2{X := 0.0,Y := 0.0}
                        Maximum := vector2{X := 1.0,Y := 1.0}
                    Offsets := margin:
                        Top := 48.0
                        Right := 832.921875
                        Bottom := 992.081055
                        Left := 832.0
                    SizeToContent := true
                    Alignment := vector2{X := 0.0,Y := 0.0}
                    
                    Widget := Title
                canvas_slot: # CanvasPanelSlot_122
                    Anchors := anchors:
                        Minimum := vector2{X := 0.0,Y := 0.0}
                        Maximum := vector2{X := 1.0,Y := 1.0}
                    Offsets := margin:
                        Top := 952.0
                        Right := 866.300293
                        Bottom := 49.081055
                        Left := 848.0
                    SizeToContent := true
                    Alignment := vector2{X := 0.0,Y := 0.0}
                    
                    Widget := BasicBG.CloseButton
                <#canvas_slot: # CanvasPanelSlot_121
                    Anchors := anchors:
                        Minimum := vector2{X := 0.0,Y := 0.0}
                        Maximum := vector2{X := 1.0,Y := 1.0}
                    Offsets := margin:
                        Top := 704.0
                        Right := 1040.681641
                        Bottom := 280.040039
                        Left := 360.0
                    SizeToContent := true
                    Alignment := vector2{X := 0.0,Y := 0.0}
                    
                    Widget := RewardFBG
                canvas_slot: # CanvasPanelSlot_48
                    Anchors := anchors:
                        Minimum := vector2{X := 0.0,Y := 0.0}
                        Maximum := vector2{X := 1.0,Y := 1.0}
                    Offsets := margin:
                        Top := 124.0
                        Right := 1040.681641
                        Bottom := 860.040039
                        Left := 360.0
                    SizeToContent := true
                    Alignment := vector2{X := 0.0,Y := 0.0}
                    
                    Widget := RewardABG
                canvas_slot: # CanvasPanelSlot_81
                    Anchors := anchors:
                        Minimum := vector2{X := 0.0,Y := 0.0}
                        Maximum := vector2{X := 1.0,Y := 1.0}
                    Offsets := margin:
                        Top := 240.0
                        Right := 1040.681641
                        Bottom := 744.040039
                        Left := 360.0
                    SizeToContent := true
                    Alignment := vector2{X := 0.0,Y := 0.0}
                    
                    Widget := RewardBBG
                canvas_slot: # CanvasPanelSlot_86
                    Anchors := anchors:
                        Minimum := vector2{X := 0.0,Y := 0.0}
                        Maximum := vector2{X := 1.0,Y := 1.0}
                    Offsets := margin:
                        Top := 472.0
                        Right := 1040.681641
                        Bottom := 512.040039
                        Left := 360.0
                    SizeToContent := true
                    Alignment := vector2{X := 0.0,Y := 0.0}
                    
                    Widget := RewardDBG
                canvas_slot: # CanvasPanelSlot_89
                    Anchors := anchors:
                        Minimum := vector2{X := 0.0,Y := 0.0}
                        Maximum := vector2{X := 1.0,Y := 1.0}
                    Offsets := margin:
                        Top := 588.0
                        Right := 1040.681641
                        Bottom := 396.040039
                        Left := 360.0
                    SizeToContent := true
                    Alignment := vector2{X := 0.0,Y := 0.0}
                    
                    Widget := RewardEBG
                canvas_slot: # CanvasPanelSlot_84
                    Anchors := anchors:
                        Minimum := vector2{X := 0.0,Y := 0.0}
                        Maximum := vector2{X := 1.0,Y := 1.0}
                    Offsets := margin:
                        Top := 356.0
                        Right := 1040.681641
                        Bottom := 628.040039
                        Left := 360.0
                    SizeToContent := true
                    Alignment := vector2{X := 0.0,Y := 0.0}
                    
                    Widget := RewardCBG
                canvas_slot: # CanvasPanelSlot_93
                    Anchors := anchors:
                        Minimum := vector2{X := 0.0,Y := 0.0}
                        Maximum := vector2{X := 1.0,Y := 1.0}
                    Offsets := margin:
                        Top := 820.0
                        Right := 1040.681641
                        Bottom := 164.040039
                        Left := 360.0
                    SizeToContent := true
                    Alignment := vector2{X := 0.0,Y := 0.0}
                    
                    Widget := RewardGBG#>
                canvas_slot: # CanvasPanelSlot_119
                    Anchors := anchors:
                        Minimum := vector2{X := 0.0,Y := 0.0}
                        Maximum := vector2{X := 1.0,Y := 1.0}
                    Offsets := margin:
                        Top := 104.504517
                        Right := 1020.219604
                        Bottom := 137.563293
                        Left := 35.507507
                    SizeToContent := true
                    ZOrder := 0
                    Alignment := vector2{X := 0.0,Y := 0.0}
                    
                    Widget := stack_box: # StackBox_44
                        Orientation := orientation.Horizontal
                        Slots := array:
                            stack_box_slot: # StackBoxSlot_1
                                HorizontalAlignment := horizontal_alignment.Center
                                VerticalAlignment := vertical_alignment.Top
                                Padding := margin:
                                    Top := 0.0
                                    Right := 0.0
                                    Bottom := 0.0
                                    Left := 20.0
                                Widget := stack_box: # StackBox_42
                                    Orientation := orientation.Vertical
                                    Slots := array:
                                        stack_box_slot: # StackBoxSlot_7
                                            HorizontalAlignment := horizontal_alignment.Left
                                            VerticalAlignment := vertical_alignment.Center
                                            Padding := margin:
                                                Top := 0.0
                                                Right := 0.0
                                                Bottom := 25.0
                                                Left := 0.0
                                            Widget := InvTitle
                                        stack_box_slot: # StackBoxSlot_0
                                            HorizontalAlignment := horizontal_alignment.Left
                                            VerticalAlignment := vertical_alignment.Center
                                            Padding := margin:
                                                Top := 10.0
                                                Right := 0.0
                                                Bottom := 10.0
                                                Left := 0.0
                                            Widget := stack_box: # StackBox_2
                                                Orientation := orientation.Horizontal
                                                Slots := array:
                                                    stack_box_slot: # StackBoxSlot_3
                                                        HorizontalAlignment := horizontal_alignment.Left
                                                        VerticalAlignment := vertical_alignment.Top
                                                        Padding := margin:
                                                            Top := 0.0
                                                            Right := 0.0
                                                            Bottom := 0.0
                                                            Left := 0.0
                                                        Widget := InvAIcon
                                                    stack_box_slot: # StackBoxSlot_4
                                                        HorizontalAlignment := horizontal_alignment.Right
                                                        VerticalAlignment := vertical_alignment.Top
                                                        Padding := margin:
                                                            Top := 7.5
                                                            Right := 0.0
                                                            Bottom := 0.0
                                                            Left := 20.0
                                                        Widget := InvAValue
                                                    
                                        stack_box_slot: # StackBoxSlot_1
                                            HorizontalAlignment := horizontal_alignment.Left
                                            VerticalAlignment := vertical_alignment.Center
                                            Padding := margin:
                                                Top := 10.0
                                                Right := 0.0
                                                Bottom := 10.0
                                                Left := 0.0
                                            Widget := stack_box: # StackBox_3
                                                Orientation := orientation.Horizontal
                                                Slots := array:
                                                    stack_box_slot: # StackBoxSlot_5
                                                        HorizontalAlignment := horizontal_alignment.Left
                                                        VerticalAlignment := vertical_alignment.Top
                                                        Padding := margin:
                                                            Top := 0.0
                                                            Right := 0.0
                                                            Bottom := 0.0
                                                            Left := 0.0
                                                        Widget := InvBIcon
                                                    stack_box_slot: # StackBoxSlot_6
                                                        HorizontalAlignment := horizontal_alignment.Right
                                                        VerticalAlignment := vertical_alignment.Top
                                                        Padding := margin:
                                                            Top := 7.5
                                                            Right := 0.0
                                                            Bottom := 0.0
                                                            Left := 20.0
                                                        Widget := InvBValue
                                                    
                                        stack_box_slot: # StackBoxSlot_2
                                            HorizontalAlignment := horizontal_alignment.Left
                                            VerticalAlignment := vertical_alignment.Center
                                            Padding := margin:
                                                Top := 10.0
                                                Right := 0.0
                                                Bottom := 10.0
                                                Left := 0.0
                                            Widget := stack_box: # StackBox_4
                                                Orientation := orientation.Horizontal
                                                Slots := array:
                                                    stack_box_slot: # StackBoxSlot_0
                                                        HorizontalAlignment := horizontal_alignment.Left
                                                        VerticalAlignment := vertical_alignment.Top
                                                        Padding := margin:
                                                            Top := 0.0
                                                            Right := 0.0
                                                            Bottom := 0.0
                                                            Left := 0.0
                                                        Widget := InvCIcon
                                                    stack_box_slot: # StackBoxSlot_1
                                                        HorizontalAlignment := horizontal_alignment.Right
                                                        VerticalAlignment := vertical_alignment.Top
                                                        Padding := margin:
                                                            Top := 7.5
                                                            Right := 0.0
                                                            Bottom := 0.0
                                                            Left := 20.0
                                                        Widget := InvCValue
                                                    
                                        stack_box_slot: # StackBoxSlot_3
                                            HorizontalAlignment := horizontal_alignment.Left
                                            VerticalAlignment := vertical_alignment.Center
                                            Padding := margin:
                                                Top := 10.0
                                                Right := 0.0
                                                Bottom := 10.0
                                                Left := 0.0
                                            Widget := stack_box: # StackBox_5
                                                Orientation := orientation.Horizontal
                                                Slots := array:
                                                    stack_box_slot: # StackBoxSlot_1
                                                        HorizontalAlignment := horizontal_alignment.Left
                                                        VerticalAlignment := vertical_alignment.Top
                                                        Padding := margin:
                                                            Top := 0.0
                                                            Right := 0.0
                                                            Bottom := 0.0
                                                            Left := 0.0
                                                        Widget := InvDIcon
                                                    stack_box_slot: # StackBoxSlot_0
                                                        HorizontalAlignment := horizontal_alignment.Right
                                                        VerticalAlignment := vertical_alignment.Top
                                                        Padding := margin:
                                                            Top := 7.5
                                                            Right := 0.0
                                                            Bottom := 0.0
                                                            Left := 20.0
                                                        Widget := InvDValue
                                                    
                                        stack_box_slot: # StackBoxSlot_4
                                            HorizontalAlignment := horizontal_alignment.Left
                                            VerticalAlignment := vertical_alignment.Center
                                            Padding := margin:
                                                Top := 10.0
                                                Right := 0.0
                                                Bottom := 10.0
                                                Left := 0.0
                                            Widget := stack_box: # StackBox
                                                Orientation := orientation.Horizontal
                                                Slots := array:
                                                    stack_box_slot: # StackBoxSlot_3
                                                        HorizontalAlignment := horizontal_alignment.Left
                                                        VerticalAlignment := vertical_alignment.Top
                                                        Padding := margin:
                                                            Top := 0.0
                                                            Right := 0.0
                                                            Bottom := 0.0
                                                            Left := 0.0
                                                        Widget := InvEIcon
                                                    stack_box_slot: # StackBoxSlot_4
                                                        HorizontalAlignment := horizontal_alignment.Right
                                                        VerticalAlignment := vertical_alignment.Top
                                                        Padding := margin:
                                                            Top := 7.5
                                                            Right := 0.0
                                                            Bottom := 0.0
                                                            Left := 20.0
                                                        Widget := InvEValue
                                                    
                                        stack_box_slot: # StackBoxSlot_5
                                            HorizontalAlignment := horizontal_alignment.Left
                                            VerticalAlignment := vertical_alignment.Center
                                            Padding := margin:
                                                Top := 10.0
                                                Right := 0.0
                                                Bottom := 10.0
                                                Left := 0.0
                                            Widget := stack_box: # StackBox_1
                                                Orientation := orientation.Horizontal
                                                Slots := array:
                                                    stack_box_slot: # StackBoxSlot_3
                                                        HorizontalAlignment := horizontal_alignment.Left
                                                        VerticalAlignment := vertical_alignment.Top
                                                        Padding := margin:
                                                            Top := 0.0
                                                            Right := 0.0
                                                            Bottom := 0.0
                                                            Left := 0.0
                                                        Widget := InvFIcon
                                                    stack_box_slot: # StackBoxSlot_4
                                                        HorizontalAlignment := horizontal_alignment.Right
                                                        VerticalAlignment := vertical_alignment.Top
                                                        Padding := margin:
                                                            Top := 7.5
                                                            Right := 0.0
                                                            Bottom := 0.0
                                                            Left := 20.0
                                                        Widget := InvFValue
                                                    
                                        stack_box_slot: # StackBoxSlot_6
                                            HorizontalAlignment := horizontal_alignment.Left
                                            VerticalAlignment := vertical_alignment.Center
                                            Padding := margin:
                                                Top := 10.0
                                                Right := 0.0
                                                Bottom := 10.0
                                                Left := 0.0
                                            Widget := stack_box: # StackBox_6
                                                Orientation := orientation.Horizontal
                                                Slots := array:
                                                    stack_box_slot: # StackBoxSlot_3
                                                        HorizontalAlignment := horizontal_alignment.Left
                                                        VerticalAlignment := vertical_alignment.Top
                                                        Padding := margin:
                                                            Top := 0.0
                                                            Right := 0.0
                                                            Bottom := 0.0
                                                            Left := 0.0
                                                        Widget := InvGIcon
                                                    stack_box_slot: # StackBoxSlot_4
                                                        HorizontalAlignment := horizontal_alignment.Right
                                                        VerticalAlignment := vertical_alignment.Top
                                                        Padding := margin:
                                                            Top := 7.5
                                                            Right := 0.0
                                                            Bottom := 0.0
                                                            Left := 20.0
                                                        Widget := InvGValue
                                                    
                                        
                            stack_box_slot: # StackBoxSlot_0
                                HorizontalAlignment := horizontal_alignment.Center
                                VerticalAlignment := vertical_alignment.Top
                                Padding := margin:
                                    Top := 0.0
                                    Right := 0.0
                                    Bottom := 0.0
                                    Left := 65.0
                                Widget := stack_box: # StackBox_43
                                    Orientation := orientation.Vertical
                                    Slots := array:
                                        stack_box_slot: # StackBoxSlot_0
                                            HorizontalAlignment := horizontal_alignment.Left
                                            VerticalAlignment := vertical_alignment.Center
                                            Padding := margin:
                                                Top := 20.0
                                                Right := 0.0
                                                Bottom := 0.0
                                                Left := 25.0
                                            Widget := stack_box: # StackBox_11
                                                Orientation := orientation.Horizontal
                                                Slots := array:
                                                    stack_box_slot: # StackBoxSlot_7
                                                        HorizontalAlignment := horizontal_alignment.Left
                                                        VerticalAlignment := vertical_alignment.Top
                                                        Padding := margin:
                                                            Top := 5.0
                                                            Right := 0.0
                                                            Bottom := 0.0
                                                            Left := 0.0
                                                        Widget := RewardAIcon
                                                    stack_box_slot: # StackBoxSlot_8
                                                        HorizontalAlignment := horizontal_alignment.Right
                                                        VerticalAlignment := vertical_alignment.Top
                                                        Padding := margin:
                                                            Top := 0.0
                                                            Right := 0.0
                                                            Bottom := 0.0
                                                            Left := 12.0
                                                        Widget := stack_box: # StackBox_10
                                                            Orientation := orientation.Vertical
                                                            Slots := array:
                                                                stack_box_slot: # StackBoxSlot_0
                                                                    HorizontalAlignment := horizontal_alignment.Left
                                                                    VerticalAlignment := vertical_alignment.Top
                                                                    Padding := margin:
                                                                        Top := -7.5
                                                                        Right := 0.0
                                                                        Bottom := 0.0
                                                                        Left := 0.0
                                                                    Widget := stack_box: # StackBox_9
                                                                        Orientation := orientation.Horizontal
                                                                        Slots := array:
                                                                            stack_box_slot: # StackBoxSlot_21
                                                                                HorizontalAlignment := horizontal_alignment.Left
                                                                                VerticalAlignment := vertical_alignment.Top
                                                                                Padding := margin:
                                                                                    Top := 0.0
                                                                                    Right := 0.0
                                                                                    Bottom := 0.0
                                                                                    Left := 0.0
                                                                                Widget := RewardALvTitle
                                                                            stack_box_slot: # StackBoxSlot_22
                                                                                HorizontalAlignment := horizontal_alignment.Right
                                                                                VerticalAlignment := vertical_alignment.Top
                                                                                Padding := margin:
                                                                                    Top := 0.0
                                                                                    Right := 0.0
                                                                                    Bottom := 0.0
                                                                                    Left := 0.0
                                                                                Widget := RewardALvValue
                                                                            
                                                                stack_box_slot: # StackBoxSlot_3
                                                                    HorizontalAlignment := horizontal_alignment.Left
                                                                    VerticalAlignment := vertical_alignment.Top
                                                                    Padding := margin:
                                                                        Top := -7.5
                                                                        Right := 0.0
                                                                        Bottom := 0.0
                                                                        Left := 0.0
                                                                    Widget := RewardAName
                                                                stack_box_slot: # StackBoxSlot_6
                                                                    HorizontalAlignment := horizontal_alignment.Left
                                                                    VerticalAlignment := vertical_alignment.Top
                                                                    Padding := margin:
                                                                        Top := -7.5
                                                                        Right := 0.0
                                                                        Bottom := 0.0
                                                                        Left := 0.0
                                                                    Widget := RewardAPercent
                                                                
                                                    
                                        stack_box_slot: # StackBoxSlot_1
                                            HorizontalAlignment := horizontal_alignment.Left
                                            VerticalAlignment := vertical_alignment.Center
                                            Padding := margin:
                                                Top := 20.0
                                                Right := 0.0
                                                Bottom := 0.0
                                                Left := 25.0
                                            Widget := stack_box: # StackBox_12
                                                Orientation := orientation.Horizontal
                                                Slots := array:
                                                    stack_box_slot: # StackBoxSlot_7
                                                        HorizontalAlignment := horizontal_alignment.Left
                                                        VerticalAlignment := vertical_alignment.Top
                                                        Padding := margin:
                                                            Top := 5.0
                                                            Right := 0.0
                                                            Bottom := 0.0
                                                            Left := 0.0
                                                        Widget := RewardBIcon
                                                    stack_box_slot: # StackBoxSlot_0
                                                        HorizontalAlignment := horizontal_alignment.Right
                                                        VerticalAlignment := vertical_alignment.Top
                                                        Padding := margin:
                                                            Top := 0.0
                                                            Right := 0.0
                                                            Bottom := 0.0
                                                            Left := 12.0
                                                        Widget := stack_box: # StackBox_13
                                                            Orientation := orientation.Vertical
                                                            Slots := array:
                                                                stack_box_slot: # StackBoxSlot_0
                                                                    HorizontalAlignment := horizontal_alignment.Left
                                                                    VerticalAlignment := vertical_alignment.Top
                                                                    Padding := margin:
                                                                        Top :=-7.5
                                                                        Right := 0.0
                                                                        Bottom := 0.0
                                                                        Left := 0.0
                                                                    Widget := stack_box: # StackBox_14
                                                                        Orientation := orientation.Horizontal
                                                                        Slots := array:
                                                                            stack_box_slot: # StackBoxSlot_21
                                                                                HorizontalAlignment := horizontal_alignment.Left
                                                                                VerticalAlignment := vertical_alignment.Top
                                                                                Padding := margin:
                                                                                    Top := 0.0
                                                                                    Right := 0.0
                                                                                    Bottom := 0.0
                                                                                    Left := 0.0
                                                                                Widget := RewardBLvTitle
                                                                            stack_box_slot: # StackBoxSlot_22
                                                                                HorizontalAlignment := horizontal_alignment.Right
                                                                                VerticalAlignment := vertical_alignment.Top
                                                                                Padding := margin:
                                                                                    Top := 0.0
                                                                                    Right := 0.0
                                                                                    Bottom := 0.0
                                                                                    Left := 0.0
                                                                                Widget := RewardBLvValue
                                                                            
                                                                stack_box_slot: # StackBoxSlot_3
                                                                    HorizontalAlignment := horizontal_alignment.Left
                                                                    VerticalAlignment := vertical_alignment.Top
                                                                    Padding := margin:
                                                                        Top := -7.5
                                                                        Right := 0.0
                                                                        Bottom := 0.0
                                                                        Left := 0.0
                                                                    Widget := RewardBName
                                                                stack_box_slot: # StackBoxSlot_6
                                                                    HorizontalAlignment := horizontal_alignment.Left
                                                                    VerticalAlignment := vertical_alignment.Top
                                                                    Padding := margin:
                                                                        Top := -7.5
                                                                        Right := 0.0
                                                                        Bottom := 0.0
                                                                        Left := 0.0
                                                                    Widget := RewardBPercent
                                                                
                                                    
                                        stack_box_slot: # StackBoxSlot_2
                                            HorizontalAlignment := horizontal_alignment.Left
                                            VerticalAlignment := vertical_alignment.Center
                                            Padding := margin:
                                                Top := 20.0
                                                Right := 0.0
                                                Bottom := 0.0
                                                Left := 25.0
                                            Widget := stack_box: # StackBox_15
                                                Orientation := orientation.Horizontal
                                                Slots := array:
                                                    stack_box_slot: # StackBoxSlot_7
                                                        HorizontalAlignment := horizontal_alignment.Left
                                                        VerticalAlignment := vertical_alignment.Top
                                                        Padding := margin:
                                                            Top := 5.0
                                                            Right := 0.0
                                                            Bottom := 0.0
                                                            Left := 0.0
                                                        Widget := RewardCIcon
                                                    stack_box_slot: # StackBoxSlot_1
                                                        HorizontalAlignment := horizontal_alignment.Right
                                                        VerticalAlignment := vertical_alignment.Top
                                                        Padding := margin:
                                                            Top := 0.0
                                                            Right := 0.0
                                                            Bottom := 0.0
                                                            Left := 12.0
                                                        Widget := stack_box: # StackBox_16
                                                            Orientation := orientation.Vertical
                                                            Slots := array:
                                                                stack_box_slot: # StackBoxSlot_0
                                                                    HorizontalAlignment := horizontal_alignment.Left
                                                                    VerticalAlignment := vertical_alignment.Top
                                                                    Padding := margin:
                                                                        Top := -7.5
                                                                        Right := 0.0
                                                                        Bottom := 0.0
                                                                        Left := 0.0
                                                                    Widget := stack_box: # StackBox_17
                                                                        Orientation := orientation.Horizontal
                                                                        Slots := array:
                                                                            stack_box_slot: # StackBoxSlot_21
                                                                                HorizontalAlignment := horizontal_alignment.Left
                                                                                VerticalAlignment := vertical_alignment.Top
                                                                                Padding := margin:
                                                                                    Top := 0.0
                                                                                    Right := 0.0
                                                                                    Bottom := 0.0
                                                                                    Left := 0.0
                                                                                Widget := RewardCLvTitle
                                                                            stack_box_slot: # StackBoxSlot_22
                                                                                HorizontalAlignment := horizontal_alignment.Right
                                                                                VerticalAlignment := vertical_alignment.Top
                                                                                Padding := margin:
                                                                                    Top := 0.0
                                                                                    Right := 0.0
                                                                                    Bottom := 0.0
                                                                                    Left := 0.0
                                                                                Widget := RewardCLvValue
                                                                            
                                                                stack_box_slot: # StackBoxSlot_3
                                                                    HorizontalAlignment := horizontal_alignment.Left
                                                                    VerticalAlignment := vertical_alignment.Top
                                                                    Padding := margin:
                                                                        Top := -7.5
                                                                        Right := 0.0
                                                                        Bottom := 0.0
                                                                        Left := 0.0
                                                                    Widget := RewardCName
                                                                stack_box_slot: # StackBoxSlot_6
                                                                    HorizontalAlignment := horizontal_alignment.Left
                                                                    VerticalAlignment := vertical_alignment.Top
                                                                    Padding := margin:
                                                                        Top := -7.5
                                                                        Right := 0.0
                                                                        Bottom := 0.0
                                                                        Left := 0.0
                                                                    Widget := RewardCPercent
                                                                
                                                    
                                        stack_box_slot: # StackBoxSlot_5
                                            HorizontalAlignment := horizontal_alignment.Left
                                            VerticalAlignment := vertical_alignment.Center
                                            Padding := margin:
                                                Top := 20.0
                                                Right := 0.0
                                                Bottom := 0.0
                                                Left := 25.0
                                            Widget := stack_box: # StackBox_18
                                                Orientation := orientation.Horizontal
                                                Slots := array:
                                                    stack_box_slot: # StackBoxSlot_7
                                                        HorizontalAlignment := horizontal_alignment.Left
                                                        VerticalAlignment := vertical_alignment.Top
                                                        Padding := margin:
                                                            Top := 5.0
                                                            Right := 0.0
                                                            Bottom := 0.0
                                                            Left := 0.0
                                                        Widget := RewardDIcon
                                                    stack_box_slot: # StackBoxSlot_1
                                                        HorizontalAlignment := horizontal_alignment.Right
                                                        VerticalAlignment := vertical_alignment.Top
                                                        Padding := margin:
                                                            Top := 0.0
                                                            Right := 0.0
                                                            Bottom := 0.0
                                                            Left := 12.0
                                                        Widget := stack_box: # StackBox_19
                                                            Orientation := orientation.Vertical
                                                            Slots := array:
                                                                stack_box_slot: # StackBoxSlot_0
                                                                    HorizontalAlignment := horizontal_alignment.Left
                                                                    VerticalAlignment := vertical_alignment.Top
                                                                    Padding := margin:
                                                                        Top := -7.5
                                                                        Right := 0.0
                                                                        Bottom := 0.0
                                                                        Left := 0.0
                                                                    Widget := stack_box: # StackBox_20
                                                                        Orientation := orientation.Horizontal
                                                                        Slots := array:
                                                                            stack_box_slot: # StackBoxSlot_21
                                                                                HorizontalAlignment := horizontal_alignment.Left
                                                                                VerticalAlignment := vertical_alignment.Top
                                                                                Padding := margin:
                                                                                    Top := 0.0
                                                                                    Right := 0.0
                                                                                    Bottom := 0.0
                                                                                    Left := 0.0
                                                                                Widget := RewardDLvTitle
                                                                            stack_box_slot: # StackBoxSlot_22
                                                                                HorizontalAlignment := horizontal_alignment.Right
                                                                                VerticalAlignment := vertical_alignment.Top
                                                                                Padding := margin:
                                                                                    Top := 0.0
                                                                                    Right := 0.0
                                                                                    Bottom := 0.0
                                                                                    Left := 0.0
                                                                                Widget := RewardDLvValue
                                                                            
                                                                stack_box_slot: # StackBoxSlot_3
                                                                    HorizontalAlignment := horizontal_alignment.Left
                                                                    VerticalAlignment := vertical_alignment.Top
                                                                    Padding := margin:
                                                                        Top := -7.5
                                                                        Right := 0.0
                                                                        Bottom := 0.0
                                                                        Left := 0.0
                                                                    Widget := RewardDName
                                                                stack_box_slot: # StackBoxSlot_6
                                                                    HorizontalAlignment := horizontal_alignment.Left
                                                                    VerticalAlignment := vertical_alignment.Top
                                                                    Padding := margin:
                                                                        Top := -7.5
                                                                        Right := 0.0
                                                                        Bottom := 0.0
                                                                        Left := 0.0
                                                                    Widget := RewardDPercent
                                                                
                                                    
                                        stack_box_slot: # StackBoxSlot_3
                                            HorizontalAlignment := horizontal_alignment.Left
                                            VerticalAlignment := vertical_alignment.Center
                                            Padding := margin:
                                                Top := 20.0
                                                Right := 0.0
                                                Bottom := 0.0
                                                Left := 20.0
                                            Widget := stack_box: # StackBox_21
                                                Orientation := orientation.Horizontal
                                                Slots := array:
                                                    stack_box_slot: # StackBoxSlot_7
                                                        HorizontalAlignment := horizontal_alignment.Left
                                                        VerticalAlignment := vertical_alignment.Top
                                                        Padding := margin:
                                                            Top := 5.0
                                                            Right := 0.0
                                                            Bottom := 0.0
                                                            Left := 0.0
                                                        Widget := RewardEIcon
                                                    stack_box_slot: # StackBoxSlot_1
                                                        HorizontalAlignment := horizontal_alignment.Right
                                                        VerticalAlignment := vertical_alignment.Top
                                                        Padding := margin:
                                                            Top := 0.0
                                                            Right := 0.0
                                                            Bottom := 0.0
                                                            Left := 12.0
                                                        Widget := stack_box: # StackBox_22
                                                            Orientation := orientation.Vertical
                                                            Slots := array:
                                                                stack_box_slot: # StackBoxSlot_0
                                                                    HorizontalAlignment := horizontal_alignment.Left
                                                                    VerticalAlignment := vertical_alignment.Top
                                                                    Padding := margin:
                                                                        Top :=-7.5
                                                                        Right := 0.0
                                                                        Bottom := 0.0
                                                                        Left := 0.0
                                                                    Widget := stack_box: # StackBox_23
                                                                        Orientation := orientation.Horizontal
                                                                        Slots := array:
                                                                            stack_box_slot: # StackBoxSlot_22
                                                                                HorizontalAlignment := horizontal_alignment.Right
                                                                                VerticalAlignment := vertical_alignment.Top
                                                                                Padding := margin:
                                                                                    Top := 0.0
                                                                                    Right := 0.0
                                                                                    Bottom := 0.0
                                                                                    Left := 0.0
                                                                                Widget := RewardECount
                                                                            
                                                                stack_box_slot: # StackBoxSlot_3
                                                                    HorizontalAlignment := horizontal_alignment.Left
                                                                    VerticalAlignment := vertical_alignment.Top
                                                                    Padding := margin:
                                                                        Top :=-7.5
                                                                        Right := 0.0
                                                                        Bottom := 0.0
                                                                        Left := 0.0
                                                                    Widget := RewardEName
                                                                stack_box_slot: # StackBoxSlot_6
                                                                    HorizontalAlignment := horizontal_alignment.Left
                                                                    VerticalAlignment := vertical_alignment.Top
                                                                    Padding := margin:
                                                                        Top :=-7.5
                                                                        Right := 0.0
                                                                        Bottom := 0.0
                                                                        Left := 0.0
                                                                    Widget := RewardEPercent
                                                                
                                                    
                                        stack_box_slot: # StackBoxSlot_4
                                            HorizontalAlignment := horizontal_alignment.Left
                                            VerticalAlignment := vertical_alignment.Center
                                            Padding := margin:
                                                Top := 20.0
                                                Right := 0.0
                                                Bottom := 0.0
                                                Left := 25.0
                                            Widget := stack_box: # StackBox_27
                                                Orientation := orientation.Horizontal
                                                Slots := array:
                                                    stack_box_slot: # StackBoxSlot_7
                                                        HorizontalAlignment := horizontal_alignment.Left
                                                        VerticalAlignment := vertical_alignment.Top
                                                        Padding := margin:
                                                            Top := 5.0
                                                            Right := 0.0
                                                            Bottom := 0.0
                                                            Left := 0.0
                                                        Widget := RewardFIcon
                                                    stack_box_slot: # StackBoxSlot_1
                                                        HorizontalAlignment := horizontal_alignment.Right
                                                        VerticalAlignment := vertical_alignment.Top
                                                        Padding := margin:
                                                            Top := 0.0
                                                            Right := 0.0
                                                            Bottom := 0.0
                                                            Left := 12.0
                                                        Widget := stack_box: # StackBox_28
                                                            Orientation := orientation.Vertical
                                                            Slots := array:
                                                                stack_box_slot: # StackBoxSlot_0
                                                                    HorizontalAlignment := horizontal_alignment.Left
                                                                    VerticalAlignment := vertical_alignment.Top
                                                                    Padding := margin:
                                                                        Top := -7.5
                                                                        Right := 0.0
                                                                        Bottom := 0.0
                                                                        Left := 0.0
                                                                    Widget := stack_box: # StackBox_29
                                                                        Orientation := orientation.Horizontal
                                                                        Slots := array:
                                                                            stack_box_slot: # StackBoxSlot_22
                                                                                HorizontalAlignment := horizontal_alignment.Right
                                                                                VerticalAlignment := vertical_alignment.Top
                                                                                Padding := margin:
                                                                                    Top := 0.0
                                                                                    Right := 0.0
                                                                                    Bottom := 0.0
                                                                                    Left := 0.0
                                                                                Widget := RewardFCount
                                                                            
                                                                stack_box_slot: # StackBoxSlot_3
                                                                    HorizontalAlignment := horizontal_alignment.Left
                                                                    VerticalAlignment := vertical_alignment.Top
                                                                    Padding := margin:
                                                                        Top := -7.5
                                                                        Right := 0.0
                                                                        Bottom := 0.0
                                                                        Left := 0.0
                                                                    Widget := RewardFName
                                                                stack_box_slot: # StackBoxSlot_6
                                                                    HorizontalAlignment := horizontal_alignment.Left
                                                                    VerticalAlignment := vertical_alignment.Top
                                                                    Padding := margin:
                                                                        Top := -7.5
                                                                        Right := 0.0
                                                                        Bottom := 0.0
                                                                        Left := 0.0
                                                                    Widget := RewardFPercent
                                                                
                                                    
                                        stack_box_slot: # StackBoxSlot_6
                                            HorizontalAlignment := horizontal_alignment.Left
                                            VerticalAlignment := vertical_alignment.Center
                                            Padding := margin:
                                                Top := 20.0
                                                Right := 0.0
                                                Bottom := 0.0
                                                Left := 25.0
                                            Widget := stack_box: # StackBox_24
                                                Orientation := orientation.Horizontal
                                                Slots := array:
                                                    stack_box_slot: # StackBoxSlot_7
                                                        HorizontalAlignment := horizontal_alignment.Left
                                                        VerticalAlignment := vertical_alignment.Top
                                                        Padding := margin:
                                                            Top := 5.0
                                                            Right := 0.0
                                                            Bottom := 0.0
                                                            Left := 0.0
                                                        Widget := RewardGIcon
                                                    stack_box_slot: # StackBoxSlot_1
                                                        HorizontalAlignment := horizontal_alignment.Right
                                                        VerticalAlignment := vertical_alignment.Top
                                                        Padding := margin:
                                                            Top := 0.0
                                                            Right := 0.0
                                                            Bottom := 0.0
                                                            Left := 12.0
                                                        Widget := stack_box: # StackBox_25
                                                            Orientation := orientation.Vertical
                                                            Slots := array:
                                                                stack_box_slot: # StackBoxSlot_0
                                                                    HorizontalAlignment := horizontal_alignment.Left
                                                                    VerticalAlignment := vertical_alignment.Top
                                                                    Padding := margin:
                                                                        Top := -7.5
                                                                        Right := 0.0
                                                                        Bottom := 0.0
                                                                        Left := 0.0
                                                                    Widget := stack_box: # StackBox_26
                                                                        Orientation := orientation.Horizontal
                                                                        Slots := array:
                                                                            stack_box_slot: # StackBoxSlot_22
                                                                                HorizontalAlignment := horizontal_alignment.Right
                                                                                VerticalAlignment := vertical_alignment.Top
                                                                                Padding := margin:
                                                                                    Top := 0.0
                                                                                    Right := 0.0
                                                                                    Bottom := 0.0
                                                                                    Left := 0.0
                                                                                Widget := RewardGCount
                                                                            
                                                                stack_box_slot: # StackBoxSlot_3
                                                                    HorizontalAlignment := horizontal_alignment.Left
                                                                    VerticalAlignment := vertical_alignment.Top
                                                                    Padding := margin:
                                                                        Top :=-7.5
                                                                        Right := 0.0
                                                                        Bottom := 0.0
                                                                        Left := 0.0
                                                                    Widget := RewardGName
                                                                stack_box_slot: # StackBoxSlot_6
                                                                    HorizontalAlignment := horizontal_alignment.Left
                                                                    VerticalAlignment := vertical_alignment.Top
                                                                    Padding := margin:
                                                                        Top := -7.5
                                                                        Right := 0.0
                                                                        Bottom := 0.0
                                                                        Left := 0.0
                                                                    Widget := RewardGPercent
                                                                
                                                    
                                        
                            
                canvas_slot: # CanvasPanelSlot_120
                    Anchors := anchors:
                        Minimum := vector2{X := 0.0,Y := 0.0}
                        Maximum := vector2{X := 1.0,Y := 1.0}
                    Offsets := margin:
                        Top := 136.0
                        Right := 16.588623
                        Bottom := 624.900879
                        Left := 972.0
                    SizeToContent := true
                    Alignment := vector2{X := 0.0,Y := 0.0}
                    
                    Widget := stack_box: # StackBox_45
                        Orientation := orientation.Vertical
                        Slots := array:
                            stack_box_slot: # StackBoxSlot_2
                                HorizontalAlignment := horizontal_alignment.Left
                                VerticalAlignment := vertical_alignment.Top
                                Padding := margin:
                                    Top := 15.0
                                    Right := 0.0
                                    Bottom := 0.0
                                    Left := 0.0
                                Widget := stack_box: # StackBox_38
                                    Orientation := orientation.Horizontal
                                    Slots := array:
                                        stack_box_slot: # StackBoxSlot_1
                                            HorizontalAlignment := horizontal_alignment.Left
                                            VerticalAlignment := vertical_alignment.Center
                                            Padding := margin:
                                                Top := 0.0
                                                Right := 0.0
                                                Bottom := 0.0
                                                Left := 60.0
                                            Widget := stack_box: # StackBox_37
                                                Orientation := orientation.Vertical
                                                Slots := array:
                                                    stack_box_slot: # StackBoxSlot_2
                                                        HorizontalAlignment := horizontal_alignment.Center
                                                        VerticalAlignment := vertical_alignment.Top
                                                        Padding := margin:
                                                            Top := 15.0
                                                            Right := 0.0
                                                            Bottom := 0.0
                                                            Left := 0.0
                                                        Widget := StatIcon
                                                    stack_box_slot: # StackBoxSlot_3
                                                        HorizontalAlignment := horizontal_alignment.Center
                                                        VerticalAlignment := vertical_alignment.Top
                                                        Padding := margin:
                                                            Top := 15.0
                                                            Right := 0.0
                                                            Bottom := 0.0
                                                            Left := 0.0
                                                        Widget := StatRankIcon
                                                    stack_box_slot: # StackBoxSlot_1
                                                        HorizontalAlignment := horizontal_alignment.Center
                                                        VerticalAlignment := vertical_alignment.Top
                                                        Padding := margin:
                                                            Top := 0.0
                                                            Right := 0.0
                                                            Bottom := 0.0
                                                            Left := 0.0
                                                        Widget := StatName
                                                    stack_box_slot: # StackBoxSlot_4
                                                        HorizontalAlignment := horizontal_alignment.Center
                                                        VerticalAlignment := vertical_alignment.Top
                                                        Padding := margin:
                                                            Top := 15.0
                                                            Right := 0.0
                                                            Bottom := 0.0
                                                            Left := 0.0
                                                        Widget := stack_box: # StackBox_36
                                                            Orientation := orientation.Horizontal
                                                            Slots := array:
                                                                stack_box_slot: # StackBoxSlot_0
                                                                    HorizontalAlignment := horizontal_alignment.Left
                                                                    VerticalAlignment := vertical_alignment.Top
                                                                    Padding := margin:
                                                                        Top := 0.0
                                                                        Right := 0.0
                                                                        Bottom := 0.0
                                                                        Left := 5.0
                                                                    Widget := StatLvTitle
                                                                stack_box_slot: # StackBoxSlot_1
                                                                    HorizontalAlignment := horizontal_alignment.Left
                                                                    VerticalAlignment := vertical_alignment.Top
                                                                    Padding := margin:
                                                                        Top := 0.0
                                                                        Right := 0.0
                                                                        Bottom := 0.0
                                                                        Left := 15.0
                                                                    Widget := StatLvValue
                                                                
                                                    
                                        stack_box_slot: # StackBoxSlot_0
                                            HorizontalAlignment := horizontal_alignment.Right
                                            VerticalAlignment := vertical_alignment.Top
                                            Padding := margin:
                                                Top := 0.0
                                                Right := 0.0
                                                Bottom := 0.0
                                                Left := 120.0
                                            Widget := stack_box: # StackBox_35
                                                Orientation := orientation.Vertical
                                                Slots := array:
                                                    stack_box_slot: # StackBoxSlot_5
                                                        HorizontalAlignment := horizontal_alignment.Left
                                                        VerticalAlignment := vertical_alignment.Top
                                                        Padding := margin:
                                                            Top := 0.0
                                                            Right := 0.0
                                                            Bottom := 0.0
                                                            Left := 0.0
                                                        Widget := StatTitle
                                                    stack_box_slot: # StackBoxSlot_0
                                                        HorizontalAlignment := horizontal_alignment.Left
                                                        VerticalAlignment := vertical_alignment.Top
                                                        Padding := margin:
                                                            Top := 0.0
                                                            Right := 0.0
                                                            Bottom := 0.0
                                                            Left := 0.0
                                                        Widget := stack_box: # StackBox_30
                                                            Orientation := orientation.Horizontal
                                                            Slots := array:
                                                                stack_box_slot: # StackBoxSlot_2
                                                                    HorizontalAlignment := horizontal_alignment.Left
                                                                    VerticalAlignment := vertical_alignment.Top
                                                                    Padding := margin:
                                                                        Top := 0.0
                                                                        Right := 0.0
                                                                        Bottom := 0.0
                                                                        Left := 7.0
                                                                    Widget := StatAIcon
                                                                stack_box_slot: # StackBoxSlot_0
                                                                    HorizontalAlignment := horizontal_alignment.Center
                                                                    VerticalAlignment := vertical_alignment.Top
                                                                    Padding := margin:
                                                                        Top := 0.0
                                                                        Right := 0.0
                                                                        Bottom := 0.0
                                                                        Left := 5.0
                                                                    Widget := StatATitle
                                                                stack_box_slot: # StackBoxSlot_1
                                                                    HorizontalAlignment := horizontal_alignment.Right
                                                                    VerticalAlignment := vertical_alignment.Top
                                                                    Padding := margin:
                                                                        Top := 0.0
                                                                        Right := 0.0
                                                                        Bottom := 0.0
                                                                        Left := 165.0
                                                                    Widget := StatAValue
                                                                
                                                    stack_box_slot: # StackBoxSlot_1
                                                        HorizontalAlignment := horizontal_alignment.Left
                                                        VerticalAlignment := vertical_alignment.Top
                                                        Padding := margin:
                                                            Top := 0.0
                                                            Right := 0.0
                                                            Bottom := 0.0
                                                            Left := 0.0
                                                        Widget := stack_box: # StackBox_31
                                                            Orientation := orientation.Horizontal
                                                            Slots := array:
                                                                stack_box_slot: # StackBoxSlot_2
                                                                    HorizontalAlignment := horizontal_alignment.Left
                                                                    VerticalAlignment := vertical_alignment.Top
                                                                    Padding := margin:
                                                                        Top := 0.0
                                                                        Right := 0.0
                                                                        Bottom := 0.0
                                                                        Left := 7.0
                                                                    Widget := StatBIcon
                                                                stack_box_slot: # StackBoxSlot_0
                                                                    HorizontalAlignment := horizontal_alignment.Center
                                                                    VerticalAlignment := vertical_alignment.Top
                                                                    Padding := margin:
                                                                        Top := 0.0
                                                                        Right := 0.0
                                                                        Bottom := 0.0
                                                                        Left := 7.5
                                                                    Widget := StatBTitle
                                                                stack_box_slot: # StackBoxSlot_1
                                                                    HorizontalAlignment := horizontal_alignment.Right
                                                                    VerticalAlignment := vertical_alignment.Top
                                                                    Padding := margin:
                                                                        Top := 0.0
                                                                        Right := 0.0
                                                                        Bottom := 0.0
                                                                        Left := 165.0
                                                                    Widget := StatBValue
                                                                
                                                    stack_box_slot: # StackBoxSlot_2
                                                        HorizontalAlignment := horizontal_alignment.Left
                                                        VerticalAlignment := vertical_alignment.Top
                                                        Padding := margin:
                                                            Top := 0.0
                                                            Right := 0.0
                                                            Bottom := 0.0
                                                            Left := 0.0
                                                        Widget := stack_box: # StackBox_32
                                                            Orientation := orientation.Horizontal
                                                            Slots := array:
                                                                stack_box_slot: # StackBoxSlot_2
                                                                    HorizontalAlignment := horizontal_alignment.Left
                                                                    VerticalAlignment := vertical_alignment.Top
                                                                    Padding := margin:
                                                                        Top := 0.0
                                                                        Right := 0.0
                                                                        Bottom := 0.0
                                                                        Left := 0.0
                                                                    Widget := StatCIcon
                                                                stack_box_slot: # StackBoxSlot_0
                                                                    HorizontalAlignment := horizontal_alignment.Center
                                                                    VerticalAlignment := vertical_alignment.Top
                                                                    Padding := margin:
                                                                        Top := 0.0
                                                                        Right := 0.0
                                                                        Bottom := 0.0
                                                                        Left := 10.0
                                                                    Widget := StatCTitle
                                                                stack_box_slot: # StackBoxSlot_3
                                                                    HorizontalAlignment := horizontal_alignment.Right
                                                                    VerticalAlignment := vertical_alignment.Top
                                                                    Padding := margin:
                                                                        Top := 0.0
                                                                        Right := 0.0
                                                                        Bottom := 0.0
                                                                        Left := 165.0
                                                                    Widget := StatCValue
                                                                
                                                    stack_box_slot: # StackBoxSlot_3
                                                        HorizontalAlignment := horizontal_alignment.Left
                                                        VerticalAlignment := vertical_alignment.Top
                                                        Padding := margin:
                                                            Top := 0.0
                                                            Right := 0.0
                                                            Bottom := 0.0
                                                            Left := 0.0
                                                        Widget := stack_box: # StackBox_33
                                                            Orientation := orientation.Horizontal
                                                            Slots := array:
                                                                stack_box_slot: # StackBoxSlot_2
                                                                    HorizontalAlignment := horizontal_alignment.Left
                                                                    VerticalAlignment := vertical_alignment.Top
                                                                    Padding := margin:
                                                                        Top := 0.0
                                                                        Right := 0.0
                                                                        Bottom := 0.0
                                                                        Left := 0.0
                                                                    Widget := StatDIcon
                                                                stack_box_slot: # StackBoxSlot_0
                                                                    HorizontalAlignment := horizontal_alignment.Center
                                                                    VerticalAlignment := vertical_alignment.Top
                                                                    Padding := margin:
                                                                        Top := 0.0
                                                                        Right := 0.0
                                                                        Bottom := 0.0
                                                                        Left := 10.0
                                                                    Widget := StatDTitle
                                                                stack_box_slot: # StackBoxSlot_3
                                                                    HorizontalAlignment := horizontal_alignment.Right
                                                                    VerticalAlignment := vertical_alignment.Top
                                                                    Padding := margin:
                                                                        Top := 0.0
                                                                        Right := 0.0
                                                                        Bottom := 0.0
                                                                        Left := 165.0
                                                                    Widget := StatDValue
                                                                
                                                    stack_box_slot: # StackBoxSlot_4
                                                        HorizontalAlignment := horizontal_alignment.Left
                                                        VerticalAlignment := vertical_alignment.Top
                                                        Padding := margin:
                                                            Top := 0.0
                                                            Right := 0.0
                                                            Bottom := 0.0
                                                            Left := 0.0
                                                        Widget := stack_box: # StackBox_34
                                                            Orientation := orientation.Horizontal
                                                            Slots := array:
                                                                stack_box_slot: # StackBoxSlot_2
                                                                    HorizontalAlignment := horizontal_alignment.Left
                                                                    VerticalAlignment := vertical_alignment.Top
                                                                    Padding := margin:
                                                                        Top := 0.0
                                                                        Right := 0.0
                                                                        Bottom := 0.0
                                                                        Left := 5.0
                                                                    Widget := StatEIcon
                                                                stack_box_slot: # StackBoxSlot_0
                                                                    HorizontalAlignment := horizontal_alignment.Center
                                                                    VerticalAlignment := vertical_alignment.Top
                                                                    Padding := margin:
                                                                        Top := 0.0
                                                                        Right := 0.0
                                                                        Bottom := 0.0
                                                                        Left := 10.0
                                                                    Widget := StatETitle
                                                                stack_box_slot: # StackBoxSlot_3
                                                                    HorizontalAlignment := horizontal_alignment.Right
                                                                    VerticalAlignment := vertical_alignment.Top
                                                                    Padding := margin:
                                                                        Top := 0.0
                                                                        Right := 0.0
                                                                        Bottom := 0.0
                                                                        Left := 165.0
                                                                    Widget := StatEValue
                                                                
                                                    
                                        
                            
                # canvas_slot: # CanvasPanelSlot_0
                #     Anchors := anchors:
                #         Minimum := vector2{X := 0.0,Y := 0.0}
                #         Maximum := vector2{X := 1.0,Y := 1.0}
                #     Offsets := margin:
                #         Top := 488.0
                #         Right := 48.066162
                #         Bottom := 529.3573
                #         Left := 968.0
                #     SizeToContent := true
                #     Alignment := vector2{X := 0.0,Y := 0.0}
                    
                    # Widget := stack_box: # StackBox_0
                    #     Orientation := orientation.Horizontal
                    #     Slots := array:
                    #         stack_box_slot: # StackBoxSlot_0
                    #             HorizontalAlignment := horizontal_alignment.Left
                    #             VerticalAlignment := vertical_alignment.Top
                    #             Padding := margin:
                    #                 Top := 0.0
                    #                 Right := 30.0
                    #                 Bottom := 0.0
                    #                 Left := 0.0
                    #             Widget := SelectEvolve
                    #         stack_box_slot: # StackBoxSlot_1
                    #             HorizontalAlignment := horizontal_alignment.Left
                    #             VerticalAlignment := vertical_alignment.Top
                    #             Padding := margin:
                    #                 Top := 0.0
                    #                 Right := 0.0
                    #                 Bottom := 0.0
                    #                 Left := 20.0
                    #             Widget := SelectEvolve2
                    #         stack_box_slot: # StackBoxSlot_1
                    #             HorizontalAlignment := horizontal_alignment.Left
                    #             VerticalAlignment := vertical_alignment.Top
                    #             Padding := margin:
                    #                 Top := 0.0
                    #                 Right := 0.0
                    #                 Bottom := 0.0
                    #                 Left := 40.0
                    #             Widget := SelectEvolve3
                    #         stack_box_slot: # StackBoxSlot_1
                    #             HorizontalAlignment := horizontal_alignment.Left
                    #             VerticalAlignment := vertical_alignment.Top
                    #             Padding := margin:
                    #                 Top := 0.0
                    #                 Right := 0.0
                    #                 Bottom := 0.0
                    #                 Left := 50.0
                    #             Widget := SelectEvolve4
                            
                canvas_slot: # CanvasPanelSlot_5
                    Anchors := anchors:
                        Minimum := vector2{X := 0.0,Y := 0.0}
                        Maximum := vector2{X := 1.0,Y := 1.0}
                    Offsets := margin:
                        Top := 608.0
                        Right := 47.207275
                        Bottom := 146.894897
                        Left := 960.0
                    SizeToContent := true
                    Alignment := vector2{X := 0.0,Y := 0.0}
                    
                    Widget := stack_box: # StackBox_47
                        Orientation := orientation.Vertical
                        Slots := array:
                            stack_box_slot: # StackBoxSlot_2
                                HorizontalAlignment := horizontal_alignment.Left
                                VerticalAlignment := vertical_alignment.Top
                                Padding := margin:
                                    Top := 10.0
                                    Right := 0.0
                                    Bottom := 0.0
                                    Left := 0.0
                                Widget := stack_box: # StackBox_41
                                    Orientation := orientation.Horizontal
                                    Slots := array:
                                        stack_box_slot: # StackBoxSlot_1
                                            HorizontalAlignment := horizontal_alignment.Center
                                            VerticalAlignment := vertical_alignment.Top
                                            Padding := margin:
                                                Top := 0.0
                                                Right := 0.0
                                                Bottom := 0.0
                                                Left := 10.0
                                            Widget := CostTitle
                                        stack_box_slot: # StackBoxSlot_0
                                            HorizontalAlignment := horizontal_alignment.Center
                                            VerticalAlignment := vertical_alignment.Center
                                            Padding := margin:
                                                Top := 0.0
                                                Right := 0.0
                                                Bottom := 0.0
                                                Left := 120.0
                                            Widget := stack_box: # StackBox_40
                                                Orientation := orientation.Vertical
                                                Slots := array:
                                                    stack_box_slot: # StackBoxSlot_1
                                                        HorizontalAlignment := horizontal_alignment.Left
                                                        VerticalAlignment := vertical_alignment.Top
                                                        Padding := margin:
                                                            Top := 0.0
                                                            Right := 0.0
                                                            Bottom := 0.0
                                                            Left := 50.0
                                                        Widget := stack_box: # StackBox_8
                                                            Orientation := orientation.Horizontal
                                                            Slots := array:
                                                                stack_box_slot: # StackBoxSlot_0
                                                                    HorizontalAlignment := horizontal_alignment.Left
                                                                    VerticalAlignment := vertical_alignment.Top
                                                                    Padding := margin:
                                                                        Top := 0.0
                                                                        Right := 0.0
                                                                        Bottom := 0.0
                                                                        Left := 0.0
                                                                    Widget := CostAIcon
                                                                stack_box_slot: # StackBoxSlot_1
                                                                    HorizontalAlignment := horizontal_alignment.Right
                                                                    VerticalAlignment := vertical_alignment.Center
                                                                    Padding := margin:
                                                                        Top := 0.0
                                                                        Right := 0.0
                                                                        Bottom := 0.0
                                                                        Left := 12.0
                                                                    Widget := CostAValue
                                                                
                                                    stack_box_slot: # StackBoxSlot_0
                                                        HorizontalAlignment := horizontal_alignment.Left
                                                        VerticalAlignment := vertical_alignment.Center
                                                        Padding := margin:
                                                            Top := 0.0
                                                            Right := 0.0
                                                            Bottom := 0.0
                                                            Left := 50.0
                                                        Widget := stack_box: # StackBox_7
                                                            Orientation := orientation.Horizontal
                                                            Slots := array:
                                                                stack_box_slot: # StackBoxSlot_0
                                                                    HorizontalAlignment := horizontal_alignment.Left
                                                                    VerticalAlignment := vertical_alignment.Top
                                                                    Padding := margin:
                                                                        Top := 0.0
                                                                        Right := 0.0
                                                                        Bottom := 0.0
                                                                        Left := 0.0
                                                                    Widget := CostBIcon
                                                                stack_box_slot: # StackBoxSlot_1
                                                                    HorizontalAlignment := horizontal_alignment.Right
                                                                    VerticalAlignment := vertical_alignment.Center
                                                                    Padding := margin:
                                                                        Top := 0.0
                                                                        Right := 0.0
                                                                        Bottom := 0.0
                                                                        Left := 12.0
                                                                    Widget := CostBValue
                                                                
                                                    
                                        
                            stack_box_slot: # StackBoxSlot_0
                                HorizontalAlignment := horizontal_alignment.Left
                                VerticalAlignment := vertical_alignment.Top
                                Padding := margin:
                                    Top := 40.0
                                    Right := 0.0
                                    Bottom := 0.0
                                    Left := 0.0
                                Widget := stack_box: # StackBox_39
                                    Orientation := orientation.Horizontal
                                    Slots := array:
                                        stack_box_slot: # StackBoxSlot_0
                                            HorizontalAlignment := horizontal_alignment.Left
                                            VerticalAlignment := vertical_alignment.Center
                                            Padding := margin:
                                                Top := 0.0
                                                Right := 50.0
                                                Bottom := 0.0
                                                Left := 0.0
                                            Widget := SelectLeft
                                        stack_box_slot: # StackBoxSlot_1
                                            HorizontalAlignment := horizontal_alignment.Left
                                            VerticalAlignment := vertical_alignment.Center
                                            Padding := margin:
                                                Top := 0.0
                                                Right := 0.0
                                                Bottom := 0.0
                                                Left := 50.0
                                            Widget := SelectRight
                                        
                            stack_box_slot: # StackBoxSlot_3
                                HorizontalAlignment := horizontal_alignment.Left
                                VerticalAlignment := vertical_alignment.Top
                                Padding := margin:
                                    Top := 15.0
                                    Right := 0.0
                                    Bottom := 0.0
                                    Left := 0.0
                                Widget := stack_box: # StackBox_46
                                    Orientation := orientation.Horizontal
                                    Slots := array:
                                        stack_box_slot: # StackBoxSlot_0
                                            HorizontalAlignment := horizontal_alignment.Left
                                            VerticalAlignment := vertical_alignment.Center
                                            Padding := margin:
                                                Top := 0.0
                                                Right := 10.0
                                                Bottom := 0.0
                                                Left := 0.0
                                            Widget := SelectActionOpen
                                        stack_box_slot: # StackBoxSlot_1
                                            HorizontalAlignment := horizontal_alignment.Left
                                            VerticalAlignment := vertical_alignment.Center
                                            Padding := margin:
                                                Top := 0.0
                                                Right := 0.0
                                                Bottom := 0.0
                                                Left := 10.0
                                            Widget := SelectActionAuto
                                        
                            
                