using { /Fortnite.com/Devices }
using { /Fortnite.com/Characters }
using { /Fortnite.com/UI }
using { /Verse.org/Simulation }
using { /Verse.org/Colors }
using { /UnrealEngine.com/Temporary/Diagnostics }
using { /UnrealEngine.com/Temporary/UI }
using { /UnrealEngine.com/Temporary/SpatialMath }
using { /Fortnite.com/Devices/CreativeAnimation } 
using {Mining_System.Code.PlayerData}
using { Mining_System.Code.Pet }

StringToMessages<localizes>(value:string)<computes> : message = "{value}"

pet_upgrade_ui := class:

    # widgets
    BGImage : texture_block = texture_block{DefaultImage := icon_set_2.bg, DefaultDesiredSize := vector2{X := 1280.0, Y := 720.0}}
    Youkai1_Evlove : button_quiet = button_quiet{}
    Youkai_Select_Right : button_quiet = button_quiet{}
    Youkai_Select_Left : button_quiet = button_quiet{}
    Youkai_Tex : texture_block = texture_block{DefaultImage := Boss_Textures.koppar, DefaultDesiredSize := vector2{X := 200.0, Y := 200.0}}
    _CloseButton : button_loud = button_loud{}

    Yokai_NAME : text_block = text_block{DefaultTextColor := NamedColors.White, DefaultJustification := text_justification.Center}

    NowLevel : text_block = text_block{DefaultTextColor := NamedColors.Red, DefaultJustification := text_justification.Left}
    NeedLevel : text_block = text_block{DefaultTextColor := NamedColors.White, DefaultJustification := text_justification.Left}

    NeedSilverOre : texture_block = texture_block{DefaultImage := icon_set_2.SilverOre, DefaultDesiredSize := vector2{X := 50.0, Y := 50.0}}
    NeedSilverOreAmount : text_block = text_block{DefaultTextColor := NamedColors.Red, DefaultJustification := text_justification.Left}
    NeedIC_Power : texture_block = texture_block{DefaultImage := icon_set_2.IC_Power, DefaultDesiredSize := vector2{X := 50.0, Y := 50.0}}
    NeedIC_PowerAmount : text_block = text_block{DefaultTextColor := NamedColors.Red, DefaultJustification := text_justification.Left}

    Player<public> : player

    var MaybeCanvas : ?canvas = false

    CloseEvent<public> : event() = event(){}

    AddGameplayUI<public>():void=
        if(PlayerUI := GetPlayerUI[Player]):
            NewCanvas := CreateCanvas()
            InputMode := player_ui_slot:
                InputMode := ui_input_mode.All
            PlayerUI.AddWidget(NewCanvas, InputMode)
            set MaybeCanvas = option{NewCanvas}

    RemoveUI<public>():void=
        if(PlayerUI := GetPlayerUI[Player], OldCanvas := MaybeCanvas?):
            PlayerUI.RemoveWidget(OldCanvas)
            set MaybeCanvas = false
            CloseEvent.Signal()

    CreateCanvas():canvas=

        CenterVector : vector2 = vector2{X:=0.5, Y:=0.5}
        FullCenterAnchors : anchors = anchors{Minimum := vector2{X:=0.0, Y:=0.0}, Maximum := vector2{X:=1.0, Y:=1.0}}
        Canvas := canvas:
            Slots := array:
                canvas_slot:
                    Anchors := anchors{Minimum := vector2{X := 0.0, Y := 0.0}, Maximum := vector2{X := 1.0, Y := 1.0}}
                    Widget := 
                        overlay: # Overlay_33
                            Slots := array:
                                overlay_slot:
                                    #========================================== BG ==========================================
                                    HorizontalAlignment := horizontal_alignment.Center
                                    VerticalAlignment := vertical_alignment.Center
                                    Padding := margin:
                                        Top := 0.0
                                        Right := 0.0
                                        Bottom := 0.0
                                        Left := 0.0
                                    Widget := BGImage

                                overlay_slot:
                                    HorizontalAlignment := horizontal_alignment.Center
                                    VerticalAlignment := vertical_alignment.Center
                                    Padding := margin:
                                        Top := 0.0
                                        Right := 0.0
                                        Bottom := 0.0
                                        Left := 0.0
                                    Widget := Youkai_Tex

                                overlay_slot:
                                    HorizontalAlignment := horizontal_alignment.Center
                                    VerticalAlignment := vertical_alignment.Center
                                    Padding := margin:
                                        Top := 225.0
                                        Right := 150.0
                                        Bottom := 0.0
                                        Left := 0.0
                                    Widget := NeedSilverOre

                                overlay_slot:
                                    HorizontalAlignment := horizontal_alignment.Center
                                    VerticalAlignment := vertical_alignment.Center
                                    Padding := margin:
                                        Top := 225.0
                                        Right := -50.0
                                        Bottom := 0.0
                                        Left := 0.0
                                    Widget := NeedIC_Power
                                #========================================= TEXT ========================================

                                overlay_slot:
                                    HorizontalAlignment := horizontal_alignment.Center
                                    VerticalAlignment := vertical_alignment.Top
                                    Padding := margin:
                                        Top := 205.0
                                        Right := 0.0
                                        Bottom := 0.0
                                        Left := 0.0
                                    Widget := text_block:
                                        DefaultText := StringToMessages("Evolve")
                                        DefaultTextColor := NamedColors.White
                                        DefaultJustification := text_justification.Center

                                overlay_slot:
                                    HorizontalAlignment := horizontal_alignment.Center
                                    VerticalAlignment := vertical_alignment.Center
                                    Padding := margin:
                                        Top := -200.0
                                        Right := 0.0
                                        Bottom := 0.0
                                        Left := 0.0
                                    Widget := Yokai_NAME

                                overlay_slot:
                                    HorizontalAlignment := horizontal_alignment.Center
                                    VerticalAlignment := vertical_alignment.Center
                                    Padding := margin:
                                        Top := 130.0
                                        Right := 0.0
                                        Bottom := 0.0
                                        Left := 0.0
                                    Widget := NowLevel

                                overlay_slot:
                                    HorizontalAlignment := horizontal_alignment.Center
                                    VerticalAlignment := vertical_alignment.Center
                                    Padding := margin:
                                        Top := 170.0
                                        Right := 0.0
                                        Bottom := 0.0
                                        Left := 0.0
                                    Widget := NeedLevel

                                overlay_slot:
                                    HorizontalAlignment := horizontal_alignment.Center
                                    VerticalAlignment := vertical_alignment.Center
                                    Padding := margin:
                                        Top := 225.0
                                        Right := 50.0
                                        Bottom := 0.0
                                        Left := 0.0
                                    Widget := NeedSilverOreAmount

                                overlay_slot:
                                    HorizontalAlignment := horizontal_alignment.Center
                                    VerticalAlignment := vertical_alignment.Center
                                    Padding := margin:
                                        Top := 225.0
                                        Right := -150.0
                                        Bottom := 0.0
                                        Left := 0.0
                                    Widget := NeedIC_PowerAmount

                                #======================================= StackBox ======================================

                                overlay_slot:
                                    HorizontalAlignment := horizontal_alignment.Center
                                    VerticalAlignment := vertical_alignment.Center
                                    Padding := margin:
                                        Top := 300.0
                                        Right := 0.0
                                        Bottom := 0.0
                                        Left := 0.0
                                    Widget := Youkai1_Evlove
                                overlay_slot:
                                    HorizontalAlignment := horizontal_alignment.Center
                                    VerticalAlignment := vertical_alignment.Center
                                    Padding := margin:
                                        Top := 0.0
                                        Right := -400.0
                                        Bottom := 0.0
                                        Left := 0.0
                                    Widget := Youkai_Select_Right
                                overlay_slot:
                                    HorizontalAlignment := horizontal_alignment.Center
                                    VerticalAlignment := vertical_alignment.Center
                                    Padding := margin:
                                        Top := 0.0
                                        Right := 400.0
                                        Bottom := 0.0
                                        Left := 0.0
                                    Widget := Youkai_Select_Left
                                overlay_slot:
                                    HorizontalAlignment := horizontal_alignment.Center
                                    VerticalAlignment := vertical_alignment.Center
                                    Padding := margin:
                                        Top := 440.0
                                        Right := 0.0
                                        Bottom := 0.0
                                        Left := 0.0
                                    Widget := _CloseButton

                                #=======================================================================================

        return Canvas

pet_upgrade_device<public> := class(creative_device):
    #SFXs
    @editable
    SFX_Open : audio_player_device = audio_player_device{}
    @editable
    SFX_Close : audio_player_device = audio_player_device{}
    @editable
    SFX_LevelUp3 : audio_player_device = audio_player_device{}
    @editable
    SFX_LevelUp4 : audio_player_device = audio_player_device{}

    @editable
    SaveDataManager : save_data_manager = save_data_manager{}

    @editable
    PetManager : pet_manager = pet_manager{}

    @editable
    _LangSelectButton : character_device = character_device{}

    @editable
    PetAddLevelDebug : button_device = button_device{}

    PETIndexArray : []int = array{1, 3, 2, 4, 5, 11, 13, 12, 14, 15}

    var PlayerSelectingYokai : [player]int = map{}
  
    OnBegin<override>()<suspends>:void=
      _LangSelectButton.InteractedWithEvent.Subscribe(SpawnWidget)
      PetAddLevelDebug.InteractedWithEvent.Subscribe(AddDebug)

    AddDebug(Agent:agent):void=
        if(Player := player[Agent]):
            PetManager.UnlockPet(Player, 1)
            PetManager.SetPetLevel(Player, 1, 30)
            SaveDataManager.SetResourcePowerupStone(Player, big_number{Value := 100.0})

    
    GetSelectionYokaiName(InPlayer : player, Index:int, Widget : pet_upgrade_ui):string=

        var TheYokaiName : string = "??????"
        var NeedStone : big_number = big_number{}
        var NeetPowerUpStone : big_number = big_number{}

        GetNowUpgrade := PetManager.GetPetUpgrade(InPlayer, Index)

        if(HasPet := SaveDataManager.GetPet[InPlayer, Index]):

            if(Index = 1):
                set TheYokaiName = "Kopper"
                Widget.Youkai_Tex.SetImage(Boss_Textures.koppar)
                set NeedStone = big_number{Value :=22.0, Exponent := 6}
                set NeetPowerUpStone = big_number{Value :=24.0, Exponent := 0}
            if(Index = 3):
                set TheYokaiName = "Kachinta"
                Widget.Youkai_Tex.SetImage(Boss_Textures.kachinta)
                set NeedStone = big_number{Value :=22.0, Exponent := 6}
                set NeetPowerUpStone = big_number{Value :=24.0, Exponent := 0}
            if(Index = 2):
                set TheYokaiName = "kogasa"
                Widget.Youkai_Tex.SetImage(Boss_Textures.kogasa)
                set NeedStone = big_number{Value :=80.0, Exponent := 18}
                set NeetPowerUpStone = big_number{Value :=256.0, Exponent := 0}
            if(Index = 4):
                set TheYokaiName = "Hagiremomen"
                Widget.Youkai_Tex.SetImage(Boss_Textures.hagiremomen)
                set NeedStone = big_number{Value :=80.0, Exponent := 18}
                set NeetPowerUpStone = big_number{Value :=256.0, Exponent := 0}
            if(Index = 5):
                set TheYokaiName = "Kajimaru"
                Widget.Youkai_Tex.SetImage(Boss_Textures.kajimaru)
                set NeedStone = big_number{Value :=540.0, Exponent := 24}
                set NeetPowerUpStone = big_number{Value :=2.0, Exponent := 3}

            if(Index = 11):
                set TheYokaiName = "DCKopper"
                Widget.Youkai_Tex.SetImage(Boss_Textures._koppa)
                set NeedStone = big_number{Value :=22.0, Exponent := 6}
                set NeetPowerUpStone = big_number{Value :=24.0, Exponent := 0}
            if(Index = 13):
                set TheYokaiName = "DCKachinta"
                Widget.Youkai_Tex.SetImage(Boss_Textures._kachinta)
                set NeedStone = big_number{Value :=22.0, Exponent := 6}
                set NeetPowerUpStone = big_number{Value :=24.0, Exponent := 0}
            if(Index = 12):
                set TheYokaiName = "DCKogasa"
                Widget.Youkai_Tex.SetImage(Boss_Textures._kogasa)
                set NeedStone = big_number{Value :=80.0, Exponent := 18}
                set NeetPowerUpStone = big_number{Value :=256.0, Exponent := 0}
            if(Index = 14):
                set TheYokaiName = "DCHagiremomen"
                Widget.Youkai_Tex.SetImage(Boss_Textures._hagiremomen)
                set NeedStone = big_number{Value :=80.0, Exponent := 18}
                set NeetPowerUpStone = big_number{Value :=256.0, Exponent := 0}
            if(Index = 15):
                set TheYokaiName = "DCKajimaru"
                Widget.Youkai_Tex.SetImage(Boss_Textures._kajimaru)
                set NeedStone = big_number{Value :=540.0, Exponent := 24}
                set NeetPowerUpStone = big_number{Value :=2.0, Exponent := 3}

        # 持ってないときのテクスチャ
        else:
            if(Index = 1 or Index = 11):
                Widget.Youkai_Tex.SetImage(Boss_Textures.Pet_1)
                set NeedStone = big_number{Value :=22.0, Exponent := 6}
                set NeetPowerUpStone = big_number{Value :=24.0, Exponent := 0}
            if(Index = 3 or Index = 13):
                Widget.Youkai_Tex.SetImage(Boss_Textures.Pet_2)
                set NeedStone = big_number{Value :=22.0, Exponent := 6}
                set NeetPowerUpStone = big_number{Value :=24.0, Exponent := 0}
            if(Index = 2 or Index = 12):
                Widget.Youkai_Tex.SetImage(Boss_Textures.Pet_3)
                set NeedStone = big_number{Value :=80.0, Exponent := 18}
                set NeetPowerUpStone = big_number{Value :=256.0, Exponent := 0}
            if(Index = 4 or Index = 14):
                Widget.Youkai_Tex.SetImage(Boss_Textures.Pet_4)
                set NeedStone = big_number{Value :=80.0, Exponent := 18}
                set NeetPowerUpStone = big_number{Value :=256.0, Exponent := 0}
            if(Index = 5 or Index = 15):
                Widget.Youkai_Tex.SetImage(Boss_Textures.Pet_5)
                set NeedStone = big_number{Value :=540.0, Exponent := 24}
                set NeetPowerUpStone = big_number{Value :=2.0, Exponent := 3}

        # Widgetの素材数を更新
        Widget.NeedSilverOreAmount.SetText(StringToMessage(BigNumber_FormatDisplay(NeedStone)))
        Widget.NeedIC_PowerAmount.SetText(StringToMessage(BigNumber_FormatDisplay(NeetPowerUpStone)))

        # 今のリソースと比べて足りなければ赤字にする
        GetNowSilverOre := SaveDataManager.GetResourceStone(Widget.Player)
        GetNowIC_Power := SaveDataManager.GetResourcePowerupStone(Widget.Player)

        EnoughStone := BigNumber_GreaterThanOrEquals(GetNowSilverOre, NeedStone)
        EnoughIC_Power := BigNumber_GreaterThanOrEquals(GetNowIC_Power, NeetPowerUpStone)

        if(EnoughStone = true):
            Widget.NeedSilverOreAmount.SetTextColor(NamedColors.White)
        else:
            Widget.NeedSilverOreAmount.SetTextColor(NamedColors.Red)
        if(EnoughIC_Power = true):
            Widget.NeedIC_PowerAmount.SetTextColor(NamedColors.White)
        else:
            Widget.NeedIC_PowerAmount.SetTextColor(NamedColors.Red)

        # レベルが足りてなかったら赤字にする

        NowPetLevel := PetManager.GetPetLevel(InPlayer, Index)
        #UIローカライズ
        var LanguageIndex : int = SaveDataManager.GetLanguage(InPlayer)
        NowLvText := if(LanguageIndex = 0){"レベル  :  Lv.{NowPetLevel}"} else {"NowLevel  :  Lv.{NowPetLevel}"}
        NeedLvText := if(LanguageIndex = 0){"必要レベル  :  Lv.30"} else {"NeedLevel  :  Lv.30"}
        if(GetNowUpgrade = 0):
        
            if(NowPetLevel < 30):
                Widget.NowLevel.SetTextColor(NamedColors.Red)
            else:
                Widget.NowLevel.SetTextColor(NamedColors.White)
            
            Widget.NowLevel.SetText(StringToMessage(NowLvText))
            Widget.NeedLevel.SetText(StringToMessages(NeedLvText))

        if(GetNowUpgrade = 1):
            Widget.NowLevel.SetTextColor(NamedColors.White)
            Widget.NeedLevel.SetTextColor(NamedColors.Yellow)
            Widget.NowLevel.SetText(StringToMessage(NowLvText))
            Widget.NeedLevel.SetText(StringToMessages("To be updated map... Please Waiting!!"))

        return TheYokaiName


    # WidgetをAgentに追加
    SpawnWidget(Agent:agent):void=
        spawn{AddSelectLanguageWidget(Agent)}

    AddSelectLanguageWidget(Agent : agent)<suspends>:void=

        Sleep(0.5)

        if(Player := player[Agent], Widget := pet_upgrade_ui{Player := Player}):

            if(not PlayerSelectingYokai[Player]):
                logic{ set PlayerSelectingYokai[Player] = 0 }
            
            Widget.AddGameplayUI()
            SFX_Open.Play(Agent)

            Widget.Youkai1_Evlove.SetText(StringToMessages("Evolve"))
            Widget.Youkai_Select_Right.SetText(StringToMessages("    >    "))
            Widget.Youkai_Select_Left.SetText(StringToMessages("    <    "))
            Widget._CloseButton.SetText(StringToMessages("Close"))

            if(YokaiIndex := PETIndexArray[PlayerSelectingYokai[Widget.Player]]):
                TheYokaiName := GetSelectionYokaiName(Player, YokaiIndex, Widget)
                Widget.Yokai_NAME.SetText(StringToMessages("{TheYokaiName}"))
                spawn{OnEvlove1ButtonClicked(Widget)}
                spawn{OnRightButtonClicked(Widget)}
                spawn{OnLeftButtonClicked(Widget)}
                spawn{OnCancelButtonClicked(Widget)}
      
    
    OnEvlove1ButtonClicked(Widget : pet_upgrade_ui)<suspends>:void=
        race:
            Widget._CloseButton.OnClick().Await()
            loop:
                Widget.Youkai1_Evlove.OnClick().Await()

                Print("Click!!", ?Duration := 10.0)
                if(Agent : agent = agent[Widget.Player]) :
                    SFX_LevelUp4.Play(Agent)

                # RewardからPetのIDとStageIndexを取得
                if(YokaiIndex := PETIndexArray[PlayerSelectingYokai[Widget.Player]]):
                    # シグナルを送ることでPetBehaviorに実行を促す
                    var StageIndex : int = 0

                    if(YokaiIndex = 1 or YokaiIndex = 11):
                        set StageIndex = 0
                    
                    if(YokaiIndex = 2 or YokaiIndex = 12):
                        set StageIndex = 0

                    if(YokaiIndex = 3 or YokaiIndex = 13):
                        set StageIndex = 1

                    if(YokaiIndex = 4 or YokaiIndex = 14):
                        set StageIndex = 1

                    if(YokaiIndex = 5 or YokaiIndex = 15):
                        set StageIndex = 2

                    Print("YokaiIndex : {YokaiIndex}", ?Duration := 10.0)

                    PetManager.PetUpgradeSignal.Signal(Widget.Player, YokaiIndex, StageIndex)

    OnRightButtonClicked(Widget : pet_upgrade_ui)<suspends>:void=
        race:
            Widget._CloseButton.OnClick().Await()
            loop:
                Widget.Youkai_Select_Right.OnClick().Await()
                if(Agent : agent = agent[Widget.Player]) :
                    SFX_LevelUp3.Play(Agent)

                # 妖怪のIndexを1ずらす
                logic{ set PlayerSelectingYokai[Widget.Player] += 1}

                if(Index := PlayerSelectingYokai[Widget.Player] >= 10):
                    logic{ set PlayerSelectingYokai[Widget.Player] = 0}

                if(YokaiIndex := PETIndexArray[PlayerSelectingYokai[Widget.Player]]):
                    TheYokaiName := GetSelectionYokaiName(Widget.Player, YokaiIndex, Widget)
                    Widget.Yokai_NAME.SetText(StringToMessages("{TheYokaiName}"))
                    Print("Selecting Index:{YokaiIndex}", ?Duration := 10.0)

    OnLeftButtonClicked(Widget : pet_upgrade_ui)<suspends>:void=
        race:
            Widget._CloseButton.OnClick().Await()
            loop:
                Widget.Youkai_Select_Left.OnClick().Await()
                if(Agent : agent = agent[Widget.Player]) :
                    SFX_LevelUp3.Play(Agent)

                # 妖怪のIndexを1ずらす
                logic{ set PlayerSelectingYokai[Widget.Player] -= 1}

                if(Index := PlayerSelectingYokai[Widget.Player] <= -1):
                    logic{ set PlayerSelectingYokai[Widget.Player] = 9}

                if(YokaiIndex := PETIndexArray[PlayerSelectingYokai[Widget.Player]]):
                    TheYokaiName := GetSelectionYokaiName(Widget.Player, YokaiIndex, Widget)
                    Widget.Yokai_NAME.SetText(StringToMessages("{TheYokaiName}"))
                    Print("Selecting Index:{YokaiIndex}", ?Duration := 10.0)

    # Cancel押したとき
    OnCancelButtonClicked(Widget : pet_upgrade_ui)<suspends>:void=
        Widget._CloseButton.OnClick().Await()
        if(Agent : agent = agent[Widget.Player]) :
            SFX_Close.Play(Agent)
        Print("Cancel")
        Widget.RemoveUI()