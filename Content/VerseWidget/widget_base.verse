using { /Verse.org/Simulation }
using { /UnrealEngine.com/Temporary/UI }
using { /UnrealEngine.com/Temporary/SpatialMath }
using { /Fortnite.com/UI }
using { /Verse.org/Colors }
using { /Fortnite.com/Devices }

widget_device<public> := class(creative_device):

    AddUI<public>(Player:player):void=
        return

CloseButtonMessage<public><localizes>:message="Close"

widget_base<public> := class<abstract>:

    Player<public> : player

    var MaybeCanvas : ?canvas = false

    BasicBG<public> : basic_background = basic_background{}

    CloseEvent<public> : event() = event(){}

    Init():void=
        BasicBG.CloseButton.OnClick().Subscribe(OnClickedCloseButton)

    AddInteractUI<public>():void=
        if(PlayerUI := GetPlayerUI[Player]):
            NewCanvas := CreateCanvas()
            PlayerUI.AddWidget(NewCanvas, player_ui_slot{InputMode := ui_input_mode.All})
            set MaybeCanvas = option{NewCanvas}

    RemoveUI<public>():void=
        if(PlayerUI := GetPlayerUI[Player], OldCanvas := MaybeCanvas?):
            PlayerUI.RemoveWidget(OldCanvas)
            set MaybeCanvas = false
            CloseEvent.Signal()

    CreateCanvas():canvas

    OnClickedCloseButton(WidgetMessage:widget_message):void=
        RemoveUI()