using { /Fortnite.com/UI }
using { /UnrealEngine.com/Temporary/UI }
using { /Verse.org/Colors }
using { /UnrealEngine.com/Temporary/SpatialMath }

StringToMessage<localizes>(Text:string):message="{Text}"

basic_background<public> := class:
    BackgroundTexture<public> : texture_block = texture_block{DefaultImage := icon_set_2.bg, DefaultDesiredSize := vector2{X := 256.0, Y := 256.0}}
    Title<public> : text_block = text_block{DefaultTextColor := NamedColors.White}
    Underline<public> : color_block = color_block{DefaultColor := NamedColors.White, DefaultDesiredSize := vector2{Y := 5.0}}
    CloseButton<public> : button_loud = button_loud{DefaultText := CloseButtonMessage}

    CreateOverlay():canvas=
        canvas:
            Slots := array:
                canvas_slot:
                    Anchors := anchors{Minimum := vector2{X := 0.0, Y := 0.0}, Maximum := vector2{X := 1.0, Y := 1.0}}
                    Widget := 
                        overlay: # Overlay_33
                            Slots := array:
                                overlay_slot: # OverlaySlot_0
                                    HorizontalAlignment := horizontal_alignment.Fill
                                    VerticalAlignment := vertical_alignment.Fill
                                    Padding := margin:
                                        Top := 0.0
                                        Right := 0.0
                                        Bottom := 0.0
                                        Left := 0.0
                                    Widget := BackgroundTexture
                                overlay_slot: # OverlaySlot_1
                                    HorizontalAlignment := horizontal_alignment.Fill
                                    VerticalAlignment := vertical_alignment.Fill
                                    Padding := margin:
                                        Top := 0.0
                                        Right := 0.0
                                        Bottom := 0.0
                                        Left := 0.0
                                    Widget := stack_box: # StackBox_425
                                        Orientation := orientation.Vertical
                                        Slots := array:
                                            stack_box_slot: # StackBoxSlot_3
                                                HorizontalAlignment := horizontal_alignment.Fill
                                                VerticalAlignment := vertical_alignment.Top
                                                Padding := margin:
                                                    Top := 0.0
                                                    Right := 0.0
                                                    Bottom := 0.0
                                                    Left := 0.0
                                                Widget := stack_box: # StackBox_515
                                                    Orientation := orientation.Vertical
                                                    Slots := array:
                                                        stack_box_slot: # StackBoxSlot_0
                                                            HorizontalAlignment := horizontal_alignment.Center
                                                            VerticalAlignment := vertical_alignment.Top
                                                            Padding := margin:
                                                                Top := 25.0
                                                                Right := 0.0
                                                                Bottom := 0.0
                                                                Left := 0.0
                                                            Widget := Title
                                                        stack_box_slot: # StackBoxSlot_1
                                                            HorizontalAlignment := horizontal_alignment.Fill
                                                            VerticalAlignment := vertical_alignment.Center
                                                            Padding := margin:
                                                                Top := 0.0
                                                                Right := 100.0
                                                                Bottom := 0.0
                                                                Left := 100.0
                                                            Widget := Underline
                                                        
                                            stack_box_slot: # StackBoxSlot_7
                                                HorizontalAlignment := horizontal_alignment.Center
                                                VerticalAlignment := vertical_alignment.Bottom
                                                Distribution := option{0.25}
                                                Padding := margin:
                                                    Top := 25.0
                                                    Right := 0.0
                                                    Bottom := 25.0
                                                    Left := 0.0
                                                Widget := CloseButton
                        

LevelText<public><localizes>:message="Lv."
basic_yokai_icon<public> := class:

    SetEmpty():void=
        LevelTitle.SetVisibility(widget_visibility.Hidden)
        LevelValue.SetText(StringToMessage(""))
        Icon.SetVisibility(widget_visibility.Hidden)
        NameButton.SetText(StringToMessage("Empty"))

    SetEmpty(Index:int):void=
        LevelTitle.SetVisibility(widget_visibility.Hidden)
        LevelValue.SetText(StringToMessage(""))
        Icon.SetVisibility(widget_visibility.Visible)
        if(Index = 1):
            Icon.SetImage(Boss_Textures.Pet_1)
        else if(Index = 2):
            Icon.SetImage(Boss_Textures.Pet_2)
        else if(Index = 3):
            Icon.SetImage(Boss_Textures.Pet_3)
        else if(Index = 4):
            Icon.SetImage(Boss_Textures.Pet_4)
        else if(Index = 5):
            Icon.SetImage(Boss_Textures.Pet_5)
        else:
            Icon.SetVisibility(widget_visibility.Hidden)
        NameButton.SetText(StringToMessage("Empty"))
    
    LevelTitle<public> : text_block = text_block{DefaultText := LevelText, DefaultTextColor := NamedColors.Orange}
    LevelValue<public> : text_block = text_block{ DefaultTextColor := NamedColors.White}
    Icon<public> : texture_block = texture_block{DefaultImage := Boss_FirstPet_Icon_2.Boss_FirstPet11_1, DefaultDesiredSize := vector2{X :=100.0, Y := 100.0}}
    NameButton<public> : button_quiet = button_quiet{}

    CreateCanvas():canvas=
        canvas:
            Slots := array:
                canvas_slot:
                    Anchors := anchors{Minimum := vector2{X := 0.0, Y := 0.0}, Maximum := vector2{X := 1.0, Y := 1.0}}
                    Widget :=
                        stack_box:
                            Orientation := orientation.Vertical
                            Slots := array:
                                stack_box_slot:
                                    HorizontalAlignment := horizontal_alignment.Fill
                                    VerticalAlignment := vertical_alignment.Fill
                                    Widget := stack_box:
                                        Orientation := orientation.Horizontal
                                        Slots := array:
                                            stack_box_slot:
                                                HorizontalAlignment := horizontal_alignment.Fill
                                                VerticalAlignment := vertical_alignment.Fill
                                                Widget := stack_box:
                                                    Orientation := orientation.Horizontal
                                                    Slots := array:
                                                        stack_box_slot:
                                                            HorizontalAlignment := horizontal_alignment.Fill
                                                            VerticalAlignment := vertical_alignment.Bottom
                                                            Widget := LevelTitle
                                                        stack_box_slot:
                                                            HorizontalAlignment := horizontal_alignment.Fill
                                                            VerticalAlignment := vertical_alignment.Bottom
                                                            Widget := LevelValue
                                            stack_box_slot:
                                                HorizontalAlignment := horizontal_alignment.Fill
                                                VerticalAlignment := vertical_alignment.Fill
                                                Widget := Icon
                                stack_box_slot:
                                    Distribution := option{1.0}
                                    HorizontalAlignment := horizontal_alignment.Fill
                                    VerticalAlignment := vertical_alignment.Fill
                                    Widget := NameButton



yokai_select_template := class:
    SelectLevelTitle<public> : text_block = text_block{DefaultTextColor := NamedColors.Orange, DefaultText := LevelText}
    SelectLevelValue<public> : text_block = text_block{DefaultTextColor := NamedColors.White}
    SelectIcon_FLV<public> : texture_block = texture_block{DefaultImage := Boss_FirstPet_Icon_2.Boss_FirstPet11_1, DefaultDesiredSize := vector2{X := 100.0, Y := 100.0}}
    SelectName<public> : text_block = text_block{DefaultTextColor := NamedColors.White}
    SelectButton : button_quiet = button_quiet{}

    CreateCanvas():canvas=
        canvas: # CanvasPanel_2
            Slots := array:
                canvas_slot: # CanvasPanelSlot_0
                    Anchors := anchors:
                        Minimum := vector2{X := 0.0,Y := 0.0}
                        Maximum := vector2{X := 1.0,Y := 1.0}
                    Offsets := margin:
                        Top := 0.0
                        Right := 0.0
                        Bottom := 0.0
                        Left := 0.0
                    SizeToContent := true
                    Alignment := vector2{X := 0.0,Y := 0.0}
                    ZOrder := 0
                    Widget := stack_box: # StackBoxMain_FLV
                        Orientation := orientation.Horizontal
                        Slots := array:
                            stack_box_slot: # StackBoxSlot_0
                                HorizontalAlignment := horizontal_alignment.Center
                                VerticalAlignment := vertical_alignment.Bottom
                                Padding := margin:
                                    Top := 0.0
                                    Right := 0.0
                                    Bottom := 25.0
                                    Left := 0.0
                                Widget := stack_box: # StackBox_152
                                    Orientation := orientation.Horizontal
                                    Slots := array:
                                        stack_box_slot: # StackBoxSlot_0
                                            HorizontalAlignment := horizontal_alignment.Left
                                            VerticalAlignment := vertical_alignment.Bottom
                                            Padding := margin:
                                                Top := 0.0
                                                Right := 0.0
                                                Bottom := 0.0
                                                Left := 0.0
                                            Widget := SelectLevelTitle
                                        stack_box_slot: # StackBoxSlot_1
                                            HorizontalAlignment := horizontal_alignment.Left
                                            VerticalAlignment := vertical_alignment.Bottom
                                            Padding := margin:
                                                Top := 0.0
                                                Right := 0.0
                                                Bottom := 0.0
                                                Left := 0.0
                                            Widget := SelectLevelValue
                                        
                            stack_box_slot: # StackBoxSlot_1
                                HorizontalAlignment := horizontal_alignment.Center
                                VerticalAlignment := vertical_alignment.Center
                                Padding := margin:
                                    Top := 0.0
                                    Right := 25.0
                                    Bottom := 0.0
                                    Left := 25.0
                                Widget := SelectIcon_FLV
                            stack_box_slot: # StackBoxSlot_2
                                HorizontalAlignment := horizontal_alignment.Center
                                VerticalAlignment := vertical_alignment.Bottom
                                Padding := margin:
                                    Top := 0.0
                                    Right := 0.0
                                    Bottom := 25.0
                                    Left := 0.0
                                Widget := stack_box: # StackBox_241
                                    Orientation := orientation.Vertical
                                    Slots := array:
                                        stack_box_slot: # StackBoxSlot_2
                                            HorizontalAlignment := horizontal_alignment.Left
                                            VerticalAlignment := vertical_alignment.Center
                                            Padding := margin:
                                                Top := 0.0
                                                Right := 0.0
                                                Bottom := 25.0
                                                Left := 0.0
                                            Widget := SelectName
                                        stack_box_slot: # StackBoxSlot_1
                                            HorizontalAlignment := horizontal_alignment.Center
                                            VerticalAlignment := vertical_alignment.Center
                                            Padding := margin:
                                                Top := 0.0
                                                Right := 0.0
                                                Bottom := 0.0
                                                Left := 0.0
                                            Widget := SelectButton
                                        
                            
yokai_stat_template := class:
    Background : texture_block = texture_block{DefaultImage := icon_set_2.bg, DefaultDesiredSize := vector2{X := 600.0, Y := 250.0}}
    #StatTitle : text_block = text_block{DefaultTextColor := NamedColors.White}
    StatAIcon : texture_block = texture_block{DefaultImage := icon_set_2.Stats_PetGather}
    StatATitle : text_block = text_block{DefaultTextColor := NamedColors.White}
    StatBIcon : texture_block = texture_block{DefaultImage := icon_set_2.Stats_Str}
    StatBTitle : text_block = text_block{DefaultTextColor := NamedColors.White}
    StatCIcon : texture_block = texture_block{DefaultImage := icon_set_2.Stats_CriRate}
    StatCTitle : text_block = text_block{DefaultTextColor := NamedColors.White}
    StatDIcon : texture_block = texture_block{DefaultImage := icon_set_2.Stats_CriDamage}
    StatDTitle : text_block = text_block{DefaultTextColor := NamedColors.White}
    StatEIcon : texture_block = texture_block{ DefaultImage := icon_set_2.IC_LastHit}
    StatETitle : text_block = text_block{DefaultTextColor := NamedColors.White}
    StatAValue : text_block = text_block{DefaultTextColor := NamedColors.White}
    StatBValue : text_block = text_block{DefaultTextColor := NamedColors.White}
    StatCValue : text_block = text_block{DefaultTextColor := NamedColors.White}
    StatDValue : text_block = text_block{DefaultTextColor := NamedColors.White}
    StatEValue : text_block = text_block{DefaultTextColor := NamedColors.White}
    

    UpdateStatsTotal<public>(PetGather:string, Strength:string, CriticalRate:string, CriticalDamage:string, LastHitDamage:string):void=
        StatAValue.SetText(StringToMessage(PetGather))
        StatBValue.SetText(StringToMessage(Strength))
        StatCValue.SetText(StringToMessage(CriticalRate))
        StatDValue.SetText(StringToMessage(CriticalDamage))
        StatEValue.SetText(StringToMessage(LastHitDamage))

    CreateCanvas():canvas=
        canvas:
            Slots := array:
                canvas_slot:
                    Anchors := anchors{Minimum := vector2{X := 0.0, Y := 0.0}, Maximum := vector2{X := 1.0, Y := 1.0}}
                    Widget :=
                        overlay: # Overlay_55
                            Slots := array:
                                overlay_slot: # OverlaySlot_0
                                    HorizontalAlignment := horizontal_alignment.Fill
                                    VerticalAlignment := vertical_alignment.Fill
                                    Padding := margin:
                                        Top := 0.0
                                        Right := 0.0
                                        Bottom := 0.0
                                        Left := 0.0
                                    Widget := Background
                                overlay_slot: # OverlaySlot_3
                                    HorizontalAlignment := horizontal_alignment.Center
                                    VerticalAlignment := vertical_alignment.Center
                                    Padding := margin:
                                        Top := 5.0
                                        Right := 5.0
                                        Bottom := 5.0
                                        Left := 5.0
                                    Widget := stack_box: # StackBox_78
                                        Orientation := orientation.Vertical
                                        Slots := array:
                                            <#stack_box_slot: # StackBoxSlot_0
                                                HorizontalAlignment := horizontal_alignment.Center
                                                VerticalAlignment := vertical_alignment.Center
                                                Padding := margin:
                                                    Top := 5.0
                                                    Right := 0.0
                                                    Bottom := 5.0
                                                    Left := 0.0
                                                Widget := StatTitle#>
                                            stack_box_slot: # StackBoxSlot_1
                                                HorizontalAlignment := horizontal_alignment.Left
                                                VerticalAlignment := vertical_alignment.Top
                                                Padding := margin:
                                                    Top := 0.0
                                                    Right := 0.0
                                                    Bottom := 0.0
                                                    Left := 0.0
                                                Widget := stack_box: # StackBox_6
                                                    Orientation := orientation.Horizontal
                                                    Slots := array:
                                                        stack_box_slot: # StackBoxSlot_0
                                                            HorizontalAlignment := horizontal_alignment.Center
                                                            VerticalAlignment := vertical_alignment.Top
                                                            Padding := margin:
                                                                Top := 0.0
                                                                Right := 0.0
                                                                Bottom := 0.0
                                                                Left := 0.0
                                                            Widget := stack_box: # StackBox_7
                                                                Orientation := orientation.Vertical
                                                                Slots := array:
                                                                    stack_box_slot: # StackBoxSlot_4
                                                                        HorizontalAlignment := horizontal_alignment.Left
                                                                        VerticalAlignment := vertical_alignment.Top
                                                                        Padding := margin:
                                                                            Top := 0.0
                                                                            Right := 0.0
                                                                            Bottom := 10.0
                                                                            Left := 0.0
                                                                        Widget := stack_box: # StackBox_8
                                                                            Orientation := orientation.Horizontal
                                                                            Slots := array:
                                                                                stack_box_slot: # StackBoxSlot_0
                                                                                    HorizontalAlignment := horizontal_alignment.Left
                                                                                    VerticalAlignment := vertical_alignment.Top
                                                                                    Padding := margin:
                                                                                        Top := 0.0
                                                                                        Right := 0.0
                                                                                        Bottom := 0.0
                                                                                        Left := 0.0
                                                                                    Widget := StatAIcon
                                                                                stack_box_slot: # StackBoxSlot_1
                                                                                    HorizontalAlignment := horizontal_alignment.Left
                                                                                    VerticalAlignment := vertical_alignment.Top
                                                                                    Padding := margin:
                                                                                        Top := 0.0
                                                                                        Right := 0.0
                                                                                        Bottom := 0.0
                                                                                        Left := 0.0
                                                                                    Widget := StatATitle
                                                                                
                                                                    stack_box_slot: # StackBoxSlot_9
                                                                        HorizontalAlignment := horizontal_alignment.Left
                                                                        VerticalAlignment := vertical_alignment.Top
                                                                        Padding := margin:
                                                                            Top := 0.0
                                                                            Right := 0.0
                                                                            Bottom := 10.0
                                                                            Left := 0.0
                                                                        Widget := stack_box: # StackBox_9
                                                                            Orientation := orientation.Horizontal
                                                                            Slots := array:
                                                                                stack_box_slot: # StackBoxSlot_0
                                                                                    HorizontalAlignment := horizontal_alignment.Left
                                                                                    VerticalAlignment := vertical_alignment.Top
                                                                                    Padding := margin:
                                                                                        Top := 0.0
                                                                                        Right := 0.0
                                                                                        Bottom := 0.0
                                                                                        Left := 0.0
                                                                                    Widget := StatBIcon
                                                                                stack_box_slot: # StackBoxSlot_1
                                                                                    HorizontalAlignment := horizontal_alignment.Left
                                                                                    VerticalAlignment := vertical_alignment.Top
                                                                                    Padding := margin:
                                                                                        Top := 0.0
                                                                                        Right := 0.0
                                                                                        Bottom := 0.0
                                                                                        Left := 0.0
                                                                                    Widget := StatBTitle
                                                                                
                                                                    stack_box_slot: # StackBoxSlot_10
                                                                        HorizontalAlignment := horizontal_alignment.Left
                                                                        VerticalAlignment := vertical_alignment.Top
                                                                        Padding := margin:
                                                                            Top := 0.0
                                                                            Right := 0.0
                                                                            Bottom := 10.0
                                                                            Left := 0.0
                                                                        Widget := stack_box: # StackBox_10
                                                                            Orientation := orientation.Horizontal
                                                                            Slots := array:
                                                                                stack_box_slot: # StackBoxSlot_0
                                                                                    HorizontalAlignment := horizontal_alignment.Left
                                                                                    VerticalAlignment := vertical_alignment.Top
                                                                                    Padding := margin:
                                                                                        Top := 0.0
                                                                                        Right := 0.0
                                                                                        Bottom := 0.0
                                                                                        Left := 0.0
                                                                                    Widget := StatCIcon
                                                                                stack_box_slot: # StackBoxSlot_1
                                                                                    HorizontalAlignment := horizontal_alignment.Left
                                                                                    VerticalAlignment := vertical_alignment.Top
                                                                                    Padding := margin:
                                                                                        Top := 0.0
                                                                                        Right := 0.0
                                                                                        Bottom := 0.0
                                                                                        Left := 0.0
                                                                                    Widget := StatCTitle
                                                                                
                                                                    stack_box_slot: # StackBoxSlot_11
                                                                        HorizontalAlignment := horizontal_alignment.Left
                                                                        VerticalAlignment := vertical_alignment.Top
                                                                        Padding := margin:
                                                                            Top := 0.0
                                                                            Right := 0.0
                                                                            Bottom := 10.0
                                                                            Left := 0.0
                                                                        Widget := stack_box: # StackBox_11
                                                                            Orientation := orientation.Horizontal
                                                                            Slots := array:
                                                                                stack_box_slot: # StackBoxSlot_0
                                                                                    HorizontalAlignment := horizontal_alignment.Left
                                                                                    VerticalAlignment := vertical_alignment.Top
                                                                                    Padding := margin:
                                                                                        Top := 0.0
                                                                                        Right := 0.0
                                                                                        Bottom := 0.0
                                                                                        Left := 0.0
                                                                                    Widget := StatDIcon
                                                                                stack_box_slot: # StackBoxSlot_1
                                                                                    HorizontalAlignment := horizontal_alignment.Left
                                                                                    VerticalAlignment := vertical_alignment.Top
                                                                                    Padding := margin:
                                                                                        Top := 0.0
                                                                                        Right := 0.0
                                                                                        Bottom := 0.0
                                                                                        Left := 0.0
                                                                                    Widget := StatDTitle
                                                                                
                                                                    stack_box_slot: # StackBoxSlot_12
                                                                        HorizontalAlignment := horizontal_alignment.Left
                                                                        VerticalAlignment := vertical_alignment.Top
                                                                        Padding := margin:
                                                                            Top := 0.0
                                                                            Right := 0.0
                                                                            Bottom := 10.0
                                                                            Left := 0.0
                                                                        Widget := stack_box: # StackBox_12
                                                                            Orientation := orientation.Horizontal
                                                                            Slots := array:
                                                                                stack_box_slot: # StackBoxSlot_0
                                                                                    HorizontalAlignment := horizontal_alignment.Left
                                                                                    VerticalAlignment := vertical_alignment.Top
                                                                                    Padding := margin:
                                                                                        Top := 0.0
                                                                                        Right := 0.0
                                                                                        Bottom := 0.0
                                                                                        Left := 0.0
                                                                                    Widget := StatEIcon
                                                                                stack_box_slot: # StackBoxSlot_1
                                                                                    HorizontalAlignment := horizontal_alignment.Left
                                                                                    VerticalAlignment := vertical_alignment.Top
                                                                                    Padding := margin:
                                                                                        Top := 0.0
                                                                                        Right := 0.0
                                                                                        Bottom := 0.0
                                                                                        Left := 0.0
                                                                                    Widget := StatETitle
                                                                                
                                                                    
                                                        stack_box_slot: # StackBoxSlot_3
                                                            HorizontalAlignment := horizontal_alignment.Left
                                                            VerticalAlignment := vertical_alignment.Top
                                                            Padding := margin:
                                                                Top := 0.0
                                                                Right := 25.0
                                                                Bottom := 0.0
                                                                Left := 25.0
                                                            Widget := stack_box: # StackBox_13
                                                                Orientation := orientation.Vertical
                                                                Slots := array:
                                                                    stack_box_slot: # StackBoxSlot_0
                                                                        HorizontalAlignment := horizontal_alignment.Left
                                                                        VerticalAlignment := vertical_alignment.Top
                                                                        Padding := margin:
                                                                            Top := 0.0
                                                                            Right := 0.0
                                                                            Bottom := 10.0
                                                                            Left := 0.0
                                                                        Widget := StatAValue
                                                                    stack_box_slot: # StackBoxSlot_1
                                                                        HorizontalAlignment := horizontal_alignment.Left
                                                                        VerticalAlignment := vertical_alignment.Top
                                                                        Padding := margin:
                                                                            Top := 0.0
                                                                            Right := 0.0
                                                                            Bottom := 10.0
                                                                            Left := 0.0
                                                                        Widget := StatBValue
                                                                    stack_box_slot: # StackBoxSlot_2
                                                                        HorizontalAlignment := horizontal_alignment.Left
                                                                        VerticalAlignment := vertical_alignment.Top
                                                                        Padding := margin:
                                                                            Top := 0.0
                                                                            Right := 0.0
                                                                            Bottom := 10.0
                                                                            Left := 0.0
                                                                        Widget := StatCValue
                                                                    stack_box_slot: # StackBoxSlot_3
                                                                        HorizontalAlignment := horizontal_alignment.Left
                                                                        VerticalAlignment := vertical_alignment.Top
                                                                        Padding := margin:
                                                                            Top := 0.0
                                                                            Right := 0.0
                                                                            Bottom := 10.0
                                                                            Left := 0.0
                                                                        Widget := StatDValue
                                                                    stack_box_slot: # StackBoxSlot_4
                                                                        HorizontalAlignment := horizontal_alignment.Left
                                                                        VerticalAlignment := vertical_alignment.Top
                                                                        Padding := margin:
                                                                            Top := 0.0
                                                                            Right := 0.0
                                                                            Bottom := 10.0
                                                                            Left := 0.0
                                                                        Widget := StatEValue
                                                                    
                                                                                                    
                                                                                        
                                                                            