using { /Fortnite.com/UI }
using { /UnrealEngine.com/Temporary/UI }
using { /Verse.org/Colors }
using { /UnrealEngine.com/Temporary/SpatialMath }
using { /Verse.org/Assets }

widget_yokai_menu<public> := class(widget_base):

    Init<override>():void=
        (super:)Init()
        spawn{AwaitYokaiClick()}
        SkillName.SetText(StringToMessage("Skill Name"))
        SkillDescription.SetText(StringToMessage("Skill Description\nSkill Description\nSkill Description"))
        SkillAdditional.SetText(StringToMessage("Skill Additional"))

        PetOwnedValue.SetText(StringToMessage("0/10"))
        PetsEquippedValue.SetText(StringToMessage("0/10"))
        TotalPetLevelsValue.SetText(StringToMessage("0"))
        TotalPetLevelsBonusValue.SetText(StringToMessage("0"))

        PetOwnedText.SetText(StringToMessage("Pet Owned"))
        PetsEquippedText.SetText(StringToMessage("Pets Equipped"))
        TotalPetLevelsText.SetText(StringToMessage("Total Pet Levels"))
        TotalPetLevelsBonusText.SetText(StringToMessage("Total Pet Levels Bonus"))
        
        StateTitle.SetText(StringToMessage("Stats"))
        PetGatherTitle.SetText(StringToMessage("Pet Gather"))
        StrehgthTitle.SetText(StringToMessage("Strength %"))
        CriticalRateTitle.SetText(StringToMessage("Critical Rate"))
        CriticalDamageTitle.SetText(StringToMessage("Critical Damage"))
        LastHitTitle.SetText(StringToMessage("Last Hit"))
        ValueTitle.SetText(StringToMessage("Total"))

        PetGatherValue.SetText(StringToMessage("0"))
        StrengthValue.SetText(StringToMessage("0"))
        CriticalRateValue.SetText(StringToMessage("0"))
        CriticalDamageValue.SetText(StringToMessage("0"))
        LastHitDamageValue.SetText(StringToMessage("0"))

        BasicBG.Title.SetText(StringToMessage("Yo-kai"))

    # 妖怪アイコン
    BasicYokaiIcon1<public> : basic_yokai_icon = basic_yokai_icon{}
    BasicYokaiIcon2<public> : basic_yokai_icon = basic_yokai_icon{}
    BasicYokaiIcon3<public> : basic_yokai_icon = basic_yokai_icon{}
    BasicYokaiIcon4<public> : basic_yokai_icon = basic_yokai_icon{}

    YokaiClickEvent<public> : event(int) = event(int){}

    UpdateYokaiIcon<public>(Index:int, Icon:texture, Name:string, Level:string):void=
        if(Index = 1):
            BasicYokaiIcon1.Icon.SetVisibility(widget_visibility.Visible)
            BasicYokaiIcon1.LevelTitle.SetVisibility(widget_visibility.Visible)
            BasicYokaiIcon1.Icon.SetImage(Icon)
            BasicYokaiIcon1.NameButton.SetText(StringToMessage(Name))
            BasicYokaiIcon1.LevelValue.SetText(StringToMessage("{Level}"))
        else if(Index = 2):
            BasicYokaiIcon1.Icon.SetVisibility(widget_visibility.Visible)
            BasicYokaiIcon1.LevelTitle.SetVisibility(widget_visibility.Visible)
            BasicYokaiIcon2.Icon.SetImage(Icon)
            BasicYokaiIcon2.NameButton.SetText(StringToMessage(Name))
            BasicYokaiIcon2.LevelValue.SetText(StringToMessage("{Level}"))
        else if(Index = 3):
            BasicYokaiIcon1.Icon.SetVisibility(widget_visibility.Visible)
            BasicYokaiIcon1.LevelTitle.SetVisibility(widget_visibility.Visible)
            BasicYokaiIcon3.Icon.SetImage(Icon)
            BasicYokaiIcon3.NameButton.SetText(StringToMessage(Name))
            BasicYokaiIcon3.LevelValue.SetText(StringToMessage("{Level}"))
        else if(Index = 4):
            BasicYokaiIcon1.Icon.SetVisibility(widget_visibility.Visible)
            BasicYokaiIcon1.LevelTitle.SetVisibility(widget_visibility.Visible)
            BasicYokaiIcon4.Icon.SetImage(Icon)
            BasicYokaiIcon4.NameButton.SetText(StringToMessage(Name))
            BasicYokaiIcon4.LevelValue.SetText(StringToMessage("{Level}"))

    SetEmpty<public>(Index:int):void=
        if(Index = 1):
            BasicYokaiIcon1.SetEmpty()
        else if(Index = 2):
            BasicYokaiIcon2.SetEmpty()
        else if(Index = 3):
            BasicYokaiIcon3.SetEmpty()
        else if(Index = 4):
            BasicYokaiIcon4.SetEmpty()

    AwaitYokaiClick<public>()<suspends>:void=
        race:
            CloseEvent.Await()
            loop:
                BasicYokaiIcon1.NameButton.OnClick().Await()
                YokaiClickEvent.Signal(1)
            loop:
                BasicYokaiIcon2.NameButton.OnClick().Await()
                YokaiClickEvent.Signal(2)
            loop:
                BasicYokaiIcon3.NameButton.OnClick().Await()
                YokaiClickEvent.Signal(3)
            loop:
                BasicYokaiIcon4.NameButton.OnClick().Await()
                YokaiClickEvent.Signal(4)

    # ペットサポート
    SkillName<public> : text_block = text_block{DefaultTextColor := NamedColors.White}
    SkillIcon<public> : texture_block = texture_block{DefaultImage := icon_set_2.IC_Boss, DefaultDesiredSize := vector2{X := 256.0, Y := 256.0}}
    SkillDescription<public> : text_block = text_block{DefaultTextColor := NamedColors.SteelBlue}
    SkillAdditional<public> : text_block = text_block{DefaultTextColor := NamedColors.White}
    
    UpdatePetSupportInfo<public>(Name:string, Icon:texture, Description:string, Additional:string):void=
        SkillName.SetText(StringToMessage(Name))
        SkillIcon.SetImage(Icon)
        SkillDescription.SetText(StringToMessage(Description))
        SkillAdditional.SetText(StringToMessage(Additional))

    # ステータスサマリー
    PetOwnedValue<public> : text_block = text_block{DefaultTextColor := NamedColors.White}
    PetsEquippedValue<public> : text_block = text_block{DefaultTextColor := NamedColors.White}
    TotalPetLevelsValue<public> : text_block = text_block{DefaultTextColor := NamedColors.White}
    TotalPetLevelsBonusValue<public> : text_block = text_block{DefaultTextColor := NamedColors.White}

    UpdateStatsSummary<public>(PetOwned:string, PetsEquipped:string, TotalPetLevels:string, TotalPetLevelsBonus:string):void=
        PetOwnedValue.SetText(StringToMessage(PetOwned))
        PetsEquippedValue.SetText(StringToMessage(PetsEquipped))
        TotalPetLevelsValue.SetText(StringToMessage(TotalPetLevels))
        TotalPetLevelsBonusValue.SetText(StringToMessage(TotalPetLevelsBonus))

    # ステータス
    PetGatherValue<public> : text_block = text_block{DefaultTextColor := NamedColors.SeaGreen}
    StrengthValue<public> : text_block = text_block{DefaultTextColor := NamedColors.SeaGreen}
    CriticalRateValue<public> : text_block = text_block{DefaultTextColor := NamedColors.SeaGreen}
    CriticalDamageValue<public> : text_block = text_block{DefaultTextColor := NamedColors.SeaGreen}
    LastHitDamageValue<public> : text_block = text_block{DefaultTextColor := NamedColors.SeaGreen}

    UpdateStatsTotal<public>(PetGather:string, Strength:string, CriticalRate:string, CriticalDamage:string, LastHitDamage:string):void=
        PetGatherValue.SetText(StringToMessage(PetGather))
        StrengthValue.SetText(StringToMessage(Strength))
        CriticalRateValue.SetText(StringToMessage(CriticalRate))
        CriticalDamageValue.SetText(StringToMessage(CriticalDamage))
        LastHitDamageValue.SetText(StringToMessage(LastHitDamage))

    PetSupportTitle<public> : text_block = text_block{DefaultTextColor := NamedColors.White}
    SubBG<public> : texture_block = texture_block{DefaultImage := icon_set_2.bg}
    PetSupportBG<public> : texture_block = texture_block{DefaultImage := icon_set_2.bg, DefaultDesiredSize := vector2{X := 600.0, Y := 450.0}}
    CenterLine<public> : color_block = color_block{DefaultColor := NamedColors.White, DefaultDesiredSize := vector2{X := 5.0}}
    StatsSummaryTitle<public> : text_block = text_block{DefaultTextColor := NamedColors.White}
    PetOwnedText<public> : text_block = text_block{DefaultTextColor := NamedColors.SteelBlue}
    PetsEquippedText<public> : text_block = text_block{DefaultTextColor := NamedColors.SteelBlue}
    TotalPetLevelsText<public> : text_block = text_block{DefaultTextColor := NamedColors.SteelBlue}
    TotalPetLevelsBonusText<public> : text_block = text_block{DefaultTextColor := NamedColors.SteelBlue}
    StateTitle<public> : text_block = text_block{DefaultTextColor := NamedColors.SteelBlue}
    PetGatherIcon<public> : texture_block = texture_block{DefaultImage := icon_set_2.Stats_PetGather}
    PetGatherTitle<public> : text_block = text_block{DefaultTextColor := NamedColors.SteelBlue}
    StrehgthIcon<public> : texture_block = texture_block{DefaultImage := icon_set_2.Stats_Str}
    StrehgthTitle<public> : text_block = text_block{DefaultTextColor := NamedColors.SteelBlue}
    CriticalRateIcon<public> : texture_block = texture_block{DefaultImage := icon_set_2.Stats_CriRate}
    CriticalRateTitle<public> : text_block = text_block{DefaultTextColor := NamedColors.SteelBlue}
    CriticalDamageIcon<public> : texture_block = texture_block{DefaultImage := icon_set_2.Stats_CriDamage}
    CriticalDamageTitle<public> : text_block = text_block{DefaultTextColor := NamedColors.SteelBlue}
    LastHitIcon<public> : texture_block = texture_block{DefaultImage := icon_set_2.IC_LastHit}
    LastHitTitle<public> : text_block = text_block{DefaultTextColor := NamedColors.SteelBlue}
    ValueTitle<public> : text_block = text_block{DefaultTextColor := NamedColors.SteelBlue}

    CreateCanvas<override>():canvas=
        canvas:
            Slots := array:
                canvas_slot:
                    Anchors := anchors:
                        Minimum := vector2{X := 0.0, Y := 0.0}
                        Maximum := vector2{X := 1.0, Y := 1.0}
                    Widget :=
                        overlay: # Overlay_40
                            Slots := array:
                                overlay_slot: # OverlaySlot_2
                                    HorizontalAlignment := horizontal_alignment.Fill
                                    VerticalAlignment := vertical_alignment.Fill
                                    Padding := margin:
                                        Top := 0.0
                                        Right := 0.0
                                        Bottom := 0.0
                                        Left := 0.0
                                    Widget := BasicBG.CreateOverlay()
                                overlay_slot: # OverlaySlot_0
                                    HorizontalAlignment := horizontal_alignment.Fill
                                    VerticalAlignment := vertical_alignment.Fill
                                    Padding := margin:
                                        Top := 125.0
                                        Right := 125.0
                                        Bottom := 125.0
                                        Left := 125.0
                                    Widget := stack_box: # StackBox_129
                                        Orientation := orientation.Vertical
                                        Slots := array:
                                            stack_box_slot: # StackBoxSlot_2
                                                HorizontalAlignment := horizontal_alignment.Center
                                                VerticalAlignment := vertical_alignment.Center
                                                Padding := margin:
                                                    Top := 0.0
                                                    Right := 0.0
                                                    Bottom := 0.0
                                                    Left := 0.0
                                                Widget := stack_box: # StackBox_585
                                                    Orientation := orientation.Horizontal
                                                    Slots := array:
                                                        stack_box_slot: # StackBoxSlot_0
                                                            HorizontalAlignment := horizontal_alignment.Fill
                                                            VerticalAlignment := vertical_alignment.Top
                                                            Padding := margin:
                                                                Top := 0.0
                                                                Right := 25.0
                                                                Bottom := 25.0
                                                                Left := 0.0
                                                            Widget := BasicYokaiIcon1.CreateCanvas()
                                                        stack_box_slot: # StackBoxSlot_1
                                                            HorizontalAlignment := horizontal_alignment.Fill
                                                            VerticalAlignment := vertical_alignment.Top
                                                            Padding := margin:
                                                                Top := 0.0
                                                                Right := 25.0
                                                                Bottom := 25.0
                                                                Left := 0.0
                                                            Widget := BasicYokaiIcon2.CreateCanvas()
                                                        stack_box_slot: # StackBoxSlot_2
                                                            HorizontalAlignment := horizontal_alignment.Fill
                                                            VerticalAlignment := vertical_alignment.Top
                                                            Padding := margin:
                                                                Top := 0.0
                                                                Right := 25.0
                                                                Bottom := 25.0
                                                                Left := 0.0
                                                            Widget := BasicYokaiIcon3.CreateCanvas()
                                                        stack_box_slot: # StackBoxSlot_3
                                                            HorizontalAlignment := horizontal_alignment.Fill
                                                            VerticalAlignment := vertical_alignment.Top
                                                            Padding := margin:
                                                                Top := 0.0
                                                                Right := 25.0
                                                                Bottom := 25.0
                                                                Left := 0.0
                                                            Widget := BasicYokaiIcon4.CreateCanvas()
                                                        
                                            stack_box_slot: # StackBoxSlot_0
                                                HorizontalAlignment := horizontal_alignment.Fill
                                                VerticalAlignment := vertical_alignment.Bottom
                                                Padding := margin:
                                                    Top := 25.0
                                                    Right := 0.0
                                                    Bottom := 0.0
                                                    Left := 0.0
                                                Widget := overlay: # Overlay_86
                                                    Slots := array:
                                                        overlay_slot: # OverlaySlot_0
                                                            HorizontalAlignment := horizontal_alignment.Fill
                                                            VerticalAlignment := vertical_alignment.Fill
                                                            Padding := margin:
                                                                Top := 0.0
                                                                Right := 0.0
                                                                Bottom := 0.0
                                                                Left := 0.0
                                                            Widget := SubBG
                                                        overlay_slot: # OverlaySlot_1
                                                            HorizontalAlignment := horizontal_alignment.Fill
                                                            VerticalAlignment := vertical_alignment.Fill
                                                            Padding := margin:
                                                                Top := 0.0
                                                                Right := 0.0
                                                                Bottom := 0.0
                                                                Left := 0.0
                                                            Widget := stack_box: # StackBox_695
                                                                Orientation := orientation.Horizontal
                                                                Slots := array:
                                                                    stack_box_slot: # StackBoxSlot_0
                                                                        HorizontalAlignment := horizontal_alignment.Fill
                                                                        VerticalAlignment := vertical_alignment.Fill
                                                                        Padding := margin:
                                                                            Top := 25.0
                                                                            Right := 25.0
                                                                            Bottom := 0.0
                                                                            Left := 25.0
                                                                        Widget := stack_box: # StackBox_781
                                                                            Orientation := orientation.Vertical
                                                                            Slots := array:
                                                                                stack_box_slot: # StackBoxSlot_0
                                                                                    HorizontalAlignment := horizontal_alignment.Center
                                                                                    VerticalAlignment := vertical_alignment.Top
                                                                                    Padding := margin:
                                                                                        Top := 0.0
                                                                                        Right := 0.0
                                                                                        Bottom := 0.0
                                                                                        Left := 0.0
                                                                                    Widget := PetSupportTitle
                                                                                stack_box_slot: # StackBoxSlot_4
                                                                                    HorizontalAlignment := horizontal_alignment.Fill
                                                                                    VerticalAlignment := vertical_alignment.Fill
                                                                                    Padding := margin:
                                                                                        Top := 0.0
                                                                                        Right := 0.0
                                                                                        Bottom := 0.0
                                                                                        Left := 0.0
                                                                                    Widget := overlay: # Overlay_182
                                                                                        Slots := array:
                                                                                            overlay_slot: # OverlaySlot_3
                                                                                                HorizontalAlignment := horizontal_alignment.Fill
                                                                                                VerticalAlignment := vertical_alignment.Fill
                                                                                                Padding := margin:
                                                                                                    Top := 0.0
                                                                                                    Right := 0.0
                                                                                                    Bottom := 0.0
                                                                                                    Left := 0.0
                                                                                                Widget := PetSupportBG
                                                                                            overlay_slot: # OverlaySlot_1
                                                                                                HorizontalAlignment := horizontal_alignment.Center
                                                                                                VerticalAlignment := vertical_alignment.Fill
                                                                                                Padding := margin:
                                                                                                    Top := 0.0
                                                                                                    Right := 0.0
                                                                                                    Bottom := 0.0
                                                                                                    Left := 0.0
                                                                                                Widget := stack_box: # StackBox_877
                                                                                                    Orientation := orientation.Vertical
                                                                                                    Slots := array:
                                                                                                        stack_box_slot: # StackBoxSlot_1
                                                                                                            HorizontalAlignment := horizontal_alignment.Center
                                                                                                            VerticalAlignment := vertical_alignment.Center
                                                                                                            Padding := margin:
                                                                                                                Top := 10.0
                                                                                                                Right := 0.0
                                                                                                                Bottom := 0.0
                                                                                                                Left := 0.0
                                                                                                            Widget := SkillName
                                                                                                        stack_box_slot: # StackBoxSlot_2
                                                                                                            HorizontalAlignment := horizontal_alignment.Center
                                                                                                            VerticalAlignment := vertical_alignment.Center
                                                                                                            Padding := margin:
                                                                                                                Top := 0.0
                                                                                                                Right := 0.0
                                                                                                                Bottom := 0.0
                                                                                                                Left := 0.0
                                                                                                            Widget := SkillIcon
                                                                                                        stack_box_slot: # StackBoxSlot_3
                                                                                                            HorizontalAlignment := horizontal_alignment.Center
                                                                                                            VerticalAlignment := vertical_alignment.Center
                                                                                                            Padding := margin:
                                                                                                                Top := 0.0
                                                                                                                Right := 0.0
                                                                                                                Bottom := 0.0
                                                                                                                Left := 0.0
                                                                                                            Widget := SkillDescription
                                                                                                        stack_box_slot: # StackBoxSlot_4
                                                                                                            HorizontalAlignment := horizontal_alignment.Center
                                                                                                            VerticalAlignment := vertical_alignment.Bottom
                                                                                                            Distribution := option{1.0}
                                                                                                            Padding := margin:
                                                                                                                Top := 0.0
                                                                                                                Right := 0.0
                                                                                                                Bottom := 10.0
                                                                                                                Left := 0.0
                                                                                                            Widget := SkillAdditional
                                                                                                        
                                                                                            
                                                                                
                                                                    stack_box_slot: # StackBoxSlot_6
                                                                        HorizontalAlignment := horizontal_alignment.Center
                                                                        VerticalAlignment := vertical_alignment.Fill
                                                                        Padding := margin:
                                                                            Top := 25.0
                                                                            Right := 0.0
                                                                            Bottom := 25.0
                                                                            Left := 0.0
                                                                        Widget := CenterLine
                                                                    stack_box_slot: # StackBoxSlot_1
                                                                        HorizontalAlignment := horizontal_alignment.Fill
                                                                        VerticalAlignment := vertical_alignment.Fill
                                                                        Padding := margin:
                                                                            Top := 25.0
                                                                            Right := 25.0
                                                                            Bottom := 25.0
                                                                            Left := 25.0
                                                                        Widget := stack_box: # StackBox_1052
                                                                            Orientation := orientation.Vertical
                                                                            Slots := array:
                                                                                stack_box_slot: # StackBoxSlot_0
                                                                                    HorizontalAlignment := horizontal_alignment.Center
                                                                                    VerticalAlignment := vertical_alignment.Top
                                                                                    Padding := margin:
                                                                                        Top := 0.0
                                                                                        Right := 0.0
                                                                                        Bottom := 50.0
                                                                                        Left := 0.0
                                                                                    Widget := StatsSummaryTitle
                                                                                stack_box_slot: # StackBoxSlot_2
                                                                                    HorizontalAlignment := horizontal_alignment.Left
                                                                                    VerticalAlignment := vertical_alignment.Top
                                                                                    Padding := margin:
                                                                                        Top := 0.0
                                                                                        Right := 25.0
                                                                                        Bottom := 0.0
                                                                                        Left := 25.0
                                                                                    Widget := stack_box: # StackBox_1145
                                                                                        Orientation := orientation.Horizontal
                                                                                        Slots := array:
                                                                                            stack_box_slot: # StackBoxSlot_0
                                                                                                HorizontalAlignment := horizontal_alignment.Left
                                                                                                VerticalAlignment := vertical_alignment.Top
                                                                                                Padding := margin:
                                                                                                    Top := 0.0
                                                                                                    Right := 25.0
                                                                                                    Bottom := 0.0
                                                                                                    Left := 0.0
                                                                                                Widget := PetOwnedText
                                                                                            stack_box_slot: # StackBoxSlot_1
                                                                                                HorizontalAlignment := horizontal_alignment.Left
                                                                                                VerticalAlignment := vertical_alignment.Top
                                                                                                Padding := margin:
                                                                                                    Top := 0.0
                                                                                                    Right := 0.0
                                                                                                    Bottom := 0.0
                                                                                                    Left := 0.0
                                                                                                Widget := PetOwnedValue
                                                                                            
                                                                                stack_box_slot: # StackBoxSlot_9
                                                                                    HorizontalAlignment := horizontal_alignment.Left
                                                                                    VerticalAlignment := vertical_alignment.Top
                                                                                    Padding := margin:
                                                                                        Top := 0.0
                                                                                        Right := 25.0
                                                                                        Bottom := 0.0
                                                                                        Left := 25.0
                                                                                    Widget := stack_box: # StackBox_1
                                                                                        Orientation := orientation.Horizontal
                                                                                        Slots := array:
                                                                                            stack_box_slot: # StackBoxSlot_0
                                                                                                HorizontalAlignment := horizontal_alignment.Left
                                                                                                VerticalAlignment := vertical_alignment.Top
                                                                                                Padding := margin:
                                                                                                    Top := 0.0
                                                                                                    Right := 25.0
                                                                                                    Bottom := 0.0
                                                                                                    Left := 0.0
                                                                                                Widget := PetsEquippedText
                                                                                            stack_box_slot: # StackBoxSlot_1
                                                                                                HorizontalAlignment := horizontal_alignment.Left
                                                                                                VerticalAlignment := vertical_alignment.Top
                                                                                                Padding := margin:
                                                                                                    Top := 0.0
                                                                                                    Right := 0.0
                                                                                                    Bottom := 0.0
                                                                                                    Left := 0.0
                                                                                                Widget := PetsEquippedValue
                                                                                            
                                                                                stack_box_slot: # StackBoxSlot_8
                                                                                    HorizontalAlignment := horizontal_alignment.Left
                                                                                    VerticalAlignment := vertical_alignment.Top
                                                                                    Padding := margin:
                                                                                        Top := 0.0
                                                                                        Right := 25.0
                                                                                        Bottom := 0.0
                                                                                        Left := 25.0
                                                                                    Widget := stack_box: # StackBox
                                                                                        Orientation := orientation.Horizontal
                                                                                        Slots := array:
                                                                                            stack_box_slot: # StackBoxSlot_0
                                                                                                HorizontalAlignment := horizontal_alignment.Left
                                                                                                VerticalAlignment := vertical_alignment.Top
                                                                                                Padding := margin:
                                                                                                    Top := 0.0
                                                                                                    Right := 25.0
                                                                                                    Bottom := 0.0
                                                                                                    Left := 0.0
                                                                                                Widget := TotalPetLevelsText
                                                                                            stack_box_slot: # StackBoxSlot_1
                                                                                                HorizontalAlignment := horizontal_alignment.Left
                                                                                                VerticalAlignment := vertical_alignment.Top
                                                                                                Padding := margin:
                                                                                                    Top := 0.0
                                                                                                    Right := 0.0
                                                                                                    Bottom := 0.0
                                                                                                    Left := 0.0
                                                                                                Widget := TotalPetLevelsValue
                                                                                            
                                                                                stack_box_slot: # StackBoxSlot_10
                                                                                    HorizontalAlignment := horizontal_alignment.Left
                                                                                    VerticalAlignment := vertical_alignment.Top
                                                                                    Padding := margin:
                                                                                        Top := 0.0
                                                                                        Right := 25.0
                                                                                        Bottom := 0.0
                                                                                        Left := 25.0
                                                                                    Widget := stack_box: # StackBox_2
                                                                                        Orientation := orientation.Horizontal
                                                                                        Slots := array:
                                                                                            stack_box_slot: # StackBoxSlot_0
                                                                                                HorizontalAlignment := horizontal_alignment.Left
                                                                                                VerticalAlignment := vertical_alignment.Top
                                                                                                Padding := margin:
                                                                                                    Top := 0.0
                                                                                                    Right := 25.0
                                                                                                    Bottom := 0.0
                                                                                                    Left := 0.0
                                                                                                Widget := TotalPetLevelsBonusText
                                                                                            stack_box_slot: # StackBoxSlot_1
                                                                                                HorizontalAlignment := horizontal_alignment.Left
                                                                                                VerticalAlignment := vertical_alignment.Top
                                                                                                Padding := margin:
                                                                                                    Top := 0.0
                                                                                                    Right := 0.0
                                                                                                    Bottom := 0.0
                                                                                                    Left := 0.0
                                                                                                Widget := TotalPetLevelsBonusValue
                                                                                            
                                                                                stack_box_slot: # StackBoxSlot_7
                                                                                    HorizontalAlignment := horizontal_alignment.Left
                                                                                    VerticalAlignment := vertical_alignment.Bottom
                                                                                    Padding := margin:
                                                                                        Top := 25.0
                                                                                        Right := 25.0
                                                                                        Bottom := 25.0
                                                                                        Left := 25.0
                                                                                    Widget := stack_box: # StackBox_1263
                                                                                        Orientation := orientation.Horizontal
                                                                                        Slots := array:
                                                                                            stack_box_slot: # StackBoxSlot_0
                                                                                                HorizontalAlignment := horizontal_alignment.Fill
                                                                                                VerticalAlignment := vertical_alignment.Top
                                                                                                Padding := margin:
                                                                                                    Top := 0.0
                                                                                                    Right := 100.0
                                                                                                    Bottom := 0.0
                                                                                                    Left := 0.0
                                                                                                Widget := stack_box: # StackBox_3
                                                                                                    Orientation := orientation.Vertical
                                                                                                    Slots := array:
                                                                                                        stack_box_slot: # StackBoxSlot_3
                                                                                                            HorizontalAlignment := horizontal_alignment.Fill
                                                                                                            VerticalAlignment := vertical_alignment.Top
                                                                                                            Padding := margin:
                                                                                                                Top := 0.0
                                                                                                                Right := 0.0
                                                                                                                Bottom := 25.0
                                                                                                                Left := 0.0
                                                                                                            Widget := StateTitle
                                                                                                        stack_box_slot: # StackBoxSlot_5
                                                                                                            HorizontalAlignment := horizontal_alignment.Left
                                                                                                            VerticalAlignment := vertical_alignment.Top
                                                                                                            Padding := margin:
                                                                                                                Top := 0.0
                                                                                                                Right := 0.0
                                                                                                                Bottom := 0.0
                                                                                                                Left := 0.0
                                                                                                            Widget := stack_box: # StackBox_1393
                                                                                                                Orientation := orientation.Horizontal
                                                                                                                Slots := array:
                                                                                                                    stack_box_slot: # StackBoxSlot_0
                                                                                                                        HorizontalAlignment := horizontal_alignment.Left
                                                                                                                        VerticalAlignment := vertical_alignment.Top
                                                                                                                        Padding := margin:
                                                                                                                            Top := 0.0
                                                                                                                            Right := 0.0
                                                                                                                            Bottom := 0.0
                                                                                                                            Left := 0.0
                                                                                                                        Widget := PetGatherIcon
                                                                                                                    stack_box_slot: # StackBoxSlot_1
                                                                                                                        HorizontalAlignment := horizontal_alignment.Left
                                                                                                                        VerticalAlignment := vertical_alignment.Top
                                                                                                                        Padding := margin:
                                                                                                                            Top := 0.0
                                                                                                                            Right := 0.0
                                                                                                                            Bottom := 0.0
                                                                                                                            Left := 0.0
                                                                                                                        Widget := PetGatherTitle
                                                                                                                    
                                                                                                        stack_box_slot: # StackBoxSlot_0
                                                                                                            HorizontalAlignment := horizontal_alignment.Left
                                                                                                            VerticalAlignment := vertical_alignment.Top
                                                                                                            Padding := margin:
                                                                                                                Top := 0.0
                                                                                                                Right := 0.0
                                                                                                                Bottom := 0.0
                                                                                                                Left := 0.0
                                                                                                            Widget := stack_box: # StackBox_5
                                                                                                                Orientation := orientation.Horizontal
                                                                                                                Slots := array:
                                                                                                                    stack_box_slot: # StackBoxSlot_0
                                                                                                                        HorizontalAlignment := horizontal_alignment.Left
                                                                                                                        VerticalAlignment := vertical_alignment.Top
                                                                                                                        Padding := margin:
                                                                                                                            Top := 0.0
                                                                                                                            Right := 0.0
                                                                                                                            Bottom := 0.0
                                                                                                                            Left := 0.0
                                                                                                                        Widget := StrehgthIcon
                                                                                                                    stack_box_slot: # StackBoxSlot_1
                                                                                                                        HorizontalAlignment := horizontal_alignment.Left
                                                                                                                        VerticalAlignment := vertical_alignment.Top
                                                                                                                        Padding := margin:
                                                                                                                            Top := 0.0
                                                                                                                            Right := 0.0
                                                                                                                            Bottom := 0.0
                                                                                                                            Left := 0.0
                                                                                                                        Widget := StrehgthTitle
                                                                                                                    
                                                                                                        stack_box_slot: # StackBoxSlot_1
                                                                                                            HorizontalAlignment := horizontal_alignment.Left
                                                                                                            VerticalAlignment := vertical_alignment.Top
                                                                                                            Padding := margin:
                                                                                                                Top := 0.0
                                                                                                                Right := 0.0
                                                                                                                Bottom := 0.0
                                                                                                                Left := 0.0
                                                                                                            Widget := stack_box: # StackBox_6
                                                                                                                Orientation := orientation.Horizontal
                                                                                                                Slots := array:
                                                                                                                    stack_box_slot: # StackBoxSlot_0
                                                                                                                        HorizontalAlignment := horizontal_alignment.Left
                                                                                                                        VerticalAlignment := vertical_alignment.Top
                                                                                                                        Padding := margin:
                                                                                                                            Top := 0.0
                                                                                                                            Right := 0.0
                                                                                                                            Bottom := 0.0
                                                                                                                            Left := 0.0
                                                                                                                        Widget := CriticalRateIcon
                                                                                                                    stack_box_slot: # StackBoxSlot_1
                                                                                                                        HorizontalAlignment := horizontal_alignment.Left
                                                                                                                        VerticalAlignment := vertical_alignment.Top
                                                                                                                        Padding := margin:
                                                                                                                            Top := 0.0
                                                                                                                            Right := 0.0
                                                                                                                            Bottom := 0.0
                                                                                                                            Left := 0.0
                                                                                                                        Widget := CriticalRateTitle
                                                                                                                    
                                                                                                        stack_box_slot: # StackBoxSlot_2
                                                                                                            HorizontalAlignment := horizontal_alignment.Left
                                                                                                            VerticalAlignment := vertical_alignment.Top
                                                                                                            Padding := margin:
                                                                                                                Top := 0.0
                                                                                                                Right := 0.0
                                                                                                                Bottom := 0.0
                                                                                                                Left := 0.0
                                                                                                            Widget := stack_box: # StackBox_7
                                                                                                                Orientation := orientation.Horizontal
                                                                                                                Slots := array:
                                                                                                                    stack_box_slot: # StackBoxSlot_0
                                                                                                                        HorizontalAlignment := horizontal_alignment.Left
                                                                                                                        VerticalAlignment := vertical_alignment.Top
                                                                                                                        Padding := margin:
                                                                                                                            Top := 0.0
                                                                                                                            Right := 0.0
                                                                                                                            Bottom := 0.0
                                                                                                                            Left := 0.0
                                                                                                                        Widget := CriticalDamageIcon
                                                                                                                    stack_box_slot: # StackBoxSlot_1
                                                                                                                        HorizontalAlignment := horizontal_alignment.Left
                                                                                                                        VerticalAlignment := vertical_alignment.Top
                                                                                                                        Padding := margin:
                                                                                                                            Top := 0.0
                                                                                                                            Right := 0.0
                                                                                                                            Bottom := 0.0
                                                                                                                            Left := 0.0
                                                                                                                        Widget := CriticalDamageTitle
                                                                                                                    
                                                                                                        
                                                                                            stack_box_slot: # StackBoxSlot_1
                                                                                                HorizontalAlignment := horizontal_alignment.Left
                                                                                                VerticalAlignment := vertical_alignment.Top
                                                                                                Padding := margin:
                                                                                                    Top := 0.0
                                                                                                    Right := 0.0
                                                                                                    Bottom := 0.0
                                                                                                    Left := 0.0
                                                                                                Widget := stack_box: # StackBox_4
                                                                                                    Orientation := orientation.Vertical
                                                                                                    Slots := array:
                                                                                                        stack_box_slot: # StackBoxSlot_2
                                                                                                            HorizontalAlignment := horizontal_alignment.Left
                                                                                                            VerticalAlignment := vertical_alignment.Top
                                                                                                            Padding := margin:
                                                                                                                Top := 0.0
                                                                                                                Right := 0.0
                                                                                                                Bottom := 25.0
                                                                                                                Left := 0.0
                                                                                                            Widget := ValueTitle
                                                                                                        stack_box_slot: # StackBoxSlot_3
                                                                                                            HorizontalAlignment := horizontal_alignment.Left
                                                                                                            VerticalAlignment := vertical_alignment.Top
                                                                                                            Padding := margin:
                                                                                                                Top := 0.0
                                                                                                                Right := 0.0
                                                                                                                Bottom := 0.0
                                                                                                                Left := 0.0
                                                                                                            Widget := stack_box: # StackBox_8
                                                                                                                Orientation := orientation.Vertical
                                                                                                                Slots := array:
                                                                                                                    stack_box_slot: # StackBoxSlot_1
                                                                                                                        HorizontalAlignment := horizontal_alignment.Right
                                                                                                                        VerticalAlignment := vertical_alignment.Top
                                                                                                                        Padding := margin:
                                                                                                                            Top := 0.0
                                                                                                                            Right := 0.0
                                                                                                                            Bottom := 0.0
                                                                                                                            Left := 0.0
                                                                                                                        Widget := PetGatherValue
                                                                                                                    stack_box_slot: # StackBoxSlot_3
                                                                                                                        HorizontalAlignment := horizontal_alignment.Right
                                                                                                                        VerticalAlignment := vertical_alignment.Top
                                                                                                                        Padding := margin:
                                                                                                                            Top := 0.0
                                                                                                                            Right := 0.0
                                                                                                                            Bottom := 0.0
                                                                                                                            Left := 0.0
                                                                                                                        Widget := StrengthValue
                                                                                                                    stack_box_slot: # StackBoxSlot_4
                                                                                                                        HorizontalAlignment := horizontal_alignment.Right
                                                                                                                        VerticalAlignment := vertical_alignment.Top
                                                                                                                        Padding := margin:
                                                                                                                            Top := 0.0
                                                                                                                            Right := 0.0
                                                                                                                            Bottom := 0.0
                                                                                                                            Left := 0.0
                                                                                                                        Widget := CriticalRateValue
                                                                                                                    stack_box_slot: # StackBoxSlot_5
                                                                                                                        HorizontalAlignment := horizontal_alignment.Right
                                                                                                                        VerticalAlignment := vertical_alignment.Top
                                                                                                                        Padding := margin:
                                                                                                                            Top := 0.0
                                                                                                                            Right := 0.0
                                                                                                                            Bottom := 0.0
                                                                                                                            Left := 0.0
                                                                                                                        Widget := CriticalDamageValue
                                                                                                                    
                                                                                                        
                                                                                            
                                                                                
                                                                    
                                                        
                                            
                                