using { /Verse.org/Simulation }
using { /UnrealEngine.com/Temporary/UI }
using { /UnrealEngine.com/Temporary/SpatialMath }
using { /Fortnite.com/UI }
using { /Verse.org/Colors }

widget_yokai_rest := class(widget_base):

    YokaiA : basic_yokai_icon = basic_yokai_icon{}
    YokaiB : basic_yokai_icon = basic_yokai_icon{}
    YokaiC : basic_yokai_icon = basic_yokai_icon{}
    YokaiD : basic_yokai_icon = basic_yokai_icon{}
    YokaiE : basic_yokai_icon = basic_yokai_icon{}
    YokaiF : basic_yokai_icon = basic_yokai_icon{}
    YokaiStat : yokai_stat_template = yokai_stat_template{}

    YokaiClickEvent : event(int) = event(int){}

    GetYokaiIcon<public>(Index:int):basic_yokai_icon=
        if(Index = 11):
            return YokaiA
        else if(Index = 12):
            return YokaiB
        else if(Index = 13):
            return YokaiC
        else if(Index = 14):
            return YokaiD
        else if(Index = 15):
            return YokaiE
        else if(Index = 16):
            return YokaiF
        else:
            return basic_yokai_icon{}

    Init<override>():void=
        (super:)Init()
        spawn{AwaitYokaiClick()}
        BasicBG.Title.SetText(StringToMessage("Yo-kai Idle"))

        YokaiStat.StatATitle.SetText(StringToMessage("Pet Gather :"))
        YokaiStat.StatBTitle.SetText(StringToMessage("Strength % :"))
        YokaiStat.StatCTitle.SetText(StringToMessage("Critical Rate :"))
        YokaiStat.StatDTitle.SetText(StringToMessage("Critical Damage :"))
        YokaiStat.StatETitle.SetText(StringToMessage("Last Hit :"))

    AwaitYokaiClick<public>()<suspends>:void=
        race:
            CloseEvent.Await()
            loop:
                YokaiA.NameButton.OnClick().Await()
                YokaiClickEvent.Signal(11)
            loop:
                YokaiB.NameButton.OnClick().Await()
                YokaiClickEvent.Signal(12)
            loop:
                YokaiC.NameButton.OnClick().Await()
                YokaiClickEvent.Signal(13)
            loop:
                YokaiD.NameButton.OnClick().Await()
                YokaiClickEvent.Signal(14)
            loop:
                YokaiE.NameButton.OnClick().Await()
                YokaiClickEvent.Signal(15)
            loop:
                YokaiF.NameButton.OnClick().Await()
                YokaiClickEvent.Signal(16)
    
    CreateCanvas<override>():canvas=
        canvas: # CanvasPanel_52
            Slots := array:
                canvas_slot: # CanvasPanelSlot_1
                    Anchors := anchors:
                        Minimum := vector2{X := 0.0,Y := 0.0}
                        Maximum := vector2{X := 1.0,Y := 1.0}
                    Offsets := margin:
                        Top := 0.0
                        Right := 0.0
                        Bottom := 0.0
                        Left := 0.0
                    SizeToContent := true
                    Alignment := vector2{X := 0.0,Y := 0.0}
                    ZOrder := 0
                    Widget := overlay: # Overlay_56
                        Slots := array:
                            overlay_slot: # OverlaySlot_0
                                HorizontalAlignment := horizontal_alignment.Fill
                                VerticalAlignment := vertical_alignment.Fill
                                Padding := margin:
                                    Top := 0.0
                                    Right := 0.0
                                    Bottom := 0.0
                                    Left := 0.0
                                Widget := BasicBG.CreateOverlay()
                            overlay_slot: # OverlaySlot_1
                                HorizontalAlignment := horizontal_alignment.Fill
                                VerticalAlignment := vertical_alignment.Fill
                                Padding := margin:
                                    Top := 100.0
                                    Right := 100.0
                                    Bottom := 100.0
                                    Left := 100.0
                                Widget := stack_box: # StackBox_83
                                    Orientation := orientation.Vertical
                                    Slots := array:
                                        stack_box_slot: # StackBoxSlot_0
                                            HorizontalAlignment := horizontal_alignment.Center
                                            VerticalAlignment := vertical_alignment.Top
                                            Padding := margin:
                                                Top := 25.0
                                                Right := 0.0
                                                Bottom := 0.0
                                                Left := 0.0
                                            Widget := stack_box: # StackBox_192
                                                Orientation := orientation.Horizontal
                                                Slots := array:
                                                    stack_box_slot: # StackBoxSlot_0
                                                        HorizontalAlignment := horizontal_alignment.Left
                                                        VerticalAlignment := vertical_alignment.Top
                                                        Padding := margin:
                                                            Top := 0.0
                                                            Right := 0.0
                                                            Bottom := 0.0
                                                            Left := 0.0
                                                        Widget := YokaiA.CreateCanvas()
                                                    stack_box_slot: # StackBoxSlot_1
                                                        HorizontalAlignment := horizontal_alignment.Left
                                                        VerticalAlignment := vertical_alignment.Top
                                                        Padding := margin:
                                                            Top := 0.0
                                                            Right := 0.0
                                                            Bottom := 0.0
                                                            Left := 0.0
                                                        Widget := YokaiB.CreateCanvas()
                                                    stack_box_slot: # StackBoxSlot_3
                                                        HorizontalAlignment := horizontal_alignment.Left
                                                        VerticalAlignment := vertical_alignment.Top
                                                        Padding := margin:
                                                            Top := 0.0
                                                            Right := 0.0
                                                            Bottom := 0.0
                                                            Left := 0.0
                                                        Widget := YokaiC.CreateCanvas()
                                                    
                                        stack_box_slot: # StackBoxSlot_4
                                            HorizontalAlignment := horizontal_alignment.Center
                                            VerticalAlignment := vertical_alignment.Top
                                            Padding := margin:
                                                Top := 25.0
                                                Right := 0.0
                                                Bottom := 0.0
                                                Left := 0.0
                                            Widget := stack_box: # StackBox
                                                Orientation := orientation.Horizontal
                                                Slots := array:
                                                    stack_box_slot: # StackBoxSlot_0
                                                        HorizontalAlignment := horizontal_alignment.Left
                                                        VerticalAlignment := vertical_alignment.Top
                                                        Padding := margin:
                                                            Top := 0.0
                                                            Right := 0.0
                                                            Bottom := 0.0
                                                            Left := 0.0
                                                        Widget := YokaiD.CreateCanvas()
                                                    stack_box_slot: # StackBoxSlot_1
                                                        HorizontalAlignment := horizontal_alignment.Left
                                                        VerticalAlignment := vertical_alignment.Top
                                                        Padding := margin:
                                                            Top := 0.0
                                                            Right := 0.0
                                                            Bottom := 0.0
                                                            Left := 0.0
                                                        Widget := YokaiE.CreateCanvas()
                                                    stack_box_slot: # StackBoxSlot_3
                                                        HorizontalAlignment := horizontal_alignment.Left
                                                        VerticalAlignment := vertical_alignment.Top
                                                        Padding := margin:
                                                            Top := 0.0
                                                            Right := 0.0
                                                            Bottom := 0.0
                                                            Left := 0.0
                                                        Widget := YokaiF.CreateCanvas()
                                                    
                                        stack_box_slot: # StackBoxSlot_5
                                            HorizontalAlignment := horizontal_alignment.Center
                                            VerticalAlignment := vertical_alignment.Center
                                            Padding := margin:
                                                Top := 25.0
                                                Right := 0.0
                                                Bottom := 0.0
                                                Left := 0.0
                                            Widget := YokaiStat.CreateCanvas()
                                        
                            
                