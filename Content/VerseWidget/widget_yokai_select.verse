using { /Fortnite.com/UI }
using { /UnrealEngine.com/Temporary/UI }
using { /Verse.org/Colors }
using { /UnrealEngine.com/Temporary/SpatialMath }
using { /Verse.org/Assets }

DebugText<localizes>:message="DEBUG MESSAGE"

widget_yokai_select := class(widget_base):
    StatA : yokai_stat_template = yokai_stat_template{}
    StatB : yokai_stat_template = yokai_stat_template{}

    SelectInfoA : yokai_select_template = yokai_select_template{}
    SelectInfoB : yokai_select_template = yokai_select_template{}

    YokaiIconA : basic_yokai_icon = basic_yokai_icon{}
    YokaiIconB : basic_yokai_icon = basic_yokai_icon{}
    YokaiIconC : basic_yokai_icon = basic_yokai_icon{}
    YokaiIconD : basic_yokai_icon = basic_yokai_icon{}
    YokaiIconE : basic_yokai_icon = basic_yokai_icon{}
    YokaiIconF : basic_yokai_icon = basic_yokai_icon{}
    YokaiIconG : basic_yokai_icon = basic_yokai_icon{}
    YokaiIconH : basic_yokai_icon = basic_yokai_icon{}
    YokaiIconI : basic_yokai_icon = basic_yokai_icon{}
    YokaiIconJ : basic_yokai_icon = basic_yokai_icon{}

    YokaiClickEvent : event(int) = event(int){}

    UpdateYokaiIcon<public>(Index:int, Icon:texture, Name:string, Level:string):void=
        if(Index = 1):
            YokaiIconA.Icon.SetVisibility(widget_visibility.Visible)
            YokaiIconA.LevelTitle.SetVisibility(widget_visibility.Visible)
            YokaiIconA.Icon.SetImage(Icon)
            YokaiIconA.NameButton.SetText(StringToMessage(Name))
            YokaiIconA.LevelValue.SetText(StringToMessage("{Level}"))
        else if(Index = 2):
            YokaiIconB.Icon.SetVisibility(widget_visibility.Visible)
            YokaiIconB.LevelTitle.SetVisibility(widget_visibility.Visible)
            YokaiIconB.Icon.SetImage(Icon)
            YokaiIconB.NameButton.SetText(StringToMessage(Name))
            YokaiIconB.LevelValue.SetText(StringToMessage("{Level}"))
        else if(Index = 3):
            YokaiIconC.Icon.SetVisibility(widget_visibility.Visible)
            YokaiIconC.LevelTitle.SetVisibility(widget_visibility.Visible)
            YokaiIconC.Icon.SetImage(Icon)
            YokaiIconC.NameButton.SetText(StringToMessage(Name))
            YokaiIconC.LevelValue.SetText(StringToMessage("{Level}"))
        else if(Index = 4):
            YokaiIconD.Icon.SetVisibility(widget_visibility.Visible)
            YokaiIconD.LevelTitle.SetVisibility(widget_visibility.Visible)
            YokaiIconD.Icon.SetImage(Icon)
            YokaiIconD.NameButton.SetText(StringToMessage(Name))
            YokaiIconD.LevelValue.SetText(StringToMessage("{Level}"))
        else if(Index = 5):
            YokaiIconE.Icon.SetVisibility(widget_visibility.Visible)
            YokaiIconE.LevelTitle.SetVisibility(widget_visibility.Visible)
            YokaiIconE.Icon.SetImage(Icon)
            YokaiIconE.NameButton.SetText(StringToMessage(Name))
            YokaiIconE.LevelValue.SetText(StringToMessage("{Level}"))
        else if(Index = 6):
            YokaiIconF.Icon.SetVisibility(widget_visibility.Visible)
            YokaiIconF.LevelTitle.SetVisibility(widget_visibility.Visible)
            YokaiIconF.Icon.SetImage(Icon)
            YokaiIconF.NameButton.SetText(StringToMessage(Name))
            YokaiIconF.LevelValue.SetText(StringToMessage("{Level}"))
        else if(Index = 7):
            YokaiIconG.Icon.SetVisibility(widget_visibility.Visible)
            YokaiIconG.LevelTitle.SetVisibility(widget_visibility.Visible)
            YokaiIconG.Icon.SetImage(Icon)
            YokaiIconG.NameButton.SetText(StringToMessage(Name))
            YokaiIconG.LevelValue.SetText(StringToMessage("{Level}"))
        else if(Index = 8):
            YokaiIconH.Icon.SetVisibility(widget_visibility.Visible)
            YokaiIconH.LevelTitle.SetVisibility(widget_visibility.Visible)
            YokaiIconH.Icon.SetImage(Icon)
            YokaiIconH.NameButton.SetText(StringToMessage(Name))
            YokaiIconH.LevelValue.SetText(StringToMessage("{Level}"))
        else if(Index = 9):
            YokaiIconI.Icon.SetVisibility(widget_visibility.Visible)
            YokaiIconI.LevelTitle.SetVisibility(widget_visibility.Visible)
            YokaiIconI.Icon.SetImage(Icon)
            YokaiIconI.NameButton.SetText(StringToMessage(Name))
            YokaiIconI.LevelValue.SetText(StringToMessage("{Level}"))
        else if(Index = 10):
            YokaiIconJ.Icon.SetVisibility(widget_visibility.Visible)
            YokaiIconJ.LevelTitle.SetVisibility(widget_visibility.Visible)
            YokaiIconJ.Icon.SetImage(Icon)
            YokaiIconJ.NameButton.SetText(StringToMessage(Name))
            YokaiIconJ.LevelValue.SetText(StringToMessage("{Level}"))

    UpdateYokaiEmptyIndex(Index:int):void=
        if(Value := Mod[Index, 5]+1):
            if(Index = 1):
                YokaiIconA.SetEmpty(Value)
            else if(Index = 2):
                YokaiIconB.SetEmpty(Value)
            else if(Index = 3):
                YokaiIconC.SetEmpty(Value)
            else if(Index = 4):
                YokaiIconD.SetEmpty(Value)
            else if(Index = 5):
                YokaiIconE.SetEmpty(Value)
            else if(Index = 6):
                YokaiIconF.SetEmpty(Value)
            else if(Index = 7):
                YokaiIconG.SetEmpty(Value)
            else if(Index = 8):
                YokaiIconH.SetEmpty(Value)
            else if(Index = 9):
                YokaiIconI.SetEmpty(Value)
            else if(Index = 10):
                YokaiIconJ.SetEmpty(Value)
    Init<override>():void=
        spawn{AwaitYokaiClick()}
        #StatA.StatTitle.SetText(StringToMessage("Stat"))
        StatA.StatAIcon.SetImage(icon_set_2.Stats_PetGather)
        StatA.StatBIcon.SetImage(icon_set_2.Stats_Str)
        StatA.StatCIcon.SetImage(icon_set_2.Stats_CriRate)
        StatA.StatDIcon.SetImage(icon_set_2.Stats_CriDamage)
        StatA.StatEIcon.SetImage(icon_set_2.IC_LastHit)

        StatA.StatATitle.SetText(StringToMessage("Pet Gather :"))
        StatA.StatBTitle.SetText(StringToMessage("Strength % :"))
        StatA.StatCTitle.SetText(StringToMessage("Critical Rate :"))
        StatA.StatDTitle.SetText(StringToMessage("Critical Damage :"))
        StatA.StatETitle.SetText(StringToMessage("Last Hit :"))

        StatA.StatAValue.SetText(StringToMessage("+ 0 %"))
        StatA.StatBValue.SetText(StringToMessage("+ 0 %"))
        StatA.StatCValue.SetText(StringToMessage("+ 0 %"))
        StatA.StatDValue.SetText(StringToMessage("+ 0 %"))
        StatA.StatEValue.SetText(StringToMessage("+ 0 %"))

        #StatB.StatTitle.SetText(StringToMessage("Stat"))
        StatB.StatAIcon.SetImage(icon_set_2.Stats_PetGather)
        StatB.StatBIcon.SetImage(icon_set_2.Stats_Str)
        StatB.StatCIcon.SetImage(icon_set_2.Stats_CriRate)
        StatB.StatDIcon.SetImage(icon_set_2.Stats_CriDamage)
        StatB.StatEIcon.SetImage(icon_set_2.IC_LastHit)

        StatB.StatATitle.SetText(StringToMessage("Pet Gather :"))
        StatB.StatBTitle.SetText(StringToMessage("Strength % :"))
        StatB.StatCTitle.SetText(StringToMessage("Critical Rate :"))
        StatB.StatDTitle.SetText(StringToMessage("Critical Damage :"))
        StatB.StatETitle.SetText(StringToMessage("Last Hit :"))
        
        StatB.StatAValue.SetText(StringToMessage("+ 0 %"))
        StatB.StatBValue.SetText(StringToMessage("+ 0 %"))
        StatB.StatCValue.SetText(StringToMessage("+ 0 %"))
        StatB.StatDValue.SetText(StringToMessage("+ 0 %"))
        StatB.StatEValue.SetText(StringToMessage("+ 0 %"))

        BasicBG.Title.SetText(StringToMessage("Yo-kai Equip"))

    AwaitYokaiClick<public>()<suspends>:void=
        race:
            CloseEvent.Await()
            loop:
                YokaiIconA.NameButton.OnClick().Await()
                YokaiClickEvent.Signal(1)
            loop:
                YokaiIconB.NameButton.OnClick().Await()
                YokaiClickEvent.Signal(2)
            loop:
                YokaiIconC.NameButton.OnClick().Await()
                YokaiClickEvent.Signal(3)
            loop:
                YokaiIconD.NameButton.OnClick().Await()
                YokaiClickEvent.Signal(4)
            loop:
                YokaiIconE.NameButton.OnClick().Await()
                YokaiClickEvent.Signal(5)
            loop:
                YokaiIconF.NameButton.OnClick().Await()
                YokaiClickEvent.Signal(6)
            loop:
                YokaiIconG.NameButton.OnClick().Await()
                YokaiClickEvent.Signal(7)
            loop:
                YokaiIconH.NameButton.OnClick().Await()
                YokaiClickEvent.Signal(8)
            loop:
                YokaiIconI.NameButton.OnClick().Await()
                YokaiClickEvent.Signal(9)
            loop:
                YokaiIconJ.NameButton.OnClick().Await()
                YokaiClickEvent.Signal(10)

    CreateCanvas<override>():canvas=
        canvas: # CanvasPanel_26
            Slots := array:
                canvas_slot: # CanvasPanelSlot_1
                    Anchors := anchors:
                        Minimum := vector2{X := 0.0,Y := 0.0}
                        Maximum := vector2{X := 1.0,Y := 1.0}
                    Offsets := margin:
                        Top := 0.0
                        Right := 0.0
                        Bottom := 0.0
                        Left := 0.0
                    SizeToContent := true
                    Alignment := vector2{X := 0.0,Y := 0.0}
                    ZOrder := 0
                    Widget := overlay: # Overlay_56
                        Slots := array:
                            overlay_slot: # OverlaySlot_0
                                HorizontalAlignment := horizontal_alignment.Fill
                                VerticalAlignment := vertical_alignment.Fill
                                Padding := margin:
                                    Top := 0.0
                                    Right := 0.0
                                    Bottom := 0.0
                                    Left := 0.0
                                Widget := BasicBG.CreateOverlay()
                            overlay_slot: # OverlaySlot_1
                                HorizontalAlignment := horizontal_alignment.Center
                                VerticalAlignment := vertical_alignment.Center
                                Padding := margin:
                                    Top := 50.0
                                    Right := 50.0
                                    Bottom := 50.0
                                    Left := 50.0
                                Widget := stack_box: # StackBox_73
                                    Orientation := orientation.Horizontal
                                    Slots := array:
                                        stack_box_slot: # StackBoxSlot_4
                                            HorizontalAlignment := horizontal_alignment.Left
                                            VerticalAlignment := vertical_alignment.Top
                                            Padding := margin:
                                                Top := 0.0
                                                Right := 0.0
                                                Bottom := 0.0
                                                Left := 0.0
                                            Widget := stack_box: # StackBox_228
                                                Orientation := orientation.Vertical
                                                Slots := array:
                                                    stack_box_slot: # StackBoxSlot_1
                                                        HorizontalAlignment := horizontal_alignment.Left
                                                        VerticalAlignment := vertical_alignment.Top
                                                        Padding := margin:
                                                            Top := 0.0
                                                            Right := 0.0
                                                            Bottom := 0.0
                                                            Left := 0.0
                                                        Widget := overlay: # Overlay_183
                                                            Slots := array:
                                                                overlay_slot: # OverlaySlot_2
                                                                    HorizontalAlignment := horizontal_alignment.Fill
                                                                    VerticalAlignment := vertical_alignment.Fill
                                                                    Padding := margin:
                                                                        Top := 10.0
                                                                        Right := 10.0
                                                                        Bottom := 10.0
                                                                        Left := 10.0
                                                                    Widget := stack_box: # StackBox_308
                                                                        Orientation := orientation.Vertical
                                                                        Slots := array:
                                                                            stack_box_slot: # StackBoxSlot_0
                                                                                HorizontalAlignment := horizontal_alignment.Left
                                                                                VerticalAlignment := vertical_alignment.Top
                                                                                Padding := margin:
                                                                                    Top := 0.0
                                                                                    Right := 0.0
                                                                                    Bottom := 0.0
                                                                                    Left := 0.0
                                                                                Widget := SelectInfoA.CreateCanvas()
                                                                            stack_box_slot:
                                                                                HorizontalAlignment := horizontal_alignment.Center
                                                                                VerticalAlignment := vertical_alignment.Center
                                                                                Padding := margin:
                                                                                    Top := 5.0
                                                                                    Right := 0.0
                                                                                    Bottom := 5.0
                                                                                    Left := 0.0
                                                                                Widget := StatA.CreateCanvas()
                    
                                                                
                                                    stack_box_slot: # StackBoxSlot_2
                                                        HorizontalAlignment := horizontal_alignment.Left
                                                        VerticalAlignment := vertical_alignment.Top
                                                        Padding := margin:
                                                            Top := 0.0
                                                            Right := 0.0
                                                            Bottom := 0.0
                                                            Left := 0.0
                                                        Widget := overlay: # Overlay
                                                            Slots := array:
                                                                # overlay_slot: # OverlaySlot_0
                                                                #     HorizontalAlignment := horizontal_alignment.Center
                                                                #     VerticalAlignment := vertical_alignment.Top
                                                                #     Padding := margin:
                                                                #         Top := 0.0
                                                                #         Right := 0.0
                                                                #         Bottom := 0.0
                                                                #         Left := 0.0
                                                                #     Widget := stack_box: # StackBox_4
                                                                #         Orientation := orientation.Vertical
                                                                #         Slots := array:
                                                                            
                                                                # overlay_slot: # OverlaySlot_1
                                                                #     HorizontalAlignment := horizontal_alignment.Fill
                                                                #     VerticalAlignment := vertical_alignment.Fill
                                                                #     Padding := margin:
                                                                #         Top := 0.0
                                                                #         Right := 0.0
                                                                #         Bottom := 0.0
                                                                #         Left := 0.0
                                                                #     Widget := StatBackground
                                                                overlay_slot: # OverlaySlot_2
                                                                    HorizontalAlignment := horizontal_alignment.Fill
                                                                    VerticalAlignment := vertical_alignment.Fill
                                                                    Padding := margin:
                                                                        Top := 10.0
                                                                        Right := 10.0
                                                                        Bottom := 10.0
                                                                        Left := 10.0
                                                                    Widget := stack_box: # StackBox_5
                                                                        Orientation := orientation.Vertical
                                                                        Slots := array:
                                                                            stack_box_slot: # StackBoxSlot_0
                                                                                HorizontalAlignment := horizontal_alignment.Left
                                                                                VerticalAlignment := vertical_alignment.Top
                                                                                Padding := margin:
                                                                                    Top := 0.0
                                                                                    Right := 0.0
                                                                                    Bottom := 0.0
                                                                                    Left := 0.0
                                                                                Widget := SelectInfoB.CreateCanvas()
                                                                            stack_box_slot:
                                                                                HorizontalAlignment := horizontal_alignment.Center
                                                                                VerticalAlignment := vertical_alignment.Center
                                                                                Padding := margin:
                                                                                    Top := 5.0
                                                                                    Right := 0.0
                                                                                    Bottom := 5.0
                                                                                    Left := 0.0
                                                                                Widget := StatB.CreateCanvas()
                                                    
                                        stack_box_slot: # StackBoxSlot_9
                                            HorizontalAlignment := horizontal_alignment.Center
                                            VerticalAlignment := vertical_alignment.Center
                                            Padding := margin:
                                                Top := 0.0
                                                Right := 0.0
                                                Bottom := 0.0
                                                Left := 0.0
                                            Widget := stack_box: # StackBox_928
                                                Orientation := orientation.Vertical
                                                Slots := array:
                                                    stack_box_slot: # StackBoxSlot_3
                                                        HorizontalAlignment := horizontal_alignment.Left
                                                        VerticalAlignment := vertical_alignment.Top
                                                        Padding := margin:
                                                            Top := 0.0
                                                            Right := 0.0
                                                            Bottom := 0.0
                                                            Left := 0.0
                                                        Widget := stack_box: # StackBox_1039
                                                            Orientation := orientation.Horizontal
                                                            Slots := array:
                                                                stack_box_slot: # StackBoxSlot_0
                                                                    HorizontalAlignment := horizontal_alignment.Left
                                                                    VerticalAlignment := vertical_alignment.Top
                                                                    Padding := margin:
                                                                        Top := 10.0
                                                                        Right := 10.0
                                                                        Bottom := 10.0
                                                                        Left := 10.0
                                                                    Widget := YokaiIconA.CreateCanvas()
                                                                stack_box_slot: # StackBoxSlot_1
                                                                    HorizontalAlignment := horizontal_alignment.Left
                                                                    VerticalAlignment := vertical_alignment.Top
                                                                    Padding := margin:
                                                                        Top := 10.0
                                                                        Right := 10.0
                                                                        Bottom := 10.0
                                                                        Left := 10.0
                                                                    Widget := YokaiIconB.CreateCanvas()
                                                                stack_box_slot: # StackBoxSlot_2
                                                                    HorizontalAlignment := horizontal_alignment.Left
                                                                    VerticalAlignment := vertical_alignment.Top
                                                                    Padding := margin:
                                                                        Top := 10.0
                                                                        Right := 10.0
                                                                        Bottom := 10.0
                                                                        Left := 10.0
                                                                    Widget := YokaiIconC.CreateCanvas()
                                                                stack_box_slot: # StackBoxSlot_3
                                                                    HorizontalAlignment := horizontal_alignment.Left
                                                                    VerticalAlignment := vertical_alignment.Top
                                                                    Padding := margin:
                                                                        Top := 10.0
                                                                        Right := 10.0
                                                                        Bottom := 10.0
                                                                        Left := 10.0
                                                                    Widget := YokaiIconD.CreateCanvas()
                                                                
                                                    stack_box_slot: # StackBoxSlot_1
                                                        HorizontalAlignment := horizontal_alignment.Left
                                                        VerticalAlignment := vertical_alignment.Top
                                                        Padding := margin:
                                                            Top := 0.0
                                                            Right := 0.0
                                                            Bottom := 0.0
                                                            Left := 0.0
                                                        Widget := stack_box: # StackBox_14
                                                            Orientation := orientation.Horizontal
                                                            Slots := array:
                                                                stack_box_slot: # StackBoxSlot_0
                                                                    HorizontalAlignment := horizontal_alignment.Left
                                                                    VerticalAlignment := vertical_alignment.Top
                                                                    Padding := margin:
                                                                        Top := 10.0
                                                                        Right := 10.0
                                                                        Bottom := 10.0
                                                                        Left := 10.0
                                                                    Widget := YokaiIconE.CreateCanvas()
                                                                stack_box_slot: # StackBoxSlot_1
                                                                    HorizontalAlignment := horizontal_alignment.Left
                                                                    VerticalAlignment := vertical_alignment.Top
                                                                    Padding := margin:
                                                                        Top := 10.0
                                                                        Right := 10.0
                                                                        Bottom := 10.0
                                                                        Left := 10.0
                                                                    Widget := YokaiIconF.CreateCanvas()
                                                                stack_box_slot: # StackBoxSlot_2
                                                                    HorizontalAlignment := horizontal_alignment.Left
                                                                    VerticalAlignment := vertical_alignment.Top
                                                                    Padding := margin:
                                                                        Top := 10.0
                                                                        Right := 10.0
                                                                        Bottom := 10.0
                                                                        Left := 10.0
                                                                    Widget := YokaiIconG.CreateCanvas()
                                                                stack_box_slot: # StackBoxSlot_3
                                                                    HorizontalAlignment := horizontal_alignment.Left
                                                                    VerticalAlignment := vertical_alignment.Top
                                                                    Padding := margin:
                                                                        Top := 10.0
                                                                        Right := 10.0
                                                                        Bottom := 10.0
                                                                        Left := 10.0
                                                                    Widget := YokaiIconH.CreateCanvas()
                                                                
                                                    stack_box_slot: # StackBoxSlot_2
                                                        HorizontalAlignment := horizontal_alignment.Left
                                                        VerticalAlignment := vertical_alignment.Top
                                                        Padding := margin:
                                                            Top := 0.0
                                                            Right := 0.0
                                                            Bottom := 0.0
                                                            Left := 0.0
                                                        Widget := stack_box: # StackBox_15
                                                            Orientation := orientation.Horizontal
                                                            Slots := array:
                                                                stack_box_slot: # StackBoxSlot_0
                                                                    HorizontalAlignment := horizontal_alignment.Left
                                                                    VerticalAlignment := vertical_alignment.Top
                                                                    Padding := margin:
                                                                        Top := 10.0
                                                                        Right := 10.0
                                                                        Bottom := 10.0
                                                                        Left := 10.0
                                                                    Widget := YokaiIconI.CreateCanvas()
                                                                stack_box_slot: # StackBoxSlot_1
                                                                    HorizontalAlignment := horizontal_alignment.Left
                                                                    VerticalAlignment := vertical_alignment.Top
                                                                    Padding := margin:
                                                                        Top := 10.0
                                                                        Right := 10.0
                                                                        Bottom := 10.0
                                                                        Left := 10.0
                                                                    Widget := YokaiIconJ.CreateCanvas()
                                                                
                                                    
                                        
                            
                