using { /Fortnite.com/Devices }
using { /Verse.org/Simulation }

using { Mining_System.Code.PlayerData }
using { Mining_System.Code.Pet }

yokai_menu_device<public> := class(widget_device):

    var WidgetPerPlayer : [player]?widget_yokai_menu = map{}

    @editable
    PetManager : pet_manager = pet_manager{}

    @editable
    OpenButton : button_device = button_device{}

    @editable
    YokaiSelectDevice : yokai_select_device = yokai_select_device{}
    @editable
    var SaveDataManager : save_data_manager = save_data_manager{}

    CloseEvent<public> : event(player) = event(player){}

    OnBegin<override>()<suspends>:void=
        loop:
            Agent := OpenButton.InteractedWithEvent.Await()
            if(Player := player[Agent]):
                AddUI(Player)

    AddUI<override>(Player:player):void=
        NewUI := widget_yokai_menu{Player := Player}
        NewUI.Init()
        NewUI.AddInteractUI()
        if(set WidgetPerPlayer[Player] = option{NewUI}){}
        spawn{AwaitYokaiSelect(Player)}
        spawn{AwaitCloseClicked(Player)}
        UpdateSlots(Player)
        UpdateStats(Player)
        UpdateSkills(Player)
        #UIローカライズ
        spawn{LocalizeText(Player)}

    LocalizeText(Player : player)<suspends>:void=
        var LanguageIndex : int = SaveDataManager.GetLanguage(Player)
        PetOwnedText := if(LanguageIndex = 0){"妖怪所持数 :"} else {"Pet Owned :"}
        EquipText := if(LanguageIndex = 0){"妖怪装備数 :"} else {"Pet Equipped :"}
        TotalLvText := if(LanguageIndex = 0){"所持妖怪レベル :"} else {"Total Pet Levels :"}
        TotalBonusText := if(LanguageIndex = 0){"所持妖怪レベルボーナス :"} else {"Total Pet Levels Bonus :"}
        GatherText := if(LanguageIndex = 0){"素材採取力 :"} else {"Pet Gather :"}
        StrText := if(LanguageIndex = 0){"採掘力 +% :"} else {"Strength +% :"}
        CRateText := if(LanguageIndex = 0){"クリティカル率 :"} else {"Critical Rate :"}
        CDamageText := if(LanguageIndex = 0){"クリティカル威力 :"} else {"Critical Damage :"}
        LastHitText := if(LanguageIndex = 0){"ラストヒットボーナス :"} else {"Last Hit Bonus :"}

        if(Element := WidgetPerPlayer[Player]?):
            Element.PetOwnedText.SetText(StringToMessage(PetOwnedText))
            Element.PetsEquippedText.SetText(StringToMessage(EquipText))
            Element.TotalPetLevelsText.SetText(StringToMessage(TotalLvText))
            Element.TotalPetLevelsBonusText.SetText(StringToMessage(TotalBonusText))
            Element.PetGatherTitle.SetText(StringToMessage(GatherText))
            Element.StrehgthTitle.SetText(StringToMessage(StrText))
            Element.CriticalRateTitle.SetText(StringToMessage(CRateText))
            Element.CriticalDamageTitle.SetText(StringToMessage(CDamageText))
            Element.LastHitTitle.SetText(StringToMessage(LastHitText))


    UpdateSlots<public>(Player:player):void=
        var UsedSlots : []int =  array{}
        if( Widget := WidgetPerPlayer[Player]?):
            if(EquipedPets := PetManager.GetEquipedPetStatus[Player]):
                for(PetElement : EquipedPets, Index := PetElement(0).Index, PetIcon := PetManager.PetIcon[PetManager.GetPetID[Player, Index]]):
                    Widget.UpdateYokaiIcon(PetElement(1).PetSlot, PetIcon, PetElement(0).PetName, "{PetElement(1).Level}")
                    set UsedSlots +=  array{PetElement(1).PetSlot}
                for(Slot := 1..4, not UsedSlots.Find[Slot]):
                    Widget.SetEmpty(Slot)
            else:
                for(Slot := 1..4):
                    Widget.SetEmpty(Slot)

    UpdateSkills<public>(Player:player):void=
        if (Widget := WidgetPerPlayer[Player]?):
            if(FirstPet := PetManager.GetPetBySlot[Player, 1], SkillIcon := FirstPet(0).PetSkillIcon[FirstPet(0).PetSkill_ID]):
                PetStatus := FirstPet(0)
                Widget.UpdatePetSupportInfo(PetStatus.PetSkillName, SkillIcon, PetStatus.PetSkillDescription, PetStatus.GetAdditionalValue())
            else:
                Widget.UpdatePetSupportInfo("Skill Name", Boss_Textures.Pet_1, "No Yokai Equiped", "")


    UpdateStats<public>(Player:player):void=
        OwnedPets := PetManager.GetPetCount(Player)
        PetEquiped := PetManager.GetEquipedPetCount(Player)
        TotalPetLevel := PetManager.GetTotalLevel(Player)
        TotalPetLevelBonus := PetManager.GetTotalLevelBonus(Player)
        if (Widget := WidgetPerPlayer[Player]?):
            Widget.UpdateStatsSummary("{OwnedPets} / 10", "{PetEquiped} / 4", "{TotalPetLevel}" ,"{BigNumber_FormatDisplay(TotalPetLevelBonus)}")
        TotalPetGather := PetManager.GetPetTotalGatherRate(Player)
        TotalPetStrength := PetManager.GetPetTotalStrength(Player)
        TotalCriticalRate := PetManager.GetPetTotalCriticalRate(Player)
        TotalCriticalDamage := PetManager.GetPetTotalCriticalDamage(Player)
        TotalLastHit := PetManager.GetPetTotalLastHit(Player)
        if (Widget := WidgetPerPlayer[Player]?):
            Widget.UpdateStatsTotal("+ {BigNumber_FormatDisplay(TotalPetGather)} %", "+ {BigNumber_FormatDisplay(TotalPetStrength)} %", "+ {BigNumber_FormatDisplay(TotalCriticalRate)} %", "+ {BigNumber_FormatDisplay(TotalCriticalDamage)} %", "+ {BigNumber_FormatDisplay(TotalLastHit)} %")

    AwaitYokaiSelect<public>(InPlayer:player)<suspends>:void=
        if (Widget := WidgetPerPlayer[InPlayer]?):
            race:
                Widget.CloseEvent.Await()
                loop:
                    Index := Widget.YokaiClickEvent.Await()
                    defer:
                        YokaiSelectDevice.AddUI(InPlayer, Index, Self)
                    Widget.RemoveUI()

    AwaitCloseClicked<public>(InPlayer:player)<suspends>:void=
        if (Widget := WidgetPerPlayer[InPlayer]?):
            race:
                block:
                    Widget.CloseEvent.Await()
                    Sleep(0.25)
                loop:
                    WidgetMessage := Widget.BasicBG.CloseButton.OnClick().Await()
                    if(WidgetMessage.Player = InPlayer):
                        CloseEvent.Signal(WidgetMessage.Player)
            
