using { /Fortnite.com/Devices }
using { /Verse.org/Simulation }

using { VerseDevice }
using { Mining_System.Code.PlayerData }
using { Mining_System.Code.Pet }

yokai_rest_device<public> := class(widget_device):

    var WidgetPerPlayer : [player]?widget_yokai_rest = map{}

    @editable
    PetManager : pet_manager = pet_manager{}

    @editable
    OpenButton : button_device = button_device{}

    @editable
    SkillNPC : character_device = character_device{}

    @editable
    YokaiSelectDevice : yokai_select_device = yokai_select_device{}
    @editable
    var SaveDataManager : save_data_manager = save_data_manager{}
    @editable
    UIHudCount : ui_hud_count = ui_hud_count{}

    CloseEvent<public> : event(player) = event(player){}

    OnBegin<override>()<suspends>:void=
        loop:
            Agent := SkillNPC.InteractedWithEvent.Await()
            if(Player := player[Agent]):
                AddUI(Player)

    AddUI<override>(Player:player):void=
        NewUI := widget_yokai_rest{Player := Player}
        NewUI.Init()
        NewUI.AddInteractUI()
        if(set WidgetPerPlayer[Player] = option{NewUI}){}
        spawn{AwaitYokaiSelect(Player)}
        spawn{AwaitCloseClicked(Player)}
        UpdateSlots(Player)
        UpdateStats(Player)
        #UIローカライズ
        spawn{LocalizeText(Player)}

    LocalizeText(Player : player)<suspends>:void=
        var LanguageIndex : int = SaveDataManager.GetLanguage(Player)
        AText := if(LanguageIndex = 0){"素材採取力 :"} else {"Pet Gather :"}
        BText := if(LanguageIndex = 0){"採掘力 % :"} else {"Strength % :"}
        CText := if(LanguageIndex = 0){"クリティカル率 :"} else {"Critical Rate :"}
        DText := if(LanguageIndex = 0){"クリティカル威力 :"} else {"Critical Damage :"}
        EText := if(LanguageIndex = 0){"ラストヒットボーナス :"} else {"Last Hit Bonus :"}

        if(Element := WidgetPerPlayer[Player]?):
            Element.YokaiStat.StatATitle.SetText(StringToMessage(AText))
            Element.YokaiStat.StatBTitle.SetText(StringToMessage(BText))
            Element.YokaiStat.StatCTitle.SetText(StringToMessage(CText))
            Element.YokaiStat.StatDTitle.SetText(StringToMessage(DText))
            Element.YokaiStat.StatETitle.SetText(StringToMessage(EText))

    UpdateSlots<public>(Player:player):void=
        var UsedSlots : []int =  array{}
        if( Widget := WidgetPerPlayer[Player]?):
            if(EquipedPets := PetManager.GetEquipedRestPetStatus[Player]):
                for(PetElement : EquipedPets, Index := PetElement(0).Index, PetIcon := PetManager.PetIcon[PetManager.GetPetID[Player, Index]]):
                    YokaiIcon := Widget.GetYokaiIcon(PetElement(1).PetSlot)
                    YokaiIcon.LevelValue.SetText(StringToMessage("{PetElement(1).Level}"))
                    YokaiIcon.Icon.SetImage(PetIcon)
                    YokaiIcon.NameButton.SetText(StringToMessage(PetElement(0).PetName))
                    set UsedSlots +=  array{PetElement(1).PetSlot}
                for(Slot := 11..16, not UsedSlots.Find[Slot]):
                    YokaiIcon := Widget.GetYokaiIcon(Slot)
                    YokaiIcon.SetEmpty(Slot)
            else:
                for(Slot := 11..16, not UsedSlots.Find[Slot]):
                    YokaiIcon := Widget.GetYokaiIcon(Slot)
                    YokaiIcon.SetEmpty(Slot)

    UpdateStats<public>(Player:player):void=
        OwnedPets := PetManager.GetPetCount(Player)
        PetEquiped := PetManager.GetEquipedPetCount(Player)
        TotalPetLevel := PetManager.GetTotalLevel(Player)
        if (Widget := WidgetPerPlayer[Player]?):
            TotalPetGather := PetManager.GetPetTotalGatherRate(Player)
            TotalPetStrength := PetManager.GetPetTotalStrength(Player)
            TotalCriticalRate := PetManager.GetPetTotalCriticalRate(Player)
            TotalCriticalDamage := PetManager.GetPetTotalCriticalDamage(Player)
            TotalLastHit := PetManager.GetPetTotalLastHit(Player)
            Widget.YokaiStat.UpdateStatsTotal("+ {BigNumber_FormatDisplay(TotalPetGather)} %", "+ {BigNumber_FormatDisplay(TotalPetStrength)} %", "+ {BigNumber_FormatDisplay(TotalCriticalRate)} %", "+ {BigNumber_FormatDisplay(TotalCriticalDamage)} %", "+ {BigNumber_FormatDisplay(TotalLastHit)} %")
                

    AwaitYokaiSelect<public>(InPlayer:player)<suspends>:void=
        if (Widget := WidgetPerPlayer[InPlayer]?):
            race:
                Widget.CloseEvent.Await()
                loop:
                    Index := Widget.YokaiClickEvent.Await()
                    defer:
                        YokaiSelectDevice.AddUI(InPlayer, Index, Self)
                    Widget.RemoveUI()

    AwaitCloseClicked<public>(InPlayer:player)<suspends>:void=
        if (Widget := WidgetPerPlayer[InPlayer]?):
            race:
                block:
                    Widget.CloseEvent.Await()
                    Sleep(0.25)
                block :
                    loop:
                        WidgetMessage := Widget.BasicBG.CloseButton.OnClick().Await()
                        if(WidgetMessage.Player = InPlayer):
                            CloseEvent.Signal(WidgetMessage.Player)
            if(Agent : agent = agent[InPlayer]) :
                UIHudCount.Update_IdleUI(Agent)