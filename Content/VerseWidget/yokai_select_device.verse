using { /Fortnite.com/Devices }
using { /Verse.org/Simulation }

using { Mining_System.Code.PlayerData }
using { Mining_System.Code.Pet }

yokai_select_device := class(creative_device):

    var WidgetPerPlayer : [player]?widget_yokai_select = map{}
    var PetSelectedSlotPerPlayer : [player]int = map{}
    var PetSelectingSlotPerPlayer : [player]int = map{}

    @editable
    PetManager : pet_manager = pet_manager{}
    @editable
    var SaveDataManager : save_data_manager = save_data_manager{}

    AddUI<public>(Player:player, Slot:int, PrevWidget:widget_device):void=
        NewUI := widget_yokai_select{Player := Player}
        NewUI.Init()
        if:
            set WidgetPerPlayer[Player] = option{NewUI}
            set PetSelectedSlotPerPlayer[Player] = Slot
            set PetSelectingSlotPerPlayer[Player] = Slot
        spawn{AwaitYokaiSelect(Player)}
        spawn{AwaitYokaiSelectingClicked(Player)}
        spawn{AwaitYokaiSelectedClicked(Player)}
        spawn{AwaitClose(Player, PrevWidget)}
        Update(Player)
        NewUI.AddInteractUI()
        #UIローカライズ
        spawn{LocalizeText(Player)}

    LocalizeText(Player : player)<suspends>:void=
        var LanguageIndex : int = SaveDataManager.GetLanguage(Player)
        AText := if(LanguageIndex = 0){"素材採取力 :"} else {"Pet Gather :"}
        BText := if(LanguageIndex = 0){"採掘力 % :"} else {"Strength % :"}
        CText := if(LanguageIndex = 0){"クリティカル率 :"} else {"Critical Rate :"}
        DText := if(LanguageIndex = 0){"クリティカル威力 :"} else {"Critical Damage :"}
        EText := if(LanguageIndex = 0){"ラストヒットボーナス :"} else {"Last Hit Bonus :"}

        if(Element := WidgetPerPlayer[Player]?):
            Element.StatA.StatATitle.SetText(StringToMessage(AText))
            Element.StatA.StatBTitle.SetText(StringToMessage(BText))
            Element.StatA.StatCTitle.SetText(StringToMessage(CText))
            Element.StatA.StatDTitle.SetText(StringToMessage(DText))
            Element.StatA.StatETitle.SetText(StringToMessage(EText))

            Element.StatB.StatATitle.SetText(StringToMessage(AText))
            Element.StatB.StatBTitle.SetText(StringToMessage(BText))
            Element.StatB.StatCTitle.SetText(StringToMessage(CText))
            Element.StatB.StatDTitle.SetText(StringToMessage(DText))
            Element.StatB.StatETitle.SetText(StringToMessage(EText))

    Update(InPlayer:player):void=
        UpdateList(InPlayer)
        UpdateInfoB(InPlayer)
        UpdateInfoA(InPlayer)

    UpdateList<public>(InPlayer:player):void=
        if (Widget := WidgetPerPlayer[InPlayer]?):
            for(Index -> PetElement : PetManager.PetStatus):
                PetIndex := Index+1
                if(PetIcon := PetManager.PetIcon[PetManager.GetPetID[InPlayer, PetElement.Index]], PetStatus := PetManager.GetPetStatus[InPlayer, Index]):
                    Name := if(PetStatus(1).PetSlot >= 1, PetStatus(1).PetSlot <= 6):
                        "{PetElement.PetName}*"
                    else:
                        "{PetElement.PetName}"
                    Widget.UpdateYokaiIcon(PetIndex, PetIcon, Name, "{PetStatus(1).Level}")
                else:
                    Widget.UpdateYokaiEmptyIndex(PetIndex)

    UpdateInfoA<public>(InPlayer:player):void=
        if (Widget := WidgetPerPlayer[InPlayer]?, SelectedSlot := PetSelectedSlotPerPlayer[InPlayer], SelectedPet := PetManager.GetPetBySlot[InPlayer, SelectedSlot], PetIcon := PetManager.PetIcon[PetManager.GetPetID[InPlayer, SelectedPet(0).Index]]):
            Print("SelectedPetSlot: {SelectedSlot}")
            Widget.SelectInfoA.SelectLevelValue.SetText(StringToMessage("{SelectedPet(1).Level}"))
            Widget.SelectInfoA.SelectIcon_FLV.SetImage(PetIcon)
            Widget.SelectInfoA.SelectName.SetText(StringToMessage(SelectedPet(0).PetName))
            Widget.SelectInfoA.SelectButton.SetText(StringToMessage("Equiped"))

            if:
                GatherRate := PetManager.GetPetGatherRate[InPlayer, SelectedPet(0).Index]
                Strength := PetManager.GetPetStrength[InPlayer, SelectedPet(0).Index]
                CriticalRate := PetManager.GetPetCriticalRate[InPlayer, SelectedPet(0).Index]
                CriticalDamage := PetManager.GetPetCriticalDamage[InPlayer, SelectedPet(0).Index]
                LastHit := PetManager.GetPetLastHit[InPlayer, SelectedPet(0).Index]
            then:
                Widget.StatA.UpdateStatsTotal("+ {BigNumber_FormatDisplay(GatherRate)} %", "+ {BigNumber_FormatDisplay(Strength)} %", "+ {BigNumber_FormatDisplay(CriticalRate)} %", "+ {BigNumber_FormatDisplay(CriticalDamage)} %", "+ {BigNumber_FormatDisplay(LastHit)} %")
            else:
                Widget.StatA.UpdateStatsTotal("+ 0 %", "+ 0 %", "+ 0 %", "+ 0 %", "+ 0 %")
        else if(Widget := WidgetPerPlayer[InPlayer]?):
            Print("SelectedPetSlot: Unknown")
            Widget.SelectInfoA.SelectLevelValue.SetText(StringToMessage("0"))
            Widget.SelectInfoA.SelectIcon_FLV.SetImage(Boss_Textures.Pet_1)
            Widget.SelectInfoA.SelectName.SetText(StringToMessage("Empty"))
            Widget.SelectInfoA.SelectButton.SetText(StringToMessage("Equiped"))

            Widget.StatA.UpdateStatsTotal("+ 0 %", "+ 0 %", "+ 0 %", "+ 0 %", "+ 0 %")

    UpdateInfoB<public>(InPlayer:player):void=
        if (Widget := WidgetPerPlayer[InPlayer]?, SelectedSlot := PetSelectingSlotPerPlayer[InPlayer], SelectedPet := PetManager.GetPetBySlot[InPlayer, SelectedSlot], PetIcon := PetManager.PetIcon[PetManager.GetPetID[InPlayer, SelectedPet(0).Index]]):
            Widget.SelectInfoB.SelectLevelValue.SetText(StringToMessage("{SelectedPet(1).Level}"))
            Widget.SelectInfoB.SelectIcon_FLV.SetImage(PetIcon)
            Widget.SelectInfoB.SelectName.SetText(StringToMessage(SelectedPet(0).PetName))
            Widget.SelectInfoB.SelectButton.SetText(StringToMessage("Equip"))

            if:
                GatherRate := PetManager.GetPetGatherRate[InPlayer, SelectedPet(0).Index]
                Strength := PetManager.GetPetStrength[InPlayer, SelectedPet(0).Index]
                CriticalRate := PetManager.GetPetCriticalRate[InPlayer, SelectedPet(0).Index]
                CriticalDamage := PetManager.GetPetCriticalDamage[InPlayer, SelectedPet(0).Index]
                LastHit := PetManager.GetPetLastHit[InPlayer, SelectedPet(0).Index]
            then:
                Widget.StatB.UpdateStatsTotal("+ {BigNumber_FormatDisplay(GatherRate)} %", "+ {BigNumber_FormatDisplay(Strength)} %", "+ {BigNumber_FormatDisplay(CriticalRate)} %", "+ {BigNumber_FormatDisplay(CriticalDamage)} %", "+ {BigNumber_FormatDisplay(LastHit)} %")
            else:
                Widget.StatB.UpdateStatsTotal("+ 0 %", "+ 0 %", "+ 0 %", "+ 0 %", "+ 0 %")

        else if(Widget := WidgetPerPlayer[InPlayer]?):
            Widget.SelectInfoB.SelectLevelValue.SetText(StringToMessage("0"))
            Widget.SelectInfoB.SelectIcon_FLV.SetImage(Boss_Textures.Pet_1)
            Widget.SelectInfoB.SelectName.SetText(StringToMessage("Empty"))
            Widget.SelectInfoB.SelectButton.SetText(StringToMessage("Equiped"))

            Widget.StatB.UpdateStatsTotal("+ 0 %", "+ 0 %", "+ 0 %", "+ 0 %", "+ 0 %")

    AwaitYokaiSelect<public>(InPlayer:player)<suspends>:void=
        if (Widget := WidgetPerPlayer[InPlayer]?):
            race:
                Widget.CloseEvent.Await()
                loop:
                    Index := Widget.YokaiClickEvent.Await()
                    if(SelectedPet := PetManager.GetPetStatus[InPlayer, Index-1], set PetSelectingSlotPerPlayer[InPlayer] = SelectedPet(1).PetSlot):
                    Update(InPlayer)

    AwaitYokaiSelectingClicked<public>(InPlayer:player)<suspends>:void=
        if (Widget := WidgetPerPlayer[InPlayer]?):
            race:
                Widget.CloseEvent.Await()
                loop:
                    Widget.SelectInfoB.SelectButton.OnClick().Await()
                    if (SelectingSlot := PetSelectingSlotPerPlayer[InPlayer], SelectedSlot := PetSelectedSlotPerPlayer[InPlayer]):
                        Print("SelectingSlot: {SelectingSlot}, SelectedSlot: {SelectedSlot}")
                        if(SelectingPet := PetManager.GetPetBySlot[InPlayer, SelectingSlot]):
                            if(SelectedPet := PetManager.GetPetBySlot[InPlayer, SelectedSlot], set PetSelectingSlotPerPlayer[InPlayer] = SelectedSlot):
                                PetManager.SetPetSlot(InPlayer, SelectedPet(0).Index, SelectingSlot)
                            else:
                                PetManager.SetPetSlot(InPlayer, SelectingPet(0).Index, SelectingSlot)
                                if(set PetSelectingSlotPerPlayer[InPlayer] = SelectedSlot){}
                            PetManager.SetPetSlot(InPlayer, SelectingPet(0).Index, SelectedSlot)

                    Update(InPlayer)

    AwaitYokaiSelectedClicked<public>(InPlayer:player)<suspends>:void=
        if (Widget := WidgetPerPlayer[InPlayer]?):
            race:
                Widget.CloseEvent.Await()
                loop:
                    Widget.SelectInfoA.SelectButton.OnClick().Await()
                    if (SelectedSlot := PetSelectedSlotPerPlayer[InPlayer], SelectedPet := PetManager.GetPetBySlot[InPlayer, SelectedSlot]):
                        if(SelectedSlot >= 11, Slot := PetManager.GetEmptyEquipSlot[InPlayer]):
                            PetManager.SetPetSlot(InPlayer, SelectedPet(0).Index, Slot)
                            if(set PetSelectingSlotPerPlayer[InPlayer] = Slot){}
                        else if(SelectedSlot <= 6, Slot := PetManager.GetEmptyRestSlot[InPlayer]):
                            PetManager.SetPetSlot(InPlayer, SelectedPet(0).Index, Slot)
                            if(set PetSelectingSlotPerPlayer[InPlayer] = Slot){}

                    Update(InPlayer)

    
    AwaitClose<public>(InPlayer:player, PrevWidget:widget_device)<suspends>:void=
        if (Widget := WidgetPerPlayer[InPlayer]?):
            race:
                Widget.CloseEvent.Await()
                loop:
                    Widget.BasicBG.CloseButton.OnClick().Await()
                    defer:
                        PrevWidget.AddUI(InPlayer)
                    Widget.RemoveUI()
                    break