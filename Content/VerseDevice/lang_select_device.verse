using { /Fortnite.com/Devices }
using { /Fortnite.com/Characters }
using { /Fortnite.com/UI }
using { /Verse.org/Simulation }
using { /Verse.org/Colors }
using { /UnrealEngine.com/Temporary/Diagnostics }
using { /UnrealEngine.com/Temporary/UI }
using { /UnrealEngine.com/Temporary/SpatialMath }
using { /Fortnite.com/Devices/CreativeAnimation } 
using {Mining_System.Code.PlayerData}



Lang_select_ui_device<public> := class(creative_device):
    #SFX
    @editable
    SFX_Open : audio_player_device = audio_player_device{}
    @editable
    SFX_Close : audio_player_device = audio_player_device{}
    @editable
    SFX_LevelUp3 : audio_player_device = audio_player_device{}

    @editable
    SaveDataManager : save_data_manager = save_data_manager{}

    @editable
    _LangSelectButton : button_device = button_device{}

    @editable
    QuestSequencer : QuestSystem.quest_sequencer = QuestSystem.quest_sequencer{}

    @editable
    QuestIndicator : map_indicator_device = map_indicator_device{}

    var CanvasPerPlayer : [player]canvas = map{}
    var UIElements : [player]tuple(text_block, text_block) = map{}

    CloseEvent<public> : event() = event(){}
    StringToMessages<localizes>(value:string)<computes> : message = "{value}"
  
    OnBegin<override>()<suspends>:void=

      for(Player : GetPlayspace().GetPlayers()):
        FirstJoinPlayer(Player)
      GetPlayspace().PlayerAddedEvent().Subscribe(FirstJoinPlayer)

      _LangSelectButton.InteractedWithEvent.Subscribe(AddSelectLanguageWidget)

    # Playerが初参加の時実行するコード
    FirstJoinPlayer(Agent:agent):void=
        if(Player : player = player[Agent]) :
            QuestIndicator.ActivateObjectivePulse(Player)
            AddSelectLanguageWidget(Agent)
        
    CheckCanvas(Player : player):void=
        if(Widget := CanvasPerPlayer[Player]) :
            return #Widget作成済みなら何もしない

        NewCanvas := CreateCanvas(Player)
        logic{set CanvasPerPlayer[Player] = NewCanvas}

    AddSelectLanguageWidget<public>(Agent : agent) : void=
        SFX_Open.Play(Agent)
        if(Player : player = player[Agent]) :
            #Canvasの有無をチェックしてAddWidget
            CheckCanvas(Player)
            if(PlayerUI : player_ui = GetPlayerUI[Player], TheCanvas := CanvasPerPlayer[Player]) :
                PlayerUI.AddWidget(TheCanvas, player_ui_slot{InputMode := ui_input_mode.All})
                
            var LanguageIndex : int = SaveDataManager.GetLanguage(Player)
            SelectText := if(LanguageIndex = 0){"言語を選択してください"} else {"Select Language"}
            SelectText2 := if(LanguageIndex = 0){"Select Language"} else {"言語を選択してください"}
            if(Element := UIElements[Player]) :
                Element(0).SetText(StringToMessages(SelectText))
                Element(1).SetText(StringToMessages(SelectText2))
      
    # 日本語を選択した時
    OnJpnButtonClicked(WidgetMessage : widget_message):void=
        ### Process for Jpn
        Print("Jpn", ?Color := NamedColors.Red)
        if(Agent : agent = agent[WidgetMessage.Player]) :
            SFX_LevelUp3.Play(Agent)
        QuestSequencer.UpdateLanguageSignal(WidgetMessage.Player)
        if:
            NowData := SaveData[WidgetMessage.Player], set SaveData[WidgetMessage.Player] = save_data:
                MakeSaveData<constructor>(NowData)
                LanguageIndex := 0

        CollapseWidget(WidgetMessage.Player)

    # Engを選択した時
    OnEngButtonClicked(WidgetMessage : widget_message):void=
        ### Process for Eng
        Print("Eng", ?Color := NamedColors.Red)
        if(Agent : agent = agent[WidgetMessage.Player]) :
            SFX_LevelUp3.Play(Agent)
        QuestSequencer.UpdateLanguageSignal(WidgetMessage.Player)
        if:
            NowData := SaveData[WidgetMessage.Player], set SaveData[WidgetMessage.Player] = save_data:
                MakeSaveData<constructor>(NowData)
                LanguageIndex := 1

        CollapseWidget(WidgetMessage.Player)

    # Cancel押したとき
    OnCancelButtonClicked(WidgetMessage : widget_message):void=
        Print("Cancel")
        if(Agent : agent = agent[WidgetMessage.Player]) :
            SFX_Close.Play(Agent)
        CollapseWidget(WidgetMessage.Player)

    CollapseWidget(Player : player):void=
        if(PlayerUI : player_ui = GetPlayerUI[Player], TheCanvas := CanvasPerPlayer[Player]) :
            PlayerUI.RemoveWidget(TheCanvas)
        CloseEvent.Signal()


    CreateCanvas(Player : player):canvas=
        # widgets
        BGImage : texture_block = texture_block{DefaultImage := Textures.UI_BG, DefaultDesiredSize := vector2{X := 1920.0, Y := 1080.0}}
        _SelectLanguage : text_block = text_block{DefaultTextColor := NamedColors.White, DefaultJustification := text_justification.Center}
        _SelectLanguage2 : text_block = text_block{DefaultTextColor := NamedColors.White, DefaultJustification := text_justification.Center}

        _JpnButton : button_quiet = button_quiet{DefaultText := StringToMessages("日本語")}
        _JpnButton.OnClick().Subscribe(OnJpnButtonClicked)

        _EngButton : button_quiet = button_quiet{DefaultText := StringToMessages("English")}
        _EngButton.OnClick().Subscribe(OnEngButtonClicked)

        _CloseButton : button_loud = button_loud{DefaultText := StringToMessages("Close")}
        _CloseButton.OnClick().Subscribe(OnCancelButtonClicked)

        CenterVector : vector2 = vector2{X:=0.5, Y:=0.5}
        CenterAnchors : anchors = anchors{Minimum := CenterVector, Maximum := CenterVector}
        Canvas := canvas:
            Slots := array:

                #========================================== BG ==========================================

                canvas_slot:
                    Anchors := CenterAnchors
                    Offsets := margin{Top := 0.0,Left := 0.0, Right := 1000.0, Bottom := 500.0}
                    Alignment := CenterVector
                    SizeToContent := false
                    Widget := BGImage

                #========================================= TEXT ========================================

                canvas_slot:
                    Anchors := CenterAnchors
                    Offsets := margin{Left := 0.0, Top := -222.5, Right := 350.0, Bottom := 50.0}
                    Alignment := CenterVector
                    SizeToContent := false
                    Widget := _SelectLanguage2
                canvas_slot:
                    Anchors := CenterAnchors
                    Offsets := margin{Left := 0.0, Top := -120.0, Right := 350.0, Bottom := 50.0}
                    Alignment := CenterVector
                    SizeToContent := false
                    Widget := _SelectLanguage

                #======================================= StackBox ======================================

                canvas_slot:
                    Anchors := CenterAnchors
                    Offsets := margin{Left := -200.0, Top := 0.0, Right := 280.0, Bottom := 130.0}
                    Alignment := CenterVector
                    SizeToContent := false
                    Widget := _JpnButton
                canvas_slot:
                    Anchors := CenterAnchors
                    Offsets := margin{Left := 200.0, Top := 0.0, Right := 280.0, Bottom := 130.0}
                    Alignment := CenterVector
                    SizeToContent := false
                    Widget := _EngButton
                canvas_slot:
                    Anchors := CenterAnchors
                    Offsets := margin{Left := 0.0, Top := 145.0, Right := 170.0, Bottom := 80.0}
                    Alignment := CenterVector
                    SizeToContent := false
                    Widget := _CloseButton

                #=======================================================================================

        if(set UIElements[Player] = (_SelectLanguage, _SelectLanguage2)){}

        return Canvas