using { /Verse.org/Assets }
using { /Verse.org/Simulation }
using { /Fortnite.com/UI }
using { /Fortnite.com/Devices }
using { /Fortnite.com/Characters }
using { /UnrealEngine.com/Temporary/UI }
using { /UnrealEngine.com/Temporary/SpatialMath }
using { /Verse.org/Simulation }
using { /UnrealEngine.com/Temporary/Diagnostics }
using { /Verse.org/Colors }
using { /Verse.org/Colors/NamedColors }
using { VerseDevice }
using { VerseWidget }
using { Mining_System.Code.PlayerData }

menu_device := class(creative_device):
    @editable
    SaveDataManager : save_data_manager = save_data_manager{}
    #SFX
    @editable
    SFX_Open : audio_player_device = audio_player_device{}
    @editable
    SFX_Close : audio_player_device = audio_player_device{}
    @editable
    SFX_LevelUp3 : audio_player_device = audio_player_device{}
    @editable
    SFX_Warp : audio_player_device = audio_player_device{}
    #Teleporter
    @editable
    TP_Theater : teleporter_device = teleporter_device{}

    # @editable
    # PlayerSpawners : []player_spawner_device = array{}
    @editable
    StrSignal : signal_remote_manager_device = signal_remote_manager_device{}

    @editable
    QuestTrigger : trigger_device = trigger_device{}

    @editable
    YokaiMenu : yokai_menu_device = yokai_menu_device{}

    @editable
    YokaiRest : yokai_rest_device = yokai_rest_device{}

    @editable
    StrDevice : Ui_str_device = Ui_str_device{}

    @editable
    BossList : ui_boss_list = ui_boss_list{}
    @editable
    LangDevice : Lang_select_ui_device = Lang_select_ui_device{}

    var MenuCanvasPerAgent : [agent]canvas = map{}
    var UIElements : [agent][int]tuple(texture_block, button_quiet, button_quiet, button_quiet, button_quiet, button_quiet, button_quiet) = map{}

    StringToMessage<public><localizes>(Text:string):message="{Text}"

    OnBegin<override>()<suspends>:void=
        StrSignal.SecondarySignalEvent.Subscribe(OnMenuSignalEvented)

    OnPlayerSpawned(Agent : agent):void=
        if (MenuCanvas := MenuCanvasPerAgent[Agent]):
            return

        if (Player := player[Agent], PlayerUI := GetPlayerUI[Player]):
            MenuCanvas : canvas = CreateMenuCanvas(Agent)
            MenuCanvas.SetVisibility(widget_visibility.Collapsed)
            if (set MenuCanvasPerAgent[Agent] = MenuCanvas){}

    UpdateUIElement(Agent : agent):void=
        if(Player : player = player[Agent]) :

    OnMenuSignalEvented(Agent : agent):void=
        SFX_Open.Play(Agent)
        OnPlayerSpawned(Agent)
        Print("OnMenuSignalEvented")

        if (Player := player[Agent], PlayerUI := GetPlayerUI[Player], MenuCanvas := MenuCanvasPerAgent[Agent]):
            MenuCanvas.SetVisibility(widget_visibility.Visible)
                
            PlayerUI.AddWidget(MenuCanvas, player_ui_slot{InputMode := ui_input_mode.All})
            #ローカライズ関数を呼び出し
            spawn{UpdateUI(Agent, Player)}

    UpdateUI(Agent : agent, Player : player)<suspends>:void=
        var LanguageIndex : int = SaveDataManager.GetLanguage(Player)
        YokaiText := if(LanguageIndex = 0){"妖怪編成"} else {"Yo-kai"}
        AutoGatherText := if(LanguageIndex = 0){"自動採掘(休憩所)"} else {"Auto Gather"}
        TheaterText := if(LanguageIndex = 0){"中村座"} else {"Theater"}
        BossListText := if(LanguageIndex = 0){"ボスリスト"} else {"Boss List"}
        StrButtonText := if(LanguageIndex = 0){"能力強化"} else {"STR Upgrade"}
        LanguageButtonText := if(LanguageIndex = 0){"言語選択"} else {"Language"}
        if(Element := UIElements[Agent][1]) :
            Element(1).SetText(StringToMessage(YokaiText))
            Element(2).SetText(StringToMessage(AutoGatherText))
            Element(3).SetText(StringToMessage(TheaterText))
            Element(4).SetText(StringToMessage(BossListText))
            Element(5).SetText(StringToMessage(StrButtonText))
            Element(6).SetText(StringToMessage(LanguageButtonText))
            UnlockedBossElement := SaveDataManager.GetDefeatedBosses(Player)
            if :
                IsDefeatedKrepper := UnlockedBossElement[0]
                IsDefeatedKrepper = 1
            then :
                Element(0).SetVisibility(widget_visibility.Visible)
                Element(4).SetVisibility(widget_visibility.Visible)
                Print("Boss_listボタン表示")
            else:
                Element(0).SetVisibility(widget_visibility.Collapsed)
                Element(4).SetVisibility(widget_visibility.Collapsed)
                Print("Boss_Listボタン非表示")

    # ボタン処理
    OnYokaiClicked(WidgetMessage : widget_message):void=
        if(Agent : agent = agent[WidgetMessage.Player]) :
            SFX_LevelUp3.Play(Agent)
        if (PlayerUI := GetPlayerUI[WidgetMessage.Player], MenuCanvas := MenuCanvasPerAgent[WidgetMessage.Player]):
            QuestTrigger.Trigger(WidgetMessage.Player)
            Print("Yokai Menu Quest Clear")
            PlayerUI.RemoveWidget(MenuCanvas)
            YokaiMenu.AddUI(WidgetMessage.Player)
            MenuCanvas.SetVisibility(widget_visibility.Collapsed)
            spawn{AwaitPageClose(WidgetMessage.Player, YokaiMenu.CloseEvent)}
            
    OnAutoGatherClicked(WidgetMessage : widget_message):void=
        if(Agent : agent = agent[WidgetMessage.Player]) :
            SFX_LevelUp3.Play(Agent)
        if (PlayerUI := GetPlayerUI[WidgetMessage.Player], MenuCanvas := MenuCanvasPerAgent[WidgetMessage.Player]):
            PlayerUI.RemoveWidget(MenuCanvas)
            YokaiRest.AddUI(WidgetMessage.Player)
            MenuCanvas.SetVisibility(widget_visibility.Collapsed)
            spawn{AwaitPageClose(WidgetMessage.Player, YokaiRest.CloseEvent)}
            
    OnTheaterClicked(WidgetMessage : widget_message):void=
        if(Agent : agent = agent[WidgetMessage.Player]) :
            SFX_Warp.Play(Agent)
            TP_Theater.Teleport(Agent)
        if (PlayerUI := GetPlayerUI[WidgetMessage.Player], MenuCanvas := MenuCanvasPerAgent[WidgetMessage.Player]):
            PlayerUI.RemoveWidget(MenuCanvas)
            MenuCanvas.SetVisibility(widget_visibility.Collapsed)
            
    OnBossListClicked(WidgetMessage : widget_message):void=
        if(Agent : agent = agent[WidgetMessage.Player]) :
            SFX_LevelUp3.Play(Agent)
        if (PlayerUI := GetPlayerUI[WidgetMessage.Player], MenuCanvas := MenuCanvasPerAgent[WidgetMessage.Player]):
            PlayerUI.RemoveWidget(MenuCanvas)
            BossList.OpenUI(WidgetMessage.Player)
            MenuCanvas.SetVisibility(widget_visibility.Collapsed)
            spawn{AwaitPageClose(WidgetMessage.Player, BossList.CloseEvent)}
            
    OnCloseButtonClicked(WidgetMessage : widget_message):void=
        if(Agent : agent = agent[WidgetMessage.Player]) :
            SFX_Close.Play(Agent)
        if (PlayerUI := GetPlayerUI[WidgetMessage.Player], MenuCanvas := MenuCanvasPerAgent[WidgetMessage.Player]):
            PlayerUI.RemoveWidget(MenuCanvas)
            MenuCanvas.SetVisibility(widget_visibility.Collapsed)

    # str
    OnStrClicked(WidgetMessage : widget_message):void=
        if(Agent : agent = agent[WidgetMessage.Player]) :
            SFX_LevelUp3.Play(Agent)
        if (PlayerUI := GetPlayerUI[WidgetMessage.Player], MenuCanvas := MenuCanvasPerAgent[WidgetMessage.Player]):
            StrDevice.OnStrSignalEvented(WidgetMessage.Player)
            MenuCanvas.SetVisibility(widget_visibility.Collapsed)
            spawn{AwaitPageClose(WidgetMessage.Player, StrDevice.CloseEvent)}

    OnLanguageClicked(WidgetMessage : widget_message):void=
        if (PlayerUI := GetPlayerUI[WidgetMessage.Player], MenuCanvas := MenuCanvasPerAgent[WidgetMessage.Player]):
            PlayerUI.RemoveWidget(MenuCanvas)
            MenuCanvas.SetVisibility(widget_visibility.Collapsed)
        if(Agent : agent = agent[WidgetMessage.Player]) :
            LangDevice.AddSelectLanguageWidget(Agent)

    AwaitPageClose<public>(Player:player, Event:event(player))<suspends>:void=
        loop:
            Target:=Event.Await()
            if(Target = Player):
                break
        OnMenuSignalEvented(Player)

    # キャンバス
    CreateMenuCanvas(Agent : agent):canvas=
        var CanvasSlots : []canvas_slot = array{}
        
        CenterVector : vector2 = vector2{X:=0.5, Y:=0.5}
        CenterAnchors : anchors = anchors{Minimum := CenterVector, Maximum := CenterVector}
        
        # Image Stock
        BGImage : texture_block = texture_block{DefaultImage := Textures.UI_BG, DefaultDesiredSize := vector2{X := 1920.0, Y := 1080.0}}
        YokaiImage : texture_block = texture_block{DefaultImage := icon_set_2.Stats_PetGather, DefaultDesiredSize := vector2{X := 50.0, Y := 50.0}}
        AutoGatherImage : texture_block = texture_block{DefaultImage := icon_set_2.IC_Yokai, DefaultDesiredSize := vector2{X := 50.0, Y := 50.0}}
        TheaterImage : texture_block = texture_block{DefaultImage := icon_set_2.IC_Ticket, DefaultDesiredSize := vector2{X := 50.0, Y := 50.0}}
        StrImage : texture_block = texture_block{DefaultImage := icon_set_2.Stats_Str, DefaultDesiredSize := vector2{X := 50.0, Y := 50.0}}
        LanguageImage : texture_block = texture_block{DefaultImage := icon_set_2.IC_LangSelect, DefaultDesiredSize := vector2{X := 50.0, Y := 50.0}}

        BossImage : texture_block = texture_block{DefaultImage := icon_set_2.IC_Boss, DefaultDesiredSize := vector2{X := 50.0, Y := 50.0}}
        var Yokai : button_quiet = button_quiet{}
        var AutoGather : button_quiet = button_quiet{}
        var Theater : button_quiet = button_quiet{}
        var Boss : button_quiet = button_quiet{}
        var StrButton : button_quiet = button_quiet{}
        var LanguageButton : button_quiet = button_quiet{}


        #========================================== BG ==========================================

        set CanvasSlots += array
        {
            canvas_slot:
                Anchors := CenterAnchors
                Offsets := margin{Top := 0.0,Left := 0.0, Right := 640.0, Bottom := 700.0}
                Alignment := CenterVector
                SizeToContent := false
                Widget := BGImage
        }

        #========================================== Header ==========================================

        MenuTextBlock := text_block{
            DefaultTextColor := NamedColors.White,
            DefaultShadowOffset := option{vector2{X:=2.5, Y:=2.5}},
            DefaultShadowColor := NamedColors.Black,
            DefaultJustification := text_justification.Center
        }
        MenuTextBlock.SetShadowOpacity(1.0)
        set CanvasSlots += array
        {
            canvas_slot:
                Anchors := CenterAnchors
                Offsets := margin{Top:= -320.0, Left:= 0.0, Right := 257.0, Bottom := 41.0}
                Alignment := CenterVector
                SizeToContent := false
                Widget := MenuTextBlock
        }

        #========================================== Main ==========================================

        Yokai.OnClick().Subscribe(OnYokaiClicked)
        AutoGather.OnClick().Subscribe(OnAutoGatherClicked)
        Theater.OnClick().Subscribe(OnTheaterClicked)
        Boss.OnClick().Subscribe(OnBossListClicked)
        StrButton.OnClick().Subscribe(OnStrClicked)
        LanguageButton.OnClick().Subscribe(OnLanguageClicked)

        set CanvasSlots += array
        {
            canvas_slot:
                Anchors := CenterAnchors
                Offsets := margin{Top:= -50.0, Left:= 0.0, Right := 480.0, Bottom := 360.0}
                Alignment := CenterVector
                SizeToContent := false
                Widget := stack_box:
                    Orientation := orientation.Vertical
                    Slots := array:
                        stack_box_slot:
                            Distribution:= option{1.0}
                            HorizontalAlignment := horizontal_alignment.Fill
                            VerticalAlignment := vertical_alignment.Center
                            Widget := Yokai
                        stack_box_slot:
                            Distribution:= option{1.0}
                            HorizontalAlignment := horizontal_alignment.Fill
                            VerticalAlignment := vertical_alignment.Center
                            Widget := AutoGather
                        stack_box_slot:
                            Distribution:= option{1.0}
                            HorizontalAlignment := horizontal_alignment.Fill
                            VerticalAlignment := vertical_alignment.Center
                            Widget := Theater
                        stack_box_slot:
                            Distribution:= option{1.0}
                            HorizontalAlignment := horizontal_alignment.Fill
                            VerticalAlignment := vertical_alignment.Center
                            Widget := Boss
                        stack_box_slot:
                            Distribution:= option{1.0}
                            HorizontalAlignment := horizontal_alignment.Fill
                            VerticalAlignment := vertical_alignment.Center
                            Widget := StrButton
                        stack_box_slot:
                            Distribution:= option{1.0}
                            HorizontalAlignment := horizontal_alignment.Fill
                            VerticalAlignment := vertical_alignment.Center
                            Widget := LanguageButton
        }

        #========================================== Wrapper ==========================================

        set CanvasSlots += array
        {
            canvas_slot:
                Anchors := CenterAnchors
                Offsets := margin{Top:= -50.0, Left:= -189.0, Right := 86.0, Bottom := 360.0}
                Alignment := CenterVector
                SizeToContent := false
                Widget := stack_box:
                    Orientation := orientation.Vertical
                    Slots := array:
                        stack_box_slot:
                            Distribution:= option{1.0}
                            HorizontalAlignment := horizontal_alignment.Center
                            VerticalAlignment := vertical_alignment.Center
                            Widget := YokaiImage
                        stack_box_slot:
                            Distribution:= option{1.0}
                            HorizontalAlignment := horizontal_alignment.Center
                            VerticalAlignment := vertical_alignment.Center
                            Widget := AutoGatherImage
                        stack_box_slot:
                            Distribution:= option{1.0}
                            HorizontalAlignment := horizontal_alignment.Center
                            VerticalAlignment := vertical_alignment.Center
                            Widget := TheaterImage
                        stack_box_slot:
                            Distribution:= option{1.0}
                            HorizontalAlignment := horizontal_alignment.Center
                            VerticalAlignment := vertical_alignment.Center
                            Widget := BossImage
                        stack_box_slot:
                            Distribution:= option{1.0}
                            HorizontalAlignment := horizontal_alignment.Center
                            VerticalAlignment := vertical_alignment.Center
                            Widget := StrImage
                        stack_box_slot:
                            Distribution:= option{1.0}
                            HorizontalAlignment := horizontal_alignment.Center
                            VerticalAlignment := vertical_alignment.Center
                            Widget := LanguageImage
        }

        #========================================== Close ==========================================

        CloseButton : button_loud = button_loud{DefaultText := StringToMessage("CLOSE")}
        CloseButton.OnClick().Subscribe(OnCloseButtonClicked)
        set CanvasSlots += array
        {
            canvas_slot:
                Anchors := CenterAnchors
                Offsets := margin{Top:= 447.0, Left:= 0.0, Right := 170.0, Bottom := 80.0}
                Alignment := CenterVector
                SizeToContent := false
                Widget := CloseButton
        }
        
        if(set UIElements[Agent] = map{}){}
        if:
            set UIElements[Agent] = map:
                1 => (BossImage, Yokai, AutoGather, Theater, Boss, StrButton, LanguageButton)
            
        Canvas : canvas = canvas:
            Slots := CanvasSlots

        return Canvas
