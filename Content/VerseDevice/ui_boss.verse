
using { /Verse.org/Assets }
using { /Verse.org/Simulation }
using { /Fortnite.com/UI }
using { /Fortnite.com/Devices }
using { /Fortnite.com/Characters }
using { /UnrealEngine.com/Temporary/UI }
using { /UnrealEngine.com/Temporary/SpatialMath }
using { /Verse.org/Simulation }
using { /UnrealEngine.com/Temporary/Diagnostics }
using { /Verse.org/Colors }
using { /Verse.org/Colors/NamedColors }

using {Mining_System.Code.PlayerData}
using {Mining_System.Code.BossFlow}


# See https://dev.epicgames.com/documentation/en-us/uefn/create-your-own-device-in-verse for how to create a verse device.

# A Verse-authored creative device that can be placed in a level

AgentToMessage<localizes>(Agent:agent):message="{Agent}"

ui_boss<public>:= class(creative_device):
    #SFX
    @editable
    SFX_Open : audio_player_device = audio_player_device{}
    @editable
    SFX_Open2 : audio_player_device = audio_player_device{}
    @editable
    SFX_Close : audio_player_device = audio_player_device{}
    @editable
    SFX_Alart : audio_player_device = audio_player_device{}
    @editable
    SFX_Anounce : audio_player_device = audio_player_device{}

    # @editable
    # PlayerSpawners : []player_spawner_device = array{}

    @editable
    HUD_Message : hud_message_device = hud_message_device{}

    @editable
    ShowAll_HUD_Message : hud_message_device = hud_message_device{}

    @editable
    TimerDevice : timer_device = timer_device{}

    @editable
    SelectJoinEvent : trigger_device = trigger_device{}

    @editable
    SelectLeaveEvent : trigger_device = trigger_device{}

    @editable
    SelectHelpEvent : trigger_device = trigger_device{}

    @editable
    BossDevice : boss_device_base = boss_device_base{}

    var BossCanvasPerAgent : [agent]canvas = map{}
    
    @editable
    BossName : string = "BossName"

    @editable
    var BossHp : editable_big_number = editable_big_number{}

    @editable
    var BossTime : string = ""

    @editable
    var PowerUpStone : editable_big_number = editable_big_number{}

    @editable
    var Stone : editable_big_number = editable_big_number{}

    @editable
    var SummonStone : editable_big_number = editable_big_number{}

    @editable
    SaveDataMamager : save_data_manager = save_data_manager{}

    @editable
    Kureppa_Class : Boss01_Kurepper_Control = Boss01_Kurepper_Control{}

    @editable
    Karakasa_Class : Boss02_Karakasa_Control = Boss02_Karakasa_Control{}

    @editable
    Enengama_Class : Boss03_Enengama_Control = Boss03_Enengama_Control{}

    @editable
    Oobanmomen_Class : Boss04_Oobanmomen_Control = Boss04_Oobanmomen_Control{}

    @editable
    Tachigarasu_Class : Boss05_Tachigarasu_Control = Boss05_Tachigarasu_Control{}

    @editable
    Yatagaran_Class : Boss06_Yatagaran_Control = Boss06_Yatagaran_Control{}

    @editable
    BossList : ui_boss_list = ui_boss_list{}

    @editable
    MiningDamageDevice : Mining_System.Code.MiningObject.mining_damage_device = Mining_System.Code.MiningObject.mining_damage_device{}

    
    StringToMessage<public><localizes>(Text:string):message="{Text}"
    var JoinedPlayers : []player = array{}
    var DisplayedPlayers : []player = array{}
    var CurrentTime : float = 0.0
    var TimerWidget : [agent]text_block = map{}
    var LevelAndNames : [agent][]tuple(text_block, text_block) = map{}

    OnBegin<override>()<suspends>:void=
        Print("")

    ConvertFloatToTime(Time:float):string=
        var Value : string = ""
        if:
            AllTimeSecond : int = Round[Time]
            Minite := AllTimeSecond / 60
            Minites := Floor(Minite)
            remainingSeconds := AllTimeSecond - (Minites * 60)
        then:
            if(remainingSeconds < 10):
                set Value = "{Minites}:0{remainingSeconds}"
            else:
                set Value = "{Minites}:{remainingSeconds}"

        Value

    OnPlayerSpawned(Agent : agent):void=
        if (BossCanvas := BossCanvasPerAgent[Agent]):
            return

        if (Player := player[Agent], PlayerUI := GetPlayerUI[Player]):
            # set DisplayedPlayers += array{Player}
            BossCanvas : canvas = CreateBossCanvas(Agent)
            BossCanvas.SetVisibility(widget_visibility.Collapsed)
            if (set BossCanvasPerAgent[Agent] = BossCanvas){}

    OnBossSignalEvented<public>(Agent : agent):void=
        SFX_Open.Play(Agent)
        OnPlayerSpawned(Agent)
        Print("OnBossSignalEvented")
        if (Player := player[Agent]):
            # BossCanvasPerAgentから該当プレイヤーのキャンバスを取得
            if (BossCanvas := BossCanvasPerAgent[Agent]):
                # UIを表示する
                if (PlayerUI := GetPlayerUI[Player]):
                    BossCanvas.SetVisibility(widget_visibility.Visible)
                    PlayerUI.AddWidget(BossCanvas, player_ui_slot{InputMode := ui_input_mode.All})
                    Print("Displaying Boss UI for Player")
            else:
                Print("No BossCanvas found for Agent")

    BossBreakPoint()<suspends>:void=
        var NowBoss : int = 0
        loop:
            Sleep(1.0)
            var State : []char = "None"

            if(BossName = "Krepper"):
                set State = Kureppa_Class.ReturnBossState()
            if(BossName = "Tachigarasu"):
                set State = Tachigarasu_Class.ReturnBossState()
            if(BossName = "Obanmomen"):
                set State = Enengama_Class.ReturnBossState()
            if(BossName = "Karakasaobake"):
                set State = Karakasa_Class.ReturnBossState()
            if(BossName = "Enengama"):
                set State = Enengama_Class.ReturnBossState()
            if(BossName = "Yatagaran"):
                set State = Yatagaran_Class.ReturnBossState()

            if(State <> "NotActive"):
                Print("Start Boss", ?Duration := 5.0)
                var UniquePlayers : []player = array{}

                for(Player : JoinedPlayers):
                    if(not UniquePlayers.Find[Player]):
                        set UniquePlayers += array{Player}

                        set NowBoss = SaveDataMamager.GetNowTicket(Player)
                        if(NowBoss > 0):
                            SaveDataMamager.SubTicket(Player)
                        
                        if(PlayerUI := GetPlayerUI[Player], BossCanvas := BossCanvasPerAgent[Player]):
                            PlayerUI.RemoveWidget(BossCanvas)
                
                BossBreakPoint2()
                break
    
    BossBreakPoint2()<suspends>:void=
        loop:
            Sleep(1.0)
            var State : []char = "None"

            if(BossName = "Krepper"):
                set State = Kureppa_Class.ReturnBossState()
            if(BossName = "Tachigarasu"):
                set State = Tachigarasu_Class.ReturnBossState()
            if(BossName = "Obanmomen"):
                set State = Enengama_Class.ReturnBossState()
            if(BossName = "Karakasaobake"):
                set State = Karakasa_Class.ReturnBossState()
            if(BossName = "Enengama"):
                set State = Enengama_Class.ReturnBossState()
            if(BossName = "Yatagaran"):
                set State = Yatagaran_Class.ReturnBossState()

            if(State = "NotActive"):
                Print("End Boss", ?Duration := 5.0)
                set JoinedPlayers = array{}
                for(Key -> Texts : LevelAndNames):
                    for(Index := 0..3, Text := Texts[Index]):
                        if(TargetPlayer := JoinedPlayers[Index]):
                            Text(0).SetText(AgentToMessage(TargetPlayer))
                            Text(1).SetText(StringToMessage("{SaveDataMamager.GetLevel(TargetPlayer)}"))
                        else:
                            Text(0).SetText(StringToMessage(""))
                            Text(1).SetText(StringToMessage(""))
                break


    DisplayPlayer<public>(WidgetMessage : widget_message)<suspends>:void=
        race:
            BossBreakPoint()
            loop:
                Sleep(1.0)

                for(Key -> Texts : LevelAndNames):
                    for(Index := 0..3, Text := Texts[Index]):
                        if(TargetPlayer := JoinedPlayers[Index]):
                            Text(0).SetText(AgentToMessage(TargetPlayer))
                            Text(1).SetText(StringToMessage("{SaveDataMamager.GetLevel(TargetPlayer)}"))
                        else:
                            Text(0).SetText(StringToMessage(""))
                            Text(1).SetText(StringToMessage(""))


    # ボタン処理
    OnHelpButtonClicked(WidgetMessage : widget_message):void=
        SFX_Anounce.Play()
        for(Player : GetPlayspace().GetPlayers()):
            LanguageIndex := SaveDataMamager.GetLanguage(Player)

            if(LanguageIndex = 0):
                ShowAll_HUD_Message.SetText(StringToMessage("{BossName}のボス戦のリクエストが来ました！"))
            if(LanguageIndex = 1):
                ShowAll_HUD_Message.SetText(StringToMessage("A request for the {BossName} Boss Battle has arrived!"))
            ShowAll_HUD_Message.Show()
        if (PlayerUI := GetPlayerUI[WidgetMessage.Player], BossCanvas := BossCanvasPerAgent[WidgetMessage.Player]):
            SelectHelpEvent.Trigger(WidgetMessage.Player)
            
    OnJoinButtonClicked(WidgetMessage : widget_message):void=
        if(Agent : agent = agent[WidgetMessage.Player]) :
            SFX_Open2.Play(Agent)
        var State : []char = "None"

        if(BossName = "Krepper"):
            set State = Kureppa_Class.ReturnBossState()
        if(BossName = "Tachigarasu"):
            set State = Tachigarasu_Class.ReturnBossState()
        if(BossName = "Obanmomen"):
            set State = Enengama_Class.ReturnBossState()
        if(BossName = "Karakasaobake"):
            set State = Karakasa_Class.ReturnBossState()
        if(BossName = "Enengama"):
            set State = Enengama_Class.ReturnBossState()
        if(BossName = "Yatagaran"):
            set State = Yatagaran_Class.ReturnBossState()

        Print("State : {State} ", ?Duration := 5.0)

        if (PlayerUI := GetPlayerUI[WidgetMessage.Player], BossCanvas := BossCanvasPerAgent[WidgetMessage.Player], State = "NotActive"):
            
            # Ticketが1以上の時参加できる
            GetTickets := SaveDataMamager.GetNowTicket(WidgetMessage.Player)
            if(GetTickets > 0):
                if(not JoinedPlayers.Find[WidgetMessage.Player], JoinedPlayers.Length < 4):
                    SelectJoinEvent.Trigger(WidgetMessage.Player)
                    set JoinedPlayers += array{WidgetMessage.Player}

                    # ボス戦参加のコード
                    spawn{StartBoss()}
                    spawn{DisplayPlayer(WidgetMessage)}
            else:
                LanguageIndex := SaveDataMamager.GetLanguage(WidgetMessage.Player)
                if(LanguageIndex = 0):
                    HUD_Message.SetText(StringToMessage("ボスチケットが足りません..."))
                if(LanguageIndex = 1):
                    HUD_Message.SetText(StringToMessage("Not enough boss tickets..."))
                HUD_Message.Show(WidgetMessage.Player)
        
        else:
            LanguageIndex := SaveDataMamager.GetLanguage(WidgetMessage.Player)
            if(LanguageIndex = 0):
                HUD_Message.SetText(StringToMessage("現在ボス戦中なので参加できません..."))
            if(LanguageIndex = 1):
                HUD_Message.SetText(StringToMessage("The boss battle is currently in progress, so you can't join..."))
            HUD_Message.Show(WidgetMessage.Player)

    StartBoss()<suspends>:void=
        Print("Press Join Button!!", ?Duration := 5.0)
        TimerDevice.Start()
        TimerDevice.SuccessEvent.Await()
        for(Player : JoinedPlayers):
            BossDevice.ChangeTeamFunc(Player)
            BossDevice.StartBossFight(option{Player})

    OnLeaveButtonClicked(WidgetMessage : widget_message):void=
        if(Agent : agent = agent[WidgetMessage.Player]) :
            SFX_Alart.Play(Agent)
        if (PlayerUI := GetPlayerUI[WidgetMessage.Player], BossCanvas := BossCanvasPerAgent[WidgetMessage.Player]):
            Print("Leave", ?Duration := 5.0)
            SelectLeaveEvent.Trigger(WidgetMessage.Player)
            set JoinedPlayers = JoinedPlayers.RemoveAllElements(WidgetMessage.Player)
            
            
    OnCloseButtonClicked(WidgetMessage : widget_message):void=
        if(Agent : agent = agent[WidgetMessage.Player]) :
            SFX_Close.Play(Agent)
        if (PlayerUI := GetPlayerUI[WidgetMessage.Player], BossCanvas := BossCanvasPerAgent[WidgetMessage.Player]):
            PlayerUI.RemoveWidget(BossCanvas)
            BossCanvas.SetVisibility(widget_visibility.Collapsed)
            set DisplayedPlayers = DisplayedPlayers.RemoveAllElements(WidgetMessage.Player)

    # キャンバス
    CreateBossCanvas(AgenTextst : agent):canvas=
        var CanvasSlots : []canvas_slot = array{}
        
        var RewardResource : big_number = big_number{Value := 0.0, Exponent := 0}
        
        CenterVector : vector2 = vector2{X:=0.5, Y:=0.5}
        CenterAnchors : anchors = anchors{Minimum := CenterVector, Maximum := CenterVector}

        if: 
            Player := player[AgenTextst]
        then:
            GetDamage :=MiningDamageDevice.GetNormalDamageAmount(Player).Amount
            RewardResources := Mining_System.Code.PlayerData.BigNumber_MultiplyByBigNumber(GetDamage, Stone.ToPersistableBigNumber())
            set RewardResource = RewardResources
        
        # Image Stock
        BGImage : texture_block = texture_block{DefaultImage := Textures.UI_BG, DefaultDesiredSize := vector2{X := 1920.0, Y := 1080.0}}
        BossImage : texture_block = texture_block{DefaultImage := BossDevice.BossImage, DefaultDesiredSize := vector2{X := 300.0, Y := 300.0}}
        Slot1Image : texture_block = texture_block{DefaultImage := Boss_FirstPet_Icon_2.null, DefaultDesiredSize := vector2{X := 65.0, Y := 65.0}}
        Slot2Image : texture_block = texture_block{DefaultImage := Boss_FirstPet_Icon_2.null, DefaultDesiredSize := vector2{X := 65.0, Y := 65.0}}
        Slot3Image : texture_block = texture_block{DefaultImage := Boss_FirstPet_Icon_2.null, DefaultDesiredSize := vector2{X := 65.0, Y := 65.0}}
        Slot4Image : texture_block = texture_block{DefaultImage := Boss_FirstPet_Icon_2.null, DefaultDesiredSize := vector2{X := 65.0, Y := 65.0}}
        HPImage : texture_block = texture_block{DefaultImage := icon_set_2.IC_HP, DefaultDesiredSize := vector2{X := 30.0, Y := 30.0}}
        TimerImage : texture_block = texture_block{DefaultImage := icon_set_2.IC_Timer, DefaultDesiredSize := vector2{X := 30.0, Y := 30.0}}
        PowerImage : texture_block = texture_block{DefaultImage := icon_set_2.IC_Power, DefaultDesiredSize := vector2{X := 40.0, Y := 40.0}}
        StoneImage : texture_block = texture_block{DefaultImage := icon_set_2.SilverOre, DefaultDesiredSize := vector2{X := 40.0, Y := 40.0}}
        SummonStoneImage : texture_block = texture_block{DefaultImage := icon_set_2.SS_1, DefaultDesiredSize := vector2{X := 40.0, Y := 40.0}}
        HelpIconImage : texture_block = texture_block{DefaultImage := icon_set_2.IC_Help, DefaultDesiredSize := vector2{X := 60.0, Y := 60.0}}
        BossIconImage : texture_block = texture_block{DefaultImage := icon_set_2.IC_Boss, DefaultDesiredSize := vector2{X := 60.0, Y := 60.0}}
        LeaveIconImage : texture_block = texture_block{DefaultImage := icon_set_2.IC_Exit, DefaultDesiredSize := vector2{X := 60.0, Y := 60.0}}

        #========================================== BG ==========================================

        set CanvasSlots += array
        {
            canvas_slot:
                Anchors := CenterAnchors
                Offsets := margin{Top := 0.0,Left := 0.0, Right := 1920.0, Bottom := 1080.0}
                Alignment := CenterVector
                SizeToContent := false
                Widget := BGImage
        }

        #========================================== Header ==========================================

        BossBattleTextBlock := text_block{
            DefaultText := StringToMessage("Boss Battle"),
            DefaultTextColor := NamedColors.White,
            DefaultShadowOffset := option{vector2{X:=2.5, Y:=2.5}},
            DefaultShadowColor := NamedColors.Black
        }
        BossBattleTextBlock.SetShadowOpacity(1.0)
        set CanvasSlots += array
        {
            canvas_slot:
                Anchors := CenterAnchors
                Offsets := margin{Top:= -490.0, Left:= 0.0, Right := 257.0, Bottom := 41.0}
                Alignment := CenterVector
                SizeToContent := false
                Widget := BossBattleTextBlock
        }

        #========================================== Main ==========================================

        Player1_LvTextBlock := text_block{
            DefaultText := if(TargetPlayer := JoinedPlayers[0], Level := SaveDataMamager.GetLevel(TargetPlayer)){StringToMessage("{Level}")} else {StringToMessage("")},
            DefaultTextColor := NamedColors.White,
            DefaultShadowOffset := option{vector2{X:=2.5, Y:=2.5}},
            DefaultShadowColor := NamedColors.Black
        }
        Player1_LvTextBlock.SetShadowOpacity(1.0)
        Player2_LvTextBlock := text_block{
            DefaultText := if(TargetPlayer := JoinedPlayers[1], Level := SaveDataMamager.GetLevel(TargetPlayer)){StringToMessage("{Level}")} else {StringToMessage("")},
            DefaultTextColor := NamedColors.White,
            DefaultShadowOffset := option{vector2{X:=2.5, Y:=2.5}},
            DefaultShadowColor := NamedColors.Black
        }
        Player2_LvTextBlock.SetShadowOpacity(1.0)
        Player3_LvTextBlock := text_block{
            DefaultText := if(TargetPlayer := JoinedPlayers[2], Level := SaveDataMamager.GetLevel(TargetPlayer)){StringToMessage("{Level}")} else {StringToMessage("")},
            DefaultTextColor := NamedColors.White,
            DefaultShadowOffset := option{vector2{X:=2.5, Y:=2.5}},
            DefaultShadowColor := NamedColors.Black
        }
        Player3_LvTextBlock.SetShadowOpacity(1.0)
        Player4_LvTextBlock := text_block{
            DefaultText := if(TargetPlayer := JoinedPlayers[3], Level := SaveDataMamager.GetLevel(TargetPlayer)){StringToMessage("{Level}")} else {StringToMessage("")},
            DefaultTextColor := NamedColors.White,
            DefaultShadowOffset := option{vector2{X:=2.5, Y:=2.5}},
            DefaultShadowColor := NamedColors.Black
        }
        Player4_LvTextBlock.SetShadowOpacity(1.0)
        Player1_NameTextBlock := text_block{
            DefaultText := if(TargetPlayer := JoinedPlayers[0]){AgentToMessage(TargetPlayer)} else {StringToMessage("")},
            DefaultTextColor := NamedColors.White,
            DefaultShadowOffset := option{vector2{X:=2.5, Y:=2.5}},
            DefaultShadowColor := NamedColors.Black
        }
        Player1_NameTextBlock.SetShadowOpacity(1.0)
        Player2_NameTextBlock := text_block{
            DefaultText := if(TargetPlayer := JoinedPlayers[1]){AgentToMessage(TargetPlayer)} else {StringToMessage("")},
            DefaultTextColor := NamedColors.White,
            DefaultShadowOffset := option{vector2{X:=2.5, Y:=2.5}},
            DefaultShadowColor := NamedColors.Black
        }
        Player2_NameTextBlock.SetShadowOpacity(1.0)
        Player3_NameTextBlock := text_block{
            DefaultText := if(TargetPlayer := JoinedPlayers[2]){AgentToMessage(TargetPlayer)} else {StringToMessage("")},
            DefaultTextColor := NamedColors.White,
            DefaultShadowOffset := option{vector2{X:=2.5, Y:=2.5}},
            DefaultShadowColor := NamedColors.Black
        }
        Player3_NameTextBlock.SetShadowOpacity(1.0)
        Player4_NameTextBlock := text_block{
            DefaultText := if(TargetPlayer := JoinedPlayers[3]){AgentToMessage(TargetPlayer)} else {StringToMessage("")},
            DefaultTextColor := NamedColors.White,
            DefaultShadowOffset := option{vector2{X:=2.5, Y:=2.5}},
            DefaultShadowColor := NamedColors.Black
        }
        Player4_NameTextBlock.SetShadowOpacity(1.0)

        if(set LevelAndNames[AgenTextst] = array{(Player1_NameTextBlock, Player1_LvTextBlock), (Player2_NameTextBlock, Player2_LvTextBlock), (Player3_NameTextBlock, Player3_LvTextBlock), (Player4_NameTextBlock, Player4_LvTextBlock)}){}

        set CanvasSlots += array
        {
            canvas_slot:
                Anchors := CenterAnchors
                Offsets := margin{Left:=-348.0 ,Top:= -29.0, Right:= 696.0, Bottom:= 703.0}
                Alignment := CenterVector
                SizeToContent := false
                Widget := stack_box:
                    Orientation := orientation.Vertical
                    Slots := array:
                        stack_box_slot:
                            Distribution:= option{0.5}
                            HorizontalAlignment := horizontal_alignment.Fill
                            VerticalAlignment := vertical_alignment.Fill
                            Widget := stack_box:
                                Orientation := orientation.Horizontal
                                Slots := array:
                                    stack_box_slot:
                                        Distribution:= option{1.0}
                                        HorizontalAlignment := horizontal_alignment.Center
                                        VerticalAlignment := vertical_alignment.Center
                                        Widget := BossImage
                        stack_box_slot:
                            Padding:=margin{Top:= 10.0}
                            Distribution:= option{0.5}
                            HorizontalAlignment := horizontal_alignment.Fill
                            VerticalAlignment := vertical_alignment.Fill
                            Widget := stack_box:
                                Orientation := orientation.Vertical
                                Slots := array:
                                    stack_box_slot:
                                        Distribution:= option{0.2}
                                        HorizontalAlignment := horizontal_alignment.Fill
                                        VerticalAlignment := vertical_alignment.Fill
                                        Widget := stack_box:
                                            Orientation := orientation.Horizontal
                                            Slots := array:
                                                stack_box_slot:
                                                    Padding:=margin{Left:= 10.0}
                                                    HorizontalAlignment := horizontal_alignment.Center
                                                    VerticalAlignment := vertical_alignment.Center
                                                    Widget := Slot1Image
                                                stack_box_slot:
                                                    Distribution:= option{0.2}
                                                    HorizontalAlignment := horizontal_alignment.Center
                                                    VerticalAlignment := vertical_alignment.Center
                                                    Widget := Player1_LvTextBlock
                                                stack_box_slot:
                                                    Distribution:= option{0.7}
                                                    HorizontalAlignment := horizontal_alignment.Left
                                                    VerticalAlignment := vertical_alignment.Center
                                                    Widget := Player1_NameTextBlock
                                    stack_box_slot:
                                        Distribution:= option{0.2}
                                        HorizontalAlignment := horizontal_alignment.Fill
                                        VerticalAlignment := vertical_alignment.Fill
                                        Widget := stack_box:
                                            Orientation := orientation.Horizontal
                                            Slots := array:
                                                stack_box_slot:
                                                    Padding:=margin{Left:= 10.0}
                                                    HorizontalAlignment := horizontal_alignment.Center
                                                    VerticalAlignment := vertical_alignment.Center
                                                    Widget := Slot2Image
                                                stack_box_slot:
                                                    Distribution:= option{0.2}
                                                    HorizontalAlignment := horizontal_alignment.Center
                                                    VerticalAlignment := vertical_alignment.Center
                                                    Widget := Player2_LvTextBlock
                                                stack_box_slot:
                                                    Distribution:= option{0.7}
                                                    HorizontalAlignment := horizontal_alignment.Left
                                                    VerticalAlignment := vertical_alignment.Center
                                                    Widget := Player2_NameTextBlock
                                    stack_box_slot:
                                        Distribution:= option{0.2}
                                        HorizontalAlignment := horizontal_alignment.Fill
                                        VerticalAlignment := vertical_alignment.Fill
                                        Widget := stack_box:
                                            Orientation := orientation.Horizontal
                                            Slots := array:
                                                stack_box_slot:
                                                    Padding:=margin{Left:= 10.0}
                                                    HorizontalAlignment := horizontal_alignment.Center
                                                    VerticalAlignment := vertical_alignment.Center
                                                    Widget := Slot3Image
                                                stack_box_slot:
                                                    Distribution:= option{0.2}
                                                    HorizontalAlignment := horizontal_alignment.Center
                                                    VerticalAlignment := vertical_alignment.Center
                                                    Widget := Player3_LvTextBlock
                                                stack_box_slot:
                                                    Distribution:= option{0.7}
                                                    HorizontalAlignment := horizontal_alignment.Left
                                                    VerticalAlignment := vertical_alignment.Center
                                                    Widget := Player3_NameTextBlock
                                    stack_box_slot:
                                        Distribution:= option{0.2}
                                        HorizontalAlignment := horizontal_alignment.Fill
                                        VerticalAlignment := vertical_alignment.Fill
                                        Widget := stack_box:
                                            Orientation := orientation.Horizontal
                                            Slots := array:
                                                stack_box_slot:
                                                    Padding:=margin{Left:= 10.0}
                                                    HorizontalAlignment := horizontal_alignment.Center
                                                    VerticalAlignment := vertical_alignment.Center
                                                    Widget := Slot4Image
                                                stack_box_slot:
                                                    Distribution:= option{0.2}
                                                    HorizontalAlignment := horizontal_alignment.Center
                                                    VerticalAlignment := vertical_alignment.Center
                                                    Widget := Player4_LvTextBlock
                                                stack_box_slot:
                                                    Distribution:= option{0.7}
                                                    HorizontalAlignment := horizontal_alignment.Left
                                                    VerticalAlignment := vertical_alignment.Center
                                                    Widget := Player4_NameTextBlock
        }


        #========================================== Boss Stat ==========================================


        BossNameTextBlock := text_block{
            DefaultText := StringToMessage("{BossName}"),
            DefaultTextColor := NamedColors.White,
            DefaultShadowOffset := option{vector2{X:=2.5, Y:=2.5}},
            DefaultShadowColor := NamedColors.Black
        }
        BossNameTextBlock.SetShadowOpacity(1.0)
        HealthTextBlock := text_block{
            DefaultText := StringToMessage("Health:"),
            DefaultTextColor := NamedColors.White,
            DefaultShadowOffset := option{vector2{X:=2.5, Y:=2.5}},
            DefaultShadowColor := NamedColors.Black
        }
        HealthTextBlock.SetShadowOpacity(1.0)
        TimeLimitTextBlock := text_block{
            DefaultText := StringToMessage("Time Limit:"),
            DefaultTextColor := NamedColors.White,
            DefaultShadowOffset := option{vector2{X:=2.5, Y:=2.5}},
            DefaultShadowColor := NamedColors.Black
        }
        TimeLimitTextBlock.SetShadowOpacity(1.0)
        BossHpTextBlock := text_block{
            DefaultText := StringToMessage("{BigNumber_FormatDisplay(BossHp.ToPersistableBigNumber())}"),
            DefaultTextColor := NamedColors.White,
            DefaultShadowOffset := option{vector2{X:=2.5, Y:=2.5}},
            DefaultShadowColor := NamedColors.Black
        }
        BossHpTextBlock.SetShadowOpacity(1.0)
        BossTimeTextBlock := text_block{
            DefaultText := StringToMessage("{BossTime}"),
            DefaultTextColor := NamedColors.White,
            DefaultShadowOffset := option{vector2{X:=2.5, Y:=2.5}},
            DefaultShadowColor := NamedColors.Black
        }
        BossTimeTextBlock.SetShadowOpacity(1.0)
        if(set TimerWidget[AgenTextst] = BossTimeTextBlock){}

        set CanvasSlots += array
        {
            canvas_slot:
                Anchors := CenterAnchors
                Offsets := margin{Top:= -373.5, Left:= 389.0, Right := 100.0, Bottom := 30.0}
                Alignment := CenterVector
                SizeToContent := false
                Widget := BossNameTextBlock
        }
        set CanvasSlots += array
        {
            canvas_slot:
                Anchors := CenterAnchors
                Offsets := margin{Top:= -282.0, Left:= 385.0, Right := 607.0, Bottom := 75.0}
                Alignment := CenterVector
                SizeToContent := false
                Widget := stack_box:
                    Orientation := orientation.Vertical
                    Slots := array:
                        stack_box_slot:
                            HorizontalAlignment := horizontal_alignment.Fill
                            VerticalAlignment := vertical_alignment.Fill
                            Widget := stack_box:
                                Orientation := orientation.Horizontal
                                Slots := array:
                                    stack_box_slot:
                                        HorizontalAlignment := horizontal_alignment.Center
                                        VerticalAlignment := vertical_alignment.Center
                                        Widget := HPImage
                                    stack_box_slot:
                                        Distribution:= option{1.0}
                                        Padding:=margin{Left:=16.0}
                                        HorizontalAlignment := horizontal_alignment.Left
                                        VerticalAlignment := vertical_alignment.Center
                                        Widget := HealthTextBlock
                                    stack_box_slot:
                                        Distribution:= option{1.0}
                                        Padding:=margin{Left:=16.0}
                                        HorizontalAlignment := horizontal_alignment.Right
                                        VerticalAlignment := vertical_alignment.Center
                                        Widget := BossHpTextBlock
                        stack_box_slot:
                            HorizontalAlignment := horizontal_alignment.Fill
                            VerticalAlignment := vertical_alignment.Fill
                            Widget := stack_box:
                                Orientation := orientation.Horizontal
                                Slots := array:
                                    stack_box_slot:
                                        HorizontalAlignment := horizontal_alignment.Center
                                        VerticalAlignment := vertical_alignment.Center
                                        Widget := TimerImage
                                    stack_box_slot:
                                        Distribution:= option{1.0}
                                        Padding:=margin{Left:=16.0}
                                        HorizontalAlignment := horizontal_alignment.Left
                                        VerticalAlignment := vertical_alignment.Center
                                        Widget := TimeLimitTextBlock
                                    stack_box_slot:
                                        Distribution:= option{1.0}
                                        Padding:=margin{Left:=16.0}
                                        HorizontalAlignment := horizontal_alignment.Right
                                        VerticalAlignment := vertical_alignment.Center
                                        Widget := BossTimeTextBlock
        }



        #========================================== Rewards ==========================================

        

        RewardsTextBlock := text_block{
            DefaultText := StringToMessage("Rewards"),
            DefaultTextColor := NamedColors.White,
            DefaultShadowOffset := option{vector2{X:=2.5, Y:=2.5}},
            DefaultShadowColor := NamedColors.Black
        }
        PowerTextBlock := text_block{
            DefaultText := StringToMessage("{BigNumber_FormatDisplay(PowerUpStone.ToPersistableBigNumber())}"),
            DefaultTextColor := NamedColors.White,
            DefaultShadowOffset := option{vector2{X:=2.5, Y:=2.5}},
            DefaultShadowColor := NamedColors.Black
        }
        PowerTextBlock.SetShadowOpacity(1.0)
        StoneTextBlock := text_block{
            DefaultText := StringToMessage("{BigNumber_FormatDisplay(RewardResource)}"),
            DefaultTextColor := NamedColors.White,
            DefaultShadowOffset := option{vector2{X:=2.5, Y:=2.5}},
            DefaultShadowColor := NamedColors.Black
        }
        StoneTextBlock.SetShadowOpacity(1.0)
        SummonStoneTextBlock := text_block{
            DefaultText := StringToMessage("{BigNumber_FormatDisplay(SummonStone.ToPersistableBigNumber())}"),
            DefaultTextColor := NamedColors.White,
            DefaultShadowOffset := option{vector2{X:=2.5, Y:=2.5}},
            DefaultShadowColor := NamedColors.Black
        }
        SummonStoneTextBlock.SetShadowOpacity(1.0)
        set CanvasSlots += array
        {
            canvas_slot:
                Anchors := CenterAnchors
                Offsets := margin{Top:= -197.5, Left:= 389.0, Right := 100.0, Bottom := 30.0}
                Alignment := CenterVector
                SizeToContent := false
                Widget := RewardsTextBlock
        }
        set CanvasSlots += array
        {
            canvas_slot:
                Anchors := CenterAnchors
                Offsets := margin{Top:= -74.0, Left:= 385.0, Right := 607.0, Bottom := 141.0}
                Alignment := CenterVector
                SizeToContent := false
                Widget := stack_box:
                    Orientation := orientation.Vertical
                    Slots := array:
                        stack_box_slot:
                            Distribution:= option{0.2}
                            HorizontalAlignment := horizontal_alignment.Fill
                            VerticalAlignment := vertical_alignment.Fill
                            Widget := stack_box:
                                Orientation := orientation.Horizontal
                                Slots := array:
                                    stack_box_slot:
                                        HorizontalAlignment := horizontal_alignment.Center
                                        VerticalAlignment := vertical_alignment.Center
                                        Widget := PowerImage
                                    stack_box_slot:
                                        Distribution:= option{1.0}
                                        Padding:=margin{Left:=16.0}
                                        HorizontalAlignment := horizontal_alignment.Left
                                        VerticalAlignment := vertical_alignment.Center
                                        Widget := PowerTextBlock
                        stack_box_slot:
                            Distribution:= option{0.2}
                            HorizontalAlignment := horizontal_alignment.Fill
                            VerticalAlignment := vertical_alignment.Fill
                            Widget := stack_box:
                                Orientation := orientation.Horizontal
                                Slots := array:
                                    stack_box_slot:
                                        HorizontalAlignment := horizontal_alignment.Center
                                        VerticalAlignment := vertical_alignment.Center
                                        Widget := StoneImage
                                    stack_box_slot:
                                        Distribution:= option{1.0}
                                        Padding:=margin{Left:=16.0}
                                        HorizontalAlignment := horizontal_alignment.Left
                                        VerticalAlignment := vertical_alignment.Center
                                        Widget := StoneTextBlock
                        stack_box_slot:
                            Distribution:= option{0.2}
                            HorizontalAlignment := horizontal_alignment.Fill
                            VerticalAlignment := vertical_alignment.Fill
                            Widget := stack_box:
                                Orientation := orientation.Horizontal
                                Slots := array:
                                    stack_box_slot:
                                        HorizontalAlignment := horizontal_alignment.Center
                                        VerticalAlignment := vertical_alignment.Center
                                        Widget := SummonStoneImage
                                    stack_box_slot:
                                        Distribution:= option{1.0}
                                        Padding:=margin{Left:=16.0}
                                        HorizontalAlignment := horizontal_alignment.Left
                                        VerticalAlignment := vertical_alignment.Center
                                        Widget := SummonStoneTextBlock
        }

        #========================================== Button ==========================================

        
        HelpButton : button_quiet = button_quiet{DefaultText := StringToMessage("Request Help")}
        HelpButton.OnClick().Subscribe(OnHelpButtonClicked)
        JoinButton : button_quiet = button_quiet{DefaultText := StringToMessage("Join")}
        JoinButton.OnClick().Subscribe(OnJoinButtonClicked)
        LeaveButton : button_quiet = button_quiet{DefaultText := StringToMessage("Leave")}
        LeaveButton.OnClick().Subscribe(OnLeaveButtonClicked)
        set CanvasSlots += array
        {
            canvas_slot:
                Anchors := CenterAnchors
                Offsets := margin{Top:= 201.0, Left:= 393.0, Right := 420.0, Bottom := 243.0}
                Alignment := CenterVector
                SizeToContent := false
                Widget := stack_box:
                    Orientation := orientation.Vertical
                    Slots := array:
                        stack_box_slot:
                            Distribution:= option{0.2}
                            HorizontalAlignment := horizontal_alignment.Fill
                            VerticalAlignment := vertical_alignment.Fill
                            Widget := stack_box:
                                Orientation := orientation.Horizontal
                                Slots := array:
                                    stack_box_slot:
                                        Padding:=margin{Right:=15.0}
                                        HorizontalAlignment := horizontal_alignment.Center
                                        VerticalAlignment := vertical_alignment.Top
                                        Widget := HelpIconImage
                                    stack_box_slot:
                                        Distribution:= option{1.0}
                                        HorizontalAlignment := horizontal_alignment.Fill
                                        VerticalAlignment := vertical_alignment.Top
                                        Widget := HelpButton
                        stack_box_slot:
                            Distribution:= option{0.2}
                            HorizontalAlignment := horizontal_alignment.Fill
                            VerticalAlignment := vertical_alignment.Fill
                            Widget := stack_box:
                                Orientation := orientation.Horizontal
                                Slots := array:
                                    stack_box_slot:
                                        Padding:=margin{Right:=15.0}
                                        HorizontalAlignment := horizontal_alignment.Center
                                        VerticalAlignment := vertical_alignment.Center
                                        Widget := BossIconImage
                                    stack_box_slot:
                                        Distribution:= option{1.0}
                                        HorizontalAlignment := horizontal_alignment.Fill
                                        VerticalAlignment := vertical_alignment.Center
                                        Widget := JoinButton
                        stack_box_slot:
                            Distribution:= option{0.2}
                            HorizontalAlignment := horizontal_alignment.Fill
                            VerticalAlignment := vertical_alignment.Fill
                            Widget := stack_box:
                                Orientation := orientation.Horizontal
                                Slots := array:
                                    stack_box_slot:
                                        Padding:=margin{Right:=15.0}
                                        HorizontalAlignment := horizontal_alignment.Center
                                        VerticalAlignment := vertical_alignment.Center
                                        Widget := LeaveIconImage
                                    stack_box_slot:
                                        Distribution:= option{1.0}
                                        HorizontalAlignment := horizontal_alignment.Fill
                                        VerticalAlignment := vertical_alignment.Center
                                        Widget := LeaveButton
        }

        #========================================== Close ==========================================

        CloseButton : button_loud = button_loud{DefaultText := StringToMessage("CLOSE")}
        CloseButton.OnClick().Subscribe(OnCloseButtonClicked)
        set CanvasSlots += array
        {
            canvas_slot:
                Anchors := CenterAnchors
                Offsets := margin{Top:= 387.0, Left:= 435.0, Right := 353.0, Bottom := 80.0}
                Alignment := CenterVector
                SizeToContent := false
                Widget := CloseButton
        }

        Canvas : canvas = canvas:
            Slots := CanvasSlots

        return Canvas