
using { /Verse.org/Assets }
using { /Verse.org/Simulation }
using { /Fortnite.com/UI }
using { /Fortnite.com/Devices }
using { /Fortnite.com/Characters }
using { /UnrealEngine.com/Temporary/UI }
using { /UnrealEngine.com/Temporary/SpatialMath }
using { /Verse.org/Simulation }
using { /UnrealEngine.com/Temporary/Diagnostics }
using { /Verse.org/Colors }
using { /Verse.org/Colors/NamedColors }
using { Mining_System.Code.MiningObject}
using { Mining_System.Code.PlayerData }
using { Mining_System.Code.MiningObject }
# See https://dev.epicgames.com/documentation/en-us/uefn/create-your-own-device-in-verse for how to create a verse device.

# A Verse-authored creative device that can be placed in a level
ui_boss_list<public>:= class(creative_device):
    #SFX
    @editable
    SFX_Open : audio_player_device = audio_player_device{}
    @editable
    SFX_Close : audio_player_device = audio_player_device{}
    @editable
    SFX_Warp : audio_player_device = audio_player_device{}

    @editable
    PlayerSpawners : []player_spawner_device = array{}
    @editable
    DebugButton : button_device = button_device{}

    @editable
    TP_Boss1 : teleporter_device = teleporter_device{}
    @editable
    TP_Boss2 : teleporter_device = teleporter_device{}
    @editable
    TP_Boss3 : teleporter_device = teleporter_device{}
    @editable
    TP_Boss4 : teleporter_device = teleporter_device{}
    @editable
    TP_Boss5 : teleporter_device = teleporter_device{}
    @editable
    TP_Boss6 : teleporter_device = teleporter_device{}

    @editable
    QuestTrigger : trigger_device = trigger_device{}
    @editable
    SaveDataManager : save_data_manager = save_data_manager{}
    @editable
    MiningDamageDevice : mining_damage_device = mining_damage_device{}


    var BosslistCanvasPerAgent : [agent]canvas = map{}
    
    StringToMessage<public><localizes>(Text: string): message = "{Text}"

    var BossUIElements : [agent][int]tuple(texture_block, texture_block, texture_block, texture_block, button_quiet, text_block, text_block, text_block, text_block) = map{}

    CloseEvent<public> : event(player) = event(player){}

    OnBegin<override>()<suspends>:void=
        DebugButton.InteractedWithEvent.Subscribe(OpenUI)

        Sleep(5.0)

        for(Player : GetPlayspace().GetPlayers()):
            SaveDataManager.CheckRecoveryBossTicket(Player)
        
        GetPlayspace().PlayerAddedEvent().Subscribe(SaveDataManager.SubscribeRBT)
        GetPlayspace().PlayerAddedEvent().Subscribe(SaveDataManager.CheckRecoveryBossTicket)

        # for (PlayerSpawner : PlayerSpawners):
        #     PlayerSpawner.SpawnedEvent.Subscribe(OnPlayerSpawned)

    OnPlayerSpawned(Agent : agent):void=
        if (BosslistCanvas := BosslistCanvasPerAgent[Agent]):
            return

        if (Player := player[Agent], PlayerUI := GetPlayerUI[Player]):
            BosslistCanvas : canvas = CreateBosslistCanvas(Agent)
            BosslistCanvas.SetVisibility(widget_visibility.Collapsed)
            if (set BosslistCanvasPerAgent[Agent] = BosslistCanvas){}

    OpenUI<public>(Agent : agent):void=
        SFX_Open.Play(Agent)
        OnPlayerSpawned(Agent)
        QuestTrigger.Trigger(Agent)
        Print("OnDebugButtonEvented")
        if (Player := player[Agent], PlayerUI := GetPlayerUI[Player], BosslistCanvas := BosslistCanvasPerAgent[Agent]):
            BosslistCanvas.SetVisibility(widget_visibility.Visible)
            PlayerUI.AddWidget(BosslistCanvas, player_ui_slot{InputMode := ui_input_mode.All})

            for(UnlockedBossElement : SaveDataManager.GetDefeatedBosses(Player), BossElement := BossUIElements[Agent][UnlockedBossElement]):
                BossElement(0).SetVisibility(widget_visibility.Visible)
                BossElement(1).SetVisibility(widget_visibility.Visible)
                BossElement(2).SetVisibility(widget_visibility.Visible)
                BossElement(3).SetVisibility(widget_visibility.Visible)
                BossElement(4).SetVisibility(widget_visibility.Visible)
                BossElement(5).SetVisibility(widget_visibility.Visible)
                BossElement(6).SetVisibility(widget_visibility.Visible)
                BossElement(7).SetVisibility(widget_visibility.Visible)
                BossElement(8).SetVisibility(widget_visibility.Visible)

    # ボタン処理
    OnTeleport1Clicked(WidgetMessage : widget_message):void=
        if(Agent : agent = agent[WidgetMessage.Player]) :
            SFX_Warp.Play(Agent)
        if (PlayerUI := GetPlayerUI[WidgetMessage.Player], BosslistCanvas := BosslistCanvasPerAgent[WidgetMessage.Player]):
            TP_Boss1.Teleport(WidgetMessage.Player)
            OnCloseButtonClicked(WidgetMessage)

    OnTeleport2Clicked(WidgetMessage : widget_message):void=
        if(Agent : agent = agent[WidgetMessage.Player]) :
            SFX_Warp.Play(Agent)
        if (PlayerUI := GetPlayerUI[WidgetMessage.Player], BosslistCanvas := BosslistCanvasPerAgent[WidgetMessage.Player]):
            TP_Boss2.Teleport(WidgetMessage.Player)
            OnCloseButtonClicked(WidgetMessage)

    OnTeleport3Clicked(WidgetMessage : widget_message):void=
        if(Agent : agent = agent[WidgetMessage.Player]) :
            SFX_Warp.Play(Agent)
        if (PlayerUI := GetPlayerUI[WidgetMessage.Player], BosslistCanvas := BosslistCanvasPerAgent[WidgetMessage.Player]):
            TP_Boss3.Teleport(WidgetMessage.Player)
            OnCloseButtonClicked(WidgetMessage)

    OnTeleport4Clicked(WidgetMessage : widget_message):void=
        if(Agent : agent = agent[WidgetMessage.Player]) :
            SFX_Warp.Play(Agent)
        if (PlayerUI := GetPlayerUI[WidgetMessage.Player], BosslistCanvas := BosslistCanvasPerAgent[WidgetMessage.Player]):
            TP_Boss4.Teleport(WidgetMessage.Player)
            OnCloseButtonClicked(WidgetMessage)
            
    OnTeleport5Clicked(WidgetMessage : widget_message):void=
        if(Agent : agent = agent[WidgetMessage.Player]) :
            SFX_Warp.Play(Agent)
        if (PlayerUI := GetPlayerUI[WidgetMessage.Player], BosslistCanvas := BosslistCanvasPerAgent[WidgetMessage.Player]):
            TP_Boss5.Teleport(WidgetMessage.Player)
            OnCloseButtonClicked(WidgetMessage)

    OnTeleport6Clicked(WidgetMessage : widget_message):void=
        if(Agent : agent = agent[WidgetMessage.Player]) :
            SFX_Warp.Play(Agent)
        if (PlayerUI := GetPlayerUI[WidgetMessage.Player], BosslistCanvas := BosslistCanvasPerAgent[WidgetMessage.Player]):
            TP_Boss6.Teleport(WidgetMessage.Player)
            OnCloseButtonClicked(WidgetMessage)

    OnCloseButtonClicked(WidgetMessage : widget_message):void=
        if(Agent : agent = agent[WidgetMessage.Player]) :
            SFX_Close.Play(Agent)
        if (PlayerUI := GetPlayerUI[WidgetMessage.Player], BosslistCanvas := BosslistCanvasPerAgent[WidgetMessage.Player]):
            PlayerUI.RemoveWidget(BosslistCanvas)
            BosslistCanvas.SetVisibility(widget_visibility.Collapsed)
            CloseEvent.Signal(WidgetMessage.Player)

    # キャンバス
    CreateBosslistCanvas(Agent : agent):canvas=
        var CanvasSlots : []canvas_slot = array{}
        
        CenterVector : vector2 = vector2{X:=0.5, Y:=0.5}
        CenterAnchors : anchors = anchors{Minimum := CenterVector, Maximum := CenterVector}

        # Image Stock
        BGImage : texture_block = texture_block{DefaultImage := Textures.UI_BG, DefaultDesiredSize := vector2{X := 1920.0, Y := 1080.0}}
        BossImage_1 : texture_block = texture_block{DefaultImage := Boss_Textures.kreppar, DefaultDesiredSize := vector2{X := 100.0, Y := 100.0}}
        BossImage_2 : texture_block = texture_block{DefaultImage := Boss_Textures.tachigarasu, DefaultDesiredSize := vector2{X := 100.0, Y := 100.0}}
        BossImage_3 : texture_block = texture_block{DefaultImage := Boss_Textures.obanmomen, DefaultDesiredSize := vector2{X := 100.0, Y := 100.0}}
        BossImage_4 : texture_block = texture_block{DefaultImage := Boss_Textures.karakasaobake, DefaultDesiredSize := vector2{X := 100.0, Y := 100.0}}
        BossImage_5 : texture_block = texture_block{DefaultImage := Boss_Textures.enengama, DefaultDesiredSize := vector2{X := 100.0, Y := 100.0}}
        BossImage_6 : texture_block = texture_block{DefaultImage := Boss_Textures.yatagaran, DefaultDesiredSize := vector2{X := 100.0, Y := 100.0}}
        IconImage_Power1 : texture_block = texture_block{DefaultImage := icon_set_2.IC_Power, DefaultDesiredSize := vector2{X := 50.0, Y := 50.0}}
        IconImage_Power2 : texture_block = texture_block{DefaultImage := icon_set_2.IC_Power, DefaultDesiredSize := vector2{X := 50.0, Y := 50.0}}
        IconImage_Power3 : texture_block = texture_block{DefaultImage := icon_set_2.IC_Power, DefaultDesiredSize := vector2{X := 50.0, Y := 50.0}}
        IconImage_Power4 : texture_block = texture_block{DefaultImage := icon_set_2.IC_Power, DefaultDesiredSize := vector2{X := 50.0, Y := 50.0}}
        IconImage_Power5 : texture_block = texture_block{DefaultImage := icon_set_2.IC_Power, DefaultDesiredSize := vector2{X := 50.0, Y := 50.0}}
        IconImage_Power6 : texture_block = texture_block{DefaultImage := icon_set_2.IC_Power, DefaultDesiredSize := vector2{X := 50.0, Y := 50.0}}
        IconImage_Summon1 : texture_block = texture_block{DefaultImage := icon_set_2.SS_1, DefaultDesiredSize := vector2{X := 50.0, Y := 50.0}}
        IconImage_Summon2 : texture_block = texture_block{DefaultImage := icon_set_2.SS_1, DefaultDesiredSize := vector2{X := 50.0, Y := 50.0}}
        IconImage_Summon3 : texture_block = texture_block{DefaultImage := icon_set_2.SS_1, DefaultDesiredSize := vector2{X := 50.0, Y := 50.0}}
        IconImage_Summon4 : texture_block = texture_block{DefaultImage := icon_set_2.SS_1, DefaultDesiredSize := vector2{X := 50.0, Y := 50.0}}
        IconImage_Summon5 : texture_block = texture_block{DefaultImage := icon_set_2.SS_1, DefaultDesiredSize := vector2{X := 50.0, Y := 50.0}}
        IconImage_Summon6 : texture_block = texture_block{DefaultImage := icon_set_2.SS_1, DefaultDesiredSize := vector2{X := 50.0, Y := 50.0}}
        IconImage_Stone1 : texture_block = texture_block{DefaultImage := icon_set_2.SilverOre, DefaultDesiredSize := vector2{X := 50.0, Y := 50.0}}
        IconImage_Stone2 : texture_block = texture_block{DefaultImage := icon_set_2.SilverOre, DefaultDesiredSize := vector2{X := 50.0, Y := 50.0}}
        IconImage_Stone3 : texture_block = texture_block{DefaultImage := icon_set_2.SilverOre, DefaultDesiredSize := vector2{X := 50.0, Y := 50.0}}
        IconImage_Stone4 : texture_block = texture_block{DefaultImage := icon_set_2.SilverOre, DefaultDesiredSize := vector2{X := 50.0, Y := 50.0}}
        IconImage_Stone5 : texture_block = texture_block{DefaultImage := icon_set_2.SilverOre, DefaultDesiredSize := vector2{X := 50.0, Y := 50.0}}
        IconImage_Stone6 : texture_block = texture_block{DefaultImage := icon_set_2.SilverOre, DefaultDesiredSize := vector2{X := 50.0, Y := 50.0}}
        
        # Button Stock
        QuietButtonTeleport1 : button_quiet = button_quiet{DefaultText := StringToMessage("Teleport")}
        QuietButtonTeleport1.OnClick().Subscribe(OnTeleport1Clicked)
        QuietButtonTeleport2 : button_quiet = button_quiet{DefaultText := StringToMessage("Teleport")}
        QuietButtonTeleport2.OnClick().Subscribe(OnTeleport2Clicked)
        QuietButtonTeleport3 : button_quiet = button_quiet{DefaultText := StringToMessage("Teleport")}
        QuietButtonTeleport3.OnClick().Subscribe(OnTeleport3Clicked)
        QuietButtonTeleport4 : button_quiet = button_quiet{DefaultText := StringToMessage("Teleport")}
        QuietButtonTeleport4.OnClick().Subscribe(OnTeleport4Clicked)
        QuietButtonTeleport5 : button_quiet = button_quiet{DefaultText := StringToMessage("Teleport")}
        QuietButtonTeleport5.OnClick().Subscribe(OnTeleport5Clicked)
        QuietButtonTeleport6 : button_quiet = button_quiet{DefaultText := StringToMessage("Teleport")}
        QuietButtonTeleport6.OnClick().Subscribe(OnTeleport6Clicked)
        # Stock END

        #========================================== BG ==========================================

        set CanvasSlots += array
        {
            canvas_slot:
                Anchors := CenterAnchors
                Offsets := margin{Top := 0.0,Left := 0.0, Right := 1920.0, Bottom := 1080.0}
                Alignment := CenterVector
                SizeToContent := false
                Widget := BGImage
        }

        #========================================== End ==========================================

        # Text Stock
        # header
        BOSSTextBlock := text_block{
            DefaultText := StringToMessage("BOSS"),
            DefaultTextColor := NamedColors.White,
            DefaultShadowOffset := option{vector2{X:=2.5, Y:=2.5}},
            DefaultShadowColor := NamedColors.Black,
            DefaultJustification := text_justification.Center
        }
        BOSSTextBlock.SetShadowOpacity(1.0)
        IconTextBlock := text_block{
            DefaultText := StringToMessage(""),
            DefaultTextColor := NamedColors.White,
            DefaultShadowOffset := option{vector2{X:=2.5, Y:=2.5}},
            DefaultShadowColor := NamedColors.Black,
            DefaultJustification := text_justification.Left
        }
        BOSSTextBlock.SetShadowOpacity(1.0)
        NameTextBlock := text_block{
            DefaultText := StringToMessage("Name"),
            DefaultTextColor := NamedColors.White,
            DefaultShadowOffset := option{vector2{X:=2.5, Y:=2.5}},
            DefaultShadowColor := NamedColors.Black,
            DefaultJustification := text_justification.Left
        }
        NameTextBlock.SetShadowOpacity(1.0)
        PowerUpTextBlock := text_block{
            DefaultText := StringToMessage("PowerUp Stone"),
            DefaultTextColor := NamedColors.White,
            DefaultShadowOffset := option{vector2{X:=2.5, Y:=2.5}},
            DefaultShadowColor := NamedColors.Black,
            DefaultJustification := text_justification.Left
        }
        PowerUpTextBlock.SetShadowOpacity(1.0)
        SummonTextBlock := text_block{
            DefaultText := StringToMessage("Summon Stone"),
            DefaultTextColor := NamedColors.White,
            DefaultShadowOffset := option{vector2{X:=2.5, Y:=2.5}},
            DefaultShadowColor := NamedColors.Black,
            DefaultJustification := text_justification.Left
        }
        SummonTextBlock.SetShadowOpacity(1.0)
        StoneTextBlock := text_block{
            DefaultText := StringToMessage("Stone"),
            DefaultTextColor := NamedColors.White,
            DefaultShadowOffset := option{vector2{X:=2.5, Y:=2.5}},
            DefaultShadowColor := NamedColors.Black,
            DefaultJustification := text_justification.Left
        }
        StoneTextBlock.SetShadowOpacity(1.0)
        TeleportTextBlock := text_block{
            DefaultText := StringToMessage("Teleport"),
            DefaultTextColor := NamedColors.White,
            DefaultShadowOffset := option{vector2{X:=2.5, Y:=2.5}},
            DefaultShadowColor := NamedColors.Black,
            DefaultJustification := text_justification.Left
        }
        TeleportTextBlock.SetShadowOpacity(1.0)
        # BossName
        BossName1TextBlock := text_block{
            DefaultText := StringToMessage("Krepper"),
            DefaultTextColor := NamedColors.White,
            DefaultShadowOffset := option{vector2{X:=2.5, Y:=2.5}},
            DefaultShadowColor := NamedColors.Black,
            DefaultJustification := text_justification.Left
        }
        BossName1TextBlock.SetShadowOpacity(1.0)
        BossName2TextBlock := text_block{
            DefaultText := StringToMessage("Tachigarasu"),
            DefaultTextColor := NamedColors.White,
            DefaultShadowOffset := option{vector2{X:=2.5, Y:=2.5}},
            DefaultShadowColor := NamedColors.Black,
            DefaultJustification := text_justification.Left
        }
        BossName2TextBlock.SetShadowOpacity(1.0)
        BossName3TextBlock := text_block{
            DefaultText := StringToMessage("Obanmomen"),
            DefaultTextColor := NamedColors.White,
            DefaultShadowOffset := option{vector2{X:=2.5, Y:=2.5}},
            DefaultShadowColor := NamedColors.Black,
            DefaultJustification := text_justification.Left
        }
        BossName3TextBlock.SetShadowOpacity(1.0)
        BossName4TextBlock := text_block{
            DefaultText := StringToMessage("Karakasaobake"),
            DefaultTextColor := NamedColors.White,
            DefaultShadowOffset := option{vector2{X:=2.5, Y:=2.5}},
            DefaultShadowColor := NamedColors.Black,
            DefaultJustification := text_justification.Left
        }
        BossName4TextBlock.SetShadowOpacity(1.0)
        BossName5TextBlock := text_block{
            DefaultText := StringToMessage("Enengama"),
            DefaultTextColor := NamedColors.White,
            DefaultShadowOffset := option{vector2{X:=2.5, Y:=2.5}},
            DefaultShadowColor := NamedColors.Black,
            DefaultJustification := text_justification.Left
        }
        BossName5TextBlock.SetShadowOpacity(1.0)
        BossName6TextBlock := text_block{
            DefaultText := StringToMessage("Yatagaran"),
            DefaultTextColor := NamedColors.White,
            DefaultShadowOffset := option{vector2{X:=2.5, Y:=2.5}},
            DefaultShadowColor := NamedColors.Black,
            DefaultJustification := text_justification.Left
        }
        BossName6TextBlock.SetShadowOpacity(1.0)
        # Boss報酬 PowerUp
        PowerUp1TextBlock := text_block{
            DefaultText := StringToMessage("10"),
            DefaultTextColor := NamedColors.White,
            DefaultShadowOffset := option{vector2{X:=2.5, Y:=2.5}},
            DefaultShadowColor := NamedColors.Black,
            DefaultJustification := text_justification.Left
        }
        PowerUp1TextBlock.SetShadowOpacity(1.0)
        PowerUp2TextBlock := text_block{
            DefaultText := StringToMessage("80"),
            DefaultTextColor := NamedColors.White,
            DefaultShadowOffset := option{vector2{X:=2.5, Y:=2.5}},
            DefaultShadowColor := NamedColors.Black,
            DefaultJustification := text_justification.Left
        }
        PowerUp2TextBlock.SetShadowOpacity(1.0)
        PowerUp3TextBlock := text_block{
            DefaultText := StringToMessage("150"),
            DefaultTextColor := NamedColors.White,
            DefaultShadowOffset := option{vector2{X:=2.5, Y:=2.5}},
            DefaultShadowColor := NamedColors.Black,
            DefaultJustification := text_justification.Left
        }
        PowerUp3TextBlock.SetShadowOpacity(1.0)
        PowerUp4TextBlock := text_block{
            DefaultText := StringToMessage("300"),
            DefaultTextColor := NamedColors.White,
            DefaultShadowOffset := option{vector2{X:=2.5, Y:=2.5}},
            DefaultShadowColor := NamedColors.Black,
            DefaultJustification := text_justification.Left
        }
        PowerUp4TextBlock.SetShadowOpacity(1.0)
        PowerUp5TextBlock := text_block{
            DefaultText := StringToMessage("600"),
            DefaultTextColor := NamedColors.White,
            DefaultShadowOffset := option{vector2{X:=2.5, Y:=2.5}},
            DefaultShadowColor := NamedColors.Black,
            DefaultJustification := text_justification.Left
        }
        PowerUp5TextBlock.SetShadowOpacity(1.0)
        PowerUp6TextBlock := text_block{
            DefaultText := StringToMessage("1.00A"),
            DefaultTextColor := NamedColors.White,
            DefaultShadowOffset := option{vector2{X:=2.5, Y:=2.5}},
            DefaultShadowColor := NamedColors.Black,
            DefaultJustification := text_justification.Left
        }
        PowerUp6TextBlock.SetShadowOpacity(1.0)
        # Boss報酬 Summon
        Summon1TextBlock := text_block{
            DefaultText := StringToMessage("33"),
            DefaultTextColor := NamedColors.White,
            DefaultShadowOffset := option{vector2{X:=2.5, Y:=2.5}},
            DefaultShadowColor := NamedColors.Black,
            DefaultJustification := text_justification.Left
        }
        Summon1TextBlock.SetShadowOpacity(1.0)
        Summon2TextBlock := text_block{
            DefaultText := StringToMessage("200"),
            DefaultTextColor := NamedColors.White,
            DefaultShadowOffset := option{vector2{X:=2.5, Y:=2.5}},
            DefaultShadowColor := NamedColors.Black,
            DefaultJustification := text_justification.Left
        }
        Summon2TextBlock.SetShadowOpacity(1.0)
        Summon3TextBlock := text_block{
            DefaultText := StringToMessage("800"),
            DefaultTextColor := NamedColors.White,
            DefaultShadowOffset := option{vector2{X:=2.5, Y:=2.5}},
            DefaultShadowColor := NamedColors.Black,
            DefaultJustification := text_justification.Left
        }
        Summon3TextBlock.SetShadowOpacity(1.0)
        Summon4TextBlock := text_block{
            DefaultText := StringToMessage("1300"),
            DefaultTextColor := NamedColors.White,
            DefaultShadowOffset := option{vector2{X:=2.5, Y:=2.5}},
            DefaultShadowColor := NamedColors.Black,
            DefaultJustification := text_justification.Left
        }
        Summon4TextBlock.SetShadowOpacity(1.0)
        Summon5TextBlock := text_block{
            DefaultText := StringToMessage("2250"),
            DefaultTextColor := NamedColors.White,
            DefaultShadowOffset := option{vector2{X:=2.5, Y:=2.5}},
            DefaultShadowColor := NamedColors.Black,
            DefaultJustification := text_justification.Left
        }
        Summon5TextBlock.SetShadowOpacity(1.0)
        Summon6TextBlock := text_block{
            DefaultText := StringToMessage("3800"),
            DefaultTextColor := NamedColors.White,
            DefaultShadowOffset := option{vector2{X:=2.5, Y:=2.5}},
            DefaultShadowColor := NamedColors.Black,
            DefaultJustification := text_justification.Left
        }
        Summon6TextBlock.SetShadowOpacity(1.0)
        #Boss報酬 Stone
        var GetDamage : big_number = big_number{}
        var RewardResource1 : big_number = big_number{}
        var RewardResource2 : big_number = big_number{}
        var RewardResource3 : big_number = big_number{}
        var RewardResource4 : big_number = big_number{}
        var RewardResource5 : big_number = big_number{}
        var RewardResource6 : big_number = big_number{}
        if(Player : player = player[Agent]) :
            set GetDamage =MiningDamageDevice.GetNormalDamageAmount(Player).Amount
            set RewardResource1 = Mining_System.Code.PlayerData.BigNumber_MultiplyByBigNumber(GetDamage, Mining_System.Code.PlayerData.big_number{Value := 250.0})
            set RewardResource2 = Mining_System.Code.PlayerData.BigNumber_MultiplyByBigNumber(GetDamage, Mining_System.Code.PlayerData.big_number{Value := 300.0})
            set RewardResource3 = Mining_System.Code.PlayerData.BigNumber_MultiplyByBigNumber(GetDamage, Mining_System.Code.PlayerData.big_number{Value := 650.0})
            set RewardResource4 = Mining_System.Code.PlayerData.BigNumber_MultiplyByBigNumber(GetDamage, Mining_System.Code.PlayerData.big_number{Value := 700.0})
            set RewardResource5 = Mining_System.Code.PlayerData.BigNumber_MultiplyByBigNumber(GetDamage, Mining_System.Code.PlayerData.big_number{Value := 1125.0})
            set RewardResource6 = Mining_System.Code.PlayerData.BigNumber_MultiplyByBigNumber(GetDamage, Mining_System.Code.PlayerData.big_number{Value := 1300.0})
        Stone1TextBlock : text_block = text_block :
            DefaultText := StringToMessage("{BigNumber_FormatDisplay(RewardResource1)}")
            DefaultTextColor := NamedColors.White
            DefaultShadowOffset := option{vector2{X:=2.5, Y:=2.5}}
            DefaultShadowColor := NamedColors.Black
            DefaultShadowOpacity := 1.0
            DefaultJustification := text_justification.Left
        Stone2TextBlock : text_block = text_block :
            DefaultText := StringToMessage("{BigNumber_FormatDisplay(RewardResource2)}")
            DefaultTextColor := NamedColors.White
            DefaultShadowOffset := option{vector2{X:=2.5, Y:=2.5}}
            DefaultShadowColor := NamedColors.Black
            DefaultShadowOpacity := 1.0
            DefaultJustification := text_justification.Left
        Stone3TextBlock : text_block = text_block :
            DefaultText := StringToMessage("{BigNumber_FormatDisplay(RewardResource3)}")
            DefaultTextColor := NamedColors.White
            DefaultShadowOffset := option{vector2{X:=2.5, Y:=2.5}}
            DefaultShadowColor := NamedColors.Black
            DefaultShadowOpacity := 1.0
            DefaultJustification := text_justification.Left
        Stone4TextBlock : text_block = text_block :
            DefaultText := StringToMessage("{BigNumber_FormatDisplay(RewardResource4)}")
            DefaultTextColor := NamedColors.White
            DefaultShadowOffset := option{vector2{X:=2.5, Y:=2.5}}
            DefaultShadowColor := NamedColors.Black
            DefaultShadowOpacity := 1.0
            DefaultJustification := text_justification.Left
        Stone5TextBlock : text_block = text_block :
            DefaultText := StringToMessage("{BigNumber_FormatDisplay(RewardResource5)}")
            DefaultTextColor := NamedColors.White
            DefaultShadowOffset := option{vector2{X:=2.5, Y:=2.5}}
            DefaultShadowColor := NamedColors.Black
            DefaultShadowOpacity := 1.0
            DefaultJustification := text_justification.Left
        Stone6TextBlock : text_block = text_block :
            DefaultText := StringToMessage("{BigNumber_FormatDisplay(RewardResource6)}")
            DefaultTextColor := NamedColors.White
            DefaultShadowOffset := option{vector2{X:=2.5, Y:=2.5}}
            DefaultShadowColor := NamedColors.Black
            DefaultShadowOpacity := 1.0
            DefaultJustification := text_justification.Left
        # Stock END

        #========================================== Header ==========================================

        set CanvasSlots += array
        {
            canvas_slot:
                Anchors := CenterAnchors
                Offsets := margin{Top:= -484.0, Left:= 0.0, Right := 257.0, Bottom := 41.0}
                Alignment := CenterVector
                SizeToContent := false
                Widget := BOSSTextBlock
        }

        #========================================== Main Container ==========================================
        
        set CanvasSlots += array
        {
            canvas_slot:
                Anchors := CenterAnchors
                Offsets := margin{Top:= 30.5, Left:= 1.0, Right := 1824.0, Bottom := 735.0}
                Alignment := CenterVector
                SizeToContent := false
                Widget := stack_box:
                    Orientation := orientation.Horizontal
                    Slots := array:
        #================================ BossIcon ================================
                        stack_box_slot:
                            Distribution:= option{0.2}
                            HorizontalAlignment := horizontal_alignment.Fill
                            VerticalAlignment := vertical_alignment.Fill
                            Widget := stack_box:
                                Orientation := orientation.Vertical
                                Slots := array:
                                    stack_box_slot:
                                        Distribution:= option{0.1}
                                        HorizontalAlignment := horizontal_alignment.Fill
                                        VerticalAlignment := vertical_alignment.Fill
                                        Widget := IconTextBlock
                                    stack_box_slot:
                                        Distribution:= option{0.2}
                                        HorizontalAlignment := horizontal_alignment.Center
                                        VerticalAlignment := vertical_alignment.Center
                                        Widget := stack_box:
                                            Orientation := orientation.Horizontal
                                            Slots := array:
                                                stack_box_slot:
                                                    HorizontalAlignment := horizontal_alignment.Center
                                                    VerticalAlignment := vertical_alignment.Center
                                                    Widget := BossImage_1
                                    stack_box_slot:
                                        Distribution:= option{0.2}
                                        HorizontalAlignment := horizontal_alignment.Center
                                        VerticalAlignment := vertical_alignment.Center
                                        Widget := stack_box:
                                            Orientation := orientation.Horizontal
                                            Slots := array:
                                                stack_box_slot:
                                                    HorizontalAlignment := horizontal_alignment.Center
                                                    VerticalAlignment := vertical_alignment.Center
                                                    Widget := BossImage_2
                                    stack_box_slot:
                                        Distribution:= option{0.2}
                                        HorizontalAlignment := horizontal_alignment.Center
                                        VerticalAlignment := vertical_alignment.Center
                                        Widget := stack_box:
                                            Orientation := orientation.Horizontal
                                            Slots := array:
                                                stack_box_slot:
                                                    HorizontalAlignment := horizontal_alignment.Center
                                                    VerticalAlignment := vertical_alignment.Center
                                                    Widget := BossImage_3
                                    stack_box_slot:
                                        Distribution:= option{0.2}
                                        HorizontalAlignment := horizontal_alignment.Center
                                        VerticalAlignment := vertical_alignment.Center
                                        Widget := stack_box:
                                            Orientation := orientation.Horizontal
                                            Slots := array:
                                                stack_box_slot:
                                                    HorizontalAlignment := horizontal_alignment.Center
                                                    VerticalAlignment := vertical_alignment.Center
                                                    Widget := BossImage_4
                                    stack_box_slot:
                                        Distribution:= option{0.2}
                                        HorizontalAlignment := horizontal_alignment.Center
                                        VerticalAlignment := vertical_alignment.Center
                                        Widget := stack_box:
                                            Orientation := orientation.Horizontal
                                            Slots := array:
                                                stack_box_slot:
                                                    HorizontalAlignment := horizontal_alignment.Center
                                                    VerticalAlignment := vertical_alignment.Center
                                                    Widget := BossImage_5
                                    stack_box_slot:
                                        Distribution:= option{0.2}
                                        HorizontalAlignment := horizontal_alignment.Center
                                        VerticalAlignment := vertical_alignment.Center
                                        Widget := stack_box:
                                            Orientation := orientation.Horizontal
                                            Slots := array:
                                                stack_box_slot:
                                                    HorizontalAlignment := horizontal_alignment.Center
                                                    VerticalAlignment := vertical_alignment.Center
                                                    Widget := BossImage_6

        #=============================== Boss Name ===============================

                        stack_box_slot:
                            Distribution:= option{0.2}
                            HorizontalAlignment := horizontal_alignment.Fill
                            VerticalAlignment := vertical_alignment.Fill
                            Widget := stack_box:
                                Orientation := orientation.Vertical
                                Slots := array:
                                    stack_box_slot:
                                        Distribution:= option{0.1}
                                        HorizontalAlignment := horizontal_alignment.Fill
                                        VerticalAlignment := vertical_alignment.Fill
                                        Widget := NameTextBlock
                                    stack_box_slot:
                                        Distribution:= option{0.2}
                                        HorizontalAlignment := horizontal_alignment.Left
                                        VerticalAlignment := vertical_alignment.Center
                                        Widget := stack_box:
                                            Orientation := orientation.Horizontal
                                            Slots := array:
                                                stack_box_slot:
                                                    HorizontalAlignment := horizontal_alignment.Center
                                                    VerticalAlignment := vertical_alignment.Center
                                                    Widget := BossName1TextBlock
                                    stack_box_slot:
                                        Distribution:= option{0.2}
                                        HorizontalAlignment := horizontal_alignment.Left
                                        VerticalAlignment := vertical_alignment.Center
                                        Widget := stack_box:
                                            Orientation := orientation.Horizontal
                                            Slots := array:
                                                stack_box_slot:
                                                    HorizontalAlignment := horizontal_alignment.Center
                                                    VerticalAlignment := vertical_alignment.Center
                                                    Widget := BossName2TextBlock
                                    stack_box_slot:
                                        Distribution:= option{0.2}
                                        HorizontalAlignment := horizontal_alignment.Left
                                        VerticalAlignment := vertical_alignment.Center
                                        Widget := stack_box:
                                            Orientation := orientation.Horizontal
                                            Slots := array:
                                                stack_box_slot:
                                                    HorizontalAlignment := horizontal_alignment.Center
                                                    VerticalAlignment := vertical_alignment.Center
                                                    Widget := BossName3TextBlock
                                    stack_box_slot:
                                        Distribution:= option{0.2}
                                        HorizontalAlignment := horizontal_alignment.Left
                                        VerticalAlignment := vertical_alignment.Center
                                        Widget := stack_box:
                                            Orientation := orientation.Horizontal
                                            Slots := array:
                                                stack_box_slot:
                                                    HorizontalAlignment := horizontal_alignment.Center
                                                    VerticalAlignment := vertical_alignment.Center
                                                    Widget := BossName4TextBlock
                                    stack_box_slot:
                                        Distribution:= option{0.2}
                                        HorizontalAlignment := horizontal_alignment.Left
                                        VerticalAlignment := vertical_alignment.Center
                                        Widget := stack_box:
                                            Orientation := orientation.Horizontal
                                            Slots := array:
                                                stack_box_slot:
                                                    HorizontalAlignment := horizontal_alignment.Center
                                                    VerticalAlignment := vertical_alignment.Center
                                                    Widget := BossName5TextBlock
                                    stack_box_slot:
                                        Distribution:= option{0.2}
                                        HorizontalAlignment := horizontal_alignment.Left
                                        VerticalAlignment := vertical_alignment.Center
                                        Widget := stack_box:
                                            Orientation := orientation.Horizontal
                                            Slots := array:
                                                stack_box_slot:
                                                    HorizontalAlignment := horizontal_alignment.Center
                                                    VerticalAlignment := vertical_alignment.Center
                                                    Widget := BossName6TextBlock
                        
        #=============================== PowerUp Item ===============================

                        stack_box_slot:
                            Distribution:= option{0.2}
                            HorizontalAlignment := horizontal_alignment.Fill
                            VerticalAlignment := vertical_alignment.Fill
                            Widget := stack_box:
                                Orientation := orientation.Vertical
                                Slots := array:
                                    stack_box_slot:
                                        Distribution:= option{0.1}
                                        HorizontalAlignment := horizontal_alignment.Fill
                                        VerticalAlignment := vertical_alignment.Fill
                                        Widget := PowerUpTextBlock
                                    stack_box_slot:
                                        Distribution:= option{0.2}
                                        HorizontalAlignment := horizontal_alignment.Left
                                        VerticalAlignment := vertical_alignment.Center
                                        Widget := stack_box:
                                            Orientation := orientation.Horizontal
                                            Slots := array:
                                                stack_box_slot:
                                                    HorizontalAlignment := horizontal_alignment.Center
                                                    VerticalAlignment := vertical_alignment.Center
                                                    Widget := IconImage_Power1
                                                stack_box_slot:
                                                    HorizontalAlignment := horizontal_alignment.Center
                                                    VerticalAlignment := vertical_alignment.Center
                                                    Widget := PowerUp1TextBlock
                                    stack_box_slot:
                                        Distribution:= option{0.2}
                                        HorizontalAlignment := horizontal_alignment.Left
                                        VerticalAlignment := vertical_alignment.Center
                                        Widget := stack_box:
                                            Orientation := orientation.Horizontal
                                            Slots := array:
                                                stack_box_slot:
                                                    HorizontalAlignment := horizontal_alignment.Center
                                                    VerticalAlignment := vertical_alignment.Center
                                                    Widget := IconImage_Power2
                                                stack_box_slot:
                                                    HorizontalAlignment := horizontal_alignment.Center
                                                    VerticalAlignment := vertical_alignment.Center
                                                    Widget := PowerUp2TextBlock
                                    stack_box_slot:
                                        Distribution:= option{0.2}
                                        HorizontalAlignment := horizontal_alignment.Left
                                        VerticalAlignment := vertical_alignment.Center
                                        Widget := stack_box:
                                            Orientation := orientation.Horizontal
                                            Slots := array:
                                                stack_box_slot:
                                                    HorizontalAlignment := horizontal_alignment.Center
                                                    VerticalAlignment := vertical_alignment.Center
                                                    Widget := IconImage_Power3
                                                stack_box_slot:
                                                    HorizontalAlignment := horizontal_alignment.Center
                                                    VerticalAlignment := vertical_alignment.Center
                                                    Widget := PowerUp3TextBlock
                                    stack_box_slot:
                                        Distribution:= option{0.2}
                                        HorizontalAlignment := horizontal_alignment.Left
                                        VerticalAlignment := vertical_alignment.Center
                                        Widget := stack_box:
                                            Orientation := orientation.Horizontal
                                            Slots := array:
                                                stack_box_slot:
                                                    HorizontalAlignment := horizontal_alignment.Center
                                                    VerticalAlignment := vertical_alignment.Center
                                                    Widget := IconImage_Power4
                                                stack_box_slot:
                                                    HorizontalAlignment := horizontal_alignment.Center
                                                    VerticalAlignment := vertical_alignment.Center
                                                    Widget := PowerUp4TextBlock
                                    stack_box_slot:
                                        Distribution:= option{0.2}
                                        HorizontalAlignment := horizontal_alignment.Left
                                        VerticalAlignment := vertical_alignment.Center
                                        Widget := stack_box:
                                            Orientation := orientation.Horizontal
                                            Slots := array:
                                                stack_box_slot:
                                                    HorizontalAlignment := horizontal_alignment.Center
                                                    VerticalAlignment := vertical_alignment.Center
                                                    Widget := IconImage_Power5
                                                stack_box_slot:
                                                    HorizontalAlignment := horizontal_alignment.Center
                                                    VerticalAlignment := vertical_alignment.Center
                                                    Widget := PowerUp5TextBlock
                                    stack_box_slot:
                                        Distribution:= option{0.2}
                                        HorizontalAlignment := horizontal_alignment.Left
                                        VerticalAlignment := vertical_alignment.Center
                                        Widget := stack_box:
                                            Orientation := orientation.Horizontal
                                            Slots := array:
                                                stack_box_slot:
                                                    HorizontalAlignment := horizontal_alignment.Center
                                                    VerticalAlignment := vertical_alignment.Center
                                                    Widget := IconImage_Power6
                                                stack_box_slot:
                                                    HorizontalAlignment := horizontal_alignment.Center
                                                    VerticalAlignment := vertical_alignment.Center
                                                    Widget := PowerUp6TextBlock
                        
        #=============================== Summon Stone ===============================

                        stack_box_slot:
                            Distribution:= option{0.2}
                            HorizontalAlignment := horizontal_alignment.Fill
                            VerticalAlignment := vertical_alignment.Fill
                            Widget := stack_box:
                                Orientation := orientation.Vertical
                                Slots := array:
                                    stack_box_slot:
                                        Widget := SummonTextBlock
                                        Distribution:= option{0.1}
                                        HorizontalAlignment := horizontal_alignment.Fill
                                        VerticalAlignment := vertical_alignment.Fill
                                    stack_box_slot:
                                        Distribution:= option{0.2}
                                        HorizontalAlignment := horizontal_alignment.Left
                                        VerticalAlignment := vertical_alignment.Center
                                        Widget := stack_box:
                                            Orientation := orientation.Horizontal
                                            Slots := array:
                                                stack_box_slot:
                                                    HorizontalAlignment := horizontal_alignment.Center
                                                    VerticalAlignment := vertical_alignment.Center
                                                    Widget := IconImage_Summon1
                                                stack_box_slot:
                                                    HorizontalAlignment := horizontal_alignment.Center
                                                    VerticalAlignment := vertical_alignment.Center
                                                    Widget := Summon1TextBlock
                                    stack_box_slot:
                                        Distribution:= option{0.2}
                                        HorizontalAlignment := horizontal_alignment.Left
                                        VerticalAlignment := vertical_alignment.Center
                                        Widget := stack_box:
                                            Orientation := orientation.Horizontal
                                            Slots := array:
                                                stack_box_slot:
                                                    HorizontalAlignment := horizontal_alignment.Center
                                                    VerticalAlignment := vertical_alignment.Center
                                                    Widget := IconImage_Summon2
                                                stack_box_slot:
                                                    HorizontalAlignment := horizontal_alignment.Center
                                                    VerticalAlignment := vertical_alignment.Center
                                                    Widget := Summon2TextBlock
                                    stack_box_slot:
                                        Distribution:= option{0.2}
                                        HorizontalAlignment := horizontal_alignment.Left
                                        VerticalAlignment := vertical_alignment.Center
                                        Widget := stack_box:
                                            Orientation := orientation.Horizontal
                                            Slots := array:
                                                stack_box_slot:
                                                    HorizontalAlignment := horizontal_alignment.Center
                                                    VerticalAlignment := vertical_alignment.Center
                                                    Widget := IconImage_Summon3
                                                stack_box_slot:
                                                    HorizontalAlignment := horizontal_alignment.Center
                                                    VerticalAlignment := vertical_alignment.Center
                                                    Widget := Summon3TextBlock
                                    stack_box_slot:
                                        Distribution:= option{0.2}
                                        HorizontalAlignment := horizontal_alignment.Left
                                        VerticalAlignment := vertical_alignment.Center
                                        Widget := stack_box:
                                            Orientation := orientation.Horizontal
                                            Slots := array:
                                                stack_box_slot:
                                                    HorizontalAlignment := horizontal_alignment.Center
                                                    VerticalAlignment := vertical_alignment.Center
                                                    Widget := IconImage_Summon4
                                                stack_box_slot:
                                                    HorizontalAlignment := horizontal_alignment.Center
                                                    VerticalAlignment := vertical_alignment.Center
                                                    Widget := Summon4TextBlock
                                    stack_box_slot:
                                        Distribution:= option{0.2}
                                        HorizontalAlignment := horizontal_alignment.Left
                                        VerticalAlignment := vertical_alignment.Center
                                        Widget := stack_box:
                                            Orientation := orientation.Horizontal
                                            Slots := array:
                                                stack_box_slot:
                                                    HorizontalAlignment := horizontal_alignment.Center
                                                    VerticalAlignment := vertical_alignment.Center
                                                    Widget := IconImage_Summon5
                                                stack_box_slot:
                                                    HorizontalAlignment := horizontal_alignment.Center
                                                    VerticalAlignment := vertical_alignment.Center
                                                    Widget := Summon5TextBlock
                                    stack_box_slot:
                                        Distribution:= option{0.2}
                                        HorizontalAlignment := horizontal_alignment.Left
                                        VerticalAlignment := vertical_alignment.Center
                                        Widget := stack_box:
                                            Orientation := orientation.Horizontal
                                            Slots := array:
                                                stack_box_slot:
                                                    HorizontalAlignment := horizontal_alignment.Center
                                                    VerticalAlignment := vertical_alignment.Center
                                                    Widget := IconImage_Summon6
                                                stack_box_slot:
                                                    HorizontalAlignment := horizontal_alignment.Center
                                                    VerticalAlignment := vertical_alignment.Center
                                                    Widget := Summon6TextBlock

        #=============================== Stone ===============================

                        stack_box_slot:
                            Distribution:= option{0.17}
                            HorizontalAlignment := horizontal_alignment.Fill
                            VerticalAlignment := vertical_alignment.Fill
                            Widget := stack_box:
                                Orientation := orientation.Vertical
                                Slots := array:
                                    stack_box_slot:
                                        Distribution:= option{0.1}
                                        HorizontalAlignment := horizontal_alignment.Fill
                                        VerticalAlignment := vertical_alignment.Fill
                                        Widget := StoneTextBlock
                                    stack_box_slot:
                                        Distribution:= option{0.2}
                                        HorizontalAlignment := horizontal_alignment.Left
                                        VerticalAlignment := vertical_alignment.Center
                                        Widget := stack_box:
                                            Orientation := orientation.Horizontal
                                            Slots := array:
                                                stack_box_slot:
                                                    Padding:=margin{Right:= 10.0}
                                                    HorizontalAlignment := horizontal_alignment.Center
                                                    VerticalAlignment := vertical_alignment.Center
                                                    Widget := IconImage_Stone1
                                                stack_box_slot:
                                                    HorizontalAlignment := horizontal_alignment.Center
                                                    VerticalAlignment := vertical_alignment.Center
                                                    Widget := Stone1TextBlock
                                    stack_box_slot:
                                        Distribution:= option{0.2}
                                        HorizontalAlignment := horizontal_alignment.Left
                                        VerticalAlignment := vertical_alignment.Center
                                        Widget := stack_box:
                                            Orientation := orientation.Horizontal
                                            Slots := array:
                                                stack_box_slot:
                                                    Padding:=margin{Right:= 10.0}
                                                    HorizontalAlignment := horizontal_alignment.Center
                                                    VerticalAlignment := vertical_alignment.Center
                                                    Widget := IconImage_Stone2
                                                stack_box_slot:
                                                    HorizontalAlignment := horizontal_alignment.Center
                                                    VerticalAlignment := vertical_alignment.Center
                                                    Widget := Stone2TextBlock
                                    stack_box_slot:
                                        Distribution:= option{0.2}
                                        HorizontalAlignment := horizontal_alignment.Left
                                        VerticalAlignment := vertical_alignment.Center
                                        Widget := stack_box:
                                            Orientation := orientation.Horizontal
                                            Slots := array:
                                                stack_box_slot:
                                                    Padding:=margin{Right:= 10.0}
                                                    HorizontalAlignment := horizontal_alignment.Center
                                                    VerticalAlignment := vertical_alignment.Center
                                                    Widget := IconImage_Stone3
                                                stack_box_slot:
                                                    HorizontalAlignment := horizontal_alignment.Center
                                                    VerticalAlignment := vertical_alignment.Center
                                                    Widget := Stone3TextBlock
                                    stack_box_slot:
                                        Distribution:= option{0.2}
                                        HorizontalAlignment := horizontal_alignment.Left
                                        VerticalAlignment := vertical_alignment.Center
                                        Widget := stack_box:
                                            Orientation := orientation.Horizontal
                                            Slots := array:
                                                stack_box_slot:
                                                    Padding:=margin{Right:= 10.0}
                                                    HorizontalAlignment := horizontal_alignment.Center
                                                    VerticalAlignment := vertical_alignment.Center
                                                    Widget := IconImage_Stone4
                                                stack_box_slot:
                                                    HorizontalAlignment := horizontal_alignment.Center
                                                    VerticalAlignment := vertical_alignment.Center
                                                    Widget := Stone4TextBlock
                                    stack_box_slot:
                                        Distribution:= option{0.2}
                                        HorizontalAlignment := horizontal_alignment.Left
                                        VerticalAlignment := vertical_alignment.Center
                                        Widget := stack_box:
                                            Orientation := orientation.Horizontal
                                            Slots := array:
                                                stack_box_slot:
                                                    Padding:=margin{Right:= 10.0}
                                                    HorizontalAlignment := horizontal_alignment.Center
                                                    VerticalAlignment := vertical_alignment.Center
                                                    Widget := IconImage_Stone5
                                                stack_box_slot:
                                                    HorizontalAlignment := horizontal_alignment.Center
                                                    VerticalAlignment := vertical_alignment.Center
                                                    Widget := Stone5TextBlock
                                    stack_box_slot:
                                        Distribution:= option{0.2}
                                        HorizontalAlignment := horizontal_alignment.Left
                                        VerticalAlignment := vertical_alignment.Center
                                        Widget := stack_box:
                                            Orientation := orientation.Horizontal
                                            Slots := array:
                                                stack_box_slot:
                                                    Padding:=margin{Right:= 10.0}
                                                    HorizontalAlignment := horizontal_alignment.Center
                                                    VerticalAlignment := vertical_alignment.Center
                                                    Widget := IconImage_Stone6
                                                stack_box_slot:
                                                    HorizontalAlignment := horizontal_alignment.Center
                                                    VerticalAlignment := vertical_alignment.Center
                                                    Widget := Stone6TextBlock

        #=============================== Teleport_New ===============================

                        stack_box_slot:
                            Distribution:= option{0.2}
                            HorizontalAlignment := horizontal_alignment.Fill
                            VerticalAlignment := vertical_alignment.Fill
                            Widget := stack_box:
                                Orientation := orientation.Vertical
                                Slots := array:
                                    stack_box_slot:
                                        Distribution:= option{0.1}
                                        HorizontalAlignment := horizontal_alignment.Fill
                                        VerticalAlignment := vertical_alignment.Fill
                                        Widget := TeleportTextBlock
                                    stack_box_slot:
                                        Distribution:= option{0.2}
                                        HorizontalAlignment := horizontal_alignment.Left
                                        VerticalAlignment := vertical_alignment.Center
                                        Widget := stack_box:
                                            Orientation := orientation.Horizontal
                                            Slots := array:
                                                stack_box_slot:
                                                    HorizontalAlignment := horizontal_alignment.Center
                                                    VerticalAlignment := vertical_alignment.Center
                                                    Widget := QuietButtonTeleport1
                                    stack_box_slot:
                                        Distribution:= option{0.2}
                                        HorizontalAlignment := horizontal_alignment.Left
                                        VerticalAlignment := vertical_alignment.Center
                                        Widget := stack_box:
                                            Orientation := orientation.Horizontal
                                            Slots := array:
                                                stack_box_slot:
                                                    HorizontalAlignment := horizontal_alignment.Center
                                                    VerticalAlignment := vertical_alignment.Center
                                                    Widget := QuietButtonTeleport2
                                    stack_box_slot:
                                        Distribution:= option{0.2}
                                        HorizontalAlignment := horizontal_alignment.Left
                                        VerticalAlignment := vertical_alignment.Center
                                        Widget := stack_box:
                                            Orientation := orientation.Horizontal
                                            Slots := array:
                                                stack_box_slot:
                                                    HorizontalAlignment := horizontal_alignment.Center
                                                    VerticalAlignment := vertical_alignment.Center
                                                    Widget := QuietButtonTeleport3
                                    stack_box_slot:
                                        Distribution:= option{0.2}
                                        HorizontalAlignment := horizontal_alignment.Left
                                        VerticalAlignment := vertical_alignment.Center
                                        Widget := stack_box:
                                            Orientation := orientation.Horizontal
                                            Slots := array:
                                                stack_box_slot:
                                                    HorizontalAlignment := horizontal_alignment.Center
                                                    VerticalAlignment := vertical_alignment.Center
                                                    Widget := QuietButtonTeleport4
                                    stack_box_slot:
                                        Distribution:= option{0.2}
                                        HorizontalAlignment := horizontal_alignment.Left
                                        VerticalAlignment := vertical_alignment.Center
                                        Widget := stack_box:
                                            Orientation := orientation.Horizontal
                                            Slots := array:
                                                stack_box_slot:
                                                    HorizontalAlignment := horizontal_alignment.Center
                                                    VerticalAlignment := vertical_alignment.Center
                                                    Widget := QuietButtonTeleport5
                                    stack_box_slot:
                                        Distribution:= option{0.2}
                                        HorizontalAlignment := horizontal_alignment.Left
                                        VerticalAlignment := vertical_alignment.Center
                                        Widget := stack_box:
                                            Orientation := orientation.Horizontal
                                            Slots := array:
                                                stack_box_slot:
                                                    HorizontalAlignment := horizontal_alignment.Center
                                                    VerticalAlignment := vertical_alignment.Center
                                                    Widget := QuietButtonTeleport6

        <#=============================== Teleport_Old ===============================
                        stack_box_slot:
                            Distribution:= option{0.2}
                            HorizontalAlignment := horizontal_alignment.Fill
                            VerticalAlignment := vertical_alignment.Fill
                            Widget := stack_box:
                                Orientation := orientation.Vertical
                                Slots := array:
                                    stack_box_slot:
                                        Distribution:= option{0.1}
                                        HorizontalAlignment := horizontal_alignment.Fill
                                        VerticalAlignment := vertical_alignment.Fill
                                        Widget := TeleportTextBlock
                                    stack_box_slot:
                                        Distribution:= option{0.2}
                                        HorizontalAlignment := horizontal_alignment.Center
                                        VerticalAlignment := vertical_alignment.Center
                                        Widget := QuietButtonTeleport1
                                    stack_box_slot:
                                        Distribution:= option{0.2}
                                        HorizontalAlignment := horizontal_alignment.Center
                                        VerticalAlignment := vertical_alignment.Center
                                        Widget := QuietButtonTeleport2
                                    stack_box_slot:
                                        Distribution:= option{0.2}
                                        HorizontalAlignment := horizontal_alignment.Center
                                        VerticalAlignment := vertical_alignment.Center
                                        Widget := QuietButtonTeleport3
                                    stack_box_slot:
                                        Distribution:= option{0.2}
                                        HorizontalAlignment := horizontal_alignment.Center
                                        VerticalAlignment := vertical_alignment.Center
                                        Widget := QuietButtonTeleport4
                                    stack_box_slot:
                                        Distribution:= option{0.2}
                                        HorizontalAlignment := horizontal_alignment.Center
                                        VerticalAlignment := vertical_alignment.Center
                                        Widget := QuietButtonTeleport5
                                    stack_box_slot:
                                        Distribution:= option{0.2}
                                        HorizontalAlignment := horizontal_alignment.Center
                                        VerticalAlignment := vertical_alignment.Center
                                        Widget := QuietButtonTeleport6
                                        #>
        }
        
        #========================================== Close ==========================================
        
        CloseButton : button_loud = button_loud{DefaultText := StringToMessage("CLOSE")}
        CloseButton.OnClick().Subscribe(OnCloseButtonClicked)
        set CanvasSlots += array
        {
            canvas_slot:
                Anchors := CenterAnchors
                Offsets := margin{Top:= 467.5, Left:= 738.0, Right := 353.0, Bottom := 80.0}
                Alignment := CenterVector
                SizeToContent := false
                Widget := CloseButton
        }

        # if(set StageUIElements[Agent] = map{}){}
        # if:
        #     set StageUIElements[Agent] = map:
        #         1 => (QuietButtonTeleport1,BossImage_1)
        #         2 => (QuietButtonTeleport2,BossImage_2)
        #         3 => (QuietButtonTeleport3,BossImage_3)
        #         4 => (QuietButtonTeleport4,BossImage_4)
        #         5 => (QuietButtonTeleport1,BossImage_5)
        #         6 => (QuietButtonTeleport2,BossImage_6)

        # for(StageElement : StageUIElements[Agent]):
        #     StageElement(0).SetVisibility(widget_visibility.Collapsed)
        #     StageElement(1).SetVisibility(widget_visibility.Collapsed):
                #   var BossUIElements : [agent][int]tuple(texture_block, texture_block, texture_block, texture_block, button_quiet, text_block, text_block, text_block) = map{}
        
        if(set BossUIElements[Agent] = map{}){}
        if:
            set BossUIElements[Agent] = map:
                1 => (BossImage_1, IconImage_Power1, IconImage_Summon1, IconImage_Stone1, QuietButtonTeleport1, BossName1TextBlock, PowerUp1TextBlock, Summon1TextBlock, Stone1TextBlock)
                2 => (BossImage_2, IconImage_Power2, IconImage_Summon2, IconImage_Stone2, QuietButtonTeleport2, BossName2TextBlock, PowerUp2TextBlock, Summon2TextBlock, Stone2TextBlock)
                3 => (BossImage_3, IconImage_Power3, IconImage_Summon3, IconImage_Stone3, QuietButtonTeleport3, BossName3TextBlock, PowerUp3TextBlock, Summon3TextBlock, Stone3TextBlock)
                4 => (BossImage_4, IconImage_Power4, IconImage_Summon4, IconImage_Stone4, QuietButtonTeleport4, BossName4TextBlock, PowerUp4TextBlock, Summon4TextBlock, Stone4TextBlock)
                5 => (BossImage_5, IconImage_Power5, IconImage_Summon5, IconImage_Stone5, QuietButtonTeleport5, BossName5TextBlock, PowerUp5TextBlock, Summon5TextBlock, Stone5TextBlock)
                6 => (BossImage_6, IconImage_Power6, IconImage_Summon6, IconImage_Stone6, QuietButtonTeleport6, BossName6TextBlock, PowerUp6TextBlock, Summon6TextBlock, Stone6TextBlock)

        for(BossElement : BossUIElements[Agent]):
            BossElement(0).SetVisibility(widget_visibility.Collapsed)
            BossElement(1).SetVisibility(widget_visibility.Collapsed)
            BossElement(2).SetVisibility(widget_visibility.Collapsed)
            BossElement(3).SetVisibility(widget_visibility.Collapsed)
            BossElement(4).SetVisibility(widget_visibility.Collapsed)
            BossElement(5).SetVisibility(widget_visibility.Collapsed)
            BossElement(6).SetVisibility(widget_visibility.Collapsed)
            BossElement(7).SetVisibility(widget_visibility.Collapsed)
            BossElement(8).SetVisibility(widget_visibility.Collapsed)

        Canvas : canvas = canvas:
            Slots := CanvasSlots

        return Canvas
