using { /Fortnite.com/Devices }
using { /Verse.org/Simulation }
using { /Fortnite.com/UI }
using { /UnrealEngine.com/Temporary/UI }
using { /Verse.org/Colors }
using { /UnrealEngine.com/Temporary/SpatialMath }

using {Mining_System.Code.MiningObject}
using {Mining_System.Code.PlayerData}
using {Mining_System.Code.Pet}
using {Mining_System.Code.Skill.Textures}
using {icon_set_2}

ui_hud_count<public> := class(creative_device):


    @editable
    SaveDataManager : save_data_manager = save_data_manager{}
    @editable
    MiningLevelDevice : mining_level_device = mining_level_device{}
    @editable
    MiningDamageDevice : mining_damage_device = mining_damage_device{}
    @editable
    PetManagerDevice : pet_manager = pet_manager{}

    #XP付与
    @editable XP_KaijuBuff : accolades_device = accolades_device{}
    @editable XP_GotCoin : accolades_device = accolades_device{}
    @editable Collectibles : []collectible_object_device = array{}
    @editable CoinMinutes : int = 8

    @editable PC_Idle : player_counter_device = player_counter_device{}
    @editable Volume_StrBuff : volume_device = volume_device{}
    @editable DefaultStrBuffTime<public>: int = 240

    var WidgetPerPlayer : [player]ui_hud_count_widget = map{}
    var BuffWidgetPerPlayer : [player]ui_buff = map{}
    var IdleWidgetPerPlayer : [player]ui_idle = map{}

    var IsUpgradeHud : [player]logic = map{}
    var IsBuffHud : [player]logic = map{}
    var StrBuffTime: [player]int = map{}
    var LowestValue : big_number = big_number{}

    StringToMessage<localizes>(Text:string):message="{Text}"

    OnBegin<override>():void=
        for(Player : GetPlayspace().GetPlayers()):
            OnPlay(Player)

        GetPlayspace().PlayerAddedEvent().Subscribe(OnPlay)
        GetPlayspace().PlayerRemovedEvent().Subscribe(OnPlayerRemoved)
        spawn{WidgetLoop()}

        Volume_StrBuff.AgentEntersEvent.Subscribe(OnEnterVolume)

        PC_Idle.CountedEvent.Subscribe(Update_IdleUI)
        PC_Idle.RemovedEvent.Subscribe(Update_IdleUI)

        for(TheCollectible : Collectibles) :
            spawn{AwaitGetCoin(TheCollectible)}

    RemoveKeyFromWidgetMap(ExampleMap:[player]ui_hud_count_widget, ElementToRemove:player):[player]ui_hud_count_widget=
        var NewMap:[player]ui_hud_count_widget = map{}
        for (Key -> Value : ExampleMap, Key <> ElementToRemove):
            set NewMap = ConcatenateMaps(NewMap, map{Key => Value})
        return NewMap

    OnPlayerRemoved(Player : player):void=
        if(PlayerUI := GetPlayerUI[Player], Widget := WidgetPerPlayer[Player]):
            PlayerUI.RemoveWidget(Widget.CreateCanvas())
            
            # WidgetPerPlayerを更新
            set WidgetPerPlayer = RemoveKeyFromWidgetMap(WidgetPerPlayer, Player)

    AwaitGetCoin(Coin : collectible_object_device)<suspends>:void=
        loop :
            Agent := Coin.CollectedEvent.Await()
            XP_GotCoin.Award(Agent)
            spawn{RemakeCoin(Agent, Coin)}

    RemakeCoin(Agent : agent, Coin : collectible_object_device)<suspends>:void=
        Sleep(60.0 * CoinMinutes)
        Coin.Respawn(Agent)
        Print("Coin Respawned")

    # OnPlay(Player:player):void=
    #     if(Widget := WidgetPerPlayer[Player]) :
    #         return #ウィジェット作成済みなら何もしない
    #     if(PlayerUI := GetPlayerUI[Player]):
    #         NewUI := ui_hud_count_widget{}
    #         PlayerUI.AddWidget(NewUI.CreateCanvas())
    #         logic{set WidgetPerPlayer[Player] = NewUI}

    OnPlay(Player : player):void=
        spawn { EnsureUIForPlayer(Player) }
    
    EnsureUIForPlayer(Player : player)<suspends>:void=
        Sleep(1.0)  # 遅延を挿入
        if(not WidgetPerPlayer[Player]):
            if(PlayerUI := GetPlayerUI[Player]):
                NewUI := ui_hud_count_widget{}
                PlayerUI.AddWidget(NewUI.CreateCanvas())
                logic{set WidgetPerPlayer[Player] = NewUI}

    CheckBuffWidget(Player : player):void=
        if(Widget := BuffWidgetPerPlayer[Player]) :
            return #ウィジェット作成済みなら何もしない
        if(PlayerUI : player_ui = GetPlayerUI[Player]) :
            NewUI := ui_buff{}
            PlayerUI.AddWidget(NewUI.CreateBuffCanvas())
            if(set BuffWidgetPerPlayer[Player] = NewUI){}
            #最初に呼び出したときに必要なら特定のウィジェットを非表示
            if(Element := BuffWidgetPerPlayer[Player]) :
                #Buff非アクティブ時
                if(not IsBuffHud[Player]) :
                    Element.BG_Buff.SetVisibility(widget_visibility.Collapsed)
                    Element.Icon_Buff.SetVisibility(widget_visibility.Collapsed)
                    Element.Text_Buff.SetVisibility(widget_visibility.Collapsed)
                #StrBuff非アクティブ時
                if( not StrBuffTime[Player]) :
                    Element.BG_StrBuff.SetVisibility(widget_visibility.Collapsed)
                    Element.Icon_StrBuff.SetVisibility(widget_visibility.Collapsed)
                    Element.Icon_StrBuff2.SetVisibility(widget_visibility.Collapsed)
                    Element.Text_StrBuff.SetVisibility(widget_visibility.Collapsed)

    CheckIdleWidget(Player : player):void=
        if(Widget := IdleWidgetPerPlayer[Player]) :
            return #ウィジェット作成済みなら何もしない
        if(PlayerUI : player_ui = GetPlayerUI[Player]) :
            NewUI : ui_idle = ui_idle{}
            PlayerUI.AddWidget(NewUI.CreateIdleCanvas())
            logic{set IdleWidgetPerPlayer[Player] = NewUI}

    WidgetLoop()<suspends>:void=
        loop:
            for(Player -> Element : WidgetPerPlayer):
                if (SaveDataManager.HasSaveData(Player)?):
                    StoneCount := SaveDataManager.GetResourceStone(Player)
                    Element.StoneCount.SetText(StringToMessage(BigNumber_FormatDisplay(StoneCount)))

                    TicketCount := SaveDataManager.GetNowResourceTicket(Player)
                    Element.IconTicketValueLeft.SetText(StringToMessage("{TicketCount}"))
                    #ResourceTicketTimer更新
                    TicketTimer := SaveDataManager.GetResourceTicket_Revival_Times(Player)
                    TicketMaxCount := SaveDataManager.GetMaxResourceTicket(Player)
                    if(TicketCount < TicketMaxCount) :
                        Element.IconTicketValueRight.SetTextColor(NamedColors.White)
                        Element.IconTicketValueRight.SetText(StringToMessage(TimeToString(TicketTimer)))
                    else :
                        Element.IconTicketValueRight.SetTextColor(NamedColors.Orange)
                        Element.IconTicketValueRight.SetText(StringToMessage("MAX"))
                    
                    BossCount := SaveDataManager.GetNowTicket(Player)
                    Element.IconBossValueLeft.SetText(StringToMessage("{BossCount}"))
                    #BossTimer更新
                    BossTimer := SaveDataManager.GetBossTicket_Revival_Times(Player)
                    BossMaxCount := SaveDataManager.GetMaxTicket(Player)
                    if(BossCount < BossMaxCount) :
                        Element.IconBossValueRight.SetTextColor(NamedColors.White)
                        Element.IconBossValueRight.SetText(StringToMessage(TimeToString(BossTimer)))
                    else :
                        Element.IconBossValueRight.SetTextColor(NamedColors.Orange)
                        Element.IconBossValueRight.SetText(StringToMessage("Ready"))

                    spawn{Update_LevelupUI(Player)} #LevelupUI
                else:
                    Print("Error: WidgetLoop")
            Sleep(1.0)

    # UpdateLastHit<public>(Player : player)<decides>:void=
    #     var PetLastHit : float = BigNumber_ToFloat(PetManager.GetEquipedPetLastHit(Player))
    #     LastHitTextBlock := StringToMessage("LastHit: +{FloatToStringWithPrecision(0.5 + PetLastHit + SaveDataManager.GetSkillLevel[Player , 17] * 0.2, 0)}%")

    TimeToString(Time : int) : string =
        var TimerString : string = ""
        var Minutes : int = 0
        var Seconds : int = 0
        FloatTime := Time * 1.0
        if(Value : int = Floor[FloatTime / 60.0]) :
            set Minutes = Value
            set Seconds = Time - Value * 60
            if(Seconds >= 10) :
                set TimerString = "{Minutes}:{Seconds}"
            else :
                set TimerString = "{Minutes}:0{Seconds}"

        return TimerString

    Update_LevelupUI(Player : player)<suspends>:void= #LevelupUIの表示/非表示
        if(not IsUpgradeHud[Player]) :
            logic{set IsUpgradeHud[Player] = true}
            Print("UpgradeHud初期化")
        if(Element := WidgetPerPlayer[Player]) :
            if(MiningLevelDevice.CanLevelUp[Player]):
                if(IsUpgradeHud[Player] = false) :
                    Element.LevelupUI.SetVisibility(widget_visibility.Visible)
                    logic{set IsUpgradeHud[Player] = true}
                    Print("UpgradeHud表示")
            else :
                if(IsUpgradeHud[Player] = true) :
                    Element.LevelupUI.SetVisibility(widget_visibility.Collapsed)
                    logic{set IsUpgradeHud[Player] = false}
                    Print("UpgradeHud非表示")
        else :
            Print("Upgrade表示失敗")

    Update_BuffUI<public>(Agent : agent, Time : int, IsBuffing : logic)<suspends>:void=
        if(Player : player = player[Agent]) :
            CheckBuffWidget(Player)
            #初期化
            if(not IsBuffHud[Player]) :
                logic{set IsBuffHud[Player] = false}

            #UI更新(表示, 非表示)の処理
            if(Element := BuffWidgetPerPlayer[Player]) :
                if(IsBuffing = true) :
                    Element.Text_Buff.SetText(StringToMessage(TimeToString(Time)))
                    Print("BuffHud更新")

                    if(IsBuffHud[Player] = false) :
                        logic{set IsBuffHud[Player] = true}
                        Element.BG_Buff.SetVisibility(widget_visibility.Visible)
                        Element.Icon_Buff.SetVisibility(widget_visibility.Visible)
                        Element.Text_Buff.SetVisibility(widget_visibility.Visible)
                        Print("BuffHud表示")

                else if(IsBuffing = false) :
                    logic{set IsBuffHud[Player] = false}
                    Element.BG_Buff.SetVisibility(widget_visibility.Collapsed)
                    Element.Icon_Buff.SetVisibility(widget_visibility.Collapsed)
                    Element.Text_Buff.SetVisibility(widget_visibility.Collapsed)
                    Print("BuffHud非表示")
            else :
                Print("Buff表示失敗")

    OnEnterVolume(Agent : agent):void=
        if(Player : player = player[Agent]) :
            if(not StrBuffTime[Player] or StrBuffTime[Player] <= 0) :
                XP_KaijuBuff.Award(Agent)
        spawn{Update_StrBuffUI(Agent)}

    Update_StrBuffUI<public>(Agent : agent)<suspends>:void=
        Sleep(0.1)
        if(Player : player = player[Agent]) :
            logic{set StrBuffTime[Player] = DefaultStrBuffTime}
            CheckBuffWidget(Player)
            if(Element := BuffWidgetPerPlayer[Player]) :
                Element.BG_StrBuff.SetVisibility(widget_visibility.Visible)
                Element.Icon_StrBuff.SetVisibility(widget_visibility.Visible)
                Element.Icon_StrBuff2.SetVisibility(widget_visibility.Visible)
                Element.Text_StrBuff.SetText(StringToMessage(TimeToString(DefaultStrBuffTime)))
                Element.Text_StrBuff.SetVisibility(widget_visibility.Visible)
                SaveDataManager.OnTurnOn(Player)
                race :
                    block :
                        Sleep(DefaultStrBuffTime * 1.0)
                    block :
                        loop :
                            EnterAgent := Volume_StrBuff.AgentEntersEvent.Await()
                            if(Agent = EnterAgent) :
                                Print("StrBuff break", ?Color := NamedColors.Red, ?Duration := 3.0)
                                break
                    block :
                        loop :
                            Sleep(1.0)
                            logic{set StrBuffTime[Player] -= 1}
                            if(Time := StrBuffTime[Player]) :
                                if(Time >= 0) :
                                    Element.Text_StrBuff.SetText(StringToMessage(TimeToString(Time)))
                                #タイマー時間切れのときの処理
                                else :
                                    break
                Element.BG_StrBuff.SetVisibility(widget_visibility.Collapsed)
                Element.Icon_StrBuff.SetVisibility(widget_visibility.Collapsed)
                Element.Icon_StrBuff2.SetVisibility(widget_visibility.Collapsed)
                Element.Text_StrBuff.SetVisibility(widget_visibility.Collapsed)
                SaveDataManager.OnTurnOff(Player)


    Update_IdleUI<public>(Agent : agent):void=
        set LowestValue = big_number{Value := 1.0, Exponent := 0}
        if :
            Player : player = player[Agent]
        then :
            CheckIdleWidget(Player)
            if(Element := IdleWidgetPerPlayer[Player]) :
                # ダメージ計算
                DamageAmount := MiningDamageDevice.GetNormalDamageAmount(Player).Amount
                GatherRate := BigNumber_MultiplyByFloat(PetManagerDevice.GetPetGatherRateFromRestSlots(Player), 0.01) # 100分率を実際の式に書き直してる
                PlayerIdle := SaveDataManager.GetPlayeridle(Player) * 0.05 + 0.1
                PetIdle := SaveDataManager.GetPetidle(Player) * 0.05 + 0.1
                ResultPetDamageAmount := BigNumber_MultiplyByFloat(BigNumber_MultiplyByBigNumber(DamageAmount, GatherRate), PetIdle)
                ResultPlayerDamageAmount := BigNumber_MultiplyByFloat(DamageAmount, PlayerIdle)
                
                AllAmount := BigNumber_Add(ResultPetDamageAmount, ResultPlayerDamageAmount)

                #休憩所の中ならペットとプレイヤーのダメージ総量を表示
                if(PC_Idle.IsCounted[Agent]) :
                    IdleCountText := "+ {BigNumber_FormatDisplay(AllAmount)}/s"

                    Element.IdleCount.SetVisibility(widget_visibility.Visible)
                    Element.IdleCount.SetText(StringToMessage(IdleCountText))
                    Print("Rest Widget表示", ?Color := NamedColors.Red, ?Duration := 3.0)
                #休憩所の外ならペットのダメージが1以上のときのみペットのダメージを表示
                else :
                    Print("PetDamage{BigNumber_FormatDisplay(ResultPetDamageAmount)}", ?Color := NamedColors.Red, ?Duration := 3.0)
                    if(BigNumber_GreaterThanOrEquals(ResultPetDamageAmount, LowestValue)?) :
                        IdleCountText := "+ {BigNumber_FormatDisplay(ResultPetDamageAmount)}/s"

                        Element.IdleCount.SetVisibility(widget_visibility.Visible)
                        Element.IdleCount.SetText(StringToMessage(IdleCountText))
                        Print("Rest Widget表示, ペットのみ", ?Color := NamedColors.Red, ?Duration := 3.0)
                    #休憩所の外でペットのダメージが0なら非表示
                    else :
                        Element.IdleCount.SetVisibility(widget_visibility.Collapsed)
                        Print("Rest Widget非表示", ?Color := NamedColors.Red, ?Duration := 3.0)
            else :
                Print("Rest Widget取得失敗", ?Color := NamedColors.Red, ?Duration := 3.0)

ui_hud_count_widget := class:
    #UpgradeHud
    LevelupUI : texture_block = texture_block: # LevelupUI
        DefaultImage := icon_set_2.Levelup_UI
        DefaultTint := color{R := 1.0, G := 1.0, B := 1.0}
        DefaultDesiredSize := vector2{X := 287.86438,Y := 51.249863}
    #通常Hudの背景
    BG_Hud : texture_block = texture_block :
        DefaultImage := BG_s
        DefaultTint := color{R := 1.0, G := 1.0, B := 1.0}
        DefaultDesiredSize := vector2{X := 262.075836, Y := 59.817024}
    BG_Hud2 : texture_block = texture_block :
        DefaultImage := BG_s
        DefaultTint := color{R := 1.0, G := 1.0, B := 1.0}
        DefaultDesiredSize := vector2{X := 262.075836, Y := 59.817024}
    BG_Hud3 : texture_block = texture_block :
        DefaultImage := BG_s
        DefaultTint := color{R := 1.0, G := 1.0, B := 1.0}
        DefaultDesiredSize := vector2{X := 262.075836, Y := 59.817024}
    #通常のHud
    EmptyText : text_block = text_block: # LevelupUIが非表示のとき用の空白(text_block)
        DefaultTextColor := color{R := 1.0, G := 1.0, B := 1.0}
        DefaultTextOpacity := 1.0
        DefaultShadowOffset := option{vector2{X := 0.0,Y := 0.0}}
        DefaultShadowColor := color{R := 1.0, G := 1.0, B := 1.0}
        DefaultShadowOpacity := 1.0

    IconStone : texture_block = texture_block: # IconStone
        DefaultImage := icon_set_2.SilverOre
        DefaultTint := color{R := 1.0, G := 1.0, B := 1.0}
        DefaultDesiredSize := vector2{X := 50.0,Y := 50.0}
    
    StoneCount : text_block = text_block: # StoneCount
        DefaultTextColor := color{R := 1.0, G := 1.0, B := 1.0}
        DefaultTextOpacity := 1.0
        DefaultShadowOffset := option{vector2{X := 2.361355,Y := 2.830423}}
        DefaultShadowColor := color{R := 0.0, G := 0.0, B := 0.0}
        DefaultShadowOpacity := 0.64
    
    IconTicket : texture_block = texture_block: # IconTicket
        DefaultImage := IC_Ticket
        DefaultTint := color{R := 1.0, G := 1.0, B := 1.0}
        DefaultDesiredSize := vector2{X := 50.0,Y := 50.0}
    
    IconTicketValueLeft : text_block = text_block: # IconTicketValueLeft
        DefaultTextColor := color{R := 1.0, G := 1.0, B := 1.0}
        DefaultTextOpacity := 1.0
        DefaultShadowOffset := option{vector2{X := 2.361355,Y := 2.830423}}
        DefaultShadowColor := color{R := 0.0, G := 0.0, B := 0.0}
        DefaultShadowOpacity := 0.64
    
    IconTicketValueRight : text_block = text_block: # IconTicketValueRight
        DefaultTextColor := color{R := 1.0, G := 1.0, B := 1.0}
        DefaultTextOpacity := 1.0
        DefaultShadowOffset := option{vector2{X := 2.361355,Y := 2.830423}}
        DefaultShadowColor := color{R := 0.0, G := 0.0, B := 0.0}
        DefaultShadowOpacity := 0.64
    
    IconBoss : texture_block = texture_block: # IconBoss
        DefaultImage := IC_Boss
        DefaultTint := color{R := 1.0, G := 1.0, B := 1.0}
        DefaultDesiredSize := vector2{X := 50.0,Y := 50.0}
    
    IconBossValueLeft : text_block = text_block: # IconBossValueLeft
        DefaultTextColor := color{R := 1.0, G := 1.0, B := 1.0}
        DefaultTextOpacity := 1.0
        DefaultShadowOffset := option{vector2{X := 2.361355,Y := 2.830423}}
        DefaultShadowColor := color{R := 0.0, G := 0.0, B := 0.0}
        DefaultShadowOpacity := 0.64
    
    IconBossValueRight : text_block = text_block: # IconBossValueRight
        DefaultTextColor := color{R := 1.0, G := 1.0, B := 1.0}
        DefaultTextOpacity := 1.0
        DefaultShadowOffset := option{vector2{X := 2.361355,Y := 2.830423}}
        DefaultShadowColor := color{R := 0.0, G := 0.0, B := 0.0}
        DefaultShadowOpacity := 0.64
        
    LastHitTexture : texture_block = texture_block{
        DefaultImage := icon_set_2.IC_LastHit,
        DefaultDesiredSize := vector2{X := 40.0, Y := 40.0}
    }     
    LastHitTextBlock : text_block = text_block{
        DefaultTextColor := color{R := 1.0, G := 1.0, B := 1.0}
        DefaultTextOpacity := 1.0
        DefaultShadowOffset := option{vector2{X := 2.36, Y := 2.83}}
        DefaultShadowColor := color{R := 0.0, G := 0.0, B := 0.0}
        DefaultShadowOpacity := 0.64
        DefaultJustification := text_justification.Center
    }


    CreateCanvas<public>():canvas=
        CenterVector : vector2 = vector2{X := 0.5, Y := 0.5}
        CenterAnchors : anchors = anchors{Minimum := CenterVector, Maximum := CenterVector}
        canvas:
            Slots := array:
                canvas_slot:
                    Anchors := anchors:
                        Minimum := vector2{X := 0.01, Y := 0.5}
                        Maximum := vector2{X := 0.01, Y := 0.5}
                    Offsets := margin{Top := 138.670349, Left := 17.278419, Right := 0.0, Bottom := 0.0}
                    SizeToContent := true
                    Alignment := vector2{X := 0.0, Y := 0.0}
                    Widget := stack_box :
                        Orientation := orientation.Vertical
                        Slots := array:
                            stack_box_slot: # StackBoxSlot_0
                                HorizontalAlignment := horizontal_alignment.Fill
                                VerticalAlignment := vertical_alignment.Fill
                                Padding := margin:
                                    Top := 0.0
                                    Right := 0.0
                                    Bottom := 0.0
                                    Left := 0.0
                                Widget := BG_Hud
                            stack_box_slot: # StackBoxSlot_1
                                HorizontalAlignment := horizontal_alignment.Fill
                                VerticalAlignment := vertical_alignment.Fill
                                Padding := margin:
                                    Top := 0.0
                                    Right := 0.0
                                    Bottom := 0.0
                                    Left := 0.0
                                Widget := BG_Hud2
                            stack_box_slot: # StackBoxSlot_2
                                HorizontalAlignment := horizontal_alignment.Fill
                                VerticalAlignment := vertical_alignment.Fill
                                Padding := margin:
                                    Top := 0.0
                                    Right := 0.0
                                    Bottom := 0.0
                                    Left := 0.0
                                Widget := BG_Hud3
                canvas_slot:
                    Anchors := anchors:
                        Minimum := vector2{X := 0.01, Y := 0.5}
                        Maximum := vector2{X := 0.01, Y := 0.5}
                    Offsets := margin{Top := 205.236923,Left := 141.402542, Right := 242.33374, Bottom := 231.755096}
                    SizeToContent := false
                    Alignment := vector2{X := 0.5, Y := 0.5}
                    Widget := GetStackbox()
                # canvas_slot:
                #     Anchors := CenterAnchors
                #     Offsets := margin{Top := -405.0, Left := 0.0, Right := 276.0, Bottom := 40.0}
                #     Alignment := CenterVector
                #     SizeToContent := false
                #     Widget := stack_box:
                #         Orientation := orientation.Horizontal
                #         Slots := array:
                #             stack_box_slot:
                #                 HorizontalAlignment := horizontal_alignment.Center
                #                 VerticalAlignment := vertical_alignment.Center
                #                 Widget := LastHitTexture
                #             stack_box_slot:
                #                 HorizontalAlignment := horizontal_alignment.Center
                #                 VerticalAlignment := vertical_alignment.Center
                #                 Padding := margin{Left :=5.0}
                #                 Widget := LastHitTextBlock

    GetStackbox():stack_box=
        stack_box: # StackBox
            Orientation := orientation.Vertical
            Slots := array:
                #LevelupHud
                stack_box_slot: # StackBoxSlot_-1
                    Distribution:= option{1.0}
                    HorizontalAlignment := horizontal_alignment.Fill
                    VerticalAlignment := vertical_alignment.Fill
                    Padding := margin:
                        Top := 0.0
                        Right := 0.0
                        Bottom := 0.0
                        Left := 0.0
                    Widget := stack_box: # StackBox_276
                        Orientation := orientation.Horizontal
                        Slots := array:
                            stack_box_slot: # StackBoxSlot_0
                                HorizontalAlignment := horizontal_alignment.Left
                                VerticalAlignment := vertical_alignment.Center
                                Padding := margin:
                                    Top := 0.0
                                    Right := 0.0
                                    Bottom := 0.0
                                    Left := 0.0
                                Widget := LevelupUI
                            stack_box_slot: # StackBoxSlot_1
                                HorizontalAlignment := horizontal_alignment.Right
                                VerticalAlignment := vertical_alignment.Center
                                Padding := margin:
                                    Top := 0.0
                                    Right := 0.0
                                    Bottom := 0.0
                                    Left := 0.0
                                Widget := EmptyText
                #リソースカウンター
                stack_box_slot: # StackBoxSlot_0
                    Distribution:= option{1.2}
                    HorizontalAlignment := horizontal_alignment.Fill
                    VerticalAlignment := vertical_alignment.Fill
                    Padding := margin:
                        Top := 0.0
                        Right := 0.0
                        Bottom := 0.0
                        Left := 0.0
                    Widget := stack_box: # StackBox_276
                        Orientation := orientation.Horizontal
                        Slots := array:
                            stack_box_slot: # StackBoxSlot_0
                                Distribution:= option{1.0}
                                HorizontalAlignment := horizontal_alignment.Left
                                VerticalAlignment := vertical_alignment.Center
                                Padding := margin:
                                    Top := 1.0
                                    Right := 0.0
                                    Bottom := 1.0
                                    Left := 10.0
                                Widget := IconStone
                            stack_box_slot: # StackBoxSlot_1
                                Distribution:= option{1.0}
                                HorizontalAlignment := horizontal_alignment.Right
                                VerticalAlignment := vertical_alignment.Center
                                Padding := margin:
                                    Top := 0.0
                                    Right := 0.0
                                    Bottom := 0.0
                                    Left := 0.0
                                Widget := StoneCount
                #チケット       
                stack_box_slot: # StackBoxSlot_1
                    Distribution:= option{1.2}
                    HorizontalAlignment := horizontal_alignment.Fill
                    VerticalAlignment := vertical_alignment.Fill
                    Padding := margin:
                        Top := 0.0
                        Right := 0.0
                        Bottom := 0.0
                        Left := 0.0
                    Widget := stack_box: # StackBox_2
                        Orientation := orientation.Horizontal
                        Slots := array:
                            stack_box_slot: # StackBoxSlot_0
                                Distribution:= option{1.0}
                                HorizontalAlignment := horizontal_alignment.Left
                                VerticalAlignment := vertical_alignment.Center
                                Padding := margin:
                                    Top := 1.0
                                    Right := 0.0
                                    Bottom := 1.0
                                    Left := 10.0
                                Widget := IconTicket
                            stack_box_slot: # StackBoxSlot_1
                                Distribution:= option{1.0}
                                HorizontalAlignment := horizontal_alignment.Left
                                VerticalAlignment := vertical_alignment.Center
                                Padding := margin:
                                    Top := 0.0
                                    Right := 0.0
                                    Bottom := 0.0
                                    Left := - 20.0
                                Widget := IconTicketValueLeft
                            stack_box_slot: # StackBoxSlot_2
                                Distribution:= option{1.0}
                                HorizontalAlignment := horizontal_alignment.Right
                                VerticalAlignment := vertical_alignment.Center
                                Padding := margin:
                                    Top := 0.0
                                    Right := 0.0
                                    Bottom := 0.0
                                    Left := - 10.0
                                Widget := IconTicketValueRight
                #ボス挑戦権
                stack_box_slot: # StackBoxSlot_2
                    Distribution:= option{1.2}
                    HorizontalAlignment := horizontal_alignment.Fill
                    VerticalAlignment := vertical_alignment.Fill
                    Padding := margin:
                        Top := 0.0
                        Right := 0.0
                        Bottom := 0.0
                        Left := 0.0
                    Widget := stack_box: # StackBox_3
                        Orientation := orientation.Horizontal
                        Slots := array:
                            stack_box_slot: # StackBoxSlot_0
                                Distribution:= option{1.0}
                                HorizontalAlignment := horizontal_alignment.Left
                                VerticalAlignment := vertical_alignment.Center
                                Padding := margin:
                                    Top := 1.0
                                    Right := 0.0
                                    Bottom := 1.0
                                    Left := 10.0
                                Widget := IconBoss
                            stack_box_slot: # StackBoxSlot_1
                                Distribution:= option{1.0}
                                HorizontalAlignment := horizontal_alignment.Left
                                VerticalAlignment := vertical_alignment.Center
                                Padding := margin:
                                    Top := 0.0
                                    Right := 0.0
                                    Bottom := 0.0
                                    Left := - 20.0
                                Widget := IconBossValueLeft
                            stack_box_slot: # StackBoxSlot_2
                                Distribution:= option{1.0}
                                HorizontalAlignment := horizontal_alignment.Right
                                VerticalAlignment := vertical_alignment.Center
                                Padding := margin:
                                    Left := - 10.0
                                    Top := 0.0
                                    Right := 0.0
                                    Bottom := 0.0
                                Widget := IconBossValueRight
                            
ui_buff := class :
    #Buff
    BG_Buff : texture_block = texture_block:
        DefaultImage := BG_s
        DefaultTint := color{R:=0.0,G:=0.530973,B:=1.0}
        DefaultDesiredSize := vector2{X := 100.0, Y := 100.0}

    Icon_Buff : texture_block = texture_block:
        DefaultImage := BuffUpgrades_2
        DefaultTint := color{R := 1.0, G := 0.226483, B := 0.0}
        DefaultDesiredSize := vector2{X := 95.0,Y := 95.0}

    Text_Buff : text_block = text_block:
        DefaultTextColor := color{R := 1.0, G := 1.0, B := 1.0}
        DefaultTextOpacity := 1.0
        DefaultShadowOffset :=  option{vector2{X := 1.472062, Y := 1.453619}}
        DefaultShadowColor := color{R := 0.0, G := 0.0, B := 0.0}
        DefaultShadowOpacity := 1.0
    #StrBuff
    BG_StrBuff : texture_block = texture_block:
        DefaultImage := BG_s
        DefaultTint := color{R:=0.0,G:=0.530973,B:=1.0}
        DefaultDesiredSize := vector2{X := 100.0, Y := 100.0}

    Icon_StrBuff : texture_block = texture_block:
        DefaultImage := Boss_1
        DefaultTint := color{R := 0.807249, G := 0.0, B :=1.0}
        DefaultDesiredSize := vector2{X := 100.0,Y := 100.0}

    Icon_StrBuff2 : texture_block = texture_block:
        DefaultImage := MinerSkill1_3
        DefaultTint := color{R := 1.0, G := 1.0, B :=1.0}
        DefaultDesiredSize := vector2{X := 70.0,Y := 70.0}

    Text_StrBuff : text_block = text_block:
        DefaultTextColor := color{R := 1.0, G := 1.0, B := 1.0}
        DefaultTextOpacity := 1.0
        DefaultShadowOffset :=  option{vector2{X := 1.472062, Y := 1.453619}}
        DefaultShadowColor := color{R := 0.0, G := 0.0, B := 0.0}
        DefaultShadowOpacity := 1.0


    CreateBuffCanvas<public>() : canvas =
        canvas :
            Slots := array :
                #Buff
                canvas_slot:
                    Anchors := anchors:
                        Minimum := vector2{X := 0.8, Y := 0.012824}
                        Maximum := vector2{X := 0.8, Y := 0.012824}
                    Offsets := margin{Top := 0.0, Left := 0.0, Right := 0.0, Bottom := 0.0}
                    SizeToContent := true
                    Alignment := vector2{X := 0.5, Y := 0.0}
                    Widget := BG_Buff
                canvas_slot:
                    Anchors := anchors:
                        Minimum := vector2{X := 0.8, Y := 0.012824}
                        Maximum := vector2{X := 0.8, Y := 0.012824}
                    Offsets := margin{Top := 2.5, Left := 0.0, Right := 0.0, Bottom := 0.0}
                    SizeToContent := true
                    Alignment := vector2{X := 0.5, Y := 0.0}
                    Widget := Icon_Buff
                canvas_slot:
                    Anchors := anchors:
                        Minimum := vector2{X := 0.8, Y := 0.012824}
                        Maximum := vector2{X := 0.8, Y := 0.012824}
                    Offsets := margin{Top := 99.877121, Left := 0.0, Right := 0.0, Bottom := 0.0}
                    SizeToContent := true
                    Alignment := vector2{X := 0.5, Y := 0.0}
                    Widget := Text_Buff
                #StrBuff
                canvas_slot:
                    Anchors := anchors:
                        Minimum := vector2{X := 0.8, Y := 0.142824}
                        Maximum := vector2{X := 0.8, Y := 0.142824}
                    Offsets := margin{Top := 0.0, Left := 0.0, Right := 0.0, Bottom := 0.0}
                    SizeToContent := true
                    Alignment := vector2{X := 0.5, Y := 0.0}
                    Widget := BG_StrBuff
                canvas_slot:
                    Anchors := anchors:
                        Minimum := vector2{X := 0.8, Y := 0.142824}
                        Maximum := vector2{X := 0.8, Y := 0.142824}
                    Offsets := margin{Top := 5.595673, Left := 0.462524, Right := 0.0, Bottom := 0.0}
                    SizeToContent := true
                    Alignment := vector2{X := 0.5, Y := 0.0}
                    Widget := Icon_StrBuff
                canvas_slot:
                    Anchors := anchors:
                        Minimum := vector2{X := 0.8, Y := 0.142824}
                        Maximum := vector2{X := 0.8, Y := 0.142824}
                    Offsets := margin{Top := -10.404327, Left := 29.462524, Right := 0.0, Bottom := 0.0}
                    SizeToContent := true
                    Alignment := vector2{X := 0.5, Y := 0.0}
                    Widget := Icon_StrBuff2
                canvas_slot:
                    Anchors := anchors:
                        Minimum := vector2{X := 0.8, Y := 0.142824}
                        Maximum := vector2{X := 0.8, Y := 0.142824}
                    Offsets := margin{Top := 103.437401, Left := 0.0, Right := 0.0, Bottom := 0.0}
                    SizeToContent := true
                    Alignment := vector2{X := 0.5, Y := 0.0}
                    Widget := Text_StrBuff

ui_idle := class :
    #休憩所
    IdleCount : text_block = text_block:
        DefaultTextColor := color{R := 0.502867, G := 1.0, B := 0.317708}
        DefaultTextOpacity := 1.0
        DefaultShadowOffset := option{vector2{X := 2.361355,Y := 2.830423}}
        DefaultShadowColor := color{R := 0.0, G := 0.0, B := 0.0}
        DefaultShadowOpacity := 0.64
    CreateIdleCanvas<public>():canvas=
        canvas :
            Slots := array :
                canvas_slot:
                    Anchors := anchors:
                        Minimum := vector2{X := 0.01, Y := 0.5}
                        Maximum := vector2{X := 0.01, Y := 0.5}
                    Offsets := margin{Top := 147.459473,Left := 286.855377, Right := 0.0, Bottom := 0.0}
                    SizeToContent := true
                    Alignment := vector2{X := 0.0, Y := 0.0}
                    Widget := IdleCount

