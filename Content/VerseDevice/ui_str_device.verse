
using { /Verse.org/Assets }
using { /Verse.org/Simulation }
using { /Fortnite.com/UI }
using { /Fortnite.com/Devices }
using { /Fortnite.com/Characters }
using { /UnrealEngine.com/Temporary/UI }
using { /UnrealEngine.com/Temporary/SpatialMath }
using { /Verse.org/Simulation }
using { /UnrealEngine.com/Temporary/Diagnostics }
using { /Verse.org/Colors }
using { /Verse.org/Colors/NamedColors }

using {Mining_System.Code.PlayerData}
using {Mining_System.Code.MiningObject}
using {Mining_System.Code.Weapon}
using {Mining_System.Code.Pet}

Ui_str_device := class(creative_device):
    #SFXs
    @editable
    SFX_Open : audio_player_device = audio_player_device{}
    @editable
    SFX_Close : audio_player_device = audio_player_device{}
    @editable
    SFX_LevelUp : audio_player_device = audio_player_device{}
    @editable
    SFX_Failed : audio_player_device = audio_player_device{}

    # @editable
    # PlayerSpawners : []player_spawner_device = array{}
    @editable
    StrSignal : signal_remote_manager_device = signal_remote_manager_device{}

    @editable
    SaveDataManager : save_data_manager = save_data_manager{}

    @editable
    WeaponDamage : weapon_damage = weapon_damage{}

    @editable
    MiningDamageDevice : mining_damage_device = mining_damage_device{}

    @editable
    MiningLevelDevice : mining_level_device = mining_level_device{}

    @editable
    PetManager : pet_manager = pet_manager{}
    @editable
    UIHudCount : ui_hud_count = ui_hud_count{}

    var StrCanvasPerAgent : [agent]canvas = map{}
    var IsAutoing : [player]logic = map{}
    var UIElements : [agent][int]tuple(text_block, text_block, text_block, text_block, text_block, text_block, text_block,
     text_block, text_block, text_block, text_block, text_block, text_block, text_block, text_block, text_block, text_block,
     text_block, text_block, text_block, text_block, text_block, text_block, text_block, text_block, text_block, text_block, text_block, button_regular) = map{}
    
    StringToMessage<public><localizes>(Text:string):message="{Text}"

    CloseEvent<public> : event(player) = event(player){}

    
    var Level : int = 1
    BaseResource : float = 10.0
    IncreaseRatioResource : float = 1.28

    SetText<localizes>(Value:[]char) : message = "{Value}"
    
    OnBegin<override>()<suspends>:void=
        StrSignal.PrimarySignalEvent.Subscribe(OnStrSignalEvented)

    UpdateUI(Agent:agent):void=
        var SilverValue : big_number = big_number{}
        var StrValue : big_number = big_number{}
        var BaseStrength  : big_number = big_number{}
        var PetStrength : big_number = big_number{}
        var TotalStrength : big_number = big_number{}
        var AddStrength : big_number = big_number{}
        var PetLevelBonus : big_number = big_number{}
        var AreaBonus : big_number = big_number{}
        var GtaherRate : big_number = big_number{}
        var Cost : big_number = big_number{}
        var TotalResource : big_number = big_number{}
        if:
            Player := player[Agent]
        then:
            set SilverValue = SaveDataManager.GetResourceStone(Player)
            set BaseStrength = MiningDamageDevice.GetBaseStrength(SaveDataManager.GetLevel(Player))
            set PetStrength = PetManager.GetPetStrengthFromEquipSlots(Player)
            set TotalStrength = BigNumber_Add(BaseStrength, BigNumber_MultiplyByBigNumber(BigNumber_MultiplyByFloat(PetStrength, 0.01),BaseStrength))
            set Level = SaveDataManager.GetLevel(Player)
            set AddStrength = BigNumber_Subtract(MiningDamageDevice.GetBaseStrength(Level+1), MiningDamageDevice.GetBaseStrength(Level))
            set PetLevelBonus  = PetManager.GetTotalLevelBonus(Player)
            set AreaBonus  = BigNumber_MultiplyByFloat(MiningDamageDevice.GetStagePercent(Player), 100.0)
            set GtaherRate  = MiningDamageDevice.GetNormalDamage(Player)
            set Cost = BigNumber_MultiplyByFloat(big_number{Value := BaseResource, Exponent := 0}, Pow(IncreaseRatioResource, (SaveDataManager.GetLevel(Player)-1)*1.0))
            set TotalResource = SaveDataManager.GetResourceStone(Player)
            set StrValue = TotalStrength
            _CostTextColor := if(MiningLevelDevice.CanLevelUp[player[Agent]]){NamedColors.White}else{NamedColors.Red}
            PlayerLevel := SaveDataManager.GetLevel(Player) 
            StrengthValue := MiningDamageDevice.GetBaseStrength(PlayerLevel)
            FormattedStrength := BigNumber_FormatDisplay(StrengthValue)
            NextStrengthValue := BigNumber_Add(StrengthValue, AddStrength)
            NextFormattedStrength := BigNumber_FormatDisplay(NextStrengthValue)

            <#エレメントリスト
            0 OreTextBlock
            1 StrTextBlock
            2 MyBaseStrengthTextBlock
            3 MyBaseStrengthTextBlock2
            4 MyBaseStrengthTextBlock3
            5 MyPetStrengthTextBlock
            6 MyTotalStrengthTextBlock
            7 MyTotalStrengthTextBlock2
            8 MyTotalStrengthTextBlock3
            9 AddStrengthTextBlock
            10 AddStrengthTextBlock2
            11 TotalResourceTextBlock
            12 SilverCostTextBlock
            13 MyLevelTextBlock
            14 MyNextLevelTextBlock
            15 BaseStrengthTextBlock
            16 BaseStrengthTextBlock2
            17 BaseStrengthTextBlock3
            18 PetStrengthTextBlock
            19 TotalStrengthTextBlock
            20 TotalStrengthTextBlock2
            21 TotalStrengthTextBlock3
            22 LvBonus_TB
            23 MyLvBonus_TB
            24 AreaBonus_TB
            25 MyAreaBonus_TB
            26 Mining_TB
            27 MyMining_TB
            28 AutoButton
            #>
            <#
            
            2 MyBaseStrengthTextBlock
            3 MyBaseStrengthTextBlock2
            4 MyBaseStrengthTextBlock3
            #>
            if(Element := UIElements[Agent][1]) :
                Element(0).SetText(SetText("{BigNumber_FormatDisplay(SilverValue)}"))
                Element(1).SetText(SetText("{BigNumber_FormatDisplay(StrValue)}"))
                Element(2).SetText(StringToMessage(FormattedStrength))
                Element(3).SetText(StringToMessage(FormattedStrength))
                Element(4).SetText(StringToMessage(NextFormattedStrength))
                Element(5).SetText(SetText("+ {BigNumber_FormatDisplay(BigNumber_MultiplyByBigNumber(BaseStrength,BigNumber_MultiplyByFloat(PetStrength,0.01)))}( + {BigNumber_FormatDisplay(PetStrength)}% )"))
                Element(6).SetText(SetText("{BigNumber_FormatDisplay(TotalStrength)}"))
                Element(7).SetText(SetText("{BigNumber_FormatDisplay(TotalStrength)}"))
                Element(8).SetText(SetText("{BigNumber_FormatDisplay(BigNumber_Add(TotalStrength, AddStrength))}"))
                Element(9).SetText(SetText("( + {BigNumber_FormatDisplay(AddStrength)} )"))
                Element(10).SetText(SetText("( + {BigNumber_FormatDisplay(AddStrength)} )"))
                Element(11).SetText(SetText("{BigNumber_FormatDisplay(TotalResource)}"))
                Element(12).SetText(SetText("{BigNumber_FormatDisplay(Cost)}"))
                Element(12).SetTextColor(_CostTextColor)   #Color
                Element(13).SetText(SetText("Now Level : {Level}"))
                Element(14).SetText(SetText("Next Level : {Level + 1}"))
                Element(23).SetText(SetText("+ {BigNumber_FormatDisplay(PetLevelBonus)}%"))
                Element(25).SetText(SetText("+ {BigNumber_FormatDisplay(AreaBonus)}%"))
                Element(27).SetText(SetText("{BigNumber_FormatDisplay(GtaherRate)}/Hit"))

                if(IsAutoing[Player] = true) :
                    Element(28).SetText(StringToMessage("Upgrading"))
                else :
                    Element(28).SetText(StringToMessage("Auto"))
            
                Element(2).SetText(SetText("{BigNumber_FormatDisplay(BaseStrength)}"))
                Element(3).SetText(SetText("{BigNumber_FormatDisplay(BaseStrength)}"))
                Element(4).SetText(SetText("{BigNumber_FormatDisplay(BigNumber_Add(BaseStrength, AddStrength))}"))
    
    SetUI(Agent : agent):void=
        if:
            Player := player[Agent]
        then:
            var LanguageIndex : int = SaveDataManager.GetLanguage(Player)
            BaseStrengthText := if(LanguageIndex = 0){"基礎採掘力 :"} else {"Base Strength :"}
            PetStrengthText := if(LanguageIndex = 0){"妖怪採掘力 :"} else {"Pet Strength :"}
            TotalStrengthText := if(LanguageIndex = 0){"採掘力合計値 :"} else {"Total Strength :"}
            PetBonusText := if(LanguageIndex = 0){"妖怪レベルボーナス :"} else {"Pet Lvel Bonus :"}
            AreaBonusText := if(LanguageIndex = 0){"エリアボーナス :"} else {"Area Bonus :"}
            MiningText := if(LanguageIndex = 0){"素材採取力 :"} else {"Mining Strength :"}
            if(Element := UIElements[Agent][1]) :
                Element(15).SetText(StringToMessage(BaseStrengthText))
                Element(16).SetText(StringToMessage(BaseStrengthText))
                Element(17).SetText(StringToMessage(BaseStrengthText))
                Element(18).SetText(StringToMessage(PetStrengthText))
                Element(19).SetText(StringToMessage(TotalStrengthText))
                Element(20).SetText(StringToMessage(TotalStrengthText))
                Element(21).SetText(StringToMessage(TotalStrengthText))
                Element(22).SetText(StringToMessage(PetBonusText))
                Element(24).SetText(StringToMessage(AreaBonusText))
                Element(26).SetText(StringToMessage(MiningText))




    OnPlayerSpawned(Agent : agent):void=
        if (StrCanvas := StrCanvasPerAgent[Agent]):
            return

        if (Player := player[Agent], PlayerUI := GetPlayerUI[Player]):
            StrCanvas : canvas = CreateStrCanvas(Agent)
            if (set StrCanvasPerAgent[Agent] = StrCanvas){}

    OnStrSignalEvented(Agent : agent):void=
        SFX_Open.Play(Agent)
        OnPlayerSpawned(Agent)
        SetUI(Agent)#文字の部分は開いた時のみ更新
        UpdateUI(Agent)
        if (Player := player[Agent], PlayerUI := GetPlayerUI[Player], StrCanvas := StrCanvasPerAgent[Agent]):
            PlayerUI.AddWidget(StrCanvas, player_ui_slot{InputMode := ui_input_mode.All})

    OnCloseButtonClicked(WidgetMessage : widget_message):void=
        if (Player := player[WidgetMessage.Player]):
            SFX_Close.Play(Player)
            if (PlayerUI := GetPlayerUI[Player], StrCanvas := StrCanvasPerAgent[Player]):
                PlayerUI.RemoveWidget(StrCanvas)
                CloseEvent.Signal(Player)
                UIHudCount.Update_IdleUI(Player)
        else:
            Print("Error: OnCloseButtonClicked")

    # ボタン処理
    OnLevelUpButtonClicked(WidgetMessage : widget_message):void=
        if (PlayerUI := GetPlayerUI[WidgetMessage.Player], StrCanvas := StrCanvasPerAgent[WidgetMessage.Player]):
            if(MiningLevelDevice.CanLevelUp[WidgetMessage.Player]):
                SFX_LevelUp.Play(WidgetMessage.Player)
                MiningLevelDevice.TryPlayerLevel(WidgetMessage.Player)
                UpdateUI(WidgetMessage.Player)
            else:
                SFX_Failed.Play(WidgetMessage.Player)

    CallAutoFunc(WidgetMessage : widget_message):void=
            spawn{OnAutoButtonClicked(WidgetMessage)}

    OnAutoButtonClicked(WidgetMessage : widget_message)<suspends>:void=
        if(not IsAutoing[WidgetMessage.Player]) :
            logic{set IsAutoing[WidgetMessage.Player] = false}
        if(IsAutoing[WidgetMessage.Player] = false) :
            logic{set IsAutoing[WidgetMessage.Player] = true}
            if (PlayerUI := GetPlayerUI[WidgetMessage.Player], StrCanvas := StrCanvasPerAgent[WidgetMessage.Player]):
                if :
                    Agent : agent = agent[WidgetMessage.Player]
                    Element := UIElements[Agent][1]
                then :
                    Element(28).SetText(StringToMessage("Upgrading"))
                race :
                    block :
                        loop :
                            ThePlayer := CloseEvent.Await()
                            if(ThePlayer = WidgetMessage.Player) :
                                Print("Player Colse UIStr")
                                break
                            else :
                                Print("Dont Close")
                    block :
                        loop :
                            if(MiningLevelDevice.CanLevelUp[WidgetMessage.Player]):
                                SFX_LevelUp.Play(WidgetMessage.Player)
                                MiningLevelDevice.TryPlayerLevel(WidgetMessage.Player)
                                UpdateUI(WidgetMessage.Player)
                                Sleep(0.5)
                            else:
                                break
                if :
                    Agent : agent = agent[WidgetMessage.Player]
                    Element := UIElements[Agent][1]
                then :
                    Element(28).SetText(StringToMessage("Auto"))
                SFX_Failed.Play(WidgetMessage.Player)
                logic{set IsAutoing[WidgetMessage.Player] = false}
        else :
            SFX_Failed.Play(WidgetMessage.Player)
    # キャンバス
    CreateStrCanvas(Agent : agent):canvas=

        var CanvasSlots : []canvas_slot = array{}

        CenterVector : vector2 = vector2{X:=0.5, Y:=0.5}
        CenterAnchors : anchors = anchors{Minimum := CenterVector, Maximum := CenterVector}

        # Image Stock
        BGImage : texture_block = texture_block{DefaultImage := Textures.UI_BG, DefaultDesiredSize := vector2{X := 1920.0, Y := 1080.0}}
        BG_sImage : texture_block = texture_block{DefaultImage := icon_set_2.BG_s, DefaultDesiredSize := vector2{X := 900.0, Y := 600.0}}
        BG_s_HalfImage : texture_block = texture_block{DefaultImage := icon_set_2.BG_s, DefaultDesiredSize := vector2{X := 900.0, Y := 300.0}}
        BG_s_HalfImage2 : texture_block = texture_block{DefaultImage := icon_set_2.BG_s, DefaultDesiredSize := vector2{X := 900.0, Y := 300.0}}

        SilverOreImage1 : texture_block = texture_block{DefaultImage := icon_set_2.SilverOre, DefaultDesiredSize := vector2{X := 50.0, Y := 50.0}}
        StrImage : texture_block = texture_block{DefaultImage := icon_set_2.Stats_Str, DefaultDesiredSize := vector2{X := 60.0, Y := 60.0}}
        StrImage2 : texture_block = texture_block{DefaultImage := icon_set_2.Stats_Str, DefaultDesiredSize := vector2{X := 60.0, Y := 60.0}}
        StrImage3 : texture_block = texture_block{DefaultImage := icon_set_2.Stats_Str, DefaultDesiredSize := vector2{X := 60.0, Y := 60.0}}
        StrImage4 : texture_block = texture_block{DefaultImage := icon_set_2.Stats_Str, DefaultDesiredSize := vector2{X := 60.0, Y := 60.0}}
        StrImage5 : texture_block = texture_block{DefaultImage := icon_set_2.Stats_Str, DefaultDesiredSize := vector2{X := 60.0, Y := 60.0}}
        StrImage6 : texture_block = texture_block{DefaultImage := icon_set_2.Stats_Str, DefaultDesiredSize := vector2{X := 60.0, Y := 60.0}}
        YokaiImage : texture_block = texture_block{DefaultImage := Mining_System.Code.Skill.Textures.Pre_Yokai_2, DefaultDesiredSize := vector2{X := 60.0, Y := 60.0}}
        IC_Yokai_LvBonus_TB : texture_block = texture_block{DefaultImage := icon_set_2.IC_Yokai, DefaultDesiredSize := vector2{X := 60.0, Y := 60.0}}
        IC_Area_TB : texture_block = texture_block{DefaultImage := icon_set_2.IC_Area, DefaultDesiredSize := vector2{X := 60.0, Y := 60.0}}
        IC_Mining_TB : texture_block = texture_block{DefaultImage := icon_set_2.IC_Mining, DefaultDesiredSize := vector2{X := 60.0, Y := 60.0}}

        var SilverValue : big_number = big_number{}
        var StrValue : big_number = big_number{}
        var BaseStrength  : big_number = big_number{}
        var PetStrength : big_number = big_number{}
        var TotalStrength : big_number = big_number{}
        var AddStrength : big_number = big_number{}
        var PetLevelBonus : big_number = big_number{}
        var AreaBonus : big_number = big_number{}
        var GtaherRate : big_number = big_number{}
        var Cost : big_number = big_number{}
        var TotalResource : big_number = big_number{}
        
        #================================== TextBlock ==================================
    
        # 所持岩
        OreTextBlock : text_block = text_block{
            DefaultTextColor := NamedColors.White,
            DefaultShadowOffset := option{vector2{X:=2.5, Y:=2.5}},
            DefaultShadowOpacity := 1.0,
            DefaultShadowColor := NamedColors.Black
        }
        # 現STR
        StrTextBlock : text_block = text_block{
            DefaultTextColor := NamedColors.White,
            DefaultShadowOffset := option{vector2{X:=2.5, Y:=2.5}},
            DefaultShadowOpacity := 1.0,
            DefaultShadowColor := NamedColors.Black
        }
    
        MyBaseStrengthTextBlock : text_block = text_block{
            DefaultTextColor := NamedColors.White,
            DefaultShadowOffset := option{vector2{X:=2.5, Y:=2.5}},
            DefaultShadowOpacity := 1.0,
            DefaultShadowColor := NamedColors.Black
        }
        MyBaseStrengthTextBlock2 : text_block = text_block{
            DefaultTextColor := NamedColors.White,
            DefaultShadowOffset := option{vector2{X:=2.5, Y:=2.5}},
            DefaultShadowOpacity := 1.0,
            DefaultShadowColor := NamedColors.Black
        }
        MyBaseStrengthTextBlock3 : text_block = text_block{
            DefaultTextColor := NamedColors.White,
            DefaultShadowOffset := option{vector2{X:=2.5, Y:=2.5}},
            DefaultShadowOpacity := 1.0,
            DefaultShadowColor := NamedColors.Black
        }
        MyPetStrengthTextBlock : text_block = text_block{
            DefaultTextColor := NamedColors.White,
            DefaultShadowOffset := option{vector2{X:=2.5, Y:=2.5}},
            DefaultShadowOpacity := 1.0,
            DefaultShadowColor := NamedColors.Black
        }
        MyTotalStrengthTextBlock : text_block = text_block{
            DefaultTextColor := NamedColors.White,
            DefaultShadowOffset := option{vector2{X:=2.5, Y:=2.5}},
            DefaultShadowOpacity := 1.0,
            DefaultShadowColor := NamedColors.Black
        }
        MyTotalStrengthTextBlock2 : text_block = text_block{
            DefaultTextColor := NamedColors.White,
            DefaultShadowOffset := option{vector2{X:=2.5, Y:=2.5}},
            DefaultShadowOpacity := 1.0,
            DefaultShadowColor := NamedColors.Black
        }
        MyTotalStrengthTextBlock3 : text_block = text_block{
            DefaultTextColor := NamedColors.White,
            DefaultShadowOffset := option{vector2{X:=2.5, Y:=2.5}},
            DefaultShadowOpacity := 1.0,
            DefaultShadowColor := NamedColors.Black
        }
        AddStrengthTextBlock : text_block = text_block{
            DefaultTextColor := color{R:=0.000000,G:=1.000000,B:=0.039124},
            DefaultShadowOffset := option{vector2{X:=2.5, Y:=2.5}},
            DefaultShadowOpacity := 1.0,
            DefaultShadowColor := NamedColors.Black
        }
        AddStrengthTextBlock2 : text_block = text_block{
            DefaultTextColor := color{R:=0.000000,G:=1.000000,B:=0.039124},
            DefaultShadowOffset := option{vector2{X:=2.5, Y:=2.5}},
            DefaultShadowOpacity := 1.0,
            DefaultShadowColor := NamedColors.Black
        }
        TotalResourceTextBlock : text_block = text_block{
            # DefaultText := StringToMessage("{BigNumber_FormatDisplay(TotalResource)}"),
            DefaultTextColor := NamedColors.White,
            DefaultShadowOffset := option{vector2{X:=2.5, Y:=2.5}},
            DefaultShadowOpacity := 1.0,
            DefaultShadowColor := NamedColors.Black
        }
        SilverCostTextBlock : text_block = text_block{
            # DefaultText := StringToMessage("{BigNumber_FormatDisplay(Cost)}"),
            DefaultTextColor := NamedColors.White,
            DefaultShadowOffset := option{vector2{X:=2.5, Y:=2.5}},
            DefaultShadowOpacity := 1.0,
            DefaultShadowColor := NamedColors.Black
        }
        MyLevelTextBlock : text_block = text_block{
            # DefaultText := StringToMessage("{Level}"),
            DefaultTextColor := NamedColors.White,
            DefaultShadowOffset := option{vector2{X:=2.5, Y:=2.5}},
            DefaultShadowOpacity := 1.0,
            DefaultShadowColor := NamedColors.Black
        }
        MyNextLevelTextBlock : text_block = text_block{
            # DefaultText := StringToMessage("{Level}"),
            DefaultTextColor := NamedColors.White,
            DefaultShadowOffset := option{vector2{X:=2.5, Y:=2.5}},
            DefaultShadowColor := NamedColors.Black
        }
        #================================== /TextBlock ==================================
        #基礎攻撃力
        var BaseStrengthTextBlock :text_block= text_block{
            DefaultTextColor := NamedColors.White,
            DefaultShadowOffset := option{vector2{X:=2.5, Y:=2.5}},
            DefaultShadowOpacity := 1.0,
            DefaultShadowColor := NamedColors.Black
        }
        var BaseStrengthTextBlock2 :text_block= text_block{
            DefaultTextColor := NamedColors.White,
            DefaultShadowOffset := option{vector2{X:=2.5, Y:=2.5}},
            DefaultShadowOpacity := 1.0,
            DefaultShadowColor := NamedColors.Black
        }
        var BaseStrengthTextBlock3 :text_block= text_block{
            DefaultTextColor := NamedColors.White,
            DefaultShadowOffset := option{vector2{X:=2.5, Y:=2.5}},
            DefaultShadowOpacity := 1.0,
            DefaultShadowColor := NamedColors.Black
        }
        #妖怪採掘力
        var PetStrengthTextBlock :text_block= text_block{
            DefaultTextColor := NamedColors.White,
            DefaultShadowOffset := option{vector2{X:=2.5, Y:=2.5}},
            DefaultShadowOpacity := 1.0,
            DefaultShadowColor := NamedColors.Black
        }
        #合計採掘力
        var TotalStrengthTextBlock :text_block= text_block{
            DefaultTextColor := NamedColors.White,
            DefaultShadowOffset := option{vector2{X:=2.5, Y:=2.5}},
            DefaultShadowOpacity := 1.0,
            DefaultShadowColor := NamedColors.Black
        }
        var TotalStrengthTextBlock2 :text_block= text_block{
            DefaultTextColor := NamedColors.White,
            DefaultShadowOffset := option{vector2{X:=2.5, Y:=2.5}},
            DefaultShadowOpacity := 1.0,
            DefaultShadowColor := NamedColors.Black
        }
        var TotalStrengthTextBlock3 :text_block= text_block{
            DefaultTextColor := NamedColors.White,
            DefaultShadowOffset := option{vector2{X:=2.5, Y:=2.5}},
            DefaultShadowOpacity := 1.0,
            DefaultShadowColor := NamedColors.Black
        }
    
        var LvBonus_TB :text_block= text_block{
            DefaultTextColor := NamedColors.White,
            DefaultShadowOffset := option{vector2{X:=2.5, Y:=2.5}},
            DefaultShadowOpacity := 1.0,
            DefaultShadowColor := NamedColors.Black
        }
        var MyLvBonus_TB : text_block = text_block{
            DefaultTextColor := NamedColors.White,
            DefaultShadowOffset := option{vector2{X:=2.5, Y:=2.5}},
            DefaultShadowOpacity := 1.0,
            DefaultShadowColor := NamedColors.Black
        }
        var AreaBonus_TB :text_block= text_block{
            DefaultTextColor := NamedColors.White,
            DefaultShadowOffset := option{vector2{X:=2.5, Y:=2.5}},
            DefaultShadowOpacity := 1.0,
            DefaultShadowColor := NamedColors.Black
        }
        var MyAreaBonus_TB : text_block = text_block{
            DefaultTextColor := NamedColors.White,
            DefaultShadowOffset := option{vector2{X:=2.5, Y:=2.5}},
            DefaultShadowOpacity := 1.0,
            DefaultShadowColor := NamedColors.Black
        }
        var Mining_TB :text_block= text_block{
            DefaultTextColor := NamedColors.White,
            DefaultShadowOffset := option{vector2{X:=2.5, Y:=2.5}},
            DefaultShadowOpacity := 1.0,
            DefaultShadowColor := NamedColors.Black
        }
        var MyMining_TB : text_block = text_block{
            DefaultTextColor := NamedColors.White,
            DefaultShadowOffset := option{vector2{X:=2.5, Y:=2.5}},
            DefaultShadowOpacity := 1.0,
            DefaultShadowColor := NamedColors.Black
        }
        AutoButton : button_regular = button_regular{}
#============================================================================

        StrengthUpgradesTextBlock := text_block :
            DefaultText := StringToMessage("Strength Upgrades")
            DefaultTextColor := NamedColors.White
            DefaultShadowOffset := option{vector2{X:=2.5, Y:=2.5}}
            DefaultShadowColor := NamedColors.Black
            DefaultShadowOpacity := 1.0

        MiningPowerTextBlock := text_block :
            DefaultText := StringToMessage("Mining Power")
            DefaultTextColor := NamedColors.White
            DefaultShadowOffset := option{vector2{X:=2.5, Y:=2.5}}
            DefaultShadowColor := NamedColors.Black
            DefaultShadowOpacity := 1.0

        WhiteLine := color_block :
            DefaultColor := NamedColors.White
            DefaultOpacity := 1.0
            DefaultDesiredSize := vector2{X :=655.0, Y := 2.0}

        WhiteLine2 := color_block :
            DefaultColor := NamedColors.White
            DefaultOpacity := 1.0
            DefaultDesiredSize := vector2{X :=655.0, Y := 2.0}

        BlankTextBlock := text_block :
            DefaultText := StringToMessage("")
            DefaultTextColor := NamedColors.White
            DefaultShadowOffset := option{vector2{X:=2.5, Y:=2.5}}
            DefaultShadowColor := NamedColors.Black
            DefaultShadowOpacity := 1.0

        BonusTextBlock := text_block :
            DefaultText := StringToMessage("Bonus")
            DefaultTextColor := NamedColors.White
            DefaultShadowOffset := option{vector2{X:=2.5, Y:=2.5}}
            DefaultShadowColor := NamedColors.Black
            DefaultShadowOpacity := 1.0

        CostTextBlock := text_block :
            DefaultText := StringToMessage("Cost :")
            DefaultTextColor := NamedColors.White
            DefaultShadowOffset := option{vector2{X:=2.5, Y:=2.5}}
            DefaultShadowColor := NamedColors.Black
            DefaultShadowOpacity := 1.0

        CloseButton : button_loud = button_loud{DefaultText := StringToMessage("CLOSE")}
        CloseButton.OnClick().Subscribe(OnCloseButtonClicked)

        LevelUpButtonMessage := if(MiningLevelDevice.CanLevelUp[player[Agent]]){StringToMessage("Level Up")}else{StringToMessage("Level Up")}
        LevelUpButton : button_regular = button_regular{DefaultText := LevelUpButtonMessage}
        LevelUpButton.OnClick().Subscribe(OnLevelUpButtonClicked)

        AutoButton.OnClick().Subscribe(CallAutoFunc)

        # --- プレイヤーのレベルと強さの取得 ---
        if (Player := player[Agent]):
            PlayerLevel := SaveDataManager.GetLevel(Player)
            PlayerStrength := MiningDamageDevice.GetBaseStrength(PlayerLevel)
            FormattedStrength := BigNumber_FormatDisplay(PlayerStrength)

            PlayerPetStrength := MiningDamageDevice.GetPetStrength(Player)
            TotalPlayerStrength := BigNumber_Add(PlayerStrength, PlayerPetStrength)
            FormattedTotalStrength := BigNumber_FormatDisplay(TotalPlayerStrength)
            NextTotalPlayerStrength := BigNumber_Add(TotalPlayerStrength, AddStrength)
            FormattedNextTotalStrength := BigNumber_FormatDisplay(NextTotalPlayerStrength)

            MyBaseStrengthTextBlock.SetText(StringToMessage(FormattedStrength))
            MyBaseStrengthTextBlock2.SetText(StringToMessage(FormattedStrength))
            MyBaseStrengthTextBlock3.SetText(StringToMessage(FormattedStrength))

            MyTotalStrengthTextBlock.SetText(StringToMessage(FormattedTotalStrength))
            MyTotalStrengthTextBlock2.SetText(StringToMessage(FormattedTotalStrength))
            MyTotalStrengthTextBlock3.SetText(StringToMessage(FormattedNextTotalStrength))

        #==========================================新UI==========================================

        #一番大きいBGと見出しテキスト(StrengthUpgradesTextBlock)
        set CanvasSlots += array
        {
            canvas_slot:
                Anchors := anchors :
                    Minimum := vector2{X := 0.5, Y := 0.5}
                    Maximum := vector2{X := 0.5, Y := 0.5}
                Offsets := margin{Top := 0.0,Left := 0.0, Right := 1920.0, Bottom := 1080.0}
                Alignment := CenterVector
                SizeToContent := false
                Widget := BGImage
        }
        set CanvasSlots += array
        {
            canvas_slot :
                Anchors := anchors :
                    Minimum := vector2{X := 0.5, Y := 0.5}
                    Maximum := vector2{X := 0.5, Y := 0.5}
                Offsets := margin{Left := 0.0, Top := -488.948944, Right := 100.0, Bottom := 30.0}
                Alignment := CenterVector
                SizeToContent := false
                Widget := StrengthUpgradesTextBlock
        }
        set CanvasSlots += array
        #真ん中3つのBG
        {
            canvas_slot:
                Anchors := CenterAnchors
                Offsets := margin{Top:= 0.0 ,Left:= 0.0, Right:= 0.0, Bottom:= 0.0}
                Alignment := CenterVector
                SizeToContent := true
                Widget := stack_box:
                    Orientation := orientation.Horizontal
                    Slots := array:
                        stack_box_slot:
                            Distribution:= option{1.0}
                            Padding := margin{Left := 0.0, Top := 0.0, Right := 0.0, Bottom := 0.0}
                            HorizontalAlignment := horizontal_alignment.Fill
                            VerticalAlignment := vertical_alignment.Fill
                            Widget := BG_sImage
                        stack_box_slot:
                            Distribution:= option{1.0}
                            Padding := margin{Left := 0.0, Top := 0.0, Right := 0.0, Bottom := 0.0}
                            HorizontalAlignment := horizontal_alignment.Fill
                            VerticalAlignment := vertical_alignment.Fill
                            Widget := stack_box :
                                Orientation := orientation.Vertical
                                Slots := array :
                                    stack_box_slot :
                                        Distribution := option{1.0}
                                        Padding := margin{Left := 10.0, Top := 0.0, Right := 0.0, Bottom := 0.0}
                                        HorizontalAlignment := horizontal_alignment.Right
                                        VerticalAlignment := vertical_alignment.Center
                                        Widget := BG_s_HalfImage
                                    stack_box_slot :
                                        Distribution := option{1.0}
                                        Padding := margin{Left := 10.0, Top := 0.0, Right := 0.0, Bottom := 0.0}
                                        HorizontalAlignment := horizontal_alignment.Right
                                        VerticalAlignment := vertical_alignment.Center
                                        Widget := BG_s_HalfImage2
        }
        set CanvasSlots += array
        #PlayerStats
        {
            canvas_slot:
                Anchors := CenterAnchors
                Offsets := margin{Top:= 0.0, Left:= -458.0, Right:= 700.0, Bottom:= 580.0}
                Alignment := CenterVector
                SizeToContent := false
                Widget := stack_box:
                    Orientation := orientation.Vertical
                    Slots := array:
                        stack_box_slot:
                            Distribution:= option{1.0}
                            Padding := margin{Left:=0.0 ,Top:= 1.0, Right:= 0.0, Bottom:= 0.0}
                            HorizontalAlignment := horizontal_alignment.Center
                            VerticalAlignment := vertical_alignment.Fill
                            Widget := MiningPowerTextBlock
                        #基礎攻撃力
                        stack_box_slot:
                            Distribution:= option{1.0}
                            Padding := margin{Left := 0.0, Top := 2.0, Right := 0.0, Bottom := 2.0}
                            HorizontalAlignment := horizontal_alignment.Fill
                            VerticalAlignment := vertical_alignment.Fill
                            Widget := stack_box :
                                Orientation := orientation.Horizontal
                                Slots := array :
                                    stack_box_slot :
                                        Padding:=margin{Left:=0.0 ,Top:= 0.0, Right:= 0.0, Bottom:= 0.0}
                                        HorizontalAlignment := horizontal_alignment.Center
                                        VerticalAlignment := vertical_alignment.Center
                                        Widget := StrImage
                                    stack_box_slot :
                                        Distribution:= option{1.0}
                                        Padding:=margin{Left:=10.0 ,Top:= 0.0, Right:= 0.0, Bottom:= 0.0}
                                        HorizontalAlignment := horizontal_alignment.Left
                                        VerticalAlignment := vertical_alignment.Center
                                        Widget := BaseStrengthTextBlock
                                    stack_box_slot :
                                        Distribution:= option{1.0}
                                        Padding:=margin{Left:=0.0 ,Top:= 0.0, Right:= 0.0, Bottom:= 0.0}
                                        HorizontalAlignment := horizontal_alignment.Right
                                        VerticalAlignment := vertical_alignment.Center
                                        Widget := MyBaseStrengthTextBlock
                        #妖怪採掘力
                        stack_box_slot:
                            Distribution:= option{1.0}
                            Padding := margin{Left := 0.0, Top := 2.0, Right := 0.0, Bottom := 2.0}
                            HorizontalAlignment := horizontal_alignment.Fill
                            VerticalAlignment := vertical_alignment.Fill
                            Widget := stack_box :
                                Orientation := orientation.Horizontal
                                Slots := array :
                                    stack_box_slot :
                                        Padding:=margin{Left:=0.0 ,Top:= 0.0, Right:= 0.0, Bottom:= 0.0}
                                        HorizontalAlignment := horizontal_alignment.Center
                                        VerticalAlignment := vertical_alignment.Center
                                        Widget := YokaiImage
                                    stack_box_slot :
                                        Distribution:= option{1.0}
                                        Padding:=margin{Left:=10.0 ,Top:= 0.0, Right:= 0.0, Bottom:= 0.0}
                                        HorizontalAlignment := horizontal_alignment.Left
                                        VerticalAlignment := vertical_alignment.Center
                                        Widget := PetStrengthTextBlock
                                    stack_box_slot :
                                        Distribution:= option{1.0}
                                        Padding:=margin{Left:=0.0 ,Top:= 0.0, Right:= 0.0, Bottom:= 0.0}
                                        HorizontalAlignment := horizontal_alignment.Right
                                        VerticalAlignment := vertical_alignment.Center
                                        Widget := MyPetStrengthTextBlock
                        #白線
                        stack_box_slot:
                            Padding := margin{Left := 0.0, Top := 0.0, Right := 0.0, Bottom := 0.0}
                            HorizontalAlignment := horizontal_alignment.Fill
                            VerticalAlignment := vertical_alignment.Fill
                            Widget := WhiteLine
                        #採掘力合計値
                        stack_box_slot:
                            Distribution:= option{1.0}
                            Padding := margin{Left := 0.0, Top := 2.0, Right := 0.0, Bottom := 2.0}
                            HorizontalAlignment := horizontal_alignment.Fill
                            VerticalAlignment := vertical_alignment.Fill
                            Widget := stack_box :
                                Orientation := orientation.Horizontal
                                Slots := array :
                                    stack_box_slot :
                                        Padding:=margin{Left:=0.0 ,Top:= 0.0, Right:= 0.0, Bottom:= 0.0}
                                        HorizontalAlignment := horizontal_alignment.Center
                                        VerticalAlignment := vertical_alignment.Center
                                        Widget := StrImage2
                                    stack_box_slot :
                                        Distribution:= option{1.0}
                                        Padding:=margin{Left:=10.0 ,Top:= 0.0, Right:= 0.0, Bottom:= 0.0}
                                        HorizontalAlignment := horizontal_alignment.Left
                                        VerticalAlignment := vertical_alignment.Center
                                        Widget := TotalStrengthTextBlock
                                    stack_box_slot :
                                        Distribution:= option{1.0}
                                        Padding:=margin{Left:=0.0 ,Top:= 0.0, Right:= 0.0, Bottom:= 0.0}
                                        HorizontalAlignment := horizontal_alignment.Right
                                        VerticalAlignment := vertical_alignment.Center
                                        Widget := MyTotalStrengthTextBlock
                        #空白
                        stack_box_slot:
                            Distribution:= option{1.0}
                            Padding := margin{Left := 0.0, Top := 0.0, Right := 0.0, Bottom := - 10.0}
                            HorizontalAlignment := horizontal_alignment.Fill
                            VerticalAlignment := vertical_alignment.Fill
                            Widget := stack_box :
                                Orientation := orientation.Horizontal
                                Slots := array :
                                    stack_box_slot :
                                        Padding:=margin{Left:=0.0 ,Top:= 0.0, Right:= 0.0, Bottom:= 0.0}
                                        HorizontalAlignment := horizontal_alignment.Fill
                                        VerticalAlignment := vertical_alignment.Fill
                                        Widget := BlankTextBlock
                        #Bonus
                        stack_box_slot:
                            Padding := margin{Left := 0.0, Top := 0.0, Right := 0.0, Bottom := 0.0}
                            HorizontalAlignment := horizontal_alignment.Center
                            VerticalAlignment := vertical_alignment.Fill
                            Widget := BonusTextBlock
                        #妖怪ボーナス
                        stack_box_slot:
                            Distribution:= option{1.0}
                            Padding := margin{Left := 0.0, Top := 0.0, Right := 0.0, Bottom := 0.0}
                            HorizontalAlignment := horizontal_alignment.Fill
                            VerticalAlignment := vertical_alignment.Fill
                            Widget := stack_box :
                                Orientation := orientation.Horizontal
                                Slots := array :
                                    stack_box_slot :
                                        Padding:=margin{Left:=0.0 ,Top:= 0.0, Right:= 0.0, Bottom:= 0.0}
                                        HorizontalAlignment := horizontal_alignment.Center
                                        VerticalAlignment := vertical_alignment.Center
                                        Widget := IC_Yokai_LvBonus_TB
                                    stack_box_slot :
                                        Distribution:= option{1.0}
                                        Padding:=margin{Left:=10.0 ,Top:= 0.0, Right:= 0.0, Bottom:= 0.0}
                                        HorizontalAlignment := horizontal_alignment.Left
                                        VerticalAlignment := vertical_alignment.Center
                                        Widget := LvBonus_TB
                                    stack_box_slot :
                                        Distribution:= option{1.0}
                                        Padding:=margin{Left:=0.0 ,Top:= 0.0, Right:= 0.0, Bottom:= 0.0}
                                        HorizontalAlignment := horizontal_alignment.Right
                                        VerticalAlignment := vertical_alignment.Center
                                        Widget := MyLvBonus_TB
                        #エリアボーナス
                        stack_box_slot:
                            Distribution:= option{1.0}
                            Padding := margin{Left := 0.0, Top := 0.0, Right := 0.0, Bottom := 0.0}
                            HorizontalAlignment := horizontal_alignment.Fill
                            VerticalAlignment := vertical_alignment.Fill
                            Widget := stack_box :
                                Orientation := orientation.Horizontal
                                Slots := array :
                                    stack_box_slot :
                                        Padding:=margin{Left:=0.0 ,Top:= 0.0, Right:= 0.0, Bottom:= 0.0}
                                        HorizontalAlignment := horizontal_alignment.Center
                                        VerticalAlignment := vertical_alignment.Center
                                        Widget := IC_Area_TB
                                    stack_box_slot :
                                        Distribution:= option{1.0}
                                        Padding:=margin{Left:=10.0 ,Top:= 0.0, Right:= 0.0, Bottom:= 0.0}
                                        HorizontalAlignment := horizontal_alignment.Left
                                        VerticalAlignment := vertical_alignment.Center
                                        Widget := AreaBonus_TB
                                    stack_box_slot :
                                        Distribution:= option{1.0}
                                        Padding:=margin{Left:=0.0 ,Top:= 0.0, Right:= 0.0, Bottom:= 0.0}
                                        HorizontalAlignment := horizontal_alignment.Right
                                        VerticalAlignment := vertical_alignment.Center
                                        Widget := MyAreaBonus_TB
                        #白線
                        stack_box_slot:
                            Padding := margin{Left := 0.0, Top := 0.0, Right := 0.0, Bottom := 0.0}
                            HorizontalAlignment := horizontal_alignment.Fill
                            VerticalAlignment := vertical_alignment.Fill
                            Widget := WhiteLine2
                        #素材採取力
                        stack_box_slot:
                            Distribution:= option{1.0}
                            Padding := margin{Left := 0.0, Top := 0.0, Right := 0.0, Bottom := 0.0}
                            HorizontalAlignment := horizontal_alignment.Fill
                            VerticalAlignment := vertical_alignment.Fill
                            Widget := stack_box :
                                Orientation := orientation.Horizontal
                                Slots := array :
                                    stack_box_slot :
                                        Padding:=margin{Left:=0.0 ,Top:= 0.0, Right:= 0.0, Bottom:= 0.0}
                                        HorizontalAlignment := horizontal_alignment.Center
                                        VerticalAlignment := vertical_alignment.Center
                                        Widget := IC_Mining_TB
                                    stack_box_slot :
                                        Distribution:= option{1.0}
                                        Padding:=margin{Left:=10.0 ,Top:= 0.0, Right:= 0.0, Bottom:= 0.0}
                                        HorizontalAlignment := horizontal_alignment.Left
                                        VerticalAlignment := vertical_alignment.Center
                                        Widget := Mining_TB
                                    stack_box_slot :
                                        Distribution:= option{1.0}
                                        Padding:=margin{Left:=0.0 ,Top:= 0.0, Right:= 0.0, Bottom:= 0.0}
                                        HorizontalAlignment := horizontal_alignment.Right
                                        VerticalAlignment := vertical_alignment.Center
                                        Widget := MyMining_TB
        }
        set CanvasSlots += array
        #NowLevel
        {
            canvas_slot:
                Anchors := CenterAnchors
                Offsets := margin{Top:= -151.0 ,Left:= 465.0, Right:= 700.0, Bottom:= 280.0}
                Alignment := CenterVector
                SizeToContent := false
                Widget := stack_box:
                    Orientation := orientation.Vertical
                    Slots := array:
                        #現在のレベル{NowLevel}
                        stack_box_slot:
                            Padding := margin{Left := 0.0, Top := 0.0, Right := 0.0, Bottom := 0.0}
                            HorizontalAlignment := horizontal_alignment.Center
                            VerticalAlignment := vertical_alignment.Fill
                            Widget := MyLevelTextBlock
                        #基礎攻撃力
                        stack_box_slot:
                            Distribution:= option{0.5}
                            HorizontalAlignment := horizontal_alignment.Fill
                            VerticalAlignment := vertical_alignment.Fill
                            Widget := stack_box :
                                Orientation := orientation.Horizontal
                                Slots := array :
                                    stack_box_slot :
                                        Padding := margin{Left := 0.0, Top := 0.0, Right := 0.0, Bottom := 0.0}
                                        HorizontalAlignment := horizontal_alignment.Center
                                        VerticalAlignment := vertical_alignment.Center
                                        Widget := StrImage3
                                    stack_box_slot :
                                        Distribution := option{1.0}
                                        Padding := margin{Left := 10.0, Top := 0.0, Right := 0.0, Bottom := 0.0}
                                        HorizontalAlignment := horizontal_alignment.Left
                                        VerticalAlignment := vertical_alignment.Center
                                        Widget := BaseStrengthTextBlock2
                                    stack_box_slot :
                                        Distribution := option{1.0}
                                        Padding := margin{Left := 0.0, Top := 0.0, Right := 0.0, Bottom := 0.0}
                                        HorizontalAlignment := horizontal_alignment.Right
                                        VerticalAlignment := vertical_alignment.Center
                                        Widget := MyBaseStrengthTextBlock2
                        #合計採掘力
                        stack_box_slot:
                            Distribution:= option{0.5}
                            HorizontalAlignment := horizontal_alignment.Fill
                            VerticalAlignment := vertical_alignment.Fill
                            Widget := stack_box :
                                Orientation := orientation.Horizontal
                                Slots := array :
                                    stack_box_slot :
                                        Padding := margin{Left := 0.0, Top := 0.0, Right := 0.0, Bottom := 0.0}
                                        HorizontalAlignment := horizontal_alignment.Center
                                        VerticalAlignment := vertical_alignment.Center
                                        Widget := StrImage4
                                    stack_box_slot :
                                        Distribution := option{1.0}
                                        Padding := margin{Left := 10.0, Top := 0.0, Right := 0.0, Bottom := 0.0}
                                        HorizontalAlignment := horizontal_alignment.Left
                                        VerticalAlignment := vertical_alignment.Center
                                        Widget := TotalStrengthTextBlock2
                                    stack_box_slot :
                                        Distribution := option{1.0}
                                        Padding := margin{Left := 0.0, Top := 0.0, Right := 0.0, Bottom := 0.0}
                                        HorizontalAlignment := horizontal_alignment.Right
                                        VerticalAlignment := vertical_alignment.Center
                                        Widget := MyTotalStrengthTextBlock2
        }
        set CanvasSlots += array
        #NextLevel
        {
            canvas_slot:
                Anchors := CenterAnchors
                Offsets := margin{Top:= 150.633972 ,Left:= 465.0, Right:= 700.0, Bottom:= 280.0}
                Alignment := CenterVector
                SizeToContent := false
                Widget := stack_box:
                    Orientation := orientation.Vertical
                    Slots := array:
                        #次のレベル{NextLevel}
                        stack_box_slot:
                            Padding := margin{Left := 0.0, Top := 0.0, Right := 0.0, Bottom := 0.0}
                            HorizontalAlignment := horizontal_alignment.Center
                            VerticalAlignment := vertical_alignment.Fill
                            Widget := MyNextLevelTextBlock
                        #基礎攻撃力
                        stack_box_slot:
                            Distribution:= option{0.5}
                            HorizontalAlignment := horizontal_alignment.Fill
                            VerticalAlignment := vertical_alignment.Fill
                            Widget := stack_box :
                                Orientation := orientation.Horizontal
                                Slots := array :
                                    stack_box_slot :
                                        Padding := margin{Left := 0.0, Top := 0.0, Right := 0.0, Bottom := 0.0}
                                        HorizontalAlignment := horizontal_alignment.Center
                                        VerticalAlignment := vertical_alignment.Center
                                        Widget := StrImage5
                                    stack_box_slot :
                                        Distribution := option{1.0}
                                        Padding := margin{Left := 10.0, Top := 0.0, Right := 0.0, Bottom := 0.0}
                                        HorizontalAlignment := horizontal_alignment.Left
                                        VerticalAlignment := vertical_alignment.Center
                                        Widget := BaseStrengthTextBlock3
                                    stack_box_slot :
                                        Distribution := option{1.0}
                                        Padding := margin{Left := 0.0, Top := 0.0, Right := 0.0, Bottom := 0.0}
                                        HorizontalAlignment := horizontal_alignment.Right
                                        VerticalAlignment := vertical_alignment.Center
                                        Widget := MyBaseStrengthTextBlock3
                                    stack_box_slot :
                                        Padding := margin{Left := 0.0, Top := 0.0, Right := 0.0, Bottom := 0.0}
                                        HorizontalAlignment := horizontal_alignment.Fill
                                        VerticalAlignment := vertical_alignment.Center
                                        Widget := AddStrengthTextBlock
                        #合計採掘力
                        stack_box_slot:
                            Distribution:= option{0.5}
                            HorizontalAlignment := horizontal_alignment.Fill
                            VerticalAlignment := vertical_alignment.Fill
                            Widget := stack_box :
                                Orientation := orientation.Horizontal
                                Slots := array :
                                    stack_box_slot :
                                        Padding := margin{Left := 0.0, Top := 0.0, Right := 0.0, Bottom := 0.0}
                                        HorizontalAlignment := horizontal_alignment.Center
                                        VerticalAlignment := vertical_alignment.Center
                                        Widget := StrImage6
                                    stack_box_slot :
                                        Distribution := option{1.0}
                                        Padding := margin{Left := 10.0, Top := 0.0, Right := 0.0, Bottom := 0.0}
                                        HorizontalAlignment := horizontal_alignment.Left
                                        VerticalAlignment := vertical_alignment.Center
                                        Widget := TotalStrengthTextBlock3
                                    stack_box_slot :
                                        Distribution := option{1.0}
                                        Padding := margin{Left := 0.0, Top := 0.0, Right := 0.0, Bottom := 0.0}
                                        HorizontalAlignment := horizontal_alignment.Right
                                        VerticalAlignment := vertical_alignment.Center
                                        Widget := MyTotalStrengthTextBlock3
                                    stack_box_slot :
                                        Padding := margin{Left := 0.0, Top := 0.0, Right := 0.0, Bottom := 0.0}
                                        HorizontalAlignment := horizontal_alignment.Fill
                                        VerticalAlignment := vertical_alignment.Center
                                        Widget := AddStrengthTextBlock2
        }
        set CanvasSlots += array
        #Cost
        {
            canvas_slot:
                Anchors := CenterAnchors
                Offsets := margin{Top:= 344.420197 ,Left:= 343.393005, Right:= 100.0, Bottom:= 30.0}
                Alignment := vector2{X := 0.0, Y := 0.5}
                SizeToContent := true
                Widget := stack_box:
                    Orientation := orientation.Horizontal
                    Slots := array:
                        stack_box_slot :
                            Distribution := option{1.0}
                            Padding := margin{Left := 0.0, Top := 0.0, Right := 50.0, Bottom := 0.0}
                            HorizontalAlignment := horizontal_alignment.Left
                            VerticalAlignment := vertical_alignment.Fill
                            Widget := CostTextBlock
                        stack_box_slot :
                            Padding := margin{Left := 0.0, Top := 0.0, Right := 10.0, Bottom := 0.0}
                            HorizontalAlignment := horizontal_alignment.Fill
                            VerticalAlignment := vertical_alignment.Fill
                            Widget := SilverOreImage1
                        stack_box_slot :
                            Padding := margin{Left := 0.0, Top := 0.0, Right := 0.0, Bottom := 0.0}
                            HorizontalAlignment := horizontal_alignment.Right
                            VerticalAlignment := vertical_alignment.Fill
                            Widget := SilverCostTextBlock
        }
        set CanvasSlots += array
        #Close
        {
            canvas_slot:
                Anchors := CenterAnchors
                Offsets := margin{Top:= 435.186005, Left:= 0.0, Right := 200.0, Bottom := 75.0}
                Alignment := CenterVector
                SizeToContent := false
                Widget := CloseButton
        }
        set CanvasSlots += array
        #LevelUp
        {
            canvas_slot:
                Anchors := CenterAnchors
                Offsets := margin{Top:= 431.738800, Left:= 472.334076, Right := 300.934937, Bottom := 78.080002}
                Alignment := CenterVector
                SizeToContent := false
                Widget := LevelUpButton
        }
        set CanvasSlots += array
        #Auto
        {
            canvas_slot:
                Anchors := CenterAnchors
                Offsets := margin{Top:= 432.435120, Left:= 777.707947, Right := 266.614899, Bottom := 78.080002}
                Alignment := CenterVector
                SizeToContent := false
                Widget := AutoButton
        }
        
        if(set UIElements[Agent] = map{}){}
        if:
            set UIElements[Agent] = map:
                1 => (OreTextBlock, StrTextBlock, MyBaseStrengthTextBlock, MyBaseStrengthTextBlock2, MyBaseStrengthTextBlock3, MyPetStrengthTextBlock,
                MyTotalStrengthTextBlock, MyTotalStrengthTextBlock2, MyTotalStrengthTextBlock3, AddStrengthTextBlock, AddStrengthTextBlock2, TotalResourceTextBlock,
                SilverCostTextBlock, MyLevelTextBlock, MyNextLevelTextBlock, BaseStrengthTextBlock, BaseStrengthTextBlock2, BaseStrengthTextBlock3, PetStrengthTextBlock,
                TotalStrengthTextBlock, TotalStrengthTextBlock2, TotalStrengthTextBlock3, LvBonus_TB, MyLvBonus_TB, AreaBonus_TB, MyAreaBonus_TB, Mining_TB,
                MyMining_TB, AutoButton)

        Canvas : canvas = canvas:
            Slots := CanvasSlots

        return Canvas
