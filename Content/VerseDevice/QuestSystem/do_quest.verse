
using { /Fortnite.com/Devices }
using { /Verse.org/Simulation }
using { /UnrealEngine.com/Temporary/Diagnostics }

do_quest := class(creative_device):

    @editable
    SaveDataManager : Mining_System.Code.PlayerData.save_data_manager = Mining_System.Code.PlayerData.save_data_manager{}

    @editable
    RemoteManager : signal_remote_manager_device = signal_remote_manager_device{}

    @editable
    QuestTrigger1 : trigger_device = trigger_device{}

    @editable
    QuestTrigger2 : trigger_device = trigger_device{}
    
    OnBegin<override>()<suspends>:void=
        spawn{AwaitPrimary()}
        spawn{SecondaryPrimary()}
        GetPlayspace().PlayerAddedEvent().Subscribe(SpawnPlayer)
        
    SpawnPlayer(Agent:agent):void=
        spawn{AwaitPrimary()}
        spawn{SecondaryPrimary()}
    
    AwaitPrimary()<suspends>:void=
        loop:
            PushAgent := RemoteManager.PrimarySignalEvent.Await()

            # Quest
            if(Player := player[PushAgent]):
                NowQuest := SaveDataManager.GetNowQuest(Player)
                if(NowQuest = 1):
                    QuestTrigger1.Trigger(Player)

    SecondaryPrimary()<suspends>:void=
        loop:
            PushAgent := RemoteManager.SecondarySignalEvent.Await()

            # Quest
            if(Player := player[PushAgent]):
                NowQuest := SaveDataManager.GetNowQuest(Player)
                if(NowQuest = 1):
                    QuestTrigger2.Trigger(Player)