using { /Fortnite.com/Devices }
using { /Verse.org/Simulation }
using { /UnrealEngine.com/Temporary/Diagnostics }
using { /Verse.org/Colors }

Indicator_Control<public>:= class(creative_device):
    @editable
    MZ_Stage1 : mutator_zone_device = mutator_zone_device{}
    @editable
    Triggers_OnCompleted : []trigger_device = array{}
    @editable
    Triggers_OnStart : []trigger_device = array{}
    @editable
    Indicator999 : map_indicator_device = map_indicator_device{}

    @editable #homeの岩
    Indicator0 : map_indicator_device = map_indicator_device{}
    @editable #門0
    Indicator2 : map_indicator_device = map_indicator_device{}
    @editable #Area1-1の真ん中
    Indicator3 : map_indicator_device = map_indicator_device{}
    @editable #Area1-1のお地蔵さん
    Indicator4 : map_indicator_device = map_indicator_device{}
    @editable #門1
    Indicator5 : map_indicator_device = map_indicator_device{}
    @editable #homeの道場
    Indicator7 : map_indicator_device = map_indicator_device{}#
    @editable #門2
    Indicator8 : map_indicator_device = map_indicator_device{}
    @editable #Area1-2のおみくじ
    Indicator9 : map_indicator_device = map_indicator_device{}
    @editable #門3
    Indicator11 : map_indicator_device = map_indicator_device{}
    @editable #中村座
    Indicator12 : map_indicator_device = map_indicator_device{}
    @editable #中村座前にいるお姉さん
    Indicator13 : map_indicator_device = map_indicator_device{}
    @editable #鍛冶場
    Indicator14 : map_indicator_device = map_indicator_device{}
    @editable #Area1-2のボス封印石
    Indicator16 : map_indicator_device = map_indicator_device{}

    var NowQuestPerPlayer : [player]int = map{}

    @editable
    QuestArrays : []quest_element = array{}

    OnBegin<override>()<suspends>:void=
        #---------------Pulse見えるようにする処理---------------
        for(Index -> TheTrigger : Triggers_OnStart) :
            spawn{SetPulse(TheTrigger, Index)}
        #-----------------------------------------------------
            
        for (QuestElement : QuestArrays):
            spawn { AwaitQuestCompletion(QuestElement) }
        for(Index -> TheTrigger : Triggers_OnCompleted) :
            spawn{StopPulse(TheTrigger)}

        #Quest7,14では室内に移動したときにIndicatorが消えるように
        MZ_Stage1.AgentExitsEvent.Subscribe(OnExitFromMZ)
        MZ_Stage1.AgentEntersEvent.Subscribe(OnEnterToMZ)
        
    AwaitQuestCompletion(QuestElement : quest_element)<suspends>:void =
        loop:
            Info := QuestElement.QuestComplateEvent.Await()
            spawn{OnQuestComplete(Info(0), Info(1))}  # Info(0)はAgent、Info(1)はQuestID
    
    OnQuestComplete(Agent:agent, QuestID:int)<suspends>:void=
        if(Player := player[Agent]):
            # クエストIDに応じたインジケーターを起動
            Index : int = QuestID + 1
            spawn { SetFirstPulse(Player, Index) }

    SetFirstPulse<public>(Player : player, Index : int)<suspends>:void=
        logic{set NowQuestPerPlayer[Player] = Index}

    SetPulse(Trigger : trigger_device, Index : int)<suspends>:void=
        loop :
            IsAgent : ?agent = Trigger.TriggeredEvent.Await()
            Print("トリガー{Index}")
            if(Agent : agent = IsAgent?) :
                if(Index = 0) :
                    for(X := 0..5) :
                        Indicator999.ActivateObjectivePulse(Agent)
                        Indicator0.ActivateObjectivePulse(Agent)
                        Sleep(0.0)
                        Print("Activated{Index}")
                if(Index = 2) :
                    for(X := 0..5) :
                        Indicator999.ActivateObjectivePulse(Agent)
                        Indicator2.ActivateObjectivePulse(Agent)
                        Sleep(0.0)
                        Print("Activated{Index}")
                if(Index = 3) :
                    for(X := 0..5) :
                        Indicator999.ActivateObjectivePulse(Agent)
                        Indicator3.ActivateObjectivePulse(Agent)
                        Sleep(0.0)
                        Print("Activated{Index}")
                if(Index = 4) :
                    for(X := 0..5) :
                        Indicator999.ActivateObjectivePulse(Agent)
                        Indicator4.ActivateObjectivePulse(Agent)
                        Sleep(0.0)
                        Print("Activated{Index}")
                if(Index = 5) :
                    for(X := 0..5) :
                        Indicator999.ActivateObjectivePulse(Agent)
                        Indicator5.ActivateObjectivePulse(Agent)
                        Sleep(0.0)
                        Print("Activated{Index}")
                if(Index = 7) :
                    for(X := 0..5) :
                        Indicator999.ActivateObjectivePulse(Agent)
                        Indicator7.ActivateObjectivePulse(Agent)
                        Sleep(0.0)
                        Print("Activated{Index}")
                if(Index = 8) :
                    for(X := 0..5) :
                        Indicator999.ActivateObjectivePulse(Agent)
                        Indicator8.ActivateObjectivePulse(Agent)
                        Sleep(0.0)
                        Print("Activated{Index}")
                if(Index = 9) :
                    for(X := 0..5) :
                        Indicator999.ActivateObjectivePulse(Agent)
                        Indicator9.ActivateObjectivePulse(Agent)
                        Sleep(0.0)
                        Print("Activated{Index}")
                if(Index = 11) :
                    for(X := 0..5) :
                        Indicator999.ActivateObjectivePulse(Agent)
                        Indicator11.ActivateObjectivePulse(Agent)
                        Sleep(0.0)
                        Print("Activated{Index}")
                if(Index = 12) :
                    for(X := 0..5) :
                        Indicator999.ActivateObjectivePulse(Agent)
                        Indicator12.ActivateObjectivePulse(Agent)
                        Sleep(0.0)
                        Print("Activated{Index}")
                if(Index = 13) :
                    for(X := 0..5) :
                        Indicator999.ActivateObjectivePulse(Agent)
                        Indicator13.ActivateObjectivePulse(Agent)
                        Sleep(0.0)
                        Print("Activated{Index}")
                if(Index = 14) :
                    for(X := 0..5) :
                        Indicator999.ActivateObjectivePulse(Agent)
                        Indicator14.ActivateObjectivePulse(Agent)
                        Sleep(0.0)
                        Print("Activated{Index}")
                if(Index = 16) :
                    for(X := 0..5) :
                        Indicator999.ActivateObjectivePulse(Agent)
                        Indicator16.ActivateObjectivePulse(Agent)
                        Sleep(0.0)
                        Print("Activated{Index}")

                #---------------全体向けのPulse消す処理---------------
                Indicator999.DeactivateObjectivePulse(Agent)
                Print("DeactivePulse ForAll")
                #---------------------------------------------------

    StopPulse(TheTrigger : trigger_device)<suspends>:void=
        #1つ前のIndicatorを停止
        loop:
            IsAgent : ?agent = TheTrigger.TriggeredEvent.Await()
            Sleep(0.1)
            if :
                Agent : agent = IsAgent?
                Player : player = player[Agent]
                Index := NowQuestPerPlayer[Player] - 1 
            then :
                if(Index = 0) :
                    Indicator0.DeactivateObjectivePulse(Agent)
                if(Index = 2) :
                    Indicator2.DeactivateObjectivePulse(Agent)
                if(Index = 3) :
                    Indicator3.DeactivateObjectivePulse(Agent)
                if(Index = 4) :
                    Indicator4.DeactivateObjectivePulse(Agent)
                if(Index = 5) :
                    Indicator5.DeactivateObjectivePulse(Agent)
                if(Index = 7) :
                    Indicator7.DeactivateObjectivePulse(Agent)
                if(Index = 8) :
                    Indicator8.DeactivateObjectivePulse(Agent)
                if(Index = 9) :
                    Indicator9.DeactivateObjectivePulse(Agent)
                if(Index = 11) :
                    Indicator11.DeactivateObjectivePulse(Agent)
                if(Index = 12) :
                    Indicator12.DeactivateObjectivePulse(Agent)
                if(Index = 13) :
                    Indicator13.DeactivateObjectivePulse(Agent)
                if(Index = 14) :
                    Indicator14.DeactivateObjectivePulse(Agent)
                if(Index = 16) :
                    Indicator16.DeactivateObjectivePulse(Agent)
                Print("Deactivated{Index}!!")

    OnExitFromMZ(Agent : agent):void=
        if :
            Player : player = player[Agent]
            Index := NowQuestPerPlayer[Player]
        then :
            if(Index = 0) :
                Indicator0.DeactivateObjectivePulse(Agent)
            if(Index = 2) :
                Indicator2.DeactivateObjectivePulse(Agent)
            if(Index = 3) :
                Indicator3.DeactivateObjectivePulse(Agent)
            if(Index = 4) :
                Indicator4.DeactivateObjectivePulse(Agent)
            if(Index = 5) :
                Indicator5.DeactivateObjectivePulse(Agent)
            if(Index = 7) :
                Indicator7.DeactivateObjectivePulse(Agent)
            if(Index = 8) :
                Indicator8.DeactivateObjectivePulse(Agent)
            if(Index = 9) :
                Indicator9.DeactivateObjectivePulse(Agent)
            if(Index = 11) :
                Indicator11.DeactivateObjectivePulse(Agent)
            if(Index = 12) :
                Indicator12.DeactivateObjectivePulse(Agent)
            if(Index = 13) :
                Indicator13.DeactivateObjectivePulse(Agent)
            if(Index = 14) :
                Indicator14.DeactivateObjectivePulse(Agent)
            if(Index = 16) :
                Indicator16.DeactivateObjectivePulse(Agent)
            Print("Deactivated{Index}!!when ExitMZ")
                
    OnEnterToMZ(Agent : agent):void=
        if :
            Player : player = player[Agent]
            Index := NowQuestPerPlayer[Player]
        then :
            if(Index = 0) :
                Indicator0.ActivateObjectivePulse(Agent)
            if(Index = 2) :
                Indicator2.ActivateObjectivePulse(Agent)
            if(Index = 3) :
                Indicator3.ActivateObjectivePulse(Agent)
            if(Index = 4) :
                Indicator4.ActivateObjectivePulse(Agent)
            if(Index = 5) :
                Indicator5.ActivateObjectivePulse(Agent)
            if(Index = 7) :
                Indicator7.ActivateObjectivePulse(Agent)
            if(Index = 8) :
                Indicator8.ActivateObjectivePulse(Agent)
            if(Index = 9) :
                Indicator9.ActivateObjectivePulse(Agent)
            if(Index = 11) :
                Indicator11.ActivateObjectivePulse(Agent)
            if(Index = 12) :
                Indicator12.ActivateObjectivePulse(Agent)
            if(Index = 13) :
                Indicator13.ActivateObjectivePulse(Agent)
            if(Index = 14) :
                Indicator14.ActivateObjectivePulse(Agent)
            if(Index = 16) :
                Indicator16.ActivateObjectivePulse(Agent)