using { /Fortnite.com/Devices }
using { /Verse.org/Simulation }
using { /Verse.org/Colors }
using {Mining_System.Code.PlayerData}
using {Mining_System.Code.Skill}
using {VerseDevice.QuestSystem}

quest_element<public> := class(creative_device):

    #セーブデータで使用

    @editable
    Title<public> : language_data = language_data{}

    @editable
    Description<public> : language_data = language_data{}

    @editable
    Progress<public> : int = 5

    @editable
    IsSaveProgress<public> : logic = false

    @editable
    QuestID : int = -1
    @editable
    SaveDataManager : save_data_manager = save_data_manager{}
    @editable
    SkillManager : skill_manager = skill_manager{}

    @editable #homeの岩
    Indicator0 : map_indicator_device = map_indicator_device{}
    @editable #門0
    Indicator2 : map_indicator_device = map_indicator_device{}
    @editable #Area1-1の真ん中
    Indicator3 : map_indicator_device = map_indicator_device{}
    @editable #Area1-1のお地蔵さん
    Indicator4 : map_indicator_device = map_indicator_device{}
    @editable #門1
    Indicator5 : map_indicator_device = map_indicator_device{}
    @editable #homeの道場
    Indicator7 : map_indicator_device = map_indicator_device{}#
    @editable #門2
    Indicator8 : map_indicator_device = map_indicator_device{}
    @editable #Area1-2のおみくじ
    Indicator9 : map_indicator_device = map_indicator_device{}
    @editable #門3
    Indicator11 : map_indicator_device = map_indicator_device{}
    @editable #中村座
    Indicator12 : map_indicator_device = map_indicator_device{}
    @editable #中村座前にいるお姉さん
    Indicator13 : map_indicator_device = map_indicator_device{}
    @editable #鍛冶場
    Indicator14 : map_indicator_device = map_indicator_device{}
    @editable #Area1-2のボス封印石
    Indicator16 : map_indicator_device = map_indicator_device{}

    QuestComplateEvent<public> : event(tuple(agent, int)) = event(tuple(agent, int)){}
    QuestProgressUpdatedEvent<public> : event(agent) = event(agent){}
    QuestEvnetForSave<public> : event(tuple(agent, int)) =  event(tuple(agent, int)){}

    var ProgressPerAgent<public> : [agent]int = map{}

    Init<public>(InAgent:agent)<suspends>:void=
        return

    Play<public><final>(InAgent:agent):void=
        if(set ProgressPerAgent[InAgent] = 0){}
    #そのクエストが完了したかどうかを判定
    CheckComplete<public>(InAgent:agent):void=
        if(CurrentProgrss := ProgressPerAgent[InAgent]):
            if(CurrentProgrss >= Progress):
                QuestComplateEvent.Signal(InAgent, QuestID)
                if(IsSaveProgress?):
                    QuestEvnetForSave.Signal((InAgent, CurrentProgrss))
    #quest_sampleのIncrementTriggerが起動したとき
    Increment<public>(InAgent:agent):void=
        if(CurrentProgress := set ProgressPerAgent[InAgent] += 1):
            QuestProgressUpdatedEvent.Signal(InAgent)
            if(IsSaveProgress?):
                QuestEvnetForSave.Signal(InAgent, CurrentProgress)
            CheckComplete(InAgent)

    Decrement<public>(InAgent:agent):void=
        if(CurrentProgress := set ProgressPerAgent[InAgent] -= 1):
            QuestProgressUpdatedEvent.Signal(InAgent)
            if(IsSaveProgress?):
                QuestEvnetForSave.Signal(InAgent, CurrentProgress)
            CheckComplete(InAgent)

    SetProgress<public>(InAgent:agent, InProgress:int):void=
        if(CurrentProgress := set ProgressPerAgent[InAgent] = InProgress):
            QuestProgressUpdatedEvent.Signal(InAgent)
            if(IsSaveProgress?):
                QuestEvnetForSave.Signal(InAgent, CurrentProgress)
            CheckComplete(InAgent)

    ResetProgress<public>(InAgent:agent):void=
        if(CurrentProgress := set ProgressPerAgent[InAgent] = 0):
            QuestProgressUpdatedEvent.Signal(InAgent)
            if(IsSaveProgress?):
                QuestEvnetForSave.Signal(InAgent, CurrentProgress)

    GetProgress<public>(InAgent:agent):int=
        if(CurrentProgrss := ProgressPerAgent[InAgent]):
            CurrentProgrss
        else:
            0

    CheckCorrectness<public>(Agent : agent)<suspends>:void=
        if :
            Player : player = player[Agent]
        then :
            Print("クエスト進度チェック", ?Color := NamedColors.Red, ? Duration := 5.0)
            UnlockedStages := SaveDataManager.GetUnlockedStages(Player)
            LastStage := UnlockedStages.Length - 1
            #クエスト1 10回叩く
            if(QuestID = 0) :

            #クエスト2 str開く
            if(QuestID = 1) :

            #クエスト3 ステージ1開放 バリア0
            if(QuestID = 2) :
                Sleep(0.1)
                if(LastStage >= 0) :
                    QuestComplateEvent.Signal(Agent, QuestID)
                    Sleep(1.0)
                    for(X := 0..5) :
                        Indicator0.DeactivateObjectivePulse(Agent)
                        Indicator2.DeactivateObjectivePulse(Agent)

            #クエスト4 岩10個破壊
            if(QuestID = 3) :

            #クエスト5 スキル習得 エリア1
            if(QuestID = 4) :
                Sleep(0.1)
                if(SkillLevel := SkillManager.ReturnSkillLevel[Player, 15]) :
                    if(SkillLevel >= 1) :
                        QuestComplateEvent.Signal(Agent, QuestID)
                        Sleep(1.0)
                        for(X := 0..5) :
                            Indicator0.DeactivateObjectivePulse(Agent)
                            Indicator4.DeactivateObjectivePulse(Agent)

            #クエスト6 ステージ1-2開放 バリア1
            if(QuestID = 5) :
                Sleep(0.1)
                if(LastStage >= 1) :
                    QuestComplateEvent.Signal(Agent, QuestID)
                    Sleep(1.0)
                    for(X := 0..5) :
                        Indicator0.DeactivateObjectivePulse(Agent)
                        Indicator5.DeactivateObjectivePulse(Agent)

            #クエスト7 テレポート
            if(QuestID = 6) :

            #クエスト8 ガチャアイテムドロップスキルを開放
            if(QuestID = 7) :
                Sleep(0.1)
                if(SkillLevel := SkillManager.ReturnSkillLevel[Player, 10]) :
                    if(SkillLevel >= 1) :
                        QuestComplateEvent.Signal(Agent, QuestID)
                        Sleep(1.0)
                        for(X := 0..5) :
                            Indicator0.DeactivateObjectivePulse(Agent)
                            Indicator7.DeactivateObjectivePulse(Agent)

            #クエスト9 ステージ1-3開放 バリア2
            if(QuestID = 8) :
                Sleep(0.1)
                if(LastStage >= 2) :
                    QuestComplateEvent.Signal(Agent, QuestID)
                    Sleep(1.0)
                    for(X := 0..5) :
                        Indicator0.DeactivateObjectivePulse(Agent)
                        Indicator8.DeactivateObjectivePulse(Agent)

            #クエスト10 ペット引く
            if(QuestID = 9) :

            #クエスト11 ペット装備
            if(QuestID = 10) :

            #クエスト12 ステージ1-4開放 バリア3
            if(QuestID = 11) :
                Sleep(0.1)
                if(LastStage >= 3) :
                    QuestComplateEvent.Signal(Agent, QuestID)
                    Sleep(1.0)
                    for(X := 0..5) :
                        Indicator0.DeactivateObjectivePulse(Agent)
                        Indicator11.DeactivateObjectivePulse(Agent)

            #クエスト13 中村座開放
            if(QuestID = 12) :
                Sleep(0.1)
                if(SkillLevel := SkillManager.ReturnSkillLevel[Player, 5]) :
                    if(SkillLevel >= 1) :
                        QuestComplateEvent.Signal(Agent, QuestID)
                        Sleep(1.0)
                        for(X := 0..5) :
                            Indicator0.DeactivateObjectivePulse(Agent)
                            Indicator12.DeactivateObjectivePulse(Agent)

            #クエスト14 チケット消費
            if(QuestID = 13) :

            #クエスト15 武器を入手
            if(QuestID = 14) :
                Sleep(0.1)
                WeaponLevel := SaveDataManager.GetWeaponLevel(Player)
                if(WeaponLevel >= 1) :
                    QuestComplateEvent.Signal(Agent, QuestID)
                    Sleep(1.0)
                    for(X := 0..5) :
                        Indicator0.DeactivateObjectivePulse(Agent)
                        Indicator14.DeactivateObjectivePulse(Agent)

            #クエスト16 武器を強化
            if(QuestID = 15) :
                
            #クエスト17 ボスを討伐
            if(QuestID = 16) :
                
            #クエスト18 ボスメニュー選択
            if(QuestID = 17) :
                
            #クエスト19 ペット進化
            if(QuestID = 18) :
                Sleep(0.1)
                var IsEvolved : logic = false
                for(PetElement : SaveDataManager.GetPets[Player]) :
                    if(PetElement(1).Level >= 31) :
                        set IsEvolved = true
                if(IsEvolved = true) :
                    QuestComplateEvent.Signal(Agent, QuestID)
                