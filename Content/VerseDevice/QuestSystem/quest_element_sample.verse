using { /Verse.org/Simulation }
using { /Fortnite.com/Devices }

quest_element_sample<public> := class(quest_element):

    @editable
    IncrementTrigger : trigger_device = trigger_device{}

    OnBegin<override>()<suspends>:void=
        IncrementTrigger.TriggeredEvent.Subscribe(OnIncrementTriggered)

    OnIncrementTriggered<private>(MaybeAgent:?agent):void=
        if(Agent := MaybeAgent?):
            Increment(Agent)
