using { /Fortnite.com/Devices }
using { /Verse.org/Simulation }
using { /Verse.org/Colors }
using { /UnrealEngine.com/Temporary/UI }
using { Mining_System.Code.PlayerData }
using { Mining_System.Code.Utility }

quest_sequencer<public> := class(creative_device):

    @editable
    CongratsHUD : hud_message_device = hud_message_device{}

    @editable
    QuestArrays : []quest_element = array{}

    @editable
    Switchs : []switch_device = array{}

    @editable
    QuestStartTriggers : []trigger_device = array{}

    var QuestUIForPlayers : [player]quest_ui = map{}

    # クエストを中断する場合は、このイベントを呼び出す
    CancelQuest<public> : event(agent) = event(agent){}
    QuestComplateEvent<public> : event(tuple(agent, int)) = event(tuple(agent, int)){}
    UpdateLanguage<public> : event(agent) = event(agent){}

    OnBegin<override>():void=
        GetPlayspace().PlayerRemovedEvent().Subscribe(OnPlayerRemoved)

    HasSaveData<public>(Player:player)<transacts>:logic=
        if (NowData := SaveData[Player]):
            return true
        else:
            Print("Error: HasSaveData")
            return false

    RemoveKeyFromWidgetMap(ExampleMap: [player]quest_ui, ElementToRemove: player): [player]quest_ui =
        var NewMap: [player]quest_ui = map{}
        for (Key -> Value : ExampleMap, Key <> ElementToRemove):
            set NewMap = ConcatenateMaps(NewMap, map{Key => Value})
        return NewMap
    
    OnPlayerRemoved(Player: player): void =
        if (QuestUI := QuestUIForPlayers[Player]):
            spawn { RemoveUIForPlayer(Player) }

        set QuestUIForPlayers = RemoveKeyFromWidgetMap(QuestUIForPlayers, Player)
            
    RemoveUIForPlayer(Player: player)<suspends>:void =
        if (QuestUI := QuestUIForPlayers[Player]):
            QuestUI.RemoveUI()

    PlaySequence<public>(InAgent: agent, StartIndex: int)<suspends>: void =
        if (Player := player[InAgent], not QuestUIForPlayers[Player]):
            if(HasSaveData(Player)?):
                Widget := quest_ui{Player := Player}
                logic{set QuestUIForPlayers[Player] = Widget}
                Widget.AddGameplayUI()
        
                race:
                    block :
                        # 1つ目の処理: クエストキャンセルの待機
                        loop:
                            RemovedAgent := CancelQuest.Await()  # プレイヤー退出イベントを待機
                            if (RemovedAgent = InAgent):
                                Widget.RemoveUI()
                                ResetQuest(Player)
                                
                                set QuestUIForPlayers = RemoveKeyFromWidgetMap(QuestUIForPlayers, Player)
                                break
                    block :
                        # 2つ目の処理: クエスト進行
                        for (QuestIndex := StartIndex..QuestArrays.Length - 1, QuestElement := QuestArrays[QuestIndex]):
                            # クエスト開始時のトリガー処理
                            if (TheTrigger := QuestStartTriggers[QuestIndex]):
                                TheTrigger.Trigger(InAgent)
                                Print("クエスト開始 {QuestIndex}")
                
                            # クエスト要素の再生と初期化
                            QuestElement.Play(InAgent)
                            spawn {QuestElement.Init(InAgent)}
                
                            # UIの初期設定
                            Widget.QuestTitle.SetText(StringToMessage(QuestElement.Title.Localize(Player)))
                            Widget.QuestDescription.SetText(StringToMessage(QuestElement.Description.Localize(Player)))
                            Widget.QuestProgress.SetText(StringToMessage("{QuestElement.GetProgress(InAgent)} / {QuestElement.Progress}"))
                
                            # 進行状況の保存処理
                            if(HasSaveData(Player)?):
                                if (QuestElement.IsSaveProgress?, NowData := SaveData[Player]):
                                    # クエスト進行状況の存在確認と取得
                                    if (NowProgressOption := NowData.Quest_Progress[QuestIndex]):
                                        # 存在する場合は進行状況を取得
                                        NowProgress := NowProgressOption
                                        QuestElement.SetProgress(InAgent, NowProgress)
                                    else:
                                        Print("スキルステータスが存在しません。")
                                else:
                                    Print("進行状況の保存が無効です。")
                            else:
                                Print("保存が無効です。")
                
                        # クエスト完了処理
                            race:
                                block:
                                    spawn{QuestElement.CheckCorrectness(InAgent)}
                                    if(HasSaveData(Player)?):
                                        QuestElement.QuestComplateEvent.Await()
                                        QuestComplateEvent.Signal(InAgent, QuestIndex)
                                        Print("クエスト完了: {QuestIndex}")
                                    else:
                                        Print("プレイヤーのデータが見つかりません。スキップします。")
                                loop:
                                    Info := QuestElement.QuestEvnetForSave.Await()
                                    if(HasSaveData(Player)?):
                                        if (NowData := SaveData[Player]):
                                            if:
                                                set SaveData[Player] = save_data:
                                                    MakeSaveData<constructor>(NowData)
                                                    Quest_Progress := AddMap(NowData.Quest_Progress, (QuestIndex, Info(1)))
                
                                # クエスト進行状況の更新イベント
                                loop:
                                    UAgent := QuestElement.QuestProgressUpdatedEvent.Await()
                                    if (UAgent = InAgent):
                                        if(HasSaveData(Player)?):
                                            Widget.QuestProgress.SetText(StringToMessage("{QuestElement.GetProgress(InAgent)} / {QuestElement.Progress}"))
                
                                # 言語更新処理（更新が必要な場合のみ）
                                loop:
                                    UAgent := UpdateLanguage.Await()
                                    Sleep(0.5)
                                    if (UAgent = InAgent):
                                        Print("言語更新中...")
                                        Widget.QuestTitle.SetText(StringToMessage(QuestElement.Title.Localize(Player)))
                                        Widget.QuestDescription.SetText(StringToMessage(QuestElement.Description.Localize(Player)))
                                    else:
                                        Sleep(0.5)  # 不要な処理を避けるために少し待機
                
                            # クエストが終了した場合の処理
                            Print("クエスト終了: {QuestIndex}")
                            CongratsHUD.Show(Player)
                        #全クエスト完了したらUI非表示
                        Widget.RemoveUI()
                        Print("クエスト終了、ウィジェットをRemove", ?Color := NamedColors.Red)
            
        else:
            Print("既にUIが存在します。")

    UpdateLanguageSignal<public>(InAgent:agent):void=
        UpdateLanguage.Signal(InAgent)

    IsComplate<public>(InAgent:agent, Index:int)<decides><transacts>:void=
        Player := player[InAgent]
        NowData := SaveData[Player]
        NowData.Quest_Completed.Find[Index]

    ResetQuest<public>(InAgent:agent):void=
        if:
            Player := player[InAgent]
            NowData := SaveData[Player]
            set SaveData[Player] = save_data:
                MakeSaveData<constructor>(NowData)
                Quest_Completed := array{}
                Quest_Progress := map{}
        then:
            CancelQuest.Signal(InAgent)

            # リセットと同時に再生しなければ削除してよい
            spawn{PlaySequence(InAgent, 0)}

