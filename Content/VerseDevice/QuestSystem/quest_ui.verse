using { /Fortnite.com/UI }
using { /Verse.org/Simulation }
using { /UnrealEngine.com/Temporary/UI }
using { /Verse.org/Colors }
using { /UnrealEngine.com/Temporary/SpatialMath }

quest_ui<public> := class:

    # UI要素
    QuestTitle<public> : text_block = text_block{DefaultTextColor := NamedColors.Orange}
    QuestDescription<public> : text_block = text_block{DefaultTextColor := NamedColors.White}
    QuestProgress<public> : text_block = text_block{DefaultTextColor := NamedColors.White}

    Player<public> : player

    var MaybeCanvas : ?canvas = false
    var QuestUIMap : [player]quest_ui = map{}

    CloseEvent<public> : event() = event(){}


    AddGameplayUI<public>():void=
        if(PlayerUI := GetPlayerUI[Player]):
            NewCanvas := CreateCanvas()
            PlayerUI.AddWidget(NewCanvas)
            set MaybeCanvas = option{NewCanvas}

    RemoveUI<public>():void=
        if(PlayerUI := GetPlayerUI[Player], OldCanvas := MaybeCanvas?):
            PlayerUI.RemoveWidget(OldCanvas)
            set MaybeCanvas = false
            CloseEvent.Signal()

    CreateCanvas():canvas=
        canvas:
            Slots := array:
                canvas_slot:
                    Anchors := anchors:
                        Minimum := vector2{X := 0.99, Y := 0.5}
                        Maximum := vector2{X := 0.99, Y := 0.5}
                    Alignment := vector2{X := 1.0, Y := 0.5}
                    SizeToContent := true
                    Widget := overlay:
                        Slots := array:
                            overlay_slot:
                                Widget := texture_block{DefaultImage := icon_set_2.Quest_BG, DefaultDesiredSize := vector2{X := 496*1.25, Y := 201*1.25}}
                            overlay_slot:
                                Padding := margin:
                                    Left := 10.0
                                    Right := 10.0
                                    Top := 10.0
                                    Bottom := 10.0
                                Widget :=
                                    stack_box:
                                        Orientation := orientation.Vertical
                                        Slots := array:
                                            stack_box_slot:
                                                HorizontalAlignment := horizontal_alignment.Center
                                                VerticalAlignment := vertical_alignment.Top
                                                Padding := margin:
                                                    Left := 10.0
                                                    Right := 10.0
                                                    Top := 10.0
                                                    Bottom := 25.0
                                                Widget := QuestTitle
                                            stack_box_slot:
                                                HorizontalAlignment := horizontal_alignment.Fill
                                                VerticalAlignment := vertical_alignment.Fill
                                                Padding := margin:
                                                    Left := 10.0
                                                    Right := 10.0
                                                    Top := 10.0
                                                    Bottom := 25.0
                                                Widget := stack_box:
                                                    Orientation := orientation.Horizontal
                                                    Slots := array:
                                                        stack_box_slot:
                                                            HorizontalAlignment := horizontal_alignment.Left
                                                            VerticalAlignment := vertical_alignment.Fill
                                                            Padding := margin:
                                                                Left := 10.0
                                                                Right := 10.0
                                                                Top := 10.0
                                                                Bottom := 10.0
                                                            Widget := QuestDescription
                                                        stack_box_slot:
                                                            HorizontalAlignment := horizontal_alignment.Right
                                                            VerticalAlignment := vertical_alignment.Fill
                                                            Padding := margin:
                                                                Left := 10.0
                                                                Right := 10.0
                                                                Top := 10.0
                                                                Bottom := 10.0
                                                            Widget := QuestProgress
                                            