
using { /Verse.org/Assets }
using { /Verse.org/Simulation }
using { /Fortnite.com/UI }
using { /Fortnite.com/Devices }
using { /Fortnite.com/Characters }
using { /UnrealEngine.com/Temporary/UI }
using { /UnrealEngine.com/Temporary/SpatialMath }
using { /Verse.org/Simulation }
using { /UnrealEngine.com/Temporary/Diagnostics }
using { /Verse.org/Colors }
using { /Verse.org/Colors/NamedColors }
using { Mining_System.Code.PlayerData }

# See https://dev.epicgames.com/documentation/en-us/uefn/create-your-own-device-in-verse for how to create a verse device.

# A Verse-authored creative device that can be placed in a level
Ui_teleport_device := class(creative_device):
    #SFX
    @editable
    SFX_Open : audio_player_device = audio_player_device{}
    @editable
    SFX_Close : audio_player_device = audio_player_device{}
    @editable
    SFX_Warp : audio_player_device = audio_player_device{}

    @editable
    PlayerSpawners : []player_spawner_device = array{}
    @editable
    TeleportButtons : []button_device = array{}

    @editable
    TP_Theater : teleporter_device = teleporter_device{}
    @editable
    TP_SkillUpgrade : teleporter_device = teleporter_device{}
    @editable
    TP_Yokai : teleporter_device = teleporter_device{}
    @editable
    TP_BlackSmith : teleporter_device = teleporter_device{}
    @editable
    TP_PortTown : teleporter_device = teleporter_device{}
    @editable
    TP_RestSpot : teleporter_device = teleporter_device{}
    @editable
    TP_Prestige : teleporter_device = teleporter_device{}

    @editable
    TP_1_1 : teleporter_device = teleporter_device{}
    @editable
    TP_1_2 : teleporter_device = teleporter_device{}
    @editable
    TP_1_3 : teleporter_device = teleporter_device{}
    @editable
    TP_1_4 : teleporter_device = teleporter_device{}
    @editable
    TP_2_1 : teleporter_device = teleporter_device{}
    @editable
    TP_2_2 : teleporter_device = teleporter_device{}
    @editable
    TP_2_3 : teleporter_device = teleporter_device{}
    @editable
    TP_2_4 : teleporter_device = teleporter_device{}
    @editable
    TP_3_1 : teleporter_device = teleporter_device{}
    @editable
    TP_3_2 : teleporter_device = teleporter_device{}
    @editable
    TP_3_3 : teleporter_device = teleporter_device{}
    @editable
    TP_3_4 : teleporter_device = teleporter_device{}

    @editable
    SaveDataManager : save_data_manager = save_data_manager{}


    var TeleportCanvasPerAgent : [agent]canvas = map{}
    
    StringToMessage<public><localizes>(Text: string): message = "{Text}"

    var StageUIElements : [agent][int]tuple(button_quiet, texture_block) = map{}
    var LocalizeUIElements : [agent]tuple(text_block, button_quiet, button_quiet, button_quiet, button_quiet, button_quiet, button_quiet) = map{}


    OnBegin<override>()<suspends>:void=
        for (TeleportButton : TeleportButtons):
            TeleportButton.InteractedWithEvent.Subscribe(OnTeleportButtonEvented)

    OnPlayerSpawned(Agent : agent):void=
        if (TeleportCanvas := TeleportCanvasPerAgent[Agent]):
            return

        if (Player := player[Agent], PlayerUI := GetPlayerUI[Player]):
            TeleportCanvas : canvas = CreateTeleportCanvas(Agent, Player)
            TeleportCanvas.SetVisibility(widget_visibility.Collapsed)
            if (set TeleportCanvasPerAgent[Agent] = TeleportCanvas){}

    OnTeleportButtonEvented(Agent : agent):void=
        SFX_Open.Play(Agent)
        OnPlayerSpawned(Agent)
        Print("OnTeleportButtonEvented")
        if (Player := player[Agent], PlayerUI := GetPlayerUI[Player], TeleportCanvas := TeleportCanvasPerAgent[Agent]):
            TeleportCanvas.SetVisibility(widget_visibility.Visible)
            PlayerUI.AddWidget(TeleportCanvas, player_ui_slot{InputMode := ui_input_mode.All})
            for(UnlockedStageElement : SaveDataManager.GetUnlockedStages(Player), StageElement := StageUIElements[Agent][UnlockedStageElement]):
                StageElement(0).SetVisibility(widget_visibility.Visible)
                StageElement(1).SetVisibility(widget_visibility.Visible)
                spawn{LocalizeText(Agent, Player)}

    LocalizeText(Agent : agent, Player : player)<suspends>:void=
        var LanguageIndex : int = SaveDataManager.GetLanguage(Player)
        TownText := if(LanguageIndex = 0){"映画村"} else {"Base"}
        TheaterText := if(LanguageIndex = 0){"中村座"} else {"Teater"}
        SkillUpgradeText := if(LanguageIndex = 0){"忍者屋敷"} else {"Skill Upgrade"}
        YokaiText := if(LanguageIndex = 0){"お化け屋敷"} else {"YO-KAI"}
        BlackSmithText := if(LanguageIndex = 0){"鍛冶屋"} else {"Black Smith"}
        PortTownText := if(LanguageIndex = 0){"港町"} else {"Port Town"}
        RestSpotText := if(LanguageIndex = 0){"休憩所"} else {"Rest Spot"}
        if(Element := LocalizeUIElements[Agent]) :
            Element(0).SetText(StringToMessage(TownText))
            Element(1).SetText(StringToMessage(TheaterText))
            Element(2).SetText(StringToMessage(SkillUpgradeText))
            Element(3).SetText(StringToMessage(YokaiText))
            Element(4).SetText(StringToMessage(BlackSmithText))
            Element(5).SetText(StringToMessage(PortTownText))
            Element(6).SetText(StringToMessage(RestSpotText))

    # ボタン処理
    OnTheaterClicked(WidgetMessage : widget_message):void=
        if(Agent : agent = agent[WidgetMessage.Player]) :
            SFX_Warp.Play(Agent)
        if (PlayerUI := GetPlayerUI[WidgetMessage.Player], TeleportCanvas := TeleportCanvasPerAgent[WidgetMessage.Player]):
            TP_Theater.Teleport(WidgetMessage.Player)
            OnCloseButtonClicked(WidgetMessage)
            
    OnSkillUpgradeClicked(WidgetMessage : widget_message):void=
        if(Agent : agent = agent[WidgetMessage.Player]) :
            SFX_Warp.Play(Agent)
        if (PlayerUI := GetPlayerUI[WidgetMessage.Player], TeleportCanvas := TeleportCanvasPerAgent[WidgetMessage.Player]):
            TP_SkillUpgrade.Teleport(WidgetMessage.Player)
            OnCloseButtonClicked(WidgetMessage)
            
    OnYokaiClicked(WidgetMessage : widget_message):void=
        if(Agent : agent = agent[WidgetMessage.Player]) :
            SFX_Warp.Play(Agent)
        if (PlayerUI := GetPlayerUI[WidgetMessage.Player], TeleportCanvas := TeleportCanvasPerAgent[WidgetMessage.Player]):
            TP_Yokai.Teleport(WidgetMessage.Player)
            OnCloseButtonClicked(WidgetMessage)
            
    OnBlackSmithClicked(WidgetMessage : widget_message):void=
        if(Agent : agent = agent[WidgetMessage.Player]) :
            SFX_Warp.Play(Agent)
        if (PlayerUI := GetPlayerUI[WidgetMessage.Player], TeleportCanvas := TeleportCanvasPerAgent[WidgetMessage.Player]):
            TP_BlackSmith.Teleport(WidgetMessage.Player)
            OnCloseButtonClicked(WidgetMessage)
            
    OnPortTownClicked(WidgetMessage : widget_message):void=
        if(Agent : agent = agent[WidgetMessage.Player]) :
            SFX_Warp.Play(Agent)
        if (PlayerUI := GetPlayerUI[WidgetMessage.Player], TeleportCanvas := TeleportCanvasPerAgent[WidgetMessage.Player]):
            TP_PortTown.Teleport(WidgetMessage.Player)
            OnCloseButtonClicked(WidgetMessage)
            
    OnRestSpotClicked(WidgetMessage : widget_message):void=
        if(Agent : agent = agent[WidgetMessage.Player]) :
            SFX_Warp.Play(Agent)
        if (PlayerUI := GetPlayerUI[WidgetMessage.Player], TeleportCanvas := TeleportCanvasPerAgent[WidgetMessage.Player]):
            TP_RestSpot.Teleport(WidgetMessage.Player)
            OnCloseButtonClicked(WidgetMessage)
            
    OnPrestigeClicked(WidgetMessage : widget_message):void=
        if(Agent : agent = agent[WidgetMessage.Player]) :
            SFX_Warp.Play(Agent)
        if (PlayerUI := GetPlayerUI[WidgetMessage.Player], TeleportCanvas := TeleportCanvasPerAgent[WidgetMessage.Player]):
            TP_Prestige.Teleport(WidgetMessage.Player)
            OnCloseButtonClicked(WidgetMessage)
            

    OnTravel1_1Clicked(WidgetMessage : widget_message):void=
        if(Agent : agent = agent[WidgetMessage.Player]) :
            SFX_Warp.Play(Agent)
        if (PlayerUI := GetPlayerUI[WidgetMessage.Player], TeleportCanvas := TeleportCanvasPerAgent[WidgetMessage.Player]):
            TP_1_1.Teleport(WidgetMessage.Player)
            OnCloseButtonClicked(WidgetMessage)

    OnTravel1_2Clicked(WidgetMessage : widget_message):void=
        if(Agent : agent = agent[WidgetMessage.Player]) :
            SFX_Warp.Play(Agent)
        if (PlayerUI := GetPlayerUI[WidgetMessage.Player], TeleportCanvas := TeleportCanvasPerAgent[WidgetMessage.Player]):
            TP_1_2.Teleport(WidgetMessage.Player)
            OnCloseButtonClicked(WidgetMessage)

    OnTravel1_3Clicked(WidgetMessage : widget_message):void=
        if(Agent : agent = agent[WidgetMessage.Player]) :
            SFX_Warp.Play(Agent)
        if (PlayerUI := GetPlayerUI[WidgetMessage.Player], TeleportCanvas := TeleportCanvasPerAgent[WidgetMessage.Player]):
            TP_1_3.Teleport(WidgetMessage.Player)
            OnCloseButtonClicked(WidgetMessage)

    OnTravel1_4Clicked(WidgetMessage : widget_message):void=
        if(Agent : agent = agent[WidgetMessage.Player]) :
            SFX_Warp.Play(Agent)
        if (PlayerUI := GetPlayerUI[WidgetMessage.Player], TeleportCanvas := TeleportCanvasPerAgent[WidgetMessage.Player]):
            TP_1_4.Teleport(WidgetMessage.Player)
            OnCloseButtonClicked(WidgetMessage)
            
    OnTravel2_1Clicked(WidgetMessage : widget_message):void=
        if(Agent : agent = agent[WidgetMessage.Player]) :
            SFX_Warp.Play(Agent)
        if (PlayerUI := GetPlayerUI[WidgetMessage.Player], TeleportCanvas := TeleportCanvasPerAgent[WidgetMessage.Player]):
            TP_2_1.Teleport(WidgetMessage.Player)
            OnCloseButtonClicked(WidgetMessage)

    OnTravel2_2Clicked(WidgetMessage : widget_message):void=
        if(Agent : agent = agent[WidgetMessage.Player]) :
            SFX_Warp.Play(Agent)
        if (PlayerUI := GetPlayerUI[WidgetMessage.Player], TeleportCanvas := TeleportCanvasPerAgent[WidgetMessage.Player]):
            TP_2_2.Teleport(WidgetMessage.Player)
            OnCloseButtonClicked(WidgetMessage)

    OnTravel2_3Clicked(WidgetMessage : widget_message):void=
        if(Agent : agent = agent[WidgetMessage.Player]) :
            SFX_Warp.Play(Agent)
        if (PlayerUI := GetPlayerUI[WidgetMessage.Player], TeleportCanvas := TeleportCanvasPerAgent[WidgetMessage.Player]):
            TP_2_3.Teleport(WidgetMessage.Player)
            OnCloseButtonClicked(WidgetMessage)

    OnTravel2_4Clicked(WidgetMessage : widget_message):void=
        if(Agent : agent = agent[WidgetMessage.Player]) :
            SFX_Warp.Play(Agent)
        if (PlayerUI := GetPlayerUI[WidgetMessage.Player], TeleportCanvas := TeleportCanvasPerAgent[WidgetMessage.Player]):
            TP_2_4.Teleport(WidgetMessage.Player)
            OnCloseButtonClicked(WidgetMessage)
            
    OnTravel3_1Clicked(WidgetMessage : widget_message):void=
        if(Agent : agent = agent[WidgetMessage.Player]) :
            SFX_Warp.Play(Agent)
        if (PlayerUI := GetPlayerUI[WidgetMessage.Player], TeleportCanvas := TeleportCanvasPerAgent[WidgetMessage.Player]):
            TP_3_1.Teleport(WidgetMessage.Player)
            OnCloseButtonClicked(WidgetMessage)

    OnTravel3_2Clicked(WidgetMessage : widget_message):void=
        if(Agent : agent = agent[WidgetMessage.Player]) :
            SFX_Warp.Play(Agent)
        if (PlayerUI := GetPlayerUI[WidgetMessage.Player], TeleportCanvas := TeleportCanvasPerAgent[WidgetMessage.Player]):
            TP_3_2.Teleport(WidgetMessage.Player)
            OnCloseButtonClicked(WidgetMessage)

    OnTravel3_3Clicked(WidgetMessage : widget_message):void=
        if(Agent : agent = agent[WidgetMessage.Player]) :
            SFX_Warp.Play(Agent)
        if (PlayerUI := GetPlayerUI[WidgetMessage.Player], TeleportCanvas := TeleportCanvasPerAgent[WidgetMessage.Player]):
            TP_3_3.Teleport(WidgetMessage.Player)
            OnCloseButtonClicked(WidgetMessage)

    OnTravel3_4Clicked(WidgetMessage : widget_message):void=
        if(Agent : agent = agent[WidgetMessage.Player]) :
            SFX_Warp.Play(Agent)
        if (PlayerUI := GetPlayerUI[WidgetMessage.Player], TeleportCanvas := TeleportCanvasPerAgent[WidgetMessage.Player]):
            TP_3_4.Teleport(WidgetMessage.Player)
            OnCloseButtonClicked(WidgetMessage)


    OnCloseButtonClicked(WidgetMessage : widget_message):void=
        if(Agent : agent = agent[WidgetMessage.Player]) :
            SFX_Close.Play(Agent)
        if (PlayerUI := GetPlayerUI[WidgetMessage.Player], TeleportCanvas := TeleportCanvasPerAgent[WidgetMessage.Player]):
            PlayerUI.RemoveWidget(TeleportCanvas)
            TeleportCanvas.SetVisibility(widget_visibility.Collapsed)


    # キャンバス
    CreateTeleportCanvas(Agent : agent, Player : player):canvas=
        var CanvasSlots : []canvas_slot = array{}
        
        CenterVector : vector2 = vector2{X:=0.5, Y:=0.5}
        CenterAnchors : anchors = anchors{Minimum := CenterVector, Maximum := CenterVector}

        
        FastTravelTextBlock := text_block{
            DefaultText := StringToMessage("Fast Travel"),
            DefaultTextColor := NamedColors.White,
            DefaultShadowOffset := option{vector2{X:=2.5, Y:=2.5}},
            DefaultShadowColor := NamedColors.Black
        }
        FastTravelTextBlock.SetShadowOpacity(1.0)
        set CanvasSlots += array
        {
            canvas_slot:
                Anchors := CenterAnchors
                Offsets := margin{Top:= -472.0, Left:= 0.0, Right := 257.0, Bottom := 41.0}
                Alignment := CenterVector
                SizeToContent := false
                Widget := FastTravelTextBlock
        }

        # Image Stock
        BGImage : texture_block = texture_block{DefaultImage := Textures.UI_BG, DefaultDesiredSize := vector2{X := 1920.0, Y := 1080.0}}
        IconImage_Skill : texture_block = texture_block{DefaultImage := Textures.Flat_Skill, DefaultDesiredSize := vector2{X := 100.0, Y := 100.0}}
        IconImage_Boss : texture_block = texture_block{DefaultImage := Textures.Flat_Boss, DefaultDesiredSize := vector2{X := 100.0, Y := 100.0}}
        IconImage_Boss2 : texture_block = texture_block{DefaultImage := Textures.Flat_Boss, DefaultDesiredSize := vector2{X := 100.0, Y := 100.0}}
        IconImage_Pet : texture_block = texture_block{DefaultImage := Textures.Flat_Yokai, DefaultDesiredSize := vector2{X := 100.0, Y := 100.0}}
        MapImage_Stage1_1 : texture_block = texture_block{DefaultImage := Textures.Stage1_1, DefaultDesiredSize := vector2{X := 240.0, Y := 135.0}}
        MapImage_Stage1_2 : texture_block = texture_block{DefaultImage := Textures.Stage1_2, DefaultDesiredSize := vector2{X := 240.0, Y := 135.0}}
        MapImage_Stage1_3 : texture_block = texture_block{DefaultImage := Textures.Stage1_3, DefaultDesiredSize := vector2{X := 240.0, Y := 135.0}}
        MapImage_Stage1_4 : texture_block = texture_block{DefaultImage := Textures.Stage1_4, DefaultDesiredSize := vector2{X := 240.0, Y := 135.0}}
        MapImage_Stage2_1 : texture_block = texture_block{DefaultImage := Textures.Stage2_1, DefaultDesiredSize := vector2{X := 240.0, Y := 135.0}}
        MapImage_Stage2_2 : texture_block = texture_block{DefaultImage := Textures.Stage2_2, DefaultDesiredSize := vector2{X := 240.0, Y := 135.0}}
        MapImage_Stage2_3 : texture_block = texture_block{DefaultImage := Textures.Stage2_3, DefaultDesiredSize := vector2{X := 240.0, Y := 135.0}}
        MapImage_Stage2_4 : texture_block = texture_block{DefaultImage := Textures.Stage2_4, DefaultDesiredSize := vector2{X := 240.0, Y := 135.0}}
        MapImage_Stage3_1 : texture_block = texture_block{DefaultImage := Textures.Stage3_1, DefaultDesiredSize := vector2{X := 240.0, Y := 135.0}}
        MapImage_Stage3_2 : texture_block = texture_block{DefaultImage := Textures.Stage3_2, DefaultDesiredSize := vector2{X := 240.0, Y := 135.0}}
        MapImage_Stage3_3 : texture_block = texture_block{DefaultImage := Textures.Stage3_3, DefaultDesiredSize := vector2{X := 240.0, Y := 135.0}}
        MapImage_Stage3_4 : texture_block = texture_block{DefaultImage := Textures.Stage3_4, DefaultDesiredSize := vector2{X := 240.0, Y := 135.0}}
        # Text Stock
        var TownTextBlock : text_block = text_block :
            DefaultTextColor := NamedColors.White
            DefaultShadowOffset := option{vector2{X:=2.5, Y:=2.5}}
            DefaultShadowColor := NamedColors.Black
            DefaultJustification := text_justification.Center
            
        var QuietButtonTheater : button_quiet = button_quiet{}
        var QuietButtonSkillUpgrade : button_quiet = button_quiet{}
        var QuietButtonYokai : button_quiet = button_quiet{}
        var QuietButtonBlackSmith : button_quiet = button_quiet{}
        var QuietButtonPortTown : button_quiet = button_quiet{}
        var QuietButtonRestSpot : button_quiet = button_quiet{}
        
        
        #========================================== BG ==========================================
        set CanvasSlots += array
        {
            canvas_slot:
                Anchors := CenterAnchors
                Offsets := margin{Top := 0.0,Left := 0.0, Right := 1920.0, Bottom := 1080.0}
                Alignment := CenterVector
                SizeToContent := false
                Widget := BGImage
        }
        #========================================== Teleport Spot [ Town ] ==========================================
        # Button Stock
        QuietButtonTheater.OnClick().Subscribe(OnTheaterClicked)
        QuietButtonSkillUpgrade.OnClick().Subscribe(OnSkillUpgradeClicked)
        QuietButtonYokai.OnClick().Subscribe(OnYokaiClicked)
        QuietButtonBlackSmith.OnClick().Subscribe(OnBlackSmithClicked)
        QuietButtonPortTown.OnClick().Subscribe(OnPortTownClicked)
        QuietButtonRestSpot.OnClick().Subscribe(OnRestSpotClicked)
        # Stock END
        set CanvasSlots += array
        {
            canvas_slot:
                Anchors := CenterAnchors
                Offsets := margin{Top:= -15.0, Left:= -765.0, Right := 305.0, Bottom := 800.0}
                Alignment := CenterVector
                SizeToContent := false
                Widget := stack_box:
                    Orientation := orientation.Vertical
                    Slots := array:
                        stack_box_slot:
                            Padding:=margin{Top:= 15.0, Bottom:= 15.0}
                            HorizontalAlignment := horizontal_alignment.Fill
                            VerticalAlignment := vertical_alignment.Fill
                            Widget := TownTextBlock
                        stack_box_slot:
                            Padding:=margin{Bottom:=10.0}
                            HorizontalAlignment := horizontal_alignment.Fill
                            VerticalAlignment := vertical_alignment.Top
                            Widget := QuietButtonTheater
                        stack_box_slot:
                            Padding:=margin{Bottom:=10.0}
                            HorizontalAlignment := horizontal_alignment.Fill
                            VerticalAlignment := vertical_alignment.Top
                            Widget := QuietButtonSkillUpgrade
                        stack_box_slot:
                            Padding:=margin{Bottom:=10.0}
                            HorizontalAlignment := horizontal_alignment.Fill
                            VerticalAlignment := vertical_alignment.Top
                            Widget := QuietButtonYokai
                        stack_box_slot:
                            Padding:=margin{Bottom:=10.0}
                            HorizontalAlignment := horizontal_alignment.Fill
                            VerticalAlignment := vertical_alignment.Top
                            Widget := QuietButtonBlackSmith
                        stack_box_slot:
                            Padding:=margin{Bottom:=10.0}
                            HorizontalAlignment := horizontal_alignment.Fill
                            VerticalAlignment := vertical_alignment.Top
                            Widget := QuietButtonPortTown
                        stack_box_slot:
                            Padding:=margin{Bottom:=10.0}
                            HorizontalAlignment := horizontal_alignment.Fill
                            VerticalAlignment := vertical_alignment.Top
                            Widget := QuietButtonRestSpot
                        # stack_box_slot:
                        #     Padding:=margin{Bottom:=10.0}
                        #     HorizontalAlignment := horizontal_alignment.Fill
                        #     VerticalAlignment := vertical_alignment.Top
                        #     Widget := QuietButtonPrestige
        }

        #========================================== Stage ==========================================

        # Button Stock
        QuietButtonTravel1_1 : button_quiet = button_quiet{DefaultText := StringToMessage("TRAVEL")}
        QuietButtonTravel1_1.OnClick().Subscribe(OnTravel1_1Clicked)
        QuietButtonTravel1_2 : button_quiet = button_quiet{DefaultText := StringToMessage("TRAVEL")}
        QuietButtonTravel1_2.OnClick().Subscribe(OnTravel1_2Clicked)
        QuietButtonTravel1_3 : button_quiet = button_quiet{DefaultText := StringToMessage("TRAVEL")}
        QuietButtonTravel1_3.OnClick().Subscribe(OnTravel1_3Clicked)
        QuietButtonTravel1_4 : button_quiet = button_quiet{DefaultText := StringToMessage("TRAVEL")}
        QuietButtonTravel1_4.OnClick().Subscribe(OnTravel1_4Clicked)

        QuietButtonTravel2_1 : button_quiet = button_quiet{DefaultText := StringToMessage("TRAVEL")}
        QuietButtonTravel2_1.OnClick().Subscribe(OnTravel2_1Clicked)
        QuietButtonTravel2_2 : button_quiet = button_quiet{DefaultText := StringToMessage("TRAVEL")}
        QuietButtonTravel2_2.OnClick().Subscribe(OnTravel2_2Clicked)
        QuietButtonTravel2_3 : button_quiet = button_quiet{DefaultText := StringToMessage("TRAVEL")}
        QuietButtonTravel2_3.OnClick().Subscribe(OnTravel2_3Clicked)
        QuietButtonTravel2_4 : button_quiet = button_quiet{DefaultText := StringToMessage("TRAVEL")}
        QuietButtonTravel2_4.OnClick().Subscribe(OnTravel2_4Clicked)

        QuietButtonTravel3_1 : button_quiet = button_quiet{DefaultText := StringToMessage("TRAVEL")}
        QuietButtonTravel3_1.OnClick().Subscribe(OnTravel3_1Clicked)
        QuietButtonTravel3_2 : button_quiet = button_quiet{DefaultText := StringToMessage("TRAVEL")}
        QuietButtonTravel3_2.OnClick().Subscribe(OnTravel3_2Clicked)
        QuietButtonTravel3_3 : button_quiet = button_quiet{DefaultText := StringToMessage("TRAVEL")}
        QuietButtonTravel3_3.OnClick().Subscribe(OnTravel3_3Clicked)
        QuietButtonTravel3_4 : button_quiet = button_quiet{DefaultText := StringToMessage("TRAVEL")}
        QuietButtonTravel3_4.OnClick().Subscribe(OnTravel3_4Clicked)
        # Stock END

        set CanvasSlots += array
        {
            canvas_slot:
                Anchors := CenterAnchors
                Offsets := margin{Top:= -15.0, Left:= 168.0, Right := 1490.0, Bottom := 800.0}
                Alignment := CenterVector
                SizeToContent := false
                Widget := stack_box:
                    Orientation := orientation.Vertical
                    Slots := array:
        #================================ UniqueIcon ================================
                        stack_box_slot:
                            Padding:=margin{Bottom:= 20.0}
                            HorizontalAlignment := horizontal_alignment.Fill
                            VerticalAlignment := vertical_alignment.Fill
                            Widget := stack_box:
                                Orientation := orientation.Horizontal
                                Slots := array:
                                    stack_box_slot:
                                        Padding:=margin{Top:= 10.0, Bottom:= 10.0}
                                        Distribution:= option{1.0}
                                        HorizontalAlignment := horizontal_alignment.Center
                                        VerticalAlignment := vertical_alignment.Center
                                        Widget := stack_box:
                                            Orientation := orientation.Horizontal
                                            Slots := array:
                                                stack_box_slot:
                                                    HorizontalAlignment := horizontal_alignment.Fill
                                                    VerticalAlignment := vertical_alignment.Fill
                                                    Widget := IconImage_Skill

                                    stack_box_slot:
                                        Padding:=margin{Top:= 10.0, Bottom:= 10.0}
                                        Distribution:= option{1.0}
                                        HorizontalAlignment := horizontal_alignment.Center
                                        VerticalAlignment := vertical_alignment.Center
                                        Widget := stack_box:
                                            Orientation := orientation.Horizontal
                                            Slots := array:
                                                stack_box_slot:
                                                    HorizontalAlignment := horizontal_alignment.Fill
                                                    VerticalAlignment := vertical_alignment.Fill
                                                    Widget := IconImage_Boss

                                    stack_box_slot:
                                        Padding:=margin{Top:= 10.0, Bottom:= 10.0}
                                        Distribution:= option{1.0}
                                        HorizontalAlignment := horizontal_alignment.Center
                                        VerticalAlignment := vertical_alignment.Center
                                        Widget := stack_box:
                                            Orientation := orientation.Horizontal
                                            Slots := array:
                                                stack_box_slot:
                                                    HorizontalAlignment := horizontal_alignment.Fill
                                                    VerticalAlignment := vertical_alignment.Fill
                                                    Widget := IconImage_Pet

                                    stack_box_slot:
                                        Padding:=margin{Top:= 10.0, Bottom:= 10.0}
                                        Distribution:= option{1.0}
                                        HorizontalAlignment := horizontal_alignment.Center
                                        VerticalAlignment := vertical_alignment.Center
                                        Widget := stack_box:
                                            Orientation := orientation.Horizontal
                                            Slots := array:
                                                stack_box_slot:
                                                    HorizontalAlignment := horizontal_alignment.Fill
                                                    VerticalAlignment := vertical_alignment.Fill
                                                    Widget := IconImage_Boss2

        #=============================== Stage 1 ===============================

                        stack_box_slot:
                            Padding:=margin{Top:= 20.0, Bottom:= 20.0}
                            HorizontalAlignment := horizontal_alignment.Fill
                            VerticalAlignment := vertical_alignment.Fill
                            Widget := stack_box:
                                Orientation := orientation.Horizontal
                                Slots := array:
                                    stack_box_slot:
                                        Padding:=margin{Top:= 10.0, Bottom:= 10.0}
                                        Distribution:= option{1.0}
                                        HorizontalAlignment := horizontal_alignment.Center
                                        VerticalAlignment := vertical_alignment.Center
                                        Widget := stack_box:
                                            Orientation := orientation.Vertical
                                            Slots := array:
                                                stack_box_slot:
                                                    HorizontalAlignment := horizontal_alignment.Fill
                                                    VerticalAlignment := vertical_alignment.Fill
                                                    Widget := MapImage_Stage1_1
                                                stack_box_slot:
                                                    HorizontalAlignment := horizontal_alignment.Fill
                                                    VerticalAlignment := vertical_alignment.Fill
                                                    Widget := QuietButtonTravel1_1

                                    stack_box_slot:
                                        Padding:=margin{Top:= 10.0, Bottom:= 10.0}
                                        Distribution:= option{1.0}
                                        HorizontalAlignment := horizontal_alignment.Center
                                        VerticalAlignment := vertical_alignment.Center
                                        Widget := stack_box:
                                            Orientation := orientation.Vertical
                                            Slots := array:
                                                stack_box_slot:
                                                    HorizontalAlignment := horizontal_alignment.Fill
                                                    VerticalAlignment := vertical_alignment.Fill
                                                    Widget := MapImage_Stage1_2
                                                stack_box_slot:
                                                    HorizontalAlignment := horizontal_alignment.Fill
                                                    VerticalAlignment := vertical_alignment.Fill
                                                    Widget := QuietButtonTravel1_2

                                    stack_box_slot:
                                        Padding:=margin{Top:= 10.0, Bottom:= 10.0}
                                        Distribution:= option{1.0}
                                        HorizontalAlignment := horizontal_alignment.Center
                                        VerticalAlignment := vertical_alignment.Center
                                        Widget := stack_box:
                                            Orientation := orientation.Vertical
                                            Slots := array:
                                                stack_box_slot:
                                                    HorizontalAlignment := horizontal_alignment.Fill
                                                    VerticalAlignment := vertical_alignment.Fill
                                                    Widget := MapImage_Stage1_3
                                                stack_box_slot:
                                                    HorizontalAlignment := horizontal_alignment.Fill
                                                    VerticalAlignment := vertical_alignment.Fill
                                                    Widget := QuietButtonTravel1_3

                                    stack_box_slot:
                                        Padding:=margin{Top:= 10.0, Bottom:= 10.0}
                                        Distribution:= option{1.0}
                                        HorizontalAlignment := horizontal_alignment.Center
                                        VerticalAlignment := vertical_alignment.Center
                                        Widget := stack_box:
                                            Orientation := orientation.Vertical
                                            Slots := array:
                                                stack_box_slot:
                                                    HorizontalAlignment := horizontal_alignment.Fill
                                                    VerticalAlignment := vertical_alignment.Fill
                                                    Widget := MapImage_Stage1_4
                                                stack_box_slot:
                                                    HorizontalAlignment := horizontal_alignment.Fill
                                                    VerticalAlignment := vertical_alignment.Fill
                                                    Widget := QuietButtonTravel1_4
                        
        #=============================== Stage 2 ===============================

                        stack_box_slot:
                            Padding:=margin{Bottom:= 20.0}
                            HorizontalAlignment := horizontal_alignment.Fill
                            VerticalAlignment := vertical_alignment.Fill
                            Widget := stack_box:
                                Orientation := orientation.Horizontal
                                Slots := array:
                                    stack_box_slot:
                                        Padding:=margin{Top:= 10.0, Bottom:= 10.0}
                                        Distribution:= option{1.0}
                                        HorizontalAlignment := horizontal_alignment.Center
                                        VerticalAlignment := vertical_alignment.Center
                                        Widget := stack_box:
                                            Orientation := orientation.Vertical
                                            Slots := array:
                                                stack_box_slot:
                                                    HorizontalAlignment := horizontal_alignment.Fill
                                                    VerticalAlignment := vertical_alignment.Fill
                                                    Widget := MapImage_Stage2_1
                                                stack_box_slot:
                                                    HorizontalAlignment := horizontal_alignment.Fill
                                                    VerticalAlignment := vertical_alignment.Fill
                                                    Widget := QuietButtonTravel2_1

                                    stack_box_slot:
                                        Padding:=margin{Top:= 10.0, Bottom:= 10.0}
                                        Distribution:= option{1.0}
                                        HorizontalAlignment := horizontal_alignment.Center
                                        VerticalAlignment := vertical_alignment.Center
                                        Widget := stack_box:
                                            Orientation := orientation.Vertical
                                            Slots := array:
                                                stack_box_slot:
                                                    HorizontalAlignment := horizontal_alignment.Fill
                                                    VerticalAlignment := vertical_alignment.Fill
                                                    Widget := MapImage_Stage2_2
                                                stack_box_slot:
                                                    HorizontalAlignment := horizontal_alignment.Fill
                                                    VerticalAlignment := vertical_alignment.Fill
                                                    Widget := QuietButtonTravel2_2

                                    stack_box_slot:
                                        Padding:=margin{Top:= 10.0, Bottom:= 10.0}
                                        Distribution:= option{1.0}
                                        HorizontalAlignment := horizontal_alignment.Center
                                        VerticalAlignment := vertical_alignment.Center
                                        Widget := stack_box:
                                            Orientation := orientation.Vertical
                                            Slots := array:
                                                stack_box_slot:
                                                    HorizontalAlignment := horizontal_alignment.Fill
                                                    VerticalAlignment := vertical_alignment.Fill
                                                    Widget := MapImage_Stage2_3
                                                stack_box_slot:
                                                    HorizontalAlignment := horizontal_alignment.Fill
                                                    VerticalAlignment := vertical_alignment.Fill
                                                    Widget := QuietButtonTravel2_3

                                    stack_box_slot:
                                        Padding:=margin{Top:= 10.0, Bottom:= 10.0}
                                        Distribution:= option{1.0}
                                        HorizontalAlignment := horizontal_alignment.Center
                                        VerticalAlignment := vertical_alignment.Center
                                        Widget := stack_box:
                                            Orientation := orientation.Vertical
                                            Slots := array:
                                                stack_box_slot:
                                                    HorizontalAlignment := horizontal_alignment.Fill
                                                    VerticalAlignment := vertical_alignment.Fill
                                                    Widget := MapImage_Stage2_4
                                                stack_box_slot:
                                                    HorizontalAlignment := horizontal_alignment.Fill
                                                    VerticalAlignment := vertical_alignment.Fill
                                                    Widget := QuietButtonTravel2_4
                                        
        #=============================== Stage 3 ===============================

                        stack_box_slot:
                            Padding:=margin{Bottom:= 20.0}
                            HorizontalAlignment := horizontal_alignment.Fill
                            VerticalAlignment := vertical_alignment.Fill
                            Widget := stack_box:
                                Orientation := orientation.Horizontal
                                Slots := array:
                                    stack_box_slot:
                                        Padding:=margin{Top:= 10.0, Bottom:= 10.0}
                                        Distribution:= option{1.0}
                                        HorizontalAlignment := horizontal_alignment.Center
                                        VerticalAlignment := vertical_alignment.Center
                                        Widget := stack_box:
                                            Orientation := orientation.Vertical
                                            Slots := array:
                                                stack_box_slot:
                                                    HorizontalAlignment := horizontal_alignment.Fill
                                                    VerticalAlignment := vertical_alignment.Fill
                                                    Widget := MapImage_Stage3_1
                                                stack_box_slot:
                                                    HorizontalAlignment := horizontal_alignment.Fill
                                                    VerticalAlignment := vertical_alignment.Fill
                                                    Widget := QuietButtonTravel3_1

                                    stack_box_slot:
                                        Padding:=margin{Top:= 10.0, Bottom:= 10.0}
                                        Distribution:= option{1.0}
                                        HorizontalAlignment := horizontal_alignment.Center
                                        VerticalAlignment := vertical_alignment.Center
                                        Widget := stack_box:
                                            Orientation := orientation.Vertical
                                            Slots := array:
                                                stack_box_slot:
                                                    HorizontalAlignment := horizontal_alignment.Fill
                                                    VerticalAlignment := vertical_alignment.Fill
                                                    Widget := MapImage_Stage3_2
                                                stack_box_slot:
                                                    HorizontalAlignment := horizontal_alignment.Fill
                                                    VerticalAlignment := vertical_alignment.Fill
                                                    Widget := QuietButtonTravel3_2

                                    stack_box_slot:
                                        Padding:=margin{Top:= 10.0, Bottom:= 10.0}
                                        Distribution:= option{1.0}
                                        HorizontalAlignment := horizontal_alignment.Center
                                        VerticalAlignment := vertical_alignment.Center
                                        Widget := stack_box:
                                            Orientation := orientation.Vertical
                                            Slots := array:
                                                stack_box_slot:
                                                    HorizontalAlignment := horizontal_alignment.Fill
                                                    VerticalAlignment := vertical_alignment.Fill
                                                    Widget := MapImage_Stage3_3
                                                stack_box_slot:
                                                    HorizontalAlignment := horizontal_alignment.Fill
                                                    VerticalAlignment := vertical_alignment.Fill
                                                    Widget := QuietButtonTravel3_3

                                    stack_box_slot:
                                        Padding:=margin{Top:= 10.0, Bottom:= 10.0}
                                        Distribution:= option{1.0}
                                        HorizontalAlignment := horizontal_alignment.Center
                                        VerticalAlignment := vertical_alignment.Center
                                        Widget := stack_box:
                                            Orientation := orientation.Vertical
                                            Slots := array:
                                                stack_box_slot:
                                                    HorizontalAlignment := horizontal_alignment.Fill
                                                    VerticalAlignment := vertical_alignment.Fill
                                                    Widget := MapImage_Stage3_4
                                                stack_box_slot:
                                                    HorizontalAlignment := horizontal_alignment.Fill
                                                    VerticalAlignment := vertical_alignment.Fill
                                                    Widget := QuietButtonTravel3_4
        }

        #========================================== Close ==========================================

        CloseButton : button_loud = button_loud{DefaultText := StringToMessage("CLOSE")}
        CloseButton.OnClick().Subscribe(OnCloseButtonClicked)
        set CanvasSlots += array
        {
            canvas_slot:
                Anchors := CenterAnchors
                Offsets := margin{Top:= 459.5, Left:= -770.0, Right := 353.0, Bottom := 80.0}
                Alignment := CenterVector
                SizeToContent := false
                Widget := CloseButton
        }
        if(set LocalizeUIElements[Agent] = (TownTextBlock, QuietButtonTheater, QuietButtonSkillUpgrade, QuietButtonYokai, QuietButtonBlackSmith, QuietButtonPortTown, QuietButtonRestSpot)){}

        if(set StageUIElements[Agent] = map{}){}
        if:
            set StageUIElements[Agent] = map:
                1 => (QuietButtonTravel1_1,MapImage_Stage1_1)
                2 => (QuietButtonTravel1_2,MapImage_Stage1_2)
                3 => (QuietButtonTravel1_3,MapImage_Stage1_3)
                4 => (QuietButtonTravel1_4,MapImage_Stage1_4)
                5 => (QuietButtonTravel2_1,MapImage_Stage2_1)
                6 => (QuietButtonTravel2_2,MapImage_Stage2_2)
                7 => (QuietButtonTravel2_3,MapImage_Stage2_3)
                8 => (QuietButtonTravel2_4,MapImage_Stage2_4)
                9 => (QuietButtonTravel3_1,MapImage_Stage3_1)
                10 => (QuietButtonTravel3_2,MapImage_Stage3_2)
                11 => (QuietButtonTravel3_3,MapImage_Stage3_3)
                12 => (QuietButtonTravel3_4,MapImage_Stage3_4)

        for(StageElement : StageUIElements[Agent]):
            StageElement(0).SetVisibility(widget_visibility.Collapsed)
            StageElement(1).SetVisibility(widget_visibility.Collapsed)

        Canvas : canvas = canvas:
            Slots := CanvasSlots

        return Canvas
