
Objects<public> := module:
    MiningObjectHealthBar<public> := module:
        Materials<public> := module:

UI<public> := module:
    MiningObject<public> := module:
        Textures<public> := module: