
using { /Fortnite.com/Devices }
using { /Verse.org/Simulation }
using { /UnrealEngine.com/Temporary/Diagnostics }
using { PlayerData }
using { PlayerData }
using { Utility }

# See https://dev.epicgames.com/documentation/en-us/uefn/create-your-own-device-in-verse for how to create a verse device.

# A Verse-authored creative device that can be placed in a level
test_weapon_device := class(creative_device):

    var WeaponManager : weapon_manager = weapon_manager{}
    var SaveDataManager : save_data_manager = save_data_manager{}
    @editable
    UpgradeButton : button_device = button_device{}
    @editable
    MoneyButton : button_device = button_device{}
    @editable
    Weapon100Button : button_device = button_device{}

    OnButtonPressed<public>(Agent:agent):void=
        if(Player := player[Agent]):
            SaveDataManager.AddWeaponLevel(Player, 1)
            
            # 武器レベルを上げた後の新しいレベルを取得して表示
            var NewLevel : int = SaveDataManager.GetWeaponLevel(Player)
            Print("New Weapon Level: {NewLevel}")

    OnMoreMoneyPressed<public>(Agent:agent):void=
        if(Player := player[Agent]):
            Resource := big_number{Value := 900.0, Exponent := 50}
            DebugResource := big_number{Value := 999.0, Exponent := 50}
            SaveDataManager.SetResourceStone(Player,Resource)
            SaveDataManager.SetResourceSummonStoneCommon(Player,DebugResource)
            SaveDataManager.SetResourceSummonStoneUncommon(Player,DebugResource)
            SaveDataManager.SetResourceSummonStoneRare(Player,DebugResource)
            SaveDataManager.SetResourceSummonStoneEpic(Player,DebugResource)
            SaveDataManager.SetResourceSummonStoneLegendary(Player,DebugResource)

    OnBossCountPressed<public>(Agent:agent):void=
        if(Player := player[Agent]):
            SaveDataManager.AddWeaponLevel(Player, 100)
            
            # 武器レベルを上げた後の新しいレベルを取得して表示
            var NewLevel : int = SaveDataManager.GetWeaponLevel(Player)
            Print("New Weapon Level: {NewLevel}")
        
    OnBegin<override>()<suspends>:void=
        UpgradeButton.InteractedWithEvent.Subscribe(OnButtonPressed)
        MoneyButton.InteractedWithEvent.Subscribe(OnMoreMoneyPressed)
        Weapon100Button.InteractedWithEvent.Subscribe(OnBossCountPressed)