using { /Fortnite.com/Devices }
using { /Verse.org/Simulation }
using { /Verse.org}
using { /Verse.org/Random }
using { /Verse.org/Colors }
using { Weapon }
using { Skill }
using { PlayerData }
using{ VerseDevice }

weapon_damage<public> := class(creative_device):

    var SaveDataManager : save_data_manager = save_data_manager{}
    var WeaponManager : weapon_manager = weapon_manager{}
    var SkillManager : skill_manager = skill_manager{}
    #恐竜バフ
    var UIHud : ui_hud_count = ui_hud_count{}

    @editable
    PlayerDamageTest : button_device = button_device{}

    
    OnBegin<override>()<suspends>:void=
        for(Player : GetPlayspace().GetPlayers()):
        #PlayerDamageTest.InteractedWithEvent.Subscribe(debugbutton)
        PlayerDamageTest.InteractedWithEvent.Subscribe(OnPlayerAttack)

    <#debugbutton(Agent : agent):void=
        if(Player := player[Agent]) :
            if(SaveDataManager.IsStrBuff[Player]?) :
                Print("IsStrBuff[Player]: true", ?Color := NamedColors.Red, ?Duration := 3.0)
            else :
                Print("IsStrBuff[Player]: false", ?Color := NamedColors.Red, ?Duration := 3.0)#>
    # プレイヤーが攻撃した際に与えるダメージを計算する関数
    CalculateDamage<public>(Agent:agent)<decides><transacts>:float=
        var Damage : float = 0.0
        if(Player := player[Agent]) :
            WeaponLevel : int = SaveDataManager.GetWeaponLevel(Player)
            Print("Weapon Level: {WeaponLevel}")
        
            BaseAttackPower : float = WeaponManager.GetWeaponAttackPower(Player)
            Print("Base Attack Power: {BaseAttackPower}")
        
            AttackPowerPercent : float = WeaponManager.GetWeaponAttackPowerPercent(Player)
        
            # 武器のクリティカル率とクリティカル倍率を取得
            WeaponCriticalRate : float = WeaponManager.GetWeaponCriticalRate(Player)
            Print("Weapon Critical Rate: {WeaponCriticalRate}")
            
            WeaponCriticalMultiplier : float = WeaponManager.GetWeaponCriticalMultiplier(Player)
        
            # スキルのクリティカル率とクリティカル倍率を取得 (Index 15, 16)
            CriticalRateSkillInfo := SaveDataManager.GetSkill[Player, 15]
            CriticalMultiplierSkillInfo := SaveDataManager.GetSkill[Player, 16]
            
            # スキルのクリティカル発生時追加倍率 (Index 18)
            CriticalPowerSurgeSkillInfo := SaveDataManager.GetSkill[Player, 18]
        
            # スキル効果を上乗せ (int を float にキャスト)
            TotalCriticalRate : float = WeaponCriticalRate + (CriticalRateSkillInfo.EffectPower * 1.0)
            TotalCriticalMultiplier : float = WeaponCriticalMultiplier + (CriticalMultiplierSkillInfo.EffectPower * 1.0)
            CriticalPowerSurgeMultiplier : float = 1.0 + (CriticalPowerSurgeSkillInfo.EffectPower * 1.0 / 100.0)  # クリティカル時の追加倍率
        
            var IsCriticalHit : int = 0
            RandomValue : float = Random.GetRandomFloat(0.0, 100.0)
            #恐竜の水浴びてたら1.2倍
            # var Multiplier : float = 1.0
            # if(SaveDataManager.GetStrBuff(Player)?) :
            #     Print("恐竜Buff", ?Color := NamedColors.Red, ?Duration := 3.0)
            #     set Multiplier = 1.2
            # else :
            #     Print("恐竜Buffなし", ?Color := NamedColors.Red, ?Duration := 3.0)
            #     set Multiplier = 1.0
        
            # クリティカルヒットの判定
            if(RandomValue <= TotalCriticalRate):
                set IsCriticalHit = 1
        
            # ダメージの基礎計算
            set Damage = BaseAttackPower * (1.0 + AttackPowerPercent / 100.0) # * Multiplier
        
            # クリティカルヒットの場合、クリティカル倍率とクリティカルパワーサージを適用
            if(IsCriticalHit = 1):
                set Damage *= (1.0 + TotalCriticalMultiplier / 100.0)  # クリティカル倍率
                set Damage *= CriticalPowerSurgeMultiplier  # クリティカルパワーサージの適用
                Print("Critical hit! Damage: {Damage}")
            else:
                Print("Normal hit. Damage: {Damage}")
        
        return Damage
    
    GetWeaponAttackPower<public>(Agent:agent)<decides><transacts>:big_number=
        var WeaponLevel : int = 0
        var BaseAttackPower : float = 0.0
        var AttackPowerPercent : float = 0.0
        if(Player := player[Agent]) :
            set WeaponLevel = SaveDataManager.GetWeaponLevel(Player)
            set BaseAttackPower = WeaponManager.GetWeaponAttackPower(Player)
            AttackPowerPercent = WeaponManager.GetWeaponAttackPowerPercent(Player)

        return big_number{Value := BaseAttackPower * (1.0 + AttackPowerPercent / 100.0), Exponent := 1}

    # プレイヤーが攻撃した時のデバッグ用イベントハンドラー
    OnPlayerAttack<public>(Agent:agent):void=
        if(Player := player[Agent]) :
            Print("Player found")
            logic:
                Damage : float = CalculateDamage[Player]
                Print("{Damage} damage.")
