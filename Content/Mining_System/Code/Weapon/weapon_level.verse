# using { /Fortnite.com/Devices }
# using { /Verse.org/Simulation }

# weapon_level_manager<public> := class(creative_device):

#     @editable
#     WeaponManager : weapon_manager = weapon_manager{}

#     @editable
#     LevelUpButton : button_device = button_device{}
#     @editable
#     SaveDataManager : save_data_manager = save_data_manager{}

#     # イベント初期化
#     OnBegin<override>()<suspends>:void=
#         Print("Weapon Level Up Button is ready.")

#     # ボタンが押された際の処理
#     OnLevelUpButtonPressed<public>(Player:player)<transacts>:void=
#         var CurrentLevel : int = WeaponManager.SaveDataManager.GetWeaponLevel(Player)  # 武器の現在のレベルを取得
#         Print("Leveling up weapon for player")

#         var Success : logic = WeaponManager.PurchaseWeaponNormal(Player, CurrentLevel)  # 武器レベルアップを実行

#         if(Success?):
#             Print("Weapon leveled up successfully!")
#         else:
#             Print("Weapon level up failed. Not enough resources.")
            

#     # リソースが十分にあるかを確認する処理（オプションで追加）
#     CheckResource<public>(Player:player)<transacts>:logic=
#         var CanPurchase : logic = WeaponManager.CanLevelUp(Player)
#         return CanPurchase