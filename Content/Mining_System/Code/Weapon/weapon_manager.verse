
using { /Fortnite.com/Devices }
using { /Verse.org/Simulation }
using { /Fortnite.com/Characters }
using { PlayerData }

weapon_manager<public> := class(creative_device):

    @editable
    SaveDataManager : save_data_manager = save_data_manager{}

    @editable
    QuestTrigger : trigger_device = trigger_device{}

    @editable
    SFX_LevelUp2 : audio_player_device = audio_player_device{}

    @editable
    ItemGranter : item_granter_device = item_granter_device{}
    @editable
    ItemGranter2 : item_granter_device = item_granter_device{}
    @editable
    ItemGranter3 : item_granter_device = item_granter_device{}
    @editable
    ItemRemover : item_remover_device = item_remover_device{}
    @editable
    TeamChanger_Blader : class_and_team_selector_device = class_and_team_selector_device{}

    # 武器の基本ステータス
    BaseAttackPower : float = 10.0
    GrowthRate : float = 10.0

    # 進化ボーナスのレベルしきい値
    LevelThreshold1 : int = 40
    LevelThreshold2 : int = 80
    Bonus1 : float = 100.0
    Bonus2 : float = 200.0

    OnBegin<override>()<suspends>:void=
        # プレイヤーリストを取得し、全員に処理を実行
        for(Player : GetPlayspace().GetPlayers()):
            if(FortniteCharacter : fort_character = Player.GetFortCharacter[]):
                GrantSwordToPlayer(Player)
        GetPlayspace().PlayerAddedEvent().Subscribe(GrantSwordToPlayer)

    GrantSwordToPlayer<public>(Player:player):void=
        WeaponLevel := SaveDataManager.GetWeaponLevel(Player)
        # 武器レベルが1以上か確認
        if(WeaponLevel >= 1 and WeaponLevel < 40):
            QuestTrigger.Trigger(Player)
            ItemGranter.GrantItem(Player)
            TeamChanger_Blader.ChangeTeamAndClass(Player)
            SaveDataManager.CheckRecoveryBossTicket(Player)
            Print("Sword granted to player with weapon level: {WeaponLevel}")
        else if(WeaponLevel >= LevelThreshold1 and WeaponLevel < LevelThreshold2):
            ItemRemover.Remove(Player)
            ItemGranter2.GrantItem(Player)
            TeamChanger_Blader.ChangeTeamAndClass(Player)
            SaveDataManager.CheckRecoveryBossTicket(Player)
            Print("Weapon level is less than 1. No sword granted.")
        else if(WeaponLevel >= LevelThreshold2):
            ItemRemover.Remove(Player)
            ItemGranter3.GrantItem(Player)
            TeamChanger_Blader.ChangeTeamAndClass(Player)
            SaveDataManager.CheckRecoveryBossTicket(Player)

    # 武器の攻撃力計算
    GetWeaponAttackPower<public>(Player:player)<transacts>:float=
        Level := SaveDataManager.GetWeaponLevel(Player)
        return BaseAttackPower * (GrowthRate * Level)

    GetNextWeaponAttackPower<public>(Player:player)<transacts>:float=
        Level := SaveDataManager.GetWeaponLevel(Player) + 1
        return BaseAttackPower * (GrowthRate * Level)
    # 攻撃力+％の計算
    GetWeaponAttackPowerPercent<public>(Player:player)<transacts>:float=
        Level := SaveDataManager.GetWeaponLevel(Player)
        var EvolutionBonus : float = 0.0

        # レベルに応じてボーナスを適用
        if(Level >= LevelThreshold2):
            set EvolutionBonus = Bonus2
        else if(Level >= LevelThreshold1):
            set EvolutionBonus = Bonus1

        PercentResult := Level * 1.0 + EvolutionBonus
        return PercentResult
    GetNextWeaponAttackPowerPercent<public>(Player:player)<transacts>:float=
        Level := SaveDataManager.GetWeaponLevel(Player) + 1
        var EvolutionBonus : float = 0.0

        # レベルに応じてボーナスを適用
        if(Level >= LevelThreshold2):
            set EvolutionBonus = Bonus2
        else if(Level >= LevelThreshold1):
            set EvolutionBonus = Bonus1

        PercentResult := Level * 1.0 + EvolutionBonus
        return PercentResult

    # クリティカル確率の計算
    GetWeaponCriticalRate<public>(Player:player)<transacts>:float=
        Level := SaveDataManager.GetWeaponLevel(Player)
        var BaseCriticalRate : float = 1.0
        var EvolutionBonus : float = 0.0

        if(Level >= LevelThreshold2):
            set EvolutionBonus = Bonus2
        else if(Level >= LevelThreshold1):
            set EvolutionBonus = Bonus1

        RateResult := BaseCriticalRate * Level + EvolutionBonus
        return RateResult
    GetNextWeaponCriticalRate<public>(Player:player)<transacts>:float=
        Level := SaveDataManager.GetWeaponLevel(Player) + 1
        var BaseCriticalRate : float = 1.0
        var EvolutionBonus : float = 0.0

        if(Level >= LevelThreshold2):
            set EvolutionBonus = Bonus2
        else if(Level >= LevelThreshold1):
            set EvolutionBonus = Bonus1

        RateResult := BaseCriticalRate * Level + EvolutionBonus
        return RateResult

    GetWeaponCriticalMultiplier<public>(Player:player)<transacts>:float=
        Level := SaveDataManager.GetWeaponLevel(Player)
        var BaseCriticalMultiplier : float = 1.5  # 基本のクリティカル倍率（デフォルトで1.5倍）
        var EvolutionBonus : float = 0.0
    
        # レベルに応じた進化ボーナスを適用
        if(Level >= LevelThreshold2):
            set EvolutionBonus = Bonus2
        else if(Level >= LevelThreshold1):
            set EvolutionBonus = Bonus1
    
        # クリティカル倍率の計算
        MultiplierResult := BaseCriticalMultiplier + (Level * 0.05) + EvolutionBonus
        return MultiplierResult

    GetNextWeaponCriticalMultiplier<public>(Player:player)<transacts>:float=
        Level := SaveDataManager.GetWeaponLevel(Player) + 1
        var BaseCriticalMultiplier : float = 1.5  # 基本のクリティカル倍率（デフォルトで1.5倍）
        var EvolutionBonus : float = 0.0
    
        # レベルに応じた進化ボーナスを適用
        if(Level >= LevelThreshold2):
            set EvolutionBonus = Bonus2
        else if(Level >= LevelThreshold1):
            set EvolutionBonus = Bonus1
    
        # クリティカル倍率の計算
        MultiplierResult := BaseCriticalMultiplier + (Level * 0.05) + EvolutionBonus
        return MultiplierResult
    # プレイヤーの武器レベルをアップグレード
    UpgradeWeaponLevel<public>(Player:player)<transacts>:void=
        if(SaveDataManager.CanLevelUp[Player]):  # レベルアップ条件をチェック
            SaveDataManager.AddWeaponLevel(Player, 1)
            Print("Weapon level upgraded to: {SaveDataManager.GetWeaponLevel(Player)}")
        else:
            Print("Not enough resources to upgrade weapon level.")

    GetWeaponPurchaseCost<public>(Player:player)<transacts>:big_number=
        Level := SaveDataManager.GetWeaponLevel(Player)
        return CalculatePurchaseCost(Level)

    # 武器のレベルをリセット
    ResetWeaponLevel<public>(Player:player)<transacts>:void=
        SaveDataManager.SetWeaponLevel(Player, 1)

    CalculatePurchaseCost<public>(Level:int)<transacts>:big_number=
        var BaseCost : big_number = big_number{Value := 2.0, Exponent := 12}  # 初期コストを設定
        var CostGrowthRate : float = 1.28  # コストの成長率
        return BigNumber_MultiplyByFloat(BaseCost, Pow(CostGrowthRate, Level * 1.0))

    # プレイヤーが購入可能かをチェックするメソッド~
    CanPurchase<public>(Player:player, Cost:big_number)<decides><transacts>:void=
        # 現在のリソースを取得
        ResourceCount := SaveDataManager.GetResourceStone(Player)
        
        # リソースが十分かを判定
        if(BigNumber_GreaterThanOrEquals(ResourceCount, Cost)?):
            Print("Player has enough resources for the purchase.")
        else:
            Print("Player does not have enough resources for the purchase.")


        # プレイヤーがアイテムを購入するメソッド
    PurchaseItem<public>(Player:player, Level:int)<transacts>:void=
        # 購入コストを計算
        Cost := CalculatePurchaseCost(Level)
        
        # 購入可能かを確認
        if(CanPurchase[Player, Cost]):  # 十分なリソースがあるかをチェック
            # リソースを減算
            SaveDataManager.SubtractResourceStone(Player, Cost)
            
            # 購入の処理（アイテムや武器のレベルを追加するなど）
            SaveDataManager.AddWeaponLevel(Player, 1)
            Print("Item purchased! Weapon level increased.")
        else:
            Print("Purchase failed due to insufficient resources.")
#-----------------------------------NomalCost---------------------------------------------

#----------------------------------- 武器レベル管理 ---------------------------------------------

    # isSpecialLevel(Level:int):logic =
    #     if(Level = 20 or Level = 40 or Level = 60 or Level = 80 or Level = 100):
    #         return true
    #     else:
    #         return false
#----------------------------------- 武器グレード管理 ---------------------------------------------

    # # プレイヤーの武器グレードを取得
    # GetWeaponGrade<public>(Player:player)<decides><transacts>:int=
    #     return SaveDataManager.GetWeaponGrade(Player)

    # # プレイヤーの武器グレードを設定
    # SetWeaponGrade<public>(Player:player, WeaponGrade:int)<decides><transacts>:void=
    #     SaveDataManager.SetWeaponGrade(Player, WeaponGrade)

    # # プレイヤーの武器グレードを加算する
    # AddWeaponGrade<public>(Player:player, WeaponGrade:int)<decides><transacts>:void=
    #     var CurrentGrade : int = GetWeaponGrade[Player]
    #     SetWeaponGrade[Player, CurrentGrade + WeaponGrade]


#-------------------------------------GradeUp--------------------------------------------------


    # 購入時のコスト計算
    # BigNumber_CalculateCostWeapon<public>(BaseCost : big_number, Start : int, Count : int, CommonRatio : float, ratio_factor: float): big_number=
    #     dynamic_ratio := CommonRatio / (1.0 + ratio_factor * Start)
        
    #     term1 := Pow(dynamic_ratio, (Start + Count)*1.0)
    #     term2 := Pow(dynamic_ratio, Start * 1.0)
        
    #     TotalCost_Value := BaseCost.Value * (term1 - term2) / (dynamic_ratio - 1.0)
    #     TotalCost_Exponent := BaseCost.Exponent
        
    #     return BigNumber_Normalize(big_number{Value:=TotalCost_Value, Exponent:=TotalCost_Exponent})

    # # 素材のコスト計算および購入可能かをチェック
    # CanPurchaseWeapon<public>(Player:player, WeaponLevel:int, RequiredMaterial:int):tuple(logic, big_number, big_number)=
    #     var CurrentStone : big_number = SaveDataManager.GetResourceStone(Player)
    #     var CurrentPowerupStone : big_number = SaveDataManager.GetResourcePowerupStone(Player)

    #     var NeededStone : big_number = CalculateMaterialCost(MaterialCost_Stone, RequiredMaterial)
    #     var NeededPowerupStone : big_number = CalculateMaterialCost(MaterialCost_PowerupStone, RequiredMaterial)

    #     # 石が足りているか確認
    #     if(BigNumber_GreaterThanOrEquals(CurrentStone, NeededStone)?):
    #         # 強化石が足りているか確認
    #         if(BigNumber_GreaterThanOrEquals(CurrentPowerupStone, NeededPowerupStone)?):
    #             return (true, NeededStone, NeededPowerupStone)
    #         else:
    #             return (false, NeededStone, NeededPowerupStone)
    #     else:
    #         return (false, NeededStone, NeededPowerupStone)

    # PurchaseWeapon<public>(Player:player, WeaponLevel:int, RequiredMaterial:int):logic=
    #     # CanPurchaseWeaponを使って購入可能か確認（トランザクション外）
    #     var PurchaseResult : tuple(logic, big_number, big_number) = CanPurchaseWeapon(Player, WeaponLevel, RequiredMaterial)
        
    #     var CanPurchase : logic = PurchaseResult(0)
    #     var NeededStone : big_number = PurchaseResult(1)
    #     var NeededPowerupStone : big_number = PurchaseResult(2)
    
    #     # ロジック型のチェックを直接returnで処理
    #     if(CanPurchase?):
    #         # トランザクションを使って素材を差し引き、武器をアップグレード
    #         PurchaseWeaponTransaction(Player, NeededStone, NeededPowerupStone, WeaponLevel)
    #         Print("Weapon purchased successfully.")
    #         return true
    #     else:
    #         Print("Purchase failed. Not enough resources.")
    #         return false
    
    # # 実際の購入処理はトランザクション内で実行
    # PurchaseWeaponTransaction<private>(Player:player, NeededStone:big_number, NeededPowerupStone:big_number, WeaponLevel:int)<transacts>:void=
    #     SaveDataManager.SubtractResourceStone(Player, NeededStone)
    #     SaveDataManager.SubtractResourcePowerupStone(Player, NeededPowerupStone)
    #     UpgradeWeapon(Player, WeaponLevel)
    
    # # 武器のアップグレード処理
    # UpgradeWeapon<public>(Player:player, WeaponLevel:int)<transacts>:void=
    #     SaveDataManager.SetLevel(Player, WeaponLevel + 1)