using { /UnrealEngine.com/Temporary/SpatialMath }
using { /Fortnite.com/Devices }
using { /Verse.org/Simulation }

vfx_device<public> := class(creative_device):

    @editable
    VFXPropAsset : creative_prop_asset = DefaultCreativePropAsset

    @editable
    DisposeTime : float = 5.0

    Spawn<public>(Pos:transform, ?CustomTime:float=DisposeTime):void=
        MaybeProp := SpawnProp(VFXPropAsset, Pos)(0)
        if(Prop := MaybeProp?):
            spawn{AutoDispose(Prop, CustomTime)}

    AutoDispose(Prop:creative_prop, Time:float)<suspends>:void=
        Sleep(Time)
        Prop.Dispose()