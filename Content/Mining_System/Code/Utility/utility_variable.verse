(Map:[ct]t).RemoveKeyFromMap<public>(ElementToRemove:ct where ct:subtype(comparable), t:type)<transacts>:[ct]t=
    var NewMap:[ct]t = map{}
    for (Key -> Value : Map, Key <> ElementToRemove):
        set NewMap = ConcatenateMaps(NewMap, map{Key => Value})
    return NewMap

AddMap<public>(Left:[ct]t, Right:tuple(ct, t) where ct:subtype(comparable), t:type)<transacts>:[ct]t=
    var NewMap:[ct]t = map{}
    for (Key -> Value : Left, Key <> Right(0)):
        set NewMap = ConcatenateMaps(NewMap, map{Key => Value})
    set NewMap = ConcatenateMaps(NewMap, map{Right(0) => Right(1)})

    NewMap
# stringをmessageに変換する
# ただし、将来のローカライズ対応に支障がでるため使用は避けること
# 固定の文字列のばあはmessageの定数を作る
StringToMessage<public><localizes>(Text:string):message="{Text}"