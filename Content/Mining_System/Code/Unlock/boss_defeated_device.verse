
using { /Fortnite.com/Devices }
using { /Fortnite.com/Characters }
using { /Verse.org/Simulation }
using { /UnrealEngine.com/Temporary/Diagnostics }
using { /UnrealEngine.com/Temporary/UI }

using { PlayerData }
using{ VerseDevice }

boss_defeated_device := class(creative_device):

    var PlayerSaveData : save_data_manager = save_data_manager{}

    @editable
    Buttons:[]button_device = array{}

    @editable UI_Bosses : []ui_boss = array{}

    @editable
    Teleporters:[]teleporter_device = array{}

    @editable
    Hud : hud_message_device = hud_message_device{}

    @editable
    BossUItrigger : trigger_device = trigger_device{}


    OnBegin<override>()<suspends>:void =
        for(Player : GetPlayspace().GetPlayers()):
            spawn{AwaitButtonSequence(Player)}

        GetPlayspace().PlayerAddedEvent().Subscribe(CallButtonSequence)

    CallButtonSequence(Player : player):void=
        spawn{AwaitButtonSequence(Player)}

    AwaitButtonSequence(Player:player)<suspends>:void =
        if(PlayerSaveData.HasSaveData(Player)?):
            loop:
                race:
                    block:
                        if (Button := Buttons[0]):
                            Button.InteractedWithEvent.Await()
                            spawn{HandleButtonPress(Player, 1)}
                            Print("クレッパー, スタート待ち")
                    block:
                        if (Button := Buttons[1]):
                            Button.InteractedWithEvent.Await()
                            spawn{HandleButtonPress(Player, 2)}
                            Print("タチガラス, スタート待ち")
                    block:
                        if (Button := Buttons[2]):
                            Button.InteractedWithEvent.Await()
                            spawn{HandleButtonPress(Player, 3)}
                            Print("オオバンモメン, スタート待ち")
                    block:
                        if (Button := Buttons[3]):
                            Button.InteractedWithEvent.Await()
                            spawn{HandleButtonPress(Player, 4)}
                            Print("カラカサ, スタート待ち")
                    block:
                        if (Button := Buttons[4]):
                            Button.InteractedWithEvent.Await()
                            spawn{HandleButtonPress(Player, 5)}
                            Print("エンエンガマ, スタート待ち")
                    block:
                        if (Button := Buttons[5]):
                            Button.InteractedWithEvent.Await()
                            spawn{HandleButtonPress(Player, 6)}
                            Print("ヤタガラン, スタート待ち")
        else:
            Print("Error: AwaitButtonSequence")

    HandleButtonPress(Player:player, BossID:int)<suspends>:void =
        if(PlayerSaveData.HasSaveData(Player)?):
            
            # ボタンを押したプレイヤーに対してのみUIを表示
            if(Agent := agent[Player], TheUI := UI_Bosses[BossID - 1]):
                TheUI.OnBossSignalEvented(Agent)
    
            # ボス戦を開始（プレイヤーごとに分離）
            spawn{StartBossFight(Player, BossID)}
        else:
            Print("Error: HandleButtonPress")
            


    Not<localizes> : message = "Not defeated the previous boss. Cannot proceed to the boss."
    AwaitBossFight(Player:player, BossID:int, PreviousBossID:int)<suspends>:void =
        if(PlayerSaveData.HasSaveData(Player)?):
            if(PreviousBossID = 0 or PlayerSaveData.IsDefeatedBoss[Player, PreviousBossID]):
                spawn { StartBossFight(Player, BossID) }
                if(TheUI := UI_Bosses[BossID - 1]):
                    TheUI.OnBossSignalEvented(Player)
            else:
                Hud.SetText(Not)
                Hud.Show(Player)

    StartBossFight<public>(Player:player, BossID:int)<suspends>:void =
        Print("Starting fight with Boss {BossID}")
        
        # Teleportersのインデックスが有効か確認
        if (Teleporter := Teleporters[BossID - 1]):
            Agent := Teleporter.TeleportedEvent.Await()
            
            if (Agent = Player):
                #spawn{OnBossTeleported(Player, BossID)}
                PlayerSaveData.DefeatBoss(Player, BossID)
                Print("Player defeated Boss {BossID}")
        else:
            Print("有効なテレポーターなし{BossID}")

    OnBossTeleported(Player:player, BossID:int)<suspends>:void =
        PlayerSaveData.DefeatBoss(Player, BossID)
        Print("Player defeated Boss {BossID}")

    