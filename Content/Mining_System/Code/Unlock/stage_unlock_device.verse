using { /Fortnite.com/Devices }
using { /Fortnite.com/Characters }
using { /Verse.org/Simulation }
using { /UnrealEngine.com/Temporary/Diagnostics }

using { PlayerData }
using { Utility }

stage_unlock_device := class(creative_device):

    var PlayerSaveData : save_data_manager = save_data_manager{}

    @editable
    Button_Debug : button_device = button_device{}

    @editable
    Buttons : []button_device = array{}

    @editable
    Barriers : []barrier_device = array{}

    @editable
    Barriers_Gacha : []barrier_device = array{}

    @editable
    QuestTriggers : []trigger_device = array{}

    @editable
    Hud : hud_message_device = hud_message_device{}
    
    # ステージコストの定義
    StageUnlockCosts : [int]big_number = map{
        1 => big_number{Value := 1.0, Exponent := 3},         # 1000
        2 => big_number{Value := 5.0, Exponent := 5},         # 500000
        3 => big_number{Value := 1.0, Exponent := 7},         # 10000000
        4 => big_number{Value := 1.0, Exponent := 9},         # 1000000000
        5 => big_number{Value := 1.0, Exponent := 12},        # 1000000000000
        6 => big_number{Value := 5.0, Exponent := 14},        # 500000000000000
        7 => big_number{Value := 1.0, Exponent := 16},        # 10000000000000000
        8 => big_number{Value := 5.0, Exponent := 18},        # 5000000000000000000
        9 => big_number{Value := 5.0, Exponent := 21},        # 5000000000000000000000
        10 => big_number{Value := 1.0, Exponent := 23},       # 100000000000000000000000
        11 => big_number{Value := 5.0, Exponent := 24},       # 5000000000000000000000000
        12 => big_number{Value := 2.5, Exponent := 27}        # 2500000000000000000000000000
    }


    OnBegin<override>()<suspends>:void =
        Button_Debug.InteractedWithEvent.Subscribe(Debug)#Debug
        for(Index -> TheButton : Buttons) :
            spawn{AwaitInteract(TheButton, Index)}

        for(Player : GetPlayspace().GetPlayers()):
            CheckAndUnlockBarriers(Player)

        GetPlayspace().PlayerAddedEvent().Subscribe(CheckAndUnlockBarriers)
        Print("UnlockDevice_Begin")

    AwaitInteract(Button : button_device, Index : int)<suspends>:void=
        loop :
            Agent := Button.InteractedWithEvent.Await()
            if(Player : player = player[Agent]) :
                UnlockStageWithButton(Player, Index + 1)

    # リソースを確認しステージをアンロック
    Not<localizes> : message = "Not enough resources !"
    Already<localizes> : message = "Already Unlocked !"
    
    UnlockStageWithButton(Player: player, StageID: int)<suspends>: void =
        if(PlayerSaveData.HasSaveData(Player)?):
            UnlockedStages := PlayerSaveData.GetUnlockedStages(Player)
            LastUnlockedStage : int = UnlockedStages.Length
            Print("Attempting to unlock StageID: {StageID}")
        
            if(StageID > LastUnlockedStage) :
                # ステージのコストを取得
                if (Cost := StageUnlockCosts[StageID]):
                    Print("Cost for Stage {StageID}: {Cost.Value}")
            
                    # 持続データの big_number<public> からプレイヤーのリソースを取得
                    PlayerResource := PlayerSaveData.GetResourceStone(Player)
            
                    # プレイヤーのリソースがステージのコストを上回っているか確認
                    if (BigNumber_GreaterThanOrEquals(PlayerResource, Cost)?):
                        Print("Player has: {PlayerResource.Value},  {PlayerResource.Exponent}")
            
                        # リソースをコスト分だけ減算
                        PlayerSaveData.SubtractResourceStone(Player, Cost)
            
                        # ステージをアンロック
                        PlayerSaveData.UnlockStage(Player, StageID)
                        Print("Stage {StageID} unlocked!")
            
                        # バリアを解除
                        CheckAndUnlockBarriers(Player)
                    else:
                        # リソースが不足している場合
                        Hud.SetText(Not)
                        Hud.Show(Player)
                        Print("Not enough resources. Player has: {BigNumber_FormatDisplay(PlayerResource)}, need: {BigNumber_FormatDisplay(Cost)}")
                else:
                    Print("No cost found for StageID: {StageID}")

            else :
                Hud.SetText(Already)
                Hud.Show(Player)
                CheckAndUnlockBarriers(Player)
                Print("Already Unlocked!")
        else:
            Print("Error: UnlockStageWithButton")
    
    # プレイヤーのステージ解放状況に基づいてバリアを解除(解放したバリアはintの配列で管理、1,2,3,4,...←これがUnlockStages)
    CheckAndUnlockBarriers(Player:player):void =
        UnlockedStages := PlayerSaveData.GetUnlockedStages(Player)
        #最後にUnlockしたstageに対応するバリアのIndex
        #例：1stage目まで解放したならUnlockStages.Length = 1 で,開放すべきBrrierIndex = 0
        LastUnlockedBarrier : int = UnlockedStages.Length - 1
        if (UnlockedStages.Length > 0):
            for (Stage : UnlockedStages):
                BarrierIndex := Stage - 1
                if (Barrier := Barriers[BarrierIndex]):
                    Print("Unlocking barrier for Stage {Stage}")
                    Barrier.AddToIgnoreList(Player)
                    #ガチャのバリア解除、BarrierIndex = 2,6,10 の時
                    if(BarrierIndex = 2 , TheBarrier := Barriers_Gacha[0]) :
                        TheBarrier.AddToIgnoreList(Player)
                        Print("Gacha1-3 Unlocked!")
                    if(BarrierIndex = 6 , TheBarrier := Barriers_Gacha[1]) :
                        TheBarrier.AddToIgnoreList(Player)
                        Print("Gacha2-3 Unlocked!")
                    if(BarrierIndex = 10 , TheBarrier := Barriers_Gacha[2]) :
                        TheBarrier.AddToIgnoreList(Player)
                        Print("Gacha3-3 Unlocked!")

                if(Trigger := QuestTriggers[BarrierIndex]) :
                    Trigger.Trigger(Player)
                    Print("QuestTrigger{BarrierIndex} Activate!")
                else:
                    Print("No barrier found for Stage {Stage}")
        else:
            Print("No unlocked stages found")

    Debug(Agent : agent):void=
        Print("UnlockDebug Interacted and Call func")
        if(Player : player = player[Agent]) :
            CheckAndUnlockBarriers(Player)