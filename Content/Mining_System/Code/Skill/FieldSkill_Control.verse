using { /Fortnite.com/Devices }
using { /Verse.org/Simulation }
using { /UnrealEngine.com/Temporary/Diagnostics }
using { /Fortnite.com/Characters }
using { /UnrealEngine.com/Temporary/SpatialMath }
using { /Verse.org/Random }
using { /Verse.org/Colors }
using { /UnrealEngine.com/Temporary/UI }
using { /Fortnite.com/UI }

using { MiningObject }
using { PlayerData }
using { Skill }
using { Weapon }

FieldSkill_Control<public>:= class(creative_device):

    var Mining_info : mining_object_info = mining_object_info{}
    var Mining_Damage : mining_damage_device = mining_damage_device{}
    var Save_Data : save_data = save_data{}
    var SaveDataManager : save_data_manager = save_data_manager{}
    var SkillManager : skill_manager = skill_manager{}
    #攻撃時のダメージ判定
    var Weapon_Damage : weapon_damage = weapon_damage{}

    @editable
    MM_Reset : movement_modulator_device = movement_modulator_device{}
    @editable
    MMs_Speed : []movement_modulator_device = array{}
    @editable
    MM_FlashJump : movement_modulator_device = movement_modulator_device{}

    @editable
    VFX_FlashJump : visual_effect_powerup_device = visual_effect_powerup_device{}
    @editable
    SFX_FlashJump : audio_player_device = audio_player_device{}
    @editable
    VFX_Powerup : damage_amplifier_powerup_device = damage_amplifier_powerup_device{}
    @editable
    SFX_OnHit : audio_player_device = audio_player_device{}
    @editable
    SFXs_FieldBGM : []audio_player_device = array{}

    @editable :
    Input_Jump : input_trigger_device = input_trigger_device{}

    @editable
    Volumes_Boss : []volume_device = array{}
    @editable
    MZs_Field : []mutator_zone_device = array{}

    var Count_FlashJumpPerAgent : [agent]int = map{}
    var Count_ChargePerAgent : [agent]int = map{}
    var Charge_IsActivePerAgent : [agent]logic = map{}
    var Count_HitPerAgent : [agent]int = map{}
    var IsExplode : [agent]logic = map{}
    #----------------------------Widget----------------------------
    var CCWidgetPerPlayer : [player]ui_ChargeCount = map{}
    var TimerCountPerPlayer : [player]int = map{}
    var IsTimerSpawned : [player]logic = map{}

    ##Field出入り時のイベント-----------------------------------------------
    OnEnterToBoss(Agent : agent) : void =
        if(not Count_FlashJumpPerAgent[Agent]) :
            logic{set Count_FlashJumpPerAgent[Agent] = 0}
        if(not Count_ChargePerAgent[Agent]) :
            logic{set Count_ChargePerAgent[Agent] = 0}
        if(not Charge_IsActivePerAgent[Agent]) :
            logic{set Charge_IsActivePerAgent[Agent] = false}
        if(not Count_HitPerAgent[Agent]) :
            logic{set Count_HitPerAgent[Agent] = 0}

        MM_Reset.Activate(Agent)
        logic{set Count_FlashJumpPerAgent[Agent] = 0}
        logic{set Count_ChargePerAgent[Agent] = 0}
        logic{set Charge_IsActivePerAgent[Agent] = false}
        logic{set Count_HitPerAgent[Agent] = 0}
        VFX_Powerup.SetDuration(0.1)
        VFX_Powerup.Pickup(Agent)
        if :
            Player : player = player[Agent]
            Char : fort_character = Agent.GetFortCharacter[]
            HP_Level := SaveDataManager.GetSkillLevel[Player , 13]
        then :
            Char.SetMaxHealth(100.0 + 10.0 * HP_Level)
            Char.SetHealth(100.0 + 10.0 * HP_Level)
            Print("PlayerHP{100.0 + 10.0 * HP_Level}")

    OnEnterToField(TheMZ : mutator_zone_device , TheSFX : audio_player_device)<suspends> : void =
        loop :
            Agent := TheMZ.AgentEntersEvent.Await()
            TheSFX.Play(Agent)
            Print("BGM Play")
            if :
                Player : player = player[Agent]
                Char : fort_character = Agent.GetFortCharacter[]
                MoveSpeed_Level := SaveDataManager.GetSkillLevel[Player , 1]
                TheMM := MMs_Speed[MoveSpeed_Level - 1]
            then :
                TheMM.Activate(Agent)
                Print("NowMoveSpeedLevel{MoveSpeed_Level}")
                Char.SetMaxHealth(100.0)
                Char.SetHealth(100.0)

    OnExitFromField(TheMZ : mutator_zone_device , TheSFX : audio_player_device)<suspends> : void =
        loop :
            Agent := TheMZ.AgentExitsEvent.Await()
            if :
                Player : player = player[Agent]
                Amount := Weapon_Damage.CalculateDamage[Player]
            then :
                Print("Amount{Amount}")
            TheSFX.Stop(Agent)
            Print("BGM Stop")

    ##-------------------------------------------------------------------

    OnInput_Jump_Pressed(Agent : agent) : void =
        if(not Count_FlashJumpPerAgent[Agent]) : 
            logic{set Count_FlashJumpPerAgent[Agent] = 0}
        if :
            #例：Levelが1の時はCountは1未満、つまりCount = 0のときのみ成立するので1回だけFlashJumpできる
            Player : player = player[Agent]
            FlashJump_Level := SaveDataManager.GetSkillLevel[Player , 2]
            Count_FlashJumpPerAgent[Agent] < FlashJump_Level
            MZ_0 := MZs_Field[0]
            MZ_1 := MZs_Field[1]
            MZ_2 := MZs_Field[2]
            Char : fort_character = Agent.GetFortCharacter[]
            Char.IsInAir[]
        then :
            if(MZ_0.IsInVolume[Agent] or MZ_1.IsInVolume[Agent] or MZ_2.IsInVolume[Agent]) :
                logic{set Count_FlashJumpPerAgent[Agent] += 1}
                MM_FlashJump.Activate(Agent)
                VFX_FlashJump.Pickup(Agent)
                SFX_FlashJump.Play(Agent)
                if(Count_FlashJumpPerAgent[Agent] = 1) :
                    spawn{FlashJump(Agent , Char)}

    FlashJump(Agent : agent , Char : fort_character)<suspends>: void =
        Print("Spawn FlashJump")
        loop :
            if(Char.IsInAir[]) :
                Sleep(0.1)
            else :
                break
        logic{set Count_FlashJumpPerAgent[Agent] = 0}
        Print("Reset FlashJump!")

    Get_CountCharge<public>(Agent : agent):int =
        #初期化------------------------------------
        if(not Count_ChargePerAgent[Agent]) :
            logic{set Count_ChargePerAgent[Agent] = 0}
        #-----------------------------------------
        var Count : int = 0
        if(Value := Count_ChargePerAgent[Agent]) :
            set Count = Value

        return Count

    Get_ChargeIsActive<public>(Agent : agent):logic =
        #初期化------------------------------------
        if(not Charge_IsActivePerAgent[Agent]) :
            logic{set Charge_IsActivePerAgent[Agent] = false}
        #-----------------------------------------
        var Count : logic = false
        if(Value := Charge_IsActivePerAgent[Agent]) :
            set Count = Value

        return Count

    Get_CountHit<public>(Agent : agent):int =
        #初期化------------------------------------
        if(not Count_HitPerAgent[Agent]) :
            logic{set Count_HitPerAgent[Agent] = 0}
        #-----------------------------------------
        var Count : int = 0
        if(Value := Count_HitPerAgent[Agent]) :
            set Count = Value

        return Count

    Get_IsExplode<public>(Agent : agent):logic =
        #初期化------------------------------------
        if(not IsExplode[Agent]) :
            logic{set IsExplode[Agent] = false}
        #-----------------------------------------
        var Count : logic = false
        if(Value := IsExplode[Agent]) :
            set Count = Value

        return Count
    Set_CountCharge<public>(Agent : agent, Value : int):void =
        logic{set Count_ChargePerAgent[Agent] = Value}

    Set_ChargeIsActive<public>(Agent : agent, Value : logic):void =
        logic{set Charge_IsActivePerAgent[Agent] = Value}

    Set_CountHit<public>(Agent : agent, Value : int):void =
        logic{set Count_HitPerAgent[Agent] = Value}

    Set_IsExplode<public>(Agent : agent)<suspends>:void =
        logic{set IsExplode[Agent] = true}
        Sleep(2.0)
        logic{set IsExplode[Agent] = false}


    OnEvent<public>(Info : mining_object_info, Agent : agent)<suspends>:void=
        Print("Damaged", ?Color := NamedColors.Red)

    Combo<public>(MoreHits_Level : int, Info : mining_object_info, Player : player)<suspends>:void=
        SkillDamage := Mining_Damage.GetDamageAmount(Player).Amount
        Info.Damage(Player, SkillDamage)
        Sleep(0.05)
        Info.Damage(Player, SkillDamage)
        if(MoreHits_Level >= 1) :
            Sleep(0.05)
            Info.Damage(Player, SkillDamage)
            if(MoreHits_Level >= 2) :
                Sleep(0.05)
                Info.Damage(Player, SkillDamage)
                if(MoreHits_Level >= 3) :
                    Sleep(0.05)
                    Info.Damage(Player, SkillDamage)
                    if(MoreHits_Level >= 4) :
                        Sleep(0.05)
                        Info.Damage(Player, SkillDamage)
                        if(MoreHits_Level >= 5) :
                            Sleep(0.05)
                            Info.Damage(Player, SkillDamage)
                            if(MoreHits_Level >= 6) :
                                Sleep(0.05)
                                Info.Damage(Player, SkillDamage)
        Print("コンボ終了", ?Color := NamedColors.Red)

    CallWidgetTimer<public>(Agent : agent, Player : player)<suspends>:void=
        logic{set TimerCountPerPlayer[Player] = 3}
        if(not IsTimerSpawned[Player]) :
            logic{set IsTimerSpawned[Player] = false}
        if(IsTimerSpawned[Player] = false) :
            spawn{HideCCWidget(Player)}

        if(TheCount := Count_ChargePerAgent[Agent]) :
            Print("Count{TheCount}", ?Color := NamedColors.Red)
            spawn{UpdateCCWidget(Player, TheCount)}

    StringToMessage<localizes>(String : string) : message = "{String}"

    UpdateCCWidget(Player : player, Count : int)<suspends>:void=
        CheckCCWidget(Player)
        NormalSize := vector2{X:= 166.259033, Y:= 44.338566}
        CriticalSize := vector2{X:= 184.286789, Y:= 44.338566}
        if(Element := CCWidgetPerPlayer[Player]) :
            Element.BG_ChargeCount.SetVisibility(widget_visibility.Visible)
            Element.Icon_ChargeCount.SetVisibility(widget_visibility.Visible)
            Element.Text_ChargeCount.SetVisibility(widget_visibility.Visible)
            if(Count < 100) :
                Element.BG_ChargeCount.SetDesiredSize(NormalSize)
                Element.Text_ChargeCount.SetText(StringToMessage("{Count}/100"))
                Element.Text_ChargeCount.SetTextColor(NamedColors.White)
            if(Count >= 100) :
                Element.BG_ChargeCount.SetDesiredSize(CriticalSize)
                Element.Text_ChargeCount.SetText(StringToMessage("PowerUp!"))
                Element.Text_ChargeCount.SetTextColor(NamedColors.Orange)

    HideCCWidget(Player : player)<suspends>:void=
        logic{set IsTimerSpawned[Player] = true}
        loop :
            logic{set TimerCountPerPlayer[Player] -= 1}
            Sleep(1.0)
            if :
                NowTime := TimerCountPerPlayer[Player]
                NowTime < 0
            then :
                break
        CheckCCWidget(Player)
        if(Element := CCWidgetPerPlayer[Player]) :
            Element.BG_ChargeCount.SetVisibility(widget_visibility.Collapsed)
            Element.Icon_ChargeCount.SetVisibility(widget_visibility.Collapsed)
            Element.Text_ChargeCount.SetVisibility(widget_visibility.Collapsed)
            Print("CCWidget非表示", ?Color := NamedColors.Red)
        logic{set IsTimerSpawned[Player] = false}

    CheckCCWidget(Player : player):void=
        if(Widget := CCWidgetPerPlayer[Player]) :
            return #Widget作成済みなら何もしない
        if(PlayerUI : player_ui = GetPlayerUI[Player]) :
            NewUI := ui_ChargeCount{}
            PlayerUI.AddWidget(NewUI.CreateCCCanvas())
            logic{set CCWidgetPerPlayer[Player] = NewUI}

    # Runs when the device is started in a running game
    OnBegin<override>()<suspends>:void=
        for(TheVolume : Volumes_Boss) :
            TheVolume.AgentEntersEvent.Subscribe(OnEnterToBoss)
        for(Index -> TheMZ : MZs_Field) :
            if(TheSFX := SFXs_FieldBGM[Index]) :
                spawn{OnEnterToField(TheMZ , TheSFX)}
                spawn{OnExitFromField(TheMZ , TheSFX)}
        Input_Jump.PressedEvent.Subscribe(OnInput_Jump_Pressed)

ui_ChargeCount := class :
    BG_ChargeCount : color_block = color_block :
        DefaultColor := NamedColors.Black
        DefaultOpacity := 0.7
        DefaultDesiredSize := vector2{X:= 166.259033, Y:= 44.338566}
    Icon_ChargeCount : texture_block = texture_block :
        DefaultImage := icon_set_2.IC_Mining
        DefaultTint := color{R:= 1.0, G:= 1.0, B:= 1.0}
        DefaultDesiredSize := vector2{X:= 34.866611, Y:= 34.866611}
    Text_ChargeCount : text_block = text_block :
        DefaultTextColor := color{R := 1.0, G := 1.0, B := 1.0}
        DefaultTextOpacity := 1.0
        DefaultShadowOffset := option{vector2{X := 0.0,Y := 0.0}}
        DefaultShadowColor := color{R := 1.0, G := 1.0, B := 1.0}
        DefaultShadowOpacity := 0.0

    CreateCCCanvas<public>():canvas=
        canvas :
            Slots := array :
                canvas_slot :
                    Anchors := anchors:
                        Minimum := vector2{X:= 0.591500, Y:= 0.524280}
                        Maximum := vector2{X:= 0.591500, Y:= 0.524280}
                    Offsets := margin{Top := 3.146622, Left := -0.640000, Right := 0.0, Bottom := 0.0}
                    SizeToContent := true
                    Alignment := vector2{X := 1.0, Y := 0.5}
                    Widget := BG_ChargeCount
                canvas_slot :
                    Anchors := anchors:
                        Minimum := vector2{X:= 0.591500, Y:= 0.524280}
                        Maximum := vector2{X:= 0.591500, Y:= 0.524280}
                    Offsets := margin{Top := 2.729148, Left := 0.0, Right := 0.0, Bottom := 0.0}
                    SizeToContent := true
                    Alignment := vector2{X := 1.0, Y := 0.5}
                    Widget := Icon_ChargeCount
                canvas_slot :
                    Anchors := anchors:
                        Minimum := vector2{X:= 0.591500, Y:= 0.524280}
                        Maximum := vector2{X:= 0.591500, Y:= 0.524280}
                    Offsets := margin{Top := 2.908783, Left := -32.728542, Right := 0.0, Bottom := 0.0}
                    SizeToContent := true
                    Alignment := vector2{X := 1.0, Y := 0.5}
                    Widget := Text_ChargeCount

