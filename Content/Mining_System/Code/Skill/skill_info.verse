
using { PlayerData }

skill_info<public> := struct<persistable>:
    Level<public> : int = 0                             # 現在のスキルレベル
    MaxLevel<public> : int = 0                          # スキルの最大レベル
    EffectPower<public> : int = 0                       # スキルの基本効果
    EffectIncrease<public> : int = 0                    # スキルレベルごとの効果上昇値
    InitialCost<public> : big_number = big_number{}     # スキルの初期コスト
    CostIncrease<public> : int = 0                      # スキルレベルごとのコスト上昇値
