using { /Fortnite.com/Devices }
using { /Verse.org/Simulation }
using { /Fortnite.com/Characters }
using { PlayerData }


skill_manager<public> := class(creative_device):
    @editable
    Button : button_device = button_device{}

    var DefaultSkillInfoData<public> : [int]skill_info = map{
        # 移動系
        1 => skill_info{Level := 0, MaxLevel := 5, EffectPower := 100, EffectIncrease := 10, InitialCost := big_number{Value := 10.0, Exponent := 6}, CostIncrease := 10},  # Move Speed
        2 => skill_info{Level := 0, MaxLevel := 2, EffectPower := 0, EffectIncrease := 1, InitialCost := big_number{Value := 500.0, Exponent := 12}, CostIncrease := 25},  # Flash Jump
        # 放置系
        3 => skill_info{Level := 0, MaxLevel := 8, EffectPower := 10, EffectIncrease := 5, InitialCost := big_number{Value := 1.00, Exponent := 9}, CostIncrease := 150},  # Player idle
        4 => skill_info{Level := 0, MaxLevel := 8, EffectPower := 10, EffectIncrease := 5, InitialCost := big_number{Value := 1.00, Exponent := 15}, CostIncrease := 150},  # Pet idle
        # バフ        
        5 => skill_info{Level := 0, MaxLevel := 1, EffectPower := 0, EffectIncrease := 1, InitialCost := big_number{Value := 1.5, Exponent := 9}, CostIncrease := 0},  # Buff Unlock
        6 => skill_info{Level := 0, MaxLevel := 10, EffectPower := 200, EffectIncrease := 50, InitialCost := big_number{Value := 100.0, Exponent := 6}, CostIncrease := 20},  # Buff effect
        7 => skill_info{Level := 0, MaxLevel := 10, EffectPower := 90, EffectIncrease := 9, InitialCost := big_number{Value := 2.00, Exponent := 9}, CostIncrease := 20},  # Buff time
        8 => skill_info{Level := 0, MaxLevel := 3, EffectPower := 1, EffectIncrease := 1, InitialCost := big_number{Value := 30.0, Exponent := 12}, CostIncrease := 100},  # Add stock
        # 召喚石
        10 => skill_info{Level := 0, MaxLevel := 1, EffectPower := 0, EffectIncrease := 10, InitialCost := big_number{Value := 2.00, Exponent := 6}, CostIncrease := 10},  # Summon stone Drop
        # ボス
        12 => skill_info{Level := 0, MaxLevel := 3, EffectPower := 1, EffectIncrease := 1, InitialCost := big_number{Value := 5.0, Exponent := 12}, CostIncrease := 10},  # Boss Add stock
        13 => skill_info{Level := 0, MaxLevel := 20, EffectPower := 100, EffectIncrease := 10, InitialCost := big_number{Value := 1.00, Exponent := 12}, CostIncrease := 10},  # Boss HP
        14 => skill_info{Level := 0, MaxLevel := 1, EffectPower := 0, EffectIncrease := 1, InitialCost := big_number{Value := 30.0, Exponent := 12}, CostIncrease := 0},  # Pet support
        # Stage1 クリティカル
        15 => skill_info{Level := 0, MaxLevel := 10, EffectPower := 5, EffectIncrease := 5, InitialCost := big_number{Value := 15.0, Exponent := 3}, CostIncrease := 10},  # Critical rate
        16 => skill_info{Level := 0, MaxLevel := 10, EffectPower := 10, EffectIncrease := 10, InitialCost := big_number{Value := 300.0, Exponent := 3}, CostIncrease := 10},  # Critical multiplier
        17 => skill_info{Level := 0, MaxLevel := 10, EffectPower := 50, EffectIncrease := 20, InitialCost := big_number{Value := 500.0, Exponent := 3}, CostIncrease := 10},  # Final hit bonus
        18 => skill_info{Level := 0, MaxLevel := 4, EffectPower := 2, EffectIncrease := 1, InitialCost := big_number{Value := 2.00, Exponent := 9}, CostIncrease := 1000},  # Critical power surge
        # Stage2 コンボ
        19 => skill_info{Level := 0, MaxLevel := 1, EffectPower := 0, EffectIncrease := 0, InitialCost := big_number{Value := 1.00, Exponent := 12}, CostIncrease := 0},  # Combo unlock
        20 => skill_info{Level := 0, MaxLevel := 9, EffectPower := 1, EffectIncrease := 1, InitialCost := big_number{Value := 15.0, Exponent := 12}, CostIncrease := 10},  # Charge count
        21 => skill_info{Level := 0, MaxLevel := 6, EffectPower := 2, EffectIncrease := 1, InitialCost := big_number{Value := 30.0, Exponent := 12}, CostIncrease := 10},  # More hits
        22 => skill_info{Level := 0, MaxLevel := 3, EffectPower := 1, EffectIncrease := 1, InitialCost := big_number{Value := 200.0, Exponent := 12}, CostIncrease := 1000},  # Combo mode
        # Stage3 範囲
        23 => skill_info{Level := 0, MaxLevel := 1, EffectPower := 0, EffectIncrease := 0, InitialCost := big_number{Value := 5.00, Exponent := 21}, CostIncrease := 0},  # Explosion unlock
        24 => skill_info{Level := 0, MaxLevel := 10, EffectPower := 1000, EffectIncrease := 100, InitialCost := big_number{Value := 250.00, Exponent := 21}, CostIncrease := 10},  # Area expansion
        25 => skill_info{Level := 0, MaxLevel := 10, EffectPower := 3000, EffectIncrease := 300, InitialCost := big_number{Value := 500.00, Exponent := 21}, CostIncrease := 10}  # Damage multiplier
    }

    var SkillStatusData<public> : [int]skill_status = map{
        # 移動系
        1 => skill_status{Index := 1, IsActive := false, CooldownTime := 0, MaxCooldownTime := 0, Charges := 1, MaxCharges := 1, LastUsedTime := 0, EffectDuration := 0},  # Move Speed
        2 => skill_status{Index := 2, IsActive := false, CooldownTime := 0, MaxCooldownTime := 0, Charges := 2, MaxCharges := 2, LastUsedTime := 0, EffectDuration := 0},  # Flash Jump
        # 放置系
        3 => skill_status{Index := 3, IsActive := false, CooldownTime := 0, MaxCooldownTime := 0, Charges := 1, MaxCharges := 1, LastUsedTime := 0, EffectDuration := 0},  # Player idle
        4 => skill_status{Index := 4, IsActive := false, CooldownTime := 0, MaxCooldownTime := 0, Charges := 1, MaxCharges := 1, LastUsedTime := 0, EffectDuration := 0},  # Pet idle
        # バフ
        5 => skill_status{Index := 5, IsActive := false, CooldownTime := 900, MaxCooldownTime := 900, Charges := 1, MaxCharges := 1, LastUsedTime := 0, EffectDuration := 90},  # Buff Unlock
        6 => skill_status{Index := 6, IsActive := false, CooldownTime := 0, MaxCooldownTime := 0, Charges := 1, MaxCharges := 1, LastUsedTime := 0, EffectDuration := 0},  # Buff effect
        7 => skill_status{Index := 7, IsActive := false, CooldownTime := 900, MaxCooldownTime := 900, Charges := 1, MaxCharges := 4, LastUsedTime := 0, EffectDuration := 90},  # Buff time
        8 => skill_status{Index := 8, IsActive := false, CooldownTime := 0, MaxCooldownTime := 0, Charges := 1, MaxCharges := 1, LastUsedTime := 0, EffectDuration := 0},  # Add stock
        # 召喚石
        10 => skill_status{Index := 10, IsActive := false, CooldownTime := 0, MaxCooldownTime := 0, Charges := 1, MaxCharges := 1, LastUsedTime := 0, EffectDuration := 0},  # Summon stone DropD
        # ボス
        12 => skill_status{Index := 12, IsActive := false, CooldownTime := 0, MaxCooldownTime := 0, Charges := 1, MaxCharges := 4, LastUsedTime := 0, EffectDuration := 0},  # Boss Add stock
        13 => skill_status{Index := 13, IsActive := false, CooldownTime := 0, MaxCooldownTime := 0, Charges := 1, MaxCharges := 1, LastUsedTime := 0, EffectDuration := 0},  # Boss HP
        14 => skill_status{Index := 14, IsActive := false, CooldownTime := 0, MaxCooldownTime := 0, Charges := 1, MaxCharges := 1, LastUsedTime := 0, EffectDuration := 0},  # Pet support
        # Stage1 クリティカル
        15 => skill_status{Index := 15, IsActive := false, CooldownTime := 0, MaxCooldownTime := 0, Charges := 1, MaxCharges := 1, LastUsedTime := 0, EffectDuration := 0},  # Critical rate
        16 => skill_status{Index := 16, IsActive := false, CooldownTime := 0, MaxCooldownTime := 0, Charges := 1, MaxCharges := 5, LastUsedTime := 0, EffectDuration := 0},  # Critical multiplier
        17 => skill_status{Index := 17, IsActive := false, CooldownTime := 0, MaxCooldownTime := 0, Charges := 1, MaxCharges := 1, LastUsedTime := 0, EffectDuration := 0},  # Final hit bonus
        18 => skill_status{Index := 18, IsActive := false, CooldownTime := 0, MaxCooldownTime := 0, Charges := 1, MaxCharges := 9, LastUsedTime := 0, EffectDuration := 0},  # Critical power surge
        # Stage2 コンボ
        19 => skill_status{Index := 19, IsActive := false, CooldownTime := 0, MaxCooldownTime := 0, Charges := 1, MaxCharges := 1, LastUsedTime := 0, EffectDuration := 0},  # Combo unlock
        20 => skill_status{Index := 20, IsActive := false, CooldownTime := 0, MaxCooldownTime := 0, Charges := 1, MaxCharges := 10, LastUsedTime := 0, EffectDuration := 0},  # Charge count
        21 => skill_status{Index := 21, IsActive := false, CooldownTime := 0, MaxCooldownTime := 0, Charges := 2, MaxCharges := 6, LastUsedTime := 0, EffectDuration := 0},  # More hits
        22 => skill_status{Index := 22, IsActive := false, CooldownTime := 0, MaxCooldownTime := 0, Charges := 1, MaxCharges := 3, LastUsedTime := 0, EffectDuration := 0},  # Combo mode
        # Stage3 範囲
        23 => skill_status{Index := 23, IsActive := false, CooldownTime := 0, MaxCooldownTime := 0, Charges := 1, MaxCharges := 1, LastUsedTime := 0, EffectDuration := 0},  # Explosion unlock
        24 => skill_status{Index := 24, IsActive := false, CooldownTime := 0, MaxCooldownTime := 0, Charges := 1, MaxCharges := 1, LastUsedTime := 0, EffectDuration := 0},  # Area expansion
        25 => skill_status{Index := 25, IsActive := false, CooldownTime := 0, MaxCooldownTime := 0, Charges := 1, MaxCharges := 1, LastUsedTime := 0, EffectDuration := 0}   # Damage multiplier
    }

    @editable
    SaveDataManager : save_data_manager = save_data_manager{}

    # スキルのレベルを呼び出す
    ReturnSkillLevel<public>(InPlayer : player, Index : int)<decides><transacts>:int=
        SkillInfo := SaveDataManager.GetSkill[InPlayer, Index]
        SkillInfo.Level

    # スキルの開放やレベルアップの際に呼び出す
    UnlockOrUpgradeSkill<public>(InPlayer: player, SkillIndex: int)<transacts><decides>: void =
        SkillInfo := SaveDataManager.GetSkill[InPlayer, SkillIndex]
        
        if:
            PlayerSaveData := SaveData[InPlayer]
    
            # スキルの初期化が必要か確認
            if (PlayerSaveData.Skill_Info.Length = 0):
                GiveInitialSkillsToPlayer(InPlayer)
            else:
                # レベルアップコストを計算
                UpgradeCost := CalculateUpgradeCost(SkillInfo)
    
                # プレイヤーの現在の所持リソースを取得
                CurrentResourceStone := SaveDataManager.GetResourceStone(InPlayer)
    
                # リソースが足りているか確認
                if (BigNumber_GreaterThanOrEquals(CurrentResourceStone, UpgradeCost)?):
                    # リソースを消費してスキルをレベルアップ
                    UpgradeSkillTransaction(InPlayer, SkillIndex, SkillInfo, UpgradeCost)
                else:
                    # 購入不可のメッセージを表示
                    Print("Not enough resources to upgrade the skill.")
        then:
    

    IsUnlockable<public>(InPlayer: player, SkillIndex: int)<decides><transacts>: void=
        SkillInfo := SaveDataManager.GetSkill[InPlayer, SkillIndex]
        Cost := CalculateUpgradeCost(SkillInfo)
        CurrentResourceStone := SaveDataManager.GetResourceStone(InPlayer)
        Print("Cost: {BigNumber_FormatDisplay(Cost)} CurrentResourceStone: {BigNumber_FormatDisplay(CurrentResourceStone)}")
        BigNumber_GreaterThanOrEquals(CurrentResourceStone, Cost)?

    # InitializeSkillTransaction<private>(InPlayer: player, Index: int)<transacts>: void=
    #     if:
    #         SkillInfo := SaveDataManager.GetSkill[InPlayer, Index]
    #     then:
    #         UpdatedSkill := skill_info{
    #             Level := SkillInfo.Level
    #             MaxLevel := SkillInfo.MaxLevel,
    #             EffectPower := SkillInfo.EffectPower,
    #             EffectIncrease := SkillInfo.EffectIncrease,
    #             InitialCost := SkillInfo.InitialCost,
    #             CostIncrease := SkillInfo.CostIncrease
    #         }
    #         SaveDataManager.SetSkill(InPlayer, Index, UpdatedSkill)
    #         Print("Skill {Index} initialized for player.")

    InitializeAllSkillTransaction<private>(InPlayer: player, Index: int)<transacts>: void=
        if:
            SkillInfo := SaveDataManager.GetSkill[InPlayer, Index]
        then:
            UpdatedSkill := skill_info{
                Level := 1,  # レベルを1に設定して初期化
                MaxLevel := SkillInfo.MaxLevel,
                EffectPower := SkillInfo.EffectPower,
                EffectIncrease := SkillInfo.EffectIncrease,
                InitialCost := SkillInfo.InitialCost,
                CostIncrease := SkillInfo.CostIncrease
            }
            SaveDataManager.SetSkill(InPlayer, Index, UpdatedSkill)
            Print("Skill {Index} initialized for player.")

    UpgradeSkillTransaction<private>(InPlayer: player, Index: int, SkillInfo: skill_info, UpgradeCost: big_number)<transacts>: void=
        CurrentResourceStone := SaveDataManager.GetResourceStone(InPlayer)
        IsEnoughResources := BigNumber_GreaterThanOrEquals(CurrentResourceStone, UpgradeCost)

        if(IsEnoughResources?):
            # リソースを消費
            SaveDataManager.SubtractResourceStone(InPlayer, UpgradeCost)

            # スキルをアップグレード
            UpdatedSkill := skill_info{
                Level := SkillInfo.Level + 1,
                MaxLevel := SkillInfo.MaxLevel,
                EffectPower := SkillInfo.EffectPower + SkillInfo.EffectIncrease,
                EffectIncrease := SkillInfo.EffectIncrease,
                InitialCost := SkillInfo.InitialCost,
                CostIncrease := SkillInfo.CostIncrease
            }
            SaveDataManager.SetSkill(InPlayer, Index, UpdatedSkill)
            Print("Skill {Index} upgraded")
        else:
            Print("Not enough Resource")
            
    # 特定のスキルのアクティブ状態を更新
    SetSkillActiveState<public>(InPlayer:player, Index:int, IsActive:logic)<transacts>:void=
        if:
            SkillStatus := SaveDataManager.GetSkillStatus[InPlayer, Index]
            UpdatedSkillStatus := skill_status{
                Index := SkillStatus.Index,
                IsActive := IsActive,
                CooldownTime := SkillStatus.CooldownTime,
                MaxCooldownTime := SkillStatus.MaxCooldownTime,
                Charges := SkillStatus.Charges,
                MaxCharges := SkillStatus.MaxCharges,
                LastUsedTime := SkillStatus.LastUsedTime,
                EffectDuration := SkillStatus.EffectDuration
            }
            SaveDataManager.SetSkillStatus(InPlayer, Index, UpdatedSkillStatus)

    # スキルの使用回数を更新
    UpdateSkillCharges<public>(InPlayer:player, Index:int, NewCharges:int)<transacts>:void=
        if:
            SkillStatus := SaveDataManager.GetSkillStatus[InPlayer, Index]
            UpdatedSkillStatus := skill_status{
                Index := SkillStatus.Index,
                IsActive := SkillStatus.IsActive,
                CooldownTime := SkillStatus.CooldownTime,
                MaxCooldownTime := SkillStatus.MaxCooldownTime,
                Charges := NewCharges,
                MaxCharges := SkillStatus.MaxCharges,
                LastUsedTime := SkillStatus.LastUsedTime,
                EffectDuration := SkillStatus.EffectDuration
            }
            SaveDataManager.SetSkillStatus(InPlayer, Index, UpdatedSkillStatus)

    # スキルのコストを計算
    CalculateUpgradeCost<public>(SkillInfo:skill_info)<transacts>:big_number=
        var NewCost : big_number = big_number{}
        BaseCost := SkillInfo.InitialCost
        LevelMultiplier := (SkillInfo.Level + 1) * 1.0
        set NewCost= BigNumber_MultiplyByFloat(BaseCost, LevelMultiplier)
        return NewCost
    
    # プレイヤーが持っている全スキルの効果を計算
    GetTotalSkillEffect<public>(InPlayer:player)<decides><transacts>:int=
        var TotalEffect : int = 0

        SkillStatuses := SaveDataManager.GetAllSkillStatuses[InPlayer]

        for(SkillElement : SkillStatuses):
            SkillIndex := SkillElement(0)
            SkillInfo := SaveDataManager.GetSkill[InPlayer, SkillIndex]
            set TotalEffect += SkillInfo.EffectPower
        
        return TotalEffect
        
    
    GiveInitialSkillsToPlayer<public>(Agent: agent)<transacts>: void=
        if (InPlayer := player[Agent]):
            # プレイヤーのセーブデータを取得
            if:
                PlayerSaveData := SaveData[InPlayer]
            
                # プレイヤーにスキルデータがない場合は初期化
                if (PlayerSaveData.Skill_Info.Length = 0):
                    # 初期スキルデータをプレイヤーにセット
                    for (SkillKey -> SkillInfo : DefaultSkillInfoData):
                        SaveDataManager.SetSkill(InPlayer, SkillKey, SkillInfo)
                        
                        # スキルステータスの初期化
                        InitiSkillStatus : skill_status = skill_status{
                            Index := SkillKey,
                            IsActive := false,
                            CooldownTime := 0,
                            MaxCooldownTime := 0,
                            Charges := 0,
                            MaxCharges := 0,
                            LastUsedTime := 0,
                            EffectDuration := 0,
                            Multiply := 2
                        }
                        SaveDataManager.SetSkillStatus(InPlayer, SkillKey, InitiSkillStatus)
                    
                    Print("Initial skill data and status have been given to player.")
                else:
                    Print("Player already has skill data.")
    

# スキルのレベルとMaxLevelを比較し、レベルアップが可能かどうかを返す
    CanUpgradeSkill<public>(InPlayer: player, SkillIndex: int)<transacts>: logic =
        if(PlayerSaveData := SaveData[InPlayer]):
            if(SkillData := PlayerSaveData.Skill_Info[SkillIndex]):
                if(SkillData.Level < SkillData.MaxLevel):
                    return true  # レベルアップ可能
                else:
                    return false  # すでに最大レベル
        return false

    GetSkillLevelStatus<public>(InPlayer: player, SkillIndex: int)<transacts>: string =
        var Lv_Drop : int = 1
        if(CanUpgradeSkill(InPlayer, SkillIndex) = false):
            return "LEVELMAX"  # レベルが最大の場合
        return "Lv.{Lv_Drop}"  # レベルアップ可能の場合

    OnBegin<override>()<suspends>:void=
        for(Player : GetPlayspace().GetPlayers()):
            if (FortniteCharacter : fort_character = Player.GetFortCharacter[]):
                GiveInitialSkillsToPlayer(Player)
        GetPlayspace().PlayerAddedEvent().Subscribe(GiveInitialSkillsToPlayer)
        Button.InteractedWithEvent.Subscribe(GiveInitialSkillsToPlayer)
        
        return
