skill_status<public> := struct<persistable>:
    Index<public> : int = 0                     # スキルのインデックス
    IsActive<public> : logic = false            # スキルが現在アクティブかどうか
    CooldownTime<public> : int = 0              # クールダウン中の残り時間
    MaxCooldownTime<public> : int = 0           # スキルの最大クールダウン時間
    Charges<public> : int = 0                   # スキルの現在の使用回数 (スタック)
    MaxCharges<public> : int = 0                # スキルの最大使用回数 (スタック)
    LastUsedTime<public> : int = 0              # 最後にスキルが使用された時刻
    EffectDuration<public> : int = 0            # スキルの効果が続く時間 (0なら一瞬のスキル)
    Multiply<public> : int = 0                  # 特定の値にいくらバフを掛けるか