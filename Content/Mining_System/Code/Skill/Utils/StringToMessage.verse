StringToMessage<public><localizes>(Text: string): message = "{Text}"