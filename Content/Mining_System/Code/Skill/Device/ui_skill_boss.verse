
using { /Verse.org/Assets }
using { /Verse.org/Simulation }
using { /Fortnite.com/UI }
using { /Fortnite.com/Devices }
using { /Fortnite.com/Characters }
using { /UnrealEngine.com/Temporary/UI }
using { /UnrealEngine.com/Temporary/SpatialMath }
using { /Verse.org/Simulation }
using { /UnrealEngine.com/Temporary/Diagnostics }
using { /Verse.org/Colors }
using { /Verse.org/Colors/NamedColors }
using { Utils }
using { PlayerData }
using { Skill }


# See https://dev.epicgames.com/documentation/en-us/uefn/create-your-own-device-in-verse for how to create a verse device.

# A Verse-authored creative device that can be placed in a level
ui_skill_boss := class(creative_device):
    #SFXs
    @editable
    SFX_Open : audio_player_device = audio_player_device{}
    @editable
    SFX_Close : audio_player_device = audio_player_device{}
    @editable
    SFX_LevelUp : audio_player_device = audio_player_device{}
    @editable
    SFX_Fail : audio_player_device = audio_player_device{}

    var _SaveData : save_data = save_data{}
    @editable
    var SaveDataManager : save_data_manager = save_data_manager{}
    @editable
    var SkillDataManager : skill_manager = skill_manager{}

    CloseEvent<public> : event(player) = event(player){}

    @editable
    SkillNPC : character_device = character_device{}

    var BossSkillCanvasPerAgent : [agent]canvas = map{}

    var UIElements : [agent]tuple(text_block, text_block) = map{}
    #Lv, Now, Will, Cost, _
    var SkillUIElements : [agent][int]tuple(text_block, text_block, text_block, text_block, text_block) = map{}


    OnBegin<override>()<suspends>:void=
        SkillNPC.InteractedWithEvent.Subscribe(OnButtonEvented)

    SetText<localizes>(Value:[]char) : message = "{Value}"
    
    # スキル関連のUI要素を全て更新
    UpdateUI(Agent:agent):void=
        UpdateUIElements(Agent)
        UpdateSkillUIElements(Agent)

    # UIElementsに格納された基本リソースのテキストを更新する
    UpdateUIElements(Agent: agent):void=
        var SilverValue : big_number = big_number{}
        var PowerupStoneValue : big_number = big_number{}
        if:
            Player := player[Agent]
        then:
            set SilverValue = SaveDataManager.GetResourceStone(Player)
            set PowerupStoneValue = SaveDataManager.GetResourcePowerupStone(Player)
            if (Elements := UIElements[Agent]):
                Elements(0).SetText(SetText("{BigNumber_FormatDisplay(SilverValue)}"))          # OreTextBlockのテキストを設定
                Elements(1).SetText(SetText("{BigNumber_FormatDisplay(PowerupStoneValue)}"))    # PowerupStoneTextBlockのテキストを設定 

    UpdateSkillUIElements(Agent:agent):void=
        var Lv_AddStock_Boss : int = 1
        var Now_AddStock_Boss : int = 1
        var Will_AddStock_Boss : int = 1
        var Cost_AddStock_Boss : big_number = big_number{}
        var Lv_HP : int = 1
        var Now_HP : int = 1
        var Will_HP : int = 1
        var Cost_HP : big_number = big_number{}
        var Lv_Support : int = 1
        var Now_Support : int = 1
        var Will_Support : int = 1
        var Cost_Support : big_number = big_number{}
        if (Player := player[Agent]):
            var LanguageIndex : int = SaveDataManager.GetLanguage(Player)
            if:
                set Lv_AddStock_Boss = SaveDataManager.GetSkillLevel[Player, 12]
                set Now_AddStock_Boss = SaveDataManager.GetPlayerSkillEffectIncrease[Player, 12]
                set Will_AddStock_Boss = SaveDataManager.GetSpecificSkillEffectIncrease[Player, 12]
                set Cost_AddStock_Boss = SaveDataManager.GetUpgradeCostForSkill[Player, 12]

                set Lv_HP = SaveDataManager.GetSkillLevel[Player, 13]
                set Now_HP = SaveDataManager.GetPlayerSkillEffectIncrease[Player, 13]
                set Will_HP = SaveDataManager.GetSpecificSkillEffectIncrease[Player, 13]
                set Cost_HP = SaveDataManager.GetUpgradeCostForSkill[Player, 13]

                set Lv_Support = SaveDataManager.GetSkillLevel[Player, 14]
                set Now_Support = SaveDataManager.GetPlayerSkillEffectIncrease[Player, 14]
                set Will_Support = SaveDataManager.GetSpecificSkillEffectIncrease[Player, 14]
                set Cost_Support = SaveDataManager.GetUpgradeCostForSkill[Player, 14]
            then:
                LvAddStockText := if(SkillDataManager.CanUpgradeSkill(player[Agent], 12)?){"Lv.{Lv_AddStock_Boss}"} else {"Lv.MAX"}
                LvAddStockTextColor := if(SkillDataManager.CanUpgradeSkill(player[Agent], 12)?){NamedColors.White} else {NamedColors.Red}
                WillAddStockText := if(SkillDataManager.CanUpgradeSkill(player[Agent], 12)?){"(+{Will_AddStock_Boss})"} else {"(MAX!)"}
                CostAddStockText := if(SkillDataManager.CanUpgradeSkill(player[Agent], 12)?){"{BigNumber_FormatDisplay(Cost_AddStock_Boss)}"} else {"-"}
                CostAddStockTextColor := if(SkillDataManager.IsUnlockable[player[Agent], 12]){NamedColors.White} else {NamedColors.Red}
                AddStockText := if(LanguageIndex = 0){"ボスに挑戦する為の\nストックの上限増加"} else {"Boss challenge\nstock increase"}

                LvPlayerHPText := if(SkillDataManager.CanUpgradeSkill(player[Agent], 13)?){"Lv.{Lv_HP}"} else {"Lv.MAX"}
                LvPlayerHPTextColor := if(SkillDataManager.CanUpgradeSkill(player[Agent], 13)?){NamedColors.White} else {NamedColors.Red}
                WillHPText := if(SkillDataManager.CanUpgradeSkill(player[Agent], 13)?){"(+{Will_HP}%)"} else {"(MAX!)"}
                CostHPText := if(SkillDataManager.CanUpgradeSkill(player[Agent], 13)?){"{BigNumber_FormatDisplay(Cost_HP)}"} else {"-"}
                CostHPTextColor := if(SkillDataManager.IsUnlockable[player[Agent], 13]){NamedColors.White} else {NamedColors.Red}
                HPText := if(LanguageIndex = 0){"ボス戦中HP増加"} else {"HP increase"}

                LvPlayerSupportText := if(SkillDataManager.CanUpgradeSkill(player[Agent], 14)?){"Lv.{Lv_Support}"} else {"Lv.MAX"}
                LvPlayerSupportTextColor := if(SkillDataManager.CanUpgradeSkill(player[Agent], 14)?){NamedColors.White} else {NamedColors.Red}
                WillSupportText := if(SkillDataManager.CanUpgradeSkill(player[Agent], 14)?){"(+{Will_Support})"} else {"(MAX!)"}
                CostSupportText := if(SkillDataManager.CanUpgradeSkill(player[Agent], 14)?){"{BigNumber_FormatDisplay(Cost_Support)}"} else {"-"}
                CostSupportTextColor := if(SkillDataManager.IsUnlockable[player[Agent], 14]){NamedColors.White} else {NamedColors.Red}
                SupportText := if(LanguageIndex = 0){"ボス戦中、先頭スロットの\n進化済み妖怪から加護を受ける"} else {"Receive the blessingYo-kai\nin the first slot"}

                if(SkillElement := SkillUIElements[Agent][1]) :
                    SkillElement(0).SetText(SetText(LvAddStockText))
                    SkillElement(0).SetTextColor(LvAddStockTextColor)
                    SkillElement(1).SetText(SetText("{Now_AddStock_Boss}"))
                    SkillElement(2).SetText(SetText(WillAddStockText))
                    SkillElement(3).SetText(SetText(CostAddStockText))
                    SkillElement(3).SetTextColor(CostAddStockTextColor)
                    SkillElement(4).SetText(SetText("{AddStockText}"))

                if(SkillElement := SkillUIElements[Agent][2]) :
                    SkillElement(0).SetText(SetText(LvPlayerHPText))
                    SkillElement(0).SetTextColor(LvPlayerHPTextColor)
                    SkillElement(1).SetText(SetText("{Now_HP}"))
                    SkillElement(2).SetText(SetText(WillHPText))
                    SkillElement(3).SetText(SetText(CostHPText))
                    SkillElement(3).SetTextColor(CostHPTextColor)
                    SkillElement(4).SetText(SetText("{HPText}"))

                if(SkillElement := SkillUIElements[Agent][3]) :
                    SkillElement(0).SetText(SetText(LvPlayerSupportText))
                    SkillElement(0).SetTextColor(LvPlayerSupportTextColor)
                    SkillElement(1).SetText(SetText("{Now_Support}"))
                    SkillElement(2).SetText(SetText(WillSupportText))
                    SkillElement(3).SetText(SetText(CostSupportText))
                    SkillElement(3).SetTextColor(CostSupportTextColor)
                    SkillElement(4).SetText(SetText("{SupportText}"))

    OnPlayerSpawned(Agent : agent):void=
        if (BossSkillCanvas := BossSkillCanvasPerAgent[Agent]):
            return

        if (Player := player[Agent], PlayerUI := GetPlayerUI[Player]):
            BossSkillCanvas : canvas = CreateBossSkillCanvas(Agent, Player)
            if (set BossSkillCanvasPerAgent[Agent] = BossSkillCanvas){}

    OnButtonEvented(Agent : agent):void=
        SFX_Open.Play(Agent)
        OnPlayerSpawned(Agent)
        UpdateUI(Agent)
        if (Player := player[Agent], PlayerUI := GetPlayerUI[Player], BossSkillCanvas := BossSkillCanvasPerAgent[Agent]):
            PlayerUI.AddWidget(BossSkillCanvas, player_ui_slot{InputMode := ui_input_mode.All})
            
    # ボタン処理
    OnLevelup1Clicked(WidgetMessage : widget_message):void=
        if (PlayerUI := GetPlayerUI[WidgetMessage.Player], BossSkillCanvas := BossSkillCanvasPerAgent[WidgetMessage.Player], Agent : agent = agent[WidgetMessage.Player]):
            if(SkillDataManager.CanUpgradeSkill(WidgetMessage.Player,12) = true):
                SFX_LevelUp.Play(Agent)
                if:
                    SkillDataManager.UnlockOrUpgradeSkill[WidgetMessage.Player,12]
                then:
                    UpdateUI(Agent)
                    SaveDataManager.CheckRecoveryBossTicket(WidgetMessage.Player)
            else:
                SFX_Fail.Play(Agent)
                return
            
    OnLevelup2Clicked(WidgetMessage : widget_message):void=
        if (PlayerUI := GetPlayerUI[WidgetMessage.Player], BossSkillCanvas := BossSkillCanvasPerAgent[WidgetMessage.Player], Agent : agent = agent[WidgetMessage.Player]):
            if(SkillDataManager.CanUpgradeSkill(WidgetMessage.Player,13) = true):
                SFX_LevelUp.Play(Agent)
                if:
                    SkillDataManager.UnlockOrUpgradeSkill[WidgetMessage.Player,13]
                then:
                    UpdateUI(Agent)
            else:
                SFX_Fail.Play(Agent)
                return

    OnLevelup4Clicked(WidgetMessage : widget_message):void=
        if (PlayerUI := GetPlayerUI[WidgetMessage.Player], BossSkillCanvas := BossSkillCanvasPerAgent[WidgetMessage.Player], Agent : agent = agent[WidgetMessage.Player]):
            if(SkillDataManager.CanUpgradeSkill(WidgetMessage.Player,14) = true):
                SFX_LevelUp.Play(Agent)
                if:
                    SkillDataManager.UnlockOrUpgradeSkill[WidgetMessage.Player,14]
                then:
                    UpdateUI(Agent)
            else:
                SFX_Fail.Play(Agent)
                return

    OnCloseButtonClicked(WidgetMessage : widget_message):void=
        if(Agent : agent = agent[WidgetMessage.Player]) :
            SFX_Close.Play(Agent)
        if (PlayerUI : player_ui = GetPlayerUI[WidgetMessage.Player], BossSkillCanvas := BossSkillCanvasPerAgent[WidgetMessage.Player]):
            PlayerUI.RemoveWidget(BossSkillCanvas)
            CloseEvent.Signal(WidgetMessage.Player)



    # キャンバス
    CreateBossSkillCanvas<public>(Agent : agent, Player : player):canvas=

        var CanvasSlots : []canvas_slot = array{}
        
        CenterVector : vector2 = vector2{X:=0.5, Y:=0.5}
        CenterAnchors : anchors = anchors{Minimum := CenterVector, Maximum := CenterVector}

        # Image Stock
        BGImage : texture_block = texture_block{DefaultImage := Textures.BG_Skill, DefaultDesiredSize := vector2{X := 1920.0, Y := 1080.0}}
        SilverOreImage1 : texture_block = texture_block{DefaultImage := Textures.SilverOre, DefaultDesiredSize := vector2{X := 50.0, Y := 50.0}}
        PowerupStoneImage : texture_block = texture_block{DefaultImage := Textures.IC_PowerupStone, DefaultDesiredSize := vector2{X := 50.0, Y := 50.0}}
        SiLverOreImage3 : texture_block = texture_block{DefaultImage := Textures.SilverOre, DefaultDesiredSize := vector2{X := 40.0, Y := 40.0}}
        SiLverOreImage4 : texture_block = texture_block{DefaultImage := Textures.SilverOre, DefaultDesiredSize := vector2{X := 40.0, Y := 40.0}}
        SiLverOreImage5 : texture_block = texture_block{DefaultImage := Textures.SilverOre, DefaultDesiredSize := vector2{X := 40.0, Y := 40.0}}
        SiLverOreImage6 : texture_block = texture_block{DefaultImage := Textures.SilverOre, DefaultDesiredSize := vector2{X := 40.0, Y := 40.0}}
        AddStock_BossImage : texture_block = texture_block{DefaultImage := Textures.Boss_1, DefaultDesiredSize := vector2{X := 200.0, Y := 200.0}}
        HPImage : texture_block = texture_block{DefaultImage := Textures.Boss_2, DefaultDesiredSize := vector2{X := 200.0, Y := 200.0}}
        DoubleRewardsImage : texture_block = texture_block{DefaultImage := Textures.Boss_3, DefaultDesiredSize := vector2{X := 200.0, Y := 200.0}}
        SupportImage : texture_block = texture_block{DefaultImage := Textures.Boss_4, DefaultDesiredSize := vector2{X := 200.0, Y := 200.0}}
        
        var SilverValue : big_number = big_number{}
        var PowerupStoneValue : big_number = big_number{}
        var Lv_AddStock_Boss : int = 1
        var Now_AddStock_Boss : int = 1
        var Will_AddStock_Boss : int = 1
        var Cost_AddStock_Boss : big_number = big_number{}
        var Lv_HP : int = 1
        var Now_HP : int = 1
        var Will_HP : int = 1
        var Cost_HP : big_number = big_number{}
        var Lv_Support : int = 1
        var Now_Support : int = 1
        var Will_Support : int = 1
        var Cost_Support : big_number = big_number{}
        #================================== TextBlock ==================================
        # 所持岩
        OreTextBlock : text_block = text_block{
            DefaultTextColor := NamedColors.White,
            DefaultShadowOffset := option{vector2{X:=2.5, Y:=2.5}},
            DefaultShadowColor := NamedColors.Black
        }
        # 所持強化アイテム
        PowerupStoneTextBlock : text_block = text_block{
            DefaultTextColor := NamedColors.White,
            DefaultShadowOffset := option{vector2{X:=2.5, Y:=2.5}},
            DefaultShadowColor := NamedColors.Black,
            DefaultJustification := text_justification.Center
        }
        #================================== Skill 1 ==================================
        Lv_AddStock_BossTextBlock : text_block = text_block{
            DefaultTextColor := NamedColors.White,
            DefaultShadowOffset := option{vector2{X:=2.5, Y:=2.5}},
            DefaultShadowColor := NamedColors.Black,
            DefaultJustification := text_justification.Center
        }
        Now_AddStock_BossTextBlock : text_block = text_block{
            DefaultTextColor := NamedColors.White,
            DefaultShadowOffset := option{vector2{X:=2.5, Y:=2.5}},
            DefaultShadowColor := NamedColors.Black,
            DefaultJustification := text_justification.Center
        }
        Will_AddStock_BossTextBlock : text_block = text_block{
            DefaultTextColor := NamedColors.White,
            DefaultShadowOffset := option{vector2{X:=2.5, Y:=2.5}},
            DefaultShadowColor := NamedColors.Black,
            DefaultJustification := text_justification.Center
        }
        Cost_AddStock_BossTextBlock : text_block = text_block{
            DefaultTextColor := NamedColors.White,
            DefaultShadowOffset := option{vector2{X:=2.5, Y:=2.5}},
            DefaultShadowColor := NamedColors.Black,
            DefaultJustification := text_justification.Center
        }
        _AddStock_BossTextBlock :text_block= text_block :
            DefaultTextColor := NamedColors.White
            DefaultShadowOffset := option{vector2{X:=2.5, Y:=2.5}}
            DefaultShadowColor := NamedColors.Black
            DefaultShadowOpacity := 1.0
            DefaultJustification := text_justification.Center
        #================================== Skill 2 ==================================
        Lv_HPTextBlock : text_block = text_block{
            DefaultTextColor := NamedColors.White,
            DefaultShadowOffset := option{vector2{X:=2.5, Y:=2.5}},
            DefaultShadowColor := NamedColors.Black,
            DefaultJustification := text_justification.Center
        }
        Now_HPTextBlock : text_block = text_block{
            DefaultTextColor := NamedColors.White,
            DefaultShadowOffset := option{vector2{X:=2.5, Y:=2.5}},
            DefaultShadowColor := NamedColors.Black,
            DefaultJustification := text_justification.Center
        }
        Will_HPTextBlock : text_block = text_block{
            DefaultTextColor := NamedColors.White,
            DefaultShadowOffset := option{vector2{X:=2.5, Y:=2.5}},
            DefaultShadowColor := NamedColors.Black,
            DefaultJustification := text_justification.Center
        }
        Cost_HPTextBlock : text_block = text_block{
            DefaultTextColor := NamedColors.White,
            DefaultShadowOffset := option{vector2{X:=2.5, Y:=2.5}},
            DefaultShadowColor := NamedColors.Black,
            DefaultJustification := text_justification.Center
        }
        _HPTextBlock :text_block= text_block :
            DefaultTextColor := NamedColors.White
            DefaultShadowOffset := option{vector2{X:=2.5, Y:=2.5}}
            DefaultShadowColor := NamedColors.Black
            DefaultShadowOpacity := 1.0
            DefaultJustification := text_justification.Center
        #================================== Skill 3 ==================================
        Lv_SupportTextBlock : text_block = text_block{
            DefaultTextColor := NamedColors.White,
            DefaultShadowOffset := option{vector2{X:=2.5, Y:=2.5}},
            DefaultShadowColor := NamedColors.Black,
            DefaultJustification := text_justification.Center
        }
        Now_SupportTextBlock : text_block = text_block{
            DefaultTextColor := NamedColors.White,
            DefaultShadowOffset := option{vector2{X:=2.5, Y:=2.5}},
            DefaultShadowColor := NamedColors.Black,
            DefaultJustification := text_justification.Center
        }
        Will_SupportTextBlock : text_block = text_block{
            DefaultTextColor := NamedColors.White,
            DefaultShadowOffset := option{vector2{X:=2.5, Y:=2.5}},
            DefaultShadowColor := NamedColors.Black,
            DefaultJustification := text_justification.Center
        }
        Cost_SupportTextBlock : text_block = text_block{
            DefaultTextColor := NamedColors.White,
            DefaultShadowOffset := option{vector2{X:=2.5, Y:=2.5}},
            DefaultShadowColor := NamedColors.Black,
            DefaultJustification := text_justification.Center
        }
        _SupportTextBlock :text_block= text_block :
            DefaultTextColor := NamedColors.White
            DefaultShadowOffset := option{vector2{X:=2.5, Y:=2.5}}
            DefaultShadowColor := NamedColors.Black
            DefaultShadowOpacity := 1.0
            DefaultJustification := text_justification.Center
        #================================== /TextBlock ==================================
        #========================================== BG ==========================================
        set CanvasSlots += array
        {
            canvas_slot:
                Anchors := CenterAnchors
                Offsets := margin{Top := 0.0,Left := 0.0, Right := 1920.0, Bottom := 1080.0}
                Alignment := CenterVector
                SizeToContent := false
                Widget := BGImage
        }
        #========================================= Header ========================================
        BossTextBlock := text_block{
            DefaultText := StringToMessage("Boss"),
            DefaultTextColor := NamedColors.White,
            DefaultShadowOffset := option{vector2{X:=2.5, Y:=2.5}},
            DefaultShadowColor := NamedColors.Black,
            DefaultJustification := text_justification.Center
        }
        BossTextBlock.SetShadowOpacity(1.0)
        set CanvasSlots += array
        {
            canvas_slot:
                Anchors := CenterAnchors
                Offsets := margin{Top:= -472.0, Left:= 0.0, Right := 257.0, Bottom := 41.0}
                Alignment := CenterVector
                SizeToContent := false
                Widget := BossTextBlock
        }

        OreTextBlock.SetShadowOpacity(1.0)
        PowerupStoneTextBlock.SetShadowOpacity(1.0)

        set CanvasSlots += array

        {
            canvas_slot:
                Anchors := CenterAnchors
                Offsets := margin{Left:=-645.5 ,Top:= -391.0, Right:= 519.0, Bottom:= 59.0}
                Alignment := CenterVector
                SizeToContent := false
                Widget := stack_box:
                    Orientation := orientation.Horizontal
                    Slots := array:
                        stack_box_slot:
                            Distribution:= option{0.5}
                            Padding:=margin{Left:=5.0 ,Top:= 5.0, Right:= 10.0, Bottom:= 5.0}
                            HorizontalAlignment := horizontal_alignment.Fill
                            VerticalAlignment := vertical_alignment.Fill
                            Widget := stack_box:
                                Orientation := orientation.Horizontal
                                Slots := array:
                                    stack_box_slot:
                                        HorizontalAlignment := horizontal_alignment.Center
                                        VerticalAlignment := vertical_alignment.Center
                                        Widget := SilverOreImage1
                                    stack_box_slot:
                                        Distribution:= option{1.0}
                                        HorizontalAlignment := horizontal_alignment.Center
                                        VerticalAlignment := vertical_alignment.Center
                                        Widget := OreTextBlock
                        stack_box_slot:
                            Padding:=margin{Left:=10.0 ,Top:= 5.0, Right:= 5.0, Bottom:= 5.0}
                            Distribution:= option{0.5}
                            HorizontalAlignment := horizontal_alignment.Fill
                            VerticalAlignment := vertical_alignment.Fill
                            Widget := stack_box:
                                Orientation := orientation.Horizontal
                                Slots := array:
                                    stack_box_slot:
                                        HorizontalAlignment := horizontal_alignment.Center
                                        VerticalAlignment := vertical_alignment.Center
                                        Widget := PowerupStoneImage
                                    stack_box_slot:
                                        Distribution:= option{1.0}
                                        HorizontalAlignment := horizontal_alignment.Center
                                        VerticalAlignment := vertical_alignment.Center
                                        Widget := PowerupStoneTextBlock
        }
        #========================================== Skill ==========================================
        CostTextBlock1 := text_block{
            DefaultText := StringToMessage("Cost :"),
            DefaultTextColor := NamedColors.White,
            DefaultShadowOffset := option{vector2{X:=2.5, Y:=2.5}},
            DefaultShadowColor := NamedColors.Black,
            DefaultJustification := text_justification.Center
        }
        CostTextBlock1.SetShadowOpacity(1.0)
        CostTextBlock2 := text_block{
            DefaultText := StringToMessage("Cost :"),
            DefaultTextColor := NamedColors.White,
            DefaultShadowOffset := option{vector2{X:=2.5, Y:=2.5}},
            DefaultShadowColor := NamedColors.Black,
            DefaultJustification := text_justification.Center
        }
        CostTextBlock2.SetShadowOpacity(1.0)
        CostTextBlock3 := text_block{
            DefaultText := StringToMessage("Cost :"),
            DefaultTextColor := NamedColors.White,
            DefaultShadowOffset := option{vector2{X:=2.5, Y:=2.5}},
            DefaultShadowColor := NamedColors.Black,
            DefaultJustification := text_justification.Center
        }
        CostTextBlock3.SetShadowOpacity(1.0)
        CostTextBlock4 := text_block{
            DefaultText := StringToMessage("Cost :"),
            DefaultTextColor := NamedColors.White,
            DefaultShadowOffset := option{vector2{X:=2.5, Y:=2.5}},
            DefaultShadowColor := NamedColors.Black,
            DefaultJustification := text_justification.Center
        }
        CostTextBlock4.SetShadowOpacity(1.0)

        #Skill 1
        AddStock_BossTextBlock := text_block{
            DefaultText := StringToMessage("Add stock"),
            DefaultTextColor := NamedColors.White,
            DefaultShadowOffset := option{vector2{X:=2.5, Y:=2.5}},
            DefaultShadowColor := NamedColors.Black,
            DefaultJustification := text_justification.Center
        }
        AddStock_BossTextBlock.SetShadowOpacity(1.0)
        Lv_AddStock_BossTextBlock.SetShadowOpacity(1.0)
        Now_AddStock_BossTextBlock.SetShadowOpacity(1.0)
        Will_AddStock_BossTextBlock.SetShadowOpacity(1.0)
        Cost_AddStock_BossTextBlock.SetShadowOpacity(1.0)
        
        #Skill 2
        HPTextBlock := text_block{
            DefaultText := StringToMessage("HP"),
            DefaultTextColor := NamedColors.White,
            DefaultShadowOffset := option{vector2{X:=2.5, Y:=2.5}},
            DefaultShadowColor := NamedColors.Black,
            DefaultJustification := text_justification.Center
        }
        HPTextBlock.SetShadowOpacity(1.0)
        Lv_HPTextBlock.SetShadowOpacity(1.0)
        Now_HPTextBlock.SetShadowOpacity(1.0)
        Will_HPTextBlock.SetShadowOpacity(1.0)
        Cost_HPTextBlock.SetShadowOpacity(1.0)

        #Skill 4
        SupportTextBlock := text_block{
            DefaultText := StringToMessage("Yo-kai support"),
            DefaultTextColor := NamedColors.White,
            DefaultShadowOffset := option{vector2{X:=2.5, Y:=2.5}},
            DefaultShadowColor := NamedColors.Black,
            DefaultJustification := text_justification.Center
        }
        SupportTextBlock.SetShadowOpacity(1.0)
        Lv_SupportTextBlock.SetShadowOpacity(1.0)
        Now_SupportTextBlock.SetShadowOpacity(1.0)
        Will_SupportTextBlock.SetShadowOpacity(1.0)
        Cost_SupportTextBlock.SetShadowOpacity(1.0)

        # Button Stock
        QuietButtonLevelup1 : button_quiet = button_quiet{DefaultText := StringToMessage("LEVEL UP")}
        QuietButtonLevelup1.OnClick().Subscribe(OnLevelup1Clicked)
        QuietButtonLevelup2 : button_quiet = button_quiet{DefaultText := StringToMessage("LEVEL UP")}
        QuietButtonLevelup2.OnClick().Subscribe(OnLevelup2Clicked)
        QuietButtonLevelup4 : button_quiet = button_quiet{DefaultText := StringToMessage("LEVEL UP")}
        QuietButtonLevelup4.OnClick().Subscribe(OnLevelup4Clicked)
        #================================ Top ================================
        set CanvasSlots += array
        {
            canvas_slot:
                Anchors := CenterAnchors
                Offsets := margin{Left:=0.0 ,Top:= -96.5, Right:= 1760.0, Bottom:= 464.0}
                Alignment := CenterVector
                SizeToContent := false
                Widget := stack_box:
                    Orientation := orientation.Horizontal
                    Slots := array:
                        stack_box_slot:
                            Padding:=margin{Left:= 50.0, Right:= 50.0}
                            Distribution:= option{0.25}
                            HorizontalAlignment := horizontal_alignment.Fill
                            VerticalAlignment := vertical_alignment.Fill
                            Widget := stack_box:
                                Orientation := orientation.Vertical
                                Slots := array:
                                    stack_box_slot:
                                        Padding:=margin{Top:= 10.0, Bottom:= 10.0}
                                        HorizontalAlignment := horizontal_alignment.Fill
                                        VerticalAlignment := vertical_alignment.Fill
                                        Widget := AddStock_BossTextBlock
                                    stack_box_slot:
                                        Padding:=margin{Top:= 15.0, Bottom:= 15.0}
                                        HorizontalAlignment := horizontal_alignment.Fill
                                        VerticalAlignment := vertical_alignment.Fill
                                        Widget := Lv_AddStock_BossTextBlock
                                    stack_box_slot:
                                        Padding:=margin{Top:= 10.0, Bottom:= 10.0}
                                        HorizontalAlignment := horizontal_alignment.Center
                                        VerticalAlignment := vertical_alignment.Center
                                        Widget := AddStock_BossImage
                                    stack_box_slot:
                                        Padding:=margin{Top:= 15.0, Bottom:= 15.0}
                                        HorizontalAlignment := horizontal_alignment.Fill
                                        VerticalAlignment := vertical_alignment.Fill
                                        Widget := _AddStock_BossTextBlock
                        stack_box_slot:
                            Padding:=margin{Left:= 50.0, Right:= 50.0}
                            Distribution:= option{0.25}
                            HorizontalAlignment := horizontal_alignment.Fill
                            VerticalAlignment := vertical_alignment.Fill
                            Widget := stack_box:
                                Orientation := orientation.Vertical
                                Slots := array:
                                    stack_box_slot:
                                        Padding:=margin{Top:= 10.0, Bottom:= 10.0}
                                        HorizontalAlignment := horizontal_alignment.Fill
                                        VerticalAlignment := vertical_alignment.Fill
                                        Widget := HPTextBlock
                                    stack_box_slot:
                                        Padding:=margin{Top:= 15.0, Bottom:= 15.0}
                                        HorizontalAlignment := horizontal_alignment.Fill
                                        VerticalAlignment := vertical_alignment.Fill
                                        Widget := Lv_HPTextBlock
                                    stack_box_slot:
                                        Padding:=margin{Top:= 10.0, Bottom:= 10.0}
                                        HorizontalAlignment := horizontal_alignment.Center
                                        VerticalAlignment := vertical_alignment.Center
                                        Widget := HPImage
                                    stack_box_slot:
                                        Padding:=margin{Top:= 15.0, Bottom:= 15.0}
                                        HorizontalAlignment := horizontal_alignment.Fill
                                        VerticalAlignment := vertical_alignment.Fill
                                        Widget := _HPTextBlock
                        stack_box_slot:
                            Padding:=margin{Left:= 50.0, Right:= 50.0}
                            Distribution:= option{0.25}
                            HorizontalAlignment := horizontal_alignment.Fill
                            VerticalAlignment := vertical_alignment.Fill
                            Widget := stack_box:
                                Orientation := orientation.Vertical
                                Slots := array:
                                    stack_box_slot:
                                        Padding:=margin{Top:= 10.0, Bottom:= 10.0}
                                        HorizontalAlignment := horizontal_alignment.Fill
                                        VerticalAlignment := vertical_alignment.Fill
                                        Widget := SupportTextBlock
                                    stack_box_slot:
                                        Padding:=margin{Top:= 15.0, Bottom:= 15.0}
                                        HorizontalAlignment := horizontal_alignment.Fill
                                        VerticalAlignment := vertical_alignment.Fill
                                        Widget := Lv_SupportTextBlock
                                    stack_box_slot:
                                        Padding:=margin{Top:= 10.0, Bottom:= 10.0}
                                        HorizontalAlignment := horizontal_alignment.Center
                                        VerticalAlignment := vertical_alignment.Center
                                        Widget := SupportImage
                                    stack_box_slot:
                                        Padding:=margin{Top:= 15.0, Bottom:= 15.0}
                                        HorizontalAlignment := horizontal_alignment.Fill
                                        VerticalAlignment := vertical_alignment.Fill
                                        Widget := _SupportTextBlock
        }
        #================================ Bottom ================================
        set CanvasSlots += array
        {
            canvas_slot:
                Anchors := CenterAnchors
                Offsets := margin{Left:=0.0 ,Top:= 249.5, Right:= 1760.0, Bottom:= 210.0}
                Alignment := CenterVector
                SizeToContent := false
                Widget := stack_box:
                    Orientation := orientation.Horizontal
                    Slots := array:
                        stack_box_slot:
                            Distribution:= option{0.25}
                            HorizontalAlignment := horizontal_alignment.Fill
                            VerticalAlignment := vertical_alignment.Fill
                            Widget := stack_box:
                                Orientation := orientation.Vertical
                                Slots := array:
                                    stack_box_slot:
                                        Padding:=margin{Top:= 10.0, Bottom:= 10.0}
                                        HorizontalAlignment := horizontal_alignment.Fill
                                        VerticalAlignment := vertical_alignment.Fill
                                        Widget := stack_box:
                                            Orientation := orientation.Horizontal
                                            Slots := array:
                                                stack_box_slot:
                                                    Padding:=margin{Top:= 10.0, Right:= 5.0, Bottom:= 10.0}
                                                    Distribution:= option{0.5}
                                                    HorizontalAlignment := horizontal_alignment.Right
                                                    VerticalAlignment := vertical_alignment.Fill
                                                    Widget := Now_AddStock_BossTextBlock
                                                stack_box_slot:
                                                    Padding:=margin{Top:= 10.0, Left:= 5.0, Bottom:= 10.0}
                                                    Distribution:= option{0.5}
                                                    HorizontalAlignment := horizontal_alignment.Left
                                                    VerticalAlignment := vertical_alignment.Fill
                                                    Widget := Will_AddStock_BossTextBlock
                                    stack_box_slot:
                                        Padding:=margin{Top:= 10.0, Bottom:= 10.0}
                                        HorizontalAlignment := horizontal_alignment.Fill
                                        VerticalAlignment := vertical_alignment.Fill
                                        Widget := stack_box:
                                            Orientation := orientation.Horizontal
                                            Slots := array:
                                                stack_box_slot:
                                                    Distribution:= option{0.5}
                                                    HorizontalAlignment := horizontal_alignment.Right
                                                    VerticalAlignment := vertical_alignment.Fill
                                                    Widget := CostTextBlock1
                                                stack_box_slot:
                                                    Distribution:= option{0.1}
                                                    HorizontalAlignment := horizontal_alignment.Center
                                                    VerticalAlignment := vertical_alignment.Fill
                                                    Widget := SiLverOreImage3
                                                stack_box_slot:
                                                    Distribution:= option{0.4}
                                                    HorizontalAlignment := horizontal_alignment.Left
                                                    VerticalAlignment := vertical_alignment.Fill
                                                    Widget := Cost_AddStock_BossTextBlock
                                    stack_box_slot:
                                        Distribution:= option{0.2}
                                        HorizontalAlignment := horizontal_alignment.Center
                                        VerticalAlignment := vertical_alignment.Center
                                        Widget := QuietButtonLevelup1
                        stack_box_slot:
                            Distribution:= option{0.25}
                            HorizontalAlignment := horizontal_alignment.Fill
                            VerticalAlignment := vertical_alignment.Fill
                            Widget := stack_box:
                                Orientation := orientation.Vertical
                                Slots := array:
                                    stack_box_slot:
                                        Padding:=margin{Top:= 10.0, Bottom:= 10.0}
                                        HorizontalAlignment := horizontal_alignment.Fill
                                        VerticalAlignment := vertical_alignment.Fill
                                        Widget := stack_box:
                                            Orientation := orientation.Horizontal
                                            Slots := array:
                                                stack_box_slot:
                                                    Padding:=margin{Top:= 10.0, Right:= 5.0, Bottom:= 10.0}
                                                    Distribution:= option{0.5}
                                                    HorizontalAlignment := horizontal_alignment.Right
                                                    VerticalAlignment := vertical_alignment.Fill
                                                    Widget := Now_HPTextBlock
                                                stack_box_slot:
                                                    Padding:=margin{Top:= 10.0, Left:= 5.0, Bottom:= 10.0}
                                                    Distribution:= option{0.5}
                                                    HorizontalAlignment := horizontal_alignment.Left
                                                    VerticalAlignment := vertical_alignment.Fill
                                                    Widget := Will_HPTextBlock
                                    stack_box_slot:
                                        Padding:=margin{Top:= 10.0, Bottom:= 10.0}
                                        HorizontalAlignment := horizontal_alignment.Fill
                                        VerticalAlignment := vertical_alignment.Fill
                                        Widget := stack_box:
                                            Orientation := orientation.Horizontal
                                            Slots := array:
                                                stack_box_slot:
                                                    Distribution:= option{0.5}
                                                    HorizontalAlignment := horizontal_alignment.Right
                                                    VerticalAlignment := vertical_alignment.Fill
                                                    Widget := CostTextBlock2
                                                stack_box_slot:
                                                    Distribution:= option{0.1}
                                                    HorizontalAlignment := horizontal_alignment.Center
                                                    VerticalAlignment := vertical_alignment.Fill
                                                    Widget := SiLverOreImage4
                                                stack_box_slot:
                                                    Distribution:= option{0.4}
                                                    HorizontalAlignment := horizontal_alignment.Left
                                                    VerticalAlignment := vertical_alignment.Fill
                                                    Widget := Cost_HPTextBlock
                                    stack_box_slot:
                                        Distribution:= option{0.2}
                                        HorizontalAlignment := horizontal_alignment.Center
                                        VerticalAlignment := vertical_alignment.Center
                                        Widget := QuietButtonLevelup2
                        stack_box_slot:
                            Distribution:= option{0.25}
                            HorizontalAlignment := horizontal_alignment.Fill
                            VerticalAlignment := vertical_alignment.Fill
                            Widget := stack_box:
                                Orientation := orientation.Vertical
                                Slots := array:
                                    stack_box_slot:
                                        Padding:=margin{Top:= 10.0, Bottom:= 10.0}
                                        HorizontalAlignment := horizontal_alignment.Fill
                                        VerticalAlignment := vertical_alignment.Fill
                                        Widget := stack_box:
                                            Orientation := orientation.Horizontal
                                            Slots := array:
                                                stack_box_slot:
                                                    Padding:=margin{Top:= 10.0, Right:= 5.0, Bottom:= 10.0}
                                                    Distribution:= option{0.5}
                                                    HorizontalAlignment := horizontal_alignment.Right
                                                    VerticalAlignment := vertical_alignment.Fill
                                                    Widget := Now_SupportTextBlock
                                                stack_box_slot:
                                                    Padding:=margin{Top:= 10.0, Left:= 5.0, Bottom:= 10.0}
                                                    Distribution:= option{0.5}
                                                    HorizontalAlignment := horizontal_alignment.Left
                                                    VerticalAlignment := vertical_alignment.Fill
                                                    Widget := Will_SupportTextBlock
                                    stack_box_slot:
                                        Padding:=margin{Top:= 10.0, Bottom:= 10.0}
                                        HorizontalAlignment := horizontal_alignment.Fill
                                        VerticalAlignment := vertical_alignment.Fill
                                        Widget := stack_box:
                                            Orientation := orientation.Horizontal
                                            Slots := array:
                                                stack_box_slot:
                                                    Distribution:= option{0.5}
                                                    HorizontalAlignment := horizontal_alignment.Right
                                                    VerticalAlignment := vertical_alignment.Fill
                                                    Widget := CostTextBlock4
                                                stack_box_slot:
                                                    Distribution:= option{0.1}
                                                    HorizontalAlignment := horizontal_alignment.Center
                                                    VerticalAlignment := vertical_alignment.Fill
                                                    Widget := SiLverOreImage6
                                                stack_box_slot:
                                                    Distribution:= option{0.4}
                                                    HorizontalAlignment := horizontal_alignment.Left
                                                    VerticalAlignment := vertical_alignment.Fill
                                                    Widget := Cost_SupportTextBlock
                                    stack_box_slot:
                                        Distribution:= option{0.2}
                                        HorizontalAlignment := horizontal_alignment.Center
                                        VerticalAlignment := vertical_alignment.Center
                                        Widget := QuietButtonLevelup4
        }
        #========================================== Close ==========================================
        CloseButton : button_loud = button_loud{DefaultText := StringToMessage("CLOSE")}
        CloseButton.OnClick().Subscribe(OnCloseButtonClicked)
        set CanvasSlots += array
        {
            canvas_slot:
                Anchors := CenterAnchors
                Offsets := margin{Top:= 447.0, Left:= 0.0, Right := 170.0, Bottom := 80.0}
                Alignment := CenterVector
                SizeToContent := false
                Widget := CloseButton
        }
        
        if(set UIElements[Agent] = (OreTextBlock, PowerupStoneTextBlock)){}
        if(set SkillUIElements[Agent] = map{}){}
        if:
            set SkillUIElements[Agent] = map:
                1 => (Lv_AddStock_BossTextBlock, Now_AddStock_BossTextBlock, Will_AddStock_BossTextBlock, Cost_AddStock_BossTextBlock, _AddStock_BossTextBlock)
                2 => (Lv_HPTextBlock, Now_HPTextBlock, Will_HPTextBlock, Cost_HPTextBlock, _HPTextBlock)
                3 => (Lv_SupportTextBlock, Now_SupportTextBlock, Will_SupportTextBlock, Cost_SupportTextBlock, _SupportTextBlock)

        Canvas : canvas = canvas:
            Slots := CanvasSlots

        return Canvas