
using { /Verse.org/Assets }
using { /Verse.org/Simulation }
using { /Fortnite.com/UI }
using { /Fortnite.com/Devices }
using { /Fortnite.com/Characters }
using { /UnrealEngine.com/Temporary/UI }
using { /UnrealEngine.com/Temporary/SpatialMath }
using { /Verse.org/Simulation }
using { /UnrealEngine.com/Temporary/Diagnostics }
using { /Verse.org/Colors }
using { /Verse.org/Colors/NamedColors }
using { Utils }
using { PlayerData }
using { Skill }

# See https://dev.epicgames.com/documentation/en-us/uefn/create-your-own-device-in-verse for how to create a verse device.

# A Verse-authored creative device that can be placed in a level
ui_skill_buff := class(creative_device):
    #SFXs
    @editable
    SFX_Open : audio_player_device = audio_player_device{}
    @editable
    SFX_Close : audio_player_device = audio_player_device{}
    @editable
    SFX_LevelUp : audio_player_device = audio_player_device{}
    @editable
    SFX_Fail : audio_player_device = audio_player_device{}


    var _SaveData : save_data = save_data{}
    @editable
    var SaveDataManager : save_data_manager = save_data_manager{}
    @editable
    var SkillDataManager : skill_manager = skill_manager{}
    CloseEvent<public> : event(player) = event(player){}

    @editable
    QuestTrigger : trigger_device = trigger_device{}
    @editable
    SkillNPC : character_device = character_device{}

    var BuffCanvasPerAgent : [agent]canvas = map{}

    var UIElements : [agent]tuple(text_block, text_block) = map{}
    #Lv, Now, Will, Cost, _
    var SkillUIElements : [agent][int]tuple(text_block, text_block, text_block, text_block, text_block) = map{}


    OnBegin<override>()<suspends>:void=
        SkillNPC.InteractedWithEvent.Subscribe(OnButtonEvented)

    SetText<localizes>(Value:[]char) : message = "{Value}"

    # スキル関連のUI要素を全て更新
    UpdateUI(Agent:agent):void=
        UpdateUIElements(Agent)
        UpdateSkillUIElements(Agent)

    # UIElementsに格納された基本リソースのテキストを更新する
    UpdateUIElements(Agent: agent):void=
        var SilverValue : big_number = big_number{}
        var PowerupStoneValue : big_number = big_number{}
        if:
            Player := player[Agent]
        then:
            set SilverValue = SaveDataManager.GetResourceStone(Player)
            set PowerupStoneValue = SaveDataManager.GetResourcePowerupStone(Player)
            if (Elements := UIElements[Agent]):
                Elements(0).SetText(SetText("{BigNumber_FormatDisplay(SilverValue)}"))          # OreTextBlockのテキストを設定
                Elements(1).SetText(SetText("{BigNumber_FormatDisplay(PowerupStoneValue)}"))    # PowerupStoneTextBlockのテキストを設定 

    UpdateSkillUIElements(Agent:agent):void=
        var Lv_Unlock : int = 1
        var Now_Unlock : int = 1
        var Will_Unlock : int = 1
        var Cost_Unlock : big_number = big_number{}
        var Lv_BuffEffect : int = 1
        var Now_BuffEffect : int = 1
        var Will_BuffEffect : int = 1
        var Cost_BuffEffect : big_number = big_number{}
        var Lv_BuffTime : int = 1
        var Now_BuffTime : int = 1
        var Will_BuffTime : int = 1
        var Cost_BuffTime : big_number = big_number{}
        var Lv_AddStock_Buff : int = 1
        var Now_AddStock_Buff : int = 1
        var Will_AddStock_Buff : int = 1
        var Cost_AddStock_Buff : big_number = big_number{}
        if(Player := player[Agent]) :
            var LanguageIndex : int = SaveDataManager.GetLanguage(Player)
            if:
                set Lv_Unlock = SaveDataManager.GetSkillLevel[Player, 5]
                set Now_Unlock = SaveDataManager.GetPlayerSkillEffectIncrease[Player, 5]
                set Will_Unlock = SaveDataManager.GetSpecificSkillEffectIncrease[Player, 5]
                set Cost_Unlock = SaveDataManager.GetUpgradeCostForSkill[Player, 5]

                set Lv_BuffEffect = SaveDataManager.GetSkillLevel[Player, 6]
                set Now_BuffEffect = SaveDataManager.GetPlayerSkillEffectIncrease[Player, 6]
                set Will_BuffEffect = SaveDataManager.GetSpecificSkillEffectIncrease[Player, 6]
                set Cost_BuffEffect = SaveDataManager.GetUpgradeCostForSkill[Player, 6]

                set Lv_BuffTime = SaveDataManager.GetSkillLevel[Player, 7]
                set Now_BuffTime = SaveDataManager.GetPlayerSkillEffectIncrease[Player, 7]
                set Will_BuffTime = SaveDataManager.GetSpecificSkillEffectIncrease[Player, 7]
                set Cost_BuffTime = SaveDataManager.GetUpgradeCostForSkill[Player, 7]

                set Lv_AddStock_Buff = SaveDataManager.GetSkillLevel[Player, 8]
                set Now_AddStock_Buff = SaveDataManager.GetPlayerSkillEffectIncrease[Player, 8]
                set Will_AddStock_Buff = SaveDataManager.GetSpecificSkillEffectIncrease[Player, 8]
                set Cost_AddStock_Buff = SaveDataManager.GetUpgradeCostForSkill[Player, 8]
            then:
                LvUnlockText := if(SkillDataManager.CanUpgradeSkill(player[Agent], 5)?){"Lv.{Lv_Unlock}"} else {"Lv.MAX"}
                LvUnlockTextColor := if(SkillDataManager.CanUpgradeSkill(player[Agent], 5)?){NamedColors.White} else {NamedColors.Red}
                WillUnlockText := if(SkillDataManager.CanUpgradeSkill(player[Agent], 5)?){"+ {Will_Unlock}"} else {"MAX!"}
                CostUnlockText := if(SkillDataManager.CanUpgradeSkill(player[Agent], 5)?){"{BigNumber_FormatDisplay(Cost_Unlock)}"} else {"-"}
                CostUnlockTextColor := if(SkillDataManager.IsUnlockable[player[Agent], 5]){NamedColors.White} else {NamedColors.Red}
                BuffText := if(LanguageIndex = 0){"バフ開放"} else {"Buff unlock"}
                
                LvBuffEffectText := if(SkillDataManager.CanUpgradeSkill(player[Agent], 6)?){"Lv.{Lv_BuffEffect}"} else {"Lv.MAX"}
                LvBuffEffectTextColor := if(SkillDataManager.CanUpgradeSkill(player[Agent], 6)?){NamedColors.White} else {NamedColors.Red}
                WillBuffEffectText := if(SkillDataManager.CanUpgradeSkill(player[Agent], 6)?){"+ {Will_BuffEffect}%"} else {"MAX!"}
                CostBuffEffectText := if(SkillDataManager.CanUpgradeSkill(player[Agent], 6)?){"{BigNumber_FormatDisplay(Cost_BuffEffect)}"} else {"-"}
                CostBuffEffectTextColor := if(SkillDataManager.IsUnlockable[player[Agent], 6]){NamedColors.White} else {NamedColors.Red}
                BuffEffectText := if(LanguageIndex = 0){"バフの中の威力上昇"} else {"Buff effect increase"}
                
                LvBuffTimeText := if(SkillDataManager.CanUpgradeSkill(player[Agent], 7)?){"Lv.{Lv_BuffTime}"} else {"Lv.MAX"}
                LvBuffTimeTextColor := if(SkillDataManager.CanUpgradeSkill(player[Agent], 7)?){NamedColors.White} else {NamedColors.Red}
                WillBuffTimeText := if(SkillDataManager.CanUpgradeSkill(player[Agent], 7)?){"+ {Will_BuffTime}sec"} else {"MAX!"}
                CostBuffTimeText := if(SkillDataManager.CanUpgradeSkill(player[Agent], 7)?){"{BigNumber_FormatDisplay(Cost_BuffTime)}"} else {"-"}
                CostBuffTimeTextColor := if(SkillDataManager.IsUnlockable[player[Agent], 7]){NamedColors.White} else {NamedColors.Red}
                BuffTimeText := if(LanguageIndex = 0){"バフの効果時間上昇"} else {"Buff time increase"}
                
                LvAddStock_BuffText := if(SkillDataManager.CanUpgradeSkill(player[Agent], 8)?){"Lv.{Lv_AddStock_Buff}"} else {"Lv.MAX"}
                LvAddStock_BuffTextColor := if(SkillDataManager.CanUpgradeSkill(player[Agent], 8)?){NamedColors.White} else {NamedColors.Red}
                WillAddStock_BuffText := if(SkillDataManager.CanUpgradeSkill(player[Agent], 8)?){"+ {Will_AddStock_Buff}"} else {"MAX!"}
                CostAddStock_BuffText := if(SkillDataManager.CanUpgradeSkill(player[Agent], 8)?){"{BigNumber_FormatDisplay(Cost_AddStock_Buff)}"} else {"-"}
                CostAddStock_BuffTextColor := if(SkillDataManager.IsUnlockable[player[Agent], 8]){NamedColors.White} else {NamedColors.Red}
                AddStockText := if(LanguageIndex = 0){"ミニゲームに参加する為の\nチケットの所持上限増加"} else {"Buff stock increase"}
                
                if(SkillElement := SkillUIElements[Agent][1]) :
                    SkillElement(0).SetText(SetText(LvUnlockText))
                    SkillElement(0).SetTextColor(LvUnlockTextColor)
                    SkillElement(1).SetText(SetText("{Now_Unlock}"))
                    SkillElement(2).SetText(SetText(WillUnlockText))
                    SkillElement(3).SetText(SetText(CostUnlockText))
                    SkillElement(3).SetTextColor(CostUnlockTextColor)
                    SkillElement(4).SetText(SetText("{BuffText}"))

                if(SkillElement := SkillUIElements[Agent][2]) :
                    SkillElement(0).SetText(SetText(LvBuffEffectText))
                    SkillElement(0).SetTextColor(LvBuffEffectTextColor)
                    SkillElement(1).SetText(SetText("{Now_BuffEffect}"))
                    SkillElement(2).SetText(SetText(WillBuffEffectText))
                    SkillElement(3).SetText(SetText(CostBuffEffectText))
                    SkillElement(3).SetTextColor(CostBuffEffectTextColor)
                    SkillElement(4).SetText(SetText("{BuffEffectText}"))

                if(SkillElement := SkillUIElements[Agent][3]) :
                    SkillElement(0).SetText(SetText(LvBuffTimeText))
                    SkillElement(0).SetTextColor(LvBuffTimeTextColor)
                    SkillElement(1).SetText(SetText("{Now_BuffTime}"))
                    SkillElement(2).SetText(SetText(WillBuffTimeText))
                    SkillElement(3).SetText(SetText(CostBuffTimeText))
                    SkillElement(3).SetTextColor(CostBuffTimeTextColor)
                    SkillElement(4).SetText(SetText("{BuffTimeText}"))

                if(SkillElement := SkillUIElements[Agent][4]) :
                    SkillElement(0).SetText(SetText(LvAddStock_BuffText))
                    SkillElement(0).SetTextColor(LvAddStock_BuffTextColor)
                    SkillElement(1).SetText(SetText("{Now_AddStock_Buff}"))
                    SkillElement(2).SetText(SetText(WillAddStock_BuffText))
                    SkillElement(3).SetText(SetText(CostAddStock_BuffText))
                    SkillElement(3).SetTextColor(CostAddStock_BuffTextColor)
                    SkillElement(4).SetText(SetText("{AddStockText}"))

    OnPlayerSpawned(Agent : agent):void=
        if (BuffCanvas := BuffCanvasPerAgent[Agent]):
            return

        if (Player := player[Agent], PlayerUI := GetPlayerUI[Player]):
            BuffCanvas : canvas = CreateBuffCanvas(Agent, Player)
            if (set BuffCanvasPerAgent[Agent] = BuffCanvas){}

    OnButtonEvented(Agent : agent):void=
        SFX_Open.Play(Agent)
        OnPlayerSpawned(Agent)
        UpdateUI(Agent)
        Print("OnSkillNPCEvented")
        if (Player := player[Agent], PlayerUI := GetPlayerUI[Player], BuffCanvas := BuffCanvasPerAgent[Agent]):
            PlayerUI.AddWidget(BuffCanvas, player_ui_slot{InputMode := ui_input_mode.All})

    # ボタン処理
    OnLevelup1Clicked(WidgetMessage : widget_message):void=
        if (PlayerUI := GetPlayerUI[WidgetMessage.Player], BuffCanvas := BuffCanvasPerAgent[WidgetMessage.Player], Agent : agent = agent[WidgetMessage.Player]):
            if(SkillDataManager.CanUpgradeSkill(WidgetMessage.Player,5) = true):
                SFX_LevelUp.Play(Agent)
                if:
                    SkillDataManager.UnlockOrUpgradeSkill[WidgetMessage.Player,5]
                then:
                    QuestTrigger.Trigger(WidgetMessage.Player)
                    SaveDataManager.CheckRecoveryResourceTicket(WidgetMessage.Player)
                    UpdateUI(Agent)
            else:
                SFX_Fail.Play(Agent)
                return
            
    OnLevelup2Clicked(WidgetMessage : widget_message):void=
        if (PlayerUI := GetPlayerUI[WidgetMessage.Player], BuffCanvas := BuffCanvasPerAgent[WidgetMessage.Player], Agent : agent = agent[WidgetMessage.Player]):
            if(SkillDataManager.CanUpgradeSkill(WidgetMessage.Player,6) = true):
                SFX_LevelUp.Play(Agent)
                if:
                    SkillDataManager.UnlockOrUpgradeSkill[WidgetMessage.Player,6]
                then:
                    UpdateUI(Agent)
            else:
                SFX_Fail.Play(Agent)
                return
    OnLevelup3Clicked(WidgetMessage : widget_message):void=
        if (PlayerUI := GetPlayerUI[WidgetMessage.Player], BuffCanvas := BuffCanvasPerAgent[WidgetMessage.Player], Agent : agent = agent[WidgetMessage.Player]):
            if(SkillDataManager.CanUpgradeSkill(WidgetMessage.Player,7) = true):
                SFX_LevelUp.Play(Agent)
                if:
                    SkillDataManager.UnlockOrUpgradeSkill[WidgetMessage.Player,7]
                then:
                    UpdateUI(Agent)
            else:
                SFX_Fail.Play(Agent)
                return
            
    OnLevelup4Clicked(WidgetMessage : widget_message):void=
        if (PlayerUI := GetPlayerUI[WidgetMessage.Player], BuffCanvas := BuffCanvasPerAgent[WidgetMessage.Player], Agent : agent = agent[WidgetMessage.Player]):
            if(SkillDataManager.CanUpgradeSkill(WidgetMessage.Player,8) = true):
                SFX_LevelUp.Play(Agent)
                if:
                    SkillDataManager.UnlockOrUpgradeSkill[WidgetMessage.Player,8]
                then:
                    SaveDataManager.CheckRecoveryResourceTicket(WidgetMessage.Player)
                    UpdateUI(Agent)
            else:
                SFX_Fail.Play(Agent)
                return

    OnCloseButtonClicked(WidgetMessage : widget_message):void=
        if(Agent : agent = agent[WidgetMessage.Player]) :
            SFX_Close.Play(Agent)
        if (PlayerUI := GetPlayerUI[WidgetMessage.Player], BuffCanvas := BuffCanvasPerAgent[WidgetMessage.Player]):
            PlayerUI.RemoveWidget(BuffCanvas)
            CloseEvent.Signal(WidgetMessage.Player)

    # キャンバス
    CreateBuffCanvas(Agent : agent, Player : player):canvas=

        var CanvasSlots : []canvas_slot = array{}
        CenterVector : vector2 = vector2{X:=0.5, Y:=0.5}
        CenterAnchors : anchors = anchors{Minimum := CenterVector, Maximum := CenterVector}

        # Image Stock
        BGImage : texture_block = texture_block{DefaultImage := Textures.BG_Skill, DefaultDesiredSize := vector2{X := 1920.0, Y := 1080.0}}
        SilverOreImage1 : texture_block = texture_block{DefaultImage := Textures.SilverOre, DefaultDesiredSize := vector2{X := 50.0, Y := 50.0}}
        PowerupStoneImage : texture_block = texture_block{DefaultImage := Textures.IC_PowerupStone, DefaultDesiredSize := vector2{X := 50.0, Y := 50.0}}
        SiLverOreImage3 : texture_block = texture_block{DefaultImage := Textures.SilverOre, DefaultDesiredSize := vector2{X := 40.0, Y := 40.0}}
        SiLverOreImage4 : texture_block = texture_block{DefaultImage := Textures.SilverOre, DefaultDesiredSize := vector2{X := 40.0, Y := 40.0}}
        SiLverOreImage5 : texture_block = texture_block{DefaultImage := Textures.SilverOre, DefaultDesiredSize := vector2{X := 40.0, Y := 40.0}}
        SiLverOreImage6 : texture_block = texture_block{DefaultImage := Textures.SilverOre, DefaultDesiredSize := vector2{X := 40.0, Y := 40.0}}
        UnlockImage : texture_block = texture_block{DefaultImage := Textures.BuffUpgrades_1, DefaultDesiredSize := vector2{X := 200.0, Y := 200.0}}
        BuffEffectImage : texture_block = texture_block{DefaultImage := Textures.BuffUpgrades_2, DefaultDesiredSize := vector2{X := 200.0, Y := 200.0}}
        BuffTimeImage : texture_block = texture_block{DefaultImage := Textures.BuffUpgrades_3, DefaultDesiredSize := vector2{X := 200.0, Y := 200.0}}
        AddStock_BuffImage : texture_block = texture_block{DefaultImage := Textures.BuffUpgrades_4, DefaultDesiredSize := vector2{X := 200.0, Y := 200.0}}
            
        var SilverValue : big_number = big_number{}
        var PowerupStoneValue : big_number = big_number{}
        var Lv_Unlock : int = 1
        var Now_Unlock : int = 1
        var Will_Unlock : int = 1
        var Cost_Unlock : big_number = big_number{}
        var Lv_BuffEffect : int = 1
        var Now_BuffEffect : int = 1
        var Will_BuffEffect : int = 1
        var Cost_BuffEffect : big_number = big_number{}
        var Lv_BuffTime : int = 1
        var Now_BuffTime : int = 1
        var Will_BuffTime : int = 1
        var Cost_BuffTime : big_number = big_number{}
        var Lv_AddStock_Buff : int = 1
        var Now_AddStock_Buff : int = 1
        var Will_AddStock_Buff : int = 1
        var Cost_AddStock_Buff : big_number = big_number{}

        
        #================================== TextBlock ==================================
        
        # 所持岩
        OreTextBlock : text_block = text_block{
            # DefaultText := StringToMessage("{BigNumber_FormatDisplay(SilverValue)}"),
            DefaultTextColor := NamedColors.White,
            DefaultShadowOffset := option{vector2{X:=2.5, Y:=2.5}},
            DefaultShadowColor := NamedColors.Black
        }
        # 所持強化アイテム
        PowerupStoneTextBlock : text_block = text_block{
            # DefaultText := StringToMessage("{BigNumber_FormatDisplay(PowerupStoneValue)}"),
            DefaultTextColor := NamedColors.White,
            DefaultShadowOffset := option{vector2{X:=2.5, Y:=2.5}},
            DefaultShadowColor := NamedColors.Black,
            DefaultJustification := text_justification.Center
        }

        #================================== Skill 1 ==================================

        Lv_UnlockTextBlock : text_block = text_block{
            # DefaultText := StringToMessage("{LvUnlockText}"),
            DefaultTextColor := NamedColors.White,
            DefaultShadowOffset := option{vector2{X:=2.5, Y:=2.5}},
            DefaultShadowColor := NamedColors.Black,
            DefaultJustification := text_justification.Center
        }
        Now_UnlockTextBlock : text_block = text_block{
            # DefaultText := StringToMessage("{Now_Unlock}"),
            DefaultTextColor := NamedColors.White,
            DefaultShadowOffset := option{vector2{X:=2.5, Y:=2.5}},
            DefaultShadowColor := NamedColors.Black,
            DefaultJustification := text_justification.Center
        }
        Will_UnlockTextBlock : text_block = text_block{
            # DefaultText := StringToMessage("({WillUnlockText})"),
            DefaultTextColor := NamedColors.White,
            DefaultShadowOffset := option{vector2{X:=2.5, Y:=2.5}},
            DefaultShadowColor := NamedColors.Black,
            DefaultJustification := text_justification.Center
        }
        Cost_UnlockTextBlock : text_block = text_block{
            # DefaultText := StringToMessage("{CostUnlockText}"),
            DefaultTextColor := NamedColors.White,
            DefaultShadowOffset := option{vector2{X:=2.5, Y:=2.5}},
            DefaultShadowColor := NamedColors.Black,
            DefaultJustification := text_justification.Center
        }
        var _UnlockTextBlock :text_block= text_block :
            DefaultTextColor := NamedColors.White
            DefaultShadowOffset := option{vector2{X:=2.5, Y:=2.5}}
            DefaultShadowColor := NamedColors.Black
            DefaultShadowOpacity := 1.0
            DefaultJustification := text_justification.Center

        #================================== Skill 2 ==================================

        Lv_BuffEffectTextBlock : text_block = text_block{
            # DefaultText := StringToMessage("{LvBuffEffectText}"),
            DefaultTextColor := NamedColors.White,
            DefaultShadowOffset := option{vector2{X:=2.5, Y:=2.5}},
            DefaultShadowColor := NamedColors.Black,
            DefaultJustification := text_justification.Center
        }
        Now_BuffEffectTextBlock : text_block = text_block{
            # DefaultText := StringToMessage("{Now_BuffEffect}%"),
            DefaultTextColor := NamedColors.White,
            DefaultShadowOffset := option{vector2{X:=2.5, Y:=2.5}},
            DefaultShadowColor := NamedColors.Black,
            DefaultJustification := text_justification.Center
        }
        Will_BuffEffectTextBlock : text_block = text_block{
            # DefaultText := StringToMessage("({WillBuffEffectText})"),
            DefaultTextColor := NamedColors.White,
            DefaultShadowOffset := option{vector2{X:=2.5, Y:=2.5}},
            DefaultShadowColor := NamedColors.Black,
            DefaultJustification := text_justification.Center
        }
        Cost_BuffEffectTextBlock : text_block = text_block{
            # DefaultText := StringToMessage("{CostBuffEffectText}"),
            DefaultTextColor := NamedColors.White,
            DefaultShadowOffset := option{vector2{X:=2.5, Y:=2.5}},
            DefaultShadowColor := NamedColors.Black,
            DefaultJustification := text_justification.Center
        }
        var _BuffEffectTextBlock :text_block= text_block :
            DefaultTextColor := NamedColors.White
            DefaultShadowOffset := option{vector2{X:=2.5, Y:=2.5}}
            DefaultShadowColor := NamedColors.Black
            DefaultShadowOpacity := 1.0
            DefaultJustification := text_justification.Center
        
        #================================== Skill 3 ==================================

        Lv_BuffTimeTextBlock : text_block = text_block{
            # DefaultText := StringToMessage("{LvBuffTimeText}"),
            DefaultTextColor := NamedColors.White,
            DefaultShadowOffset := option{vector2{X:=2.5, Y:=2.5}},
            DefaultShadowColor := NamedColors.Black,
            DefaultJustification := text_justification.Center
        }
        Now_BuffTimeTextBlock : text_block = text_block{
            # DefaultText := StringToMessage("{Now_BuffTime}sec"),
            DefaultTextColor := NamedColors.White,
            DefaultShadowOffset := option{vector2{X:=2.5, Y:=2.5}},
            DefaultShadowColor := NamedColors.Black,
            DefaultJustification := text_justification.Center
        }
        Will_BuffTimeTextBlock : text_block = text_block{
            # DefaultText := StringToMessage("({WillBuffTimeText})"),
            DefaultTextColor := NamedColors.White,
            DefaultShadowOffset := option{vector2{X:=2.5, Y:=2.5}},
            DefaultShadowColor := NamedColors.Black,
            DefaultJustification := text_justification.Center
        }
        Cost_BuffTimeTextBlock : text_block = text_block{
            # DefaultText := StringToMessage("{CostBuffTimeText}"),
            DefaultTextColor := NamedColors.White,
            DefaultShadowOffset := option{vector2{X:=2.5, Y:=2.5}},
            DefaultShadowColor := NamedColors.Black,
            DefaultJustification := text_justification.Center
        }
        var _BuffTimeTextBlock :text_block= text_block :
            DefaultTextColor := NamedColors.White
            DefaultShadowOffset := option{vector2{X:=2.5, Y:=2.5}}
            DefaultShadowColor := NamedColors.Black
            DefaultJustification := text_justification.Center
            DefaultShadowOpacity := 1.0
            DefaultOverflowPolicy := text_overflow_policy.Ellipsis
        #================================== Skill 4 ==================================

        Lv_AddStock_BuffTextBlock : text_block = text_block{
            # DefaultText := StringToMessage("{LvAddStock_BuffText}"),
            DefaultTextColor := NamedColors.White,
            DefaultShadowOffset := option{vector2{X:=2.5, Y:=2.5}},
            DefaultShadowColor := NamedColors.Black,
            DefaultJustification := text_justification.Center
        }
        Now_AddStock_BuffTextBlock : text_block = text_block{
            # DefaultText := StringToMessage("{Now_AddStock_Buff}"),
            DefaultTextColor := NamedColors.White,
            DefaultShadowOffset := option{vector2{X:=2.5, Y:=2.5}},
            DefaultShadowColor := NamedColors.Black,
            DefaultJustification := text_justification.Center
        }
        Will_AddStock_BuffTextBlock : text_block = text_block{
            # DefaultText := StringToMessage("({WillAddStock_BuffText})"),
            DefaultTextColor := NamedColors.White,
            DefaultShadowOffset := option{vector2{X:=2.5, Y:=2.5}},
            DefaultShadowColor := NamedColors.Black,
            DefaultJustification := text_justification.Center
        }
        Cost_AddStock_BuffTextBlock : text_block = text_block{
            # DefaultText := StringToMessage("{CostAddStock_BuffText}"),
            DefaultTextColor := NamedColors.White,
            DefaultShadowOffset := option{vector2{X:=2.5, Y:=2.5}},
            DefaultShadowColor := NamedColors.Black,
            DefaultJustification := text_justification.Center
        }
        var _AddStock_BuffTextBlock :text_block= text_block :
            DefaultTextColor := NamedColors.White
            DefaultShadowOffset := option{vector2{X:=2.5, Y:=2.5}}
            DefaultShadowColor := NamedColors.Black
            DefaultShadowOpacity := 1.0
            DefaultJustification := text_justification.Center


        #================================== /TextBlock ==================================

        
        #========================================== BG ==========================================

        set CanvasSlots += array
        {
            canvas_slot:
                Anchors := CenterAnchors
                Offsets := margin{Top := 0.0,Left := 0.0, Right := 1920.0, Bottom := 1080.0}
                Alignment := CenterVector
                SizeToContent := false
                Widget := BGImage
        }

        #========================================= Header ========================================

        BuffTextBlock := text_block{
            DefaultText := StringToMessage("Buff"),
            DefaultTextColor := NamedColors.White,
            DefaultShadowOffset := option{vector2{X:=2.5, Y:=2.5}},
            DefaultShadowColor := NamedColors.Black,
            DefaultJustification := text_justification.Center
        }
        BuffTextBlock.SetShadowOpacity(1.0)
        set CanvasSlots += array
        {
            canvas_slot:
                Anchors := CenterAnchors
                Offsets := margin{Top:= -472.0, Left:= 0.0, Right := 257.0, Bottom := 41.0}
                Alignment := CenterVector
                SizeToContent := false
                Widget := BuffTextBlock
        }
        set CanvasSlots += array
        {
            canvas_slot:
                Anchors := CenterAnchors
                Offsets := margin{Left:=-645.5 ,Top:= -391.0, Right:= 519.0, Bottom:= 59.0}
                Alignment := CenterVector
                SizeToContent := false
                Widget := stack_box:
                    Orientation := orientation.Horizontal
                    Slots := array:
                        stack_box_slot:
                            Distribution:= option{0.5}
                            Padding:=margin{Left:=5.0 ,Top:= 5.0, Right:= 10.0, Bottom:= 5.0}
                            HorizontalAlignment := horizontal_alignment.Fill
                            VerticalAlignment := vertical_alignment.Fill
                            Widget := stack_box:
                                Orientation := orientation.Horizontal
                                Slots := array:
                                    stack_box_slot:
                                        HorizontalAlignment := horizontal_alignment.Center
                                        VerticalAlignment := vertical_alignment.Center
                                        Widget := SilverOreImage1
                                    stack_box_slot:
                                        Distribution:= option{1.0}
                                        HorizontalAlignment := horizontal_alignment.Center
                                        VerticalAlignment := vertical_alignment.Center
                                        Widget := OreTextBlock
                        stack_box_slot:
                            Padding:=margin{Left:=10.0 ,Top:= 5.0, Right:= 5.0, Bottom:= 5.0}
                            Distribution:= option{0.5}
                            HorizontalAlignment := horizontal_alignment.Fill
                            VerticalAlignment := vertical_alignment.Fill
                            Widget := stack_box:
                                Orientation := orientation.Horizontal
                                Slots := array:
                                    stack_box_slot:
                                        HorizontalAlignment := horizontal_alignment.Center
                                        VerticalAlignment := vertical_alignment.Center
                                        Widget := PowerupStoneImage
                                    stack_box_slot:
                                        Distribution:= option{1.0}
                                        HorizontalAlignment := horizontal_alignment.Center
                                        VerticalAlignment := vertical_alignment.Center
                                        Widget := PowerupStoneTextBlock
        }

        #========================================== Skill ==========================================

        CostTextBlock1 := text_block{
            DefaultText := StringToMessage("Cost :"),
            DefaultTextColor := NamedColors.White,
            DefaultShadowOffset := option{vector2{X:=2.5, Y:=2.5}},
            DefaultShadowColor := NamedColors.Black,
            DefaultJustification := text_justification.Center
        }
        CostTextBlock1.SetShadowOpacity(1.0)
        CostTextBlock2 := text_block{
            DefaultText := StringToMessage("Cost :"),
            DefaultTextColor := NamedColors.White,
            DefaultShadowOffset := option{vector2{X:=2.5, Y:=2.5}},
            DefaultShadowColor := NamedColors.Black,
            DefaultJustification := text_justification.Center
        }
        CostTextBlock2.SetShadowOpacity(1.0)
        CostTextBlock3 := text_block{
            DefaultText := StringToMessage("Cost :"),
            DefaultTextColor := NamedColors.White,
            DefaultShadowOffset := option{vector2{X:=2.5, Y:=2.5}},
            DefaultShadowColor := NamedColors.Black,
            DefaultJustification := text_justification.Center
        }
        CostTextBlock3.SetShadowOpacity(1.0)
        CostTextBlock4 := text_block{
            DefaultText := StringToMessage("Cost :"),
            DefaultTextColor := NamedColors.White,
            DefaultShadowOffset := option{vector2{X:=2.5, Y:=2.5}},
            DefaultShadowColor := NamedColors.Black,
            DefaultJustification := text_justification.Center
        }
        CostTextBlock4.SetShadowOpacity(1.0)

        #Skill 1
        UnlockTextBlock := text_block{
            DefaultText := StringToMessage("Unlock"),
            DefaultTextColor := NamedColors.White,
            DefaultShadowOffset := option{vector2{X:=2.5, Y:=2.5}},
            DefaultShadowColor := NamedColors.Black,
            DefaultJustification := text_justification.Center
        }

        UnlockTextBlock.SetShadowOpacity(1.0)
        Lv_UnlockTextBlock.SetShadowOpacity(1.0)
        Now_UnlockTextBlock.SetShadowOpacity(1.0)
        Will_UnlockTextBlock.SetShadowOpacity(1.0)
        Cost_UnlockTextBlock.SetShadowOpacity(1.0)
        
        #Skill 2
        BuffEffectTextBlock := text_block{
            DefaultText := StringToMessage("Buff effect"),
            DefaultTextColor := NamedColors.White,
            DefaultShadowOffset := option{vector2{X:=2.5, Y:=2.5}},
            DefaultShadowColor := NamedColors.Black,
            DefaultJustification := text_justification.Center
        }

        BuffEffectTextBlock.SetShadowOpacity(1.0)
        Lv_BuffEffectTextBlock.SetShadowOpacity(1.0)
        Now_BuffEffectTextBlock.SetShadowOpacity(1.0)
        Will_BuffEffectTextBlock.SetShadowOpacity(1.0)
        Cost_BuffEffectTextBlock.SetShadowOpacity(1.0)
        
        #Skill 3
        BuffTimeTextBlock := text_block{
            DefaultText := StringToMessage("Buff time"),
            DefaultTextColor := NamedColors.White,
            DefaultShadowOffset := option{vector2{X:=2.5, Y:=2.5}},
            DefaultShadowColor := NamedColors.Black,
            DefaultJustification := text_justification.Center
        }

        BuffTimeTextBlock.SetShadowOpacity(1.0)
        Lv_BuffTimeTextBlock.SetShadowOpacity(1.0)
        Now_BuffTimeTextBlock.SetShadowOpacity(1.0)
        Will_BuffTimeTextBlock.SetShadowOpacity(1.0)
        Cost_BuffTimeTextBlock.SetShadowOpacity(1.0)
        
        #Skill 4
        AddStock_BuffTextBlock := text_block{
            DefaultText := StringToMessage("Add stock"),
            DefaultTextColor := NamedColors.White,
            DefaultShadowOffset := option{vector2{X:=2.5, Y:=2.5}},
            DefaultShadowColor := NamedColors.Black,
            DefaultJustification := text_justification.Center
        }
        AddStock_BuffTextBlock.SetShadowOpacity(1.0)
        Lv_AddStock_BuffTextBlock.SetShadowOpacity(1.0)
        Now_AddStock_BuffTextBlock.SetShadowOpacity(1.0)
        Will_AddStock_BuffTextBlock.SetShadowOpacity(1.0)
        Cost_AddStock_BuffTextBlock.SetShadowOpacity(1.0)

        # Button Stock
        QuietButtonLevelup1 : button_quiet = button_quiet{DefaultText := StringToMessage("LEVEL UP")}
        QuietButtonLevelup1.OnClick().Subscribe(OnLevelup1Clicked)
        QuietButtonLevelup2 : button_quiet = button_quiet{DefaultText := StringToMessage("LEVEL UP")}
        QuietButtonLevelup2.OnClick().Subscribe(OnLevelup2Clicked)
        QuietButtonLevelup3 : button_quiet = button_quiet{DefaultText := StringToMessage("LEVEL UP")}
        QuietButtonLevelup3.OnClick().Subscribe(OnLevelup3Clicked)
        QuietButtonLevelup4 : button_quiet = button_quiet{DefaultText := StringToMessage("LEVEL UP")}
        QuietButtonLevelup4.OnClick().Subscribe(OnLevelup4Clicked)

        #================================ Top ================================

        set CanvasSlots += array
        {
            canvas_slot:
                Anchors := CenterAnchors
                Offsets := margin{Left:=0.0 ,Top:= -96.5, Right:= 1760.0, Bottom:= 464.0}
                Alignment := CenterVector
                SizeToContent := false
                Widget := stack_box:
                    Orientation := orientation.Horizontal
                    Slots := array:
                        stack_box_slot:
                            Padding:=margin{Left:= 50.0, Right:= 50.0}
                            Distribution:= option{0.25}
                            HorizontalAlignment := horizontal_alignment.Fill
                            VerticalAlignment := vertical_alignment.Fill
                            Widget := stack_box:
                                Orientation := orientation.Vertical
                                Slots := array:
                                    stack_box_slot:
                                        Padding:=margin{Top:= 10.0, Bottom:= 10.0}
                                        HorizontalAlignment := horizontal_alignment.Fill
                                        VerticalAlignment := vertical_alignment.Fill
                                        Widget := UnlockTextBlock
                                    stack_box_slot:
                                        Padding:=margin{Top:= 15.0, Bottom:= 15.0}
                                        HorizontalAlignment := horizontal_alignment.Fill
                                        VerticalAlignment := vertical_alignment.Fill
                                        Widget := Lv_UnlockTextBlock
                                    stack_box_slot:
                                        Padding:=margin{Top:= 10.0, Bottom:= 10.0}
                                        HorizontalAlignment := horizontal_alignment.Center
                                        VerticalAlignment := vertical_alignment.Center
                                        Widget := UnlockImage
                                    stack_box_slot:
                                        Padding:=margin{Top:= 15.0, Bottom:= 15.0}
                                        HorizontalAlignment := horizontal_alignment.Fill
                                        VerticalAlignment := vertical_alignment.Fill
                                        Widget := _UnlockTextBlock
                        stack_box_slot:
                            Padding:=margin{Left:= 50.0, Right:= 50.0}
                            Distribution:= option{0.25}
                            HorizontalAlignment := horizontal_alignment.Fill
                            VerticalAlignment := vertical_alignment.Fill
                            Widget := stack_box:
                                Orientation := orientation.Vertical
                                Slots := array:
                                    stack_box_slot:
                                        Padding:=margin{Top:= 10.0, Bottom:= 10.0}
                                        HorizontalAlignment := horizontal_alignment.Fill
                                        VerticalAlignment := vertical_alignment.Fill
                                        Widget := BuffEffectTextBlock
                                    stack_box_slot:
                                        Padding:=margin{Top:= 15.0, Bottom:= 15.0}
                                        HorizontalAlignment := horizontal_alignment.Fill
                                        VerticalAlignment := vertical_alignment.Fill
                                        Widget := Lv_BuffEffectTextBlock
                                    stack_box_slot:
                                        Padding:=margin{Top:= 10.0, Bottom:= 10.0}
                                        HorizontalAlignment := horizontal_alignment.Center
                                        VerticalAlignment := vertical_alignment.Center
                                        Widget := BuffEffectImage
                                    stack_box_slot:
                                        Padding:=margin{Top:= 15.0, Bottom:= 15.0}
                                        HorizontalAlignment := horizontal_alignment.Fill
                                        VerticalAlignment := vertical_alignment.Fill
                                        Widget := _BuffEffectTextBlock
                        stack_box_slot:
                            Padding:=margin{Left:= 50.0, Right:= 50.0}
                            Distribution:= option{0.25}
                            HorizontalAlignment := horizontal_alignment.Fill
                            VerticalAlignment := vertical_alignment.Fill
                            Widget := stack_box:
                                Orientation := orientation.Vertical
                                Slots := array:
                                    stack_box_slot:
                                        Padding:=margin{Top:= 10.0, Bottom:= 10.0}
                                        HorizontalAlignment := horizontal_alignment.Fill
                                        VerticalAlignment := vertical_alignment.Fill
                                        Widget := BuffTimeTextBlock
                                    stack_box_slot:
                                        Padding:=margin{Top:= 15.0, Bottom:= 15.0}
                                        HorizontalAlignment := horizontal_alignment.Fill
                                        VerticalAlignment := vertical_alignment.Fill
                                        Widget := Lv_BuffTimeTextBlock
                                    stack_box_slot:
                                        Padding:=margin{Top:= 10.0, Bottom:= 10.0}
                                        HorizontalAlignment := horizontal_alignment.Center
                                        VerticalAlignment := vertical_alignment.Center
                                        Widget := BuffTimeImage
                                    stack_box_slot:
                                        Padding:=margin{Top:= 15.0, Bottom:= 15.0}
                                        HorizontalAlignment := horizontal_alignment.Fill
                                        VerticalAlignment := vertical_alignment.Fill
                                        Widget := _BuffTimeTextBlock
                        stack_box_slot:
                            Padding:=margin{Left:= 50.0, Right:= 50.0}
                            Distribution:= option{0.25}
                            HorizontalAlignment := horizontal_alignment.Fill
                            VerticalAlignment := vertical_alignment.Fill
                            Widget := stack_box:
                                Orientation := orientation.Vertical
                                Slots := array:
                                    stack_box_slot:
                                        Padding:=margin{Top:= 10.0, Bottom:= 10.0}
                                        HorizontalAlignment := horizontal_alignment.Fill
                                        VerticalAlignment := vertical_alignment.Fill
                                        Widget := AddStock_BuffTextBlock
                                    stack_box_slot:
                                        Padding:=margin{Top:= 15.0, Bottom:= 15.0}
                                        HorizontalAlignment := horizontal_alignment.Fill
                                        VerticalAlignment := vertical_alignment.Fill
                                        Widget := Lv_AddStock_BuffTextBlock
                                    stack_box_slot:
                                        Padding:=margin{Top:= 10.0, Bottom:= 10.0}
                                        HorizontalAlignment := horizontal_alignment.Center
                                        VerticalAlignment := vertical_alignment.Center
                                        Widget := AddStock_BuffImage
                                    stack_box_slot:
                                        Padding:=margin{Top:= 15.0, Bottom:= 15.0}
                                        HorizontalAlignment := horizontal_alignment.Fill
                                        VerticalAlignment := vertical_alignment.Fill
                                        Widget := _AddStock_BuffTextBlock
        }

        
        #================================ Bottom ================================

        set CanvasSlots += array
        {
            canvas_slot:
                Anchors := CenterAnchors
                Offsets := margin{Left:=0.0 ,Top:= 249.5, Right:= 1760.0, Bottom:= 210.0}
                Alignment := CenterVector
                SizeToContent := false
                Widget := stack_box:
                    Orientation := orientation.Horizontal
                    Slots := array:
                        stack_box_slot:
                            Distribution:= option{0.25}
                            HorizontalAlignment := horizontal_alignment.Fill
                            VerticalAlignment := vertical_alignment.Fill
                            Widget := stack_box:
                                Orientation := orientation.Vertical
                                Slots := array:
                                    stack_box_slot:
                                        Padding:=margin{Top:= 10.0, Bottom:= 10.0}
                                        HorizontalAlignment := horizontal_alignment.Fill
                                        VerticalAlignment := vertical_alignment.Fill
                                        Widget := stack_box:
                                            Orientation := orientation.Horizontal
                                            Slots := array:
                                                stack_box_slot:
                                                    Padding:=margin{Top:= 10.0, Right:= 5.0, Bottom:= 10.0}
                                                    Distribution:= option{0.5}
                                                    HorizontalAlignment := horizontal_alignment.Right
                                                    VerticalAlignment := vertical_alignment.Fill
                                                    Widget := Now_UnlockTextBlock
                                                stack_box_slot:
                                                    Padding:=margin{Top:= 10.0, Left:= 5.0, Bottom:= 10.0}
                                                    Distribution:= option{0.5}
                                                    HorizontalAlignment := horizontal_alignment.Left
                                                    VerticalAlignment := vertical_alignment.Fill
                                                    Widget := Will_UnlockTextBlock
                                    stack_box_slot:
                                        Padding:=margin{Top:= 10.0, Bottom:= 10.0}
                                        HorizontalAlignment := horizontal_alignment.Fill
                                        VerticalAlignment := vertical_alignment.Fill
                                        Widget := stack_box:
                                            Orientation := orientation.Horizontal
                                            Slots := array:
                                                stack_box_slot:
                                                    Distribution:= option{0.5}
                                                    HorizontalAlignment := horizontal_alignment.Right
                                                    VerticalAlignment := vertical_alignment.Fill
                                                    Widget := CostTextBlock1
                                                stack_box_slot:
                                                    Distribution:= option{0.1}
                                                    HorizontalAlignment := horizontal_alignment.Center
                                                    VerticalAlignment := vertical_alignment.Fill
                                                    Widget := SiLverOreImage3
                                                stack_box_slot:
                                                    Distribution:= option{0.4}
                                                    HorizontalAlignment := horizontal_alignment.Left
                                                    VerticalAlignment := vertical_alignment.Fill
                                                    Widget := Cost_UnlockTextBlock
                                    stack_box_slot:
                                        Distribution:= option{0.2}
                                        HorizontalAlignment := horizontal_alignment.Center
                                        VerticalAlignment := vertical_alignment.Center
                                        Widget := QuietButtonLevelup1
                        stack_box_slot:
                            Distribution:= option{0.25}
                            HorizontalAlignment := horizontal_alignment.Fill
                            VerticalAlignment := vertical_alignment.Fill
                            Widget := stack_box:
                                Orientation := orientation.Vertical
                                Slots := array:
                                    stack_box_slot:
                                        Padding:=margin{Top:= 10.0, Bottom:= 10.0}
                                        HorizontalAlignment := horizontal_alignment.Fill
                                        VerticalAlignment := vertical_alignment.Fill
                                        Widget := stack_box:
                                            Orientation := orientation.Horizontal
                                            Slots := array:
                                                stack_box_slot:
                                                    Padding:=margin{Top:= 10.0, Right:= 5.0, Bottom:= 10.0}
                                                    Distribution:= option{0.5}
                                                    HorizontalAlignment := horizontal_alignment.Right
                                                    VerticalAlignment := vertical_alignment.Fill
                                                    Widget := Now_BuffEffectTextBlock
                                                stack_box_slot:
                                                    Padding:=margin{Top:= 10.0, Left:= 5.0, Bottom:= 10.0}
                                                    Distribution:= option{0.5}
                                                    HorizontalAlignment := horizontal_alignment.Left
                                                    VerticalAlignment := vertical_alignment.Fill
                                                    Widget := Will_BuffEffectTextBlock
                                    stack_box_slot:
                                        Padding:=margin{Top:= 10.0, Bottom:= 10.0}
                                        HorizontalAlignment := horizontal_alignment.Fill
                                        VerticalAlignment := vertical_alignment.Fill
                                        Widget := stack_box:
                                            Orientation := orientation.Horizontal
                                            Slots := array:
                                                stack_box_slot:
                                                    Distribution:= option{0.5}
                                                    HorizontalAlignment := horizontal_alignment.Right
                                                    VerticalAlignment := vertical_alignment.Fill
                                                    Widget := CostTextBlock2
                                                stack_box_slot:
                                                    Distribution:= option{0.1}
                                                    HorizontalAlignment := horizontal_alignment.Center
                                                    VerticalAlignment := vertical_alignment.Fill
                                                    Widget := SiLverOreImage4
                                                stack_box_slot:
                                                    Distribution:= option{0.4}
                                                    HorizontalAlignment := horizontal_alignment.Left
                                                    VerticalAlignment := vertical_alignment.Fill
                                                    Widget := Cost_BuffEffectTextBlock
                                    stack_box_slot:
                                        Distribution:= option{0.2}
                                        HorizontalAlignment := horizontal_alignment.Center
                                        VerticalAlignment := vertical_alignment.Center
                                        Widget := QuietButtonLevelup2

                        stack_box_slot:
                            Distribution:= option{0.25}
                            HorizontalAlignment := horizontal_alignment.Fill
                            VerticalAlignment := vertical_alignment.Fill
                            Widget := stack_box:
                                Orientation := orientation.Vertical
                                Slots := array:
                                    stack_box_slot:
                                        Padding:=margin{Top:= 10.0, Bottom:= 10.0}
                                        HorizontalAlignment := horizontal_alignment.Fill
                                        VerticalAlignment := vertical_alignment.Fill
                                        Widget := stack_box:
                                            Orientation := orientation.Horizontal
                                            Slots := array:
                                                stack_box_slot:
                                                    Padding:=margin{Top:= 10.0, Right:= 5.0, Bottom:= 10.0}
                                                    Distribution:= option{0.5}
                                                    HorizontalAlignment := horizontal_alignment.Right
                                                    VerticalAlignment := vertical_alignment.Fill
                                                    Widget := Now_BuffTimeTextBlock
                                                stack_box_slot:
                                                    Padding:=margin{Top:= 10.0, Left:= 5.0, Bottom:= 10.0}
                                                    Distribution:= option{0.5}
                                                    HorizontalAlignment := horizontal_alignment.Left
                                                    VerticalAlignment := vertical_alignment.Fill
                                                    Widget := Will_BuffTimeTextBlock
                                    stack_box_slot:
                                        Padding:=margin{Top:= 10.0, Bottom:= 10.0}
                                        HorizontalAlignment := horizontal_alignment.Fill
                                        VerticalAlignment := vertical_alignment.Fill
                                        Widget := stack_box:
                                            Orientation := orientation.Horizontal
                                            Slots := array:
                                                stack_box_slot:
                                                    Distribution:= option{0.5}
                                                    HorizontalAlignment := horizontal_alignment.Right
                                                    VerticalAlignment := vertical_alignment.Fill
                                                    Widget := CostTextBlock3
                                                stack_box_slot:
                                                    Distribution:= option{0.1}
                                                    HorizontalAlignment := horizontal_alignment.Center
                                                    VerticalAlignment := vertical_alignment.Fill
                                                    Widget := SiLverOreImage5
                                                stack_box_slot:
                                                    Distribution:= option{0.4}
                                                    HorizontalAlignment := horizontal_alignment.Left
                                                    VerticalAlignment := vertical_alignment.Fill
                                                    Widget := Cost_BuffTimeTextBlock
                                    stack_box_slot:
                                        Distribution:= option{0.2}
                                        HorizontalAlignment := horizontal_alignment.Center
                                        VerticalAlignment := vertical_alignment.Center
                                        Widget := QuietButtonLevelup3
                        stack_box_slot:
                            Distribution:= option{0.25}
                            HorizontalAlignment := horizontal_alignment.Fill
                            VerticalAlignment := vertical_alignment.Fill
                            Widget := stack_box:
                                Orientation := orientation.Vertical
                                Slots := array:
                                    stack_box_slot:
                                        Padding:=margin{Top:= 10.0, Bottom:= 10.0}
                                        HorizontalAlignment := horizontal_alignment.Fill
                                        VerticalAlignment := vertical_alignment.Fill
                                        Widget := stack_box:
                                            Orientation := orientation.Horizontal
                                            Slots := array:
                                                stack_box_slot:
                                                    Padding:=margin{Top:= 10.0, Right:= 5.0, Bottom:= 10.0}
                                                    Distribution:= option{0.5}
                                                    HorizontalAlignment := horizontal_alignment.Right
                                                    VerticalAlignment := vertical_alignment.Fill
                                                    Widget := Now_AddStock_BuffTextBlock
                                                stack_box_slot:
                                                    Padding:=margin{Top:= 10.0, Left:= 5.0, Bottom:= 10.0}
                                                    Distribution:= option{0.5}
                                                    HorizontalAlignment := horizontal_alignment.Left
                                                    VerticalAlignment := vertical_alignment.Fill
                                                    Widget := Will_AddStock_BuffTextBlock
                                    stack_box_slot:
                                        Padding:=margin{Top:= 10.0, Bottom:= 10.0}
                                        HorizontalAlignment := horizontal_alignment.Fill
                                        VerticalAlignment := vertical_alignment.Fill
                                        Widget := stack_box:
                                            Orientation := orientation.Horizontal
                                            Slots := array:
                                                stack_box_slot:
                                                    Distribution:= option{0.5}
                                                    HorizontalAlignment := horizontal_alignment.Right
                                                    VerticalAlignment := vertical_alignment.Fill
                                                    Widget := CostTextBlock4
                                                stack_box_slot:
                                                    Distribution:= option{0.1}
                                                    HorizontalAlignment := horizontal_alignment.Center
                                                    VerticalAlignment := vertical_alignment.Fill
                                                    Widget := SiLverOreImage6
                                                stack_box_slot:
                                                    Distribution:= option{0.4}
                                                    HorizontalAlignment := horizontal_alignment.Left
                                                    VerticalAlignment := vertical_alignment.Fill
                                                    Widget := Cost_AddStock_BuffTextBlock
                                    stack_box_slot:
                                        Distribution:= option{0.2}
                                        HorizontalAlignment := horizontal_alignment.Center
                                        VerticalAlignment := vertical_alignment.Center
                                        Widget := QuietButtonLevelup4
        }


        #========================================== Close ==========================================

        CloseButton : button_loud = button_loud{DefaultText := StringToMessage("CLOSE")}
        CloseButton.OnClick().Subscribe(OnCloseButtonClicked)
        set CanvasSlots += array
        {
            canvas_slot:
                Anchors := CenterAnchors
                Offsets := margin{Top:= 435.0, Left:= 658.0, Right := 353.0, Bottom := 80.0}
                Alignment := CenterVector
                SizeToContent := false
                Widget := CloseButton
        }
        
        if(set UIElements[Agent] = (OreTextBlock, PowerupStoneTextBlock)){}
        if(set SkillUIElements[Agent] = map{}){}
        if:
            set SkillUIElements[Agent] = map:
                1 => (Lv_UnlockTextBlock, Now_UnlockTextBlock, Will_UnlockTextBlock, Cost_UnlockTextBlock, _UnlockTextBlock)
                2 => (Lv_BuffEffectTextBlock, Now_BuffEffectTextBlock, Will_BuffEffectTextBlock, Cost_BuffEffectTextBlock, _BuffEffectTextBlock)
                3 => (Lv_BuffTimeTextBlock, Now_BuffTimeTextBlock, Will_BuffTimeTextBlock, Cost_BuffTimeTextBlock, _BuffTimeTextBlock)
                4 => (Lv_AddStock_BuffTextBlock, Now_AddStock_BuffTextBlock, Will_AddStock_BuffTextBlock, Cost_AddStock_BuffTextBlock, _AddStock_BuffTextBlock)


        Canvas : canvas = canvas:
            Slots := CanvasSlots

        return Canvas