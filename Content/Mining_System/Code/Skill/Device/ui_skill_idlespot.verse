
using { /Verse.org/Assets }
using { /Verse.org/Simulation }
using { /Fortnite.com/UI }
using { /Fortnite.com/Devices }
using { /Fortnite.com/Characters }
using { /UnrealEngine.com/Temporary/UI }
using { /UnrealEngine.com/Temporary/SpatialMath }
using { /Verse.org/Simulation }
using { /UnrealEngine.com/Temporary/Diagnostics }
using { /Verse.org/Colors }
using { /Verse.org/Colors/NamedColors }
using { Utils }
using { PlayerData }
using { Skill }


# See https://dev.epicgames.com/documentation/en-us/uefn/create-your-own-device-in-verse for how to create a verse device.

# A Verse-authored creative device that can be placed in a level
ui_skill_idlespot := class(creative_device):
    #SFXs
    @editable
    SFX_Open : audio_player_device = audio_player_device{}
    @editable
    SFX_Close : audio_player_device = audio_player_device{}
    @editable
    SFX_LevelUp : audio_player_device = audio_player_device{}
    @editable
    SFX_Fail : audio_player_device = audio_player_device{}


    var _SaveData : save_data = save_data{}
    @editable
    var SaveDataManager : save_data_manager = save_data_manager{}
    @editable
    var SkillDataManager : skill_manager = skill_manager{}
    CloseEvent<public> : event(player) = event(player){}

    @editable
    SkillNPC : character_device = character_device{}

    var IdleSpotCanvasPerAgent : [agent]canvas = map{}
    var UIElements : [agent]tuple(text_block, text_block) = map{}
    var SkillUIElements : [agent][int]tuple(text_block, text_block, text_block, text_block, text_block) = map{}


    OnBegin<override>()<suspends>:void=
        SkillNPC.InteractedWithEvent.Subscribe(OnButtonEvented)

    SetText<localizes>(Value:[]char) : message = "{Value}"
    
    # スキル関連のUI要素を全て更新
    UpdateUI(Agent:agent):void=
        UpdateUIElements(Agent)
        UpdateSkillUIElements(Agent)

    # UIElementsに格納された基本リソースのテキストを更新する
    UpdateUIElements(Agent: agent):void=
        var SilverValue : big_number = big_number{}
        var PowerupStoneValue : big_number = big_number{}
        if:
            Player := player[Agent]
        then:
            set SilverValue = SaveDataManager.GetResourceStone(Player)
            set PowerupStoneValue = SaveDataManager.GetResourcePowerupStone(Player)
            if (Elements := UIElements[Agent]):
                Elements(0).SetText(SetText("{BigNumber_FormatDisplay(SilverValue)}"))          # OreTextBlockのテキストを設定
                Elements(1).SetText(SetText("{BigNumber_FormatDisplay(PowerupStoneValue)}"))    # PowerupStoneTextBlockのテキストを設定 

    UpdateSkillUIElements(Agent:agent):void=
        var Lv_Playeridle : int = 1 #3
        var Now_Playeridle : int = 1
        var Will_Playeridle : int = 1
        var Cost_Playeridle : big_number = big_number{}
        var Lv_Petidle : int = 1    #4
        var Now_Petidle : int = 1
        var Will_Petidle : int = 1
        var Cost_Petidle : big_number = big_number{}
        if(Player := player[Agent]) :
            var LanguageIndex : int = SaveDataManager.GetLanguage(Player)
            if:
                set Lv_Playeridle = SaveDataManager.GetSkillLevel[Player, 3]
                set Now_Playeridle = SaveDataManager.GetPlayerSkillEffectIncrease[Player, 3]
                set Will_Playeridle = SaveDataManager.GetSpecificSkillEffectIncrease[Player, 3]
                set Cost_Playeridle = SaveDataManager.GetUpgradeCostForSkill[Player, 3]

                set Lv_Petidle = SaveDataManager.GetSkillLevel[Player, 4]
                set Now_Petidle = SaveDataManager.GetPlayerSkillEffectIncrease[Player, 4]
                set Will_Petidle = SaveDataManager.GetSpecificSkillEffectIncrease[Player, 4]
                set Cost_Petidle = SaveDataManager.GetUpgradeCostForSkill[Player, 4]
            then:
                LvPlayerIdleText := if(SkillDataManager.CanUpgradeSkill(player[Agent], 3)?){"Lv.{Lv_Playeridle}"} else {"Lv.MAX"}
                LvPlayerIdleTextColor := if(SkillDataManager.CanUpgradeSkill(player[Agent], 3)?){NamedColors.White} else {NamedColors.Red}
                WillPlayerIdleText := if(SkillDataManager.CanUpgradeSkill(player[Agent], 3)?){"+ {Will_Playeridle} %"} else {"MAX!"}
                CostPlayerIdleText := if(SkillDataManager.CanUpgradeSkill(player[Agent], 3)?){"{BigNumber_FormatDisplay(Cost_Playeridle)}"} else {"-"}
                CostPlayerIdleTextColor := if(SkillDataManager.IsUnlockable[player[Agent], 3]){NamedColors.White} else {NamedColors.Red}
                IdleText := if(LanguageIndex = 0){"休憩スポットにてプレイヤー\n放置中の報酬増加"} else {"Get more rewards when idle"}
                
                LvPetidleIdleText := if(SkillDataManager.CanUpgradeSkill(player[Agent], 4)?){"Lv.{Lv_Petidle}"} else {"Lv.MAX"}
                LvPetidleIdleTextColor := if(SkillDataManager.CanUpgradeSkill(player[Agent], 4)?){NamedColors.White} else {NamedColors.Red}
                WillPetidleTextBlock := if(SkillDataManager.CanUpgradeSkill(player[Agent], 4)?){"+ {Will_Petidle} %"} else {"MAX!"}
                CostPetIdleText := if(SkillDataManager.CanUpgradeSkill(player[Agent], 4)?){"{BigNumber_FormatDisplay(Cost_Petidle)}"} else {"-"}
                CostPetIdleTextColor := if(SkillDataManager.IsUnlockable[player[Agent], 4]){NamedColors.White} else {NamedColors.Red}
                PetIdleText := if(LanguageIndex = 0){"休憩スポットにてペット\n放置中の報酬増加"} else {"Get more rewards when idle"}

                if(SkillElement := SkillUIElements[Agent][1]) :
                    SkillElement(0).SetText(SetText(LvPlayerIdleText))
                    SkillElement(0).SetTextColor(LvPlayerIdleTextColor)
                    SkillElement(1).SetText(SetText("{Now_Playeridle}"))
                    SkillElement(2).SetText(SetText(WillPlayerIdleText))
                    SkillElement(3).SetText(SetText(CostPlayerIdleText))
                    SkillElement(3).SetTextColor(CostPlayerIdleTextColor)
                    SkillElement(4).SetText(SetText("{IdleText}"))

                if(SkillElement := SkillUIElements[Agent][2]) :
                    SkillElement(0).SetText(SetText(LvPetidleIdleText))
                    SkillElement(0).SetTextColor(LvPetidleIdleTextColor)
                    SkillElement(1).SetText(SetText("{Now_Petidle}"))
                    SkillElement(2).SetText(SetText(WillPetidleTextBlock))
                    SkillElement(3).SetText(SetText(CostPetIdleText))
                    SkillElement(3).SetTextColor(CostPetIdleTextColor)
                    SkillElement(4).SetText(SetText("{PetIdleText}"))

    OnPlayerSpawned(Agent : agent):void=
        if (IdleSpotCanvas := IdleSpotCanvasPerAgent[Agent]):
            return

        if (Player := player[Agent], PlayerUI := GetPlayerUI[Player]):
            IdleSpotCanvas : canvas = CreateIdleSpotCanvas(Agent, Player)
            if (set IdleSpotCanvasPerAgent[Agent] = IdleSpotCanvas){}

    OnButtonEvented(Agent : agent):void=
        SFX_Open.Play(Agent)
        OnPlayerSpawned(Agent)
        UpdateUI(Agent)
        Print("OnSkillNPCEvented")
        if (Player := player[Agent], PlayerUI := GetPlayerUI[Player], IdleSpotCanvas := IdleSpotCanvasPerAgent[Agent]):
            PlayerUI.AddWidget(IdleSpotCanvas, player_ui_slot{InputMode := ui_input_mode.All})

    # ボタン処理
    OnLevelup1Clicked(WidgetMessage : widget_message):void=
        if (PlayerUI := GetPlayerUI[WidgetMessage.Player], IdleSpotCanvas := IdleSpotCanvasPerAgent[WidgetMessage.Player], Agent : agent = agent[WidgetMessage.Player]):
            if(SkillDataManager.CanUpgradeSkill(WidgetMessage.Player,3) = true):
                SFX_LevelUp.Play(Agent)
                if:
                    SkillDataManager.UnlockOrUpgradeSkill[WidgetMessage.Player,3]
                then:
                    UpdateUI(Agent)
            else:
                SFX_Fail.Play(Agent)
                return
        
    OnLevelup2Clicked(WidgetMessage : widget_message):void=
        if (PlayerUI := GetPlayerUI[WidgetMessage.Player], IdleSpotCanvas := IdleSpotCanvasPerAgent[WidgetMessage.Player], Agent : agent = agent[WidgetMessage.Player]):
            if(SkillDataManager.CanUpgradeSkill(WidgetMessage.Player,4) = true):
                SFX_LevelUp.Play(Agent)
                if:
                    SkillDataManager.UnlockOrUpgradeSkill[WidgetMessage.Player,4]
                then:
                    UpdateUI(Agent)
            else:
                SFX_Fail.Play(Agent)
                return

    OnCloseButtonClicked(WidgetMessage : widget_message):void=
        if(Agent : agent = agent[WidgetMessage.Player]) :
            SFX_Close.Play(Agent)
        if (PlayerUI := GetPlayerUI[WidgetMessage.Player], IdleSpotCanvas := IdleSpotCanvasPerAgent[WidgetMessage.Player]):
            PlayerUI.RemoveWidget(IdleSpotCanvas)
            CloseEvent.Signal(WidgetMessage.Player)
            
    # キャンバス
    CreateIdleSpotCanvas(Agent : agent, Player : player):canvas=
                
        var CanvasSlots : []canvas_slot = array{}
        
        CenterVector : vector2 = vector2{X:=0.5, Y:=0.5}
        CenterAnchors : anchors = anchors{Minimum := CenterVector, Maximum := CenterVector}

        # Image Stock
        BGImage : texture_block = texture_block{DefaultImage := Textures.BG_Skill, DefaultDesiredSize := vector2{X := 1920.0, Y := 1080.0}}
        SilverOreImage1 : texture_block = texture_block{DefaultImage := Textures.SilverOre, DefaultDesiredSize := vector2{X := 50.0, Y := 50.0}}
        PowerupStoneImage : texture_block = texture_block{DefaultImage := Textures.IC_PowerupStone, DefaultDesiredSize := vector2{X := 50.0, Y := 50.0}}
        SiLverOreImage3 : texture_block = texture_block{DefaultImage := Textures.SilverOre, DefaultDesiredSize := vector2{X := 40.0, Y := 40.0}}
        SiLverOreImage4 : texture_block = texture_block{DefaultImage := Textures.SilverOre, DefaultDesiredSize := vector2{X := 40.0, Y := 40.0}}
        PlayeridleImage : texture_block = texture_block{DefaultImage := Textures.IdleSpot_1, DefaultDesiredSize := vector2{X := 200.0, Y := 200.0}}
        PetidleImage : texture_block = texture_block{DefaultImage := Textures.IdleSpot_2, DefaultDesiredSize := vector2{X := 200.0, Y := 200.0}}
        
        var SilverValue : big_number = big_number{}
        var PowerupStoneValue : big_number = big_number{}
        var Lv_Playeridle : int = 1 #3
        var Now_Playeridle : int = 1
        var Will_Playeridle : int = 1
        var Cost_Playeridle : big_number = big_number{}
        var Lv_Petidle : int = 1    #4
        var Now_Petidle : int = 1
        var Will_Petidle : int = 1
        var Cost_Petidle : big_number = big_number{}

        #================================== TextBlock ==================================
        # 所持岩
        OreTextBlock : text_block = text_block{
            # DefaultText := StringToMessage("{BigNumber_FormatDisplay(SilverValue)}"),
            DefaultTextColor := NamedColors.White,
            DefaultShadowOffset := option{vector2{X:=2.5, Y:=2.5}},
            DefaultShadowColor := NamedColors.Black
        }
        # 所持強化アイテム
        PowerupStoneTextBlock : text_block = text_block{
            # DefaultText := StringToMessage("{BigNumber_FormatDisplay(PowerupStoneValue)}"),
            DefaultTextColor := NamedColors.White,
            DefaultShadowOffset := option{vector2{X:=2.5, Y:=2.5}},
            DefaultShadowColor := NamedColors.Black,
            DefaultJustification := text_justification.Center
        }
        #================================== Skill 1 ==================================
        Lv_PlayeridleTextBlock : text_block = text_block{
            # DefaultText := StringToMessage("{LvPlayerIdleText}"),
            DefaultTextColor := NamedColors.White,
            DefaultShadowOffset := option{vector2{X:=2.5, Y:=2.5}},
            DefaultShadowColor := NamedColors.Black,
            DefaultJustification := text_justification.Center
        }
        Now_PlayeridleTextBlock : text_block = text_block{
            # DefaultText := StringToMessage("{Now_Playeridle}%"),
            DefaultTextColor := NamedColors.White,
            DefaultShadowOffset := option{vector2{X:=2.5, Y:=2.5}},
            DefaultShadowColor := NamedColors.Black,
            DefaultJustification := text_justification.Center
        }
        Will_PlayeridleTextBlock : text_block = text_block{
            # DefaultText := StringToMessage("({WillPlayerIdleText})"),
            DefaultTextColor := NamedColors.White,
            DefaultShadowOffset := option{vector2{X:=2.5, Y:=2.5}},
            DefaultShadowColor := NamedColors.Black,
            DefaultJustification := text_justification.Center
        }
        Cost_PlayeridleTextBlock : text_block = text_block{
            # DefaultText := StringToMessage("{CostPlayerIdleText}"),
            DefaultTextColor := NamedColors.White,
            DefaultShadowOffset := option{vector2{X:=2.5, Y:=2.5}},
            DefaultShadowColor := NamedColors.Black,
            DefaultJustification := text_justification.Center
        }
        var _PlayeridleTextBlock : text_block= text_block :
            DefaultTextColor := NamedColors.White
            DefaultShadowOffset := option{vector2{X:=2.5, Y:=2.5}}
            DefaultShadowColor := NamedColors.Black
            DefaultShadowOpacity := 1.0
            DefaultJustification := text_justification.Center
        #================================== Skill 2 ==================================
        Lv_PetidleTextBlock : text_block = text_block{
            # DefaultText := StringToMessage("{(LvPetidleIdleText)}"),
            DefaultTextColor := NamedColors.White,
            DefaultShadowOffset := option{vector2{X:=2.5, Y:=2.5}},
            DefaultShadowColor := NamedColors.Black,
            DefaultJustification := text_justification.Center
        }
        Now_PetidleTextBlock : text_block = text_block{
            # DefaultText := StringToMessage("{Now_Petidle}%"),
            DefaultTextColor := NamedColors.White,
            DefaultShadowOffset := option{vector2{X:=2.5, Y:=2.5}},
            DefaultShadowColor := NamedColors.Black,
            DefaultJustification := text_justification.Center
        }
        Will_PetidleTextBlock : text_block = text_block{
            # DefaultText := StringToMessage("({WillPetidleTextBlock})"),
            DefaultTextColor := NamedColors.White,
            DefaultShadowOffset := option{vector2{X:=2.5, Y:=2.5}},
            DefaultShadowColor := NamedColors.Black,
            DefaultJustification := text_justification.Center
        }
        Cost_PetidleTextBlock : text_block = text_block{
            # DefaultText := StringToMessage("{CostPetIdleText}"),
            DefaultTextColor := NamedColors.White,
            DefaultShadowOffset := option{vector2{X:=2.5, Y:=2.5}},
            DefaultShadowColor := NamedColors.Black,
            DefaultJustification := text_justification.Center
        }
        var _PetidleTextBlock : text_block = text_block :
            DefaultTextColor := NamedColors.White
            DefaultShadowOffset := option{vector2{X:=2.5, Y:=2.5}}
            DefaultShadowColor := NamedColors.Black
            DefaultShadowOpacity := 1.0
            DefaultJustification := text_justification.Center
        #================================== /TextBlock ==================================
        #========================================== BG ==========================================

        set CanvasSlots += array
        {
            canvas_slot:
                Anchors := CenterAnchors
                Offsets := margin{Top := 0.0,Left := 0.0, Right := 1920.0, Bottom := 1080.0}
                Alignment := CenterVector
                SizeToContent := false
                Widget := BGImage
        }

        #========================================= Header ========================================
        
        IdleSpotTextBlock := text_block{
            DefaultText := StringToMessage("Idle Spot"),
            DefaultTextColor := NamedColors.White,
            DefaultShadowOffset := option{vector2{X:=2.5, Y:=2.5}},
            DefaultShadowColor := NamedColors.Black
        }
        IdleSpotTextBlock.SetShadowOpacity(1.0)
        set CanvasSlots += array
        {
            canvas_slot:
                Anchors := CenterAnchors
                Offsets := margin{Top:= -472.0, Left:= 0.0, Right := 257.0, Bottom := 41.0}
                Alignment := CenterVector
                SizeToContent := false
                Widget := IdleSpotTextBlock
        }

        OreTextBlock.SetShadowOpacity(1.0)
        PowerupStoneTextBlock.SetShadowOpacity(1.0)

        #========================================== Player Stats ==========================================

        set CanvasSlots += array
        {
            canvas_slot:
                Anchors := CenterAnchors
                Offsets := margin{Left:=-645.5 ,Top:= -391.0, Right:= 519.0, Bottom:= 59.0}
                Alignment := CenterVector
                SizeToContent := false
                Widget := stack_box:
                    Orientation := orientation.Horizontal
                    Slots := array:
                        stack_box_slot:
                            Distribution:= option{0.5}
                            Padding:=margin{Left:=5.0 ,Top:= 5.0, Right:= 10.0, Bottom:= 5.0}
                            HorizontalAlignment := horizontal_alignment.Fill
                            VerticalAlignment := vertical_alignment.Fill
                            Widget := stack_box:
                                Orientation := orientation.Horizontal
                                Slots := array:
                                    stack_box_slot:
                                        HorizontalAlignment := horizontal_alignment.Center
                                        VerticalAlignment := vertical_alignment.Center
                                        Widget := SilverOreImage1
                                    stack_box_slot:
                                        Distribution:= option{1.0}
                                        HorizontalAlignment := horizontal_alignment.Center
                                        VerticalAlignment := vertical_alignment.Center
                                        Widget := OreTextBlock
                        stack_box_slot:
                            Padding:=margin{Left:=10.0 ,Top:= 5.0, Right:= 5.0, Bottom:= 5.0}
                            Distribution:= option{0.5}
                            HorizontalAlignment := horizontal_alignment.Fill
                            VerticalAlignment := vertical_alignment.Fill
                            Widget := stack_box:
                                Orientation := orientation.Horizontal
                                Slots := array:
                                    stack_box_slot:
                                        HorizontalAlignment := horizontal_alignment.Center
                                        VerticalAlignment := vertical_alignment.Center
                                        Widget := PowerupStoneImage
                                    stack_box_slot:
                                        Distribution:= option{1.0}
                                        HorizontalAlignment := horizontal_alignment.Center
                                        VerticalAlignment := vertical_alignment.Center
                                        Widget := PowerupStoneTextBlock
        }
        #========================================== Skill ==========================================

        CostTextBlock1 := text_block{
            DefaultText := StringToMessage("Cost :"),
            DefaultTextColor := NamedColors.White,
            DefaultShadowOffset := option{vector2{X:=2.5, Y:=2.5}},
            DefaultShadowColor := NamedColors.Black,
            DefaultJustification := text_justification.Center
        }
        CostTextBlock1.SetShadowOpacity(1.0)
        CostTextBlock2 := text_block{
            DefaultText := StringToMessage("Cost :"),
            DefaultTextColor := NamedColors.White,
            DefaultShadowOffset := option{vector2{X:=2.5, Y:=2.5}},
            DefaultShadowColor := NamedColors.Black,
            DefaultJustification := text_justification.Center
        }
        CostTextBlock2.SetShadowOpacity(1.0)

        #Skill 1
        PlayeridleTextBlock := text_block{
            DefaultText := StringToMessage("Player idle"),
            DefaultTextColor := NamedColors.White,
            DefaultShadowOffset := option{vector2{X:=2.5, Y:=2.5}},
            DefaultShadowColor := NamedColors.Black,
            DefaultJustification := text_justification.Center
        }

        Now_PlayeridleTextBlock.SetShadowOpacity(1.0)
        Will_PlayeridleTextBlock.SetShadowOpacity(1.0)
        PlayeridleTextBlock.SetShadowOpacity(1.0)
        Lv_PlayeridleTextBlock.SetShadowOpacity(1.0)
        Cost_PlayeridleTextBlock.SetShadowOpacity(1.0)
        
        #Skill 2
        PetidleTextBlock := text_block{
            DefaultText := StringToMessage("Yo-kai idle"),
            DefaultTextColor := NamedColors.White,
            DefaultShadowOffset := option{vector2{X:=2.5, Y:=2.5}},
            DefaultShadowColor := NamedColors.Black,
            DefaultJustification := text_justification.Center
        }
        PetidleTextBlock.SetShadowOpacity(1.0)
        Lv_PetidleTextBlock.SetShadowOpacity(1.0)
        Now_PetidleTextBlock.SetShadowOpacity(1.0)
        Will_PetidleTextBlock.SetShadowOpacity(1.0)
        Cost_PetidleTextBlock.SetShadowOpacity(1.0)
        
        # Button Stock
        QuietButtonLevelup1 : button_quiet = button_quiet{DefaultText := StringToMessage("LEVEL UP")}
        QuietButtonLevelup1.OnClick().Subscribe(OnLevelup1Clicked)
        QuietButtonLevelup2 : button_quiet = button_quiet{DefaultText := StringToMessage("LEVEL UP")}
        QuietButtonLevelup2.OnClick().Subscribe(OnLevelup2Clicked)

        #================================ Top ================================

        set CanvasSlots += array
        {
            canvas_slot:
                Anchors := CenterAnchors
                Offsets := margin{Left:=0.0 ,Top:= -96.5, Right:= 1760.0, Bottom:= 464.0}
                Alignment := CenterVector
                SizeToContent := false
                Widget := stack_box:
                    Orientation := orientation.Horizontal
                    Slots := array:
                        stack_box_slot:
                            Padding:=margin{Left:= 50.0, Right:= 50.0}
                            Distribution:= option{0.25}
                            HorizontalAlignment := horizontal_alignment.Fill
                            VerticalAlignment := vertical_alignment.Fill
                            Widget := stack_box:
                                Orientation := orientation.Vertical
                                Slots := array:
                                    stack_box_slot:
                                        Padding:=margin{Top:= 10.0, Bottom:= 10.0}
                                        HorizontalAlignment := horizontal_alignment.Fill
                                        VerticalAlignment := vertical_alignment.Fill
                                        Widget := PlayeridleTextBlock
                                    stack_box_slot:
                                        Padding:=margin{Top:= 15.0, Bottom:= 15.0}
                                        HorizontalAlignment := horizontal_alignment.Fill
                                        VerticalAlignment := vertical_alignment.Fill
                                        Widget := Lv_PlayeridleTextBlock
                                    stack_box_slot:
                                        Padding:=margin{Top:= 10.0, Bottom:= 10.0}
                                        HorizontalAlignment := horizontal_alignment.Center
                                        VerticalAlignment := vertical_alignment.Center
                                        Widget := PlayeridleImage
                                    stack_box_slot:
                                        Padding:=margin{Top:= 15.0, Bottom:= 15.0}
                                        HorizontalAlignment := horizontal_alignment.Fill
                                        VerticalAlignment := vertical_alignment.Fill
                                        Widget := _PlayeridleTextBlock
                        stack_box_slot:
                            Padding:=margin{Left:= 50.0, Right:= 50.0}
                            Distribution:= option{0.25}
                            HorizontalAlignment := horizontal_alignment.Fill
                            VerticalAlignment := vertical_alignment.Fill
                            Widget := stack_box:
                                Orientation := orientation.Vertical
                                Slots := array:
                                    stack_box_slot:
                                        Padding:=margin{Top:= 10.0, Bottom:= 10.0}
                                        HorizontalAlignment := horizontal_alignment.Fill
                                        VerticalAlignment := vertical_alignment.Fill
                                        Widget := PetidleTextBlock
                                    stack_box_slot:
                                        Padding:=margin{Top:= 15.0, Bottom:= 15.0}
                                        HorizontalAlignment := horizontal_alignment.Fill
                                        VerticalAlignment := vertical_alignment.Fill
                                        Widget := Lv_PetidleTextBlock
                                    stack_box_slot:
                                        Padding:=margin{Top:= 10.0, Bottom:= 10.0}
                                        HorizontalAlignment := horizontal_alignment.Center
                                        VerticalAlignment := vertical_alignment.Center
                                        Widget := PetidleImage
                                    stack_box_slot:
                                        Padding:=margin{Top:= 15.0, Bottom:= 15.0}
                                        HorizontalAlignment := horizontal_alignment.Fill
                                        VerticalAlignment := vertical_alignment.Fill
                                        Widget := _PetidleTextBlock
        }

        
        #================================ Bottom ================================

        set CanvasSlots += array
        {
            canvas_slot:
                Anchors := CenterAnchors
                Offsets := margin{Left:=0.0 ,Top:= 249.5, Right:= 1760.0, Bottom:= 210.0}
                Alignment := CenterVector
                SizeToContent := false
                Widget := stack_box:
                    Orientation := orientation.Horizontal
                    Slots := array:
                        stack_box_slot:
                            Distribution:= option{0.25}
                            HorizontalAlignment := horizontal_alignment.Fill
                            VerticalAlignment := vertical_alignment.Fill
                            Widget := stack_box:
                                Orientation := orientation.Vertical
                                Slots := array:
                                    stack_box_slot:
                                        Padding:=margin{Top:= 10.0, Bottom:= 10.0}
                                        HorizontalAlignment := horizontal_alignment.Fill
                                        VerticalAlignment := vertical_alignment.Fill
                                        Widget := stack_box:
                                            Orientation := orientation.Horizontal
                                            Slots := array:
                                                stack_box_slot:
                                                    Padding:=margin{Top:= 10.0, Right:= 5.0, Bottom:= 10.0}
                                                    Distribution:= option{0.5}
                                                    HorizontalAlignment := horizontal_alignment.Right
                                                    VerticalAlignment := vertical_alignment.Fill
                                                    Widget := Now_PlayeridleTextBlock
                                                stack_box_slot:
                                                    Padding:=margin{Top:= 10.0, Left:= 5.0, Bottom:= 10.0}
                                                    Distribution:= option{0.5}
                                                    HorizontalAlignment := horizontal_alignment.Left
                                                    VerticalAlignment := vertical_alignment.Fill
                                                    Widget := Will_PlayeridleTextBlock
                                    stack_box_slot:
                                        Padding:=margin{Top:= 10.0, Bottom:= 10.0}
                                        HorizontalAlignment := horizontal_alignment.Fill
                                        VerticalAlignment := vertical_alignment.Fill
                                        Widget := stack_box:
                                            Orientation := orientation.Horizontal
                                            Slots := array:
                                                stack_box_slot:
                                                    Distribution:= option{0.5}
                                                    HorizontalAlignment := horizontal_alignment.Right
                                                    VerticalAlignment := vertical_alignment.Fill
                                                    Widget := CostTextBlock1
                                                stack_box_slot:
                                                    Distribution:= option{0.1}
                                                    HorizontalAlignment := horizontal_alignment.Center
                                                    VerticalAlignment := vertical_alignment.Fill
                                                    Widget := SiLverOreImage3
                                                stack_box_slot:
                                                    Distribution:= option{0.4}
                                                    HorizontalAlignment := horizontal_alignment.Left
                                                    VerticalAlignment := vertical_alignment.Fill
                                                    Widget := Cost_PlayeridleTextBlock
                                    stack_box_slot:
                                        Distribution:= option{0.2}
                                        HorizontalAlignment := horizontal_alignment.Center
                                        VerticalAlignment := vertical_alignment.Center
                                        Widget := QuietButtonLevelup1
                        stack_box_slot:
                            Distribution:= option{0.25}
                            HorizontalAlignment := horizontal_alignment.Fill
                            VerticalAlignment := vertical_alignment.Fill
                            Widget := stack_box:
                                Orientation := orientation.Vertical
                                Slots := array:
                                    stack_box_slot:
                                        Padding:=margin{Top:= 10.0, Bottom:= 10.0}
                                        HorizontalAlignment := horizontal_alignment.Fill
                                        VerticalAlignment := vertical_alignment.Fill
                                        Widget := stack_box:
                                            Orientation := orientation.Horizontal
                                            Slots := array:
                                                stack_box_slot:
                                                    Padding:=margin{Top:= 10.0, Right:= 5.0, Bottom:= 10.0}
                                                    Distribution:= option{0.5}
                                                    HorizontalAlignment := horizontal_alignment.Right
                                                    VerticalAlignment := vertical_alignment.Fill
                                                    Widget := Now_PetidleTextBlock
                                                stack_box_slot:
                                                    Padding:=margin{Top:= 10.0, Left:= 5.0, Bottom:= 10.0}
                                                    Distribution:= option{0.5}
                                                    HorizontalAlignment := horizontal_alignment.Left
                                                    VerticalAlignment := vertical_alignment.Fill
                                                    Widget := Will_PetidleTextBlock
                                    stack_box_slot:
                                        Padding:=margin{Top:= 10.0, Bottom:= 10.0}
                                        HorizontalAlignment := horizontal_alignment.Fill
                                        VerticalAlignment := vertical_alignment.Fill
                                        Widget := stack_box:
                                            Orientation := orientation.Horizontal
                                            Slots := array:
                                                stack_box_slot:
                                                    Distribution:= option{0.5}
                                                    HorizontalAlignment := horizontal_alignment.Right
                                                    VerticalAlignment := vertical_alignment.Fill
                                                    Widget := CostTextBlock2
                                                stack_box_slot:
                                                    Distribution:= option{0.1}
                                                    HorizontalAlignment := horizontal_alignment.Center
                                                    VerticalAlignment := vertical_alignment.Fill
                                                    Widget := SiLverOreImage4
                                                stack_box_slot:
                                                    Distribution:= option{0.4}
                                                    HorizontalAlignment := horizontal_alignment.Left
                                                    VerticalAlignment := vertical_alignment.Fill
                                                    Widget := Cost_PetidleTextBlock
                                    stack_box_slot:
                                        Distribution:= option{0.2}
                                        HorizontalAlignment := horizontal_alignment.Center
                                        VerticalAlignment := vertical_alignment.Center
                                        Widget := QuietButtonLevelup2
        }

        #========================================== Close ==========================================

        CloseButton : button_loud = button_loud{DefaultText := StringToMessage("CLOSE")}
        CloseButton.OnClick().Subscribe(OnCloseButtonClicked)
        set CanvasSlots += array
        {
            canvas_slot:
                Anchors := CenterAnchors
                Offsets := margin{Top:= 451.0, Left:= 438.0, Right := 353.0, Bottom := 80.0}
                Alignment := CenterVector
                SizeToContent := false
                Widget := CloseButton
        }
        
        if(set UIElements[Agent] = (OreTextBlock, PowerupStoneTextBlock)){}
        if(set SkillUIElements[Agent] = map{}){}
        if:
            set SkillUIElements[Agent] = map:
                1 => (Lv_PlayeridleTextBlock, Now_PlayeridleTextBlock, Will_PlayeridleTextBlock, Cost_PlayeridleTextBlock, _PlayeridleTextBlock)
                2 => (Lv_PetidleTextBlock, Now_PetidleTextBlock, Will_PetidleTextBlock, Cost_PetidleTextBlock, _PetidleTextBlock)


        Canvas : canvas = canvas:
            Slots := CanvasSlots

        return Canvas