
using { /Verse.org/Assets }
using { /Verse.org/Simulation }
using { /Fortnite.com/UI }
using { /Fortnite.com/Devices }
using { /Fortnite.com/Characters }
using { /UnrealEngine.com/Temporary/UI }
using { /UnrealEngine.com/Temporary/SpatialMath }
using { /Verse.org/Simulation }
using { /UnrealEngine.com/Temporary/Diagnostics }
using { /Verse.org/Colors }
using { /Verse.org/Colors/NamedColors }
using { Utils }
using { PlayerData }
using { Skill }


# See https://dev.epicgames.com/documentation/en-us/uefn/create-your-own-device-in-verse for how to create a verse device.

# A Verse-authored creative device that can be placed in a level
ui_skill_miner1 := class(creative_device):
    #SFXs
    @editable
    SFX_Open : audio_player_device = audio_player_device{}
    @editable
    SFX_Close : audio_player_device = audio_player_device{}
    @editable
    SFX_LevelUp : audio_player_device = audio_player_device{}
    @editable
    SFX_Fail : audio_player_device = audio_player_device{}


    var _SaveData : save_data = save_data{}
    @editable
    var SaveDataManager : save_data_manager = save_data_manager{}
    @editable
    var SkillDataManager : skill_manager = skill_manager{}
    CloseEvent<public> : event(player) = event(player){}


    @editable
    QuestTrigger : trigger_device = trigger_device{}

    @editable
    SkillButton : button_device = button_device{}


    var miner1CanvasPerAgent : [agent]canvas = map{}
    var UIElements : [agent]tuple(text_block, text_block) = map{}
    var SkillUIElements : [agent][int]tuple(text_block, text_block, text_block, text_block, text_block) = map{}

    OnBegin<override>()<suspends>:void=
        SkillButton.InteractedWithEvent.Subscribe(OnButtonEvented)

    SetText<localizes>(Value:[]char) : message = "{Value}"

    # スキル関連のUI要素を全て更新
    UpdateUI(Agent:agent):void=
        UpdateUIElements(Agent)
        UpdateSkillUIElements(Agent)

    # UIElementsに格納された基本リソースのテキストを更新する
    UpdateUIElements(Agent: agent):void=
        var SilverValue : big_number = big_number{}
        var PowerupStoneValue : big_number = big_number{}
        if:
            Player := player[Agent]
        then:
            set SilverValue = SaveDataManager.GetResourceStone(Player)
            set PowerupStoneValue = SaveDataManager.GetResourcePowerupStone(Player)
            if (Elements := UIElements[Agent]):
                Elements(0).SetText(SetText("{BigNumber_FormatDisplay(SilverValue)}"))          # OreTextBlockのテキストを設定
                Elements(1).SetText(SetText("{BigNumber_FormatDisplay(PowerupStoneValue)}"))    # PowerupStoneTextBlockのテキストを設定 

    UpdateSkillUIElements(Agent:agent):void=
        var Lv_CriticalRate : int = 1   #15
        var Now_CriticalRate : int = 1
        var Will_CriticalRate : int = 1
        var Cost_CriticalRate : big_number = big_number{}
        var Lv_CriticalMultiplier : int = 1 #16
        var Now_CriticalMultiplier : int = 1
        var Will_CriticalMultiplier : int = 1
        var Cost_CriticalMultiplier : big_number = big_number{}
        var Lv_FinalHitBonus : int = 1  #17
        var Now_FinalHitBonus : int = 1
        var Will_FinalHitBonus : int = 1
        var Cost_FinalHitBonus : big_number = big_number{}
        var Lv_CriticalPowerSurge : int = 1 #18
        var Now_CriticalPowerSurge : int = 1
        var Will_CriticalPowerSurge : int = 1
        var Cost_CriticalPowerSurge : big_number = big_number{}
        if:
            Player := player[Agent]
        then:
            var LanguageIndex : int = SaveDataManager.GetLanguage(Player)
            if:
                set Lv_CriticalRate = SaveDataManager.GetSkillLevel[Player, 15]
                set Now_CriticalRate = SaveDataManager.GetPlayerSkillEffectIncrease[Player, 15]
                set Will_CriticalRate = SaveDataManager.GetSpecificSkillEffectIncrease[Player, 15]
                set Cost_CriticalRate = SaveDataManager.GetUpgradeCostForSkill[Player, 15]

                set Lv_CriticalMultiplier = SaveDataManager.GetSkillLevel[Player, 16]
                set Now_CriticalMultiplier = SaveDataManager.GetPlayerSkillEffectIncrease[Player, 16]
                set Will_CriticalMultiplier = SaveDataManager.GetSpecificSkillEffectIncrease[Player, 16]
                set Cost_CriticalMultiplier = SaveDataManager.GetUpgradeCostForSkill[Player, 16]

                set Lv_FinalHitBonus = SaveDataManager.GetSkillLevel[Player, 17]
                set Now_FinalHitBonus = SaveDataManager.GetPlayerSkillEffectIncrease[Player, 17]
                set Will_FinalHitBonus = SaveDataManager.GetSpecificSkillEffectIncrease[Player, 17]
                set Cost_FinalHitBonus = SaveDataManager.GetUpgradeCostForSkill[Player, 17]

                set Lv_CriticalPowerSurge = SaveDataManager.GetSkillLevel[Player, 18]
                set Now_CriticalPowerSurge = SaveDataManager.GetPlayerSkillEffectIncrease[Player, 18]
                set Will_CriticalPowerSurge = SaveDataManager.GetSpecificSkillEffectIncrease[Player, 18]
                set Cost_CriticalPowerSurge = SaveDataManager.GetUpgradeCostForSkill[Player, 18]
            then:
                LvCriticalRateText := if(SkillDataManager.CanUpgradeSkill(player[Agent], 15)?){"Lv.{Lv_CriticalRate}"} else {"Lv.MAX"}
                LvCriticalRateTextColor := if(SkillDataManager.CanUpgradeSkill(player[Agent], 15)?){NamedColors.White} else {NamedColors.Red}
                WillCriticalRateText := if(SkillDataManager.CanUpgradeSkill(player[Agent], 15)?){"+ {Will_CriticalRate} %"} else {"MAX!"}
                CostCriticalRateText := if(SkillDataManager.CanUpgradeSkill(player[Agent], 15)?){"{BigNumber_FormatDisplay(Cost_CriticalRate)}"} else {"-"}
                CostCriticalRateTextColor := if(SkillDataManager.IsUnlockable[player[Agent], 15]){NamedColors.White} else {NamedColors.Red}
                CCRateText := if(LanguageIndex = 0){"クリティカル確率上昇"} else {"Critical rate increase"}
                
                LvCriticalMultiplierText := if(SkillDataManager.CanUpgradeSkill(player[Agent], 16)?){"Lv.{Lv_CriticalMultiplier}"} else {"Lv.MAX"}
                LvCriticalMultiplierTextColor := if(SkillDataManager.CanUpgradeSkill(player[Agent], 16)?){NamedColors.White} else {NamedColors.Red}
                WillCriticalMultiplierText := if(SkillDataManager.CanUpgradeSkill(player[Agent], 16)?){"+ {Will_CriticalMultiplier} %"} else {"MAX!"}
                CostCriticalMultiplierText := if(SkillDataManager.CanUpgradeSkill(player[Agent], 16)?){"{BigNumber_FormatDisplay(Cost_CriticalMultiplier)}"} else {"-"}
                CostCriticalMultiplierTextColor := if(SkillDataManager.IsUnlockable[player[Agent], 16]){NamedColors.White} else {NamedColors.Red}
                CCMultiText := if(LanguageIndex = 0){"クリティカル倍率上昇"} else {"Critical multiplier increase"}

                LvFinalHitBonusText := if(SkillDataManager.CanUpgradeSkill(player[Agent], 17)?){"Lv.{Lv_FinalHitBonus}"} else {"Lv.MAX"}
                LvFinalHitBonusTextColor := if(SkillDataManager.CanUpgradeSkill(player[Agent], 17)?){NamedColors.White} else {NamedColors.Red}
                WillFinalHitBonusText := if(SkillDataManager.CanUpgradeSkill(player[Agent], 17)?){"+ {Will_FinalHitBonus} %"} else {"MAX!"}
                CostFinalHitBonusText := if(SkillDataManager.CanUpgradeSkill(player[Agent], 17)?){"{BigNumber_FormatDisplay(Cost_FinalHitBonus)}"} else {"-"}
                CostFinalHitBonusTextColor := if(SkillDataManager.IsUnlockable[player[Agent], 17]){NamedColors.White} else {NamedColors.Red}
                BonusText := if(LanguageIndex = 0){"ラストヒット時に貰える\nリソースが増える"} else {"Get more resources\non last hit"}

                LvCriticalPowerSurgeText := if(SkillDataManager.CanUpgradeSkill(player[Agent], 18)?){"Lv.{Lv_CriticalPowerSurge}"} else {"Lv.MAX"}
                LvAddStockTextColor := if(SkillDataManager.CanUpgradeSkill(player[Agent], 18)?){NamedColors.White} else {NamedColors.Red}
                WillCriticalPowerSurgeText := if(SkillDataManager.CanUpgradeSkill(player[Agent], 18)?){"+ {Will_CriticalPowerSurge} %"} else {"MAX!"}
                CostCriticalPowerSurgeText := if(SkillDataManager.CanUpgradeSkill(player[Agent], 18)?){"{BigNumber_FormatDisplay(Cost_CriticalPowerSurge)}"} else {"-"}
                CostCriticalPowerSurgeTextColor := if(SkillDataManager.IsUnlockable[player[Agent], 18]){NamedColors.White} else {NamedColors.Red}
                CCSurgeText := if(LanguageIndex = 0){"クリティカル時に\nよりダメージが増えるチャンスを得る"} else {"Chance for critical multiplier\nto increase on critical hits"}

                if(SkillElement := SkillUIElements[Agent][1]) :
                    SkillElement(0).SetText(SetText(LvCriticalRateText))
                    SkillElement(0).SetTextColor(LvCriticalRateTextColor)
                    SkillElement(1).SetText(SetText("{Now_CriticalRate} %"))
                    SkillElement(2).SetText(SetText(WillCriticalRateText))
                    SkillElement(3).SetText(SetText(CostCriticalRateText))
                    SkillElement(3).SetTextColor(CostCriticalRateTextColor)
                    SkillElement(4).SetText(SetText("{CCRateText}"))

                if(SkillElement := SkillUIElements[Agent][2]) :
                    SkillElement(0).SetText(SetText(LvCriticalMultiplierText))
                    SkillElement(0).SetTextColor(LvCriticalMultiplierTextColor)
                    SkillElement(1).SetText(SetText("{Now_CriticalMultiplier}%"))
                    SkillElement(2).SetText(SetText(WillCriticalMultiplierText))
                    SkillElement(3).SetText(SetText(CostCriticalMultiplierText))
                    SkillElement(3).SetTextColor(CostCriticalMultiplierTextColor)
                    SkillElement(4).SetText(SetText("{CCMultiText}"))
                    
                if(SkillElement := SkillUIElements[Agent][3]) :
                    SkillElement(0).SetText(SetText(LvFinalHitBonusText))
                    SkillElement(0).SetTextColor(LvFinalHitBonusTextColor)
                    SkillElement(1).SetText(SetText("{Now_FinalHitBonus}%"))
                    SkillElement(2).SetText(SetText(WillFinalHitBonusText))
                    SkillElement(3).SetText(SetText(CostFinalHitBonusText))
                    SkillElement(3).SetTextColor(CostFinalHitBonusTextColor)
                    SkillElement(4).SetText(SetText("{BonusText}"))

                if(SkillElement := SkillUIElements[Agent][4]) :
                    SkillElement(0).SetText(SetText(LvCriticalPowerSurgeText))
                    SkillElement(0).SetTextColor(LvAddStockTextColor)
                    SkillElement(1).SetText(SetText("{Now_CriticalPowerSurge}00%"))
                    SkillElement(2).SetText(SetText(WillCriticalPowerSurgeText))
                    SkillElement(3).SetText(SetText(CostCriticalPowerSurgeText))
                    SkillElement(3).SetTextColor(CostCriticalPowerSurgeTextColor)
                    SkillElement(4).SetText(SetText("{CCSurgeText}"))

    OnPlayerSpawned(Agent : agent):void=
        if (miner1Canvas := miner1CanvasPerAgent[Agent]):
            return

        if (Player := player[Agent], PlayerUI := GetPlayerUI[Player]):
            miner1Canvas : canvas = Createminer1Canvas(Agent)
            if (set miner1CanvasPerAgent[Agent] = miner1Canvas){}

    OnButtonEvented(Agent : agent):void=
        SFX_Open.Play(Agent)
        OnPlayerSpawned(Agent)
        UpdateUI(Agent)
        Print("OnSkillButtonEvented")
        if (Player := player[Agent], PlayerUI := GetPlayerUI[Player], miner1Canvas := miner1CanvasPerAgent[Agent]):
            PlayerUI.AddWidget(miner1Canvas, player_ui_slot{InputMode := ui_input_mode.All})

    # ボタン処理
    OnLevelup1Clicked(WidgetMessage : widget_message):void=
        if (PlayerUI := GetPlayerUI[WidgetMessage.Player], miner1Canvas := miner1CanvasPerAgent[WidgetMessage.Player], Agent : agent = agent[WidgetMessage.Player]):
            if(SkillDataManager.CanUpgradeSkill(WidgetMessage.Player,15) = true):
                SFX_LevelUp.Play(Agent)
                if:
                    SkillDataManager.UnlockOrUpgradeSkill[WidgetMessage.Player,15]
                then:
                    QuestTrigger.Trigger(WidgetMessage.Player)
                    UpdateUI(Agent)
            else:
                SFX_Fail.Play(Agent)
                return
            
    OnLevelup2Clicked(WidgetMessage : widget_message):void=
        if (PlayerUI := GetPlayerUI[WidgetMessage.Player], miner1Canvas := miner1CanvasPerAgent[WidgetMessage.Player], Agent : agent = agent[WidgetMessage.Player]):
            if(SkillDataManager.CanUpgradeSkill(WidgetMessage.Player,16) = true):
                SFX_LevelUp.Play(Agent)
                if:
                    SkillDataManager.UnlockOrUpgradeSkill[WidgetMessage.Player,16]
                then:
                    UpdateUI(Agent)
            else:
                SFX_Fail.Play(Agent)
                return

    OnLevelup3Clicked(WidgetMessage : widget_message):void=
        if (PlayerUI := GetPlayerUI[WidgetMessage.Player], miner1Canvas := miner1CanvasPerAgent[WidgetMessage.Player], Agent : agent = agent[WidgetMessage.Player]):
            if(SkillDataManager.CanUpgradeSkill(WidgetMessage.Player,17) = true):
                SFX_LevelUp.Play(Agent)
                if:
                    SkillDataManager.UnlockOrUpgradeSkill[WidgetMessage.Player,17]
                then:
                    UpdateUI(Agent)
            else:
                SFX_Fail.Play(Agent)
                return
            
    OnLevelup4Clicked(WidgetMessage : widget_message):void=
        if (PlayerUI := GetPlayerUI[WidgetMessage.Player], miner1Canvas := miner1CanvasPerAgent[WidgetMessage.Player], Agent : agent = agent[WidgetMessage.Player]):
            if(SkillDataManager.CanUpgradeSkill(WidgetMessage.Player,18) = true):
                SFX_LevelUp.Play(Agent)
                if:
                    SkillDataManager.UnlockOrUpgradeSkill[WidgetMessage.Player,18]
                then:
                    UpdateUI(Agent)
            else:
                SFX_Fail.Play(Agent)
                return

    OnCloseButtonClicked(WidgetMessage : widget_message):void=
        if(Agent : agent = agent[WidgetMessage.Player]) :
            SFX_Close.Play(Agent)
        if (PlayerUI := GetPlayerUI[WidgetMessage.Player], miner1Canvas := miner1CanvasPerAgent[WidgetMessage.Player]):
            PlayerUI.RemoveWidget(miner1Canvas)
            CloseEvent.Signal(WidgetMessage.Player)

    # キャンバス
    Createminer1Canvas(Agent : agent):canvas=
        var CanvasSlots : []canvas_slot = array{}
        
        CenterVector : vector2 = vector2{X:=0.5, Y:=0.5}
        CenterAnchors : anchors = anchors{Minimum := CenterVector, Maximum := CenterVector}

        # Image Stock
        BGImage : texture_block = texture_block{DefaultImage := Textures.BG_Skill, DefaultDesiredSize := vector2{X := 1920.0, Y := 1080.0}}
        SilverOreImage1 : texture_block = texture_block{DefaultImage := Textures.SilverOre, DefaultDesiredSize := vector2{X := 50.0, Y := 50.0}}
        PowerupStoneImage : texture_block = texture_block{DefaultImage := Textures.IC_PowerupStone, DefaultDesiredSize := vector2{X := 50.0, Y := 50.0}}
        SiLverOreImage3 : texture_block = texture_block{DefaultImage := Textures.SilverOre, DefaultDesiredSize := vector2{X := 40.0, Y := 40.0}}
        SiLverOreImage4 : texture_block = texture_block{DefaultImage := Textures.SilverOre, DefaultDesiredSize := vector2{X := 40.0, Y := 40.0}}
        SiLverOreImage5 : texture_block = texture_block{DefaultImage := Textures.SilverOre, DefaultDesiredSize := vector2{X := 40.0, Y := 40.0}}
        SiLverOreImage6 : texture_block = texture_block{DefaultImage := Textures.SilverOre, DefaultDesiredSize := vector2{X := 40.0, Y := 40.0}}
        CriticalRateImage : texture_block = texture_block{DefaultImage := Textures.MinerSkill1_1, DefaultDesiredSize := vector2{X := 200.0, Y := 200.0}}
        CriticalMultiplierImage : texture_block = texture_block{DefaultImage := Textures.MinerSkill1_2, DefaultDesiredSize := vector2{X := 200.0, Y := 200.0}}
        FinalHitBonusImage : texture_block = texture_block{DefaultImage := Textures.MinerSkill1_3, DefaultDesiredSize := vector2{X := 200.0, Y := 200.0}}
        CriticalPowerSurgeImage : texture_block = texture_block{DefaultImage := Textures.MinerSkill1_4, DefaultDesiredSize := vector2{X := 200.0, Y := 200.0}}
            
        var SilverValue : big_number = big_number{}
        var PowerupStoneValue : big_number = big_number{}
        var Lv_CriticalRate : int = 1   #15
        var Now_CriticalRate : int = 1
        var Will_CriticalRate : int = 1
        var Cost_CriticalRate : big_number = big_number{}
        var Lv_CriticalMultiplier : int = 1 #16
        var Now_CriticalMultiplier : int = 1
        var Will_CriticalMultiplier : int = 1
        var Cost_CriticalMultiplier : big_number = big_number{}
        var Lv_FinalHitBonus : int = 1  #17
        var Now_FinalHitBonus : int = 1
        var Will_FinalHitBonus : int = 1
        var Cost_FinalHitBonus : big_number = big_number{}
        var Lv_CriticalPowerSurge : int = 1 #18
        var Now_CriticalPowerSurge : int = 1
        var Will_CriticalPowerSurge : int = 1
        var Cost_CriticalPowerSurge : big_number = big_number{}
        
        #================================== TextBlock ==================================
        
        # 所持岩
        OreTextBlock : text_block = text_block{
            DefaultTextColor := NamedColors.White,
            DefaultShadowOffset := option{vector2{X:=2.5, Y:=2.5}},
            DefaultShadowColor := NamedColors.Black
        }
        # 所持強化アイテム
        PowerupStoneTextBlock : text_block = text_block{
            DefaultTextColor := NamedColors.White,
            DefaultShadowOffset := option{vector2{X:=2.5, Y:=2.5}},
            DefaultShadowColor := NamedColors.Black,
            DefaultJustification := text_justification.Center
        }

        #================================== Skill 1 ==================================

        Lv_CriticalRateTextBlock : text_block = text_block{
            DefaultTextColor := NamedColors.White,
            DefaultShadowOffset := option{vector2{X:=2.5, Y:=2.5}},
            DefaultShadowColor := NamedColors.Black,
            DefaultJustification := text_justification.Center
        }
        Now_CriticalRateTextBlock : text_block = text_block{
            DefaultTextColor := NamedColors.White,
            DefaultShadowOffset := option{vector2{X:=2.5, Y:=2.5}},
            DefaultShadowColor := NamedColors.Black,
            DefaultJustification := text_justification.Center
        }
        Will_CriticalRateTextBlock : text_block = text_block{
            DefaultTextColor := NamedColors.White,
            DefaultShadowOffset := option{vector2{X:=2.5, Y:=2.5}},
            DefaultShadowColor := NamedColors.Black,
            DefaultJustification := text_justification.Center
        }
        Cost_CriticalRateTextBlock : text_block = text_block{
            DefaultTextColor := NamedColors.White,
            DefaultShadowOffset := option{vector2{X:=2.5, Y:=2.5}},
            DefaultShadowColor := NamedColors.Black,
            DefaultJustification := text_justification.Center
        }
        var _CriticalRateTextBlock : text_block = text_block{
            DefaultTextColor := NamedColors.White,
            DefaultShadowOffset := option{vector2{X:=2.5, Y:=2.5}},
            DefaultShadowColor := NamedColors.Black,
            DefaultShadowOpacity := 1.0,
            DefaultJustification := text_justification.Center
        }

        #================================== Skill 2 ==================================

        Lv_CriticalMultiplierTextBlock : text_block = text_block{
            DefaultTextColor := NamedColors.White,
            DefaultShadowOffset := option{vector2{X:=2.5, Y:=2.5}},
            DefaultShadowColor := NamedColors.Black,
            DefaultJustification := text_justification.Center
        }
        Now_CriticalMultiplierTextBlock : text_block = text_block{
            DefaultTextColor := NamedColors.White,
            DefaultShadowOffset := option{vector2{X:=2.5, Y:=2.5}},
            DefaultShadowColor := NamedColors.Black,
            DefaultJustification := text_justification.Center
        }
        Will_CriticalMultiplierTextBlock : text_block = text_block{
            DefaultTextColor := NamedColors.White,
            DefaultShadowOffset := option{vector2{X:=2.5, Y:=2.5}},
            DefaultShadowColor := NamedColors.Black,
            DefaultJustification := text_justification.Center
        }
        Cost_CriticalMultiplierTextBlock : text_block = text_block{
            DefaultTextColor := NamedColors.White,
            DefaultShadowOffset := option{vector2{X:=2.5, Y:=2.5}},
            DefaultShadowColor := NamedColors.Black,
            DefaultJustification := text_justification.Center
        }
        var _CriticalMultiplierTextBlock :text_block= text_block{
            DefaultTextColor := NamedColors.White,
            DefaultShadowOffset := option{vector2{X:=2.5, Y:=2.5}},
            DefaultShadowColor := NamedColors.Black,
            DefaultShadowOpacity := 1.0,
            DefaultJustification := text_justification.Center
        }
        
        #================================== Skill 3 ==================================

        Lv_FinalHitBonusTextBlock : text_block = text_block{
            DefaultTextColor := NamedColors.White,
            DefaultShadowOffset := option{vector2{X:=2.5, Y:=2.5}},
            DefaultShadowColor := NamedColors.Black,
            DefaultJustification := text_justification.Center
        }
        Now_FinalHitBonusTextBlock : text_block = text_block{
            DefaultTextColor := NamedColors.White,
            DefaultShadowOffset := option{vector2{X:=2.5, Y:=2.5}},
            DefaultShadowColor := NamedColors.Black,
            DefaultJustification := text_justification.Center
        }
        Will_FinalHitBonusTextBlock : text_block = text_block{
            DefaultTextColor := NamedColors.White,
            DefaultShadowOffset := option{vector2{X:=2.5, Y:=2.5}},
            DefaultShadowColor := NamedColors.Black,
            DefaultJustification := text_justification.Center
        }
        Cost_FinalHitBonusTextBlock : text_block = text_block{
            DefaultTextColor := NamedColors.White,
            DefaultShadowOffset := option{vector2{X:=2.5, Y:=2.5}},
            DefaultShadowColor := NamedColors.Black,
            DefaultJustification := text_justification.Center
        }
        var _FinalHitBonusTextBlock :text_block= text_block{
            DefaultTextColor := NamedColors.White,
            DefaultShadowOffset := option{vector2{X:=2.5, Y:=2.5}},
            DefaultShadowColor := NamedColors.Black,
            DefaultShadowOpacity := 1.0,
            DefaultJustification := text_justification.Center,
            DefaultOverflowPolicy := text_overflow_policy.Ellipsis
        }
        
        #================================== Skill 4 ==================================

        Lv_CriticalPowerSurgeTextBlock : text_block = text_block{
            DefaultTextColor := NamedColors.White,
            DefaultShadowOffset := option{vector2{X:=2.5, Y:=2.5}},
            DefaultShadowColor := NamedColors.Black,
            DefaultJustification := text_justification.Center
        }
        Now_CriticalPowerSurgeTextBlock : text_block = text_block{
            DefaultTextColor := NamedColors.White,
            DefaultShadowOffset := option{vector2{X:=2.5, Y:=2.5}},
            DefaultShadowColor := NamedColors.Black,
            DefaultJustification := text_justification.Center
        }
        Will_CriticalPowerSurgeTextBlock : text_block = text_block{
            DefaultTextColor := NamedColors.White,
            DefaultShadowOffset := option{vector2{X:=2.5, Y:=2.5}},
            DefaultShadowColor := NamedColors.Black,
            DefaultJustification := text_justification.Center
        }
        Cost_CriticalPowerSurgeTextBlock : text_block = text_block{
            DefaultTextColor := NamedColors.White,
            DefaultShadowOffset := option{vector2{X:=2.5, Y:=2.5}},
            DefaultShadowColor := NamedColors.Black,
            DefaultJustification := text_justification.Center
        }
        var _CriticalPowerSurgeTextBlock :text_block= text_block{
            DefaultTextColor := NamedColors.White,
            DefaultShadowOffset := option{vector2{X:=2.5, Y:=2.5}},
            DefaultShadowColor := NamedColors.Black,
            DefaultShadowOpacity := 1.0,
            DefaultJustification := text_justification.Center
        }

        #================================== /TextBlock ==================================
        
        #========================================== BG ==========================================

        set CanvasSlots += array
        {
            canvas_slot:
                Anchors := CenterAnchors
                Offsets := margin{Top := 0.0,Left := 0.0, Right := 1920.0, Bottom := 1080.0}
                Alignment := CenterVector
                SizeToContent := false
                Widget := BGImage
        }

        #========================================= Header ========================================

        miner1TextBlock := text_block{
            DefaultText := StringToMessage("Miner Skill 1"),
            DefaultTextColor := NamedColors.White,
            DefaultShadowOffset := option{vector2{X:=2.5, Y:=2.5}},
            DefaultShadowColor := NamedColors.Black,
            DefaultJustification := text_justification.Center
        }
        miner1TextBlock.SetShadowOpacity(1.0)
        set CanvasSlots += array
        {
            canvas_slot:
                Anchors := CenterAnchors
                Offsets := margin{Top:= -472.0, Left:= 0.0, Right := 257.0, Bottom := 41.0}
                Alignment := CenterVector
                SizeToContent := false
                Widget := miner1TextBlock
        }

        OreTextBlock.SetShadowOpacity(1.0)
        PowerupStoneTextBlock.SetShadowOpacity(1.0)

        set CanvasSlots += array

        {
            canvas_slot:
                Anchors := CenterAnchors
                Offsets := margin{Left:=-645.5 ,Top:= -391.0, Right:= 519.0, Bottom:= 59.0}
                Alignment := CenterVector
                SizeToContent := false
                Widget := stack_box:
                    Orientation := orientation.Horizontal
                    Slots := array:
                        stack_box_slot:
                            Distribution:= option{0.5}
                            Padding:=margin{Left:=5.0 ,Top:= 5.0, Right:= 10.0, Bottom:= 5.0}
                            HorizontalAlignment := horizontal_alignment.Fill
                            VerticalAlignment := vertical_alignment.Fill
                            Widget := stack_box:
                                Orientation := orientation.Horizontal
                                Slots := array:
                                    stack_box_slot:
                                        HorizontalAlignment := horizontal_alignment.Center
                                        VerticalAlignment := vertical_alignment.Center
                                        Widget := SilverOreImage1
                                    stack_box_slot:
                                        Distribution:= option{1.0}
                                        HorizontalAlignment := horizontal_alignment.Center
                                        VerticalAlignment := vertical_alignment.Center
                                        Widget := OreTextBlock
                        stack_box_slot:
                            Padding:=margin{Left:=10.0 ,Top:= 5.0, Right:= 5.0, Bottom:= 5.0}
                            Distribution:= option{0.5}
                            HorizontalAlignment := horizontal_alignment.Fill
                            VerticalAlignment := vertical_alignment.Fill
                            Widget := stack_box:
                                Orientation := orientation.Horizontal
                                Slots := array:
                                    stack_box_slot:
                                        HorizontalAlignment := horizontal_alignment.Center
                                        VerticalAlignment := vertical_alignment.Center
                                        Widget := PowerupStoneImage
                                    stack_box_slot:
                                        Distribution:= option{1.0}
                                        HorizontalAlignment := horizontal_alignment.Center
                                        VerticalAlignment := vertical_alignment.Center
                                        Widget := PowerupStoneTextBlock
        }

        #========================================== Skill ==========================================

        CostTextBlock1 := text_block{
            DefaultText := StringToMessage("Cost :"),
            DefaultTextColor := NamedColors.White,
            DefaultShadowOffset := option{vector2{X:=2.5, Y:=2.5}},
            DefaultShadowColor := NamedColors.Black,
            DefaultJustification := text_justification.Center
        }
        CostTextBlock1.SetShadowOpacity(1.0)
        CostTextBlock2 := text_block{
            DefaultText := StringToMessage("Cost :"),
            DefaultTextColor := NamedColors.White,
            DefaultShadowOffset := option{vector2{X:=2.5, Y:=2.5}},
            DefaultShadowColor := NamedColors.Black,
            DefaultJustification := text_justification.Center
        }
        CostTextBlock2.SetShadowOpacity(1.0)
        CostTextBlock3 := text_block{
            DefaultText := StringToMessage("Cost :"),
            DefaultTextColor := NamedColors.White,
            DefaultShadowOffset := option{vector2{X:=2.5, Y:=2.5}},
            DefaultShadowColor := NamedColors.Black,
            DefaultJustification := text_justification.Center
        }
        CostTextBlock3.SetShadowOpacity(1.0)
        CostTextBlock4 := text_block{
            DefaultText := StringToMessage("Cost :"),
            DefaultTextColor := NamedColors.White,
            DefaultShadowOffset := option{vector2{X:=2.5, Y:=2.5}},
            DefaultShadowColor := NamedColors.Black,
            DefaultJustification := text_justification.Center
        }
        CostTextBlock4.SetShadowOpacity(1.0)

        #Skill 1
        CriticalRateTextBlock := text_block{
            DefaultText := StringToMessage("Critical rate"),
            DefaultTextColor := NamedColors.White,
            DefaultShadowOffset := option{vector2{X:=2.5, Y:=2.5}},
            DefaultShadowColor := NamedColors.Black,
            DefaultJustification := text_justification.Center
        }
        CriticalRateTextBlock.SetShadowOpacity(1.0)
        Lv_CriticalRateTextBlock.SetShadowOpacity(1.0)
        Now_CriticalRateTextBlock.SetShadowOpacity(1.0)
        Will_CriticalRateTextBlock.SetShadowOpacity(1.0)
        Cost_CriticalRateTextBlock.SetShadowOpacity(1.0)
        
        #Skill 2
        CriticalMultiplierTextBlock := text_block{
            DefaultText := StringToMessage("Critical multiplier"),
            DefaultTextColor := NamedColors.White,
            DefaultShadowOffset := option{vector2{X:=2.5, Y:=2.5}},
            DefaultShadowColor := NamedColors.Black,
            DefaultJustification := text_justification.Center
        }
        CriticalMultiplierTextBlock.SetShadowOpacity(1.0)
        Lv_CriticalMultiplierTextBlock.SetShadowOpacity(1.0)
        Now_CriticalMultiplierTextBlock.SetShadowOpacity(1.0)
        Will_CriticalMultiplierTextBlock.SetShadowOpacity(1.0)
        Cost_CriticalMultiplierTextBlock.SetShadowOpacity(1.0)
        
        #Skill 3
        FinalHitBonusTextBlock := text_block{
            DefaultText := StringToMessage("Final hit bonus"),
            DefaultTextColor := NamedColors.White,
            DefaultShadowOffset := option{vector2{X:=2.5, Y:=2.5}},
            DefaultShadowColor := NamedColors.Black,
            DefaultJustification := text_justification.Center
        }
        FinalHitBonusTextBlock.SetShadowOpacity(1.0)
        Lv_FinalHitBonusTextBlock.SetShadowOpacity(1.0)
        Now_FinalHitBonusTextBlock.SetShadowOpacity(1.0)
        Will_FinalHitBonusTextBlock.SetShadowOpacity(1.0)
        Cost_FinalHitBonusTextBlock.SetShadowOpacity(1.0)
        
        #Skill 4
        CriticalPowerSurgeTextBlock := text_block{
            DefaultText := StringToMessage("Critical power surge"),
            DefaultTextColor := NamedColors.White,
            DefaultShadowOffset := option{vector2{X:=2.5, Y:=2.5}},
            DefaultShadowColor := NamedColors.Black,
            DefaultJustification := text_justification.Center
        }
        CriticalPowerSurgeTextBlock.SetShadowOpacity(1.0)
        Lv_CriticalPowerSurgeTextBlock.SetShadowOpacity(1.0)
        Now_CriticalPowerSurgeTextBlock.SetShadowOpacity(1.0)
        Will_CriticalPowerSurgeTextBlock.SetShadowOpacity(1.0)
        Cost_CriticalPowerSurgeTextBlock.SetShadowOpacity(1.0)

        # Button Stock
        QuietButtonLevelup1 : button_quiet = button_quiet{DefaultText := StringToMessage("LEVEL UP")}
        QuietButtonLevelup1.OnClick().Subscribe(OnLevelup1Clicked)
        QuietButtonLevelup2 : button_quiet = button_quiet{DefaultText := StringToMessage("LEVEL UP")}
        QuietButtonLevelup2.OnClick().Subscribe(OnLevelup2Clicked)
        QuietButtonLevelup3 : button_quiet = button_quiet{DefaultText := StringToMessage("LEVEL UP")}
        QuietButtonLevelup3.OnClick().Subscribe(OnLevelup3Clicked)
        QuietButtonLevelup4 : button_quiet = button_quiet{DefaultText := StringToMessage("LEVEL UP")}
        QuietButtonLevelup4.OnClick().Subscribe(OnLevelup4Clicked)

        #================================ Top ================================

        set CanvasSlots += array
        {
            canvas_slot:
                Anchors := CenterAnchors
                Offsets := margin{Left:=0.0 ,Top:= -96.5, Right:= 1760.0, Bottom:= 464.0}
                Alignment := CenterVector
                SizeToContent := false
                Widget := stack_box:
                    Orientation := orientation.Horizontal
                    Slots := array:
                        stack_box_slot:
                            Padding:=margin{Left:= 50.0, Right:= 50.0}
                            Distribution:= option{0.25}
                            HorizontalAlignment := horizontal_alignment.Fill
                            VerticalAlignment := vertical_alignment.Fill
                            Widget := stack_box:
                                Orientation := orientation.Vertical
                                Slots := array:
                                    stack_box_slot:
                                        Padding:=margin{Top:= 10.0, Bottom:= 10.0}
                                        HorizontalAlignment := horizontal_alignment.Fill
                                        VerticalAlignment := vertical_alignment.Fill
                                        Widget := CriticalRateTextBlock
                                    stack_box_slot:
                                        Padding:=margin{Top:= 15.0, Bottom:= 15.0}
                                        HorizontalAlignment := horizontal_alignment.Fill
                                        VerticalAlignment := vertical_alignment.Fill
                                        Widget := Lv_CriticalRateTextBlock
                                    stack_box_slot:
                                        Padding:=margin{Top:= 10.0, Bottom:= 10.0}
                                        HorizontalAlignment := horizontal_alignment.Center
                                        VerticalAlignment := vertical_alignment.Center
                                        Widget := CriticalRateImage
                                    stack_box_slot:
                                        Padding:=margin{Top:= 15.0, Bottom:= 15.0}
                                        HorizontalAlignment := horizontal_alignment.Fill
                                        VerticalAlignment := vertical_alignment.Fill
                                        Widget := _CriticalRateTextBlock
                        stack_box_slot:
                            Padding:=margin{Left:= 50.0, Right:= 50.0}
                            Distribution:= option{0.25}
                            HorizontalAlignment := horizontal_alignment.Fill
                            VerticalAlignment := vertical_alignment.Fill
                            Widget := stack_box:
                                Orientation := orientation.Vertical
                                Slots := array:
                                    stack_box_slot:
                                        Padding:=margin{Top:= 10.0, Bottom:= 10.0}
                                        HorizontalAlignment := horizontal_alignment.Fill
                                        VerticalAlignment := vertical_alignment.Fill
                                        Widget := CriticalMultiplierTextBlock
                                    stack_box_slot:
                                        Padding:=margin{Top:= 15.0, Bottom:= 15.0}
                                        HorizontalAlignment := horizontal_alignment.Fill
                                        VerticalAlignment := vertical_alignment.Fill
                                        Widget := Lv_CriticalMultiplierTextBlock
                                    stack_box_slot:
                                        Padding:=margin{Top:= 10.0, Bottom:= 10.0}
                                        HorizontalAlignment := horizontal_alignment.Center
                                        VerticalAlignment := vertical_alignment.Center
                                        Widget := CriticalMultiplierImage
                                    stack_box_slot:
                                        Padding:=margin{Top:= 15.0, Bottom:= 15.0}
                                        HorizontalAlignment := horizontal_alignment.Fill
                                        VerticalAlignment := vertical_alignment.Fill
                                        Widget := _CriticalMultiplierTextBlock
                        stack_box_slot:
                            Padding:=margin{Left:= 50.0, Right:= 50.0}
                            Distribution:= option{0.25}
                            HorizontalAlignment := horizontal_alignment.Fill
                            VerticalAlignment := vertical_alignment.Fill
                            Widget := stack_box:
                                Orientation := orientation.Vertical
                                Slots := array:
                                    stack_box_slot:
                                        Padding:=margin{Top:= 10.0, Bottom:= 10.0}
                                        HorizontalAlignment := horizontal_alignment.Fill
                                        VerticalAlignment := vertical_alignment.Fill
                                        Widget := FinalHitBonusTextBlock
                                    stack_box_slot:
                                        Padding:=margin{Top:= 15.0, Bottom:= 15.0}
                                        HorizontalAlignment := horizontal_alignment.Fill
                                        VerticalAlignment := vertical_alignment.Fill
                                        Widget := Lv_FinalHitBonusTextBlock
                                    stack_box_slot:
                                        Padding:=margin{Top:= 10.0, Bottom:= 10.0}
                                        HorizontalAlignment := horizontal_alignment.Center
                                        VerticalAlignment := vertical_alignment.Center
                                        Widget := FinalHitBonusImage
                                    stack_box_slot:
                                        Padding:=margin{Top:= 15.0, Bottom:= 15.0}
                                        HorizontalAlignment := horizontal_alignment.Fill
                                        VerticalAlignment := vertical_alignment.Fill
                                        Widget := _FinalHitBonusTextBlock
                        stack_box_slot:
                            Padding:=margin{Left:= 50.0, Right:= 50.0}
                            Distribution:= option{0.25}
                            HorizontalAlignment := horizontal_alignment.Fill
                            VerticalAlignment := vertical_alignment.Fill
                            Widget := stack_box:
                                Orientation := orientation.Vertical
                                Slots := array:
                                    stack_box_slot:
                                        Padding:=margin{Top:= 10.0, Bottom:= 10.0}
                                        HorizontalAlignment := horizontal_alignment.Fill
                                        VerticalAlignment := vertical_alignment.Fill
                                        Widget := CriticalPowerSurgeTextBlock
                                    stack_box_slot:
                                        Padding:=margin{Top:= 15.0, Bottom:= 15.0}
                                        HorizontalAlignment := horizontal_alignment.Fill
                                        VerticalAlignment := vertical_alignment.Fill
                                        Widget := Lv_CriticalPowerSurgeTextBlock
                                    stack_box_slot:
                                        Padding:=margin{Top:= 10.0, Bottom:= 10.0}
                                        HorizontalAlignment := horizontal_alignment.Center
                                        VerticalAlignment := vertical_alignment.Center
                                        Widget := CriticalPowerSurgeImage
                                    stack_box_slot:
                                        Padding:=margin{Top:= 15.0, Bottom:= 15.0}
                                        HorizontalAlignment := horizontal_alignment.Fill
                                        VerticalAlignment := vertical_alignment.Fill
                                        Widget := _CriticalPowerSurgeTextBlock
        }

        
        #================================ Bottom ================================

        set CanvasSlots += array
        {
            canvas_slot:
                Anchors := CenterAnchors
                Offsets := margin{Left:=0.0 ,Top:= 249.5, Right:= 1760.0, Bottom:= 210.0}
                Alignment := CenterVector
                SizeToContent := false
                Widget := stack_box:
                    Orientation := orientation.Horizontal
                    Slots := array:
                        stack_box_slot:
                            Distribution:= option{0.25}
                            HorizontalAlignment := horizontal_alignment.Fill
                            VerticalAlignment := vertical_alignment.Fill
                            Widget := stack_box:
                                Orientation := orientation.Vertical
                                Slots := array:
                                    stack_box_slot:
                                        Padding:=margin{Top:= 10.0, Bottom:= 10.0}
                                        HorizontalAlignment := horizontal_alignment.Fill
                                        VerticalAlignment := vertical_alignment.Fill
                                        Widget := stack_box:
                                            Orientation := orientation.Horizontal
                                            Slots := array:
                                                stack_box_slot:
                                                    Padding:=margin{Top:= 10.0, Right:= 5.0, Bottom:= 10.0}
                                                    Distribution:= option{0.5}
                                                    HorizontalAlignment := horizontal_alignment.Right
                                                    VerticalAlignment := vertical_alignment.Fill
                                                    Widget := Now_CriticalRateTextBlock
                                                stack_box_slot:
                                                    Padding:=margin{Top:= 10.0, Left:= 5.0, Bottom:= 10.0}
                                                    Distribution:= option{0.5}
                                                    HorizontalAlignment := horizontal_alignment.Left
                                                    VerticalAlignment := vertical_alignment.Fill
                                                    Widget := Will_CriticalRateTextBlock
                                    stack_box_slot:
                                        Padding:=margin{Top:= 10.0, Bottom:= 10.0}
                                        HorizontalAlignment := horizontal_alignment.Fill
                                        VerticalAlignment := vertical_alignment.Fill
                                        Widget := stack_box:
                                            Orientation := orientation.Horizontal
                                            Slots := array:
                                                stack_box_slot:
                                                    Distribution:= option{0.5}
                                                    HorizontalAlignment := horizontal_alignment.Right
                                                    VerticalAlignment := vertical_alignment.Fill
                                                    Widget := CostTextBlock1
                                                stack_box_slot:
                                                    Distribution:= option{0.1}
                                                    HorizontalAlignment := horizontal_alignment.Center
                                                    VerticalAlignment := vertical_alignment.Fill
                                                    Widget := SiLverOreImage3
                                                stack_box_slot:
                                                    Distribution:= option{0.4}
                                                    HorizontalAlignment := horizontal_alignment.Left
                                                    VerticalAlignment := vertical_alignment.Fill
                                                    Widget := Cost_CriticalRateTextBlock
                                    stack_box_slot:
                                        Distribution:= option{0.2}
                                        HorizontalAlignment := horizontal_alignment.Center
                                        VerticalAlignment := vertical_alignment.Center
                                        Widget := QuietButtonLevelup1
                        stack_box_slot:
                            Distribution:= option{0.25}
                            HorizontalAlignment := horizontal_alignment.Fill
                            VerticalAlignment := vertical_alignment.Fill
                            Widget := stack_box:
                                Orientation := orientation.Vertical
                                Slots := array:
                                    stack_box_slot:
                                        Padding:=margin{Top:= 10.0, Bottom:= 10.0}
                                        HorizontalAlignment := horizontal_alignment.Fill
                                        VerticalAlignment := vertical_alignment.Fill
                                        Widget := stack_box:
                                            Orientation := orientation.Horizontal
                                            Slots := array:
                                                stack_box_slot:
                                                    Padding:=margin{Top:= 10.0, Right:= 5.0, Bottom:= 10.0}
                                                    Distribution:= option{0.5}
                                                    HorizontalAlignment := horizontal_alignment.Right
                                                    VerticalAlignment := vertical_alignment.Fill
                                                    Widget := Now_CriticalMultiplierTextBlock
                                                stack_box_slot:
                                                    Padding:=margin{Top:= 10.0, Left:= 5.0, Bottom:= 10.0}
                                                    Distribution:= option{0.5}
                                                    HorizontalAlignment := horizontal_alignment.Left
                                                    VerticalAlignment := vertical_alignment.Fill
                                                    Widget := Will_CriticalMultiplierTextBlock
                                    stack_box_slot:
                                        Padding:=margin{Top:= 10.0, Bottom:= 10.0}
                                        HorizontalAlignment := horizontal_alignment.Fill
                                        VerticalAlignment := vertical_alignment.Fill
                                        Widget := stack_box:
                                            Orientation := orientation.Horizontal
                                            Slots := array:
                                                stack_box_slot:
                                                    Distribution:= option{0.5}
                                                    HorizontalAlignment := horizontal_alignment.Right
                                                    VerticalAlignment := vertical_alignment.Fill
                                                    Widget := CostTextBlock2
                                                stack_box_slot:
                                                    Distribution:= option{0.1}
                                                    HorizontalAlignment := horizontal_alignment.Center
                                                    VerticalAlignment := vertical_alignment.Fill
                                                    Widget := SiLverOreImage4
                                                stack_box_slot:
                                                    Distribution:= option{0.4}
                                                    HorizontalAlignment := horizontal_alignment.Left
                                                    VerticalAlignment := vertical_alignment.Fill
                                                    Widget := Cost_CriticalMultiplierTextBlock
                                    stack_box_slot:
                                        Distribution:= option{0.2}
                                        HorizontalAlignment := horizontal_alignment.Center
                                        VerticalAlignment := vertical_alignment.Center
                                        Widget := QuietButtonLevelup2

                        stack_box_slot:
                            Distribution:= option{0.25}
                            HorizontalAlignment := horizontal_alignment.Fill
                            VerticalAlignment := vertical_alignment.Fill
                            Widget := stack_box:
                                Orientation := orientation.Vertical
                                Slots := array:
                                    stack_box_slot:
                                        Padding:=margin{Top:= 10.0, Bottom:= 10.0}
                                        HorizontalAlignment := horizontal_alignment.Fill
                                        VerticalAlignment := vertical_alignment.Fill
                                        Widget := stack_box:
                                            Orientation := orientation.Horizontal
                                            Slots := array:
                                                stack_box_slot:
                                                    Padding:=margin{Top:= 10.0, Right:= 5.0, Bottom:= 10.0}
                                                    Distribution:= option{0.5}
                                                    HorizontalAlignment := horizontal_alignment.Right
                                                    VerticalAlignment := vertical_alignment.Fill
                                                    Widget := Now_FinalHitBonusTextBlock
                                                stack_box_slot:
                                                    Padding:=margin{Top:= 10.0, Left:= 5.0, Bottom:= 10.0}
                                                    Distribution:= option{0.5}
                                                    HorizontalAlignment := horizontal_alignment.Left
                                                    VerticalAlignment := vertical_alignment.Fill
                                                    Widget := Will_FinalHitBonusTextBlock
                                    stack_box_slot:
                                        Padding:=margin{Top:= 10.0, Bottom:= 10.0}
                                        HorizontalAlignment := horizontal_alignment.Fill
                                        VerticalAlignment := vertical_alignment.Fill
                                        Widget := stack_box:
                                            Orientation := orientation.Horizontal
                                            Slots := array:
                                                stack_box_slot:
                                                    Distribution:= option{0.5}
                                                    HorizontalAlignment := horizontal_alignment.Right
                                                    VerticalAlignment := vertical_alignment.Fill
                                                    Widget := CostTextBlock3
                                                stack_box_slot:
                                                    Distribution:= option{0.1}
                                                    HorizontalAlignment := horizontal_alignment.Center
                                                    VerticalAlignment := vertical_alignment.Fill
                                                    Widget := SiLverOreImage5
                                                stack_box_slot:
                                                    Distribution:= option{0.4}
                                                    HorizontalAlignment := horizontal_alignment.Left
                                                    VerticalAlignment := vertical_alignment.Fill
                                                    Widget := Cost_FinalHitBonusTextBlock
                                    stack_box_slot:
                                        Distribution:= option{0.2}
                                        HorizontalAlignment := horizontal_alignment.Center
                                        VerticalAlignment := vertical_alignment.Center
                                        Widget := QuietButtonLevelup3
                        stack_box_slot:
                            Distribution:= option{0.25}
                            HorizontalAlignment := horizontal_alignment.Fill
                            VerticalAlignment := vertical_alignment.Fill
                            Widget := stack_box:
                                Orientation := orientation.Vertical
                                Slots := array:
                                    stack_box_slot:
                                        Padding:=margin{Top:= 10.0, Bottom:= 10.0}
                                        HorizontalAlignment := horizontal_alignment.Fill
                                        VerticalAlignment := vertical_alignment.Fill
                                        Widget := stack_box:
                                            Orientation := orientation.Horizontal
                                            Slots := array:
                                                stack_box_slot:
                                                    Padding:=margin{Top:= 10.0, Right:= 5.0, Bottom:= 10.0}
                                                    Distribution:= option{0.5}
                                                    HorizontalAlignment := horizontal_alignment.Right
                                                    VerticalAlignment := vertical_alignment.Fill
                                                    Widget := Now_CriticalPowerSurgeTextBlock
                                                stack_box_slot:
                                                    Padding:=margin{Top:= 10.0, Left:= 5.0, Bottom:= 10.0}
                                                    Distribution:= option{0.5}
                                                    HorizontalAlignment := horizontal_alignment.Left
                                                    VerticalAlignment := vertical_alignment.Fill
                                                    Widget := Will_CriticalPowerSurgeTextBlock
                                    stack_box_slot:
                                        Padding:=margin{Top:= 10.0, Bottom:= 10.0}
                                        HorizontalAlignment := horizontal_alignment.Fill
                                        VerticalAlignment := vertical_alignment.Fill
                                        Widget := stack_box:
                                            Orientation := orientation.Horizontal
                                            Slots := array:
                                                stack_box_slot:
                                                    Distribution:= option{0.5}
                                                    HorizontalAlignment := horizontal_alignment.Right
                                                    VerticalAlignment := vertical_alignment.Fill
                                                    Widget := CostTextBlock4
                                                stack_box_slot:
                                                    Distribution:= option{0.1}
                                                    HorizontalAlignment := horizontal_alignment.Center
                                                    VerticalAlignment := vertical_alignment.Fill
                                                    Widget := SiLverOreImage6
                                                stack_box_slot:
                                                    Distribution:= option{0.4}
                                                    HorizontalAlignment := horizontal_alignment.Left
                                                    VerticalAlignment := vertical_alignment.Fill
                                                    Widget := Cost_CriticalPowerSurgeTextBlock
                                    stack_box_slot:
                                        Distribution:= option{0.2}
                                        HorizontalAlignment := horizontal_alignment.Center
                                        VerticalAlignment := vertical_alignment.Center
                                        Widget := QuietButtonLevelup4
        }


        #========================================== Close ==========================================

        CloseButton : button_loud = button_loud{DefaultText := StringToMessage("CLOSE")}
        CloseButton.OnClick().Subscribe(OnCloseButtonClicked)
        set CanvasSlots += array
        {
            canvas_slot:
                Anchors := CenterAnchors
                Offsets := margin{Top:= 435.0, Left:= 658.0, Right := 353.0, Bottom := 80.0}
                Alignment := CenterVector
                SizeToContent := false
                Widget := CloseButton
        }
        
        if(set UIElements[Agent] = (OreTextBlock, PowerupStoneTextBlock)){}
        if(set SkillUIElements[Agent] = map{}){}
        if:
            set SkillUIElements[Agent] = map:
                1 => (Lv_CriticalRateTextBlock, Now_CriticalRateTextBlock, Will_CriticalRateTextBlock, Cost_CriticalRateTextBlock, _CriticalRateTextBlock)
                2 => (Lv_CriticalMultiplierTextBlock, Now_CriticalMultiplierTextBlock, Will_CriticalMultiplierTextBlock, Cost_CriticalMultiplierTextBlock, _CriticalMultiplierTextBlock)
                3 => (Lv_FinalHitBonusTextBlock, Now_FinalHitBonusTextBlock, Will_FinalHitBonusTextBlock, Cost_FinalHitBonusTextBlock, _FinalHitBonusTextBlock)
                4 => (Lv_CriticalPowerSurgeTextBlock, Now_CriticalPowerSurgeTextBlock, Will_CriticalPowerSurgeTextBlock, Cost_CriticalPowerSurgeTextBlock, _CriticalPowerSurgeTextBlock)

        Canvas : canvas = canvas:
            Slots := CanvasSlots

        return Canvas
