
using { /Verse.org/Assets }
using { /Verse.org/Simulation }
using { /Fortnite.com/UI }
using { /Fortnite.com/Devices }
using { /Fortnite.com/Characters }
using { /UnrealEngine.com/Temporary/UI }
using { /UnrealEngine.com/Temporary/SpatialMath }
using { /Verse.org/Simulation }
using { /UnrealEngine.com/Temporary/Diagnostics }
using { /Verse.org/Colors }
using { /Verse.org/Colors/NamedColors }
using { Utils }
using { PlayerData }
using { Skill }


# See https://dev.epicgames.com/documentation/en-us/uefn/create-your-own-device-in-verse for how to create a verse device.

# A Verse-authored creative device that can be placed in a level
ui_skill_miner2 := class(creative_device):
    #SFXs
    @editable
    SFX_Open : audio_player_device = audio_player_device{}
    @editable
    SFX_Close : audio_player_device = audio_player_device{}
    @editable
    SFX_LevelUp : audio_player_device = audio_player_device{}
    @editable
    SFX_Fail : audio_player_device = audio_player_device{}


    var _SaveData : save_data = save_data{}
    @editable
    var SaveDataManager : save_data_manager = save_data_manager{}
    @editable
    var SkillDataManager : skill_manager = skill_manager{}
    CloseEvent<public> : event(player) = event(player){}


    @editable
    SkillButton : button_device = button_device{}

    var miner2CanvasPerAgent : [agent]canvas = map{}
    var UIElements : [agent]tuple(text_block, text_block) = map{}
    var SkillUIElements : [agent][int]tuple(text_block, text_block, text_block, text_block, text_block) = map{}




    OnBegin<override>()<suspends>:void=
        SkillButton.InteractedWithEvent.Subscribe(OnButtonEvented)

    SetText<localizes>(Value:[]char) : message = "{Value}"
    # スキル関連のUI要素を全て更新
    UpdateUI(Agent:agent):void=
        UpdateUIElements(Agent)
        UpdateSkillUIElements(Agent)

    # UIElementsに格納された基本リソースのテキストを更新する
    UpdateUIElements(Agent: agent):void=
        var SilverValue : big_number = big_number{}
        var PowerupStoneValue : big_number = big_number{}
        if:
            Player := player[Agent]
        then:
            set SilverValue = SaveDataManager.GetResourceStone(Player)
            set PowerupStoneValue = SaveDataManager.GetResourcePowerupStone(Player)
            if (Elements := UIElements[Agent]):
                Elements(0).SetText(SetText("{BigNumber_FormatDisplay(SilverValue)}"))          # OreTextBlockのテキストを設定
                Elements(1).SetText(SetText("{BigNumber_FormatDisplay(PowerupStoneValue)}"))    # PowerupStoneTextBlockのテキストを設定 

    UpdateSkillUIElements(Agent:agent):void=
        var Lv_ComboUnlock : int = 1    #19
        var Now_ComboUnlock : int = 1
        var Will_ComboUnlock : int = 1
        var Cost_ComboUnlock : big_number = big_number{}
        var Lv_ChargeCount : int = 1    #20
        var Now_ChargeCount : int = 1
        var Will_ChargeCount : int = 1
        var Cost_ChargeCount : big_number = big_number{}
        var Lv_MoreHits : int = 1   #21
        var Now_MoreHits : int = 1
        var Will_MoreHits : int = 1
        var Cost_MoreHits : big_number = big_number{}
        var Lv_ComboMode : int = 1  #22
        var Now_ComboMode : int = 1
        var Will_ComboMode : int = 1
        var Cost_ComboMode : big_number = big_number{}
        if:
            Player := player[Agent]
        then:
            var LanguageIndex : int = SaveDataManager.GetLanguage(Player)
            if:
                set Lv_ComboUnlock = SaveDataManager.GetSkillLevel[Player, 19]
                set Now_ComboUnlock = SaveDataManager.GetPlayerSkillEffectIncrease[Player, 19]
                set Will_ComboUnlock = SaveDataManager.GetSpecificSkillEffectIncrease[Player, 19]
                set Cost_ComboUnlock = SaveDataManager.GetUpgradeCostForSkill[Player, 19]

                set Lv_ChargeCount = SaveDataManager.GetSkillLevel[Player, 20]
                set Now_ChargeCount = SaveDataManager.GetPlayerSkillEffectIncrease[Player, 20]
                set Will_ChargeCount = SaveDataManager.GetSpecificSkillEffectIncrease[Player, 20]
                set Cost_ChargeCount = SaveDataManager.GetUpgradeCostForSkill[Player, 20]

                set Lv_MoreHits = SaveDataManager.GetSkillLevel[Player, 21]
                set Now_MoreHits = SaveDataManager.GetPlayerSkillEffectIncrease[Player, 21]
                set Will_MoreHits = SaveDataManager.GetSpecificSkillEffectIncrease[Player, 21]
                set Cost_MoreHits = SaveDataManager.GetUpgradeCostForSkill[Player, 21]

                set Lv_ComboMode = SaveDataManager.GetSkillLevel[Player, 22]
                set Now_ComboMode = SaveDataManager.GetPlayerSkillEffectIncrease[Player, 22]
                set Will_ComboMode = SaveDataManager.GetSpecificSkillEffectIncrease[Player, 22]
                set Cost_ComboMode = SaveDataManager.GetUpgradeCostForSkill[Player, 22]
            then:
                LvComboUnlockText := if(SkillDataManager.CanUpgradeSkill(player[Agent], 19)?){"Lv.{Lv_ComboUnlock}"} else {"Lv.MAX"}
                LvComboUnlockTextColor := if(SkillDataManager.CanUpgradeSkill(player[Agent], 19)?){NamedColors.White} else {NamedColors.Red}
                WillComboUnlockText := if(SkillDataManager.CanUpgradeSkill(player[Agent], 19)?){"+ {Will_ComboUnlock}"} else {"MAX!"}
                CostComboUnlockText := if(SkillDataManager.CanUpgradeSkill(player[Agent], 19)?){"{BigNumber_FormatDisplay(Cost_ComboUnlock)}"} else {"-"}
                CostComboUnlockTextColor := if(SkillDataManager.IsUnlockable[player[Agent], 19]){NamedColors.White} else {NamedColors.Red}
                ChargeUnlockText := if(LanguageIndex = 0){"100回叩くと、次の攻撃が\n300%の連撃になる"} else {"After 100 hitsnext attack\nbecomes a 300% damage combo"}

                LvChargeCountText := if(SkillDataManager.CanUpgradeSkill(player[Agent], 20)?){"Lv.{Lv_ChargeCount}"} else {"Lv.MAX"}
                LvChargeCountTextColor := if(SkillDataManager.CanUpgradeSkill(player[Agent], 20)?){NamedColors.White} else {NamedColors.Red}
                WillChargeCountText := if(SkillDataManager.CanUpgradeSkill(player[Agent], 20)?){"+ {Will_ChargeCount}"} else {"MAX!"}
                CostChargeCountText := if(SkillDataManager.CanUpgradeSkill(player[Agent], 20)?){"{BigNumber_FormatDisplay(Cost_ChargeCount)}"} else {"-"}
                CostChargeCountTextColor := if(SkillDataManager.IsUnlockable[player[Agent], 20]){NamedColors.White} else {NamedColors.Red}
                ChargeCountText := if(LanguageIndex = 0){"1回の攻撃で溜まる\nカウントが増える"} else {"More charge per attack"}

                LvMoreHitsText := if(SkillDataManager.CanUpgradeSkill(player[Agent], 21)?){"Lv.{Lv_MoreHits}"} else {"Lv.MAX"}
                LvMoreHitsTextColor := if(SkillDataManager.CanUpgradeSkill(player[Agent], 21)?){NamedColors.White} else {NamedColors.Red}
                WillMoreHitsText := if(SkillDataManager.CanUpgradeSkill(player[Agent], 21)?){"+ {Will_MoreHits}"} else {"MAX!"}
                CostMoreHitsText := if(SkillDataManager.CanUpgradeSkill(player[Agent], 21)?){"{BigNumber_FormatDisplay(Cost_MoreHits)}"} else {"-"}
                CostMoreHitsTextColor := if(SkillDataManager.IsUnlockable[player[Agent], 21]){NamedColors.White} else {NamedColors.Red}
                MoreHitsText := if(LanguageIndex = 0){"連撃の攻撃力増加"} else {"More attacks in a combo"}

                LvComboModeText := if(SkillDataManager.CanUpgradeSkill(player[Agent], 22)?){"Lv.{Lv_ComboMode}"} else {"Lv.MAX"}
                LvComboModeTextColor := if(SkillDataManager.CanUpgradeSkill(player[Agent], 22)?){NamedColors.White} else {NamedColors.Red}
                WillComboModeText := if(SkillDataManager.CanUpgradeSkill(player[Agent], 22)?){"+ {Will_ComboMode}"} else {"MAX!"}
                CostComboModeText := if(SkillDataManager.CanUpgradeSkill(player[Agent], 22)?){"{BigNumber_FormatDisplay(Cost_ComboMode)}"} else {"-"}
                CostComboModeTextColor := if(SkillDataManager.IsUnlockable[player[Agent], 22]){NamedColors.White} else {NamedColors.Red}
                ComboModeText := if(LanguageIndex = 0){"連撃状態を持続\nできるようになる"} else {"More attacks during combo state"}

                if(SkillElement := SkillUIElements[Agent][1]) :
                    SkillElement(0).SetText(SetText(LvComboUnlockText))
                    SkillElement(0).SetTextColor(LvComboUnlockTextColor)
                    SkillElement(1).SetText(SetText("{Now_ComboUnlock}"))
                    SkillElement(2).SetText(SetText(WillComboUnlockText))
                    SkillElement(3).SetText(SetText(CostComboUnlockText))
                    SkillElement(3).SetTextColor(CostComboUnlockTextColor)
                    SkillElement(4).SetText(SetText("{ChargeUnlockText}"))

                if(SkillElement := SkillUIElements[Agent][2]) :
                    SkillElement(0).SetText(SetText(LvChargeCountText))
                    SkillElement(0).SetTextColor(LvChargeCountTextColor)
                    SkillElement(1).SetText(SetText("{Now_ChargeCount}"))
                    SkillElement(2).SetText(SetText(WillChargeCountText))
                    SkillElement(3).SetText(SetText(CostChargeCountText))
                    SkillElement(3).SetTextColor(CostChargeCountTextColor)
                    SkillElement(4).SetText(SetText("{ChargeCountText}"))

                if(SkillElement := SkillUIElements[Agent][3]) :
                    SkillElement(0).SetText(SetText(LvMoreHitsText))
                    SkillElement(0).SetTextColor(LvMoreHitsTextColor)
                    SkillElement(1).SetText(SetText("{Now_MoreHits}"))
                    SkillElement(2).SetText(SetText(WillMoreHitsText))
                    SkillElement(3).SetText(SetText(CostMoreHitsText))
                    SkillElement(3).SetTextColor(CostMoreHitsTextColor)
                    SkillElement(4).SetText(SetText("{MoreHitsText}"))

                if(SkillElement := SkillUIElements[Agent][4]) :
                    SkillElement(0).SetText(SetText(LvComboModeText))
                    SkillElement(0).SetTextColor(LvComboModeTextColor)
                    SkillElement(1).SetText(SetText("{Now_ComboMode}"))
                    SkillElement(2).SetText(SetText(WillComboModeText))
                    SkillElement(3).SetText(SetText(CostComboModeText))
                    SkillElement(3).SetTextColor(CostComboModeTextColor)
                    SkillElement(4).SetText(SetText("{ComboModeText}"))

    OnPlayerSpawned(Agent : agent):void=
        if (miner2Canvas := miner2CanvasPerAgent[Agent]):
            return

        if (Player := player[Agent], PlayerUI := GetPlayerUI[Player]):
            miner2Canvas : canvas = Createminer2Canvas(Agent)
            if (set miner2CanvasPerAgent[Agent] = miner2Canvas){}

    OnButtonEvented(Agent : agent):void=
        SFX_Open.Play(Agent)
        OnPlayerSpawned(Agent)
        UpdateUI(Agent)
        Print("OnSkillButtonEvented")
        if (Player := player[Agent], PlayerUI := GetPlayerUI[Player], miner2Canvas := miner2CanvasPerAgent[Agent]):
            PlayerUI.AddWidget(miner2Canvas, player_ui_slot{InputMode := ui_input_mode.All})

    # ボタン処理
    OnLevelup1Clicked(WidgetMessage : widget_message):void=
        if (PlayerUI := GetPlayerUI[WidgetMessage.Player], miner2Canvas := miner2CanvasPerAgent[WidgetMessage.Player], Agent : agent = agent[WidgetMessage.Player]):
            if(SkillDataManager.CanUpgradeSkill(WidgetMessage.Player,19) = true):
                SFX_LevelUp.Play(Agent)
                if:
                    SkillDataManager.UnlockOrUpgradeSkill[WidgetMessage.Player,19]
                then:
                    UpdateUI(Agent)
            else:
                SFX_Fail.Play(Agent)
                return
            
    OnLevelup2Clicked(WidgetMessage : widget_message):void=
        if (PlayerUI := GetPlayerUI[WidgetMessage.Player], miner2Canvas := miner2CanvasPerAgent[WidgetMessage.Player], Agent : agent = agent[WidgetMessage.Player]):
            if(SkillDataManager.CanUpgradeSkill(WidgetMessage.Player,20) = true):
                SFX_LevelUp.Play(Agent)
                if:
                    SkillDataManager.UnlockOrUpgradeSkill[WidgetMessage.Player,20]
                then:
                    UpdateUI(Agent)
            else:
                SFX_Fail.Play(Agent)
                return

    OnLevelup3Clicked(WidgetMessage : widget_message):void=
        if (PlayerUI := GetPlayerUI[WidgetMessage.Player], miner2Canvas := miner2CanvasPerAgent[WidgetMessage.Player], Agent : agent = agent[WidgetMessage.Player]):
            if(SkillDataManager.CanUpgradeSkill(WidgetMessage.Player,21) = true):
                SFX_LevelUp.Play(Agent)
                if:
                    SkillDataManager.UnlockOrUpgradeSkill[WidgetMessage.Player,21]
                then:
                    UpdateUI(Agent)
            else:
                SFX_Fail.Play(Agent)
                return
            
    OnLevelup4Clicked(WidgetMessage : widget_message):void=
        if (PlayerUI := GetPlayerUI[WidgetMessage.Player], miner2Canvas := miner2CanvasPerAgent[WidgetMessage.Player], Agent : agent = agent[WidgetMessage.Player]):
            if(SkillDataManager.CanUpgradeSkill(WidgetMessage.Player,22) = true):
                SFX_LevelUp.Play(Agent)
                if:
                    SkillDataManager.UnlockOrUpgradeSkill[WidgetMessage.Player,22]
                then:
                    UpdateUI(Agent)
            else:
                SFX_Fail.Play(Agent)
                return

    OnCloseButtonClicked(WidgetMessage : widget_message):void=
        if(Agent : agent = agent[WidgetMessage.Player]) :
            SFX_Close.Play(Agent)
        if (PlayerUI := GetPlayerUI[WidgetMessage.Player], miner2Canvas := miner2CanvasPerAgent[WidgetMessage.Player]):
            PlayerUI.RemoveWidget(miner2Canvas)
            CloseEvent.Signal(WidgetMessage.Player)

    # キャンバス
    Createminer2Canvas(Agent : agent):canvas=
        var CanvasSlots : []canvas_slot = array{}
        
        CenterVector : vector2 = vector2{X:=0.5, Y:=0.5}
        CenterAnchors : anchors = anchors{Minimum := CenterVector, Maximum := CenterVector}

        # Image Stock
        BGImage : texture_block = texture_block{DefaultImage := Textures.BG_Skill, DefaultDesiredSize := vector2{X := 1920.0, Y := 1080.0}}
        SilverOreImage1 : texture_block = texture_block{DefaultImage := Textures.SilverOre, DefaultDesiredSize := vector2{X := 50.0, Y := 50.0}}
        PowerupStoneImage : texture_block = texture_block{DefaultImage := Textures.IC_PowerupStone, DefaultDesiredSize := vector2{X := 50.0, Y := 50.0}}
        SiLverOreImage3 : texture_block = texture_block{DefaultImage := Textures.SilverOre, DefaultDesiredSize := vector2{X := 40.0, Y := 40.0}}
        SiLverOreImage4 : texture_block = texture_block{DefaultImage := Textures.SilverOre, DefaultDesiredSize := vector2{X := 40.0, Y := 40.0}}
        SiLverOreImage5 : texture_block = texture_block{DefaultImage := Textures.SilverOre, DefaultDesiredSize := vector2{X := 40.0, Y := 40.0}}
        SiLverOreImage6 : texture_block = texture_block{DefaultImage := Textures.SilverOre, DefaultDesiredSize := vector2{X := 40.0, Y := 40.0}}
        ComboUnlockImage : texture_block = texture_block{DefaultImage := Textures.MinerSkill2_1, DefaultDesiredSize := vector2{X := 200.0, Y := 200.0}}
        ChargeCountImage : texture_block = texture_block{DefaultImage := Textures.MinerSkill2_2, DefaultDesiredSize := vector2{X := 200.0, Y := 200.0}}
        MoreHitsImage : texture_block = texture_block{DefaultImage := Textures.MinerSkill2_3, DefaultDesiredSize := vector2{X := 200.0, Y := 200.0}}
        ComboModeImage : texture_block = texture_block{DefaultImage := Textures.MinerSkill2_4, DefaultDesiredSize := vector2{X := 200.0, Y := 200.0}}
            
        var SilverValue : big_number = big_number{}
        var PowerupStoneValue : big_number = big_number{}
        var Lv_ComboUnlock : int = 1    #19
        var Now_ComboUnlock : int = 1
        var Will_ComboUnlock : int = 1
        var Cost_ComboUnlock : big_number = big_number{}
        var Lv_ChargeCount : int = 1    #20
        var Now_ChargeCount : int = 1
        var Will_ChargeCount : int = 1
        var Cost_ChargeCount : big_number = big_number{}
        var Lv_MoreHits : int = 1   #21
        var Now_MoreHits : int = 1
        var Will_MoreHits : int = 1
        var Cost_MoreHits : big_number = big_number{}
        var Lv_ComboMode : int = 1  #22
        var Now_ComboMode : int = 1
        var Will_ComboMode : int = 1
        var Cost_ComboMode : big_number = big_number{}
        
        #================================== TextBlock ==================================
        
        # 所持岩
        OreTextBlock : text_block = text_block{
            DefaultTextColor := NamedColors.White,
            DefaultShadowOffset := option{vector2{X:=2.5, Y:=2.5}},
            DefaultShadowColor := NamedColors.Black
        }
        # 所持強化アイテム
        PowerupStoneTextBlock : text_block = text_block{
            DefaultTextColor := NamedColors.White,
            DefaultShadowOffset := option{vector2{X:=2.5, Y:=2.5}},
            DefaultShadowColor := NamedColors.Black,
            DefaultJustification := text_justification.Center
        }

        #================================== Skill 1 ==================================

        Lv_ComboUnlockTextBlock : text_block = text_block{
            DefaultTextColor := NamedColors.White,
            DefaultShadowOffset := option{vector2{X:=2.5, Y:=2.5}},
            DefaultShadowColor := NamedColors.Black,
            DefaultJustification := text_justification.Center
        }
        Now_ComboUnlockTextBlock : text_block = text_block{
            DefaultTextColor := NamedColors.White,
            DefaultShadowOffset := option{vector2{X:=2.5, Y:=2.5}},
            DefaultShadowColor := NamedColors.Black,
            DefaultJustification := text_justification.Center
        }
        Will_ComboUnlockTextBlock : text_block = text_block{
            DefaultTextColor := NamedColors.White,
            DefaultShadowOffset := option{vector2{X:=2.5, Y:=2.5}},
            DefaultShadowColor := NamedColors.Black,
            DefaultJustification := text_justification.Center
        }
        Cost_ComboUnlockTextBlock : text_block = text_block{
            DefaultTextColor := NamedColors.White,
            DefaultShadowOffset := option{vector2{X:=2.5, Y:=2.5}},
            DefaultShadowColor := NamedColors.Black,
            DefaultJustification := text_justification.Center
        }
        var _ComboUnlockTextBlock :text_block= text_block{
            DefaultTextColor := NamedColors.White,
            DefaultShadowOffset := option{vector2{X:=2.5, Y:=2.5}},
            DefaultShadowColor := NamedColors.Black,
            DefaultShadowOpacity := 1.0,
            DefaultJustification := text_justification.Center
        }
        #================================== Skill 2 ==================================
        Lv_ChargeCountTextBlock : text_block = text_block{
            DefaultTextColor := NamedColors.White,
            DefaultShadowOffset := option{vector2{X:=2.5, Y:=2.5}},
            DefaultShadowColor := NamedColors.Black,
            DefaultJustification := text_justification.Center
        }
        Now_ChargeCountTextBlock : text_block = text_block{
            DefaultTextColor := NamedColors.White,
            DefaultShadowOffset := option{vector2{X:=2.5, Y:=2.5}},
            DefaultShadowColor := NamedColors.Black,
            DefaultJustification := text_justification.Center
        }
        Will_ChargeCountTextBlock : text_block = text_block{
            DefaultTextColor := NamedColors.White,
            DefaultShadowOffset := option{vector2{X:=2.5, Y:=2.5}},
            DefaultShadowColor := NamedColors.Black,
            DefaultJustification := text_justification.Center
        }
        Cost_ChargeCountTextBlock : text_block = text_block{
            DefaultTextColor := NamedColors.White,
            DefaultShadowOffset := option{vector2{X:=2.5, Y:=2.5}},
            DefaultShadowColor := NamedColors.Black,
            DefaultJustification := text_justification.Center
        }
        var _ChargeCountTextBlock :text_block= text_block{
            DefaultTextColor := NamedColors.White,
            DefaultShadowOffset := option{vector2{X:=2.5, Y:=2.5}},
            DefaultShadowColor := NamedColors.Black,
            DefaultShadowOpacity := 1.0,
            DefaultJustification := text_justification.Center
        }
        #================================== Skill 3 ==================================
        Lv_MoreHitsTextBlock : text_block = text_block{
            DefaultTextColor := NamedColors.White,
            DefaultShadowOffset := option{vector2{X:=2.5, Y:=2.5}},
            DefaultShadowColor := NamedColors.Black,
            DefaultJustification := text_justification.Center
        }
        Now_MoreHitsTextBlock : text_block = text_block{
            DefaultTextColor := NamedColors.White,
            DefaultShadowOffset := option{vector2{X:=2.5, Y:=2.5}},
            DefaultShadowColor := NamedColors.Black,
            DefaultJustification := text_justification.Center
        }
        Will_MoreHitsTextBlock : text_block = text_block{
            DefaultTextColor := NamedColors.White,
            DefaultShadowOffset := option{vector2{X:=2.5, Y:=2.5}},
            DefaultShadowColor := NamedColors.Black,
            DefaultJustification := text_justification.Center
        }
        Cost_MoreHitsTextBlock : text_block = text_block{
            DefaultTextColor := NamedColors.White,
            DefaultShadowOffset := option{vector2{X:=2.5, Y:=2.5}},
            DefaultShadowColor := NamedColors.Black,
            DefaultJustification := text_justification.Center
        }
        var _MoreHitsTextBlock :text_block= text_block{
            DefaultTextColor := NamedColors.White,
            DefaultShadowOffset := option{vector2{X:=2.5, Y:=2.5}},
            DefaultShadowColor := NamedColors.Black,
            DefaultShadowOpacity := 1.0,
            DefaultJustification := text_justification.Center,
            DefaultOverflowPolicy := text_overflow_policy.Ellipsis
        }
        #================================== Skill 4 ==================================

        Lv_ComboModeTextBlock : text_block = text_block{
            DefaultTextColor := NamedColors.White,
            DefaultShadowOffset := option{vector2{X:=2.5, Y:=2.5}},
            DefaultShadowColor := NamedColors.Black,
            DefaultJustification := text_justification.Center
        }
        Now_ComboModeTextBlock : text_block = text_block{
            DefaultTextColor := NamedColors.White,
            DefaultShadowOffset := option{vector2{X:=2.5, Y:=2.5}},
            DefaultShadowColor := NamedColors.Black,
            DefaultJustification := text_justification.Center
        }
        Will_ComboModeTextBlock : text_block = text_block{
            DefaultTextColor := NamedColors.White,
            DefaultShadowOffset := option{vector2{X:=2.5, Y:=2.5}},
            DefaultShadowColor := NamedColors.Black,
            DefaultJustification := text_justification.Center
        }
        Cost_ComboModeTextBlock : text_block = text_block{
            DefaultTextColor := NamedColors.White,
            DefaultShadowOffset := option{vector2{X:=2.5, Y:=2.5}},
            DefaultShadowColor := NamedColors.Black,
            DefaultJustification := text_justification.Center
        }
        var _ComboModeTextBlock :text_block= text_block{
            DefaultTextColor := NamedColors.White,
            DefaultShadowOffset := option{vector2{X:=2.5, Y:=2.5}},
            DefaultShadowColor := NamedColors.Black,
            DefaultShadowOpacity := 1.0,
            DefaultJustification := text_justification.Center
        }

        #================================== /TextBlock ==================================
        
        #========================================== BG ==========================================

        set CanvasSlots += array
        {
            canvas_slot:
                Anchors := CenterAnchors
                Offsets := margin{Top := 0.0,Left := 0.0, Right := 1920.0, Bottom := 1080.0}
                Alignment := CenterVector
                SizeToContent := false
                Widget := BGImage
        }

        #========================================= Header ========================================

        miner2TextBlock := text_block{
            DefaultText := StringToMessage("Miner Skill 2"),
            DefaultTextColor := NamedColors.White,
            DefaultShadowOffset := option{vector2{X:=2.5, Y:=2.5}},
            DefaultShadowColor := NamedColors.Black,
            DefaultJustification := text_justification.Center
        }
        miner2TextBlock.SetShadowOpacity(1.0)
        set CanvasSlots += array
        {
            canvas_slot:
                Anchors := CenterAnchors
                Offsets := margin{Top:= -472.0, Left:= 0.0, Right := 257.0, Bottom := 41.0}
                Alignment := CenterVector
                SizeToContent := false
                Widget := miner2TextBlock
        }

        OreTextBlock.SetShadowOpacity(1.0)
        PowerupStoneTextBlock.SetShadowOpacity(1.0)

        set CanvasSlots += array

        {
            canvas_slot:
                Anchors := CenterAnchors
                Offsets := margin{Left:=-645.5 ,Top:= -391.0, Right:= 519.0, Bottom:= 59.0}
                Alignment := CenterVector
                SizeToContent := false
                Widget := stack_box:
                    Orientation := orientation.Horizontal
                    Slots := array:
                        stack_box_slot:
                            Distribution:= option{0.5}
                            Padding:=margin{Left:=5.0 ,Top:= 5.0, Right:= 10.0, Bottom:= 5.0}
                            HorizontalAlignment := horizontal_alignment.Fill
                            VerticalAlignment := vertical_alignment.Fill
                            Widget := stack_box:
                                Orientation := orientation.Horizontal
                                Slots := array:
                                    stack_box_slot:
                                        HorizontalAlignment := horizontal_alignment.Center
                                        VerticalAlignment := vertical_alignment.Center
                                        Widget := SilverOreImage1
                                    stack_box_slot:
                                        Distribution:= option{1.0}
                                        HorizontalAlignment := horizontal_alignment.Center
                                        VerticalAlignment := vertical_alignment.Center
                                        Widget := OreTextBlock
                        stack_box_slot:
                            Padding:=margin{Left:=10.0 ,Top:= 5.0, Right:= 5.0, Bottom:= 5.0}
                            Distribution:= option{0.5}
                            HorizontalAlignment := horizontal_alignment.Fill
                            VerticalAlignment := vertical_alignment.Fill
                            Widget := stack_box:
                                Orientation := orientation.Horizontal
                                Slots := array:
                                    stack_box_slot:
                                        HorizontalAlignment := horizontal_alignment.Center
                                        VerticalAlignment := vertical_alignment.Center
                                        Widget := PowerupStoneImage
                                    stack_box_slot:
                                        Distribution:= option{1.0}
                                        HorizontalAlignment := horizontal_alignment.Center
                                        VerticalAlignment := vertical_alignment.Center
                                        Widget := PowerupStoneTextBlock
        }

        #========================================== Skill ==========================================

        CostTextBlock1 := text_block{
            DefaultText := StringToMessage("Cost :"),
            DefaultTextColor := NamedColors.White,
            DefaultShadowOffset := option{vector2{X:=2.5, Y:=2.5}},
            DefaultShadowColor := NamedColors.Black,
            DefaultJustification := text_justification.Center
        }
        CostTextBlock1.SetShadowOpacity(1.0)
        CostTextBlock2 := text_block{
            DefaultText := StringToMessage("Cost :"),
            DefaultTextColor := NamedColors.White,
            DefaultShadowOffset := option{vector2{X:=2.5, Y:=2.5}},
            DefaultShadowColor := NamedColors.Black,
            DefaultJustification := text_justification.Center
        }
        CostTextBlock2.SetShadowOpacity(1.0)
        CostTextBlock3 := text_block{
            DefaultText := StringToMessage("Cost :"),
            DefaultTextColor := NamedColors.White,
            DefaultShadowOffset := option{vector2{X:=2.5, Y:=2.5}},
            DefaultShadowColor := NamedColors.Black,
            DefaultJustification := text_justification.Center
        }
        CostTextBlock3.SetShadowOpacity(1.0)
        CostTextBlock4 := text_block{
            DefaultText := StringToMessage("Cost :"),
            DefaultTextColor := NamedColors.White,
            DefaultShadowOffset := option{vector2{X:=2.5, Y:=2.5}},
            DefaultShadowColor := NamedColors.Black,
            DefaultJustification := text_justification.Center
        }
        CostTextBlock4.SetShadowOpacity(1.0)

        #Skill 1
        ComboUnlockTextBlock := text_block{
            DefaultText := StringToMessage("Combo unlock"),
            DefaultTextColor := NamedColors.White,
            DefaultShadowOffset := option{vector2{X:=2.5, Y:=2.5}},
            DefaultShadowColor := NamedColors.Black,
            DefaultJustification := text_justification.Center
        }
        ComboUnlockTextBlock.SetShadowOpacity(1.0)
        Lv_ComboUnlockTextBlock.SetShadowOpacity(1.0)
        Now_ComboUnlockTextBlock.SetShadowOpacity(1.0)
        Will_ComboUnlockTextBlock.SetShadowOpacity(1.0)
        Cost_ComboUnlockTextBlock.SetShadowOpacity(1.0)
        
        #Skill 2
        ChargeCountTextBlock := text_block{
            DefaultText := StringToMessage("Charge count"),
            DefaultTextColor := NamedColors.White,
            DefaultShadowOffset := option{vector2{X:=2.5, Y:=2.5}},
            DefaultShadowColor := NamedColors.Black,
            DefaultJustification := text_justification.Center
        }
        ChargeCountTextBlock.SetShadowOpacity(1.0)
        Lv_ChargeCountTextBlock.SetShadowOpacity(1.0)
        Now_ChargeCountTextBlock.SetShadowOpacity(1.0)
        Will_ChargeCountTextBlock.SetShadowOpacity(1.0)
        Cost_ChargeCountTextBlock.SetShadowOpacity(1.0)
        
        #Skill 3
        MoreHitsTextBlock := text_block{
            DefaultText := StringToMessage("More hits"),
            DefaultTextColor := NamedColors.White,
            DefaultShadowOffset := option{vector2{X:=2.5, Y:=2.5}},
            DefaultShadowColor := NamedColors.Black,
            DefaultJustification := text_justification.Center
        }
        MoreHitsTextBlock.SetShadowOpacity(1.0)
        Lv_MoreHitsTextBlock.SetShadowOpacity(1.0)
        Now_MoreHitsTextBlock.SetShadowOpacity(1.0)
        Will_MoreHitsTextBlock.SetShadowOpacity(1.0)
        Cost_MoreHitsTextBlock.SetShadowOpacity(1.0)
        
        #Skill 4
        ComboModeTextBlock := text_block{
            DefaultText := StringToMessage("Combo mode"),
            DefaultTextColor := NamedColors.White,
            DefaultShadowOffset := option{vector2{X:=2.5, Y:=2.5}},
            DefaultShadowColor := NamedColors.Black,
            DefaultJustification := text_justification.Center
        }
        ComboModeTextBlock.SetShadowOpacity(1.0)
        Lv_ComboModeTextBlock.SetShadowOpacity(1.0)
        Now_ComboModeTextBlock.SetShadowOpacity(1.0)
        Will_ComboModeTextBlock.SetShadowOpacity(1.0)
        Cost_ComboModeTextBlock.SetShadowOpacity(1.0)

        # Button Stock
        QuietButtonLevelup1 : button_quiet = button_quiet{DefaultText := StringToMessage("LEVEL UP")}
        QuietButtonLevelup1.OnClick().Subscribe(OnLevelup1Clicked)
        QuietButtonLevelup2 : button_quiet = button_quiet{DefaultText := StringToMessage("LEVEL UP")}
        QuietButtonLevelup2.OnClick().Subscribe(OnLevelup2Clicked)
        QuietButtonLevelup3 : button_quiet = button_quiet{DefaultText := StringToMessage("LEVEL UP")}
        QuietButtonLevelup3.OnClick().Subscribe(OnLevelup3Clicked)
        QuietButtonLevelup4 : button_quiet = button_quiet{DefaultText := StringToMessage("LEVEL UP")}
        QuietButtonLevelup4.OnClick().Subscribe(OnLevelup4Clicked)

        #================================ Top ================================

        set CanvasSlots += array
        {
            canvas_slot:
                Anchors := CenterAnchors
                Offsets := margin{Left:=0.0 ,Top:= -96.5, Right:= 1760.0, Bottom:= 464.0}
                Alignment := CenterVector
                SizeToContent := false
                Widget := stack_box:
                    Orientation := orientation.Horizontal
                    Slots := array:
                        stack_box_slot:
                            Padding:=margin{Left:= 50.0, Right:= 50.0}
                            Distribution:= option{0.25}
                            HorizontalAlignment := horizontal_alignment.Fill
                            VerticalAlignment := vertical_alignment.Fill
                            Widget := stack_box:
                                Orientation := orientation.Vertical
                                Slots := array:
                                    stack_box_slot:
                                        Padding:=margin{Top:= 10.0, Bottom:= 10.0}
                                        HorizontalAlignment := horizontal_alignment.Fill
                                        VerticalAlignment := vertical_alignment.Fill
                                        Widget := ComboUnlockTextBlock
                                    stack_box_slot:
                                        Padding:=margin{Top:= 15.0, Bottom:= 15.0}
                                        HorizontalAlignment := horizontal_alignment.Fill
                                        VerticalAlignment := vertical_alignment.Fill
                                        Widget := Lv_ComboUnlockTextBlock
                                    stack_box_slot:
                                        Padding:=margin{Top:= 10.0, Bottom:= 10.0}
                                        HorizontalAlignment := horizontal_alignment.Center
                                        VerticalAlignment := vertical_alignment.Center
                                        Widget := ComboUnlockImage
                                    stack_box_slot:
                                        Padding:=margin{Top:= 15.0, Bottom:= 15.0}
                                        HorizontalAlignment := horizontal_alignment.Fill
                                        VerticalAlignment := vertical_alignment.Fill
                                        Widget := _ComboUnlockTextBlock
                        stack_box_slot:
                            Padding:=margin{Left:= 50.0, Right:= 50.0}
                            Distribution:= option{0.25}
                            HorizontalAlignment := horizontal_alignment.Fill
                            VerticalAlignment := vertical_alignment.Fill
                            Widget := stack_box:
                                Orientation := orientation.Vertical
                                Slots := array:
                                    stack_box_slot:
                                        Padding:=margin{Top:= 10.0, Bottom:= 10.0}
                                        HorizontalAlignment := horizontal_alignment.Fill
                                        VerticalAlignment := vertical_alignment.Fill
                                        Widget := ChargeCountTextBlock
                                    stack_box_slot:
                                        Padding:=margin{Top:= 15.0, Bottom:= 15.0}
                                        HorizontalAlignment := horizontal_alignment.Fill
                                        VerticalAlignment := vertical_alignment.Fill
                                        Widget := Lv_ChargeCountTextBlock
                                    stack_box_slot:
                                        Padding:=margin{Top:= 10.0, Bottom:= 10.0}
                                        HorizontalAlignment := horizontal_alignment.Center
                                        VerticalAlignment := vertical_alignment.Center
                                        Widget := ChargeCountImage
                                    stack_box_slot:
                                        Padding:=margin{Top:= 15.0, Bottom:= 15.0}
                                        HorizontalAlignment := horizontal_alignment.Fill
                                        VerticalAlignment := vertical_alignment.Fill
                                        Widget := _ChargeCountTextBlock
                        stack_box_slot:
                            Padding:=margin{Left:= 50.0, Right:= 50.0}
                            Distribution:= option{0.25}
                            HorizontalAlignment := horizontal_alignment.Fill
                            VerticalAlignment := vertical_alignment.Fill
                            Widget := stack_box:
                                Orientation := orientation.Vertical
                                Slots := array:
                                    stack_box_slot:
                                        Padding:=margin{Top:= 10.0, Bottom:= 10.0}
                                        HorizontalAlignment := horizontal_alignment.Fill
                                        VerticalAlignment := vertical_alignment.Fill
                                        Widget := MoreHitsTextBlock
                                    stack_box_slot:
                                        Padding:=margin{Top:= 15.0, Bottom:= 15.0}
                                        HorizontalAlignment := horizontal_alignment.Fill
                                        VerticalAlignment := vertical_alignment.Fill
                                        Widget := Lv_MoreHitsTextBlock
                                    stack_box_slot:
                                        Padding:=margin{Top:= 10.0, Bottom:= 10.0}
                                        HorizontalAlignment := horizontal_alignment.Center
                                        VerticalAlignment := vertical_alignment.Center
                                        Widget := MoreHitsImage
                                    stack_box_slot:
                                        Padding:=margin{Top:= 15.0, Bottom:= 15.0}
                                        HorizontalAlignment := horizontal_alignment.Fill
                                        VerticalAlignment := vertical_alignment.Fill
                                        Widget := _MoreHitsTextBlock
                        stack_box_slot:
                            Padding:=margin{Left:= 50.0, Right:= 50.0}
                            Distribution:= option{0.25}
                            HorizontalAlignment := horizontal_alignment.Fill
                            VerticalAlignment := vertical_alignment.Fill
                            Widget := stack_box:
                                Orientation := orientation.Vertical
                                Slots := array:
                                    stack_box_slot:
                                        Padding:=margin{Top:= 10.0, Bottom:= 10.0}
                                        HorizontalAlignment := horizontal_alignment.Fill
                                        VerticalAlignment := vertical_alignment.Fill
                                        Widget := ComboModeTextBlock
                                    stack_box_slot:
                                        Padding:=margin{Top:= 15.0, Bottom:= 15.0}
                                        HorizontalAlignment := horizontal_alignment.Fill
                                        VerticalAlignment := vertical_alignment.Fill
                                        Widget := Lv_ComboModeTextBlock
                                    stack_box_slot:
                                        Padding:=margin{Top:= 10.0, Bottom:= 10.0}
                                        HorizontalAlignment := horizontal_alignment.Center
                                        VerticalAlignment := vertical_alignment.Center
                                        Widget := ComboModeImage
                                    stack_box_slot:
                                        Padding:=margin{Top:= 15.0, Bottom:= 15.0}
                                        HorizontalAlignment := horizontal_alignment.Fill
                                        VerticalAlignment := vertical_alignment.Fill
                                        Widget := _ComboModeTextBlock
        }

        
        #================================ Bottom ================================

        set CanvasSlots += array
        {
            canvas_slot:
                Anchors := CenterAnchors
                Offsets := margin{Left:=0.0 ,Top:= 249.5, Right:= 1760.0, Bottom:= 210.0}
                Alignment := CenterVector
                SizeToContent := false
                Widget := stack_box:
                    Orientation := orientation.Horizontal
                    Slots := array:
                        stack_box_slot:
                            Distribution:= option{0.25}
                            HorizontalAlignment := horizontal_alignment.Fill
                            VerticalAlignment := vertical_alignment.Fill
                            Widget := stack_box:
                                Orientation := orientation.Vertical
                                Slots := array:
                                    stack_box_slot:
                                        Padding:=margin{Top:= 10.0, Bottom:= 10.0}
                                        HorizontalAlignment := horizontal_alignment.Fill
                                        VerticalAlignment := vertical_alignment.Fill
                                        Widget := stack_box:
                                            Orientation := orientation.Horizontal
                                            Slots := array:
                                                stack_box_slot:
                                                    Padding:=margin{Top:= 10.0, Right:= 5.0, Bottom:= 10.0}
                                                    Distribution:= option{0.5}
                                                    HorizontalAlignment := horizontal_alignment.Right
                                                    VerticalAlignment := vertical_alignment.Fill
                                                    Widget := Now_ComboUnlockTextBlock
                                                stack_box_slot:
                                                    Padding:=margin{Top:= 10.0, Left:= 5.0, Bottom:= 10.0}
                                                    Distribution:= option{0.5}
                                                    HorizontalAlignment := horizontal_alignment.Left
                                                    VerticalAlignment := vertical_alignment.Fill
                                                    Widget := Will_ComboUnlockTextBlock
                                    stack_box_slot:
                                        Padding:=margin{Top:= 10.0, Bottom:= 10.0}
                                        HorizontalAlignment := horizontal_alignment.Fill
                                        VerticalAlignment := vertical_alignment.Fill
                                        Widget := stack_box:
                                            Orientation := orientation.Horizontal
                                            Slots := array:
                                                stack_box_slot:
                                                    Distribution:= option{0.5}
                                                    HorizontalAlignment := horizontal_alignment.Right
                                                    VerticalAlignment := vertical_alignment.Fill
                                                    Widget := CostTextBlock1
                                                stack_box_slot:
                                                    Distribution:= option{0.1}
                                                    HorizontalAlignment := horizontal_alignment.Center
                                                    VerticalAlignment := vertical_alignment.Fill
                                                    Widget := SiLverOreImage3
                                                stack_box_slot:
                                                    Distribution:= option{0.4}
                                                    HorizontalAlignment := horizontal_alignment.Left
                                                    VerticalAlignment := vertical_alignment.Fill
                                                    Widget := Cost_ComboUnlockTextBlock
                                    stack_box_slot:
                                        Distribution:= option{0.2}
                                        HorizontalAlignment := horizontal_alignment.Center
                                        VerticalAlignment := vertical_alignment.Center
                                        Widget := QuietButtonLevelup1
                        stack_box_slot:
                            Distribution:= option{0.25}
                            HorizontalAlignment := horizontal_alignment.Fill
                            VerticalAlignment := vertical_alignment.Fill
                            Widget := stack_box:
                                Orientation := orientation.Vertical
                                Slots := array:
                                    stack_box_slot:
                                        Padding:=margin{Top:= 10.0, Bottom:= 10.0}
                                        HorizontalAlignment := horizontal_alignment.Fill
                                        VerticalAlignment := vertical_alignment.Fill
                                        Widget := stack_box:
                                            Orientation := orientation.Horizontal
                                            Slots := array:
                                                stack_box_slot:
                                                    Padding:=margin{Top:= 10.0, Right:= 5.0, Bottom:= 10.0}
                                                    Distribution:= option{0.5}
                                                    HorizontalAlignment := horizontal_alignment.Right
                                                    VerticalAlignment := vertical_alignment.Fill
                                                    Widget := Now_ChargeCountTextBlock
                                                stack_box_slot:
                                                    Padding:=margin{Top:= 10.0, Left:= 5.0, Bottom:= 10.0}
                                                    Distribution:= option{0.5}
                                                    HorizontalAlignment := horizontal_alignment.Left
                                                    VerticalAlignment := vertical_alignment.Fill
                                                    Widget := Will_ChargeCountTextBlock
                                    stack_box_slot:
                                        Padding:=margin{Top:= 10.0, Bottom:= 10.0}
                                        HorizontalAlignment := horizontal_alignment.Fill
                                        VerticalAlignment := vertical_alignment.Fill
                                        Widget := stack_box:
                                            Orientation := orientation.Horizontal
                                            Slots := array:
                                                stack_box_slot:
                                                    Distribution:= option{0.5}
                                                    HorizontalAlignment := horizontal_alignment.Right
                                                    VerticalAlignment := vertical_alignment.Fill
                                                    Widget := CostTextBlock2
                                                stack_box_slot:
                                                    Distribution:= option{0.1}
                                                    HorizontalAlignment := horizontal_alignment.Center
                                                    VerticalAlignment := vertical_alignment.Fill
                                                    Widget := SiLverOreImage4
                                                stack_box_slot:
                                                    Distribution:= option{0.4}
                                                    HorizontalAlignment := horizontal_alignment.Left
                                                    VerticalAlignment := vertical_alignment.Fill
                                                    Widget := Cost_ChargeCountTextBlock
                                    stack_box_slot:
                                        Distribution:= option{0.2}
                                        HorizontalAlignment := horizontal_alignment.Center
                                        VerticalAlignment := vertical_alignment.Center
                                        Widget := QuietButtonLevelup2

                        stack_box_slot:
                            Distribution:= option{0.25}
                            HorizontalAlignment := horizontal_alignment.Fill
                            VerticalAlignment := vertical_alignment.Fill
                            Widget := stack_box:
                                Orientation := orientation.Vertical
                                Slots := array:
                                    stack_box_slot:
                                        Padding:=margin{Top:= 10.0, Bottom:= 10.0}
                                        HorizontalAlignment := horizontal_alignment.Fill
                                        VerticalAlignment := vertical_alignment.Fill
                                        Widget := stack_box:
                                            Orientation := orientation.Horizontal
                                            Slots := array:
                                                stack_box_slot:
                                                    Padding:=margin{Top:= 10.0, Right:= 5.0, Bottom:= 10.0}
                                                    Distribution:= option{0.5}
                                                    HorizontalAlignment := horizontal_alignment.Right
                                                    VerticalAlignment := vertical_alignment.Fill
                                                    Widget := Now_MoreHitsTextBlock
                                                stack_box_slot:
                                                    Padding:=margin{Top:= 10.0, Left:= 5.0, Bottom:= 10.0}
                                                    Distribution:= option{0.5}
                                                    HorizontalAlignment := horizontal_alignment.Left
                                                    VerticalAlignment := vertical_alignment.Fill
                                                    Widget := Will_MoreHitsTextBlock
                                    stack_box_slot:
                                        Padding:=margin{Top:= 10.0, Bottom:= 10.0}
                                        HorizontalAlignment := horizontal_alignment.Fill
                                        VerticalAlignment := vertical_alignment.Fill
                                        Widget := stack_box:
                                            Orientation := orientation.Horizontal
                                            Slots := array:
                                                stack_box_slot:
                                                    Distribution:= option{0.5}
                                                    HorizontalAlignment := horizontal_alignment.Right
                                                    VerticalAlignment := vertical_alignment.Fill
                                                    Widget := CostTextBlock3
                                                stack_box_slot:
                                                    Distribution:= option{0.1}
                                                    HorizontalAlignment := horizontal_alignment.Center
                                                    VerticalAlignment := vertical_alignment.Fill
                                                    Widget := SiLverOreImage5
                                                stack_box_slot:
                                                    Distribution:= option{0.4}
                                                    HorizontalAlignment := horizontal_alignment.Left
                                                    VerticalAlignment := vertical_alignment.Fill
                                                    Widget := Cost_MoreHitsTextBlock
                                    stack_box_slot:
                                        Distribution:= option{0.2}
                                        HorizontalAlignment := horizontal_alignment.Center
                                        VerticalAlignment := vertical_alignment.Center
                                        Widget := QuietButtonLevelup3
                        stack_box_slot:
                            Distribution:= option{0.25}
                            HorizontalAlignment := horizontal_alignment.Fill
                            VerticalAlignment := vertical_alignment.Fill
                            Widget := stack_box:
                                Orientation := orientation.Vertical
                                Slots := array:
                                    stack_box_slot:
                                        Padding:=margin{Top:= 10.0, Bottom:= 10.0}
                                        HorizontalAlignment := horizontal_alignment.Fill
                                        VerticalAlignment := vertical_alignment.Fill
                                        Widget := stack_box:
                                            Orientation := orientation.Horizontal
                                            Slots := array:
                                                stack_box_slot:
                                                    Padding:=margin{Top:= 10.0, Right:= 5.0, Bottom:= 10.0}
                                                    Distribution:= option{0.5}
                                                    HorizontalAlignment := horizontal_alignment.Right
                                                    VerticalAlignment := vertical_alignment.Fill
                                                    Widget := Now_ComboModeTextBlock
                                                stack_box_slot:
                                                    Padding:=margin{Top:= 10.0, Left:= 5.0, Bottom:= 10.0}
                                                    Distribution:= option{0.5}
                                                    HorizontalAlignment := horizontal_alignment.Left
                                                    VerticalAlignment := vertical_alignment.Fill
                                                    Widget := Will_ComboModeTextBlock
                                    stack_box_slot:
                                        Padding:=margin{Top:= 10.0, Bottom:= 10.0}
                                        HorizontalAlignment := horizontal_alignment.Fill
                                        VerticalAlignment := vertical_alignment.Fill
                                        Widget := stack_box:
                                            Orientation := orientation.Horizontal
                                            Slots := array:
                                                stack_box_slot:
                                                    Distribution:= option{0.5}
                                                    HorizontalAlignment := horizontal_alignment.Right
                                                    VerticalAlignment := vertical_alignment.Fill
                                                    Widget := CostTextBlock4
                                                stack_box_slot:
                                                    Distribution:= option{0.1}
                                                    HorizontalAlignment := horizontal_alignment.Center
                                                    VerticalAlignment := vertical_alignment.Fill
                                                    Widget := SiLverOreImage6
                                                stack_box_slot:
                                                    Distribution:= option{0.4}
                                                    HorizontalAlignment := horizontal_alignment.Left
                                                    VerticalAlignment := vertical_alignment.Fill
                                                    Widget := Cost_ComboModeTextBlock
                                    stack_box_slot:
                                        Distribution:= option{0.2}
                                        HorizontalAlignment := horizontal_alignment.Center
                                        VerticalAlignment := vertical_alignment.Center
                                        Widget := QuietButtonLevelup4
        }


        #========================================== Close ==========================================

        CloseButton : button_loud = button_loud{DefaultText := StringToMessage("CLOSE")}
        CloseButton.OnClick().Subscribe(OnCloseButtonClicked)
        set CanvasSlots += array
        {
            canvas_slot:
                Anchors := CenterAnchors
                Offsets := margin{Top:= 435.0, Left:= 658.0, Right := 353.0, Bottom := 80.0}
                Alignment := CenterVector
                SizeToContent := false
                Widget := CloseButton
        }
        
        if(set UIElements[Agent] = (OreTextBlock, PowerupStoneTextBlock)){}
        if(set SkillUIElements[Agent] = map{}){}
        if:
            set SkillUIElements[Agent] = map:
                1 => (Lv_ComboUnlockTextBlock, Now_ComboUnlockTextBlock, Will_ComboUnlockTextBlock, Cost_ComboUnlockTextBlock, _ComboUnlockTextBlock)
                2 => (Lv_ChargeCountTextBlock, Now_ChargeCountTextBlock, Will_ChargeCountTextBlock, Cost_ChargeCountTextBlock, _ChargeCountTextBlock)
                3 => (Lv_MoreHitsTextBlock, Now_MoreHitsTextBlock, Will_MoreHitsTextBlock, Cost_MoreHitsTextBlock, _MoreHitsTextBlock)
                4 => (Lv_ComboModeTextBlock, Now_ComboModeTextBlock, Will_ComboModeTextBlock, Cost_ComboModeTextBlock, _ComboModeTextBlock)

        Canvas : canvas = canvas:
            Slots := CanvasSlots

        return Canvas