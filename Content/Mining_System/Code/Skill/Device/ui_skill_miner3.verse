
using { /Verse.org/Assets }
using { /Verse.org/Simulation }
using { /Fortnite.com/UI }
using { /Fortnite.com/Devices }
using { /Fortnite.com/Characters }
using { /UnrealEngine.com/Temporary/UI }
using { /UnrealEngine.com/Temporary/SpatialMath }
using { /Verse.org/Simulation }
using { /UnrealEngine.com/Temporary/Diagnostics }
using { /Verse.org/Colors }
using { /Verse.org/Colors/NamedColors }
using { Utils }
using { PlayerData }
using { Skill }


# See https://dev.epicgames.com/documentation/en-us/uefn/create-your-own-device-in-verse for how to create a verse device.

# A Verse-authored creative device that can be placed in a level
ui_skill_miner3 := class(creative_device):
    #SFXs
    @editable
    SFX_Open : audio_player_device = audio_player_device{}
    @editable
    SFX_Close : audio_player_device = audio_player_device{}
    @editable
    SFX_LevelUp : audio_player_device = audio_player_device{}
    @editable
    SFX_Fail : audio_player_device = audio_player_device{}

    
    var _SaveData : save_data = save_data{}
    @editable
    var SaveDataManager : save_data_manager = save_data_manager{}
    @editable
    var SkillDataManager : skill_manager = skill_manager{}
    CloseEvent<public> : event(player) = event(player){}

    @editable
    SkillButton : button_device = button_device{}

    var miner3CanvasPerAgent : [agent]canvas = map{}
    var UIElements : [agent]tuple(text_block, text_block) = map{}
    var SkillUIElements : [agent][int]tuple(text_block, text_block, text_block, text_block, text_block) = map{}
    OnBegin<override>()<suspends>:void=
        SkillButton.InteractedWithEvent.Subscribe(OnButtonEvented)


    SetText<localizes>(Value:[]char) : message = "{Value}"
    # スキル関連のUI要素を全て更新
    UpdateUI(Agent:agent):void=
        UpdateUIElements(Agent)
        UpdateSkillUIElements(Agent)

    # UIElementsに格納された基本リソースのテキストを更新する
    UpdateUIElements(Agent: agent):void=
        var SilverValue : big_number = big_number{}
        var PowerupStoneValue : big_number = big_number{}
        if:
            Player := player[Agent]
        then:
            set SilverValue = SaveDataManager.GetResourceStone(Player)
            set PowerupStoneValue = SaveDataManager.GetResourcePowerupStone(Player)
            if (Elements := UIElements[Agent]):
                Elements(0).SetText(SetText("{BigNumber_FormatDisplay(SilverValue)}"))          # OreTextBlockのテキストを設定
                Elements(1).SetText(SetText("{BigNumber_FormatDisplay(PowerupStoneValue)}"))    # PowerupStoneTextBlockのテキストを設定 

    UpdateSkillUIElements(Agent:agent):void=
        var Lv_ExplosionUnlock : int = 1    #23
        var Now_ExplosionUnlock : int = 1
        var Will_ExplosionUnlock : int = 1
        var Cost_ExplosionUnlock : big_number = big_number{}
        var Lv_AreaExpansion : int = 1  #24
        var Now_AreaExpansion : int = 1
        var Will_AreaExpansion : int = 1
        var Cost_AreaExpansion : big_number = big_number{}
        var Lv_DamageMultiplier : int = 1   #25
        var Now_DamageMultiplier : int = 1
        var Will_DamageMultiplier : int = 1
        var Cost_DamageMultiplier : big_number = big_number{}
        if:
            Player := player[Agent]
        then:
            var LanguageIndex : int = SaveDataManager.GetLanguage(Player)
            if:
                set Lv_ExplosionUnlock = SaveDataManager.GetSkillLevel[Player, 23]
                set Now_ExplosionUnlock = SaveDataManager.GetPlayerSkillEffectIncrease[Player, 23]
                set Will_ExplosionUnlock = SaveDataManager.GetSpecificSkillEffectIncrease[Player, 23]
                set Cost_ExplosionUnlock = SaveDataManager.GetUpgradeCostForSkill[Player, 23]

                set Lv_AreaExpansion = SaveDataManager.GetSkillLevel[Player, 24]
                set Now_AreaExpansion = SaveDataManager.GetPlayerSkillEffectIncrease[Player, 24]
                set Will_AreaExpansion = SaveDataManager.GetSpecificSkillEffectIncrease[Player, 24]
                set Cost_AreaExpansion = SaveDataManager.GetUpgradeCostForSkill[Player, 24]

                set Lv_DamageMultiplier = SaveDataManager.GetSkillLevel[Player, 25]
                set Now_DamageMultiplier = SaveDataManager.GetPlayerSkillEffectIncrease[Player, 25]
                set Will_DamageMultiplier = SaveDataManager.GetSpecificSkillEffectIncrease[Player, 25]
                set Cost_DamageMultiplier = SaveDataManager.GetUpgradeCostForSkill[Player, 25]
            then:
                LvExplosionUnlockText := if(SkillDataManager.CanUpgradeSkill(player[Agent], 23)?){"Lv.{Lv_ExplosionUnlock}"} else {"Lv.MAX"}
                LvExplosionUnlockTextColor := if(SkillDataManager.CanUpgradeSkill(player[Agent], 23)?){NamedColors.White} else {NamedColors.Red}
                WillExplosionUnlockText := if(SkillDataManager.CanUpgradeSkill(player[Agent], 23)?){"+ {Will_ExplosionUnlock}"} else {"MAX!"}
                CostExplosionUnlockText := if(SkillDataManager.CanUpgradeSkill(player[Agent], 23)?){"{BigNumber_FormatDisplay(Cost_ExplosionUnlock)}"} else {"-"}
                CostExplosionUnlockTextColor := if(SkillDataManager.IsUnlockable[player[Agent], 23]){NamedColors.White} else {NamedColors.Red}
                ExpUnText := if(LanguageIndex = 0){"ラストヒット時に衝撃波\nを起こし、近くの岩に\nダメージを与える"} else {"Triggers an explosion on\nlast hit dealing damage\nto nearby stone"}

                LvAreaExpansionText := if(SkillDataManager.CanUpgradeSkill(player[Agent], 24)?){"Lv.{Lv_AreaExpansion}"} else {"Lv.MAX"}
                LvAreaExpansionTextColor := if(SkillDataManager.CanUpgradeSkill(player[Agent], 24)?){NamedColors.White} else {NamedColors.Red}
                WillAreaExpansionText := if(SkillDataManager.CanUpgradeSkill(player[Agent], 24)?){"+ {Will_AreaExpansion}"} else {"MAX!"}
                CostAreaExpansionText := if(SkillDataManager.CanUpgradeSkill(player[Agent], 24)?){"{BigNumber_FormatDisplay(Cost_AreaExpansion)}"} else {"-"}
                CostAreaExpansionTextColor := if(SkillDataManager.IsUnlockable[player[Agent], 24]){NamedColors.White} else {NamedColors.Red}
                AreaExpandText := if(LanguageIndex = 0){"衝撃波の範囲の増加"} else {"Explosion radius increases"}

                LvDamageMultiplierText := if(SkillDataManager.CanUpgradeSkill(player[Agent], 25)?){"Lv.{Lv_DamageMultiplier}"} else {"Lv.MAX"}
                LvDamageMultiplierTextColor := if(SkillDataManager.CanUpgradeSkill(player[Agent], 25)?){NamedColors.White} else {NamedColors.Red}
                WillDamageMultiplierText := if(SkillDataManager.CanUpgradeSkill(player[Agent], 25)?){"+ {Will_DamageMultiplier} %"} else {"MAX!"}
                CostDamageMultiplierText := if(SkillDataManager.CanUpgradeSkill(player[Agent], 25)?){"{BigNumber_FormatDisplay(Cost_DamageMultiplier)}"} else {"-"}
                CostDamageMultiplierTextColor := if(SkillDataManager.IsUnlockable[player[Agent], 25]){NamedColors.White} else {NamedColors.Red}
                DamageMultiText := if(LanguageIndex = 0){"衝撃波で与える\nダメージの増加"} else {"Increase in\nthe damage multiplier applied"}

                if(SkillElement := SkillUIElements[Agent][1]) :
                    SkillElement(0).SetText(SetText(LvExplosionUnlockText))
                    SkillElement(0).SetTextColor(LvExplosionUnlockTextColor)
                    SkillElement(1).SetText(SetText("{Now_ExplosionUnlock}"))
                    SkillElement(2).SetText(SetText(WillExplosionUnlockText))
                    SkillElement(3).SetText(SetText(CostExplosionUnlockText))
                    SkillElement(3).SetTextColor(CostExplosionUnlockTextColor)
                    SkillElement(4).SetText(SetText("{ExpUnText}"))
                if(SkillElement := SkillUIElements[Agent][2]) :
                    SkillElement(0).SetText(SetText(LvAreaExpansionText))
                    SkillElement(0).SetTextColor(LvAreaExpansionTextColor)
                    SkillElement(1).SetText(SetText("{Now_AreaExpansion}"))
                    SkillElement(2).SetText(SetText(WillAreaExpansionText))
                    SkillElement(3).SetText(SetText(CostAreaExpansionText))
                    SkillElement(3).SetTextColor(CostAreaExpansionTextColor)
                    SkillElement(4).SetText(SetText("{AreaExpandText}"))
                if(SkillElement := SkillUIElements[Agent][3]) :
                    SkillElement(0).SetText(SetText(LvDamageMultiplierText))
                    SkillElement(0).SetTextColor(LvDamageMultiplierTextColor)
                    SkillElement(1).SetText(SetText("{Now_DamageMultiplier} %"))
                    SkillElement(2).SetText(SetText(WillDamageMultiplierText))
                    SkillElement(3).SetText(SetText(CostDamageMultiplierText))
                    SkillElement(3).SetTextColor(CostDamageMultiplierTextColor)
                    SkillElement(4).SetText(SetText("{DamageMultiText}"))

    OnPlayerSpawned(Agent : agent):void=
        if (miner3Canvas := miner3CanvasPerAgent[Agent]):
            return

        if (Player := player[Agent], PlayerUI := GetPlayerUI[Player]):
            miner3Canvas : canvas = Createminer3Canvas(Agent)
            if (set miner3CanvasPerAgent[Agent] = miner3Canvas){}

    OnButtonEvented(Agent : agent):void=
        SFX_Open.Play(Agent)
        OnPlayerSpawned(Agent)
        UpdateUI(Agent)
        Print("OnSkillButtonEvented")
        if (Player := player[Agent], PlayerUI := GetPlayerUI[Player], miner3Canvas := miner3CanvasPerAgent[Agent]):
            PlayerUI.AddWidget(miner3Canvas, player_ui_slot{InputMode := ui_input_mode.All})

    # ボタン処理
    OnLevelup1Clicked(WidgetMessage : widget_message):void=
        if (PlayerUI := GetPlayerUI[WidgetMessage.Player], miner3Canvas := miner3CanvasPerAgent[WidgetMessage.Player], Agent : agent = agent[WidgetMessage.Player]):
            if(SkillDataManager.CanUpgradeSkill(WidgetMessage.Player,23) = true):
                SFX_LevelUp.Play(Agent)
                if:
                    SkillDataManager.UnlockOrUpgradeSkill[WidgetMessage.Player,23]
                then:
                    UpdateUI(Agent)
            else:
                SFX_Fail.Play(Agent)
                return
            
    OnLevelup2Clicked(WidgetMessage : widget_message):void=
        if (PlayerUI := GetPlayerUI[WidgetMessage.Player], miner3Canvas := miner3CanvasPerAgent[WidgetMessage.Player], Agent : agent = agent[WidgetMessage.Player]):
            if(SkillDataManager.CanUpgradeSkill(WidgetMessage.Player,24) = true):
                SFX_LevelUp.Play(Agent)
                if:
                    SkillDataManager.UnlockOrUpgradeSkill[WidgetMessage.Player,24]
                then:
                    UpdateUI(Agent)
            else:
                SFX_Fail.Play(Agent)
                return

    OnLevelup3Clicked(WidgetMessage : widget_message):void=
        if (PlayerUI := GetPlayerUI[WidgetMessage.Player], miner3Canvas := miner3CanvasPerAgent[WidgetMessage.Player], Agent : agent = agent[WidgetMessage.Player]):
            if(SkillDataManager.CanUpgradeSkill(WidgetMessage.Player,25) = true):
                SFX_LevelUp.Play(Agent)
                if:
                    SkillDataManager.UnlockOrUpgradeSkill[WidgetMessage.Player,25]
                then:
                    UpdateUI(Agent)
            else:
                SFX_Fail.Play(Agent)
                return
            
    OnLevelup4Clicked(WidgetMessage : widget_message):void=
        if (PlayerUI := GetPlayerUI[WidgetMessage.Player], miner3Canvas := miner3CanvasPerAgent[WidgetMessage.Player]):

    OnCloseButtonClicked(WidgetMessage : widget_message):void=
        if(Agent : agent = agent[WidgetMessage.Player]) :
            SFX_Close.Play(Agent)
        if (PlayerUI := GetPlayerUI[WidgetMessage.Player], miner3Canvas := miner3CanvasPerAgent[WidgetMessage.Player]):
            PlayerUI.RemoveWidget(miner3Canvas)
            CloseEvent.Signal(WidgetMessage.Player)

    # キャンバス
    Createminer3Canvas(Agent : agent):canvas=
        var CanvasSlots : []canvas_slot = array{}
        
        CenterVector : vector2 = vector2{X:=0.5, Y:=0.5}
        CenterAnchors : anchors = anchors{Minimum := CenterVector, Maximum := CenterVector}

        # Image Stock
        BGImage : texture_block = texture_block{DefaultImage := Textures.BG_Skill, DefaultDesiredSize := vector2{X := 1920.0, Y := 1080.0}}
        SilverOreImage1 : texture_block = texture_block{DefaultImage := Textures.SilverOre, DefaultDesiredSize := vector2{X := 50.0, Y := 50.0}}
        PowerupStoneImage : texture_block = texture_block{DefaultImage := Textures.IC_PowerupStone, DefaultDesiredSize := vector2{X := 50.0, Y := 50.0}}
        SiLverOreImage3 : texture_block = texture_block{DefaultImage := Textures.SilverOre, DefaultDesiredSize := vector2{X := 40.0, Y := 40.0}}
        SiLverOreImage4 : texture_block = texture_block{DefaultImage := Textures.SilverOre, DefaultDesiredSize := vector2{X := 40.0, Y := 40.0}}
        SiLverOreImage5 : texture_block = texture_block{DefaultImage := Textures.SilverOre, DefaultDesiredSize := vector2{X := 40.0, Y := 40.0}}
        SiLverOreImage6 : texture_block = texture_block{DefaultImage := Textures.SilverOre, DefaultDesiredSize := vector2{X := 40.0, Y := 40.0}}
        ExplosionUnlockImage : texture_block = texture_block{DefaultImage := Textures.MinerSkill3_1, DefaultDesiredSize := vector2{X := 200.0, Y := 200.0}}
        AreaExpansionImage : texture_block = texture_block{DefaultImage := Textures.MinerSkill3_2, DefaultDesiredSize := vector2{X := 200.0, Y := 200.0}}
        DamageMultiplierImage : texture_block = texture_block{DefaultImage := Textures.MinerSkill3_3, DefaultDesiredSize := vector2{X := 200.0, Y := 200.0}}

        var SilverValue : big_number = big_number{}
        var PowerupStoneValue : big_number = big_number{}
        var Lv_ExplosionUnlock : int = 1    #23
        var Now_ExplosionUnlock : int = 1
        var Will_ExplosionUnlock : int = 1
        var Cost_ExplosionUnlock : big_number = big_number{}
        var Lv_AreaExpansion : int = 1  #24
        var Now_AreaExpansion : int = 1
        var Will_AreaExpansion : int = 1
        var Cost_AreaExpansion : big_number = big_number{}
        var Lv_DamageMultiplier : int = 1   #25
        var Now_DamageMultiplier : int = 1
        var Will_DamageMultiplier : int = 1
        var Cost_DamageMultiplier : big_number = big_number{}
    
        
        #================================== TextBlock ==================================
        
        # 所持岩
        OreTextBlock : text_block = text_block{
            # DefaultText := StringToMessage("{BigNumber_FormatDisplay(SilverValue)}"),
            DefaultTextColor := NamedColors.White,
            DefaultShadowOffset := option{vector2{X:=2.5, Y:=2.5}},
            DefaultShadowColor := NamedColors.Black
        }
        # 所持強化アイテム
        PowerupStoneTextBlock : text_block = text_block{
            # DefaultText := StringToMessage("{BigNumber_FormatDisplay(PowerupStoneValue)}"),
            DefaultTextColor := NamedColors.White,
            DefaultShadowOffset := option{vector2{X:=2.5, Y:=2.5}},
            DefaultShadowColor := NamedColors.Black,
            DefaultJustification := text_justification.Center
        }
    
        #================================== Skill 1 ==================================
    
        Lv_ExplosionUnlockTextBlock : text_block = text_block{
            # DefaultText := StringToMessage("{}"),
            DefaultTextColor := NamedColors.White,
            DefaultShadowOffset := option{vector2{X:=2.5, Y:=2.5}},
            DefaultShadowColor := NamedColors.Black,
            DefaultJustification := text_justification.Center
        }
        Now_ExplosionUnlockTextBlock : text_block = text_block{
            # DefaultText := StringToMessage("{Now_ExplosionUnlock}"),
            DefaultTextColor := NamedColors.White,
            DefaultShadowOffset := option{vector2{X:=2.5, Y:=2.5}},
            DefaultShadowColor := NamedColors.Black,
            DefaultJustification := text_justification.Center
        }
        Will_ExplosionUnlockTextBlock : text_block = text_block{
            # DefaultText := StringToMessage("({WillExplosionUnlockText})"),
            DefaultTextColor := NamedColors.White,
            DefaultShadowOffset := option{vector2{X:=2.5, Y:=2.5}},
            DefaultShadowColor := NamedColors.Black,
            DefaultJustification := text_justification.Center
        }
        Cost_ExplosionUnlockTextBlock : text_block = text_block{
            # DefaultText := StringToMessage("{CostExplosionUnlockText}"),
            DefaultTextColor := NamedColors.White,
            DefaultShadowOffset := option{vector2{X:=2.5, Y:=2.5}},
            DefaultShadowColor := NamedColors.Black,
            DefaultJustification := text_justification.Center
        }
        var _ExplosionUnlockTextBlock :text_block= text_block{
            DefaultTextColor := NamedColors.White,
            DefaultShadowOffset := option{vector2{X:=2.5, Y:=2.5}},
            DefaultShadowColor := NamedColors.Black,
            DefaultShadowOpacity := 1.0,
            DefaultJustification := text_justification.Center
        }
    
        #================================== Skill 2 ==================================
    
        Lv_AreaExpansionTextBlock : text_block = text_block{
            # DefaultText := StringToMessage("{LvAreaExpansionText}"),
            DefaultTextColor := NamedColors.White,
            DefaultShadowOffset := option{vector2{X:=2.5, Y:=2.5}},
            DefaultShadowColor := NamedColors.Black,
            DefaultJustification := text_justification.Center
        }
        Now_AreaExpansionTextBlock : text_block = text_block{
            # DefaultText := StringToMessage("{Now_AreaExpansion}"),
            DefaultTextColor := NamedColors.White,
            DefaultShadowOffset := option{vector2{X:=2.5, Y:=2.5}},
            DefaultShadowColor := NamedColors.Black,
            DefaultJustification := text_justification.Center
        }
        Will_AreaExpansionTextBlock : text_block = text_block{
            # DefaultText := StringToMessage("({WillAreaExpansionText})"),
            DefaultTextColor := NamedColors.White,
            DefaultShadowOffset := option{vector2{X:=2.5, Y:=2.5}},
            DefaultShadowColor := NamedColors.Black,
            DefaultJustification := text_justification.Center
        }
        Cost_AreaExpansionTextBlock : text_block = text_block{
            # DefaultText := StringToMessage("{CostAreaExpansionText}"),
            DefaultTextColor := NamedColors.White,
            DefaultShadowOffset := option{vector2{X:=2.5, Y:=2.5}},
            DefaultShadowColor := NamedColors.Black,
            DefaultJustification := text_justification.Center
        }
        var _AreaExpansionTextBlock :text_block= text_block{
            DefaultTextColor := NamedColors.White,
            DefaultShadowOffset := option{vector2{X:=2.5, Y:=2.5}},
            DefaultShadowColor := NamedColors.Black,
            DefaultShadowOpacity := 1.0,
            DefaultJustification := text_justification.Center
        }
    
        #================================== Skill 3 ==================================
    
        Lv_DamageMultiplierTextBlock : text_block = text_block{
            # DefaultText := StringToMessage("{LvDamageMultiplierText}"),
            DefaultTextColor := NamedColors.White,
            DefaultShadowOffset := option{vector2{X:=2.5, Y:=2.5}},
            DefaultShadowColor := NamedColors.Black,
            DefaultJustification := text_justification.Center
        }
        Now_DamageMultiplierTextBlock : text_block = text_block{
            # DefaultText := StringToMessage("{Now_DamageMultiplier}%"),
            DefaultTextColor := NamedColors.White,
            DefaultShadowOffset := option{vector2{X:=2.5, Y:=2.5}},
            DefaultShadowColor := NamedColors.Black,
            DefaultJustification := text_justification.Center
        }
        Will_DamageMultiplierTextBlock : text_block = text_block{
            # DefaultText := StringToMessage("({})"),
            DefaultTextColor := NamedColors.White,
            DefaultShadowOffset := option{vector2{X:=2.5, Y:=2.5}},
            DefaultShadowColor := NamedColors.Black,
            DefaultJustification := text_justification.Center
        }
        Cost_DamageMultiplierTextBlock : text_block = text_block{
            # DefaultText := StringToMessage("{CostDamageMultiplierText}"),
            DefaultTextColor := NamedColors.White,
            DefaultShadowOffset := option{vector2{X:=2.5, Y:=2.5}},
            DefaultShadowColor := NamedColors.Black,
            DefaultJustification := text_justification.Center
        }
        var _DamageMultiplierTextBlock :text_block= text_block{
            DefaultTextColor := NamedColors.White,
            DefaultShadowOffset := option{vector2{X:=2.5, Y:=2.5}},
            DefaultShadowColor := NamedColors.Black,
            DefaultShadowOpacity := 1.0,
            DefaultJustification := text_justification.Center,
            DefaultOverflowPolicy := text_overflow_policy.Ellipsis
        }
    
        #================================== /TextBlock ==================================
        
        
        #========================================== BG ==========================================

        set CanvasSlots += array
        {
            canvas_slot:
                Anchors := CenterAnchors
                Offsets := margin{Top := 0.0,Left := 0.0, Right := 1920.0, Bottom := 1080.0}
                Alignment := CenterVector
                SizeToContent := false
                Widget := BGImage
        }

        #========================================= Header ========================================

        miner3TextBlock := text_block{
            DefaultText := StringToMessage("Miner Skill 3"),
            DefaultTextColor := NamedColors.White,
            DefaultShadowOffset := option{vector2{X:=2.5, Y:=2.5}},
            DefaultShadowColor := NamedColors.Black,
            DefaultJustification := text_justification.Center
        }
        miner3TextBlock.SetShadowOpacity(1.0)
        set CanvasSlots += array
        {
            canvas_slot:
                Anchors := CenterAnchors
                Offsets := margin{Top:= -472.0, Left:= 0.0, Right := 257.0, Bottom := 41.0}
                Alignment := CenterVector
                SizeToContent := false
                Widget := miner3TextBlock
        }

        OreTextBlock.SetShadowOpacity(1.0)
        PowerupStoneTextBlock.SetShadowOpacity(1.0)

        set CanvasSlots += array

        {
            canvas_slot:
                Anchors := CenterAnchors
                Offsets := margin{Left:=-645.5 ,Top:= -391.0, Right:= 519.0, Bottom:= 59.0}
                Alignment := CenterVector
                SizeToContent := false
                Widget := stack_box:
                    Orientation := orientation.Horizontal
                    Slots := array:
                        stack_box_slot:
                            Distribution:= option{0.5}
                            Padding:=margin{Left:=5.0 ,Top:= 5.0, Right:= 10.0, Bottom:= 5.0}
                            HorizontalAlignment := horizontal_alignment.Fill
                            VerticalAlignment := vertical_alignment.Fill
                            Widget := stack_box:
                                Orientation := orientation.Horizontal
                                Slots := array:
                                    stack_box_slot:
                                        HorizontalAlignment := horizontal_alignment.Center
                                        VerticalAlignment := vertical_alignment.Center
                                        Widget := SilverOreImage1
                                    stack_box_slot:
                                        Distribution:= option{1.0}
                                        HorizontalAlignment := horizontal_alignment.Center
                                        VerticalAlignment := vertical_alignment.Center
                                        Widget := OreTextBlock
                        stack_box_slot:
                            Padding:=margin{Left:=10.0 ,Top:= 5.0, Right:= 5.0, Bottom:= 5.0}
                            Distribution:= option{0.5}
                            HorizontalAlignment := horizontal_alignment.Fill
                            VerticalAlignment := vertical_alignment.Fill
                            Widget := stack_box:
                                Orientation := orientation.Horizontal
                                Slots := array:
                                    stack_box_slot:
                                        HorizontalAlignment := horizontal_alignment.Center
                                        VerticalAlignment := vertical_alignment.Center
                                        Widget := PowerupStoneImage
                                    stack_box_slot:
                                        Distribution:= option{1.0}
                                        HorizontalAlignment := horizontal_alignment.Center
                                        VerticalAlignment := vertical_alignment.Center
                                        Widget := PowerupStoneTextBlock
        }

        #========================================== Skill ==========================================

        CostTextBlock1 := text_block{
            DefaultText := StringToMessage("Cost :"),
            DefaultTextColor := NamedColors.White,
            DefaultShadowOffset := option{vector2{X:=2.5, Y:=2.5}},
            DefaultShadowColor := NamedColors.Black,
            DefaultJustification := text_justification.Center
        }
        CostTextBlock1.SetShadowOpacity(1.0)
        CostTextBlock2 := text_block{
            DefaultText := StringToMessage("Cost :"),
            DefaultTextColor := NamedColors.White,
            DefaultShadowOffset := option{vector2{X:=2.5, Y:=2.5}},
            DefaultShadowColor := NamedColors.Black,
            DefaultJustification := text_justification.Center
        }
        CostTextBlock2.SetShadowOpacity(1.0)
        CostTextBlock3 := text_block{
            DefaultText := StringToMessage("Cost :"),
            DefaultTextColor := NamedColors.White,
            DefaultShadowOffset := option{vector2{X:=2.5, Y:=2.5}},
            DefaultShadowColor := NamedColors.Black,
            DefaultJustification := text_justification.Center
        }
        CostTextBlock3.SetShadowOpacity(1.0)
        CostTextBlock4 := text_block{
            DefaultText := StringToMessage("Cost :"),
            DefaultTextColor := NamedColors.White,
            DefaultShadowOffset := option{vector2{X:=2.5, Y:=2.5}},
            DefaultShadowColor := NamedColors.Black,
            DefaultJustification := text_justification.Center
        }
        CostTextBlock4.SetShadowOpacity(1.0)

        #Skill 1
        ExplosionUnlockTextBlock := text_block{
            DefaultText := StringToMessage("Explosion unlock"),
            DefaultTextColor := NamedColors.White,
            DefaultShadowOffset := option{vector2{X:=2.5, Y:=2.5}},
            DefaultShadowColor := NamedColors.Black,
            DefaultJustification := text_justification.Center
        }
        ExplosionUnlockTextBlock.SetShadowOpacity(1.0)
        Lv_ExplosionUnlockTextBlock.SetShadowOpacity(1.0)
        Now_ExplosionUnlockTextBlock.SetShadowOpacity(1.0)
        Will_ExplosionUnlockTextBlock.SetShadowOpacity(1.0)
        Cost_ExplosionUnlockTextBlock.SetShadowOpacity(1.0)
        
        #Skill 2
        AreaExpansionTextBlock := text_block{
            DefaultText := StringToMessage("Area expansion"),
            DefaultTextColor := NamedColors.White,
            DefaultShadowOffset := option{vector2{X:=2.5, Y:=2.5}},
            DefaultShadowColor := NamedColors.Black,
            DefaultJustification := text_justification.Center
        }
        AreaExpansionTextBlock.SetShadowOpacity(1.0)
        Lv_AreaExpansionTextBlock.SetShadowOpacity(1.0)
        Now_AreaExpansionTextBlock.SetShadowOpacity(1.0)
        Will_AreaExpansionTextBlock.SetShadowOpacity(1.0)
        Cost_AreaExpansionTextBlock.SetShadowOpacity(1.0)
        
        #Skill 3
        DamageMultiplierTextBlock := text_block{
            DefaultText := StringToMessage("Damage multiplier"),
            DefaultTextColor := NamedColors.White,
            DefaultShadowOffset := option{vector2{X:=2.5, Y:=2.5}},
            DefaultShadowColor := NamedColors.Black,
            DefaultJustification := text_justification.Center
        }
        DamageMultiplierTextBlock.SetShadowOpacity(1.0)
        Lv_DamageMultiplierTextBlock.SetShadowOpacity(1.0)
        Now_DamageMultiplierTextBlock.SetShadowOpacity(1.0)
        Will_DamageMultiplierTextBlock.SetShadowOpacity(1.0)
        Cost_DamageMultiplierTextBlock.SetShadowOpacity(1.0)

        # Button Stock
        QuietButtonLevelup1 : button_quiet = button_quiet{DefaultText := StringToMessage("LEVEL UP")}
        QuietButtonLevelup1.OnClick().Subscribe(OnLevelup1Clicked)
        QuietButtonLevelup2 : button_quiet = button_quiet{DefaultText := StringToMessage("LEVEL UP")}
        QuietButtonLevelup2.OnClick().Subscribe(OnLevelup2Clicked)
        QuietButtonLevelup3 : button_quiet = button_quiet{DefaultText := StringToMessage("LEVEL UP")}
        QuietButtonLevelup3.OnClick().Subscribe(OnLevelup3Clicked)
        QuietButtonLevelup4 : button_quiet = button_quiet{DefaultText := StringToMessage("LEVEL UP")}
        QuietButtonLevelup4.OnClick().Subscribe(OnLevelup4Clicked)

        #================================ Top ================================

        set CanvasSlots += array
        {
            canvas_slot:
                Anchors := CenterAnchors
                Offsets := margin{Left:=0.0 ,Top:= -96.5, Right:= 1760.0, Bottom:= 464.0}
                Alignment := CenterVector
                SizeToContent := false
                Widget := stack_box:
                    Orientation := orientation.Horizontal
                    Slots := array:
                        stack_box_slot:
                            Padding:=margin{Left:= 50.0, Right:= 50.0}
                            Distribution:= option{0.25}
                            HorizontalAlignment := horizontal_alignment.Fill
                            VerticalAlignment := vertical_alignment.Fill
                            Widget := stack_box:
                                Orientation := orientation.Vertical
                                Slots := array:
                                    stack_box_slot:
                                        Padding:=margin{Top:= 10.0, Bottom:= 10.0}
                                        HorizontalAlignment := horizontal_alignment.Fill
                                        VerticalAlignment := vertical_alignment.Fill
                                        Widget := ExplosionUnlockTextBlock
                                    stack_box_slot:
                                        Padding:=margin{Top:= 15.0, Bottom:= 15.0}
                                        HorizontalAlignment := horizontal_alignment.Fill
                                        VerticalAlignment := vertical_alignment.Fill
                                        Widget := Lv_ExplosionUnlockTextBlock
                                    stack_box_slot:
                                        Padding:=margin{Top:= 10.0, Bottom:= 10.0}
                                        HorizontalAlignment := horizontal_alignment.Center
                                        VerticalAlignment := vertical_alignment.Center
                                        Widget := ExplosionUnlockImage
                                    stack_box_slot:
                                        Padding:=margin{Top:= 15.0, Bottom:= 15.0}
                                        HorizontalAlignment := horizontal_alignment.Fill
                                        VerticalAlignment := vertical_alignment.Fill
                                        Widget := _ExplosionUnlockTextBlock
                        stack_box_slot:
                            Padding:=margin{Left:= 50.0, Right:= 50.0}
                            Distribution:= option{0.25}
                            HorizontalAlignment := horizontal_alignment.Fill
                            VerticalAlignment := vertical_alignment.Fill
                            Widget := stack_box:
                                Orientation := orientation.Vertical
                                Slots := array:
                                    stack_box_slot:
                                        Padding:=margin{Top:= 10.0, Bottom:= 10.0}
                                        HorizontalAlignment := horizontal_alignment.Fill
                                        VerticalAlignment := vertical_alignment.Fill
                                        Widget := AreaExpansionTextBlock
                                    stack_box_slot:
                                        Padding:=margin{Top:= 15.0, Bottom:= 15.0}
                                        HorizontalAlignment := horizontal_alignment.Fill
                                        VerticalAlignment := vertical_alignment.Fill
                                        Widget := Lv_AreaExpansionTextBlock
                                    stack_box_slot:
                                        Padding:=margin{Top:= 10.0, Bottom:= 10.0}
                                        HorizontalAlignment := horizontal_alignment.Center
                                        VerticalAlignment := vertical_alignment.Center
                                        Widget := AreaExpansionImage
                                    stack_box_slot:
                                        Padding:=margin{Top:= 15.0, Bottom:= 15.0}
                                        HorizontalAlignment := horizontal_alignment.Fill
                                        VerticalAlignment := vertical_alignment.Fill
                                        Widget := _AreaExpansionTextBlock
                        stack_box_slot:
                            Padding:=margin{Left:= 50.0, Right:= 50.0}
                            Distribution:= option{0.25}
                            HorizontalAlignment := horizontal_alignment.Fill
                            VerticalAlignment := vertical_alignment.Fill
                            Widget := stack_box:
                                Orientation := orientation.Vertical
                                Slots := array:
                                    stack_box_slot:
                                        Padding:=margin{Top:= 10.0, Bottom:= 10.0}
                                        HorizontalAlignment := horizontal_alignment.Fill
                                        VerticalAlignment := vertical_alignment.Fill
                                        Widget := DamageMultiplierTextBlock
                                    stack_box_slot:
                                        Padding:=margin{Top:= 15.0, Bottom:= 15.0}
                                        HorizontalAlignment := horizontal_alignment.Fill
                                        VerticalAlignment := vertical_alignment.Fill
                                        Widget := Lv_DamageMultiplierTextBlock
                                    stack_box_slot:
                                        Padding:=margin{Top:= 10.0, Bottom:= 10.0}
                                        HorizontalAlignment := horizontal_alignment.Center
                                        VerticalAlignment := vertical_alignment.Center
                                        Widget := DamageMultiplierImage
                                    stack_box_slot:
                                        Padding:=margin{Top:= 15.0, Bottom:= 15.0}
                                        HorizontalAlignment := horizontal_alignment.Fill
                                        VerticalAlignment := vertical_alignment.Fill
                                        Widget := _DamageMultiplierTextBlock
        }

        
        #================================ Bottom ================================

        set CanvasSlots += array
        {
            canvas_slot:
                Anchors := CenterAnchors
                Offsets := margin{Left:=0.0 ,Top:= 249.5, Right:= 1760.0, Bottom:= 210.0}
                Alignment := CenterVector
                SizeToContent := false
                Widget := stack_box:
                    Orientation := orientation.Horizontal
                    Slots := array:
                        stack_box_slot:
                            Distribution:= option{0.25}
                            HorizontalAlignment := horizontal_alignment.Fill
                            VerticalAlignment := vertical_alignment.Fill
                            Widget := stack_box:
                                Orientation := orientation.Vertical
                                Slots := array:
                                    stack_box_slot:
                                        Padding:=margin{Top:= 10.0, Bottom:= 10.0}
                                        HorizontalAlignment := horizontal_alignment.Fill
                                        VerticalAlignment := vertical_alignment.Fill
                                        Widget := stack_box:
                                            Orientation := orientation.Horizontal
                                            Slots := array:
                                                stack_box_slot:
                                                    Padding:=margin{Top:= 10.0, Right:= 5.0, Bottom:= 10.0}
                                                    Distribution:= option{0.5}
                                                    HorizontalAlignment := horizontal_alignment.Right
                                                    VerticalAlignment := vertical_alignment.Fill
                                                    Widget := Now_ExplosionUnlockTextBlock
                                                stack_box_slot:
                                                    Padding:=margin{Top:= 10.0, Left:= 5.0, Bottom:= 10.0}
                                                    Distribution:= option{0.5}
                                                    HorizontalAlignment := horizontal_alignment.Left
                                                    VerticalAlignment := vertical_alignment.Fill
                                                    Widget := Will_ExplosionUnlockTextBlock
                                    stack_box_slot:
                                        Padding:=margin{Top:= 10.0, Bottom:= 10.0}
                                        HorizontalAlignment := horizontal_alignment.Fill
                                        VerticalAlignment := vertical_alignment.Fill
                                        Widget := stack_box:
                                            Orientation := orientation.Horizontal
                                            Slots := array:
                                                stack_box_slot:
                                                    Distribution:= option{0.5}
                                                    HorizontalAlignment := horizontal_alignment.Right
                                                    VerticalAlignment := vertical_alignment.Fill
                                                    Widget := CostTextBlock1
                                                stack_box_slot:
                                                    Distribution:= option{0.1}
                                                    HorizontalAlignment := horizontal_alignment.Center
                                                    VerticalAlignment := vertical_alignment.Fill
                                                    Widget := SiLverOreImage3
                                                stack_box_slot:
                                                    Distribution:= option{0.4}
                                                    HorizontalAlignment := horizontal_alignment.Left
                                                    VerticalAlignment := vertical_alignment.Fill
                                                    Widget := Cost_ExplosionUnlockTextBlock
                                    stack_box_slot:
                                        Distribution:= option{0.2}
                                        HorizontalAlignment := horizontal_alignment.Center
                                        VerticalAlignment := vertical_alignment.Center
                                        Widget := QuietButtonLevelup1
                        stack_box_slot:
                            Distribution:= option{0.25}
                            HorizontalAlignment := horizontal_alignment.Fill
                            VerticalAlignment := vertical_alignment.Fill
                            Widget := stack_box:
                                Orientation := orientation.Vertical
                                Slots := array:
                                    stack_box_slot:
                                        Padding:=margin{Top:= 10.0, Bottom:= 10.0}
                                        HorizontalAlignment := horizontal_alignment.Fill
                                        VerticalAlignment := vertical_alignment.Fill
                                        Widget := stack_box:
                                            Orientation := orientation.Horizontal
                                            Slots := array:
                                                stack_box_slot:
                                                    Padding:=margin{Top:= 10.0, Right:= 5.0, Bottom:= 10.0}
                                                    Distribution:= option{0.5}
                                                    HorizontalAlignment := horizontal_alignment.Right
                                                    VerticalAlignment := vertical_alignment.Fill
                                                    Widget := Now_AreaExpansionTextBlock
                                                stack_box_slot:
                                                    Padding:=margin{Top:= 10.0, Left:= 5.0, Bottom:= 10.0}
                                                    Distribution:= option{0.5}
                                                    HorizontalAlignment := horizontal_alignment.Left
                                                    VerticalAlignment := vertical_alignment.Fill
                                                    Widget := Will_AreaExpansionTextBlock
                                    stack_box_slot:
                                        Padding:=margin{Top:= 10.0, Bottom:= 10.0}
                                        HorizontalAlignment := horizontal_alignment.Fill
                                        VerticalAlignment := vertical_alignment.Fill
                                        Widget := stack_box:
                                            Orientation := orientation.Horizontal
                                            Slots := array:
                                                stack_box_slot:
                                                    Distribution:= option{0.5}
                                                    HorizontalAlignment := horizontal_alignment.Right
                                                    VerticalAlignment := vertical_alignment.Fill
                                                    Widget := CostTextBlock2
                                                stack_box_slot:
                                                    Distribution:= option{0.1}
                                                    HorizontalAlignment := horizontal_alignment.Center
                                                    VerticalAlignment := vertical_alignment.Fill
                                                    Widget := SiLverOreImage4
                                                stack_box_slot:
                                                    Distribution:= option{0.4}
                                                    HorizontalAlignment := horizontal_alignment.Left
                                                    VerticalAlignment := vertical_alignment.Fill
                                                    Widget := Cost_AreaExpansionTextBlock
                                    stack_box_slot:
                                        Distribution:= option{0.2}
                                        HorizontalAlignment := horizontal_alignment.Center
                                        VerticalAlignment := vertical_alignment.Center
                                        Widget := QuietButtonLevelup2

                        stack_box_slot:
                            Distribution:= option{0.25}
                            HorizontalAlignment := horizontal_alignment.Fill
                            VerticalAlignment := vertical_alignment.Fill
                            Widget := stack_box:
                                Orientation := orientation.Vertical
                                Slots := array:
                                    stack_box_slot:
                                        Padding:=margin{Top:= 10.0, Bottom:= 10.0}
                                        HorizontalAlignment := horizontal_alignment.Fill
                                        VerticalAlignment := vertical_alignment.Fill
                                        Widget := stack_box:
                                            Orientation := orientation.Horizontal
                                            Slots := array:
                                                stack_box_slot:
                                                    Padding:=margin{Top:= 10.0, Right:= 5.0, Bottom:= 10.0}
                                                    Distribution:= option{0.5}
                                                    HorizontalAlignment := horizontal_alignment.Right
                                                    VerticalAlignment := vertical_alignment.Fill
                                                    Widget := Now_DamageMultiplierTextBlock
                                                stack_box_slot:
                                                    Padding:=margin{Top:= 10.0, Left:= 5.0, Bottom:= 10.0}
                                                    Distribution:= option{0.5}
                                                    HorizontalAlignment := horizontal_alignment.Left
                                                    VerticalAlignment := vertical_alignment.Fill
                                                    Widget := Will_DamageMultiplierTextBlock
                                    stack_box_slot:
                                        Padding:=margin{Top:= 10.0, Bottom:= 10.0}
                                        HorizontalAlignment := horizontal_alignment.Fill
                                        VerticalAlignment := vertical_alignment.Fill
                                        Widget := stack_box:
                                            Orientation := orientation.Horizontal
                                            Slots := array:
                                                stack_box_slot:
                                                    Distribution:= option{0.5}
                                                    HorizontalAlignment := horizontal_alignment.Right
                                                    VerticalAlignment := vertical_alignment.Fill
                                                    Widget := CostTextBlock3
                                                stack_box_slot:
                                                    Distribution:= option{0.1}
                                                    HorizontalAlignment := horizontal_alignment.Center
                                                    VerticalAlignment := vertical_alignment.Fill
                                                    Widget := SiLverOreImage5
                                                stack_box_slot:
                                                    Distribution:= option{0.4}
                                                    HorizontalAlignment := horizontal_alignment.Left
                                                    VerticalAlignment := vertical_alignment.Fill
                                                    Widget := Cost_DamageMultiplierTextBlock
                                    stack_box_slot:
                                        Distribution:= option{0.2}
                                        HorizontalAlignment := horizontal_alignment.Center
                                        VerticalAlignment := vertical_alignment.Center
                                        Widget := QuietButtonLevelup3
        }


        #========================================== Close ==========================================

        CloseButton : button_loud = button_loud{DefaultText := StringToMessage("CLOSE")}
        CloseButton.OnClick().Subscribe(OnCloseButtonClicked)
        set CanvasSlots += array
        {
            canvas_slot:
                Anchors := CenterAnchors
                Offsets := margin{Top:= 447.0, Left:= 0.0, Right := 170.0, Bottom := 80.0}
                Alignment := CenterVector
                SizeToContent := false
                Widget := CloseButton
        }
        
        if(set UIElements[Agent] = (OreTextBlock, PowerupStoneTextBlock)){}
        if(set SkillUIElements[Agent] = map{}){}
        if:
            set SkillUIElements[Agent] = map:
                1 => (Lv_ExplosionUnlockTextBlock, Now_ExplosionUnlockTextBlock, Will_ExplosionUnlockTextBlock, Cost_ExplosionUnlockTextBlock, _ExplosionUnlockTextBlock)
                2 => (Lv_AreaExpansionTextBlock, Now_AreaExpansionTextBlock, Will_AreaExpansionTextBlock, Cost_AreaExpansionTextBlock, _AreaExpansionTextBlock)
                3 => (Lv_DamageMultiplierTextBlock, Now_DamageMultiplierTextBlock, Will_DamageMultiplierTextBlock, Cost_DamageMultiplierTextBlock, _DamageMultiplierTextBlock)


        Canvas : canvas = canvas:
            Slots := CanvasSlots

        return Canvas