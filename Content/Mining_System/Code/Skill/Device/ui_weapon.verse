
using { /Verse.org/Assets }
using { /Verse.org/Simulation }
using { /Fortnite.com/UI }
using { /Fortnite.com/Devices }
using { /Fortnite.com/Characters }
using { /UnrealEngine.com/Temporary/UI }
using { /UnrealEngine.com/Temporary/SpatialMath }
using { /Verse.org/Simulation }
using { /UnrealEngine.com/Temporary/Diagnostics }
using { /Verse.org/Colors }
using { /Verse.org/Colors/NamedColors }
using { Utils }
using { PlayerData }

using {Mining_System.Code.PlayerData}
using {Mining_System.Code.MiningObject}
using {Mining_System.Code.Weapon}

ui_weapon<public> := class(creative_device):
    @editable
    SFX_LevelUp2 : audio_player_device = audio_player_device{}
    @editable
    SFX_LevelUp3 : audio_player_device = audio_player_device{}
    @editable
    SFX_Fail : audio_player_device = audio_player_device{}
    @editable
    SaveDataManager : save_data_manager = save_data_manager{}
    @editable
    NPC : character_device = character_device{}
    @editable
    WeaponDamage : weapon_damage = weapon_damage{}
    @editable
    WeaponManager : weapon_manager = weapon_manager{}
    @editable
    MiningDamageDevice : mining_damage_device = mining_damage_device{}
    @editable
    MiningLevelDevice : mining_level_device = mining_level_device{}

    # 通常購入時の初期コスト
    NormalPurchaseCost_Stone<public>: big_number = big_number{Value:=2.00, Exponent:=12}
    #武器の背景色
    Color_Rare : color = color{R := 0.0, G := 0.367893, B := 1.0}
    Color_Epic : color = color{R := 0.441209, G := 0.0, B := 1.0}
    Color_Mythic : color = color{R := 1.0, G := 0.690055, B := 0.0}

    @editable
    QuestTrigger : trigger_device = trigger_device{}
    @editable
    QuestTrigger2 : trigger_device = trigger_device{}

    var WeaponCanvasPerAgent : [agent]canvas = map{}
    var UIElements : [agent]tuple(
        #背景と白線
        texture_block, texture_block, color_block, color_block,
        #現在の武器
        color_block, texture_block, text_block,
        #次の武器
        color_block, texture_block, text_block,
        #矢印
        text_block,
        #ローカライズテキスト
        text_block, text_block, text_block, text_block,
        #現在の実数値
        text_block, text_block, text_block, text_block,
        #矢印
        text_block, text_block, text_block, text_block,
        #次の実数値
        text_block, text_block, text_block, text_block,
        #コスト
        text_block, texture_block, text_block,
        #ボタン
        button_quiet, button_loud
        ) = map{}

    CloseEvent<public> : event(player) = event(player){}

    OnBegin<override>()<suspends>:void=
        NPC.InteractedWithEvent.Subscribe(OnWeaponEvented)

    SetText<localizes>(Value:[]char) : message = "{Value}"
    # スキル関連のUI要素を全て更新
    UpdateUI(Agent:agent):void=
        UpdateUIElements(Agent)

    UpdateUIElements(Agent:agent):void=
        if:
            Player := player[Agent]
            ResourceCount := SaveDataManager.GetResourceStone(Player)
            Level := SaveDataManager.GetWeaponLevel(Player)
            Cost := WeaponManager.CalculatePurchaseCost(Level)
        then:
            var LanguageIndex : int = SaveDataManager.GetLanguage(Player)

            Value_LocalLevel := SaveDataManager.GetWeaponLevel(Player)

            Value_NowBaseStr := WeaponManager.GetWeaponAttackPower(Player)
            Value_NowStrMulti := WeaponManager.GetWeaponAttackPowerPercent(Player)
            Value_NowCriRate  := WeaponManager.GetWeaponCriticalRate(Player)
            Value_NowCriMulti  := WeaponManager.GetWeaponCriticalMultiplier(Player)

            Value_NextBaseStr := WeaponManager.GetNextWeaponAttackPower(Player)
            Value_NextStrMulti := WeaponManager.GetNextWeaponAttackPowerPercent(Player)
            Value_NextCriRate  := WeaponManager.GetNextWeaponCriticalRate(Player)
            Value_NextCriMulti  := WeaponManager.GetNextWeaponCriticalMultiplier(Player)
            
            Value_LocalCost := WeaponManager.GetWeaponPurchaseCost(Player)

            MyCostTextColor := if(BigNumber_GreaterThanOrEquals(ResourceCount, Cost)?){NamedColors.White} else {NamedColors.Red}
            BaseStrText := if(LanguageIndex = 0){"基礎攻撃力 :"} else {"Base Attack Power :"}
            StrMultiText := if(LanguageIndex = 0){"攻撃倍率 :"} else {"Attack Multiplier :"}
            CriRateText := if(LanguageIndex = 0){"クリティカル率 :"} else {"Critical Rate :"}
            CriDamageText := if(LanguageIndex = 0){"クリティカル倍率 :"} else {"Critical Multiplier :"}
            if(Element := UIElements[Agent]) :
                #現在の武器
                if(Level = 0) :
                    Element(4).SetVisibility(widget_visibility.Collapsed)
                    Element(5).SetVisibility(widget_visibility.Collapsed)
                    Element(6).SetVisibility(widget_visibility.Collapsed)
                else :
                    if(Level < 40) :
                        Element(4).SetColor(Color_Rare)
                        Element(6).SetText(StringToMessage("Lv.{Level}/40"))
                    else if(Level >= 40, Level < 80) :
                        Element(4).SetColor(Color_Epic)
                        Element(6).SetText(StringToMessage("Lv.{Level}/80"))
                    else if(Level >= 80) :
                        Element(4).SetColor(Color_Mythic)
                        Element(6).SetText(StringToMessage("Lv.{Level}/999"))
                    Element(4).SetVisibility(widget_visibility.Visible)
                    Element(5).SetVisibility(widget_visibility.Visible)
                    Element(6).SetVisibility(widget_visibility.Visible)
                #次の武器
                if((Level + 1) < 40) :
                    Element(7).SetColor(Color_Rare)
                    Element(9).SetText(StringToMessage("Lv.{Level + 1}/40"))

                if((Level + 1) = 40) :
                    Element(7).SetColor(Color_Epic)
                    Element(9).SetText(StringToMessage("Lv.{Level + 1}/40 (Evolve!)"))

                if((Level + 1) > 40, (Level + 1) < 80) :
                    Element(9).SetText(StringToMessage("Lv.{Level + 1}/80"))
                    Element(7).SetColor(Color_Epic)

                if((Level + 1) = 80) :
                    Element(7).SetColor(Color_Epic)
                    Element(9).SetText(StringToMessage("Lv.{Level + 1}/80 (Evolve!)"))

                if((Level + 1) > 80) :
                    Element(9).SetText(StringToMessage("Lv.{Level + 1}/999"))
                    Element(7).SetColor(Color_Mythic)
                #ローカライズテキスト
                Element(11).SetText(StringToMessage(BaseStrText))
                Element(12).SetText(StringToMessage(StrMultiText))
                Element(13).SetText(StringToMessage(CriRateText))
                Element(14).SetText(StringToMessage(CriDamageText))
                #現在の実数値
                Element(15).SetText(StringToMessage("{FloatToStringWithPrecision(Value_NowBaseStr, 0)}"))
                Element(16).SetText(StringToMessage("× {FloatToStringWithPrecision(Value_NowStrMulti + 100.0, 0)}%"))
                Element(17).SetText(StringToMessage("{FloatToStringWithPrecision(Value_NowCriRate, 0)}%"))
                Element(18).SetText(StringToMessage("× {FloatToStringWithPrecision(Value_NowCriMulti + 100.0, 0)}%"))
                #次の実数値
                Element(23).SetText(StringToMessage("{FloatToStringWithPrecision(Value_NextBaseStr, 0)}"))
                Element(24).SetText(StringToMessage("× {FloatToStringWithPrecision(Value_NextStrMulti + 100.0, 0)}%"))
                Element(25).SetText(StringToMessage("{FloatToStringWithPrecision(Value_NextCriRate, 0)}%"))
                Element(26).SetText(StringToMessage("× {FloatToStringWithPrecision(Value_NextCriMulti + 100.0, 0)}%"))
                #コスト
                Element(29).SetTextColor(MyCostTextColor)
                Element(29).SetText(StringToMessage("{BigNumber_FormatDisplay(Cost)}"))

    OnPlayerSpawned(Agent : agent):void=
        if (WeaponCanvas := WeaponCanvasPerAgent[Agent]):
            return

        if (Player := player[Agent], PlayerUI := GetPlayerUI[Player]):
            WeaponCanvas : canvas = CreateWeaponCanvas(Agent)
            if (set WeaponCanvasPerAgent[Agent] = WeaponCanvas){}

    OnWeaponEvented(Agent : agent):void=
        OnPlayerSpawned(Agent)
        UpdateUI(Agent)
        Print("OnWeaponEvented")

        if (Player := player[Agent], PlayerUI := GetPlayerUI[Player], WeaponCanvas := WeaponCanvasPerAgent[Agent]):
            PlayerUI.AddWidget(WeaponCanvas, player_ui_slot{InputMode := ui_input_mode.All})

    # ボタン処理
    OnLevelUpButtonClicked(WidgetMessage : widget_message):void=
        if :
            Player : player = player[WidgetMessage.Player]
            PlayerUI := GetPlayerUI[WidgetMessage.Player]
            WeaponCanvas := WeaponCanvasPerAgent[WidgetMessage.Player]
        then :
            ResourceCount := SaveDataManager.GetResourceStone(Player)
            Level := SaveDataManager.GetWeaponLevel(WidgetMessage.Player)
            Cost := WeaponManager.CalculatePurchaseCost(Level)
            if(BigNumber_GreaterThanOrEquals(ResourceCount, Cost)?):
                #購入成功時の処理
                SFX_LevelUp3.Play(Player)
                SaveDataManager.SubtractResourceStone(Player, Cost)# リソースを減算
                SaveDataManager.AddWeaponLevel(Player, 1)
                UpdateUI(Player)
                QuestTrigger2.Trigger(Player)
                if(Level=0 or Level=39 or Level=79):
                    WeaponManager.GrantSwordToPlayer(WidgetMessage.Player)
                    SFX_LevelUp2.Play(Player)
                    QuestTrigger.Trigger(Player)
            else:
                #購入失敗時の処理
                SFX_Fail.Play(Player)
            

    OnCloseButtonClicked(WidgetMessage : widget_message):void=
        if (PlayerUI := GetPlayerUI[WidgetMessage.Player], WeaponCanvas := WeaponCanvasPerAgent[WidgetMessage.Player]):
            PlayerUI.RemoveWidget(WeaponCanvas)
            CloseEvent.Signal(WidgetMessage.Player)



    # キャンバス
    CreateWeaponCanvas(Agent : agent):canvas=
        var CanvasSlots : []canvas_slot = array{}
        
        CenterVector : vector2 = vector2{X:=0.5, Y:=0.5}
        CenterAnchors : anchors = anchors{Minimum := CenterVector, Maximum := CenterVector}
        #背景と白線
        Image_bg : texture_block = texture_block:
            DefaultImage := icon_set_2.bg
            DefaultDesiredSize := vector2{X := 1858.864136, Y := 999.300232}
        Image_BG_s : texture_block = texture_block:
            DefaultImage := icon_set_2.BG_s
            DefaultDesiredSize := vector2{X := 1373.675537, Y := 795.110901}
        CB_Line : color_block = color_block:
            DefaultColor := NamedColors.White
            DefaultOpacity := 1.0
            DefaultDesiredSize := vector2{X := 900.0, Y := 3.0}
        CB_Line1 : color_block = color_block:
            DefaultColor := NamedColors.White
            DefaultOpacity := 1.0
            DefaultDesiredSize := vector2{X := 900.0, Y := 3.0}
        #現在の武器とレベル
        CB_NowWeapon : color_block = color_block:
            DefaultColor := color{R := 0.0, G := 0.367893, B := 1.0}
            DefaultOpacity := 0.496
            DefaultDesiredSize := vector2{X := 200.0, Y := 200.0}
        Image_NowWeapon : texture_block = texture_block:
            DefaultImage := icon_set_2.Kinetic_Blade
            DefaultDesiredSize := vector2{X := 200.0, Y := 200.0}
        TB_NowLv : text_block = text_block:
            DefaultText := StringToMessage("Lv.0/40")
            DefaultTextColor := NamedColors.White
            DefaultShadowOpacity := 0.0
        #次の武器とレベル
        CB_NextWeapon : color_block = color_block:
            DefaultColor := color{R := 0.0, G := 0.367893, B := 1.0}
            DefaultOpacity := 0.496
            DefaultDesiredSize := vector2{X := 200.0, Y := 200.0}
        Image_NextWeapon : texture_block = texture_block:
            DefaultImage := icon_set_2.Kinetic_Blade
            DefaultDesiredSize := vector2{X := 200.0, Y := 200.0}
        TB_NextLv : text_block = text_block:
            DefaultText := StringToMessage("Lv.1/40")
            DefaultTextColor := color{R := 0.090634, G := 1.0, B := 0.049568}
            DefaultShadowOpacity := 0.0
        #矢印
        TB_Arrow : text_block = text_block:
            DefaultText := StringToMessage(">>>")
            DefaultTextColor := NamedColors.White
            DefaultShadowOpacity := 0.0
        #ローカライズテキスト
        TB_BaseStr : text_block = text_block:
            DefaultText := StringToMessage("(仮)")
            DefaultTextColor := NamedColors.White
            DefaultShadowOpacity := 0.0
        TB_StrMulti : text_block = text_block:
            DefaultText := StringToMessage("(仮)")
            DefaultTextColor := NamedColors.White
            DefaultShadowOpacity := 0.0
        TB_CriRate : text_block = text_block:
            DefaultText := StringToMessage("(仮)")
            DefaultTextColor := NamedColors.White
            DefaultShadowOpacity := 0.0
        TB_CriDamage : text_block = text_block:
            DefaultText := StringToMessage("(仮)")
            DefaultTextColor := NamedColors.White
            DefaultShadowOpacity := 0.0
        #現在の実数値
        TB_NowBaseStr : text_block = text_block:
            DefaultText := StringToMessage("0")
            DefaultTextColor := NamedColors.White
            DefaultShadowOpacity := 0.0
        TB_NowStrMulti : text_block = text_block:
            DefaultText := StringToMessage("0%")
            DefaultTextColor := NamedColors.White
            DefaultShadowOpacity := 0.0
        TB_NowCriRate : text_block = text_block:
            DefaultText := StringToMessage("0%")
            DefaultTextColor := NamedColors.White
            DefaultShadowOpacity := 0.0
        TB_NowCriDamage : text_block = text_block:
            DefaultText := StringToMessage("0")
            DefaultTextColor := NamedColors.White
            DefaultShadowOpacity := 0.0
        #矢印
        TB_Arrow1 : text_block = text_block:
            DefaultText := StringToMessage(">>>")
            DefaultTextColor := NamedColors.White
            DefaultShadowOpacity := 0.0
        TB_Arrow2 : text_block = text_block:
            DefaultText := StringToMessage(">>>")
            DefaultTextColor := NamedColors.White
            DefaultShadowOpacity := 0.0
        TB_Arrow3 : text_block = text_block:
            DefaultText := StringToMessage(">>>")
            DefaultTextColor := NamedColors.White
            DefaultShadowOpacity := 0.0
        TB_Arrow4 : text_block = text_block:
            DefaultText := StringToMessage(">>>")
            DefaultTextColor := NamedColors.White
            DefaultShadowOpacity := 0.0
        #次の実数値
        TB_NextBaseStr : text_block = text_block:
            DefaultText := StringToMessage("(仮)")
            DefaultTextColor := color{R := 0.090634, G := 1.0, B := 0.049568}
            DefaultShadowOpacity := 0.0
        TB_NextStrMulti : text_block = text_block:
            DefaultText := StringToMessage("(仮)")
            DefaultTextColor := color{R := 0.090634, G := 1.0, B := 0.049568}
            DefaultShadowOpacity := 0.0
        TB_NextCriRate : text_block = text_block:
            DefaultText := StringToMessage("(仮)")
            DefaultTextColor := color{R := 0.090634, G := 1.0, B := 0.049568}
            DefaultShadowOpacity := 0.0
        TB_NextCriDamage : text_block = text_block:
            DefaultText := StringToMessage("(仮)")
            DefaultTextColor := color{R := 0.090634, G := 1.0, B := 0.049568}
            DefaultShadowOpacity := 0.0
        #コスト
        TB_Cost : text_block = text_block:
            DefaultText := StringToMessage("Cost :")
            DefaultTextColor := NamedColors.White
            DefaultShadowOpacity := 0.0
        Image_Ore : texture_block = texture_block:
            DefaultImage := icon_set_2.SilverOre
            DefaultDesiredSize := vector2{X := 50.0, Y := 50.0}
        TB_Cost1 : text_block = text_block:
            DefaultText := StringToMessage("(仮)")
            DefaultTextColor := NamedColors.White
            DefaultShadowOpacity := 0.0
        #ボタン
        LevelUpButton : button_quiet = button_quiet{DefaultText := StringToMessage("Upgrade !")}
        LevelUpButton.OnClick().Subscribe(OnLevelUpButtonClicked)
        CloseButton : button_loud = button_loud{DefaultText := StringToMessage("CLOSE")}
        CloseButton.OnClick().Subscribe(OnCloseButtonClicked)

        set CanvasSlots += array:
            #背景と白線
            canvas_slot:
                Anchors := CenterAnchors
                Offsets := margin{Left := 0.0, Top := 0.0, Right := 0.0, Bottom := 0.0}
                SizeToContent := true
                Alignment := vector2{X := 0.5, Y := 0.5}
                Widget := Image_bg
            canvas_slot:
                Anchors := CenterAnchors
                Offsets := margin{Left := 0.0, Top := 22.489239, Right := 0.0, Bottom := 0.0}
                SizeToContent := true
                Alignment := vector2{X := 0.5, Y := 0.5}
                Widget := Image_BG_s
            canvas_slot:
                Anchors := CenterAnchors
                Offsets := margin{Left := 0.0, Top := 197.316010, Right := 0.0, Bottom := 0.0}
                SizeToContent := true
                Alignment := vector2{X := 0.5, Y := 0.5}
                Widget := CB_Line
            canvas_slot:
                Anchors := CenterAnchors
                Offsets := margin{Left := 0.0, Top := -66.429825, Right := 0.0, Bottom := 0.0}
                SizeToContent := true
                Alignment := vector2{X := 0.5, Y := 0.5}
                Widget := CB_Line1
            #現在の武器
            canvas_slot:
                Anchors := CenterAnchors
                Offsets := margin{Left := -230.0, Top := -250.0, Right := 0.0, Bottom := 0.0}
                SizeToContent := true
                Alignment := vector2{X := 0.5, Y := 0.5}
                Widget := CB_NowWeapon
            canvas_slot:
                Anchors := CenterAnchors
                Offsets := margin{Left := -230.0, Top := -250.0, Right := 0.0, Bottom := 0.0}
                SizeToContent := true
                Alignment := vector2{X := 0.5, Y := 0.5}
                Widget := Image_NowWeapon
            canvas_slot:
                Anchors := CenterAnchors
                Offsets := margin{Left := -230.0, Top := -110.0, Right := 0.0, Bottom := 0.0}
                SizeToContent := true
                Alignment := vector2{X := 0.5, Y := 0.5}
                Widget := TB_NowLv
            #次の武器
            canvas_slot:
                Anchors := CenterAnchors
                Offsets := margin{Left := 230.0, Top := -250.0, Right := 0.0, Bottom := 0.0}
                SizeToContent := true
                Alignment := vector2{X := 0.5, Y := 0.5}
                Widget := CB_NextWeapon
            canvas_slot:
                Anchors := CenterAnchors
                Offsets := margin{Left := 230.0, Top := -250.0, Right := 0.0, Bottom := 0.0}
                SizeToContent := true
                Alignment := vector2{X := 0.5, Y := 0.5}
                Widget := Image_NextWeapon
            canvas_slot:
                Anchors := CenterAnchors
                Offsets := margin{Left := 161.246735, Top := -110.0, Right := 0.0, Bottom := 0.0}
                SizeToContent := true
                Alignment := vector2{X := 0.0, Y := 0.5}
                Widget := TB_NextLv
            #矢印
            canvas_slot:
                Anchors := CenterAnchors
                Offsets := margin{Left := 0.0, Top := -250.0, Right := 0.0, Bottom := 0.0}
                SizeToContent := true
                Alignment := vector2{X := 0.5, Y := 0.5}
                Widget := TB_Arrow
            #スタックボックス_ステータス
            canvas_slot:
                Anchors := CenterAnchors
                Offsets := margin{Left := -14.444442, Top := 65.217537, Right := 744.144165, Bottom := 253.213211}
                SizeToContent := false
                Alignment := vector2{X := 0.5, Y := 0.5}
                Widget := stack_box:
                    Orientation := orientation.Horizontal
                    Slots := array:
                        #ローカライズテキスト
                        stack_box_slot:
                            Distribution:= option{0.4}
                            Padding:=margin{Left:=0.0 ,Top:= 0.0, Right:= 0.0, Bottom:= 0.0}
                            HorizontalAlignment := horizontal_alignment.Fill
                            VerticalAlignment := vertical_alignment.Fill
                            Widget := stack_box:
                                Orientation := orientation.Vertical
                                Slots := array:
                                    stack_box_slot:
                                        Distribution := option{1.0}
                                        Padding:=margin{Left:=0.0 ,Top:= 0.0, Right:= 0.0, Bottom:= 0.0}
                                        HorizontalAlignment := horizontal_alignment.Right
                                        VerticalAlignment := vertical_alignment.Center
                                        Widget := TB_BaseStr
                                    stack_box_slot:
                                        Distribution := option{1.0}
                                        Padding:=margin{Left:=0.0 ,Top:= 0.0, Right:= 0.0, Bottom:= 0.0}
                                        HorizontalAlignment := horizontal_alignment.Right
                                        VerticalAlignment := vertical_alignment.Center
                                        Widget := TB_StrMulti
                                    stack_box_slot:
                                        Distribution := option{1.0}
                                        Padding:=margin{Left:=0.0 ,Top:= 0.0, Right:= 0.0, Bottom:= 0.0}
                                        HorizontalAlignment := horizontal_alignment.Right
                                        VerticalAlignment := vertical_alignment.Center
                                        Widget := TB_CriRate
                                    stack_box_slot:
                                        Distribution := option{1.0}
                                        Padding:=margin{Left:=0.0 ,Top:= 0.0, Right:= 0.0, Bottom:= 0.0}
                                        HorizontalAlignment := horizontal_alignment.Right
                                        VerticalAlignment := vertical_alignment.Center
                                        Widget := TB_CriDamage
                        #現在の実数値
                        stack_box_slot:
                            Distribution:= option{0.21}
                            Padding:=margin{Left:=0.0 ,Top:= 0.0, Right:= 0.0, Bottom:= 0.0}
                            HorizontalAlignment := horizontal_alignment.Fill
                            VerticalAlignment := vertical_alignment.Fill
                            Widget := stack_box:
                                Orientation := orientation.Vertical
                                Slots := array:
                                    stack_box_slot:
                                        Distribution := option{1.0}
                                        Padding:=margin{Left:=0.0 ,Top:= 0.0, Right:= 0.0, Bottom:= 0.0}
                                        HorizontalAlignment := horizontal_alignment.Right
                                        VerticalAlignment := vertical_alignment.Center
                                        Widget := TB_NowBaseStr
                                    stack_box_slot:
                                        Distribution := option{1.0}
                                        Padding:=margin{Left:=0.0 ,Top:= 0.0, Right:= 0.0, Bottom:= 0.0}
                                        HorizontalAlignment := horizontal_alignment.Right
                                        VerticalAlignment := vertical_alignment.Center
                                        Widget := TB_NowStrMulti
                                    stack_box_slot:
                                        Distribution := option{1.0}
                                        Padding:=margin{Left:=0.0 ,Top:= 0.0, Right:= 0.0, Bottom:= 0.0}
                                        HorizontalAlignment := horizontal_alignment.Right
                                        VerticalAlignment := vertical_alignment.Center
                                        Widget := TB_NowCriRate
                                    stack_box_slot:
                                        Distribution := option{1.0}
                                        Padding:=margin{Left:=0.0 ,Top:= 0.0, Right:= 0.0, Bottom:= 0.0}
                                        HorizontalAlignment := horizontal_alignment.Right
                                        VerticalAlignment := vertical_alignment.Center
                                        Widget := TB_NowCriDamage
                        #矢印
                        stack_box_slot:
                            Distribution:= option{0.2}
                            Padding:=margin{Left:=0.0 ,Top:= 0.0, Right:= 0.0, Bottom:= 0.0}
                            HorizontalAlignment := horizontal_alignment.Fill
                            VerticalAlignment := vertical_alignment.Fill
                            Widget := stack_box:
                                Orientation := orientation.Vertical
                                Slots := array:
                                    stack_box_slot:
                                        Distribution := option{1.0}
                                        Padding:=margin{Left:=0.0 ,Top:= 0.0, Right:= 0.0, Bottom:= 0.0}
                                        HorizontalAlignment := horizontal_alignment.Right
                                        VerticalAlignment := vertical_alignment.Center
                                        Widget := TB_Arrow1
                                    stack_box_slot:
                                        Distribution := option{1.0}
                                        Padding:=margin{Left:=0.0 ,Top:= 0.0, Right:= 0.0, Bottom:= 0.0}
                                        HorizontalAlignment := horizontal_alignment.Right
                                        VerticalAlignment := vertical_alignment.Center
                                        Widget := TB_Arrow2
                                    stack_box_slot:
                                        Distribution := option{1.0}
                                        Padding:=margin{Left:=0.0 ,Top:= 0.0, Right:= 0.0, Bottom:= 0.0}
                                        HorizontalAlignment := horizontal_alignment.Right
                                        VerticalAlignment := vertical_alignment.Center
                                        Widget := TB_Arrow3
                                    stack_box_slot:
                                        Distribution := option{1.0}
                                        Padding:=margin{Left:=0.0 ,Top:= 0.0, Right:= 0.0, Bottom:= 0.0}
                                        HorizontalAlignment := horizontal_alignment.Right
                                        VerticalAlignment := vertical_alignment.Center
                                        Widget := TB_Arrow4
                        #次の実数値
                        stack_box_slot:
                            Distribution:= option{0.24}
                            Padding:=margin{Left:=0.0 ,Top:= 0.0, Right:= 0.0, Bottom:= 0.0}
                            HorizontalAlignment := horizontal_alignment.Fill
                            VerticalAlignment := vertical_alignment.Fill
                            Widget := stack_box:
                                Orientation := orientation.Vertical
                                Slots := array:
                                    stack_box_slot:
                                        Distribution := option{1.0}
                                        Padding:=margin{Left:=0.0 ,Top:= 0.0, Right:= 0.0, Bottom:= 0.0}
                                        HorizontalAlignment := horizontal_alignment.Right
                                        VerticalAlignment := vertical_alignment.Center
                                        Widget := TB_NextBaseStr
                                    stack_box_slot:
                                        Distribution := option{1.0}
                                        Padding:=margin{Left:=0.0 ,Top:= 0.0, Right:= 0.0, Bottom:= 0.0}
                                        HorizontalAlignment := horizontal_alignment.Right
                                        VerticalAlignment := vertical_alignment.Center
                                        Widget := TB_NextStrMulti
                                    stack_box_slot:
                                        Distribution := option{1.0}
                                        Padding:=margin{Left:=0.0 ,Top:= 0.0, Right:= 0.0, Bottom:= 0.0}
                                        HorizontalAlignment := horizontal_alignment.Right
                                        VerticalAlignment := vertical_alignment.Center
                                        Widget := TB_NextCriRate
                                    stack_box_slot:
                                        Distribution := option{1.0}
                                        Padding:=margin{Left:=0.0 ,Top:= 0.0, Right:= 0.0, Bottom:= 0.0}
                                        HorizontalAlignment := horizontal_alignment.Right
                                        VerticalAlignment := vertical_alignment.Center
                                        Widget := TB_NextCriDamage
            #スタックボックス_コスト
            canvas_slot:
                Anchors := CenterAnchors
                Offsets := margin{Left := -127.726913, Top := 239.453857, Right := 273.518036, Bottom := 63.252415}
                SizeToContent := false
                Alignment := vector2{X := 0.0, Y := 0.5}
                Widget := stack_box :
                    Orientation := orientation.Horizontal
                    Slots := array:
                        stack_box_slot:
                            Distribution := option{1.0}
                            Padding:=margin{Left:=0.0 ,Top:= 0.0, Right:= 0.0, Bottom:= 0.0}
                            HorizontalAlignment := horizontal_alignment.Left
                            VerticalAlignment := vertical_alignment.Center
                            Widget := TB_Cost
                        stack_box_slot:
                            Distribution := option{1.0}
                            Padding:=margin{Left:=40.0 ,Top:= 0.0, Right:= 0.0, Bottom:= 0.0}
                            HorizontalAlignment := horizontal_alignment.Center
                            VerticalAlignment := vertical_alignment.Center
                            Widget := Image_Ore
                        stack_box_slot:
                            Distribution := option{1.0}
                            Padding:=margin{Left:=0.0 ,Top:= 0.0, Right:= 0.0, Bottom:= 0.0}
                            HorizontalAlignment := horizontal_alignment.Right
                            VerticalAlignment := vertical_alignment.Center
                            Widget := TB_Cost1
        #ボタン
            canvas_slot:
                Anchors := CenterAnchors
                Offsets := margin{Left := 0.0, Top := 284.289520, Right := 260.173981, Bottom := 74.470024}
                SizeToContent := false
                Alignment := vector2{X := 0.5, Y := 0.0}
                Widget := LevelUpButton
            canvas_slot:
                Anchors := CenterAnchors
                Offsets := margin{Left := 0.0, Top := 376.919281, Right := 476.507507, Bottom := 68.766380}
                SizeToContent := false
                Alignment := vector2{X := 0.5, Y := 0.0}
                Widget := CloseButton

        
        if(set UIElements[Agent] = (
            # 0 背景と白線
            Image_bg, Image_BG_s, CB_Line, CB_Line1,
            # 4 現在の武器
            CB_NowWeapon, Image_NowWeapon, TB_NowLv,
            # 7 次の武器
            CB_NextWeapon, Image_NextWeapon, TB_NextLv,
            # 10 矢印
            TB_Arrow,
            # 11 ローカライズテキスト
            TB_BaseStr, TB_StrMulti, TB_CriRate, TB_CriDamage,
            # 15 現在の実数値
            TB_NowBaseStr, TB_NowStrMulti, TB_NowCriRate, TB_NowCriDamage,
            # 19 矢印
            TB_Arrow1, TB_Arrow2, TB_Arrow3, TB_Arrow4,
            # 23 次の実数値
            TB_NextBaseStr, TB_NextStrMulti, TB_NextCriRate, TB_NextCriDamage,
            # 27 コスト
            TB_Cost, Image_Ore, TB_Cost1,
            # 30 ボタン
            LevelUpButton, CloseButton
        )){}

        Canvas : canvas = canvas:
            Slots := CanvasSlots

        return Canvas
