using { /Verse.org/Simulation }

using {MiningObject}
using {VerseWidget}

var GlobalData<public> : weak_map(session, global_data) = map{}

GetGlobalData<public>()<transacts>:global_data=
    if(Data := GlobalData[GetSession()]):
        Data
    else if(Data := set GlobalData[GetSession()] = global_data{}):
        Data
    else:
        global_data{}

global_data<public> := class:

    var MiningObjectWidget<public> : [player]mining_object_widget = map{}

    FirstPlayedThisMapEvent<public> : event(player) = event(player){}
    FirstPlayedThisVersionEvent<public> : event(player) = event(player){}

    YokaiSlotUpdateEvent<public> : event(player) = event(player){}
    UnlockedStageEvent<public> : event(player) = event(player){}