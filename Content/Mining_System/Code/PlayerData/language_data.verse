using { /Verse.org/Simulation }
using { /Verse.org/Simulation }

using {Mining_System.Code.PlayerData}

language_data<public> := class<concrete>:

    @editable_text_box{MultiLine := true}
    Japanese<public> : string = ""

    @editable_text_box{MultiLine := true}
    English<public> : string = ""

    Localize<public>(Language:int):string=
        LanguageString := case(Language):
            0 => Japanese
            1 => English
            _ => Japanese

        # 言語が未設定の場合のデフォルト値
        if(LanguageString.Length <= 0):
            Japanese
        else:
            LanguageString

    Localize<public>(Player:player):string=
        if(NowData := SaveData[Player]):
            Localize(NowData.LanguageIndex)
        else:
            Localize(0)