using { /Fortnite.com/Devices }
using { /Verse.org/Simulation }

using {Pet}
using {Skill}
using { Profile }
using { VerseDevice.QuestSystem }

rebirth_device<public> := class(creative_device):

    @editable
    SaveDataManager : save_data_manager = save_data_manager{}

    @editable
    QuestDevice : quest_sequencer = quest_sequencer{}

    Rebirth<public>(InPlayer:player):void=
        SaveDataManager.AddResourceDiamond(InPlayer, BigNumber_PrestigeResource(SaveDataManager.GetResourceStone(InPlayer)))
        ResetAllResources(InPlayer)
        ResetPlayerLevel(InPlayer)
        ResetStageUnlock(InPlayer)
        ResetPlayerSkills(InPlayer)
        ResetPlayerPets(InPlayer)
        ResetBossDefeated(InPlayer)
        ResetQuest(InPlayer)
        ResetWeapon(InPlayer)

    ResetAllResources<public>(InPlayer:player):void=
        SaveDataManager.SetResourceStone(InPlayer, big_number{Value := 0.0})
        SaveDataManager.SetResourceSummonStoneCommon(InPlayer, big_number{Value := 0.0})
        SaveDataManager.SetResourceSummonStoneUncommon(InPlayer, big_number{Value := 0.0})
        SaveDataManager.SetResourceSummonStoneRare(InPlayer, big_number{Value := 0.0})
        SaveDataManager.SetResourceSummonStoneEpic(InPlayer, big_number{Value := 0.0})
        SaveDataManager.SetResourceSummonStoneLegendary(InPlayer, big_number{Value := 0.0})
        SaveDataManager.SetResourcePowerupStone(InPlayer, big_number{Value := 0.0})

    ResetPlayerLevel<public>(InPlayer:player):void=
        SaveDataManager.SetLevel(InPlayer, 1)
        SaveDataManager.SetXP(InPlayer, big_number{Value := 0.0})

    ResetStageUnlock<public>(InPlayer:player):void=
        SaveDataManager.ResetStageUnlocked(InPlayer)
        GetGlobalData().UnlockedStageEvent.Signal(InPlayer)

    ResetPlayerSkills<public>(InPlayer:player):void=
        if:
            NowData := SaveData[InPlayer]
            set SaveData[InPlayer] = save_data:
                MakeSaveData<constructor>(NowData)
                Skill_Info := map{}
                Skill_Status := map{}
        then:

    ResetPlayerPets<public>(InPlayer:player):void=
        if:
            NowData := SaveData[InPlayer]
            set SaveData[InPlayer] = save_data:
                MakeSaveData<constructor>(NowData)
                Pet_Info := map{}
        then:
            GetGlobalData().YokaiSlotUpdateEvent.Signal(InPlayer)

    ResetBossDefeated<public>(InPlayer:player):void=
        SaveDataManager.ResetBossDefeated(InPlayer)

    ResetQuest<public>(InPlayer:player):void=
        QuestDevice.ResetQuest(InPlayer)

    ResetWeapon<public>(InPlayer:player):void=
        SaveDataManager.SetWeaponLevel(InPlayer, 0)

    BigNumber_PrestigeResource(BigResource: big_number): big_number =
        ConstantValue: big_number = big_number{Value := 5.0, Exponent := -33}
        PrestigeResource: big_number = BigNumber_MultiplyByBigNumber(BigResource, ConstantValue)
        PrestigeResourceNormalized: big_number = BigNumber_Normalize(PrestigeResource)
    
        return PrestigeResourceNormalized