using { /Verse.org/Simulation }

using { Utility }
using { Pet }
using { Skill }

var SaveData<public> : weak_map(player, save_data) = map{}

 # XPの総量からレベルを計算するか、XPは次の経験値への量を表すかはまだ未定, 総量の場合はLevelは使用されないかも

SaveDataInit<public>(Player:player)<decides><transacts>:void=
    not SaveData[Player]
    set SaveData[Player] = save_data{}

MakeSaveData<public><constructor>(NowData:save_data)<transacts> := save_data:
    LastPlayedVersion := NowData.LastPlayedVersion
    Level := NowData.Level
    XP := NowData.XP
    Resource_Ticket := NowData.Resource_Ticket
    Resource_Ticket_Revival_Time := NowData.Resource_Ticket_Revival_Time
    Boss_Ticket_Revival_Time := NowData.Boss_Ticket_Revival_Time
    Resource_Ticket_Revival_Times := NowData.Resource_Ticket_Revival_Times
    Boss_Ticket_Revival_Times := NowData.Boss_Ticket_Revival_Times
    Resource_Stone := NowData.Resource_Stone
    Resource_Count_SummonStone_Common := NowData.Resource_Count_SummonStone_Common
    Resource_Count_SummonStone_Uncommon := NowData.Resource_Count_SummonStone_Uncommon
    Resource_Count_SummonStone_Rare := NowData.Resource_Count_SummonStone_Rare
    Resource_Count_SummonStone_Epic := NowData.Resource_Count_SummonStone_Epic
    Resource_Count_SummonStone_Legendary := NowData.Resource_Count_SummonStone_Legendary
    Resource_Count_PowerupStone := NowData.Resource_Count_PowerupStone
    Resource_Count_Diamond := NowData.Resource_Count_Diamond
    Stage_Unlocked := NowData.Stage_Unlocked
    Boss_Defeated := NowData.Boss_Defeated
    Gacha_Unlocked := NowData.Gacha_Unlocked
    Quest_Completed := NowData.Quest_Completed
    Quest_Progress := NowData.Quest_Progress
    Pet_Info := NowData.Pet_Info
    Skill_Info := NowData.Skill_Info
    Skill_Status := NowData.Skill_Status
    WeaponLevel := NowData.WeaponLevel
    
    LanguageIndex := NowData.LanguageIndex

save_data<public> := class<final><persistable>:

    LastPlayedVersion : int = 0
    
    Level : int = 1
    XP : big_number = big_number{}

    Resource_Ticket : logic = true
    Resource_Ticket_Revival_Times : int = 900
    Boss_Ticket_Revival_Times : int = 600

    Resource_Stone : big_number = big_number{}
    
    Resource_Count_SummonStone_Common : big_number = big_number{}
    Resource_Count_SummonStone_Uncommon : big_number = big_number{}
    Resource_Count_SummonStone_Rare : big_number = big_number{}
    Resource_Count_SummonStone_Epic : big_number = big_number{}
    Resource_Count_SummonStone_Legendary : big_number = big_number{}

    Resource_Count_PowerupStone : big_number = big_number{}

    Resource_Count_Diamond : big_number = big_number{}

    Stage_Unlocked : []int = array{}
    Boss_Defeated : []int = array{}
    Gacha_Unlocked : []int = array{}
    Quest_Completed<public> : []int = array{}
    Quest_Progress<public> : [int]int = map{}

    Pet_Info<public> : [int]pet_info = map{}
    Skill_Info<public> : [int]skill_info = map{}
    Skill_Status<public> : [int]skill_status = map{}


    WeaponLevel : int = 0

    LanguageIndex<public> : int = 0
    
    #廃棄
    Pet_Levels : [int]int = map{}
    Resource_Ticket_Revival_Time : float = 900.0
    Boss_Ticket_Revival_Time : float = 900.0





## big_number周り

using { /Verse.org/Simulation }

editable_big_number<public> := struct<concrete>:
    @editable
    Value<public> : float = 0.0

    @editable
    Exponent<public> : int = 0

big_number<public> := struct<persistable>:
    Value<public> : float = 0.0
    Exponent<public> : int = 0

(EBigNumber:editable_big_number).ToPersistableBigNumber<public>()<transacts>:big_number=
    return big_number{Value := EBigNumber.Value, Exponent := EBigNumber.Exponent}

# 2つの big_number を加算する
BigNumber_Add<public>(NumberA : big_number, NumberB : big_number)<transacts>:big_number=
    CommonExponent : int = Max(NumberA.Exponent, NumberB.Exponent)

    NewValue : float =  (NumberA.Value * Pow(10.0, (NumberA.Exponent - CommonExponent) * 1.0)) + 
                        (NumberB.Value * Pow(10.0, (NumberB.Exponent - CommonExponent) * 1.0))

    BigNumber_Normalize(big_number{Value := NewValue, Exponent := CommonExponent})

# 1つの big_number から別の big_number を減算する
BigNumber_Subtract<public>(NumberA : big_number, NumberB : big_number)<transacts>:big_number=
    CommonExponent : int = Max(NumberA.Exponent, NumberB.Exponent)

    NewValue : float =  (NumberA.Value * Pow(10.0, (NumberA.Exponent - CommonExponent) * 1.0)) - 
                        (NumberB.Value * Pow(10.0, (NumberB.Exponent - CommonExponent) * 1.0))

    BigNumber_Normalize(big_number{Value := NewValue, Exponent := CommonExponent})

# 2つの big_number を掛け合わせる
BigNumber_MultiplyByBigNumber<public>(NumberA : big_number, NumberB : big_number)<transacts>:big_number=
    NewValue := NumberA.Value * NumberB.Value
    NewExponent := NumberA.Exponent + NumberB.Exponent

    BigNumber_Normalize(big_number{Value := NewValue, Exponent := NewExponent})

# big_number と float を掛け合わせる
BigNumber_MultiplyByFloat<public>(NumberA : big_number, NumberB : float)<transacts>:big_number=
    NewValue := NumberA.Value * NumberB
    NewExponent := NumberA.Exponent

    BigNumber_Normalize(big_number{Value := NewValue, Exponent := NewExponent})

# 1つの big_number が別の big_number より大きいか等しいかを確認する
BigNumber_GreaterThanOrEquals<public>(NumberA : big_number, NumberB : big_number)<transacts>:logic=
    if:
        NumberA.Exponent > NumberB.Exponent or (NumberA.Exponent = NumberB.Exponent and NumberA.Value >= NumberB.Value)
    then:
        return true

    return false

# X個のアイテムを購入するのに必要なコストを計算する
#   Start は現在所有しているアイテムの数、Count は購入したい数を表す
#   CommonRatio は次の購入ごとの価格のスケーリングを示す。1.0に設定されると、価格スケーリングは発生しない。
BigNumber_CalculateCostRange(BaseCost : big_number, Start : int, Count : int, CommonRatio : float):big_number=
    term1 := Pow(CommonRatio, (Start + Count)*1.0);
    term2 := Pow(CommonRatio, Start * 1.0);
 
    TotalCost_Value := BaseCost.Value * (term1 - term2) / (CommonRatio - 1.0);
    TotalCost_Exponent := BaseCost.Exponent;
 
    BigNumber_Normalize(big_number{Value:=TotalCost_Value, Exponent:=TotalCost_Exponent});


# 購入可能なアイテム数を計算する
#   CommonRatio は次の購入ごとの価格のスケーリングを示す。1.0に設定されると、価格スケーリングは発生しない。
BigNumber_CalculateAffordableItems(CurrentMoney : big_number, BaseCost : big_number, AmountOwned : int, CommonRatio : float)<suspends>:int=
    var AffordableItems : int = 1
    var MaxIterations : int = 1000  # ループ上限
    var IterationCount : int = 0

    loop:
        Affordable := BigNumber_GreaterThanOrEquals(CurrentMoney, BigNumber_CalculateCostRange(BaseCost, AmountOwned, AffordableItems, CommonRatio))
        if(not Affordable? or IterationCount >= MaxIterations):  # 上限を超えたら終了
            break
        
        set AffordableItems += 1
        set IterationCount += 1  # イテレーションをカウント
        Sleep(0.1)  # 負荷を分散
    
    if(IterationCount >= MaxIterations):
        Print("Warning: Loop exceeded maximum iterations: BigNumber_CalculateAffordableItems")

    return AffordableItems - 1


# Formats a big_number into a string
BigNumber_FormatDisplay<public>(ValueIN : big_number)<transacts>:string=
    if (ValueIN.Value = 0):
        return "0"
 
    CorrectedNumber := BigNumber_NormalizeToNearestNumberSet(ValueIN)
 
    if:
        CorrectedNumber.Exponent < 3
        ValueAsInt := Floor[CorrectedNumber.Value]
    then:
        return "{ValueAsInt}"
    
    var ValueAsString : string = "{CorrectedNumber.Value}"
 
    var TargetLength : int = 4
    if(CorrectedNumber.Value >= 100.0):
        set TargetLength = 3
 
    if:
        set ValueAsString = ValueAsString.Remove[TargetLength, ValueAsString.Length]
    
    if:
        SuffixIndex : int = Quotient[CorrectedNumber.Exponent, 3]
        set ValueAsString += NumberSuffixes[SuffixIndex]      
 
    ValueAsString

# Helper function to convert float to string with specified precision
FloatToStringWithPrecision<public>(Value : float, Precision : int)<transacts>:string=
    # Precisionが0の場合、小数点以下を除去して整数として返す
    if (Precision = 0):
        if (ValueAsInt := Floor[Value]):
            return "{ValueAsInt}"
        else:
            Print("Floor演算中にエラーが発生")
            return "Error"

    var FinalValue : float = 0.0
    # Precisionが1以上の場合、指定された精度で処理する
    Multiplier : float = Pow(10.0, Precision * 1.0)
    Print("Multiplier: {Multiplier}")
    
    # 'Floor' 呼び出しに '[]' を使用し、計算結果を取得
    if (RoundedValue := Floor[Value * Multiplier]):
        set FinalValue = RoundedValue*1.0 / Multiplier
        Print("RoundedValue: {RoundedValue}, FinalValue: {FinalValue}")
    else:
        Print("Floor演算中にエラーが発生")
        return "Error"

    # 値を文字列に変換（手動でPrecision桁までの文字列を作成）
    ValueAsString : string = "{FinalValue}"  # 一旦floatを文字列に変換

    # 小数点の位置を探す
    var DotIndex : int = -1
    logic:
        DotIndex = ValueAsString.Find["."]
    
    # 小数点が見つかった場合、桁数を制限
    if (DotIndex >= 0):  # Findが成功したかどうかを確認
        # 小数点以下の桁数が指定Precisionより長い場合、削る
        RequiredLength := DotIndex + Precision + 1  # 小数点とPrecision桁を含めた長さ
        if (ValueAsString.Length > RequiredLength):
            if:
                ValueAsString = ValueAsString.Remove[RequiredLength, ValueAsString.Length - RequiredLength]
    else:
        # 小数点が無い場合そのまま返す
        return ValueAsString

    # 末尾のゼロを削除する
    loop:
        if (ValueAsString.Length > 0 and ValueAsString[ValueAsString.Length - 1] = '0'):
            if:
                ValueAsString = ValueAsString.Remove[ValueAsString.Length - 1, 1]
        else:
            break

    # 最後の文字が "." の場合、それも削除
    if (ValueAsString.Length > 0 and ValueAsString[ValueAsString.Length - 1] = '.'):
        if:
            ValueAsString = ValueAsString.Remove[ValueAsString.Length - 1, 1]

    return ValueAsString


NumberSuffixes<public> : []string = array{"-","A","B","C","D","E","F","G","H","I","J","K","K","M","N","O","P","Q","R","S","T","U","V","W","X","Y","Z","Aa","Ab","Ac","Ad","Ae","Af","Ag","Ah","Ai","Aj","Ak","Al","Am","An","Ao","Ap","Aq","Ar","As","At","Au","Av","Aw","Ax","Ay","Az",
"Ba","Bb","Bc","Bd","Be","Bf","Bg","Bh","Bi","Bj","Bk","Bl","Bm","Bn","Bo","Bp","Bq","Br","Bs","Bt","Bu","Bv","Bw","Bx","By","Bz","Ca","Cb","Cc","Cd","Ce","Cf","Cg","Ch","Ci","Cj","Ck","Cl","Cm","Cn","Co","Cp","Cq","Cr","Cs","Ct","Cu","Cv","Cw","Cx","Cy","Cz"}


# big_number を、Value フィールドが一桁の整数になるように変換する
#   例)  14.2 * 10^2  は 1.42 * 10^3 になる
BigNumber_Normalize<public>(Number : big_number)<transacts>:big_number=
    var NewExponent : int = Number.Exponent
    var NewValue : float = Number.Value
    #loop上限を設定
    var MaxIterations : int = 1000
    var IterationCount : int = 0

    if (NewValue <= 0.0):
        return NewResult := big_number{}

    loop:
        if (NewValue < 10.0 or IterationCount >= MaxIterations):
            break
        set NewValue *= 0.1
        set NewExponent += 1
        set IterationCount += 1
        if(IterationCount >= MaxIterations):
            Print("Warning: Normalize max iterations (lower).")
            break

    set IterationCount = 0
    
    loop:
        if (NewValue >= 1.0 or IterationCount >= MaxIterations):
            break
        set NewValue *= 10.0
        set NewExponent -= 1
        set IterationCount += 1
        if(IterationCount >= MaxIterations):
            Print("Warning: Normalize  max iterations (upper).")
            break

    # 浮動小数点の精度問題に対処
    Epsilon : float = 1e-10
    if (Abs(NewValue - 1.0) < Epsilon):
        set NewValue = 1.0

    Result := big_number:
        Value := NewValue
        Exponent := NewExponent


# big_number を、Value フィールドが1桁、2桁、または3桁の整数になるように変換する。これは文字列フォーマットに使用される。
#   例)  1.42K、14.2K、142K はすべて有効。
#        1420k は無効で、1.42M に変換されるべき。
BigNumber_NormalizeToNearestNumberSet<public>(Number : big_number)<transacts>:big_number=
    var NewExponent : int = Number.Exponent
    var NewValue : float = Number.Value
    var MaxIterations : int = 100
    var IterationCount : int = 0

    loop:
        if (Mod[NewExponent, 3] = 0 or IterationCount >= MaxIterations):  # ループ条件をチェック
            break

        set NewValue *= 10.0
        set NewExponent -= 1
        set IterationCount += 1
        if(IterationCount >= MaxIterations):
            Print("Warning: Loop exceeded max iterations in NormalizeToNearestNumberSet")
            break

    Result := big_number:
        Value := NewValue
        Exponent := NewExponent


BigNumber_MultiplyByInt<public>(NumberA: big_number, IntMultiplier: int)<transacts>:big_number=
    FloatMultiplier : float = IntMultiplier * 1.0
    return BigNumber_MultiplyByFloat(NumberA, FloatMultiplier)


BigNumber_Divide<public>(NumberA : big_number, NumberB : big_number)<transacts>:big_number=
    NewValue := NumberA.Value / NumberB.Value
    NewExponent := NumberA.Exponent - NumberB.Exponent
    BigNumber_Normalize(big_number{Value := NewValue, Exponent := NewExponent})
    
BigNumber_ToFloat<public>(Number: big_number)<transacts>:float=
    return Number.Value * Pow(10.0, Number.Exponent * 1.0)