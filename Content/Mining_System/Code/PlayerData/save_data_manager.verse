using { /Fortnite.com/Devices }
using { /Verse.org/Simulation }
using { /UnrealEngine.com/Temporary }
using { /Verse.org/Colors }

using { Profile }
using { Code.Utility }
using { Pet }
using { Skill }
using { VerseDevice.QuestSystem }

save_data_manager<public> := class(creative_device):

    #セーブデータに関する更新を行った場合、この値を増やす
    @editable
    CurrentVersion : int = 4

    @editable
    QuestSequencer : quest_sequencer = quest_sequencer{}
    @editable
    Indicator : Indicator_Control = Indicator_Control{}

    var NKMRZTicket_Revival_IsActive : [player]logic = map{}
    var BossTicket_Revival_IsActive : [player]logic = map{}
    var BossTicket_Lock : [player]logic = map{}
    #恐竜バフ
    # @editable Switch_StrBuff : switch_device = switch_device{}
    var IsStrBuff: [player]logic = map{}

    # --- レベル周り --- 
    SetLevel<public>(Player:player, Level:int)<transacts>:void=
        if:
            NowData := SaveData[Player]
            set SaveData[Player] = save_data:
                MakeSaveData<constructor>(NowData)
                Level := Level
        then:

    GetLevel<public>(Player:player)<transacts>:int=
        if(NowData := SaveData[Player]):
            NowData.Level
        else:
            0

    AddLevel<public>(Player:player, Level:int)<transacts>:void=
        SetLevel(Player, GetLevel(Player) + Level)


    HasSaveData<public>(Player:player)<transacts>:logic=
        if (NowData := SaveData[Player]):
            return true
        else:
            Print("Error: HasSaveData")
            return false

    # 言語取得
    GetLanguage<public>(Player:player):int=
        var Language : int = 0
        if(NowData := SaveData[Player]):
            set Language = NowData.LanguageIndex
        
        Language

    # --- 経験値周り ---
    SetXP<public>(Player: player, XP: big_number)<transacts>: void =
        if:
            NowData := SaveData[Player]
            set SaveData[Player] = save_data:
                MakeSaveData<constructor>(NowData)
                XP := XP
        then:
    
    GetXP<public>(Player: player)<transacts>: big_number =
        var NewValue: float = 0.0
        var NewExponent: int = 0
        if (NowData := SaveData[Player]):
            NowData.XP
        else:
            Result := big_number:
                Value := NewValue
                Exponent := NewExponent

    AddXP<public>(Player: player, XP: big_number)<transacts>: void =
        CurrentXP := GetXP(Player)
        UpdatedXP := BigNumber_Add(CurrentXP, XP)
        SetXP(Player, UpdatedXP)

    # --- 武器周り ---

    SetWeaponLevel<public>(Player:player, WeaponLevel:int)<transacts>:void=
        if:
            NowData := SaveData[Player]
            set SaveData[Player] = save_data:
                MakeSaveData<constructor>(NowData)
                WeaponLevel := WeaponLevel
        then:

    GetWeaponLevel<public>(Player:player)<transacts>:int=
        if(NowData := SaveData[Player]):
            NowData.WeaponLevel
        else:
            0

    AddWeaponLevel<public>(Player:player, WeaponLevel:int)<transacts>:void=
        SetWeaponLevel(Player, GetWeaponLevel(Player) + WeaponLevel)

    # レベルアップが可能かをチェック
    CanLevelUp<public>(Player:player)<decides><transacts>:void=
        ResourceCount := GetResourceStone(Player)
        RequiredResource := BigNumber_MultiplyByFloat(big_number{Value := 10.0, Exponent := 0}, Pow(1.28, (GetWeaponLevel(Player) - 1) * 1.0))
        
        if(BigNumber_GreaterThanOrEquals(ResourceCount, RequiredResource)?):
            Print("Can level up.")
        else:
            Print("not enough resources")

    # SetWeaponGrade<public>(Player:player, WeaponGrade:int)<transacts>:void=
    #     if:
    #         NowData := SaveData[Player]
    #         set SaveData[Player] = save_data:
    #             MakeSaveData<constructor>(NowData)
    #             WeaponGrade := WeaponGrade
    #     then:

    # GetWeaponGrade<public>(Player:player)<transacts>:int=
    #     if(NowData := SaveData[Player]):
    #         NowData.WeaponGrade
    #     else:
    #         0

    # AddWeaponGrade<public>(Player:player, WeaponGrade:int)<transacts>:void=
    #     SetWeaponGrade(Player, GetWeaponGrade(Player) + WeaponGrade)



    # Resouece_Count_Stone

    
    # SetResourceStone<public>(Player:player, Count:int)<transacts>:void=
    #     if:
    #         NowData := SaveData[Player]
    #         set SaveData[Player] = save_data:
    #             MakeSaveData<constructor>(NowData)
    #             Resouece_Count_Stone := Count
    #     then:

    # GetResourceStone<public>(Player:player)<transacts>:int=
    #     if(NowData := SaveData[Player]):
    #         NowData.Resouece_Count_Stone
    #     else:
    #         0

    # AddResourceStone<public>(Player:player, Count:int)<transacts>:void=
    #     SetResourceStone(Player, GetResourceStone(Player) + Count)

    # SubtractResourceStone<public>(Player:player, Count:int)<transacts><decides>:void =
    #     if(GetResourceStone(Player) >= Count):
    #         SetResourceStone(Player, GetResourceStone(Player) - Count)

    # ^^^^^^^^^^^^^^^^^^^^^^^^^^^^^^^^^^^^^^^^^^^^^^^^^^^^^^^^^^^^^^^^^^^^^^^^^^^^^^^^^^^^^^^^^^
    #Buff受け渡し
    GetStrBuff<public>(Player : player)<transacts>:logic=
        var State :logic= false
        if(IsStrBuff[Player] = false) :
            set State = false
            Print("State: false", ?Color := NamedColors.Red, ?Duration := 3.0)
        else if(IsStrBuff[Player] = true) :
            set State = true
            Print("State: true", ?Color := NamedColors.Red, ?Duration := 3.0)
        return State

    OnTurnOn<public>(Player : player)<transacts>:void=
        logic{set IsStrBuff[Player] = true}
        Print("TurnOn", ?Color := NamedColors.Red, ?Duration := 3.0)

    OnTurnOff<public>(Player : player)<transacts>:void=
        logic{set IsStrBuff[Player] = false}
        Print("TurnOff", ?Color := NamedColors.Red, ?Duration := 3.0)

    # ResourceTicket
    GetMaxResourceTicket<public>(Player:player):int=

        var MaxTicket : int = 0

        if:
            MaxTickets := GetSkill[Player, 8].EffectPower
        then:
            set MaxTicket = MaxTickets
        
        MaxTicket

    AddMaxResourceTicket<public>(Player:player):void=
        if:
            MaxTickets := GetSkill[Player, 8]
        then:
            SetSkill(Player, 8, skill_info{
                Level := MaxTickets.Level + 1,
                MaxLevel := MaxTickets.MaxLevel,
                EffectPower := MaxTickets.EffectPower,
                EffectIncrease := MaxTickets.EffectIncrease,
                InitialCost := MaxTickets.InitialCost,
                CostIncrease := MaxTickets.CostIncrease
                })
            Print("Level : {MaxTickets.Level + 1}")
            CheckRecoveryResourceTicket(Player)

    GetNowResourceTicket<public>(Player:player):int=
        var Ticket : int = 0

        if:
            NowTickets := GetSkillStatus[Player, 8].Charges
        then:
            set Ticket = NowTickets
        
        Ticket

    AddResourceTicket<public>(Player:player):skill_status=
        var SetTickets : skill_status = skill_status{}
        if:
            Tickets := GetSkillStatus[Player, 8]
        then:
            set SetTickets = skill_status{
                Index := Tickets.Index,
                IsActive := Tickets.IsActive, 
                CooldownTime := Tickets.CooldownTime, 
                MaxCooldownTime := Tickets.MaxCooldownTime, 
                Charges := Tickets.Charges + 1, 
                MaxCharges := Tickets.MaxCharges, 
                LastUsedTime := Tickets.LastUsedTime, 
                EffectDuration := Tickets.EffectDuration}

            SetSkillStatus(Player, 8, SetTickets)

        return SetTickets

    SubResourceTicket<public>(Player:player):void=
        if:
            Tickets := GetSkillStatus[Player, 8]
        then:
            SetTickets := skill_status{
                Index := Tickets.Index,
                IsActive := Tickets.IsActive, 
                CooldownTime := Tickets.CooldownTime, 
                MaxCooldownTime := Tickets.MaxCooldownTime, 
                Charges := Tickets.Charges - 1, 
                MaxCharges := Tickets.MaxCharges, 
                LastUsedTime := Tickets.LastUsedTime, 
                EffectDuration := Tickets.EffectDuration}

            Print("Sub and Recovering Boss Ticket!!", ?Duration := 5.0)
            SetSkillStatus(Player, 8, SetTickets)

            CheckRecoveryResourceTicket(Player)

    SetResourceTicket_Revival_Times<public>(Player: player, Count: int)<transacts>: void =
        if:
            NowData := SaveData[Player]
            set SaveData[Player] = save_data:
                MakeSaveData<constructor>(NowData)
                Resource_Ticket_Revival_Times := Count
        then:

    GetResourceTicket_Revival_Times<public>(Player: player)<transacts>: int =
        var Resource_Ticket_Revival_Times : int = -1
        if(NowData := SaveData[Player]):
            set Resource_Ticket_Revival_Times = NowData.Resource_Ticket_Revival_Times
        else:
            Print("Error: GetResourceTicket_Revival_Times")
        
        Resource_Ticket_Revival_Times

    # 一定時間でチケットを回復させる
    SubscribeRRBT<public>(Player:player):void=
        CheckRecoveryResourceTicket(Player)

    CheckRecoveryResourceTicket<public>(Player:player):void=
        if(NKMRZTicket_Revival_IsActive[Player] = false):
            spawn{RecoveryResourceTicket(Player)}
        if(not NKMRZTicket_Revival_IsActive[Player]):
            spawn{RecoveryResourceTicket(Player)}

    RecoveryResourceTicket<public>(Player:player)<suspends>:void=
        if(HasSaveData(Player)?):
            CheckNowTicket := GetNowResourceTicket(Player)
            CheckMaxTicket := GetMaxResourceTicket(Player)

            if(CheckNowTicket < CheckMaxTicket):
                Print("Start Recovery Nakamuraza Ticket!! NowTicket : {CheckNowTicket}, MaxTicket : {CheckMaxTicket}", ?Duration := 5.0)
                loop:
                    logic{set NKMRZTicket_Revival_IsActive[Player] = true}
                    Sleep(1.0)
                    ResourceTicket_Revival_Times := GetResourceTicket_Revival_Times(Player)
                    Print("{ResourceTicket_Revival_Times}")

                    # 0じゃないとき、1ずつ減らす
                    if(ResourceTicket_Revival_Times >= 0):
                        SetResourceTicket_Revival_Times(Player, ResourceTicket_Revival_Times - 1)

                    # 時間が経ったらチケットを回復。Maxを超えたらBreak
                    else:
                        Print("Recovery Resource Ticket!!", ?Duration := 5.0)
                        AddResourceTicket(Player)
                        NowTicket := GetNowResourceTicket(Player)
                        MaxTicket := GetMaxResourceTicket(Player)

                        SetResourceTicket_Revival_Times(Player, 900)

                        if(NowTicket >= MaxTicket):
                            logic{set NKMRZTicket_Revival_IsActive[Player] = false}
                            Print("Resource Ticket is Max!!", ?Duration := 5.0)
                            break

            else:
                Print("Max Resource Ticket!!", ?Duration := 5.0)
        else:
            Print("Error: RecoveryResourceTicket")

    # -----------------------------------------------------------------------------------------------------
    # BossTickets
    SetBossTicket_Revival_Times<public>(Player: player, Count: int)<transacts>: void =
        if:
            NowData := SaveData[Player]
            set SaveData[Player] = save_data:
                MakeSaveData<constructor>(NowData)
                Boss_Ticket_Revival_Times := Count
        then:

    GetBossTicket_Revival_Times<public>(Player: player)<transacts>: int =
        var Boss_Ticket_Revival_Times : int = 0
        if(HasSaveData(Player)?):  # 存在を確認
            if(NowData := SaveData[Player]):
                set Boss_Ticket_Revival_Times = NowData.Boss_Ticket_Revival_Times
                Print("Success: GetBossTicket_Revival_Times")
        else:
            Print("Error: GetBossTicket_Revival_Times")
        
        Boss_Ticket_Revival_Times

    # チケット関係
    GetMaxTicket<public>(Player:player):int=

        var MaxTicket : int = 0

        if:
            MaxTickets := GetSkillLevel[Player, 12]
        then:
            set MaxTicket = MaxTickets +1
        
        MaxTicket

    AddMaxTicket<public>(Player:player):void=
        if:
            MaxTickets := GetSkill[Player, 12]
        then:
            SetSkill(Player, 12, skill_info{
                Level := MaxTickets.Level + 1,
                MaxLevel := MaxTickets.MaxLevel,
                EffectPower := MaxTickets.EffectPower,
                EffectIncrease := MaxTickets.EffectIncrease,
                InitialCost := MaxTickets.InitialCost,
                CostIncrease := MaxTickets.CostIncrease
                })
            Print("Level : {MaxTickets.Level + 1}")
            CheckRecoveryBossTicket(Player)

    GetNowTicket<public>(Player:player):int=
        var Ticket : int = 0

        if:
            NowTickets := GetSkillStatus[Player, 12].Charges
        then:
            set Ticket = NowTickets
        
        Ticket

    AddTicket<public>(Player:player):skill_status=
        var SetTickets : skill_status = skill_status{}
        if:
            Tickets := GetSkillStatus[Player, 12]
        then:
            set SetTickets = skill_status{
                Index := Tickets.Index,
                IsActive := Tickets.IsActive, 
                CooldownTime := Tickets.CooldownTime, 
                MaxCooldownTime := Tickets.MaxCooldownTime, 
                Charges := Tickets.Charges + 1, 
                MaxCharges := Tickets.MaxCharges, 
                LastUsedTime := Tickets.LastUsedTime, 
                EffectDuration := Tickets.EffectDuration}

            SetSkillStatus(Player, 12, SetTickets)

        return SetTickets

    SubTicket<public>(Player:player):void=
        if (BossTicket_Lock[Player]?):
            return

        if:
            set BossTicket_Lock[Player] = true
            Tickets := GetSkillStatus[Player, 12]
        then:
            SetTickets := skill_status{
                Index := Tickets.Index,
                IsActive := Tickets.IsActive, 
                CooldownTime := Tickets.CooldownTime, 
                MaxCooldownTime := Tickets.MaxCooldownTime, 
                Charges := Tickets.Charges - 1, 
                MaxCharges := Tickets.MaxCharges, 
                LastUsedTime := Tickets.LastUsedTime, 
                EffectDuration := Tickets.EffectDuration
            }

            Print("Sub and Recovering Boss Ticket!!", ?Duration := 5.0)
            SetSkillStatus(Player, 12, SetTickets)

            CheckRecoveryBossTicket(Player)
        if:
            set BossTicket_Lock[Player] = false
            
    # 一定時間でチケットを回復させる
    SubscribeRBT<public>(Player:player):void=
        CheckRecoveryBossTicket(Player)

    CheckRecoveryBossTicket<public>(Player:player):void=
        if(BossTicket_Revival_IsActive[Player] = false):
            spawn{RecoveryBossTicket(Player)}
        if(not BossTicket_Revival_IsActive[Player]):
            spawn{RecoveryBossTicket(Player)}

    RecoveryBossTicket<public>(Player:player)<suspends>:void=
        if (HasSaveData(Player)?):
            CheckNowTicket := GetNowTicket(Player)
            CheckMaxTicket := GetMaxTicket(Player)

            if(CheckNowTicket < CheckMaxTicket):
                # Print("Start Recovery Boss Ticket!! NowTicket : {CheckNowTicket}, MaxTicket : {CheckMaxTicket}", ?Duration := 5.0)
                loop:
                    logic{set BossTicket_Revival_IsActive[Player] = true}
                    Sleep(1.0)
                    BossTicket_Revival_Times := GetBossTicket_Revival_Times(Player)
                    # Print("{BossTicket_Revival_Times}")
                    if (BossTicket_Revival_Times < 0):
                        Print("Error: RecoveryBossTicket")
                        break  # 不正な値の場合ループを終了
                    else:
                        Print("BossTicket_Revival_Times: {BossTicket_Revival_Times}")

                    # 0じゃないとき、1ずつ減らす
                    if(BossTicket_Revival_Times >= 1):
                        SetBossTicket_Revival_Times(Player, BossTicket_Revival_Times - 1)

                    # 時間が経ったらチケットを回復。Maxを超えたらBreak
                    else:
                        Print("Recovery Boss Ticket!!", ?Duration := 5.0)
                        AddTicket(Player)
                        NowTicket := GetNowTicket(Player)
                        MaxTicket := GetMaxTicket(Player)

                        SetBossTicket_Revival_Times(Player, 300) 

                        if(NowTicket >= MaxTicket):
                            logic{set BossTicket_Revival_IsActive[Player] = false}
                            Print("Boss Ticket is Max!!", ?Duration := 5.0)
                            break

            else:
                Print("Max Boss Ticket!!", ?Duration := 5.0)
        else:
            Print("No save data found : RecoveryBossTicket.")

    # ^^^^^^^^^^^^^^^^^^^^^^^^^^^^^^^^^^^^^^^^^^^^^^^^^^^^^^^^^^^^^^^^^^^^^^^^^^^^^^^^^^^^^^^^^^

    # ResourceStone

    SetResourceStone<public>(Player: player, Count: big_number)<transacts>: void =
        if:
            NowData := SaveData[Player]
            set SaveData[Player] = save_data:
                MakeSaveData<constructor>(NowData)
                Resource_Stone := Count
        then:
        
    AddResourceStone<public>(Player: player, Count: big_number)<transacts>: void =
        if:
            # 現在のリソースを取得
            NowData := SaveData[Player]
            MyNumber := NowData.Resource_Stone
            CountNumber := Count
            UpdatedResource := BigNumber_Add(MyNumber, CountNumber)

            # 更新されたリソースを保存
            SetResourceStone(Player, UpdatedResource)
        then:
        
    GetResourceStone<public>(Player:player)<transacts>:big_number=
        var NewValue : float = 0.0
        var NewExponent : int = 0
    
        # SaveDataが存在するかを確認
        if (NowData := SaveData[Player]):
            # Print("Player's save data found, returning Resource_Stone")
            return NowData.Resource_Stone  # データが存在する場合は石のリソースを返す
        else:
            # データが存在しない場合、デフォルト値を返す
            Print("Error: No save data found for player or invalid player.")
            return big_number{Value := NewValue, Exponent := NewExponent}
            
        
    SubtractResourceStone<public>(Player: player, Count: big_number)<transacts>: void =
        if:
            # 現在のリソースを取得
            NowData := SaveData[Player]
            CurrentBigNumber := NowData.Resource_Stone
    
            # 現在のリソースが十分かどうかを確認
            if (BigNumber_GreaterThanOrEquals(CurrentBigNumber, Count)?):
                # リソースを減算
                UpdatedResource := BigNumber_Subtract(CurrentBigNumber, Count)
                # 更新されたリソースを保存
                SetResourceStone(Player, UpdatedResource)
            else:
                Print("Not enough resources to subtract.")

    # --- AFK周り(Cube追記) --- 


    GetPlayeridle<public>(Player: player)<transacts>: int =
        var ReturnPlayerIdle: int = 0
        if (NowData := SaveData[Player]):
            # スキルのインデックス（例: スキルインデックスが1の場合）
            SkillIndex := 3
            # 失敗する可能性のある式を変数に保存
            if:
                SkillLevel := GetSkillLevel[Player, SkillIndex]
                set ReturnPlayerIdle = SkillLevel
        return ReturnPlayerIdle
    

    GetPetidle<public>(Player : player)<transacts>: int =
        var ReturnPetIdle: int = 0
        if (NowData := SaveData[Player]):
            # スキルのインデックス（例: スキルインデックスが2の場合）
            SkillIndex := 4
            # 失敗する可能性のある式を変数に保存
            if:
                SkillLevel := GetSkillLevel[Player, SkillIndex]
                set ReturnPetIdle = SkillLevel
        return ReturnPetIdle
        

    # Resource_Count_SummonStone_Common

    SetResourceSummonStoneCommon<public>(Player: player, Count: big_number)<transacts>: void =
        if:
            NowData := SaveData[Player]
            set SaveData[Player] = save_data:
                MakeSaveData<constructor>(NowData)
                Resource_Count_SummonStone_Common := Count
        then:
    
    AddResourceSummonStoneCommon<public>(Player: player, Count: big_number)<transacts>: void =
        if:
            NowData := SaveData[Player]
            MyNumber := NowData.Resource_Count_SummonStone_Common
            UpdatedResource := BigNumber_Add(MyNumber, Count)
    
            SetResourceSummonStoneCommon(Player, UpdatedResource)
        then:
    
    GetResourceSummonStoneCommon<public>(Player: player)<transacts>: big_number =
        var NewValue: float = 0.0
        var NewExponent: int = 0
        if (NowData := SaveData[Player]):
            NowData.Resource_Count_SummonStone_Common
        else:
            Result := big_number:
                Value := NewValue
                Exponent := NewExponent
    
    SubtractResourceSummonStoneCommon<public>(Player: player, Count: big_number)<transacts>: void =
        if:
            NowData := SaveData[Player]
            CurrentBigNumber := NowData.Resource_Count_SummonStone_Common
    
            if (BigNumber_GreaterThanOrEquals(CurrentBigNumber, Count)?):
                UpdatedResource := BigNumber_Subtract(CurrentBigNumber, Count)
                SetResourceSummonStoneCommon(Player, UpdatedResource)
            else:
                Print("Not enough resources to subtract.")

    # Resouece_Count_SummonStone_Uncommon

    SetResourceSummonStoneUncommon<public>(Player: player, Count: big_number)<transacts>: void =
        if:
            NowData := SaveData[Player]
            set SaveData[Player] = save_data:
                MakeSaveData<constructor>(NowData)
                Resource_Count_SummonStone_Uncommon := Count
        then:
    
    AddResourceSummonStoneUncommon<public>(Player: player, Count: big_number)<transacts>: void =
        if:
            NowData := SaveData[Player]
            MyNumber := NowData.Resource_Count_SummonStone_Uncommon
            UpdatedResource := BigNumber_Add(MyNumber, Count)
            SetResourceSummonStoneUncommon(Player, UpdatedResource)
        then:
    
    GetResourceSummonStoneUncommon<public>(Player: player)<transacts>: big_number =
        var NewValue: float = 0.0
        var NewExponent: int = 0
        if (NowData := SaveData[Player]):
            NowData.Resource_Count_SummonStone_Uncommon
        else:
            Result := big_number:
                Value := NewValue
                Exponent := NewExponent
    
    SubtractResourceSummonStoneUncommon<public>(Player: player, Count: big_number)<transacts>: void =
        if:
            NowData := SaveData[Player]
            CurrentBigNumber := NowData.Resource_Count_SummonStone_Uncommon
            if (BigNumber_GreaterThanOrEquals(CurrentBigNumber, Count)?):
                UpdatedResource := BigNumber_Subtract(CurrentBigNumber, Count)
                SetResourceSummonStoneUncommon(Player, UpdatedResource)
            else:
                Print("Not enough resources to subtract.")

    # Resouece_Count_SummonStone_Rare

    SetResourceSummonStoneRare<public>(Player: player, Count: big_number)<transacts>: void =
        if:
            NowData := SaveData[Player]
            set SaveData[Player] = save_data:
                MakeSaveData<constructor>(NowData)
                Resource_Count_SummonStone_Rare := Count
        then:
    
    AddResourceSummonStoneRare<public>(Player: player, Count: big_number)<transacts>: void =
        if:
            NowData := SaveData[Player]
            MyNumber := NowData.Resource_Count_SummonStone_Rare
            UpdatedResource := BigNumber_Add(MyNumber, Count)
            SetResourceSummonStoneRare(Player, UpdatedResource)
        then:
    
    GetResourceSummonStoneRare<public>(Player: player)<transacts>: big_number =
        var NewValue: float = 0.0
        var NewExponent: int = 0
        if (NowData := SaveData[Player]):
            NowData.Resource_Count_SummonStone_Rare
        else:
            Result := big_number:
                Value := NewValue
                Exponent := NewExponent
    
    SubtractResourceSummonStoneRare<public>(Player: player, Count: big_number)<transacts>: void =
        if:
            NowData := SaveData[Player]
            CurrentBigNumber := NowData.Resource_Count_SummonStone_Rare
            if (BigNumber_GreaterThanOrEquals(CurrentBigNumber, Count)?):
                UpdatedResource := BigNumber_Subtract(CurrentBigNumber, Count)
                SetResourceSummonStoneRare(Player, UpdatedResource)
            else:
                Print("Not enough resources to subtract.")

    # Resouece_Count_SummonStone_Epic

    SetResourceSummonStoneEpic<public>(Player: player, Count: big_number)<transacts>: void =
        if:
            NowData := SaveData[Player]
            set SaveData[Player] = save_data:
                MakeSaveData<constructor>(NowData)
                Resource_Count_SummonStone_Epic := Count
        then:
    
    AddResourceSummonStoneEpic<public>(Player: player, Count: big_number)<transacts>: void =
        if:
            NowData := SaveData[Player]
            MyNumber := NowData.Resource_Count_SummonStone_Epic
            UpdatedResource := BigNumber_Add(MyNumber, Count)
            SetResourceSummonStoneEpic(Player, UpdatedResource)
        then:
    
    GetResourceSummonStoneEpic<public>(Player: player)<transacts>: big_number =
        var NewValue: float = 0.0
        var NewExponent: int = 0
        if (NowData := SaveData[Player]):
            NowData.Resource_Count_SummonStone_Epic
        else:
            Result := big_number:
                Value := NewValue
                Exponent := NewExponent
    
    SubtractResourceSummonStoneEpic<public>(Player: player, Count: big_number)<transacts>: void =
        if:
            NowData := SaveData[Player]
            CurrentBigNumber := NowData.Resource_Count_SummonStone_Epic
            if (BigNumber_GreaterThanOrEquals(CurrentBigNumber, Count)?):
                UpdatedResource := BigNumber_Subtract(CurrentBigNumber, Count)
                SetResourceSummonStoneEpic(Player, UpdatedResource)
            else:
                Print("Not enough resources to subtract.")

    # Resouece_Count_SummonStone_Legendary

    SetResourceSummonStoneLegendary<public>(Player: player, Count: big_number)<transacts>: void =
        if:
            NowData := SaveData[Player]
            set SaveData[Player] = save_data:
                MakeSaveData<constructor>(NowData)
                Resource_Count_SummonStone_Legendary := Count
        then:
    
    AddResourceSummonStoneLegendary<public>(Player: player, Count: big_number)<transacts>: void =
        if:
            NowData := SaveData[Player]
            MyNumber := NowData.Resource_Count_SummonStone_Legendary
            UpdatedResource := BigNumber_Add(MyNumber, Count)
            SetResourceSummonStoneLegendary(Player, UpdatedResource)
        then:
    
    GetResourceSummonStoneLegendary<public>(Player: player)<transacts>: big_number =
        var NewValue: float = 0.0
        var NewExponent: int = 0
        if (NowData := SaveData[Player]):
            NowData.Resource_Count_SummonStone_Legendary
        else:
            Result := big_number:
                Value := NewValue
                Exponent := NewExponent
    
    SubtractResourceSummonStoneLegendary<public>(Player: player, Count: big_number)<transacts>: void =
        if:
            NowData := SaveData[Player]
            CurrentBigNumber := NowData.Resource_Count_SummonStone_Legendary
            if (BigNumber_GreaterThanOrEquals(CurrentBigNumber, Count)?):
                UpdatedResource := BigNumber_Subtract(CurrentBigNumber, Count)
                SetResourceSummonStoneLegendary(Player, UpdatedResource)
            else:
                Print("Not enough resources to subtract.")

    # Resouece_Count_PowerupStone

    SetResourcePowerupStone<public>(Player: player, Count: big_number)<transacts>: void =
        if:
            NowData := SaveData[Player]
            set SaveData[Player] = save_data:
                MakeSaveData<constructor>(NowData)
                Resource_Count_PowerupStone := Count
        then:
    
    AddResourcePowerupStone<public>(Player: player, Count: big_number)<transacts>: void =
        if:
            NowData := SaveData[Player]
            MyNumber := NowData.Resource_Count_PowerupStone
            UpdatedResource := BigNumber_Add(MyNumber, Count)
            SetResourcePowerupStone(Player, UpdatedResource)
        then:
    
    GetResourcePowerupStone<public>(Player: player)<transacts>: big_number =
        var NewValue: float = 0.0
        var NewExponent: int = 0
        if (NowData := SaveData[Player]):
            NowData.Resource_Count_PowerupStone
        else:
            Result := big_number:
                Value := NewValue
                Exponent := NewExponent
    
    SubtractResourcePowerupStone<public>(Player: player, Count: big_number)<transacts>: void =
        if:
            NowData := SaveData[Player]
            CurrentBigNumber := NowData.Resource_Count_PowerupStone
            if (BigNumber_GreaterThanOrEquals(CurrentBigNumber, Count)?):
                UpdatedResource := BigNumber_Subtract(CurrentBigNumber, Count)
                SetResourcePowerupStone(Player, UpdatedResource)
            else:
                Print("Not enough resources to subtract.")

    # Resouece_Count_Diamond

    SetResourceDiamond<public>(Player: player, Count: big_number)<transacts>: void =
        if:
            NowData := SaveData[Player]
            set SaveData[Player] = save_data:
                MakeSaveData<constructor>(NowData)
                Resource_Count_Diamond := Count
        then:
    
    AddResourceDiamond<public>(Player: player, Count: big_number)<transacts>: void =
        if:
            NowData := SaveData[Player]
            MyNumber := NowData.Resource_Count_Diamond
            UpdatedResource := BigNumber_Add(MyNumber, Count)
            SetResourceDiamond(Player, UpdatedResource)
        then:
    
    GetResourceDiamond<public>(Player: player)<transacts>: big_number =
        var NewValue: float = 0.0
        var NewExponent: int = 0
        if (NowData := SaveData[Player]):
            NowData.Resource_Count_Diamond
        else:
            Result := big_number:
                Value := NewValue
                Exponent := NewExponent
    
    SubtractResourceDiamond<public>(Player: player, Count: big_number)<transacts>: void =
        if:
            NowData := SaveData[Player]
            CurrentBigNumber := NowData.Resource_Count_Diamond
            if (BigNumber_GreaterThanOrEquals(CurrentBigNumber, Count)?):
                UpdatedResource := BigNumber_Subtract(CurrentBigNumber, Count)
                SetResourceDiamond(Player, UpdatedResource)
            else:
                Print("Not enough resources to subtract.")

    # Stage_Unlocked

    ResetStageUnlocked<public>(Player:player)<transacts>:void=
        if:
            NowData := SaveData[Player]
            set SaveData[Player] = save_data:
                MakeSaveData<constructor>(NowData)
                Stage_Unlocked := array{}
        then:

    UnlockStage<public>(Player:player, Stage:int)<transacts>:void=
        if:
            NowData := SaveData[Player]
            set SaveData[Player] = save_data:
                MakeSaveData<constructor>(NowData)
                Stage_Unlocked := NowData.Stage_Unlocked + array{Stage}
        then:
        
    IsUnlockedStage<public>(Player:player, Stage:int)<transacts><decides>:void=
        NowData := SaveData[Player]
        NowData.Stage_Unlocked.Find[Stage]

    GetUnlockedStages<public>(Player:player)<transacts>:[]int=
        if(NowData := SaveData[Player]):
            NowData.Stage_Unlocked
        else:
            array{}


    # Boss_Defeated

    ResetBossDefeated<public>(Player:player)<transacts>:void=
        if:
            NowData := SaveData[Player]
            set SaveData[Player] = save_data:
                MakeSaveData<constructor>(NowData)
                Boss_Defeated := array{}
        then:

    DefeatBoss<public>(Player:player, Boss:int)<transacts>:void=
        if:
            NowData := SaveData[Player]
            set SaveData[Player] = save_data:
                MakeSaveData<constructor>(NowData)
                Boss_Defeated := NowData.Boss_Defeated + array{Boss}
        then:
            Print("{Boss}番目のボスをセーブ")

    IsDefeatedBoss<public>(Player:player, Boss:int)<transacts><decides>:void=
        NowData := SaveData[Player]
        NowData.Boss_Defeated.Find[Boss]

    GetDefeatedBosses<public>(Player:player)<transacts>:[]int=
        if(NowData := SaveData[Player]):
            NowData.Boss_Defeated
        else:
            array{}

    # Gacha_Unlocked

    ResetGachaUnlocked<public>(Player:player)<transacts>:void=
        if:
            NowData := SaveData[Player]
            set SaveData[Player] = save_data:
                MakeSaveData<constructor>(NowData)
                Gacha_Unlocked := array{}
        then:

    UnlockGacha<public>(Player:player, Gacha:int)<transacts>:void=
        if:
            NowData := SaveData[Player]
            set SaveData[Player] = save_data:
                MakeSaveData<constructor>(NowData)
                Gacha_Unlocked := NowData.Gacha_Unlocked + array{Gacha}
        then:

    IsUnlockedGacha<public>(Player:player, Gacha:int)<transacts><decides>:void=
        NowData := SaveData[Player]
        NowData.Gacha_Unlocked.Find[Gacha]

    GetUnlockedGachas<public>(Player:player)<transacts>:[]int=
        if(NowData := SaveData[Player]):
            NowData.Gacha_Unlocked
        else:
            array{}

    # Quest InProgress

    GetNowQuest<public>(Player:player)<transacts>:int=
        var NowQuest : int = 0
        if:
            NowData := SaveData[Player]
            Quest_Progress := NowData.Quest_Completed
            Index := Quest_Progress.Length - 1
            set NowQuest = Quest_Progress[Index]
        NowQuest

    # Quest_Completed

    ResetQuestCompleted<public>(Player:player)<transacts>:void=
        if:
            NowData := SaveData[Player]
            set SaveData[Player] = save_data:
                MakeSaveData<constructor>(NowData)
                Quest_Completed := array{}
        then:

    CompleteQuest<public>(Player:player, Quest:int)<transacts>:void=
        if:
            NowData := SaveData[Player]
            set SaveData[Player] = save_data:
                MakeSaveData<constructor>(NowData)
                Quest_Completed := NowData.Quest_Completed + array{Quest}
        then:

    IsCompletedQuest<public>(Player:player, Quest:int)<transacts><decides>:void=
        NowData := SaveData[Player]
        NowData.Quest_Completed.Find[Quest]

    GetCompletedQuests<public>(Player:player)<transacts>:[]int=
        if(NowData := SaveData[Player]):
            NowData.Quest_Completed
        else:
            array{}

    # Pet

    GetPet<public>(InPlayer:player, Index:int)<decides><transacts>:pet_info=
        SaveData[InPlayer].Pet_Info[Index]

    GetPets<public>(InPlayer:player)<decides><transacts>:[]tuple(int, pet_info)=
        for(Key -> Element : SaveData[InPlayer].Pet_Info):
            (Key, Element)

    SetPet<public>(InPlayer:player, Index:int, PetData:pet_info)<transacts>:void=
        if:
            NowData := SaveData[InPlayer]
            set SaveData[InPlayer] = save_data:
                MakeSaveData<constructor>(NowData)
                Pet_Info := AddMap(NowData.Pet_Info, (Index, PetData))
        then:
    # Skill

    GetSkill<public>(InPlayer:player, Index:int)<decides><transacts>:skill_info=
        if(SkillInfo := SaveData[InPlayer].Skill_Info[Index]):
            return SkillInfo
        else:
            # スキルが存在しない場合、仮のスキル情報を返す
            return skill_info{
                Level := 0,
                MaxLevel := 5,
                EffectPower := 0,
                EffectIncrease := 0,
                InitialCost := big_number{Value := 100.0, Exponent := 0},
                CostIncrease := 0
            }
    
    # スキル情報を設定する関数
    SetSkill<public>(InPlayer:player, Index:int, SkillData:skill_info)<transacts>:void=
        if:
            NowData := SaveData[InPlayer]
            set SaveData[InPlayer] = save_data:
                MakeSaveData<constructor>(NowData)
                Skill_Info := AddMap(NowData.Skill_Info, (Index, SkillData))
        then:

            #-----------------------------------

    # スキルの現在のレベルを取得する関数
    GetSkillLevel<public>(InPlayer: player, SkillIndex: int)<decides><transacts>: int =
        SkillInfo := GetSkill[InPlayer, SkillIndex]
        return SkillInfo.Level

    # プレイヤーのスキルレベルに基づいて、EffectIncreaseを取得
    GetPlayerSkillEffectIncrease<public>(InPlayer:player, SkillIndex:int)<decides><transacts>:int=
        SkillInfo := GetSkill[InPlayer, SkillIndex]
        return SkillInfo.EffectPower

    # 特定のスキルのEffectIncreaseを取得
    GetSpecificSkillEffectIncrease<public>(InPlayer:player, SkillIndex:int)<decides><transacts>:int=
        SkillInfo := GetSkill[InPlayer, SkillIndex]
        return SkillInfo.EffectIncrease

        # 特定のスキルの EffectPower を取得する関数
    GetSkillEffectPower<public>(InPlayer:player, SkillIndex:int)<decides><transacts>:int=
        SkillInfo := GetSkill[InPlayer, SkillIndex]
        return SkillInfo.EffectPower

    # プレイヤーのスキルレベルに基づいて、次のアップグレードコストを計算
    GetUpgradeCostForSkill<public>(InPlayer:player, SkillIndex:int)<decides><transacts>:big_number=
        SkillInfo := GetSkill[InPlayer, SkillIndex]
        BaseCost := SkillInfo.InitialCost
        LevelMultiplier := (SkillInfo.Level+1) * 1.0
        BigNumber_Result := BigNumber_MultiplyByFloat(BaseCost, LevelMultiplier)
        return BigNumber_Result

            #-----------------------------------

    # 全スキル情報を取得
    GetSkills<public>(InPlayer:player)<decides><transacts>:[]tuple(int, skill_info)=
        var SkillList : []tuple(int, skill_info) = array{}
        
        for(Key -> Element : SaveData[InPlayer].Skill_Info):
            set SkillList += array{(Key, Element)}
        
        return SkillList

    # スキルの状態（skill_status）を取得
    GetSkillStatus<public>(InPlayer:player, Index:int)<decides><transacts>:skill_status=
        var Result : skill_status = skill_status{}  # デフォルト初期化
    
        # プレイヤーデータが存在するか確認
        if (HasSaveData(InPlayer)?):
            if (NowData := SaveData[InPlayer]):
                # Skill_Statusが存在し、インデックスが範囲内にあるか確認
                if (Index < NowData.Skill_Status.Length):
                    # スキルステータスを変数名 `SkillStatus` で取得
                    if (SkillStatus := NowData.Skill_Status[Index]):
                        # スキルステータスが見つかった場合、そのまま続行
                        set Result = SkillStatus
                    else:
                        Print("既存のスキルステータスが見つからない。初期化します。")
                        # スキルステータスが見つからない場合、初期化する
                        InitiSkill : skill_status = skill_status{
                            Index := Index,
                            IsActive := false,
                            CooldownTime := 0,
                            MaxCooldownTime := 0,
                            Charges := 0,
                            MaxCharges := 0,
                            LastUsedTime := 0,
                            EffectDuration := 0,
                            Multiply := 2
                        }
        
                        # 初期化されたスキルステータスを保存
                        SetSkillStatus(InPlayer, Index, InitiSkill)
                        set Result = InitiSkill
                else:
                    Print("スキルステータスのインデックスが範囲外です。初期化します。")
                    # インデックスが範囲外の場合、初期化データを設定
                    InitiSkill : skill_status = skill_status{
                        Index := Index,
                        IsActive := false,
                        CooldownTime := 0,
                        MaxCooldownTime := 0,
                        Charges := 0,
                        MaxCharges := 0,
                        LastUsedTime := 0,
                        EffectDuration := 0,
                        Multiply := 2
                    }
        
                    # 初期化されたスキルステータスを保存
                    SetSkillStatus(InPlayer, Index, InitiSkill)
                    set Result = InitiSkill
        else:
            Print("プレイヤーデータが見つかりません。")
            set Result = skill_status{}
            
        return Result

    # 全スキルの状態を取得
    GetAllSkillStatuses<public>(InPlayer:player)<decides><transacts>:[]tuple(int, skill_status)=
        var StatusList : []tuple(int, skill_status) = array{}
        
        for(Key -> Element : SaveData[InPlayer].Skill_Status):
            set StatusList += array{(Key, Element)}
        
        return StatusList

    # スキルの状態(SkillStatus)を設定
    SetSkillStatus<public>(InPlayer:player, Index:int, StatusData:skill_status)<transacts>:void=
        if:
            NowData := SaveData[InPlayer]
            # 現在のデータを更新
            set SaveData[InPlayer] = save_data:
                MakeSaveData<constructor>(NowData)
                Skill_Status := AddMap(NowData.Skill_Status,(Index,StatusData))
        then:

    # バフを使用する関数
    BuffTimeStatus<public>(InPlayer:player, Index:int)<suspends>:void=
        loop:
            Sleep(1.0)
            Print("BuffTimeStatus")
            if(SkillStatus := GetSkillStatus[InPlayer, Index]):
                Time := SkillStatus.EffectDuration - 1
                Print("バフ残り時間 : {Time}")
                
                if(Time >= 0):

                    SetSkill_status : skill_status = skill_status{
                        Index := SkillStatus.Index,
                        IsActive := true,
                        CooldownTime := SkillStatus.CooldownTime,
                        MaxCooldownTime := SkillStatus.MaxCooldownTime,
                        Charges := SkillStatus.Charges,
                        MaxCharges := SkillStatus.MaxCharges,
                        LastUsedTime := SkillStatus.LastUsedTime,
                        EffectDuration := Time,
                        Multiply := SkillStatus.Multiply
                    }

                    SetSkillStatus(InPlayer, Index, SetSkill_status)
                
                else:
                    # 終わったらSkill Statusは空にする
                    SetSkillStatus(InPlayer, Index, skill_status{})
                    break
    
    # --- セーブデータの初期化周りの実装です。基本いじらなくて大丈夫です ---
    OnBegin<override>()<suspends>:void=
        for(Player : GetPlayspace().GetPlayers()):
            OnPlay(Player)
        GetPlayspace().PlayerAddedEvent().Subscribe(OnPlay)
        spawn{AwaitQuestComplate()}

    OnPlay<public>(Player:player):void=
        if(SaveDataInit[Player]):
            Print("SaveDataManagerInit")
            GetGlobalData().FirstPlayedThisMapEvent.Signal(Player)

        else if(NowData := SaveData[Player]):
            Print("Has NowData")
            if:
                NowData.LastPlayedVersion < CurrentVersion
                set SaveData[Player] = save_data:
                    MakeSaveData<constructor>(NowData)
                    LastPlayedVersion := CurrentVersion
            then:
                GetGlobalData().FirstPlayedThisVersionEvent.Signal(Player)

        GetGlobalData().YokaiSlotUpdateEvent.Signal(Player)
        #----------Quest呼び出し----------
        if(NowData := SaveData[Player]):
            CompletedQuests := GetCompletedQuests(Player)
            LastComplateQuests := SortBy(CompletedQuests, QuestSort)
            if(Last := LastComplateQuests[0]):
                spawn{QuestSequencer.PlaySequence(Player, Last + 1)}
                spawn{Indicator.SetFirstPulse(Player, Last + 1)}
            else:
                #最初のクエスト
                spawn{QuestSequencer.PlaySequence(Player, 0)}
                spawn{Indicator.SetFirstPulse(Player, 0)}
        #--------------------------------

    AwaitQuestComplate<public>()<suspends>:void=
        loop:
            Info := QuestSequencer.QuestComplateEvent.Await()
            if(Player := player[Info(0)]):
                CompleteQuest(Player, Info(1))
    # --- ここまで ---

QuestSort<public>(Left:int, Right:int)<computes><decides>:void=
    Left > Right
