
pet_info<public> := struct<persistable>:
    Level<public> : int = 1
    Upgrade<public> : int = 0
    PetSlot<public> : int = -1