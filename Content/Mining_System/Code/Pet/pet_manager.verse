using { /Fortnite.com/Devices }
using { /Verse.org/Simulation }
using { /Verse.org/Assets }
using { /Fortnite.com/Characters }
using { Boss_Textures }

using { PlayerData }
using { Utility }
using { Profile }

pet_manager<public> := class(creative_device):

    @editable
    PetStatus<public> : []pet_status = array{}

    @editable
    SaveDataManager : save_data_manager = save_data_manager{}

    PetIcon<public> : [tuple(int, int)]texture = map:
        (1, 0) => koppar
        (2, 0) => kogasa
        (3, 0) => kachinta
        (4, 0) => hagiremomen
        (5, 0) => kajimaru
        (1, 1) => kreppar
        (2, 1) => karakasaobake
        (3, 1) => tachigarasu
        (4, 1) => obanmomen
        (5, 1) => enengama
        (1, 2) => kreppar
        (2, 2) => karakasaobake
        (3, 2) => tachigarasu
        (4, 2) => obanmomen
        (5, 2) => enengama

        (11, 0) => _koppa
        (12, 0) => _kogasa
        (13, 0) => _kachinta
        (14, 0) => _hagiremomen
        (15, 0) => _kajimaru
        (11, 1) => _creppa
        (12, 1) => _karakasaobake
        (13, 1) => _tachigarasu
        (14, 1) => _obanmomen
        (15, 1) => _enengama
        (11, 2) => _creppa
        (12, 2) => _karakasaobake
        (13, 2) => _tachigarasu
        (14, 2) => _obanmomen
        (15, 2) => _enengama

    GetEquipedPetStatus<public>(InPlayer:player)<decides><transacts>:[]tuple(pet_status, pet_info)=
        var Result : []tuple(pet_status, pet_info) = array{}

        for(PetElement : PetStatus, HasPet := SaveDataManager.GetPet[InPlayer, PetElement.Index], HasPet.PetSlot >= 1, HasPet.PetSlot <= 4):
            set Result += array:
                (PetElement, HasPet)
        
        return Result

    GetPetID<public>(InPlayer:player, Index:int)<decides><transacts>:tuple(int, int)=
        PetData := SaveDataManager.GetPet[InPlayer, Index]
        PetInfo := PetData
        (Index, Min(PetData.Upgrade, 1))

    GetEquipedRestPetStatus<public>(InPlayer:player)<decides><transacts>:[]tuple(pet_status, pet_info)=
        var Result : []tuple(pet_status, pet_info) = array{}
    
        for(PetElement : PetStatus, HasPet := SaveDataManager.GetPet[InPlayer, PetElement.Index], HasPet.PetSlot >= 11, HasPet.PetSlot <= 16):
            set Result += array:
                (PetElement, HasPet)
    
        return Result

    GetPetStatus<public>(InPlayer:player, Index:int)<decides><transacts>:tuple(pet_status, pet_info)=
        Status := PetStatus[Index]
        PetData := SaveDataManager.GetPet[InPlayer, Status.Index]
        return (Status, PetData) 

    GetPetStatusByIndex<public>(Index:int)<decides><transacts>:pet_status=
        (
            for(PetElement : PetStatus, PetElement.Index = Index):
                PetElement
        )[0]

    UnlockPet<public>(InPlayer:player, Index:int):void=
        if:
            HasPet := SaveDataManager.GetPet[InPlayer, Index]
        then:
            Print("Unlocking Pet for ID: {Index}, HasPet: {HasPet.Level}")
            if(HasPet.Level < 30, HasPet.Upgrade = 0):
                LevelUp := HasPet.Level+1
                SaveDataManager.SetPet(
                        InPlayer,
                        Index,
                        pet_info{Level := LevelUp, Upgrade := HasPet.Upgrade, PetSlot := HasPet.PetSlot}
                    )
            if(HasPet.Level < 100, HasPet.Upgrade = 1):
                LevelUp := HasPet.Level+1
                SaveDataManager.SetPet(
                        InPlayer,
                        Index,
                        pet_info{Level := LevelUp, Upgrade := HasPet.Upgrade, PetSlot := HasPet.PetSlot}
                    )
                    

        else:
            Print("No existing pet found for ID: {Index}, creating new entry.")
            SaveDataManager.SetPet(
                InPlayer,
                Index,
                pet_info{
                    Level := 1,
                    Upgrade := 0,
                    PetSlot := GetEmptySlot(InPlayer)
                }
            )

        GetGlobalData().YokaiSlotUpdateEvent.Signal(InPlayer)

    GetEmptySlot<public>(InPlayer:player)<transacts>:int=
        UsedSlots := for(PetElement : PetStatus, Slots := GetPetSlot[InPlayer, PetElement.Index]):
                Slots
        for(Slot := 1..4, not UsedSlots.Find[Slot]):
            return Slot
        for(Slot := 11..16, not UsedSlots.Find[Slot]):
            return Slot
        return 0

    GetEmptyEquipSlot<public>(InPlayer:player)<transacts><decides>:int=
        UsedSlots := for(PetElement : PetStatus, Slots := GetPetSlot[InPlayer, PetElement.Index]):
                Slots
        (
            for(Slot := 1..4, not UsedSlots.Find[Slot]):
                Slot
        )[0]

    GetEmptyRestSlot<public>(InPlayer:player)<transacts><decides>:int=
        UsedSlots := for(PetElement : PetStatus, Slots := GetPetSlot[InPlayer, PetElement.Index]):
                Slots
        (
            for(Slot := 11..16, not UsedSlots.Find[Slot]):
                Slot
        )[0]

    GetPetCount<public>(InPlayer:player)<transacts>:int=
        if(HasPet := SaveDataManager.GetPets[InPlayer]):
            return HasPet.Length
        else:
            return 0

    GetTotalLevel<public>(InPlayer:player)<transacts>:int=
        var Result : int = 0

        for(PetElement : SaveDataManager.GetPets[InPlayer]):
            set Result += PetElement(1).Level

        return Result

    GetEquipedPetCount<public>(InPlayer:player)<transacts>:int=
        if(HasPet := GetEquipedPetStatus[InPlayer]):
            return HasPet.Length
        else:
            return 0

    # -1 : 未装備, 1 - 4 : 装備スロット, 11 - 16 : 放置 // 
    GetPetSlot<public>(InPlayer:player, Index:int)<decides><transacts>:int=
        SaveDataManager.GetPet[InPlayer, Index].PetSlot

    SetPetSlot<public>(InPlayer:player, Index:int, Slot:int):void=
        if:
            CurrentData := SaveData[InPlayer]
            CurrentPetInfo := CurrentData.Pet_Info[Index]
            set SaveData[InPlayer] = save_data:
                MakeSaveData<constructor>(CurrentData)
                Pet_Info := AddMap(
                        CurrentData.Pet_Info,
                        (
                            Index,
                            pet_info{
                                Level := CurrentPetInfo.Level
                                Upgrade := CurrentPetInfo.Upgrade
                                PetSlot := Slot}
                        )
                        
                    )
        then:
            Print("SetPetSlot 成功")
        else:
            Print("SetPetSlot 失敗")
        GetGlobalData().YokaiSlotUpdateEvent.Signal(InPlayer)

    GetPetByIndex<public>(Index:int)<decides><transacts>:pet_status=
        (
            for(PetElement : PetStatus, PetElement.Index = Index):
                PetElement
        )[0]

    GetPetBySlot<public>(InPlayer:player, Slot:int)<decides><transacts>:tuple(pet_status, pet_info)=
        (
            for(PetElement : PetStatus, Info := (SaveDataManager.GetPet[InPlayer, PetElement.Index]), Info.PetSlot = Slot):
                Print("Called GetPetBySlot")
                (PetElement, Info)
        )[0]

    GetPetGatherRate<public>(InPlayer:player, Index:int)<decides><transacts>:big_number=
        var Result : big_number = big_number{Value := 0.0, Exponent := 0}
        PetInfo := SaveDataManager.GetPet[InPlayer, Index]
        Status := GetPetByIndex[Index]
        var BaseStatus : float = Status.GatherRate
        # Print("Initial BaseStatus for PetIndex {Index}: {BaseStatus}")
        # Print("Pet Level for PetIndex {Index}: {PetInfo.Level}")
        for(LevelCount := 1.. Min(PetInfo.Level, 30)):
            set BaseStatus += BaseStatus*0.05
            # Print("BaseStatus after 5% increment for Level {LevelCount}: {BaseStatus}")
        if(PetInfo.Level > 30):
            for(LevelCount := 31.. Min(PetInfo.Level, 70)):
                set BaseStatus += BaseStatus*0.1
        if(PetInfo.Level > 70):
            for(LevelCount := 71.. Min(PetInfo.Level, 100)):
                set BaseStatus += BaseStatus*0.15
        set Result = BigNumber_Add(BigNumber_MultiplyByFloat(big_number{Value := 1.0, Exponent := 0}, BaseStatus), Result)

        return Result

    GetPetCriticalRate<public>(InPlayer:player, Index:int)<decides><transacts>:big_number=
        PetInfo := SaveDataManager.GetPet[InPlayer, Index]
        Status := GetPetByIndex[Index]
        var BaseStatus : float = Status.CriticalRate
        for(LevelCount := 1.. Min(PetInfo.Level, 30)):
            set BaseStatus += BaseStatus*0.05
        if(PetInfo.Level > 30):
            for(LevelCount := 31.. Min(PetInfo.Level, 70)):
                set BaseStatus += BaseStatus*0.1
        if(PetInfo.Level > 70):
            for(LevelCount := 71.. Min(PetInfo.Level, 100)):
                set BaseStatus += BaseStatus*0.15

        return BigNumber_MultiplyByFloat(big_number{Value := 1.0, Exponent := 0}, BaseStatus)

    GetPetCriticalDamage<public>(InPlayer:player, Index:int)<decides><transacts>:big_number=
        PetInfo := SaveDataManager.GetPet[InPlayer, Index]
        Status := GetPetByIndex[Index]
        var BaseStatus : float = Status.CriticalDamage
        for(LevelCount := 1.. Min(PetInfo.Level, 30)):
            set BaseStatus += BaseStatus*0.05
        if(PetInfo.Level > 30):
            for(LevelCount := 31.. Min(PetInfo.Level, 70)):
                set BaseStatus += BaseStatus*0.1
        if(PetInfo.Level > 70):
            for(LevelCount := 71.. Min(PetInfo.Level, 100)):
                set BaseStatus += BaseStatus*0.15

        return BigNumber_MultiplyByFloat(big_number{Value := 1.0, Exponent := 0}, BaseStatus)

    GetPetLastHit<public>(InPlayer:player, Index:int)<decides><transacts>:big_number=
        PetInfo := SaveDataManager.GetPet[InPlayer, Index]
        Status := GetPetByIndex[Index]
        var BaseStatus : float = Status.LastHit
        for(LevelCount := 1.. Min(PetInfo.Level, 30)):
            set BaseStatus += BaseStatus*0.05
        if(PetInfo.Level > 30):
            for(LevelCount := 31.. Min(PetInfo.Level, 70)):
                set BaseStatus += BaseStatus*0.1
        if(PetInfo.Level > 70):
            for(LevelCount := 71.. Min(PetInfo.Level, 100)):
                set BaseStatus += BaseStatus*0.15

        return BigNumber_MultiplyByFloat(big_number{Value := 1.0, Exponent := 0}, BaseStatus)

    GetPetStrength<public>(InPlayer:player, Index:int)<decides><transacts>:big_number=
        PetInfo := SaveDataManager.GetPet[InPlayer, Index]
        Status := GetPetByIndex[Index]
        var BaseStatus : float = Status.STRPower
        for(LevelCount := 1.. Min(PetInfo.Level, 30)):
            set BaseStatus += BaseStatus*0.05
        if(PetInfo.Level > 30):
            for(LevelCount := 31.. Min(PetInfo.Level, 70)):
                set BaseStatus += BaseStatus*0.1
        if(PetInfo.Level > 70):
            for(LevelCount := 71.. Min(PetInfo.Level, 100)):
                set BaseStatus += BaseStatus*0.15

        return BigNumber_MultiplyByFloat(big_number{Value := 1.0, Exponent := 0}, BaseStatus)

    GetPetTotalStrength<public>(InPlayer:player)<transacts>:big_number=
        var Result : big_number = big_number{}

        for(PetElement : SaveDataManager.GetPets[InPlayer], STRPower := GetPetStrength[InPlayer, PetElement(0)]):
            set Result = BigNumber_Add(Result, STRPower)

        return Result

    GetPetTotalGatherRate<public>(InPlayer:player)<transacts>:big_number=
        var Result : big_number = big_number{}

        for(PetElement : SaveDataManager.GetPets[InPlayer], GatherRate := GetPetGatherRate[InPlayer, PetElement(0)]):
            set Result = BigNumber_Add(Result, GatherRate)

        return Result

    GetPetTotalCriticalRate<public>(InPlayer:player)<transacts>:big_number=
        var Result : big_number = big_number{}

        for(PetElement : SaveDataManager.GetPets[InPlayer], CriticalRate := GetPetCriticalRate[InPlayer, PetElement(0)]):
            set Result = BigNumber_Add(Result, CriticalRate)

        return Result

    GetPetTotalCriticalDamage<public>(InPlayer:player)<transacts>:big_number=
        var Result : big_number = big_number{}

        for(PetElement : SaveDataManager.GetPets[InPlayer], CriticalDamage := GetPetCriticalDamage[InPlayer, PetElement(0)]):
            set Result = BigNumber_Add(Result, CriticalDamage)

        return Result

    GetPetTotalLastHit<public>(InPlayer:player)<transacts>:big_number=
        var Result : big_number = big_number{}

        for(PetElement : SaveDataManager.GetPets[InPlayer], LastHit := GetPetLastHit[InPlayer, PetElement(0)]):
            set Result = BigNumber_Add(Result, LastHit)

        return Result
        
    GetPetIndexInFirstSlot<public>(InPlayer:player)<decides><transacts>:int=
        var Result : []tuple(pet_status, pet_info) = GetEquipedPetStatus[InPlayer]
        var Index : int = -1
    
        for(TupleElement : Result):
            Status := TupleElement(0)
            Info := TupleElement(1)
            if(Info.PetSlot = 1):
                set Index = Status.Index
    
        return Index
        
    GetPetUpgradeInFirstSlot<public>(InPlayer:player)<decides><transacts>:int=
        var Result : []tuple(pet_status, pet_info) = GetEquipedPetStatus[InPlayer]
        var Upgrade : int = -1
    
        for(TupleElement : Result):
            Status := TupleElement(0)
            Info := TupleElement(1)
            if(Info.PetSlot = 1):
                set Upgrade = Info.Upgrade
    
        return Upgrade

    GetTotalLevelBonus<public>(InPlayer:player)<transacts>:big_number=
        var TotalLevel : int = 0

        if:
            for(PetElement : SaveDataManager.GetPets[InPlayer]):
                PetIndex := PetElement(0)
                PetInfo := SaveDataManager.GetPet[InPlayer, PetIndex]
                set TotalLevel += PetInfo.Level

        # 合計レベルに400.0を掛けてTotalLevelBonusを計算
        var TotalLevelBonus : float = TotalLevel * 400.0

        return BigNumber_MultiplyByFloat(big_number{Value := 1.0, Exponent := 0}, TotalLevelBonus)

    # Cube追記です
    GetPetUpgradeByIndex<public>(InPlayer:player, Index:int)<decides><transacts>:int=
        var Result : []tuple(pet_status, pet_info) = GetEquipedPetStatus[InPlayer]
        var Upgrade : int = -1
    
        for(TupleElement : Result):
            Status := TupleElement(0)
            Info := TupleElement(1)
            if(Info.PetSlot = Index):
                set Upgrade = Info.Upgrade
    
        return Upgrade
        
    GetPetsInSlots1_4<public>(InPlayer:player)<decides><transacts>:[]tuple(int, int)=
        var Result : []tuple(int, int) = array{}
    
        var EquipedPets : []tuple(pet_status, pet_info) = GetEquipedPetStatus[InPlayer]
    
        for(TupleElement : EquipedPets):
            Status := TupleElement(0)
            Info := TupleElement(1)
            
            # Slot1~4 のペットのインデックスとUpgradeを配列に追加
            set Result += array{(Status.Index, Info.Upgrade)}
    
        return Result

    GetPetsInSlots11_16<public>(InPlayer:player)<decides><transacts>:[]tuple(int, int)=
        var Result : []tuple(int, int) = array{}
    
        var EquipedPets : []tuple(pet_status, pet_info) = GetEquipedRestPetStatus[InPlayer]
    
        for(TupleElement : EquipedPets):
            Status := TupleElement(0)
            Info := TupleElement(1)
            
            # Slot11~16 のペットのインデックスと Upgrade を配列に追加
            set Result += array{(Status.Index, Info.Upgrade)}
    
        return Result

    GetPetEachGatherRateFromIndex<public>(InPlayer:player, Index:int)<transacts>:big_number =
        var Result : big_number = big_number{Value := 0.0, Exponent := 0}
    
        for(PetElement : SaveDataManager.GetPets[InPlayer]):
            PetIndex := PetElement(0)
            PetSlot := PetElement(1).PetSlot
            
            # PET IDとPET Indexが一致するとき
            if(PetSlot >= 1 and PetSlot <= 4, GetPetID[InPlayer, PetIndex] = Index):
                if:
                    GatherRate := GetPetGatherRate[InPlayer, PetIndex]
                then:
                    set Result = BigNumber_Add(Result, GatherRate)
    
        return Result
    
    GetPetGatherRateFromEquipSlots<public>(InPlayer:player)<transacts>:big_number =
        var Result : big_number = big_number{Value := 0.0, Exponent := 0}
    
        for(PetElement : SaveDataManager.GetPets[InPlayer]):
            PetIndex := PetElement(0)
            PetSlot := PetElement(1).PetSlot
            if(PetSlot >= 1 and PetSlot <= 4):
                if:
                    GatherRate := GetPetGatherRate[InPlayer, PetIndex]
                then:
                    set Result = BigNumber_Add(Result, GatherRate)
    
        return Result
    
    GetPetGatherRateFromRestSlots<public>(InPlayer:player)<transacts>:big_number =
        var Result : big_number = big_number{Value := 0.0, Exponent := 0}
    
        for(PetElement : SaveDataManager.GetPets[InPlayer]):
            PetIndex := PetElement(0)
            PetSlot := PetElement(1).PetSlot
            Print("PetSlot for PetIndex {PetIndex}: {PetSlot}")
            if(PetSlot >= 11 and PetSlot <= 16):
                if:
                    GatherRate := GetPetGatherRate[InPlayer, PetElement(0)]
                    Print("GatherRate for PetIndex {PetIndex}: {BigNumber_FormatDisplay(GatherRate)}")
                then:
                    set Result = BigNumber_Add(Result, GatherRate)
    
        return Result

    # 装備中のペットのSTRを百分率で返す
    GetPetStrengthFromEquipSlots<public>(Player:player)<transacts>:big_number =
        var Result : big_number = big_number{Value := 0.0, Exponent := 0}
    
        for(PetElement : SaveDataManager.GetPets[Player]):
            PetIndex := PetElement(0)
            if(PetIndex >= 1 and PetIndex <= 4):
                if:
                    PetStrength := GetPetStrength[Player, PetIndex]
                    set Result = BigNumber_Add(Result, PetStrength)
    
        return Result

    # 装備中のペットのCri確率を百分率で返す
    GetPetCriticalRateFromEquipSlots<public>(InPlayer:player)<transacts>:big_number =
        var Result : big_number = big_number{Value := 0.0, Exponent := 0}
    
        for(PetElement : SaveDataManager.GetPets[InPlayer]):
            PetIndex := PetElement(0)
            if(PetIndex >= 1 and PetIndex <= 4):
                if:
                    _PetStatus := GetPetByIndex[PetIndex]
                    CriticalRate := _PetStatus.CriticalRate
                    set Result = BigNumber_Add(Result, big_number{Value := CriticalRate, Exponent := 0})
    
        return Result

    # 装備中のペットのCri倍率を百分率で返す
    GetPetCriticalDamageFromEquipSlots<public>(InPlayer:player)<transacts>:big_number =
        var Result : big_number = big_number{Value := 0.0, Exponent := 0}
    
        for(PetElement : SaveDataManager.GetPets[InPlayer]):
            PetIndex := PetElement(0)
            if(PetIndex >= 1 and PetIndex <= 4):
                if:
                    _PetStatus := GetPetByIndex[PetIndex]
                    CriticalDamage := _PetStatus.CriticalDamage
                    set Result = BigNumber_Add(Result, big_number{Value := CriticalDamage, Exponent := 0})
    
        return Result

    # 装備中のペットのLastHitを百分率で返す
    GetPetLastHitBonusFromEquipSlots<public>(InPlayer:player)<transacts>:big_number =
        var Result : big_number = big_number{Value := 0.0, Exponent := 0}
    
        for(PetElement : SaveDataManager.GetPets[InPlayer]):
            PetIndex := PetElement(0)
            if(PetIndex >= 1 and PetIndex <= 4):
                if:
                    _PetStatus := GetPetByIndex[PetIndex]
                    LastHit := _PetStatus.LastHit
                    set Result = BigNumber_Add(Result, big_number{Value := LastHit, Exponent := 0})
    
        return Result


    # ----------------------------------------------------------------------------------------------------
    # 書き足しポイント
    GetPetLevel<public>(InPlayer:player, Index:int):int=

        var GetLevel : int = 0

        if:
            HasPet := SaveDataManager.GetPet[InPlayer, Index]
        then:
            set GetLevel = HasPet.Level

        GetLevel

    GetPetUpgrade<public>(InPlayer:player, Index:int):int=

        var GetLevel : int = 0

        if:
            HasPet := SaveDataManager.GetPet[InPlayer, Index]
        then:
            set GetLevel = HasPet.Upgrade

        GetLevel

    SetPetLevel<public>(InPlayer:player, Index:int, Count:int):void=

        if:
            HasPet := SaveDataManager.GetPet[InPlayer, Index]
        then:
            PetInfo: pet_info = pet_info{Level := Count ,Upgrade:= HasPet.Upgrade, PetSlot:= HasPet.PetSlot}

            SaveDataManager.SetPet(InPlayer, Index, PetInfo)

    SetPetUpgrade<public>(InPlayer:player, Index:int, Count:int):void=

        if:
            HasPet := SaveDataManager.GetPet[InPlayer, Index]
        then:
            PetInfo: pet_info = pet_info{Level := HasPet.Level ,Upgrade:= Count, PetSlot:= HasPet.PetSlot}

            SaveDataManager.SetPet(InPlayer, Index, PetInfo)

    UpgradePet<public>(InPlayer:player, Index:int, Chibi_NPC_DespawnTrigger:trigger_device, Big_NPC_SpawnTrigger:trigger_device, StageIndex:int):void=
        GetNPCLevel := GetPetLevel(InPlayer, Index)

        var NeedStone : big_number = big_number{}
        var NeetPowerUpStone : big_number = big_number{}

        # Playerのリソースの取得
        PlayerStone := SaveDataManager.GetResourceStone(InPlayer)
        PlayerPoweUpStone := SaveDataManager.GetResourcePowerupStone(InPlayer)

        Print("Run UpgradePet", ?Duration := 10.0)

        if:
            PetUpgrade := GetPetUpgradeByIndex[InPlayer, Index]
        then:

            # 30レベル、70レベル以上は進化するまでレベル上げられない認識で書いてます。
            if(GetNPCLevel = 30):

                # 必要なリソースの数
                # ステージごとに設定
                if(StageIndex = 0):
                    set NeedStone = big_number{Value :=22.0, Exponent := 6}
                    set NeetPowerUpStone = big_number{Value :=24.0, Exponent := 0}
                if(StageIndex = 1):
                    set NeedStone = big_number{Value :=80.0, Exponent := 18}
                    set NeetPowerUpStone = big_number{Value :=256.0, Exponent := 0}
                if(StageIndex = 2):
                    set NeedStone = big_number{Value :=540.0, Exponent := 24}
                    set NeetPowerUpStone = big_number{Value :=2.0, Exponent := 3}

                # リソースが足りてるかの判定
                StoneLogic := BigNumber_GreaterThanOrEquals(PlayerStone, NeedStone)
                PowerUpLogic := BigNumber_GreaterThanOrEquals(PlayerPoweUpStone, NeetPowerUpStone)

                Print("PlayerStone : {BigNumber_FormatDisplay(PlayerStone)}" ,?Duration := 10.0)
                Print("NeedStone : {BigNumber_FormatDisplay(NeedStone)}" ,?Duration := 10.0)

                # BigNumberを引いて購入するコード
                if(StoneLogic = true, PowerUpLogic = true):
                    SubtractPlayerStone := BigNumber_Subtract(PlayerStone, NeedStone)
                    SubtractPlayerPoweUpStone := BigNumber_Subtract(PlayerStone, NeetPowerUpStone)

                    SaveDataManager.SetResourceStone(InPlayer, SubtractPlayerStone)
                    SaveDataManager.SetResourcePowerupStone(InPlayer, SubtractPlayerPoweUpStone)

                    # Petの実際のUpgrade処理
                    SetPetUpgrade(InPlayer, Index, 1)
                    Chibi_NPC_DespawnTrigger.Trigger(InPlayer)
                    Big_NPC_SpawnTrigger.Trigger(InPlayer)
                    SetPetLevel(InPlayer, Index, 31)

            # アップデート予定
            if(GetNPCLevel = 999):

                # 必要なリソースの数
                # ステージごとに設定
                if(StageIndex = 1):
                    set NeedStone = big_number{Value :=110.0, Exponent := 6}
                    set NeetPowerUpStone = big_number{Value :=72.0, Exponent := 0}
                if(StageIndex = 2):
                    set NeedStone = big_number{Value :=400.0, Exponent := 18}
                    set NeetPowerUpStone = big_number{Value :=768.0, Exponent := 0}
                if(StageIndex = 3):
                    set NeedStone = big_number{Value :=2.70, Exponent := 27}
                    set NeetPowerUpStone = big_number{Value :=6.0, Exponent := 3}

                # リソースが足りてるかの判定
                StoneLogic := BigNumber_GreaterThanOrEquals(PlayerStone, NeedStone)
                PowerUpLogic := BigNumber_GreaterThanOrEquals(PlayerPoweUpStone, NeetPowerUpStone)

                # BigNumberを引いて購入するコード
                if(StoneLogic = true, PowerUpLogic = true):
                    SubtractPlayerStone := BigNumber_Subtract(PlayerStone, NeedStone)
                    SubtractPlayerPoweUpStone := BigNumber_Subtract(PlayerStone, NeetPowerUpStone)

                    SaveDataManager.SetResourceStone(InPlayer, SubtractPlayerStone)
                    SaveDataManager.SetResourcePowerupStone(InPlayer, SubtractPlayerPoweUpStone)

                    # Petの実際のUpgrade処理
                    SetPetUpgrade(InPlayer, Index, 2)
                    Chibi_NPC_DespawnTrigger.Trigger(InPlayer)
                    Big_NPC_SpawnTrigger.Trigger(InPlayer)

            else:
                Print("Pet Cant upgrade...")

        else:
            Print("Need more Level!!")

    # アップグレードした時送るシグナル
    PetUpgradeSignal<public> : event(tuple(player,int,int)) = event(tuple(player,int,int)){}

    SignalPetUpgrade<public>(InPlayer:player, Index:int, StageIndex:int):void=
        PetUpgradeSignal.Signal(InPlayer, Index, StageIndex)
        Print("Send Signal!!", ?Duration := 10.0)

    # ----------------------------------------------------------------------------------------------------
    
    <#CheckSkill<public>(InPlayer:player):void=
        if(MyPet := GetPetBySlot[InPlayer, 1], FortChar := InPlayer.GetFortCharacter[]):
            if(MyPet(0).PetSkill_ID = 0):
                if(MyPet(1).Upgrade = 0):
                    FortChar.SetShield(50.0)
                else:
                    FortChar.SetShield(100.0)
            else:
                if(MyPet(1).Upgrade = 0):
                    FortChar.SetShield(0.0)
                else:
                    FortChar.SetShield(0.0)pet_manager#>