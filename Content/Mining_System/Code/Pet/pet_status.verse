using { /Verse.org/Simulation }
using { /Fortnite.com/Devices }
using { /Verse.org/Assets }

pet_status<public> := class(creative_device):

    @editable
    PetName<public> : string = ""

    @editable
    Index<public> : int = 0

    @editable
    GatherRate<public> : float = 0.0

    @editable
    STRPower<public> : float = 0.0

    @editable
    CriticalRate<public> : float = 0.0

    @editable
    CriticalDamage<public> : float = 0.0

    @editable
    LastHit<public> : float = 0.0

    # 百分率で設定しているのでUIはそのままで、計算の際は100で割って小数の割合にしてから計算してください

    @editable
    PetSkill_ID<public> : int = 1

    @editable
    PetSkillName<public> : string = "Skill Name"

    @editable_text_box {MultiLine := true}
    PetSkillDescription<public> : string = "Skill Description"

    @editable
    AdditionalValue<public> : string = "100%"

    PetSkillIcon<public> : [int]texture = map:
        1 => PetSkill_ID_set.PetSkill_1
        2 => PetSkill_ID_set.PetSkill_2
        3 => PetSkill_ID_set.PetSkill_3
        4 => PetSkill_ID_set.PetSkill_4
        5 => PetSkill_ID_set.PetSkill_5

    GetAdditionalValue<public>():string=
        AdditionalValue