using { /Verse.org/Simulation }
using { /Fortnite.com/Devices }
using { /Verse.org/Random }

pet_unlock_check := class(creative_device):
    @editable
    UnlockButton : button_device = button_device{}

    @editable
    PetManager : pet_manager = pet_manager{}

    OnBegin<override>()<suspends>:void=
        loop:
            Agent := UnlockButton.InteractedWithEvent.Await()
            if(Player := player[Agent]):
                RandomNumber := GetRandomInt(1, 5)
                PetManager.UnlockPet(Player, RandomNumber)