using { /Fortnite.com/Devices }
using { /Verse.org/Simulation }
using { /UnrealEngine.com/Temporary/SpatialMath }
using { /Fortnite.com/Devices/CreativeAnimation }
using { /Verse.org/Random }
using { /UnrealEngine.com/Temporary/UI }
using { /Verse.org/Colors }
using { HPBar }
using { VFXSystem }
using { Profile }
using { Utility }
using { PlayerData }
using { Skill }
using {VerseDevice}
using { HPBar }

mining_object_device<public> := class(creative_device):
    #XP付与
    @editable XP_OnBreak : accolades_device = accolades_device{}
    @editable XP_GotSumSto : accolades_device = accolades_device{}

    @editable VFXChance : int = 30
    @editable VFX_OnDamge1 : creative_prop_asset = DefaultCreativePropAsset
    @editable VFX_OnDamge2 : creative_prop_asset = DefaultCreativePropAsset
    
    @editable T_SummonStone_R1 : trigger_device = trigger_device{}
    @editable T_SummonStone_R2 : trigger_device = trigger_device{}
    @editable T_SummonStone_R3 : trigger_device = trigger_device{}
    @editable T_SummonStone_R4 : trigger_device = trigger_device{}
    @editable T_SummonStone_R5 : trigger_device = trigger_device{}
    @editable VFX_Area : visual_effect_powerup_device = visual_effect_powerup_device{}
    @editable SFX_Area : audio_player_device = audio_player_device{}
    @editable SFX_OnBroken : audio_player_device = audio_player_device{}
    @editable VFX_Powerup : damage_amplifier_powerup_device = damage_amplifier_powerup_device{}
    @editable VFX_OnHit : visual_effect_powerup_device = visual_effect_powerup_device{}
    @editable SFX_OnHit : audio_player_device = audio_player_device{}

    @editable
    InfoArray : []mining_object_info = array{}

    @editable
    MaxHealth : editable_big_number = editable_big_number{Value := 100.0}

    @editable
    DamagedScale : vector3 = vector3{X := 1.0, Y := 1.0, Z := 1.0}

    @editable
    HPBarOffset : float = 100.0

    @editable
    HPBarDevice : hp_bar_device = hp_bar_device{}

    @editable
    DebugButton : button_device = button_device{}

    @editable
    GlobalDestroyEvent : channel_device = channel_device{}

    @editable
    GlobalDestroyEventTrigger : trigger_device = trigger_device{}

    @editable
    VFXDevice : vfx_device = vfx_device{}

    @editable
    MiningObjectWidgetManager : mining_object_widget_manager = mining_object_widget_manager{}

    @editable
    AutoRestoreTime : ?float = option{5.0}

    @editable
    TransmitMode : transmit_mode = transmit_mode.AllPlayers

    @editable
    MiningDamageDevice : mining_damage_device = mining_damage_device{}

    @editable
    SaveDataManager : Mining_System.Code.PlayerData.save_data_manager = Mining_System.Code.PlayerData.save_data_manager{}

    @editable
    UI_HUD_Count : ui_hud_count = ui_hud_count{}

    @editable
    PetManager : Pet.pet_manager = Pet.pet_manager{}


    # @editable
    # HPBarDeviceUI : hp_bar_device_ui = hp_bar_device_ui{}

    @editable
    QuestTrigger1 : trigger_device = trigger_device{}

    @editable
    QuestTrigger2 : trigger_device = trigger_device{}

    var FieldSkill : FieldSkill_Control = FieldSkill_Control{}

    #--------------------------------------------------NPC Signal--------------------------------------------------
    var NPCMiningDevice : ?int = false
    var SignalDamagedPlayer<public> : ?agent = false
    var SignalDestroyedPlayer<public> : ?agent = false
    var DamagedPropManipulator<public> : int = 0
    var DestroyedPropManipulator<public> : int = 0
    NPCDamagedEvent : event(mining_object_info) = event(mining_object_info){}
    NPCDestroyedEvent : event(prop_manipulator_device) = event(prop_manipulator_device){}
    #--------------------------------------------------------------------------------------------------------------

    SummonStonePercents : tuple(float, float, float, float, float)= (76.5, 18.0, 4.3, 1.0, 0.2)

    OnBegin<override>()<suspends>:void=
        for(Info : InfoArray):
            set Info.DefaultTransform = Info.TargetProp.GetTransform()
            Info.RestoreHealth(MaxHealth.ToPersistableBigNumber())
            spawn{AwaitDestroy(Info)}
            spawn{AwaitDamage(Info)}
            spawn{AwaitPickaxeDamage(Info)}

    SignalPlayerDamagedEventFunction()<suspends>:tuple(?agent, int)=
        return (SignalDamagedPlayer, DamagedPropManipulator)

    SignalPlayerDestroyedEventFunction()<suspends>:tuple(?agent, int)=
        return (SignalDestroyedPlayer, DestroyedPropManipulator)

    # サイズ変更
    Rescale(Info:mining_object_info):void=
        if (NowHealth := Info.NowHealth?, AnimController := Info.TargetProp.GetAnimationController[]):
            # 現在のHPと最大HPを比較して割合を計算
            Percent := BigNumber_Divide(NowHealth, MaxHealth.ToPersistableBigNumber()).Value / 10.0
    #         Print("Percent: {Percent}")
    
    #         # スケール範囲を (2,2,2) から (1,1,1) の間に設定
    #         MinScale := vector3{X := 1.0, Y := 1.0, Z := 1.0}
    #         MaxScale := vector3{X := 2.0, Y := 2.0, Z := 2.0}
    
    #         # Percentに基づいて新しいスケールを計算
    #         NewScale := Lerp(MinScale, MaxScale, Percent)
    #         Print("NewScale: {NewScale}")
    
    #         # デフォルトスケールとの差分を計算し、DeltaScale として設定
    #         DefaultScale := Info.DefaultTransform.Scale
    #         DeltaScale := NewScale - DefaultScale  # デフォルトスケールとの差分を基準に設定
    
    #         # アニメーション用のキーフレームを設定し、DeltaScale のみでスケール変更を実現
    #         KeyFrames := array:
    #             keyframe_delta:
    #                 Time := 0.5
    #                 DeltaLocation := vector3{}  # 位置の変更はなし
    #                 DeltaRotation := rotation{}  # 回転の変更はなし
    #                 DeltaScale := DeltaScale  # 計算済みの DeltaScale を設定
    
    #         # アニメーションを適用
    #         AnimController.SetAnimation(KeyFrames, ?Mode := animation_mode.OneShot)
    #         AnimController.Play()
    
    
    
    
    

    AwaitAutoRestore(Info:mining_object_info)<suspends>:void=
        if(Time := AutoRestoreTime?):
            Sleep(Time)
            Info.RestoreHealth(MaxHealth.ToPersistableBigNumber())
            Rescale(Info)

    AwaitPickaxeDamage(Info:mining_object_info)<suspends>:void=
        loop:
            Agent := Info.PropManipulator.DamagedEvent.Await()

            #--------------------------NPC Signal--------------------------
            set SignalDamagedPlayer = option{Agent}
            NPCDamagedEvent.Signal(Info)
            if(NPCMiningDevice = false, Index := InfoArray.Find[Info]):
                set DamagedPropManipulator = Index
                set NPCMiningDevice = option{Index}
            #--------------------------------------------------------------

            Info.PropManipulator.RestoreHealth()
            if(Player := player[Agent]):

                var DamageInfo : damage_info = MiningDamageDevice.GetDamageAmount(Player)
                var DamageAmount : big_number = DamageInfo.Amount

                # Quest
                Print("GetResource : {DamageAmount.Value}")

                if(DamageAmount.Value >= 1.0):
                    QuestTrigger1.Trigger(Player)

                # Skillに応じてDamageを増やす
                SkillIndex := 5 # DamageBuffCheck
                if(Data := SaveDataManager.GetSkillStatus[Player, SkillIndex]):
                    Active := Data.IsActive
                    if(Active = true):
                        # スキル6のEffectPowerを参照してダメージ倍率を変更
                        if(SkillData := SaveDataManager.GetSkill[Player, 6]):
                            # EffectPower を float 型に変換
                            var EffectPower : float = SkillData.EffectPower*1.0
                            var EffectPowerAsFloat : float = EffectPower / 100.0  # 200% -> 2.0 に変換
                            GetDamageMultiply : big_number = big_number{Value := EffectPowerAsFloat, Exponent := 0}
    
                            # ダメージに倍率を適用
                            FinalDamage := BigNumber_MultiplyByBigNumber(DamageAmount, GetDamageMultiply)
                            Print("EffectPower : {EffectPowerAsFloat}")
    
                            # 新しいダメージ情報を設定
                            set DamageInfo = damage_info{InstigatorPlayer := DamageInfo.InstigatorPlayer, Amount := FinalDamage, DamageType := DamageInfo.DamageType, Multiplier := DamageInfo.Multiplier}
    
                Info.Damage(DamageInfo)
                
            #---------------------FieldSkill-------------------------------
            #spawn{FieldSkill.OnEvent(Info, Agent)}
            if :
                Player : player = player[Agent]
                1 = SaveDataManager.GetSkillLevel[Player , 19]
                Value := SaveDataManager.GetSkillLevel[Player , 20] + 1
                MoreHits_Level := SaveDataManager.GetSkillLevel[Player , 21]
                ComboMode_Level := SaveDataManager.GetSkillLevel[Player , 22]
            then :
                Count_Charge := FieldSkill.Get_CountCharge(Agent)
                ChargeIsActive := FieldSkill.Get_ChargeIsActive(Agent)
                CountHit := FieldSkill.Get_CountHit(Agent)
                Print("CCスキル有効, Count{Count_Charge + Value}", ?Color := NamedColors.Red)
                SkillDamage := MiningDamageDevice.GetDamageAmount(Player).Amount
                if(Count_Charge < 100) :
                    FieldSkill.Set_CountCharge(Agent, Count_Charge + Value)

                NewCountCharge := FieldSkill.Get_CountCharge(Agent)
                if(NewCountCharge >= 100) :
                    #連撃発動可能になった時の処理（エフェクト付与、ステート変更など）
                    if(ChargeIsActive = false) :
                        FieldSkill.Set_ChargeIsActive(Agent, true)
                        VFX_Powerup.SetDuration(-1.0)
                        VFX_Powerup.Pickup(Agent)
                        Print("チャージVFX{NewCountCharge}", ?Color := NamedColors.Red)
                    #連撃発動時の処理（エフェクト取り除き、ステート変更など）
                    else if(ChargeIsActive = true) :
                        if(CountHit <= ComboMode_Level) :
                            FieldSkill.Set_CountHit(Agent, CountHit + 1)
                            VFX_OnHit.Pickup(Agent)
                            SFX_OnHit.Play(Agent)
                            spawn{FieldSkill.Combo(MoreHits_Level, Info, Player)}
                        #連撃終了時に各変数とエフェクトを初期化
                        NewCountHit := FieldSkill.Get_CountHit(Agent)
                        if(NewCountHit > ComboMode_Level) :
                            VFX_Powerup.SetDuration(0.1)
                            VFX_Powerup.Pickup(Agent)
                            FieldSkill.Set_ChargeIsActive(Agent, false)
                            FieldSkill.Set_CountCharge(Agent, Value)
                            FieldSkill.Set_CountHit(Agent, 0)
                            Print("IsActive初期化", ?Color := NamedColors.Red)
                #連撃Widgetをアップデートする関数を呼び出し
                spawn{FieldSkill.CallWidgetTimer(Agent, Player)}
            #---------------------ダメージ時のエフェクト-------------------------------
            spawn{DamageVFX_Func(Info)}

    DamageVFX_Func(Info : mining_object_info)<suspends>:void=
        RockPos : vector3 = Info.PropManipulator.GetTransform().Translation
        var PropAsset : creative_prop_asset = VFX_OnDamge1
        Spawnable := GetRandomFloat(0.0, 100.0)
        if(Spawnable < VFXChance * 1.0) :
            WhichVFX := GetRandomFloat(0.0, 100.0)
            if(WhichVFX > 50.0) :
                set PropAsset = VFX_OnDamge1
                Print("VFX = 1", ?Color := NamedColors.Red)
            if(WhichVFX <= 50.0) :
                set PropAsset = VFX_OnDamge2
                Print("VFX = 2", ?Color := NamedColors.Red)
            SpawnedProp : ?creative_prop = SpawnProp(PropAsset, RockPos, rotation{})(0)
            if(TheProp : creative_prop = SpawnedProp?) :
                Sleep(3.0)
                TheProp.Dispose()
                Print("VFX Spawn!", ?Color := NamedColors.Red)
        else :
            Print("VFX not spawn", ?Color := NamedColors.Red)

            #---------------------ここまで-------------------------------

    AwaitDestroy(Info:mining_object_info)<suspends>:void=
        var DamageAmountInfo : big_number = big_number{}
        var GiveResource : big_number = big_number{}
        loop:
            DestroyPlayer := Info.DestroyEvent.Await()

            #--------------------------NPC Signal--------------------------
            if(Index := InfoArray.Find[Info], Index = NPCMiningDevice?):
                set SignalDestroyedPlayer = option{DestroyPlayer}
                set DamagedPropManipulator = Index
                NPCDestroyedEvent.Signal(Info.PropManipulator)
                set NPCMiningDevice = false
            #--------------------------------------------------------------

            # 破壊時に
            DamageInfo := MiningDamageDevice.GetDamageAmount(DestroyPlayer)
            SaveDataManager.AddResourceStone(DestroyPlayer, DamageInfo.Amount)
            set DamageAmountInfo = DamageInfo.Amount


            #---------------------FinalHitBonus-----------------------------
            if(FinalHitBonus := 0.5 + SaveDataManager.GetSkillLevel[DestroyPlayer , 17] * 0.2) :
                set GiveResource  = BigNumber_MultiplyByFloat(DamageAmountInfo , FinalHitBonus)
                SaveDataManager.AddResourceStone(DestroyPlayer , GiveResource)
                # FinalHitBonusをDamageInfoのAmountに加算
                set DamageAmountInfo = BigNumber_Add(DamageAmountInfo, GiveResource)
            #---------------------------------------------------------------
            NewDamageInfo := damage_info:
                InstigatorPlayer := DamageInfo.InstigatorPlayer
                Amount := DamageAmountInfo
                DamageType := DamageInfo.DamageType
                Multiplier := DamageInfo.Multiplier

            MiningObjectWidgetManager.AddBrakeScroll(NewDamageInfo)

            # Quest関係
            # 岩を壊すクエストの時に実行
            NowQuest := SaveDataManager.GetNowQuest(DestroyPlayer)
            Print("NowQuest : {NowQuest}")
            if(NowQuest = 2):
                QuestTrigger2.Trigger(DestroyPlayer)
                Print("岩壊すクエスト進行用トリガー起動")

            if(TargetHPBar := HPBarDevice.CurrentHPBar?):
                HPBarDevice.Dispose(TargetHPBar.ID)  # IDを指定してDisposeを呼び出す
            VFXDevice.Spawn(Info.TargetProp.GetTransform())
            spawn{AwaitAutoRestore(Info)}

            DestroiedRock_Pos : vector3 = Info.PropManipulator.GetTransform().Translation
            if(Agent : agent = agent[DestroyPlayer]) :
                XP_OnBreak.Award(Agent)
                #---------------------SummonStone-------------------------------
                if(SummonStoneDrop := SaveDataManager.GetSkillLevel[DestroyPlayer , 10]) :
                    Getable := GetRandomFloat(0.0 , 100.0)
                    if :
                        SummonStoneDrop = 1
                        Getable <= 100.0
                    then :
                        var GetSummonIndex : int = -1
                        IsDrop := GetRandomFloat(0.0, 100.0)
                        RandomReward := GetRandomFloat(0.0, 100.0)
                        SummonReward := PlayerData.big_number{Value := 1.0}
                        if(IsDrop < 10.0) :
                            XP_GotSumSto.Award(Agent)
                            if(RandomReward < 76.5):
                                SaveDataManager.AddResourceSummonStoneCommon(DestroyPlayer,SummonReward)
                                set GetSummonIndex = 0
                                Print("Reward is SummonStoneCommon!!")
                            if(76.5 <= RandomReward < 94.5):
                                SaveDataManager.AddResourceSummonStoneUncommon(DestroyPlayer,SummonReward)
                                set GetSummonIndex = 1
                                Print("Reward is SummonStoneUncommon!!")
                            if(94.5 <= RandomReward < 98.8):
                                SaveDataManager.AddResourceSummonStoneRare(DestroyPlayer,SummonReward)
                                set GetSummonIndex = 2
                                Print("Reward is SummonStoneRare!!")
                            if(98.8 <= RandomReward < 99.8):
                                SaveDataManager.AddResourceSummonStoneEpic(DestroyPlayer,SummonReward)
                                set GetSummonIndex = 3
                                Print("Reward is SummonStoneEpic!!")
                            if(99.8 <= RandomReward <= 100.0):
                                SaveDataManager.AddResourceSummonStoneLegendary(DestroyPlayer,SummonReward)
                                set GetSummonIndex = 4
                                Print("Reward is SummonStoneLegendary!!")
        
                        if(GetSummonIndex = 0):
                            T_SummonStone_R1.Trigger(Agent)
                            Print("Trigger")
                        if(GetSummonIndex = 1):
                            T_SummonStone_R2.Trigger(Agent)
                            Print("Trigger")
                        if(GetSummonIndex = 2):
                            T_SummonStone_R3.Trigger(Agent)
                            Print("Trigger")
                        if(GetSummonIndex = 3):
                            T_SummonStone_R4.Trigger(Agent)
                            Print("Trigger")
                        if(GetSummonIndex = 4):
                            T_SummonStone_R5.Trigger(Agent)
                            Print("Trigger")
                    else :
                        Print("SummonLevel not 1")
        
                else :
                    Print("Summonスキル取得失敗")
                #---------------------FieldSkill-------------------------------
                ExplodeValue := FieldSkill.Get_IsExplode(Agent)
                if :
                    ExplodeValue = false
                    1 = SaveDataManager.GetSkillLevel[DestroyPlayer , 23]
                    AE_Level := SaveDataManager.GetSkillLevel[DestroyPlayer , 24]
                    DM_Level := SaveDataManager.GetSkillLevel[DestroyPlayer , 25]
                then :
                    #一度爆発スキル起動したら2秒間起動しないように（爆発先でさらに岩が壊れた場合再度スキル起動するのを防ぐ）
                    spawn{FieldSkill.Set_IsExplode(Agent)}
                    ResultBigNumber := BigNumber_MultiplyByInt(MiningDamageDevice.GetDamageAmount(DestroyPlayer).Amount, 30 + DM_Level * 3)
                    BaseDistance : float = 1000.0
                    MaxDistance : float = BaseDistance + (AE_Level * 100.0)
                    for (NearbyInfo : InfoArray):
                        NearbyPropPos : vector3 = NearbyInfo.PropManipulator.GetTransform().Translation
                
                        DiffX : float = DestroiedRock_Pos.X - NearbyPropPos.X
                        DiffY : float = DestroiedRock_Pos.Y - NearbyPropPos.Y
                        DiffZ : float = DestroiedRock_Pos.Z - NearbyPropPos.Z
                
                        CalculatedDistance : float = (DiffX * DiffX) + (DiffY * DiffY) + (DiffZ * DiffZ)
                
                        if (CalculatedDistance <= MaxDistance * MaxDistance):
                            Print("Nearby prop found within distance: {CalculatedDistance}")
                            # 範囲内のオブジェクトに対しての処理
                            VFX_Area.Pickup(Agent)
                            NearbyInfo.Damage(DestroyPlayer , ResultBigNumber)
                            SFX_Area.Play(Agent)
                #--------------------------------------------------------------
                SFX_OnBroken.Play(Agent)
                Print("Finished OnDestroy")

            case(TransmitMode):
                transmit_mode.AllPlayers=>
                    for(Player : Info.DamagedPlayers):
                        if(EventChannel := Info.SingleDestroyEvent?):
                            EventChannel.Transmit(option{Player})
                        GlobalDestroyEvent.Transmit(option{Player})
                        GlobalDestroyEventTrigger.Trigger(Player)
    
                transmit_mode.LastPlayer=>
                    if(EventChannel := Info.SingleDestroyEvent?):
                        EventChannel.Transmit(option{DestroyPlayer})
                    GlobalDestroyEvent.Transmit(option{DestroyPlayer})
                    GlobalDestroyEventTrigger.Trigger(DestroyPlayer)


    AwaitDamage(Info:mining_object_info)<suspends>:void=
        loop:
            DamageInfo := Info.DamageEvent.Await()
            if(SaveDataManager.HasSaveData(DamageInfo.InstigatorPlayer)?):
                SaveDataManager.AddResourceStone(DamageInfo.InstigatorPlayer, DamageInfo.Amount)

                if (Health := Info.NowHealth?, Percent := BigNumber_Divide(Health, MaxHealth.ToPersistableBigNumber()).Value / 10.0):
                    if (Current := HPBarDevice.CurrentHPBar?):
                        Current.HPBarCanvas.SetVisibility(widget_visibility.Visible)
                        HPBarDevice.SetHPBarPercent(Current.ID, Percent)  #更新
                        HPBarDevice.SetHPText(Current.ID, Health, MaxHealth.ToPersistableBigNumber())
                        if (Player := player[DamageInfo.InstigatorPlayer]):
                            if:
                                set HPBarDevice.CountdownMap[Player] = 10
                    else:
                        # 新しい HP バーを作成して進行状況を設定
                        if (PlayerUI := GetPlayerUI[DamageInfo.InstigatorPlayer]):
                            HPBarTransform := transform{}
                            if (NewHPBar := HPBarDevice.CreateHPBar[Info, HPBarTransform, PlayerUI]):
                                PlayerUI.AddWidget(NewHPBar.HPBarCanvas)
                                HPBarDevice.SetHPBarPercent(NewHPBar.ID, Percent)  #更新
                                HPBarDevice.SetHPText(NewHPBar.ID, Health, MaxHealth.ToPersistableBigNumber())
                                set HPBarDevice.CurrentHPBar = option{NewHPBar}
                                if (Player := player[DamageInfo.InstigatorPlayer]):
                                    spawn{StartCountdownForPlayer(Player)}
                    
                    

                    set HPBarDevice.CurrentTarget = option{Info}

                    # Rescale(Info)
                    
                MiningObjectWidgetManager.AddBasicScroll(DamageInfo)
            else:
                Print("Error: AwaitDamage")


    StartCountdownForPlayer(Player:player)<suspends>:void=
        if:
            set HPBarDevice.CountdownMap[Player] = 10
        loop:
            if (Count := HPBarDevice.CountdownMap[Player], Count <= 0):
                if (TargetHPBar := HPBarDevice.CurrentHPBar?):
                    TargetHPBar.RemoveFromUI()
                    set HPBarDevice.CurrentHPBar = false
                if:
                    set HPBarDevice.CountdownMap[Player] = -1
                break
    
            if (Count := HPBarDevice.CountdownMap[Player], Count > 0):
                if:
                    set HPBarDevice.CountdownMap[Player] = Count - 1
            Sleep(1.0)
