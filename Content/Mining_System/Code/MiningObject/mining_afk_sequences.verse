using { /Fortnite.com/Devices }
using { /Verse.org/Simulation }
using { /Fortnite.com/Playspaces }
using { /Fortnite.com/Characters }
using { /UnrealEngine.com/Temporary/SpatialMath }
using { /UnrealEngine.com/Temporary/Diagnostics }

using { Pet }
using { PlayerData }
using { MiningObject }

PETCLASS_Sequence := class<concrete>():

    @editable
    PETNAME : string = "NAME"

    @editable
    var MiningCharacters : []creative_prop = array{}

mining_afk_sequences := class(creative_device):

    @editable
    PLayerAFKVolume : player_counter_device = player_counter_device{}

    @editable
    MiningCharacters : []PETCLASS_Sequence = array{}

    @editable
    MiningArea : []creative_prop = array{}

    @editable
    MiningPosition1 : []creative_prop = array{}

    @editable
    MiningPosition2 : []creative_prop = array{}

    @editable
    MiningPosition3 : []creative_prop = array{}

    @editable
    MiningPosition4 : []creative_prop = array{}

    @editable
    MiningPosition5 : []creative_prop = array{}

    @editable
    MiningPosition6 : []creative_prop = array{}

    @editable
    AFKTrigger : trigger_device = trigger_device{}

    var MiningIndexs : []int = array{0, 1, 2, 3, 4, 5}

    @editable
    PetManagerDevice : pet_manager = pet_manager{}

    @editable
    MiningDamageDevice : mining_damage_device = mining_damage_device{}

    @editable
    SaveDataManager : save_data_manager = save_data_manager{}

    OnBegin<override>()<suspends>:void=
        
        for(Player : GetPlayspace().GetPlayers()):
            spawn{PlayMiningSequence(Player)}

        GetPlayspace().PlayerAddedEvent().Subscribe(AddEvent)
        # GetPlayspace().PlayerRemovedEvent().Subscribe(OnPlayerRemoved)

        Sleep(5.0)
        for(Player : GetPlayspace().GetPlayers()):
            AFKTrigger.Trigger(Player)

    RaceMining(Player:player)<suspends>:void=
        if(SaveDataManager.HasSaveData(Player)?):
            race:
                MiningResource(Player)

                block:
                    loop:
                        TriggeredAgent := AFKTrigger.TriggeredEvent.Await()
                        if(player[TriggeredAgent?] = Player):
                            break
        else:
            Print("Error: RaceMining.")
        
    MiningResource(Player:player)<suspends>:void=
        loop:
            Sleep(2.0)

            if(SaveDataManager.HasSaveData(Player)?):
                Resource := SaveDataManager.GetResourceStone(Player)

                # ダメージ計算
                DamageAmount := MiningDamageDevice.GetNormalDamageAmount(Player).Amount
                GatherRate := BigNumber_MultiplyByFloat(PetManagerDevice.GetPetGatherRateFromRestSlots(Player), 0.01) # 100分率を実際の式に書き直してる
                PlayerIdle := SaveDataManager.GetPlayeridle(Player) * 0.05 + 0.1
                PetIdle := SaveDataManager.GetPetidle(Player) * 0.05 + 0.1

                ResultPetDamageAmount := BigNumber_MultiplyByFloat(BigNumber_MultiplyByBigNumber(DamageAmount, GatherRate), PetIdle)
                ResultPlayerDamageAmount := BigNumber_MultiplyByFloat(DamageAmount, PlayerIdle)

                SaveDataManager.AddResourceStone(Player, ResultPetDamageAmount)
                # Print("ResultPlayerDamageAmount: {(BigNumber_FormatDisplay(ResultPlayerDamageAmount))}")
                # Print("ResultPetDamageAmount: {(BigNumber_FormatDisplay(ResultPetDamageAmount))}")
                if(PLayerAFKVolume.IsCounted[Player]):
                    Print("Player is in AFK Zone!!")
                    SaveDataManager.AddResourceStone(Player, ResultPlayerDamageAmount)

                    # Print("Now Resource : {Resource.Value * Resource.Exponent}, PlayerIdle : {PlayerIdle}, PetIdle : {PetIdle}", ?Duration := 5.0)
            else:
                Print("Error: MiningResource")

    AddEvent(Player:agent):void=
        spawn{PlayMiningSequence(Player)}

    # 整数配列をソートする関数
    SortIntegers(Numbers : []int) : []int =
        var SortedNumbers : []int = Numbers
        for (i := 0..SortedNumbers.Length - 2):
            for (j := 0..SortedNumbers.Length - i - 2):
                if (SortedNumbers[j] > SortedNumbers[j + 1]):
                    if:
                        Temp := SortedNumbers[j]
                        V := SortedNumbers[j + 1]
                        set SortedNumbers[j] = V
                        set SortedNumbers[j + 1] = Temp
        return SortedNumbers

    ReturnPetIndex_FromInt(PETIndex:int, Upgrade:int):int=
        var PETIndexInt : int = 0

        # Kureppa
        if(PETIndex = 1):
            if(Upgrade = 0):
                set PETIndexInt = 0
            else:
                set PETIndexInt = 1

        # Kogasa
        if(PETIndex = 2):
            if(Upgrade = 0):
                set PETIndexInt = 2
            else:
                set PETIndexInt = 3

        # Kachinta
        if(PETIndex = 3):
            if(Upgrade = 0):
                set PETIndexInt = 4
            else:
                set PETIndexInt = 5

        # Ittanmomen
        if(PETIndex = 4):
            if(Upgrade = 0):
                set PETIndexInt = 6
            else:
                set PETIndexInt = 7

        # Narigama
        if(PETIndex = 5):
            if(Upgrade = 0):
                set PETIndexInt = 8
            else:
                set PETIndexInt = 9

        # DCKopper
        if(PETIndex = 11):
            if(Upgrade = 0):
                set PETIndexInt = 20
            else:
                set PETIndexInt = 21

        # DCKogasa
        if(PETIndex = 12):
            if(Upgrade = 0):
                set PETIndexInt = 22
            else:
                set PETIndexInt = 23

        # DCKachinta
        if(PETIndex = 13):
            if(Upgrade = 0):
                set PETIndexInt = 24
            else:
                set PETIndexInt = 25

        # DCIttanmomen
        if(PETIndex = 14):
            if(Upgrade = 0):
                set PETIndexInt = 26
            else:
                set PETIndexInt = 27

        # DCNarigama
        if(PETIndex = 15):
            if(Upgrade = 0):
                set PETIndexInt = 28
            else:
                set PETIndexInt = 29

        return PETIndexInt

    PlayMiningSequence(Player:agent)<suspends>:void=
        var NPCs : []int = array{}
        var TempPETCLASS : []PETCLASS_Sequence = array{}

        # ============----------------------------------------------------------------============
        # PETのスロットを埋める(テストプレイ用)

        #if(ThePlayer := player[Player]):

            # Kopper
            #PetManagerDevice.UnlockPet(ThePlayer, 1)
            #PetManagerDevice.SetPetSlot(ThePlayer, 1, 11)

            # Kachinta
            #PetManagerDevice.UnlockPet(ThePlayer, 2)
            #PetManagerDevice.SetPetSlot(ThePlayer, 2, 11)

            # Ittanmomen
            #PetManagerDevice.UnlockPet(ThePlayer, 3)
            #PetManagerDevice.SetPetSlot(ThePlayer, 3, 11)

            # Kogasa
            #PetManagerDevice.UnlockPet(ThePlayer, 4)
            #PetManagerDevice.SetPetSlot(ThePlayer, 4, 11)

            # Narigama
            #PetManagerDevice.UnlockPet(ThePlayer, 5)
            #PetManagerDevice.SetPetSlot(ThePlayer, 5, 11)

        # ============----------------------------------------------------------------============

        # PETのスロットからペットを取得
        if(PetStatus := PetManagerDevice.GetPetsInSlots11_16[player[Player]]):

            # スロット11のペットを取得
            if(ThePet := PetStatus[0]):
                PETIndex := ThePet(0)
                PETUpgrade := ThePet(1)
                GetIndex := ReturnPetIndex_FromInt(PETIndex, PETUpgrade)
                set NPCs += array{GetIndex}
                #Print("PETINDEX IS {GetIndex}", ?Duration:= 5.0)

            # スロット12のペットを取得
            if(ThePet := PetStatus[1]):
                PETIndex := ThePet(0)
                PETUpgrade := ThePet(1)
                GetIndex := ReturnPetIndex_FromInt(PETIndex, PETUpgrade)
                set NPCs += array{GetIndex}
                #Print("PETINDEX IS {GetIndex}", ?Duration:= 5.0)

            # スロット13のペットを取得
            if(ThePet := PetStatus[2]):
                PETIndex := ThePet(0)
                PETUpgrade := ThePet(1)
                GetIndex := ReturnPetIndex_FromInt(PETIndex, PETUpgrade)
                set NPCs += array{GetIndex}
                #Print("PETINDEX IS {GetIndex}", ?Duration:= 5.0)

            # スロット14のペットを取得
            if(ThePet := PetStatus[3]):
                PETIndex := ThePet(0)
                PETUpgrade := ThePet(1)
                GetIndex := ReturnPetIndex_FromInt(PETIndex, PETUpgrade)
                set NPCs += array{GetIndex}
                #Print("PETINDEX IS {GetIndex}", ?Duration:= 5.0)

            # スロット15のペットを取得
            if(ThePet := PetStatus[4]):
                PETIndex := ThePet(0)
                PETUpgrade := ThePet(1)
                GetIndex := ReturnPetIndex_FromInt(PETIndex, PETUpgrade)
                set NPCs += array{GetIndex}
                #Print("PETINDEX IS {GetIndex}", ?Duration:= 5.0)

            # スロット16のペットを取得
            if(ThePet := PetStatus[5]):
                PETIndex := ThePet(0)
                PETUpgrade := ThePet(1)
                GetIndex := ReturnPetIndex_FromInt(PETIndex, PETUpgrade)
                set NPCs += array{GetIndex}
                #Print("PETINDEX IS {GetIndex}", ?Duration:= 5.0)

        # ------------------------------------------------------

        loop:
            EnteredAgent := AFKTrigger.TriggeredEvent.Await()
            if(EnteredAgent? = Player, PlayerP := player[Player]):
                spawn{RaceMining(PlayerP)}

                # Mining Characterの体数だけ実行
                for(Index -> MiningCharacter : MiningCharacters):
                    for(FollowNPC : NPCs, Index = FollowNPC):
                        set TempPETCLASS += array{MiningCharacter}
                        Print("{Index}", ?Duration := 5.0)

                # 特定ひとつのBPだけ使う
                var TempMiningIndexs : []int = array{}
                var MiningIndex : int = 0

                if(Mining := MiningIndexs[0]):
                    set MiningIndex = Mining

                for(i := 1..MiningIndexs.Length - 1):
                    if(Element := MiningIndexs[i]):
                        set TempMiningIndexs += array{Element}
                set MiningIndexs = TempMiningIndexs

                var IntTransform1 : transform = transform{}
                var IntTransform2 : transform = transform{}
                var IntTransform3 : transform = transform{}
                var IntTransform4 : transform = transform{}
                var IntTransform5 : transform = transform{}
                var IntTransform6 : transform = transform{}

                for(Index -> MiningCharacter : TempPETCLASS):
                    if:
                        NPCBP := MiningCharacter.MiningCharacters[MiningIndex]
                    then:
                        var TeleportLocation : vector3 = vector3{}
                        var TeleportRotation : rotation = rotation{}

                        if(Index = 0, MiningBP := MiningPosition1[MiningIndex]):
                            set IntTransform1 = NPCBP.GetTransform()
                            set TeleportLocation = MiningBP.GetTransform().Translation
                            set TeleportRotation = MiningBP.GetTransform().Rotation

                        if(Index = 1, MiningBP := MiningPosition2[MiningIndex]):
                            set IntTransform2 = NPCBP.GetTransform()
                            set TeleportLocation = MiningBP.GetTransform().Translation
                            set TeleportRotation = MiningBP.GetTransform().Rotation

                        if(Index = 2, MiningBP := MiningPosition3[MiningIndex]):
                            set IntTransform3 = NPCBP.GetTransform()
                            set TeleportLocation = MiningBP.GetTransform().Translation
                            set TeleportRotation = MiningBP.GetTransform().Rotation

                        if(Index = 3, MiningBP := MiningPosition4[MiningIndex]):
                            set IntTransform4 = NPCBP.GetTransform()
                            set TeleportLocation = MiningBP.GetTransform().Translation
                            set TeleportRotation = MiningBP.GetTransform().Rotation

                        if(Index = 4, MiningBP := MiningPosition5[MiningIndex]):
                            set IntTransform5 = NPCBP.GetTransform()
                            set TeleportLocation = MiningBP.GetTransform().Translation
                            set TeleportRotation = MiningBP.GetTransform().Rotation

                        if(Index = 5, MiningBP := MiningPosition6[MiningIndex]):
                            set IntTransform6 = NPCBP.GetTransform()
                            set TeleportLocation = MiningBP.GetTransform().Translation
                            set TeleportRotation = MiningBP.GetTransform().Rotation
                        
                        if(NPCBP.TeleportTo[TeleportLocation,TeleportRotation]):
                        Print("Teleport {TeleportLocation}")

                loop:
                    ExitedAgent := AFKTrigger.TriggeredEvent.Await()
                    if(ExitedAgent? = Player):
                        break

                # 元の配列に戻す
                set MiningIndexs += array{MiningIndex}
                set MiningIndexs = SortIntegers(MiningIndexs)

                for(Index -> NPC : NPCs,MiningCharacter : TempPETCLASS):
                    if:
                        NPCBP := MiningCharacter.MiningCharacters[MiningIndex]
                        if(Index = 0, MiningBP := MiningPosition1[MiningIndex]):
                            NPCBP.TeleportTo[IntTransform1]
                        if(Index = 1, MiningBP := MiningPosition2[MiningIndex]):
                            NPCBP.TeleportTo[IntTransform2]
                        if(Index = 2, MiningBP := MiningPosition3[MiningIndex]):
                            NPCBP.TeleportTo[IntTransform3]
                        if(Index = 3, MiningBP := MiningPosition4[MiningIndex]):
                            NPCBP.TeleportTo[IntTransform4]
                        if(Index = 4, MiningBP := MiningPosition5[MiningIndex]):
                            NPCBP.TeleportTo[IntTransform5]
                        if(Index = 5, MiningBP := MiningPosition6[MiningIndex]):
                            NPCBP.TeleportTo[IntTransform6]
                
                spawn{PlayMiningSequence(Player)}
                break
        