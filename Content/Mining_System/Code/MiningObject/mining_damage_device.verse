using { /Fortnite.com/Devices }
using { /Verse.org/Simulation }
using { /Verse.org/Random }
using { /Verse.org }

using { PlayerData }
using { Utility }

using { Pet }

# UI等への情報を渡す用, 中身は好きに増やして大丈夫です。
damage_info<public> := class:
    InstigatorPlayer<public> : player
    Amount<public> : big_number = big_number{}
    DamageType<public> : damage_type = damage_type.Normal
    Multiplier : int = 1

damage_type<public> := enum:
    Normal
    Critical

mining_damage_device<public> := class(creative_device):

    @editable
    SaveDataManager : save_data_manager = save_data_manager{}

    @editable
    PetManagerDevice : pet_manager = pet_manager{}

    @editable
    BaseSTR : float = 1.65

    @editable
    IncreaseRatioSTR : float = 1.28

    StageGatherRates : [int]float = map:
        0 => 0.01
        1 => 0.05
        2 => 0.10
        3 => 0.15
        4 => 0.20
        5 => 0.25
        6 => 0.30
        7 => 0.40
        8 => 0.50
        9 => 0.60
        10 => 0.7
        11 => 0.8

    GetNormalDamageAmount<public>(Target:player):damage_info=
        Print("GetNormalDamageAmount")
        Level := SaveDataManager.GetLevel(Target)
        AddNumber := big_number{Value := (Level-1) * 1.0, Exponent := 0}
        
        AddA := BigNumber_Add(GetBaseStrength(SaveDataManager.GetLevel(Target)), BigNumber_MultiplyByBigNumber(GetBaseStrength(SaveDataManager.GetLevel(Target)),BigNumber_MultiplyByFloat(PetManagerDevice.GetPetStrengthFromEquipSlots(Target),0.01)))
        AddB := BigNumber_Add(BigNumber_MultiplyByBigNumber(AddA, GetStagePercent(Target)), BigNumber_MultiplyByBigNumber(AddA, GetTotalLevelBonus(Target)))
        AddC := BigNumber_Add(AddA,AddB)

        damage_info:
            InstigatorPlayer := Target
            Amount := AddC
            DamageType := damage_type.Normal
            Multiplier := 1

    GetNormalDamage<public>(Target:player):big_number=
        Print("GetNormalDamageAmount")
        Level := SaveDataManager.GetLevel(Target)
        AddNumber := big_number{Value := (Level-1) * 1.0, Exponent := 0}
        
        AddA := BigNumber_Add(GetBaseStrength(SaveDataManager.GetLevel(Target)), BigNumber_MultiplyByBigNumber(GetBaseStrength(SaveDataManager.GetLevel(Target)),BigNumber_MultiplyByFloat(PetManagerDevice.GetPetStrengthFromEquipSlots(Target),0.01)))
        AddB := BigNumber_Add(BigNumber_MultiplyByBigNumber(AddA, GetStagePercent(Target)), BigNumber_MultiplyByBigNumber(AddA, GetTotalLevelBonus(Target)))
        AddC := BigNumber_Add(AddA,AddB)

        return AddC

    # ここに書き足していってAmountを最終のものにしてください。critical_infoのクラスの中身は好きに増やして大丈夫です。using { /Skill }
    GetCriticalDamageAmount<public>(Target:player):damage_info=
        Print("GetCriticalDamageAmount")
    
        var DefaultCriticalMultiplier : float = 1.5
        var RandomMultiplier : int = 1  # 初期値は1
        Level := SaveDataManager.GetLevel(Target)
        AddNumber := big_number{Value := (Level-1) * 1.0, Exponent := 0}
    
        if (CriticalMultiplier := SaveDataManager.GetSkillLevel[Target, 16] * 0.1):
            set DefaultCriticalMultiplier += CriticalMultiplier
    
        if:
            EffectPower : int = (SaveDataManager.GetSkillLevel[Target, 18]) + 1
        then:
            if (EffectPower > 1):
                set RandomMultiplier = Random.GetRandomInt(1, EffectPower)

        AddCri := big_number{Value := DefaultCriticalMultiplier, Exponent := 0}

        AddA := BigNumber_Add(GetBaseStrength(SaveDataManager.GetLevel(Target)), BigNumber_MultiplyByBigNumber(GetBaseStrength(SaveDataManager.GetLevel(Target)),BigNumber_MultiplyByFloat(PetManagerDevice.GetPetStrengthFromEquipSlots(Target),0.01)))
        AddB := BigNumber_Add(BigNumber_MultiplyByBigNumber(AddA, GetStagePercent(Target)), BigNumber_MultiplyByBigNumber(AddA, GetTotalLevelBonus(Target)))
        AddC := BigNumber_Add(AddA,AddB)
        AddD := BigNumber_MultiplyByBigNumber(AddC, BigNumber_MultiplyByFloat(BigNumber_Add(AddCri, GetPetCriticalDamage(Target)), RandomMultiplier*1.0))
                
        return damage_info:
            InstigatorPlayer := Target
            Amount := AddD
            DamageType := damage_type.Critical
            Multiplier := RandomMultiplier

    GetDamageAmount<public>(Target:player):damage_info=
        if(IsCritical(Target)?):
            Print("Critical")
            GetCriticalDamageAmount(Target)
        else:
            Print("Normal")
            GetNormalDamageAmount(Target)

    # クリティカルかどうかの判定を書く
    IsCritical<public>(Target:player)<transacts>:logic=
        if :
            CriticalRate := SaveDataManager.GetSkillLevel[Target , 15] * 5.0 + BigNumber_ToFloat(GetPetCriticalRatio(Target))
            GetRandomFloat(0.0, 100.0) < CriticalRate
        then :
            return true
        else:
            return false

    # 割合で返す
    GetStagePercent<public>(Target:player):big_number=
        UnlockedStages := SaveDataManager.GetUnlockedStages(Target)
        var Result : float = 0.0
        for(StageIndex : UnlockedStages, Rate := StageGatherRates[StageIndex]):
            set Result += Rate
        return big_number{Value := Result, Exponent := 0}
            
    GetBaseStrength<public>(Level : int):big_number=
        # Print("Level: {Level}")
    
        # BaseSTRをbig_numberで初期化
        BaseStr := big_number{Value := BaseSTR, Exponent := 0}
    
        # # 指定された値よりLevelが大きい場合は処理を行う
        # LimitLevel := 173  # 仮設定
    
        # if (Level > LimitLevel):
        #     Print("Level is too high, adjusting...")
    
        #     AdjustedLevel := (Level * 1.0) / 2.0
    
        #     StrValue := BaseStr.Value * Pow(IncreaseRatioSTR, (AdjustedLevel - 1.0))
        #     StrExponent := BaseStr.Exponent
    
        #     # 正規化した後、値を2倍にして調整
        #     AdjustedStr := BigNumber_Normalize(big_number{Value := StrValue, Exponent := StrExponent})
            
        #     logic {AdjustedStr.Value = AdjustedStr.Value * 2}
        #     Print("Adjusted Display: {BigNumber_FormatDisplay(AdjustedStr)}")
    
        #     return AdjustedStr
        # else:
        Str := BigNumber_MultiplyByFloat(BaseStr , Pow(IncreaseRatioSTR, (Level - 1) * 1.0))

        # Print("Display: {BigNumber_FormatDisplay(big_number{Value := Str.Value, Exponent := Str.Exponent})}")

        return BigNumber_Normalize(Str)
    
    
    
    # 百分率で返す
    GetPetStrength<public>(Target:player):big_number=

        TotalStrength := PetManagerDevice.GetPetStrengthFromEquipSlots(Target)

        return TotalStrength

    GetPetTotalStrength<public>(Target:player):big_number=
        var Result : float = 0.0
        
        TotalSTR := PetManagerDevice.GetPetTotalStrength(Target)

        return TotalSTR

    # 割合で返す
    GetPetTotalGatherRatio<public>(Target:player):big_number=
        
        TotalGather := BigNumber_MultiplyByFloat(PetManagerDevice.GetPetGatherRateFromEquipSlots(Target), 0.01)

        return TotalGather

    # 割合で返す
    GetPetStrengthRatio<public>(Target:player):big_number=

        TotalStrength := BigNumber_MultiplyByFloat(PetManagerDevice.GetPetStrengthFromEquipSlots(Target), 0.01)

        return TotalStrength
    # 割合で返す
    GetPetCriticalRatio<public>(Target:player)<transacts>:big_number=

        TotalCriticaRatio := BigNumber_MultiplyByFloat(PetManagerDevice.GetPetCriticalRateFromEquipSlots(Target), 0.01)

        return TotalCriticaRatio
    # 割合で返す
    GetPetCriticalDamage<public>(Target:player):big_number=

        TotalCriticalDamage := BigNumber_MultiplyByFloat(PetManagerDevice.GetPetCriticalDamageFromEquipSlots(Target), 0.01)

        return TotalCriticalDamage

    GetTotalLevelBonus<public>(Target:player):big_number=

        TotalLevelBonus := BigNumber_MultiplyByFloat(PetManagerDevice.GetTotalLevelBonus(Target), 0.01)

        return TotalLevelBonus