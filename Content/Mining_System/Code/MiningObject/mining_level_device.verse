using { /Verse.org/Simulation }
using { /Fortnite.com/Devices }

using { PlayerData }
using { Utility }

mining_level_device<public> := class(creative_device):

    BaseResource : float = 10.0
    IncreaseRatioResource : float = 1.28

    @editable
    SaveDataManager : save_data_manager = save_data_manager{}

    TryPlayerLevel<public>(Player:player):void=
        if(CanLevelUp[Player]):
            RequiredResource := BigNumber_MultiplyByFloat(big_number{Value := BaseResource, Exponent := 0}, Pow(IncreaseRatioResource, (SaveDataManager.GetLevel(Player)-1)*1.0))
            SaveDataManager.SubtractResourceStone(Player , RequiredResource)
            SaveDataManager.AddLevel(Player, 1)

    LoopTryPlayerLevel<public>(Player:player):void=
        if(CanLevelUp[Player]):
            SaveDataManager.AddLevel(Player, 1)
            LoopTryPlayerLevel(Player)

    CanLevelUp<public>(Player:player)<decides><transacts>:void=
        ResourceCount := SaveDataManager.GetResourceStone(Player)
        RequiredResource := BigNumber_MultiplyByFloat(big_number{Value := BaseResource, Exponent := 0}, Pow(IncreaseRatioResource, (SaveDataManager.GetLevel(Player)-1)*1.0))
        BigNumber_GreaterThanOrEquals(ResourceCount, RequiredResource)?