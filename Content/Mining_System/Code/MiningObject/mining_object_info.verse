using { /Verse.org/Simulation }
using { /Fortnite.com/Devices }
using { /UnrealEngine.com/Temporary/SpatialMath }
using { /UnrealEngine.com/Temporary }

using { Utility }
using { PlayerData }

mining_object_info<public> := class<concrete><unique>:
    
    @editable
    TargetProp<public> : creative_prop = creative_prop{}

    @editable
    PropManipulator<public> : prop_manipulator_device = prop_manipulator_device{}

    @editable
    SingleDestroyEvent<public> : ?channel_device = false

    var DefaultTransform : transform = transform{}
    var NowHealth  : ?big_number = false
    var DamagedPlayers : []player = array{}

    DestroyEvent : event(player) = event(player){}
    DamageEvent : event(damage_info) = event(damage_info){}

    RestoreHealth(Value:big_number):void=
        set NowHealth = option{Value}
        set DamagedPlayers = array{}
        PropManipulator.ShowProps()

    Damage<public>(Player:player, Amount:big_number):void=
        Damage(
            damage_info{
                InstigatorPlayer := Player,
                Amount := Amount
            }
        )

    Damage<public>(DamageInfo:damage_info):void=
        if(Health := NowHealth?):
            AddDamagedPlayer(DamageInfo.InstigatorPlayer)
            NewHealth := BigNumber_Subtract(Health, DamageInfo.Amount)
            Print("NewHealth : {BigNumber_FormatDisplay(NewHealth)}")
            if(not BigNumber_GreaterThanOrEquals(NewHealth, big_number{Value := 0.0, Exponent := 1})?):
                Print("破壊された")
                set NowHealth = false
                PropManipulator.HideProps()
                DestroyEvent.Signal(DamageInfo.InstigatorPlayer)
            else:
                Print("破壊されてない")
                set NowHealth = option{NewHealth}
                DamageEvent.Signal(DamageInfo)
        else:
            Print("破壊されている")

    AddDamagedPlayer(Player:player):void=
        if(not DamagedPlayers.Find[Player]):
            set DamagedPlayers += array{Player}

transmit_mode<public> := enum:
    LastPlayer
    AllPlayers