using { /Verse.org/Simulation }
using { /UnrealEngine.com/Temporary/UI }
using { /Verse.org/Assets }
using { /Fortnite.com/UI }
using { /Verse.org/Colors }
using { /UnrealEngine.com/Temporary/SpatialMath }

using { Utility }

scroll_info := class:
    Icon : texture_block
    MainText : text_block
    SubText : text_block
    OffsetColor : color_block

mining_object_widget<public> := class:

    Player<public> : player

    var NowWidget : ?canvas = false

    ScrollSpeed : float = 10.0
    ScrollUpdateTime : float = 0.1
    ScrollLifetime : float = 5.0
    ScrollPos : vector2 = vector2{X := 0.5, Y := 0.5}

    AddGameplayUI():void=
        if(PlayerUI := GetPlayerUI[Player], not NowWidget?):
            NewCanvas := CreateCanvas()
            PlayerUI.AddWidget(NewCanvas)
            set NowWidget = option{NewCanvas}

    RemoveUI():void=
        if(PlayerUI := GetPlayerUI[Player], OldCanvas := NowWidget?):
            PlayerUI.RemoveWidget(OldCanvas)
            set NowWidget = false

    CreateCanvas():canvas=
        canvas:

    AddScrollObject(Icon:texture, MainText:string, SubText:string, MainColor:color, SubColor:color):void=
        SInfo := scroll_info:
            Icon := texture_block{DefaultImage := Icon}
            MainText := text_block{DefaultText := StringToMessage(MainText), DefaultTextColor := MainColor}
            SubText := text_block{DefaultText := StringToMessage(SubText), DefaultTextColor := SubColor}
            OffsetColor := color_block{DefaultOpacity := 0.0}

        StackBox := stack_box:
            Orientation := orientation.Horizontal
            Slots := array:
                stack_box_slot:
                    Widget := SInfo.Icon
                stack_box_slot:
                    Widget := SInfo.MainText
                stack_box_slot:
                    Widget := SInfo.SubText
                stack_box_slot:
                    Widget := SInfo.OffsetColor
        spawn:
            UpdateScrollObject(StackBox, SInfo)

    UpdateScrollObject(Object:stack_box, SInfo:scroll_info)<suspends>:void=
        var NowHeight : vector2 = vector2{}
        var PassedTime : float = 0.0

        if(NowCanvas := NowWidget?):
            NewCanvasSlot := canvas_slot:
                Anchors := anchors{Minimum := ScrollPos + NowHeight, Maximum := ScrollPos + NowHeight}
                Alignment := vector2{X := 0.5, Y := 1.0}
                SizeToContent := true
                Widget := Object

            NowCanvas.AddWidget(NewCanvasSlot)
            race:
                Sleep(ScrollLifetime)
                loop:
                    Sleep(ScrollUpdateTime)
                    set PassedTime += ScrollUpdateTime
                    PassedPercent := PassedTime / ScrollLifetime
                    Index := if(Value := Floor[PassedPercent * 10.0]) then Value else 0
                    OpacityArrays := array{1.0, 0.9, 0.8, 0.7, 0.6, 0.5, 0.4, 0.3, 0.2, 0.1, 0.0}

                    set NowHeight = vector2{Y := NowHeight.Y + ScrollSpeed}
                    SInfo.OffsetColor.SetDesiredSize(NowHeight)
                    if(Opacity := OpacityArrays[Index]):
                        SInfo.MainText.SetTextOpacity(Opacity)
                        SInfo.SubText.SetTextOpacity(Opacity)

            NowCanvas.RemoveWidget(Object)