using { /Fortnite.com/Devices }
using { /Verse.org/Simulation }
using { /Verse.org/Colors }

using { Profile }
using { UI.MiningObject.Textures }
using { Utility }
using { PlayerData }

mining_object_widget_manager<public> := class(creative_device):

    AddBasicScroll(DamageInfo:damage_info):void=
        if(WClass := GetGlobalData().MiningObjectWidget[DamageInfo.InstigatorPlayer]):
            case(DamageInfo.DamageType):
                damage_type.Normal =>
                    WClass.AddScrollObject(T_UI_MiningObject_Sample, "{BigNumber_FormatDisplay(DamageInfo.Amount)}", "", NamedColors.White, NamedColors.Yellow)
                damage_type.Critical =>
                    var TextColor : color = NamedColors.White  # 初期値として安全な値を設定

                    # Multiplierに基づいて色を変更
                    if (DamageInfo.Multiplier = 2):
                        set TextColor = NamedColors.Orange
                    else if (DamageInfo.Multiplier = 3):
                        set TextColor = NamedColors.OrangeRed
                    else if (DamageInfo.Multiplier = 4):
                        set TextColor = NamedColors.Red
                    else if (DamageInfo.Multiplier = 5):
                        set TextColor = NamedColors.MediumVioletRed
                    else if (DamageInfo.Multiplier = 6):
                        set TextColor = NamedColors.Violet

                    # 色を指定してUIに表示
                    if (DamageInfo.Multiplier = 1):
                        WClass.AddScrollObject(FlatSkill2_144, "{BigNumber_FormatDisplay(DamageInfo.Amount)}", "(Critical!)", NamedColors.Yellow, NamedColors.Yellow)
                    else:
                        WClass.AddScrollObject(FlatSkill2_144, "{BigNumber_FormatDisplay(DamageInfo.Amount)}", "(x{DamageInfo.Multiplier})", TextColor, TextColor)

    AddBrakeScroll(DamageInfo:damage_info):void=
        if(WClass := GetGlobalData().MiningObjectWidget[DamageInfo.InstigatorPlayer]):
            case(DamageInfo.DamageType):
                damage_type.Normal =>
                    WClass.AddScrollObject(IC_LastHit, "{BigNumber_FormatDisplay(DamageInfo.Amount)}", "", NamedColors.Cyan, NamedColors.Yellow)
                damage_type.Critical =>
                    var TextColor : color = NamedColors.White  # 初期値として安全な値を設定
                    var Text_Color : color = NamedColors.White

                    # Multiplierに基づいて色を変更
                    set Text_Color = NamedColors.Cyan
                    if (DamageInfo.Multiplier = 2):
                        set TextColor = NamedColors.Orange
                    else if (DamageInfo.Multiplier = 3):
                        set TextColor = NamedColors.OrangeRed
                    else if (DamageInfo.Multiplier = 4):
                        set TextColor = NamedColors.Red
                    else if (DamageInfo.Multiplier = 5):
                        set TextColor = NamedColors.MediumVioletRed
                    else if (DamageInfo.Multiplier = 6):
                        set TextColor = NamedColors.Violet

                    # 色を指定してUIに表示
                    if (DamageInfo.Multiplier = 1):
                        WClass.AddScrollObject(IC_LastHit, "{BigNumber_FormatDisplay(DamageInfo.Amount)}", "(Critical!)", Text_Color, NamedColors.Yellow)
                    else:
                        WClass.AddScrollObject(IC_LastHit, "{BigNumber_FormatDisplay(DamageInfo.Amount)}", "(x{DamageInfo.Multiplier})", Text_Color, TextColor)
            
    OnBegin<override>()<suspends>:void=
        for(Player : GetPlayspace().GetPlayers()):
            OnPlay(Player)

        GetPlayspace().PlayerAddedEvent().Subscribe(OnPlay)

    OnPlay(Player:player):void=
        if:
            not GetGlobalData().MiningObjectWidget[Player]
            set GetGlobalData().MiningObjectWidget[Player] = mining_object_widget{Player := Player}
        then{}

        if(WClass := GetGlobalData().MiningObjectWidget[Player]):
            WClass.AddGameplayUI()