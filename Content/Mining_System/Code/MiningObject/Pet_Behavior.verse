# < --==============================アップデート予定==============================-- >

# Mining時にprop_manipulatorの方向を見てないことがある
# Miningのアニメーションが流れた瞬間にTeleportでRotaitonを明確に決めてしまうようにする

# < --===========================================================================-- >

using { /Fortnite.com/AI }
using { /Verse.org/Assets }
using { /Fortnite.com/Devices }
using { /Verse.org/Simulation }
using { /Fortnite.com/Characters }
using { /Fortnite.com/Animation/PlayAnimation }
using { /UnrealEngine.com/Temporary/Diagnostics }
using { /UnrealEngine.com/Temporary/SpatialMath }
using { /Verse.org/Random }
using { /Verse.org/Colors }
using { Pet }
using { Profile }
using { PlayerData }

PETCLASS := class<concrete>():

    @editable
    PETNAME : string = "NAME"

    @editable
    NPC_Spawner : npc_spawner_device = npc_spawner_device{}

    @editable
    SpawnTrigger : trigger_device = trigger_device{}

    @editable
    DespawnTrigger : trigger_device = trigger_device{}

    @editable
    Mining : ?animation_sequence = false

    var SpawnCount : int = 0

Pet_Behavior := class(creative_device):
    @editable VFXChance : int = 30
    @editable VFX_OnDamge1 : creative_prop_asset = DefaultCreativePropAsset
    @editable VFX_OnDamge2 : creative_prop_asset = DefaultCreativePropAsset

    @editable
    PETLIST : []PETCLASS = array{}

    @editable
    TeleportDistance : float = 3000.0

    @editable
    RunDistance : float = 1000.0

    @editable
    ReachRadius : float = 150.0

    @editable
    StackTime : float = 5.0

    @editable
    WaitTime : float = 0.0

    @editable
    MaxIslandPlayers : int = 4

    @editable
    LockOnPlayer : logic = false

    @editable
    WaitingTrigger : trigger_device = trigger_device{}

    @editable
    FollowingTrigger : trigger_device = trigger_device{}

    @editable
    MiningVolume : volume_device = volume_device{}

    @editable
    MiningDevices : []mining_object_device = array{}

    @editable
    MiningDamage : mining_damage_device = mining_damage_device{}

    @editable
    PetManager : pet_manager = pet_manager{}

    @editable
    AFKTrigger : trigger_device = trigger_device{}

    @editable
    QuestTrigger : trigger_device = trigger_device{}

    NPCDamagedEvent : event(tuple(mining_object_info, mining_object_device)) = event(tuple(mining_object_info, mining_object_device)){}

    var EquippedPets : [player][]int = map{}

    @editable
    DebugButton : button_device = button_device{}

    OnBegin<override>()<suspends>:void=

        # PETに登録されているデバイスに対してコード実行
        for(Index -> PET : PETLIST):
            PETNAME := PET.PETNAME
            NPC_Spawner := PET.NPC_Spawner
            SpawnTrigger := PET.SpawnTrigger
            DespawnTrigger := PET.DespawnTrigger
            SpawnCount := PET.SpawnCount

            spawn{SpawnTrigger.SpawnTriggerEvent(NPC_Spawner, DespawnTrigger, Index)}
        # 自動でついてくるの無効にしたい場合はここコメントアウト
        spawn{AwaitPetUpdate()}
        spawn{AwaitingPetUpdate()}
        DebugButton.InteractedWithEvent.Subscribe(Debug)
        MiningDevicesRace()

    # ----------------------------------------------------------------------------------------------------
    # 書き足しポイント
    Debug(Agent:agent):void=
        if(Player := player[Agent]):

    AwaitingPetUpdate()<suspends>:void=
        Print("PETUPDATEAWAIT" ,?Duration := 20.0)
        loop:
            Signal := PetManager.PetUpgradeSignal.Await()
            Print("Recieve Signal!!")
            Player := Signal(0)
            PETID := Signal(1)
            StageIndex := Signal(2)
            ChibiBehaviorIndex := ReturnPetIndex_FromInt(Signal(1), 0)
            BigBehaviorIndex := ReturnPetIndex_FromInt(Signal(1), 1)
            Print("ChibiBehaviorIndex:{ChibiBehaviorIndex}", ?Duration := 10.0)
            Print("BigBehaviorIndex:{BigBehaviorIndex}", ?Duration := 10.0)
            if(ChibiPetTrigger := PETLIST[ChibiBehaviorIndex].DespawnTrigger, BigPetTrigger := PETLIST[BigBehaviorIndex].SpawnTrigger):
                PetManager.UpgradePet(Player, PETID, ChibiPetTrigger, BigPetTrigger, StageIndex)

    # ----------------------------------------------------------------------------------------------------

    AwaitPetUpdate<public>()<suspends>:void=
        loop:
            Player := GetGlobalData().YokaiSlotUpdateEvent.Await()
            QuestTrigger.Trigger(Player)
            if(not EquippedPets[Player]):
                for(PetElement : PetManager.GetEquipedPetStatus[Player], PetID := PetManager.GetPetID[Player, PetElement(0).Index]):
                    BehavirIndex := ReturnPetIndex_FromInt(PetID)
                    if(TargetPetList := PETLIST[BehavirIndex]):
                        TargetPetList.SpawnTrigger.Trigger(Player)
            else if(LastEquippedPets := EquippedPets[Player]):
                var UsedPets : []int =array{}
                for(PetElement : PetManager.GetEquipedPetStatus[Player], PetID := PetManager.GetPetID[Player, PetElement(0).Index]):
                    if(not LastEquippedPets.Find[PetElement(0).Index]):
                        BehavirIndex := ReturnPetIndex_FromInt(PetID)
                        if(TargetPetList := PETLIST[BehavirIndex]):
                            TargetPetList.SpawnTrigger.Trigger(Player)
                            set UsedPets += array{PetElement(0).Index}
                            Print("Added: {PetElement(0).Index}")
                    else:
                        set UsedPets += array{PetElement(0).Index}
                for(PetElement : LastEquippedPets, not UsedPets.Find[PetElement], PetID := PetManager.GetPetID[Player, PetElement]):
                    Print("PetUpdate: {PetElement}")
                    BehavirIndex := ReturnPetIndex_FromInt(PetID)
                    if(TargetPetList := PETLIST[BehavirIndex]):
                        TargetPetList.DespawnTrigger.Trigger(Player)

            if(set EquippedPets[Player] = array{}){}
            for(PetElement : PetManager.GetEquipedPetStatus[Player]):
                if(set EquippedPets[Player] += array{PetElement(0).Index}){}

    # SaveDataのPetIndexから実際のIndexに置き換える
    ReturnPetIndex_FromInt(PETIndex:int, Upgrade:int):int=
        var PETIndexInt : int = 0

        # Kureppa
        if(PETIndex = 1):
            if(Upgrade = 0):
                set PETIndexInt = 0
            else:
                set PETIndexInt = 1

        # Kogasa
        if(PETIndex = 2):
            if(Upgrade = 0):
                set PETIndexInt = 2
            else:
                set PETIndexInt = 3

        # Kachinta
        if(PETIndex = 3):
            if(Upgrade = 0):
                set PETIndexInt = 4
            else:
                set PETIndexInt = 5

        # Ittanmomen
        if(PETIndex = 4):
            if(Upgrade = 0):
                set PETIndexInt = 6
            else:
                set PETIndexInt = 7

        # Narigama
        if(PETIndex = 5):
            if(Upgrade = 0):
                set PETIndexInt = 8
            else:
                set PETIndexInt = 9

        # DCKopper
        if(PETIndex = 11):
            if(Upgrade = 0):
                set PETIndexInt = 20
            else:
                set PETIndexInt = 21

        # DCKogasa
        if(PETIndex = 12):
            if(Upgrade = 0):
                set PETIndexInt = 26
            else:
                set PETIndexInt = 27

        # DCKachinta
        if(PETIndex = 13):
            if(Upgrade = 0):
                set PETIndexInt = 22
            else:
                set PETIndexInt = 23

        # DCIttanmomen
        if(PETIndex = 14):
            if(Upgrade = 0):
                set PETIndexInt = 24
            else:
                set PETIndexInt = 25

        # DCNarigama
        if(PETIndex = 15):
            if(Upgrade = 0):
                set PETIndexInt = 28
            else:
                set PETIndexInt = 29

        return PETIndexInt

    TriggerAFK(Agent:agent)<suspends>:void=
        AFKTrigger.Trigger(Agent)
        Sleep(0.5)
        AFKTrigger.Trigger(Agent)

    # トリガーした時にNPCをスポーン/カスタムスポーンイベントを起こすコード
    # プレイヤーの位置にNPCSpawnerをテレポートしてスポーンしてから元の位置に戻す
    (SpawnTrigger:trigger_device).SpawnTriggerEvent(NPC_Spawner:npc_spawner_device, DespawnTrigger:trigger_device, Index:int)<suspends>:void=

        # トリガーするまで待機
        PlayerAgentOption := SpawnTrigger.TriggeredEvent.Await()

        # NPCのスポーンイベント
        if:
            PlayerAgent := PlayerAgentOption?
            PlayerCharacter := PlayerAgent.GetFortCharacter[]
        
        then:
            BasicTransform := NPC_Spawner.GetTransform()
            TeleportTransform := PlayerCharacter.GetTransform()

            spawn{TriggerAFK(PlayerAgent)}

            SpawnCount := GetSpawnCount(Index)

            if(SpawnCount <= MaxIslandPlayers - 1):
                AddSpawnCount(Index)
                if(NPC_Spawner.TeleportTo[TeleportTransform]):

                Sleep(0.1)
            
                spawn{NPC_Spawner.CustomSpawnedEvent(PlayerAgent,DespawnTrigger, Index)}
                NPC_Spawner.Spawn()

                Sleep(0.1)

                if(NPC_Spawner.TeleportTo[BasicTransform]):

                Print("Spawn!!")
            
            else:
                Print("Failed Spawn!! Maybe limited spawn...")

        # 次回用のSpawnTriggerEventを登録しておく
        SpawnTrigger.SpawnTriggerEvent(NPC_Spawner, DespawnTrigger, Index)

    #SpawnCount
    GetSpawnCount(Index:int):int=
        var ReturnInt : int = 0
        if(PET := PETLIST[Index]):
            set ReturnInt = PET.SpawnCount
        return ReturnInt

    AddSpawnCount(Index:int):void=
        if(PET := PETLIST[Index]):
            set PET.SpawnCount += 1
            Print("SpawnCountPET : {PET.SpawnCount}")

    SubtructSpawnCount(Index:int):void=
        if(PET := PETLIST[Index]):
            set PET.SpawnCount -= 1
            Print("SpawnCountPET : {PET.SpawnCount}")
    
    # 指定秒数後終わるNavigateTo
    (Navigatable:navigatable).RaceNavigateTo(Target:navigation_target, State:movement_type, UpdateTime:float, ?OReachRadius:float=0.0)<suspends>:void=
        Movement : movement_type = State
        race:
            Navigatable.NavigateTo(Target,?MovementType:=Movement,?ReachRadius:=OReachRadius)
            Sleep(UpdateTime)

    # NPCをPlayerの位置に転送
    (NPCCharacter:fort_character).TeleportToPlayer(Target:fort_character):void=

        PlayerTransform := Target.GetTransform()

        if:
            NPCCharacter.TeleportTo[PlayerTransform.Translation, PlayerTransform.Rotation]

    # NPCの状態を取得する
    (NPCCharacter:fort_character).GetStatus(Target:fort_character, LastTransform:vector3, StackCount:int):tuple([]char,vector3,int)=
        
        # 変数
        var NPCStatus : []char = "None"
        var StackCountInstance : int = StackCount

        # Transform
        PlayerTransform := Target.GetTransform().Translation
        NPCTransform := NPCCharacter.GetTransform().Translation
        DistanceTransform := Distance(PlayerTransform,NPCTransform)
        DistanceLastTransform := Distance(LastTransform,NPCTransform)

        # それぞれ状態の決定
        if(DistanceTransform > RunDistance):
            set NPCStatus = "Run"
            set StackCountInstance = 0
        if(DistanceTransform < RunDistance):
            set NPCStatus = "Walk"
            set StackCountInstance = 0
        if(DistanceTransform < ReachRadius):
            set NPCStatus = "Stay"
            set StackCountInstance = 0

        # Stack対策(60ticks同じ場所にいた時スタックと判定する)
        if(DistanceLastTransform = 0.0, NPCStatus <> "Stay"):
            if(StackCount >= 30):
                set NPCStatus = "Stack"
                set StackCountInstance = 0
            else:
                set StackCountInstance = StackCount + 1

        # Teleport距離にいる時はNPCStatusを上書きする
        if(DistanceTransform > TeleportDistance):
            set NPCStatus = "Teleport"
            set StackCountInstance = 0
        
        return (NPCStatus, NPCTransform, StackCountInstance)

    # Waitingを実行するコード
    NPCWaiting(Player:agent)<suspends>:void=
        loop:
            TriggeredAgent := WaitingTrigger.TriggeredEvent.Await()
            if(TriggeredAgent? = Player):
                break

    # Followingを実行するコード
    NPCFollowing(Player:agent)<suspends>:void=
        loop:
            TriggeredAgent := FollowingTrigger.TriggeredEvent.Await()
            if(TriggeredAgent? = Player):
                break

    # NPCをDespawnするコード
    (DespawnTrigger:trigger_device).DespawnTriggerEvent(Index:int, NPC:agent, Player:agent)<suspends>:void=
        loop:
            TriggerAgent := DespawnTrigger.TriggeredEvent.Await()
            if(TriggerPlayer := TriggerAgent?, TriggerPlayer = Player, NPCCharacter := NPC.GetFortCharacter[]):
                SubtructSpawnCount(Index)
                NPCCharacter.SetVulnerability(true)
                NPCCharacter.Damage(5000.0)
                spawn{TriggerAFK(Player)}
                break

    # Miningでダメージを与える処理
    DoNPCDamage(TheMiningDevice:mining_object_device,Info:mining_object_info, Player:player, Index:int)<suspends>:void=
        # PetからIndexを取得してGatherRateを取得する
        var PETID : int = 0

        if(Index = 0 or Index = 1):
            set PETID = 1
        if(Index = 2 or Index = 3):
            set PETID = 2
        if(Index = 4 or Index = 5):
            set PETID = 3
        if(Index = 6 or Index = 7):
            set PETID = 4
        if(Index = 8 or Index = 9):
            set PETID = 5

        if(Index = 20 or Index = 21):
            set PETID = 11
        if(Index = 22 or Index = 23):
            set PETID = 12
        if(Index = 24 or Index = 25):
            set PETID = 13
        if(Index = 26 or Index = 27):
            set PETID = 14
        if(Index = 28 or Index = 29):
            set PETID = 15

        if:
            GatherRate := PetManager.GetPetGatherRate[Player, PETID]
        then:
            # 壊れるまで攻撃し続ける
            race:
                block:
                    TheMiningDevice.NPCDestroyedEvent.Await()
                    Print("NPCDestroyed!!", ?Duration:=10.0)
                block:
                    loop:
                        Player_STR := MiningDamage.GetNormalDamage(Player)
                        ResultGatherRate := PlayerData.BigNumber_MultiplyByBigNumber(Player_STR, BigNumber_MultiplyByFloat(GatherRate, 0.01))
                        Print("GatherRate : {PlayerData.BigNumber_FormatDisplay(PlayerData.BigNumber_MultiplyByBigNumber(Player_STR, BigNumber_MultiplyByFloat(GatherRate, 0.01)))}", ?Duration := 10.0)
                        Sleep(1.0)
                        DamageInfo := damage_info{InstigatorPlayer := Player, Amount := ResultGatherRate, DamageType := damage_type.Normal}
                        Info.Damage(DamageInfo)
                        spawn{DamageVFX_Func(Info)}
                        Print("Damaging!!", ?Duration:=10.0)

            Print("END!!!", ?Duration:=10.0)

    DamageVFX_Func(Info : mining_object_info)<suspends>:void=
        RockPos : vector3 = Info.PropManipulator.GetTransform().Translation
        var PropAsset : creative_prop_asset = VFX_OnDamge1
        Spawnable := GetRandomFloat(0.0, 100.0)
        if(Spawnable < VFXChance * 1.0) :
            WhichVFX := GetRandomFloat(0.0, 100.0)
            if(WhichVFX > 50.0) :
                set PropAsset = VFX_OnDamge1
                Print("VFX = 1", ?Color := NamedColors.Red)
            if(WhichVFX <= 50.0) :
                set PropAsset = VFX_OnDamge2
                Print("VFX = 2", ?Color := NamedColors.Red)
            SpawnedProp : ?creative_prop = SpawnProp(PropAsset, RockPos, rotation{})(0)
            if(TheProp : creative_prop = SpawnedProp?) :
                Sleep(3.0)
                TheProp.Dispose()
                Print("VFX Spawn!", ?Color := NamedColors.Red)
        else :
            Print("VFX not spawn", ?Color := NamedColors.Red)

    MiningDevicesRace():void=
        for(MiningDevice : MiningDevices):
            spawn{MiningRace(MiningDevice)}

    MiningRace(MiningDevice:mining_object_device)<suspends>:void=
        loop:
            TheMiningObject := MiningDevice.NPCDamagedEvent.Await()
            NPCDamagedEvent.Signal(TheMiningObject, MiningDevice)
            Print("Signal!!")

    # NPCがスポーンした時のカスタムイベント
    (NPC_Spawner:npc_spawner_device).CustomSpawnedEvent(Player:agent, DespawnTrigger:trigger_device, Index:int)<suspends>:void=

        # NPCがSpawnするまで待機
        NPCAgent := NPC_Spawner.SpawnedEvent.Await()

        spawn{DespawnTrigger.DespawnTriggerEvent(Index,NPCAgent,Player)}

        var LastTransform : vector3 = vector3{}
        var StackCount : int = 0
        var DamagedAgent : ?agent = false
        var DestroyedAgent : ?agent = false
        var DamagedPropManipulator : int = 0
        var DestroyedPropManipulator : int = 0

        # CharacterとNavigatableの定義
        loop:

            if:
                PlayerType := player[Player]
                PlayerCharacter := Player.GetFortCharacter[]
                NPCCharacter := NPCAgent.GetFortCharacter[]
                Navigatable := NPCCharacter.GetNavigatable[]
                FocusInterface := NPCCharacter.GetFocusInterface[]
                PlayerCharacter.IsActive[]

            # Playerに追従するコード
            then:

                NavigationTarget := MakeNavigationTarget(Player)

                var NPCStatus : []char = NPCCharacter.GetStatus(PlayerCharacter,LastTransform,StackCount)(0)
                set StackCount = NPCCharacter.GetStatus(PlayerCharacter,LastTransform,StackCount)(2)
                set LastTransform = NPCCharacter.GetStatus(PlayerCharacter,LastTransform,StackCount)(1)
                var PropManipulator : prop_manipulator_device = prop_manipulator_device{}
                var TheMiningDevice : mining_object_device = mining_object_device{}
                var MiningObjectInfo : mining_object_info = mining_object_info{}
                var Mining : logic = false
                var Waiting : logic = false
                MiningRadius : float = 100.0

                race:
                    block:
                        Navigation :=
                            #Print("{NPCStatus}")
                            #Print("{StackCount}")

                            if(NPCStatus = "Walk"):
                                Navigatable.RaceNavigateTo(NavigationTarget,movement_types.Walking,1.0)
                            if(NPCStatus = "Run"):
                                Navigatable.RaceNavigateTo(NavigationTarget,movement_types.Running,3.0)
                            if(NPCStatus = "Teleport"):
                                Sleep(0.1)
                                NPCCharacter.TeleportToPlayer(PlayerCharacter)
                            if(NPCStatus = "Stay"):
                                Sleep(1.0)
                            if(NPCStatus = "Stack"):
                                Sleep(0.2)
                                NPCCharacter.TeleportToPlayer(PlayerCharacter)

                        # NavigateToが終わって結果がReachedの時、n秒待機
                        if (Navigation = navigation_result.Reached):
                            Navigatable.Wait(?Duration := WaitTime)

                    # Playerを見続ける
                    block:
                        FocusInterface.MaintainFocus(Player)

                    # Mining状態の時
                    block:
                        MiningAwait := NPCDamagedEvent.Await()
                        set MiningObjectInfo = MiningAwait(0)
                        set TheMiningDevice = MiningAwait(1)
                        set PropManipulator = MiningObjectInfo.PropManipulator
                        set DamagedAgent = TheMiningDevice.SignalPlayerDamagedEventFunction()(0)
                        set DamagedPropManipulator = TheMiningDevice.SignalPlayerDestroyedEventFunction()(1)
                        
                        Print("NPCDamagedSignalを受け取りました！")
                        if(DA := DamagedAgent?, DA = Player):
                            set DamagedAgent = false
                            Print("Get Prop Damage!!")
                            set Mining = true
                    # Waitingの状態
                    block:
                        NPCWaiting(Player)
                        set Waiting = true

                if(Mining = true):
                    PropManipulatorTranslation := PropManipulator.GetTransform().Translation
                    MiningNavigationTarget := MakeNavigationTarget(PropManipulatorTranslation)

                    var MiningCantFind : logic = false

                    race:
                        block:
                            loop:
                                Navigatable.RaceNavigateTo(MiningNavigationTarget,movement_types.Running,2.5, ?OReachRadius := 100.0)

                                # NPCが採掘位置にいるかの確認
                                NPCTranslation := NPCCharacter.GetTransform().Translation
                                NPCPropDistance := Distance(NPCTranslation, PropManipulatorTranslation)
                                if(NPCPropDistance >= MiningRadius + 10.0):
                                    break
                        block:
                            # 5秒後にNPCが採掘位置にいるかの確認、いない場合はTeleportする
                            Sleep(3.0)
                            NPCTranslation := NPCCharacter.GetTransform().Translation
                            NPCPropDistance := Distance(NPCTranslation, PropManipulatorTranslation)
                            if(NPCPropDistance >= MiningRadius + 10.0):
                                TeleportTranslation := vector3{X := PropManipulatorTranslation.X + MiningRadius, Y:= PropManipulatorTranslation.Y + MiningRadius, Z:= PropManipulatorTranslation.Z}
                                if(NPCCharacter.TeleportTo[TeleportTranslation,rotation{}]):

                        block:
                            DestroyManipulator := TheMiningDevice.NPCDestroyedEvent.Await()
                            set MiningCantFind = true

                    # Mining Animationの再生。Propが壊れた時アニメーションの再生をやめる
                    if(AnimController := NPCCharacter.GetPlayAnimationController[], MiningAnimation := PETLIST[Index].Mining?):
                        race:
                            
                            <# 本来のMining Animation Code
                            block:
                                loop:
                                    Sleep(0.0)
                                    Instance := AnimController.PlayAndAwait(MiningAnimation,?BlendInTime:=0.25,?BlendOutTime:=0.25)#>

                            # 32.00 update 対策コード
                            # 冗長すぎだろ！！！！！😡
                            # アップデート撲滅運動
                            block:

                                # Chibi
                                if(PETLIST[Index].PETNAME = "Kopper" or PETLIST[Index].PETNAME = "DCKopper"):
                                    loop:
                                        Sleep(0.0)
                                        Instance := AnimController.PlayAndAwait(MiningAnim.Kopper_Mining,?BlendInTime:=0.25,?BlendOutTime:=0.25)
                                if(PETLIST[Index].PETNAME = "Kachinta" or PETLIST[Index].PETNAME = "DCKachinta"):
                                    loop:
                                        Sleep(0.0)
                                        Instance := AnimController.PlayAndAwait(MiningAnim.Kachinta_Mining,?BlendInTime:=0.25,?BlendOutTime:=0.25)
                                if(PETLIST[Index].PETNAME = "Kogasa" or PETLIST[Index].PETNAME = "DCKogasa"):
                                    loop:
                                        Sleep(0.0)
                                        Instance := AnimController.PlayAndAwait(MiningAnim.Kogasa_Mining,?BlendInTime:=0.25,?BlendOutTime:=0.25)
                                if(PETLIST[Index].PETNAME = "Ittanmomen" or PETLIST[Index].PETNAME = "DCIttanmomen"):
                                    loop:
                                        Sleep(0.0)
                                        Instance := AnimController.PlayAndAwait(MiningAnim.Ittanmomen_Mining,?BlendInTime:=0.25,?BlendOutTime:=0.25)
                                if(PETLIST[Index].PETNAME = "Narigama" or PETLIST[Index].PETNAME = "DCNarigama"):
                                    loop:
                                        Sleep(0.0)
                                        Instance := AnimController.PlayAndAwait(MiningAnim.Narigama_Mining,?BlendInTime:=0.25,?BlendOutTime:=0.25)
                                
                                # Boss
                                if(PETLIST[Index].PETNAME = "Kureppa" or PETLIST[Index].PETNAME = "DCKureppa"):
                                    loop:
                                        Sleep(0.0)
                                        Instance := AnimController.PlayAndAwait(MiningAnim.Kureppa_Mining,?BlendInTime:=0.25,?BlendOutTime:=0.25)
                                if(PETLIST[Index].PETNAME = "Tachigarasu" or PETLIST[Index].PETNAME = "DCTachigarasu"):
                                    loop:
                                        Sleep(0.0)
                                        Instance := AnimController.PlayAndAwait(MiningAnim.Tachigarasu_Mining,?BlendInTime:=0.25,?BlendOutTime:=0.25)
                                if(PETLIST[Index].PETNAME = "Karakasa" or PETLIST[Index].PETNAME = "DCKarakasa"):
                                    loop:
                                        Sleep(0.0)
                                        Instance := AnimController.PlayAndAwait(MiningAnim.Karakasaobake_Mining,?BlendInTime:=0.25,?BlendOutTime:=0.25)
                                if(PETLIST[Index].PETNAME = "Oobanmomen" or PETLIST[Index].PETNAME = "DCOobanmomen"):
                                    loop:
                                        Sleep(0.0)
                                        Instance := AnimController.PlayAndAwait(MiningAnim.Oobanmomen_Mining,?BlendInTime:=0.25,?BlendOutTime:=0.25)
                                if(PETLIST[Index].PETNAME = "Enengama" or PETLIST[Index].PETNAME = "DCEnengama"):
                                    loop:
                                        Sleep(0.0)
                                        Instance := AnimController.PlayAndAwait(MiningAnim.Enengama_Mining,?BlendInTime:=0.25,?BlendOutTime:=0.25)

                            block:
                                loop:
                                    DestroyManipulator := TheMiningDevice.NPCDestroyedEvent.Await()
                                    set DestroyedAgent = TheMiningDevice.SignalPlayerDestroyedEventFunction()(0)
                                    set DestroyedPropManipulator = TheMiningDevice.SignalPlayerDestroyedEventFunction()(1)
                                    if(DA := DestroyedAgent?, DA = Player or DestroyedPropManipulator = DamagedPropManipulator):
                                        set DestroyedAgent = false
                                        Print("Destroy!", ?Duration:=10.0)
                                        break
                            block:
                                FocusInterface.MaintainFocus(PropManipulatorTranslation)

                            block:
                                DoNPCDamage(TheMiningDevice, MiningObjectInfo, PlayerType, Index)
                                Print("A!", ?Duration:=10.0)
                            
                            block:
                                loop:
                                    Sleep(1.0)
                                    set NPCStatus = NPCCharacter.GetStatus(PlayerCharacter,LastTransform,StackCount)(0)
                                    if(NPCStatus = "Teleport" or MiningCantFind = true):
                                        break
                                        
                # Waiting状態の時、待機場所にテレポートする
                if(Waiting = true):
                    WaitingTriggerPosition := WaitingTrigger.GetTransform().Translation
                    TeleportPosition := vector3{X:=WaitingTriggerPosition.X, Y:=WaitingTriggerPosition.Y, Z:=WaitingTriggerPosition.Z + 50.0 + (Index * 150.0)}
                    loop:
                        Sleep(0.1)
                        NPCLocation := NPCCharacter.GetTransform().Translation

                        NPCCharacter.Hide()

                        # NPCがテレポートできてないとき、Teleportを再実行する
                        if(NPCLocation.X - TeleportPosition.X >= 100.0):
                            if(NPCCharacter.TeleportTo[TeleportPosition,rotation{}]):
                        else:
                            Print("NPC Teleport to Waiting Place")
                            break
                    
                    NPCFollowing(Player)
                    NPCCharacter.Show()
                    NPCCharacter.TeleportToPlayer(PlayerCharacter)
                    Print("Waiting Break")

            # NPCがいない/Playerがいない時、Breakする
            else:

                # NPCがDespawnしていた時、Breakする
                if(NPCCharacter := NPCAgent.GetFortCharacter[], NPCCharacter.IsActive[]):
                else:
                    Print("NPC is Despawned!! Break")
                    break

                # PlayerがEliminate中なだけの場合、特に何もしない
                if(ThePlayer := player[Player], ThePlayer.IsActive[]):
                    Print("Player is Active!! No Break")
                    Sleep(0.1)

                # Playerがもう島にいない尚且つNPCが残ってる場合Despawn/Break
                else:
                    DespawnTrigger.Trigger(Player)
                    Print("Player is not Active!! Break")
                    break