using { /Fortnite.com/Devices }
using { /Verse.org/Simulation }
using { /UnrealEngine.com/Temporary/SpatialMath }
using { /UnrealEngine.com/Temporary/UI }
using { /Verse.org/Colors }
using { /Fortnite.com/UI }

using { Objects.MiningObjectHealthBar.Materials }
using { Code.PlayerData }
using { Utility  }

hp_bar_device<public> := class(creative_device):

    var HPBars<public> : [comparable]hp_bar_info = map{}
    var InvisibleColorBlock<public> : color_block = color_block{DefaultOpacity := 0.0}
    var CurrentTarget<public> : ?comparable = false
    var CurrentHPBar<public> : ?hp_bar_info = false
    var IDCounter<public> : int = 0
    var CountdownMap<public> : [player]int = map{}
    var SaveDataManager : save_data_manager = save_data_manager{}
    
    CreateHPBar<public>(Target:comparable, Pos:transform, PlayerUI:player_ui)<decides><transacts>:hp_bar_info=

        CenterVector : vector2 = vector2{X := 0.5, Y := 0.5}
        CenterAnchors : anchors = anchors{Minimum := CenterVector, Maximum := CenterVector}

        # 枠
        FrameTexture : texture_block = texture_block{
            DefaultImage := Objects.UI_Healthbar_outside,
            DefaultDesiredSize := vector2{X := 880.0, Y := 43.0}
        }
        # HPバー
        HPTexture : texture_block = texture_block{
            DefaultImage := Objects.UI_Healthbar_inside,
            DefaultDesiredSize := vector2{X := 817.0, Y := 43.0}
        }     
        TextWidget := text_block{
            DefaultTextColor := color{R := 1.0, G := 1.0, B := 1.0}
            DefaultTextOpacity := 1.0
            DefaultShadowOffset := option{vector2{X := 2.36, Y := 2.83}}
            DefaultShadowColor := color{R := 0.0, G := 0.0, B := 0.0}
            DefaultShadowOpacity := 0.64
            DefaultJustification := text_justification.Center
        }
        LastHitWidget := text_block{
            DefaultTextColor := color{R := 1.0, G := 1.0, B := 1.0}
            DefaultTextOpacity := 1.0
            DefaultShadowOffset := option{vector2{X := 2.36, Y := 2.83}}
            DefaultShadowColor := color{R := 0.0, G := 0.0, B := 0.0}
            DefaultShadowOpacity := 0.64
            DefaultJustification := text_justification.Center
        }


        NewCanvas : canvas = canvas:
            Slots := array:
                canvas_slot:
                        Anchors := CenterAnchors
                        Offsets := margin{Top := -451.0, Left := 0.0, Right := 880.0, Bottom := 43.0}
                        Alignment := CenterVector
                        SizeToContent := false
                        ZOrder := 0
                        Widget := FrameTexture
                        
                canvas_slot:
                    Anchors := CenterAnchors
                    Offsets := margin{Top := -451.0, Left := 0.0, Right := 817.0, Bottom := 43.0}
                    Alignment := CenterVector
                    SizeToContent := false
                    ZOrder := 1
                    Widget := stack_box:
                        Orientation := orientation.Horizontal
                        Slots := array:
                            stack_box_slot:
                                Distribution := option{1.0}
                                HorizontalAlignment := horizontal_alignment.Fill
                                VerticalAlignment := vertical_alignment.Fill
                                Widget := HPTexture
                            stack_box_slot:
                                HorizontalAlignment := horizontal_alignment.Fill
                                VerticalAlignment := vertical_alignment.Fill
                                Widget := InvisibleColorBlock
                           
                canvas_slot:
                    Anchors := CenterAnchors
                    Offsets := margin{Top := -496.0, Left := 0.0, Right := 257.0, Bottom := 41.0}
                    Alignment := CenterVector
                    Widget := TextWidget
                    SizeToContent := false
                    ZOrder := 3

        NewHPBarInfo := hp_bar_info:
            HPBarCanvas := NewCanvas
            HPBarMover := InvisibleColorBlock
            HPBarTexture := HPTexture
            HPBarBackTexture := FrameTexture
            AssignedPlayerUI := option{PlayerUI}
            ID := IDCounter
            HPText := TextWidget

            
        set HPBars[IDCounter] = NewHPBarInfo
        Print("新しいHPバーが追加されました。ID: {IDCounter}")

        set IDCounter = IDCounter + 1

        set CurrentHPBar = option{NewHPBarInfo}  # 現在の HP バーとして設定

        Print("新しいHPバーが追加されました。ID: {NewHPBarInfo.ID}")

        return NewHPBarInfo

        
    SetHPBarPercent<public>(Target:comparable, Percent:float):void=
        if(TargetHPBar := HPBars[Target]):
            NewWidth := 817.0 - (817.0 * Percent)
            TargetHPBar.HPBarMover.SetDesiredSize(vector2{X := NewWidth, Y := 43.0})
            Print("HPバーの新しい幅: {NewWidth}")

    SetHPText<public>(Target:comparable, Health:PlayerData.big_number, MaxHealth:PlayerData.big_number):void=
        if (TargetHPBar := HPBars[Target]):
            HPTextMessage := StringToMessage("Stone HP: {PlayerData.BigNumber_FormatDisplay(Health)} / {PlayerData.BigNumber_FormatDisplay(MaxHealth)}")
            TargetHPBar.HPText.SetText(HPTextMessage)
            Print("HPバーのテキストが更新されました: {PlayerData.BigNumber_FormatDisplay(Health)} / {PlayerData.BigNumber_FormatDisplay(MaxHealth)}")
            

    Dispose<public>(TargetID:int):void=
        if(TargetHPBar := HPBars[TargetID]):
            TargetHPBar.RemoveFromUI()
            set HPBars = HPBars.RemoveKeyFromMap(TargetID)
            Print("HPバーが削除されました")
            HPBarCanvas := canvas{}

        set CurrentHPBar = false
    