# using { /Verse.org/Assets }
# using { /Verse.org/Simulation }
# using { /Fortnite.com/UI }
# using { /Fortnite.com/Devices }
# using { /Fortnite.com/Characters }
# using { /UnrealEngine.com/Temporary/UI }
# using { /UnrealEngine.com/Temporary/SpatialMath }
# using { /Verse.org/Simulation }
# using { /UnrealEngine.com/Temporary/Diagnostics }
# using { /Verse.org/Colors }
# using { /Verse.org/Colors/NamedColors }
# using { PlayerData }
# using { MiningObject }
# using { Objects.MiningObjectHealthBar.Materials }
# using { Utility }

# hp_bar_device_ui<public> := class(creative_device):
#     var HPBars<public> : [comparable]hp_bar_info = map{}
#     var Canvas : canvas = canvas{}

#     # 枠
#     FrameTexture<public> : texture_block = texture_block{
#         DefaultImage := Objects.UI_Healthbar_outside,
#         DefaultDesiredSize := vector2{X := 880.0, Y := 43.0}
#     }

    # # HPバー
    # HPTexture<public> : texture_block = texture_block{
    #     DefaultImage := Objects.UI_Healthbar_inside,
    #     DefaultDesiredSize := vector2{X := 880.0, Y := 43.0}
    # }

    # InvisibleColorBlock<public> : color_block = color_block{DefaultOpacity := 0.0}
    # トランザクション内で行うHPバーの設定や管理
    # AddHPBarWidget<public>(PlayerUI:player_ui):void=
    #     # 中心の位置ベクトルとアンカー設定
    #     CenterVector : vector2 = vector2{X := 0.5, Y := 0.5}
    #     CenterAnchors : anchors = anchors{Minimum := CenterVector, Maximum := CenterVector}
    
    #     # 新しい canvas を作成
    #     NewCanvas := canvas{}
    
    #     # HPバーの外枠（FrameTexture）
    #     FrameCanvasSlot := canvas_slot:
    #         Anchors := CenterAnchors
    #         Offsets := margin{Top := -451.0, Left := 0.0, Right := 880.0, Bottom := 43.0}
    #         Alignment := CenterVector
    #         SizeToContent := false
    #         ZOrder := 0
    #         Widget := texture_block{
    #             DefaultImage := Objects.UI_Healthbar_outside,
    #             DefaultDesiredSize := vector2{X := 880.0, Y := 43.0}
    #         }
    #     NewCanvas.AddWidget(FrameCanvasSlot)
    
    #     # ProgressStackBox の作成
    #     ProgressStackBox := stack_box{Orientation := orientation.Horizontal}
    #     HPBarCanvasSlot := canvas_slot:
    #         Anchors := CenterAnchors
    #         Offsets := margin{Top := -451.0, Left := 0.0, Right := 880.0, Bottom := 43.0}
    #         Alignment := CenterVector
    #         SizeToContent := false
    #         ZOrder := 1
    #         Widget := ProgressStackBox
    #     NewCanvas.AddWidget(HPBarCanvasSlot)
    
    #     # スタックボックス内の空白領域
    #     ProgressStackBox.AddWidget(stack_box_slot:
    #         HorizontalAlignment := horizontal_alignment.Fill
    #         VerticalAlignment := vertical_alignment.Fill
    #         Widget := color_block{DefaultOpacity := 0.0}  # InvisibleColorBlock
    #     )
    
    #     # スタックボックス内のHPバー（進行度部分）
    #     ProgressStackBox.AddWidget(stack_box_slot:
    #         HorizontalAlignment := horizontal_alignment.Fill
    #         VerticalAlignment := vertical_alignment.Fill
    #         Distribution := option{1.0}
    #         Widget := texture_block{
    #             DefaultImage := Objects.UI_Healthbar_inside,
    #             DefaultDesiredSize := vector2{X := 880.0, Y := 43.0}
    #         }
    #     )
    
    #     # HPバーのテキスト（進行度のテキスト表示）
    #     TextWidget := text_block{
    #         DefaultTextColor := color{R := 1.0, G := 1.0, B := 1.0}
    #         DefaultTextOpacity := 1.0
    #         DefaultShadowOffset := option{vector2{X := 2.36, Y := 2.83}}
    #         DefaultShadowColor := color{R := 0.0, G := 0.0, B := 0.0}
    #         DefaultShadowOpacity := 0.64
    #         DefaultJustification := text_justification.Center
    #     }
    #     TextCanvasSlot := canvas_slot:
    #         Anchors := CenterAnchors
    #         Offsets := margin{Top := -496.0, Left := 0.0, Right := 257.0, Bottom := 41.0}
    #         Alignment := CenterVector
    #         Widget := TextWidget
    #         SizeToContent := false
    #         ZOrder := 3
    #     NewCanvas.AddWidget(TextCanvasSlot)
    
    #     # プレイヤーUIに追加
    #     PlayerUI.AddWidget(NewCanvas)
    
    # SetHPBarPercent<public>(Target:comparable, Percent:float)<decides>:void=
    #     if(TargetHPBar := HPBars[Target]):
    #         # `InvisibleColorBlock` の幅を HP の割合に応じて調整
    #         NewWidth := 880.0 * Percent  # 0 から 880 の間で変動
    #         TargetHPBar.HPBarMover.SetDesiredSize(vector2{X := NewWidth, Y := 43.0})  # HPBarMover は InvisibleColorBlock

    # # HPバーの削除
    # Dispose<public>(Target:comparable):void=
    #     if(TargetHPBar := HPBars[Target]):
    #         TargetHPBar.RemoveFromUI()  # UIから削除
    #         set HPBars = HPBars.RemoveKeyFromMap(Target)    
    
    # # キャンバスを表示
    # ShowCanvas<public>(Player:player, Target:comparable):void=
    #     if(PlayerUI := GetPlayerUI[Player]):
    #         PlayerUI.AddWidget(Canvas, player_ui_slot{})
    