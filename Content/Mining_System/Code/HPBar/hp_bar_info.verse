using { /Fortnite.com/Devices }
using { /UnrealEngine.com/Temporary/UI }
using { /UnrealEngine.com/Temporary/SpatialMath }
using { /Fortnite.com/UI }


using { Objects.MiningObjectHealthBar.Materials }

hp_bar_info<public> := class:
    HPBarCanvas<public> : canvas
    HPBarMover<public> : color_block
    HPBarBackTexture<public> : texture_block
    HPBarTexture<public> : texture_block
    AssignedPlayerUI<public> : ?player_ui
    ID<public> : int
    HPText<public> : text_block

    # HPバーのサイズをパーセントに応じて変更
    SetPercent(Percent:float)<decides>:void=
        NewWidth := 880.0 * Percent
        HPBarMover.SetDesiredSize(vector2{X := NewWidth, Y := 43.0})
        Print("HPバーの新しい幅: {NewWidth}")  # デバッグ用の出力

    # HPバーの削除
    RemoveFromUI<public>():void=
        if(PlayerUI := AssignedPlayerUI?):
            PlayerUI.RemoveWidget(HPBarCanvas)  # HPバー全体の Canvas を削除
            Print("HPバーの Canvas が削除されました")
    

# hp_bar_info<public> := class:

#     HPBarProp<public> : creative_prop
#     HPBarMaterial<public> : M_MiningObjectHealthBar_material

#     Init()<transacts>:void=
#         HPBarProp.SetMaterial(HPBarMaterial)

#     SetPercent(Percent:float)<transacts>:void=
#         set HPBarMaterial.Percent = Percent

#     Dispose():void=
#         HPBarProp.Dispose()
