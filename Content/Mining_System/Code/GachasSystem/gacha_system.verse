using { /Fortnite.com/Devices }
using { /Verse.org/Simulation }
using { /Verse.org/Random }
using { /Verse.org/Assets }
using { /Verse.org/Colors }
using {Utility}
using {PlayerData}
using { Pet }
using {VerseWidget}

# 使用するタマゴのリソースタイプ
cost_egg_type<public> := enum:
    Common
    Uncommon
    Rare
    Epic
    Legendary

# ガチャから出てくる報酬の種類
gacha_reward_type<public> := enum:
    Yokai
    WeaponUpgrade
    Resouece

# ガチャから出てくる妖怪の種類
gacha_reward_yokai_type<public> := enum:
    YokaiA
    YokaiB
    YokaiARare
    YokaiBRare

gacha_system<public> := class(creative_device):

    # ステージが分かれてるのでそれに応じて内容物、コストが変わる
    @editable
    StageIndex<public> : int= 0

    # 石のリソースがどれだけ必要か
    @editable
    CostStone<public> : editable_big_number = editable_big_number{}

    # 召喚石がどれだけ必要か
    @editable
    CostEgg<public> : editable_big_number = editable_big_number{}

    # 石のリソースがどれだけ必要か
    @editable
    EvloveCostStone<public> : editable_big_number = editable_big_number{}

    # 召喚石がどれだけ必要か
    @editable
    EvloveCostEgg<public> : editable_big_number = editable_big_number{}

    @editable
    SaveDataManager : save_data_manager = save_data_manager{}

    # 妖怪Aが出てくる確率(gacha_reward_typeがYokaiの時)
    @editable
    RewardYokaiAPercent<public> : float = 45.0

    # 妖怪Bが出てくる確率(gacha_reward_typeがYokaiの時)
    @editable
    RewardYokaiBPercent<public> : float = 45.0

    # 妖怪A(レア)が出てくる確率(gacha_reward_typeがYokaiの時)
    @editable
    RewardYokaiARarePercent<public> : float = 5.0

    # 妖怪B(レア)が出てくる確率(gacha_reward_typeがYokaiの時)
    @editable
    RewardYokaiBRarePercent<public> : float = 5.0

    @editable
    PetManager : pet_manager = pet_manager{}

    @editable
    GachaDevice : gacha_device = gacha_device{}

    @editable
    OpenButton : button_device = button_device{}

    # Resoueceが出た時のリワードの報酬設定？
    CommonReward<public> : tuple(float, float, float, float) = (0.10, 0.05, 0.10, 0.75)
    UncommonReward<public> : tuple(float, float, float, float) = (0.15, 0.10, 0.20, 0.55)
    RareReward<public> : tuple(float, float, float, float) = (0.20, 0.30, 0.30, 0.20)
    EpicReward<public> : tuple(float, float, float, float) = (0.30, 0.40, 0.20, 0.10)
    LegendaryReward<public> : tuple(float, float, float, float) = (0.40, 0.50, 0.09, 0.01)

    <#
    CommonReward<public> : tuple(float, float, float, float) = (0.40, 0.05, 0.10, 0.15)
    UncommonReward<public> : tuple(float, float, float, float) = (0.60, 0.05, 0.10, 0.25)
    RareReward<public> : tuple(float, float, float, float) = (0.65, 0.05, 0.08, 0.30)
    EpicReward<public> : tuple(float, float, float, float) = (0.70, 0.05, 0.10, 0.15)
    LegendaryReward<public> : tuple(float, float, float, float) = (0.75, 0.20, 0.04, 0.01)
    #>
    OnBegin<override>()<suspends>:void=
        loop:
            Agent := OpenButton.InteractedWithEvent.Await()
            if(Player := player[Agent]):
                GachaDevice.SetGacha(Player, Self)

    # ステージ1の妖怪のID
    YokaiTypeIndexStage1<public> : [gacha_reward_yokai_type]int = map:
        gacha_reward_yokai_type.YokaiA => 1
        gacha_reward_yokai_type.YokaiB => 3
        gacha_reward_yokai_type.YokaiARare => 11
        gacha_reward_yokai_type.YokaiBRare => 13

    # ステージ2の妖怪のID
    YokaiTypeIndexStage2<public> : [gacha_reward_yokai_type]int = map:
        gacha_reward_yokai_type.YokaiA => 4
        gacha_reward_yokai_type.YokaiB => 2
        gacha_reward_yokai_type.YokaiARare => 14
        gacha_reward_yokai_type.YokaiBRare => 12

    # ステージ3の妖怪のID
    YokaiTypeIndexStage3<public> : [gacha_reward_yokai_type]int = map:
        gacha_reward_yokai_type.YokaiA => 5
        gacha_reward_yokai_type.YokaiB => 3
        gacha_reward_yokai_type.YokaiARare => 15
        gacha_reward_yokai_type.YokaiBRare => 13

    # UI用の特殊リソースアイコン
    EggIcon<public> : [cost_egg_type]texture = map:
        cost_egg_type.Common => icon_set_2.SS_1
        cost_egg_type.Uncommon => icon_set_2.SS_2
        cost_egg_type.Rare => icon_set_2.SS_3
        cost_egg_type.Epic => icon_set_2.SS_4
        cost_egg_type.Legendary => icon_set_2.SS_5

    # Stage1にかかる特殊リソースの値
    CostPerEggStage1<public> : [cost_egg_type]big_number = map:
        cost_egg_type.Common => big_number{Value := 1.0}
        cost_egg_type.Uncommon => big_number{Value := 1.0}
        cost_egg_type.Rare => big_number{Value := 1.0}
        cost_egg_type.Epic => big_number{Value := 1.0}
        cost_egg_type.Legendary => big_number{Value := 1.0}

    # Stage2にかかる特殊リソースの値
    CostPerEggStage2<public> : [cost_egg_type]big_number = map:
        cost_egg_type.Common => big_number{Value := 130.0}
        cost_egg_type.Uncommon => big_number{Value := 130.0}
        cost_egg_type.Rare => big_number{Value := 130.0}
        cost_egg_type.Epic => big_number{Value := 130.0}
        cost_egg_type.Legendary => big_number{Value := 130.0}

    # Stage3にかかる特殊リソースの値
    CostPerEggStage3<public> : [cost_egg_type]big_number = map:
        cost_egg_type.Common => big_number{Value := 3.0,Exponent := 3}
        cost_egg_type.Uncommon => big_number{Value := 3.0,Exponent := 3}
        cost_egg_type.Rare => big_number{Value := 3.0,Exponent := 3}
        cost_egg_type.Epic => big_number{Value := 3.0,Exponent := 3}
        cost_egg_type.Legendary => big_number{Value := 3.0,Exponent := 3}

    # Stage1にかかるリソースの値
    CostPerStoneStage1<public> : [cost_egg_type]big_number = map:
        cost_egg_type.Common => big_number{Value := 20.0, Exponent :=6}
        cost_egg_type.Uncommon => big_number{Value := 300.0, Exponent :=6}
        cost_egg_type.Rare => big_number{Value := 4.5, Exponent :=9}
        cost_egg_type.Epic => big_number{Value := 67.5, Exponent :=9}
        cost_egg_type.Legendary => big_number{Value := 1.01, Exponent :=12}

    # Stage2にかかるリソースの値
    CostPerStoneStage2<public> : [cost_egg_type]big_number = map:
        cost_egg_type.Common => big_number{Value := 30.0, Exponent :=15}
        cost_egg_type.Uncommon => big_number{Value := 450.0, Exponent :=15}
        cost_egg_type.Rare => big_number{Value := 6.75, Exponent :=18}
        cost_egg_type.Epic => big_number{Value := 101.0, Exponent :=18}
        cost_egg_type.Legendary => big_number{Value := 1.51, Exponent :=21}

    # Stage3にかかるリソースの値
    CostPerStoneStage3<public> : [cost_egg_type]big_number = map:
        cost_egg_type.Common => big_number{Value := 300.0, Exponent :=21}
        cost_egg_type.Uncommon => big_number{Value := 4.5, Exponent :=24}
        cost_egg_type.Rare => big_number{Value := 67.5, Exponent :=24}
        cost_egg_type.Epic => big_number{Value := 1.01, Exponent :=27}
        cost_egg_type.Legendary => big_number{Value := 15.1, Exponent :=27}

    # StageIndexに応じて得られるリワードを変える
    GetRewards<public>():[]int=
        if(StageIndex = 0):
            for(Element : YokaiTypeIndexStage1):
                Element
        else if(StageIndex = 1):
            for(Element : YokaiTypeIndexStage2):
                Element
        else if(StageIndex = 2):
            for(Element : YokaiTypeIndexStage3):
                Element
        else:
            array{}

    # StageIndex取得用
    GetStageIndex<public>():int=
        StageIndex
    
    GetCostPerEgg<public>(CostEggType:cost_egg_type)<decides><transacts>:big_number=
        if(StageIndex = 0):
            return CostPerEggStage1[CostEggType]
        else if(StageIndex = 1):
            return CostPerEggStage2[CostEggType]
        else if(StageIndex = 2):
            return CostPerEggStage3[CostEggType]
        else:
            return big_number{Value := 0.0}

    GetCostPerStone<public>(CostEggType:cost_egg_type)<transacts><decides>:big_number=
        if(StageIndex = 0):
            return CostPerStoneStage1[CostEggType]
        else if(StageIndex = 1):
            return CostPerStoneStage2[CostEggType]
        else if(StageIndex = 2):
            return CostPerStoneStage3[CostEggType]
        else:
            return big_number{Value := 0.0}

    Roll<public>(InPlayer:player, CostEggType:cost_egg_type):tuple(logic, int)=
        StageYokaiTypes := array:
            YokaiTypeIndexStage1
            YokaiTypeIndexStage2
            YokaiTypeIndexStage3

        if(CanRoll[InPlayer, CostEggType]):
            Print("CanRoll")
            Percent := GetPercentbyEgg(CostEggType)
            PickedPercent := GetRandomFloat(0.0, 1.0)
            #所持ペットの総数が0の場合、確定で妖怪がでるようにする
            PetCount := PetManager.GetPetCount(InPlayer)
            Print("ペット数{PetCount}", ?Color := NamedColors.Red)
            #Print("PickedPercent: {PickedPercent}")
            #Print("Percent: {Percent(0)}")
            #妖怪当たり
            if(PickedPercent <= Percent(0) or (PetCount = 0)):
                YokaiType := GetGachaYokaiType()
                Print("StageIndex: {StageIndex}")  # 妖怪の種類とステージを表示
                if(RandomYokaiIndex := StageYokaiTypes[StageIndex][YokaiType]):
                    Print("Unlocked Pet with ID: {RandomYokaiIndex}")
                    Print("StageIndex: {StageIndex}, RandomYokaiIndex: {RandomYokaiIndex}")
                    PetManager.UnlockPet(InPlayer, RandomYokaiIndex)
                    Print("RollResult - Returning ID: {RandomYokaiIndex}")
                    return (true, RandomYokaiIndex)
            #アプグレ素材当たり
            else if(PickedPercent <= Percent(0) + Percent(1)):
                Resouece := GetGachaResource(0)
                SaveDataManager.AddResourcePowerupStone(InPlayer, big_number{Value := Resouece*1.0})
            else if(PickedPercent <= Percent(0) + Percent(1) + Percent(2)):
                Resouece := GetGachaResource(1)
                SaveDataManager.AddResourcePowerupStone(InPlayer, big_number{Value := Resouece*1.0})
            else:
                Resouece := GetGachaResource(2)
                SaveDataManager.AddResourcePowerupStone(InPlayer, big_number{Value := Resouece*1.0})
            return (true, -1)

        return (false, -1)

    CanRoll<public>(InPlayer:player, CostEggType:cost_egg_type)<decides><transacts>:void=
        RequiredEgg := GetCostPerEgg[CostEggType]
        RequiredStone := GetCostPerStone[CostEggType]
        
        Funcs := GetEggFunctions(CostEggType)
        NowEggs := Funcs(3)(InPlayer)
        
        # 必要なエッグとストーンを持っているか確認
        Print("Now Resource : {PlayerData.BigNumber_FormatDisplay(SaveDataManager.GetResourceStone(InPlayer))}")
        (BigNumber_GreaterThanOrEquals(NowEggs, RequiredEgg)? and BigNumber_GreaterThanOrEquals(SaveDataManager.GetResourceStone(InPlayer), RequiredStone)?)
    

    # 0 = 妖怪, 1 = 素材大, 2 = 素材中, 3 = 素材小
    GetPercentbyEgg<public>(CostEggType:cost_egg_type):tuple(float, float, float, float)=
        case(CostEggType):
            cost_egg_type.Common=>
                return CommonReward
            cost_egg_type.Uncommon=>
                return UncommonReward
            cost_egg_type.Rare=>
                return RareReward
            cost_egg_type.Epic=>
                return EpicReward
            cost_egg_type.Legendary=>
                return LegendaryReward

    # 確率の合計が100%の場合
    GetGachaYokaiType():gacha_reward_yokai_type=
        PickedPercent := GetRandomFloat(0.0, 100.0)
        if(PickedPercent <= RewardYokaiAPercent):
            return gacha_reward_yokai_type.YokaiA
        else if(PickedPercent <= RewardYokaiAPercent + RewardYokaiBPercent):
            return gacha_reward_yokai_type.YokaiB
        else if(PickedPercent <= RewardYokaiAPercent + RewardYokaiBPercent + RewardYokaiARarePercent):
            return gacha_reward_yokai_type.YokaiARare
        else:
            return gacha_reward_yokai_type.YokaiBRare

    GetGachaResource<public>(Index:int):int=
        if(Index = 0):
            if(StageIndex = 0):
                return 9
            if(StageIndex = 1):
                return 135
            if(StageIndex = 2):
                return 540
        if(Index = 1):
            if(StageIndex = 0):
                return 3
            if(StageIndex = 1):
                return 45
            if(StageIndex = 2):
                return 180
        if(Index = 2):
            if(StageIndex = 0):
                return 1
            if(StageIndex = 1):
                return 15
            if(StageIndex = 2):
                return 60
        # 対応する数値がない場合
        Print("Error: GetGachaResource / 対応する数値が入っていません。")
        return 0

    GetEggFunctions(CostEggType:cost_egg_type)<transacts>:tuple(type{_(:player, :big_number)<transacts>:void}, type{_(:player, :big_number)<transacts>:void}, type{_(:player, :big_number)<transacts>:void} ,type{_(:player)<transacts>:big_number})=
        
        case(CostEggType):
            cost_egg_type.Common=>
                (SaveDataManager.SetResourceSummonStoneCommon, SaveDataManager.AddResourceSummonStoneCommon, SaveDataManager.SubtractResourceSummonStoneCommon, SaveDataManager.GetResourceSummonStoneCommon)
            cost_egg_type.Uncommon=>
                (SaveDataManager.SetResourceSummonStoneUncommon, SaveDataManager.AddResourceSummonStoneUncommon, SaveDataManager.SubtractResourceSummonStoneUncommon, SaveDataManager.GetResourceSummonStoneUncommon)
            cost_egg_type.Rare=>
                (SaveDataManager.SetResourceSummonStoneRare, SaveDataManager.AddResourceSummonStoneRare, SaveDataManager.SubtractResourceSummonStoneRare, SaveDataManager.GetResourceSummonStoneRare)
            cost_egg_type.Epic=>
                (SaveDataManager.SetResourceSummonStoneEpic, SaveDataManager.AddResourceSummonStoneEpic, SaveDataManager.SubtractResourceSummonStoneEpic, SaveDataManager.GetResourceSummonStoneEpic)
            cost_egg_type.Legendary=>
                (SaveDataManager.SetResourceSummonStoneLegendary, SaveDataManager.AddResourceSummonStoneLegendary, SaveDataManager.SubtractResourceSummonStoneLegendary, SaveDataManager.GetResourceSummonStoneLegendary)