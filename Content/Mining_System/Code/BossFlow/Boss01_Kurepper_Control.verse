using { /Fortnite.com/Devices }
using { /Verse.org/Simulation }
using { /UnrealEngine.com/Temporary/Diagnostics }
using { /Fortnite.com/Characters }
using { /UnrealEngine.com/Assets }
using { /UnrealEngine.com/Temporary/SpatialMath }
using { /Verse.org/Random }
using { /Verse.org/Assets }
using { /Verse.org/Colors }
using { /UnrealEngine.com/Temporary/UI }
using { /Fortnite.com/UI }

using { Pet }
using { Weapon }
using { PlayerData }

boss_device_base<public> := class(creative_device):
    #XP付与
    @editable XP_ElimedBoss : accolades_device = accolades_device{}

    @editable
    BossStartTrigger : trigger_device = trigger_device{}

    @editable
    ChangeTeamAndClass : class_and_team_selector_device = class_and_team_selector_device{}

    @editable
    NP_Resource : visual_effect_powerup_device = visual_effect_powerup_device{}
    @editable
    AD_Resource : audio_player_device = audio_player_device{}

    @editable
    NP_PowerUpResource : visual_effect_powerup_device = visual_effect_powerup_device{}
    @editable
    AD_PowerUpResource : audio_player_device = audio_player_device{}

    @editable
    NP_SummonStone_R1 : visual_effect_powerup_device = visual_effect_powerup_device{}
    @editable
    AD_SummonStone_R1 : audio_player_device = audio_player_device{}

    @editable
    NP_SummonStone_R2 : visual_effect_powerup_device = visual_effect_powerup_device{}
    @editable
    AD_SummonStone_R2 : audio_player_device = audio_player_device{}

    @editable
    NP_SummonStone_R3 : visual_effect_powerup_device = visual_effect_powerup_device{}
    @editable
    AD_SummonStone_R3 : audio_player_device = audio_player_device{}

    @editable
    NP_SummonStone_R4 : visual_effect_powerup_device = visual_effect_powerup_device{}
    @editable
    AD_SummonStone_R4 : audio_player_device = audio_player_device{}

    @editable
    NP_SummonStone_R5 : visual_effect_powerup_device = visual_effect_powerup_device{}
    @editable
    AD_SummonStone_R5 : audio_player_device = audio_player_device{}

    BossImage<public> : texture = Boss_Textures.kreppar

    ChangeTeamFunc<public>(Agent : agent):void=
        ChangeTeamAndClass.ChangeTeamAndClass(Agent)

    StartBossFight<public>(Agent : ?agent) : void=
        if(Player := Agent?):
            Print("A")
            BossStartTrigger.Trigger(Player)
        return

    RewardNiagara(Agent:agent, GetSummonStone:int)<suspends>:void=
        NP_Resource.Pickup(Agent)
        AD_Resource.Play(Agent)
        Sleep(1.0)
        NP_PowerUpResource.Pickup(Agent)
        AD_PowerUpResource.Play(Agent)
        Sleep(1.0)
        if(GetSummonStone = 0):
            NP_SummonStone_R1.Pickup(Agent)
            AD_SummonStone_R1.Play(Agent)
        if(GetSummonStone = 1):
            NP_SummonStone_R2.Pickup(Agent)
            AD_SummonStone_R2.Play(Agent)
        if(GetSummonStone = 2):
            NP_SummonStone_R3.Pickup(Agent)
            AD_SummonStone_R3.Play(Agent)
        if(GetSummonStone = 3):
            NP_SummonStone_R4.Pickup(Agent)
            AD_SummonStone_R4.Play(Agent)
        if(GetSummonStone = 4):
            NP_SummonStone_R5.Pickup(Agent)
            AD_SummonStone_R5.Play(Agent)

Boss01_Kurepper_Control<public> := class(boss_device_base):

    BossImage<override> : texture = Boss_Textures.kreppar

    @editable
    SpawnPads : []player_spawner_device = array{}
    @editable
    PlayerCheck : player_counter_device = player_counter_device{}

    @editable
    StartSeq_Boss : cinematic_sequence_device = cinematic_sequence_device{}

    @editable
    EndSeq_Boss : cinematic_sequence_device = cinematic_sequence_device{}

    @editable
    DPSCheckSeq_Boss : cinematic_sequence_device = cinematic_sequence_device{}

    @editable
    TimeOverSeq_Boss : cinematic_sequence_device = cinematic_sequence_device{}
    @editable
    DamagedEffectSeq_Boss : cinematic_sequence_device = cinematic_sequence_device{}
    
    @editable
    MotionsSeq_Boss : []cinematic_sequence_device = array{}
    @editable
    DPSed_MotionSeq_Boss : []cinematic_sequence_device = array{}

    @editable
    TimeLimit_Second : float = 300.0
    @editable
    var Health : float = 100.0

    @editable
    DamageTrigger  : prop_manipulator_device = prop_manipulator_device{}

    @editable
    WeaponCheckers : []conditional_button_device = array{}
    @editable
    TeamChanger_Default : class_and_team_selector_device = class_and_team_selector_device{}
    @editable
    TeamChanger_Blader : class_and_team_selector_device = class_and_team_selector_device{}
    @editable
    RewardButton : button_device = button_device{}
    #UI
    @editable
    UI_GotReward : hud_message_device = hud_message_device{}
    #トリガー
    @editable
    StartTrigger : trigger_device = trigger_device{}
    @editable
    EndTrigger : trigger_device = trigger_device{}

    @editable
    EnterTeleporter : teleporter_device = teleporter_device{}
    @editable
    ExitTeleporter : teleporter_device = teleporter_device{}
    @editable
    DefaultTeleporter : teleporter_device = teleporter_device{}

    @editable
    BGM : audio_player_device = audio_player_device{}

    #ペットの能力関連
    @editable
    VFX_Powerup : damage_amplifier_powerup_device = damage_amplifier_powerup_device{}
    @editable
    VFX_Lightning2 : creative_prop_asset = DefaultCreativePropAsset
    @editable
    SFX_OnHit : audio_player_device = audio_player_device{}
    @editable
    SFX_OnHeal : audio_player_device = audio_player_device{}
    @editable
    Zeus_Granter : item_granter_device = item_granter_device{}
    @editable
    Zeus_Remover : item_remover_device = item_remover_device{}

    var Pet_Manager : pet_manager = pet_manager{}
    var IsActivable_Powerup : [agent]int = map{}#0 = 起動不可, 1 = 15s待機中, 2 = 起動可能
    #攻撃時のダメージ判定
    var Weapon_Damage : weapon_damage = weapon_damage{}

    #エフェクト
    @editable
    VFX_SquirtUp : creative_prop_asset = DefaultCreativePropAsset

    @editable
    VFX_SquirtUp_DPSed : creative_prop_asset = DefaultCreativePropAsset

    @editable
    VFX_Fall : creative_prop_asset = DefaultCreativePropAsset

    @editable
    VFX_Fall_DPSed : creative_prop_asset = DefaultCreativePropAsset
    #ダメージを与えるデバイス
    @editable
    DV_SquirtUp : []damage_volume_device = array{}

    @editable
    DV_SquirtUp_DPSed : []damage_volume_device = array{}
    @editable
    DamagedNockBack : movement_modulator_device = movement_modulator_device{}

    @editable
    AreaSize : []creative_prop = array{}
    #水球が現れてキャラを包むエフェクト
    @editable
    DPS01 : creative_prop = creative_prop{}
    #水球が弾けるエフェクト
    @editable
    VFX_WaterPrisonExplode : creative_prop_asset = DefaultCreativePropAsset

    @editable
    Seq_WaterPrison : cinematic_sequence_device = cinematic_sequence_device{}
    
    @editable
    var DPSHealth : float = 10.0

    @editable
    DPSDamageTrigger : prop_manipulator_device = prop_manipulator_device{}

    @editable
    Krepper_Rotator : creative_prop = creative_prop{}

    @editable
    Boss_Timer : timer_device = timer_device{}

    @editable
    ZoneVolume : volume_device = volume_device{}

    @editable
    SaveDataManager : PlayerData.save_data_manager = PlayerData.save_data_manager{}

    @editable
    MiningDamageDevice : MiningObject.mining_damage_device = MiningObject.mining_damage_device{}

    #ダメージチェックモーションが起動したか "UnTrigger" = 未起動 , "Trigger" = 起動中 , "Triggered" = 起動済み
    var DPSChecker : []char = "UnTrigger"
    #再生するモーション番号0~n , StartSeq = -1 , EndSeq = 999
    var MotionValue : int = -1

    var DefaultHealth : float = - 100.0

    var DamageCheckHealth : float = 50.0
    #プレイヤーの攻撃力初期値
    var DamageValue : float = 0.0

    #"NotActive" = ゾーン内に誰もいない , "BeforeGame" = 誰かがゾーンに入ったけど最初のムービー中 , "InGame" = ボス戦中
    var BossState<public> : []char = "NotActive"

    var Count : int = 0

    var Count02 : int = 0

    var PropsArray : []creative_prop = array{}
    

    var OriginPos : vector3 = vector3{X := 0.0 , Y := 0.0 , Z := 0.0}
    var DPSTeleportPos : vector3 = vector3{X := 0.0 , Y := 0.0 , Z := 0.0}

    #ボスフィールドの座標最大値最小値
    var MinX : float = 0.0
    var MaxX : float = 0.0
    var MinY : float = 0.0
    var MaxY : float = 0.0

    var RandomPosX : float = 0.0
    var RandomPosY : float = 0.0
    #"Unbroken" = クレッパーがDPS技を出すまでのstateで、水球が壊れていない状態
    #"CouldntBreak" = クレッパーがDPS技を出してからのstateで、水球が壊れていない状態（壊せなかった）
    #"Broken" = 水球が壊れた状態
    var WaterPrisonState : []char = "Unbroken"

    var DPSDefaultHealth : float = - 100.0

    var UseDVArray : []damage_volume_device = array{}

    var Yaw : float = 0.0
    var Pitch : float = 0.0
    var Roll : float = 0.0
    
    var IsGotReward : [agent]logic = map{}

    ReturnBossState<public>():[]char=
        return BossState

    StartBossFight<override>(Agent : ?agent) : void =
        #ボス戦のチームにいるプレイヤーをテレポートさせ、全員テレポートさせたらモーション開始
        if(Agent02 := Agent?) :
            for(ThePlayer : GetPlayspace().GetPlayers()) :
                if :
                    TeamCollection := GetPlayspace().GetTeamCollection()
                    BossTeam := TeamCollection.GetTeams()[1]
                    TeamCollection.IsOnTeam[ThePlayer , BossTeam]
                then :
                    EnterTeleporter.Teleport(ThePlayer)
                    Print("TeleportTo BossField")
            if(BossState = "NotActive") :
                set BossState = "BeforeGame"
                spawn{PlayMotions(Agent02)}
                spawn{TimerCheck(Agent02)}

    PlayMotions(Agent02 : agent)<suspends> : void = 
        loop :
            #プレイヤーカウンター内にプレイヤーがいるとき
            if(0 < PlayerCheck.GetCount()) :
                if(MotionValue = -1) :
                    Print("StartSeq.Play")
                    spawn{ShowBBWidget()}

                    StartSeq_Boss.Play(Agent02)
                    set MotionValue += 1
                    BGM.Play()
                    #ボリューム内のプレイヤーを取得してタイマー取り付け
                    set Count = 0
                    PlayerList : []player = GetPlayspace().GetPlayers()
                    for(Players : PlayerList) :
                        if(PlayerCheck.IsCounted[PlayerList[Count]] , ThePlayer := PlayerList[Count]) :
                            Boss_Timer.Start(ThePlayer)
                            spawn{OnDamaged(ThePlayer)}
                            spawn{CheckAbility_01(ThePlayer)}
                            Print("タイマー取り付け！")
                        set Count += 1
                        Sleep(0.1)
                    set Count = 0
                    StartSeq_Boss.StoppedEvent.Await()

                    #各変数を初期化
                    if(PlayerCheck.GetCount() = 1):
                        set Health = DefaultHealth * 1
                    if(PlayerCheck.GetCount() = 2):
                        set Health = DefaultHealth * 1.5
                    if(PlayerCheck.GetCount() = 3):
                        set Health = DefaultHealth * 2
                    if(PlayerCheck.GetCount() = 4):
                        set Health = DefaultHealth * 2.5
                    set DamageCheckHealth = Health * 0.6
                    #MaxHealthをセット
                    set MaxHealth = Health
                    set DPSChecker = "UnTrigger"
                    set BossState = "InGame"
                    Print("Health = {Health} , DamageCheckHealth ={DamageCheckHealth}")

                else if(DPSChecker = "Trigger" , BossState <> "NotActive"):
                    Print("DPSCheck")
                    DPSCheckSeq_Boss.Play()
                    spawn{MotionDPS()}
                    DPSCheckSeq_Boss.StoppedEvent.Await()
                    set DPSChecker = "Triggered"

                #クレッパーなら MotionValue = 0~3
                else if(0 <= MotionValue <= MotionsSeq_Boss.Length -2) :
                    Print("Playing Motion")
                    if(SelectedMotion := MotionsSeq_Boss[MotionValue] , DPSChecker = "UnTrigger" or "Triggered") :
                        #----------モーション名テキスト変更----------
                        var Frame : int = 0
                        if(MotionValue = 1):
                            set Frame = 240
                        if(MotionValue = 2):
                            set Frame = 390
                        if(MotionValue = 3):
                            set Frame = 411
                        spawn{SetMotionText(MotionValue, Frame)}
                        #-------------------------------------------
                        if(MotionValue = 0) :
                                SelectedMotion.Play()
                                Print("{MotionValue}")
                                SelectedMotion.StoppedEvent.Await()
                        else if(Selected_DPSed_Motion := DPSed_MotionSeq_Boss[MotionValue - 1]) :
                            if(MotionValue = 1) :
                                spawn{Motion01()}
                            if(MotionValue = 3) :
                                spawn{Motion03()}

                            if(DPSChecker = "Triggered") :
                                Selected_DPSed_Motion.Play()
                                Print("{MotionValue}DPS後")
                                Selected_DPSed_Motion.StoppedEvent.Await()
                            else :
                                SelectedMotion.Play()
                                Print("{MotionValue}DPS前")
                                SelectedMotion.StoppedEvent.Await()

                        set MotionValue += 1

                #クレッパーなら MotionValue = 4
                else if(MotionValue = MotionsSeq_Boss.Length -1) :
                    if(SelectedMotion := MotionsSeq_Boss[MotionValue] , Selected_DPSed_Motion := DPSed_MotionSeq_Boss[MotionValue - 1] , DPSChecker = "UnTrigger" or "Triggered") :
                        #----------モーション名テキスト変更----------
                        var Frame : int = 0
                        if(MotionValue = 4):
                            set Frame = 210
                        spawn{SetMotionText(MotionValue, Frame)}
                        #-------------------------------------------
                        Print("{MotionValue}")
                        spawn{Motion04()}
                        if(DPSChecker = "Triggered") :
                            Selected_DPSed_Motion.Play()
                            Print("{MotionValue}DPS後")
                            set MotionValue = 0
                            Selected_DPSed_Motion.StoppedEvent.Await()
                        else :
                            SelectedMotion.Play()
                            Print("{MotionValue}DPS前")
                            set MotionValue = 0
                            SelectedMotion.StoppedEvent.Await()

                else if(MotionValue >= 999):
                    EndSeq_Boss.Play(Agent02)
                    set MotionValue = -1
                    EndSeq_Boss.StoppedEvent.Await()
                    #報酬部屋にテレポートさせる処理
                    ExitPos := ExitTeleporter.GetTransform().Translation
                    ExitRotation := ExitTeleporter.GetTransform().Rotation
                    for(ThePlayer : GetPlayspace().GetPlayers()) :
                        if :
                            TeamCollection := GetPlayspace().GetTeamCollection()
                            BossTeam := TeamCollection.GetTeams()[1]
                            TeamCollection.IsOnTeam[ThePlayer , BossTeam]
                            Char : fort_character = ThePlayer.GetFortCharacter[]
                        then :
                            ExitTeleporter.Teleport(ThePlayer)
                            XP_ElimedBoss.Award(ThePlayer)
                            if(Agent : agent = agent[ThePlayer]):
                                if (SaveDataManager.GetWeaponLevel(ThePlayer) = 0):
                                    TeamChanger_Default.ChangeTeamAndClass(Agent)
                                    Print("チームリセット一般人,ゾーン")
                                else:
                                    TeamChanger_Blader.ChangeTeamAndClass(Agent)
                                    Print("チームリセット刀,ゾーン")
                    #ここまで
                    BGM.Stop()
                    Print("break , End")
                    break

            #プレイヤーカウンター内にプレイヤーがいないとき
            else if(0 = PlayerCheck.GetCount()) :
                set BossState = "NotActive"
                HideBBWidget()
                DPS_HideBBWidget()
                set MotionValue = -1
                BGM.Stop()
                EndTrigger.Trigger()
                Print("break , No Players")
                break

            Sleep(0.1)

    EndMotions() : void =
        EndTrigger.Trigger()
        if(1 <= MotionValue <= MotionsSeq_Boss.Length -1 , SelectedMotion := MotionsSeq_Boss[MotionValue] , Selected_DPSed_Motion := DPSed_MotionSeq_Boss[MotionValue - 1]) :
            if(DPSChecker <> "Triggered") :
                SelectedMotion.Stop()
            else :
                Selected_DPSed_Motion.Stop()

        else if(MotionValue = 0 , SelectedMotion := MotionsSeq_Boss[MotionsSeq_Boss.Length -1] , Selected_DPSed_Motion := DPSed_MotionSeq_Boss[DPSed_MotionSeq_Boss.Length - 1]) :
            if(DPSChecker <> "Triggered") :
                SelectedMotion.Stop()
            else :
                Selected_DPSed_Motion.Stop()
        set MotionValue = 999
        Print("Stopped Motions")

    WeaponCheck(Agent:agent) : void =
        spawn{DamageEffect()}
        if :
            BossState = "InGame"
        then :
            Print("WeapoCheck InGame")
        if :
            Player := player[Agent]
        then :
            Print("WeapoCheck Player")
            if :
                var WeaponValue : float = Weapon_Damage.CalculateDamage[Player]
            then :
                if(SaveDataManager.GetStrBuff(Player)?):
                    set WeaponValue = WeaponValue * 1.2
                    Print("恐竜バフ", ?Color := NamedColors.Red, ?Duration := 3.0)
                set DamageValue = WeaponValue
                Print("Damaged {DamageValue}")

        spawn{CheckAbility_02(Agent)}

        set Health -= DamageValue
        Print("Health = {Health}")
        if(BossState = "InGame") :
            ChangeHPBar(Health)

        #体力が半分以下になったとき1度だけDPSCheckerをTriggerにset
        if(BossState = "InGame" , Health <= DamageCheckHealth , DPSChecker = "UnTrigger") :
            Print("Trigger")
            set DPSChecker = "Trigger"

        #体力が0以下になったときにMotionValueを999にsetしてボス戦終了シーケンスを再生
        if(BossState = "InGame" , Health <= 0.0) :
            Print("Health <= 0")
            set BossState = "NotActive"
            HideBBWidget()
            DPS_HideBBWidget()
            EndMotions()

    ##-----------------ここからペット能力--------------------------------##
    CheckAbility_01(Agent : agent)<suspends>:void=
        Print("Spawned CheckAbility_01")
        #ボス戦開始時にクレッパーとエンエンガマの能力チェック
        if :
            Char : fort_character = Agent.GetFortCharacter[]
            Player : player = player[Agent]
            Index := Pet_Manager.GetPetIndexInFirstSlot[Player]
            Upgrade := Pet_Manager.GetPetUpgradeInFirstSlot[Player]
            PetSup := SaveDataManager.GetSkillLevel[Player , 14]
            PetSup = 1
        then :
            if(Upgrade >= 1) :
                Print("Having Ability")
                if(Index = 1) :
                    if(Upgrade = 1) :
                        Char.SetShield(50.0)
                        Print("Kure 30~69lv")
                    if(Upgrade = 2) :
                        Char.SetShield(100.0)
                        Print("Kure 70~??lv")
                if(Index = 5) :
                    if(not IsActivable_Powerup[Agent]) :
                        logic{set IsActivable_Powerup[Agent] = 0}
                    race :
                        block :
                            EndTrigger.TriggeredEvent.Await()
                            logic{set IsActivable_Powerup[Agent] = 0}
                            Print("!!race : Finished!!")
                        block :
                            Sleep(15.0)
                            VFX_Powerup.SetDuration(-1.0)
                            VFX_Powerup.Pickup(Agent)
                            logic{set IsActivable_Powerup[Agent] = 2}

                if(Index = 2) :
                    Zeus_Granter.GrantItem(Agent)

    CheckAbility_02(Agent : agent)<suspends>:void=
        #ボスにダメージを与えた時にカラカサ,オオバンモメン,タチガラスの能力をチェック
        if :
            Char : fort_character = Agent.GetFortCharacter[]
            Player : player = player[Agent]
            Index := Pet_Manager.GetPetIndexInFirstSlot[Player]
            Upgrade := Pet_Manager.GetPetUpgradeInFirstSlot[Player]
            PetSup := SaveDataManager.GetSkillLevel[Player , 14]
            PetSup = 1
        then :
            if(Upgrade >= 1) :
                if(Index = 4) :
                    var Threshold : int = 10
                    if(Upgrade = 1 ) :
                        set Threshold = 10
                    if(Upgrade = 2 ) :
                        set Threshold = 15
                    if(BossState = "InGame" , GetRandomInt(0 , 100) <= Threshold) :
                        SpawnedProp : ?creative_prop = SpawnProp(VFX_Lightning2 , Krepper_Rotator.GetTransform().Translation , rotation{})(0)
                        if(PlayerCheck.GetCount() = 1):
                            set DamageValue = DamageValue + DefaultHealth * 1 * 0.03
                        if(PlayerCheck.GetCount() = 2):
                            set DamageValue = DamageValue + DefaultHealth * 1.5 * 0.03
                        if(PlayerCheck.GetCount() = 3):
                            set DamageValue = DamageValue + DefaultHealth * 2 * 0.03
                        if(PlayerCheck.GetCount() = 4):
                            set DamageValue = DamageValue + DefaultHealth * 2.5 * 0.03
                        Print("Kara {Threshold}%")
                        Sleep(0.1)
                        SFX_OnHit.Play(Agent)
                        Sleep(1.0)
                        if(TheProp : creative_prop = SpawnedProp?) :
                            TheProp.Dispose()

                if(Index = 5) :
                    if(IsActivable_Powerup[Agent] = 2) :
                        SFX_OnHit.Play(Agent)
                        if(Upgrade = 1) :
                            set DamageValue *= 3.0
                            Print("Enen 30~69lv")
                        if(Upgrade = 2) :
                            set DamageValue *= 5.0
                            Print("Enen 70~??lv")
                        VFX_Powerup.SetDuration(0.1)
                        VFX_Powerup.Pickup(Agent)
                        logic{set IsActivable_Powerup[Agent] = 0}
                        spawn{CheckAbility_01(Agent)}

                if(Index = 3) :
                    var NewHealth : float = 0.0
                    if(Upgrade = 1 , Char.GetHealth() <= 99.0) :
                        set NewHealth = Char.GetHealth() + 1.0
                        Char.SetHealth(NewHealth)
                        SFX_OnHeal.Play(Agent)
                        Print("Ooban 0~69lv")
                    if(Upgrade = 2 , Char.GetHealth() <= 98.0) :
                        set NewHealth = Char.GetHealth() + 2.0
                        Char.SetHealth(NewHealth)
                        SFX_OnHeal.Play(Agent)
                        Print("Ooban 70~??lv")

                if(Index = 2) :
                    if :
                        BossState = "InGame"
                        Zeus := WeaponCheckers[0]
                        Zeus.IsHoldingItem[Agent , 1]
                        var WeaponValue : float = Weapon_Damage.CalculateDamage[Player]
                    then :
                        if(SaveDataManager.GetStrBuff(Player)?):
                            set WeaponValue = WeaponValue*1.2
                        var Damage_Multiplier : float = 2.0
                        if(Upgrade = 1) :
                            set Damage_Multiplier = 1.0
                            Print("Tachi 30~69lv")
                        if(Upgrade = 2) :
                            set Damage_Multiplier = 1.5
                            Print("Tachi 70~??lv")
                        set DamageValue = WeaponValue * Damage_Multiplier

    ##ここまでペット能力##

    TimerCheck(Agent : agent)<suspends> : void =
        race :
            block :
                EndTrigger.TriggeredEvent.Await()
                Print("Timer Stop")
            block :
                var IsTimerChecked : logic = false
                Sleep(TimeLimit_Second)
                for(ThePlayer : GetPlayspace().GetPlayers()) :
                    if(PlayerCheck.IsCounted[ThePlayer] , Char : fort_character = ThePlayer.GetFortCharacter[]) :
                        Char.SetVulnerability(false)
                        Char.PutInStasis(stasis_args{AllowTurning := false , AllowFalling := false , AllowEmotes := false})
                        Print("Freeze & Valner")
                        if(IsTimerChecked = false) :
                            set IsTimerChecked = true
                            TimeOverSeq_Boss.Play(ThePlayer)
                            Print("Play")
                TimeOverSeq_Boss.StoppedEvent.Await()
                for(ThePlayer : GetPlayspace().GetPlayers()) :
                    if(PlayerCheck.IsCounted[ThePlayer] , Char : fort_character = ThePlayer.GetFortCharacter[]) :
                        Char.ReleaseFromStasis()
                        DefaultTeleporter.Teleport(ThePlayer)
                        Char.SetVulnerability(true)
                        Print("Player TP & Release")

    OnExitFromVolume(Agent : agent) : void =
        Boss_Timer.Reset(Agent)
        if(PlayerCheck.GetCount() = 0) :
            HideBBWidget()
            DPS_HideBBWidget()
        Zeus_Remover.Remove(Agent)
        Print("Timer Completed !")

    DamageEffect()<suspends> : void =
        if(BossState = "InGame") :
            DamagedEffectSeq_Boss.Play()
            Print("Spawned Damage Effect!")
    
    OnDamaged(Agent : agent)<suspends>:void=
        race :
            block :
                EndTrigger.TriggeredEvent.Await()
                Print("!!race : Finished!!")
            block :
                if(Char : fort_character = Agent.GetFortCharacter[]) :
                    loop :
                        if(PlayerCheck.IsCounted[Agent]) :
                            Char.DamagedEvent().Await()
                            if(Char.IsDownButNotOut[]) :
                                Print("ダウン中")
                            else :
                                DamagedNockBack.Activate(Agent)
                                Print("ノックバック！")
                        else :
                            break
                    Print("プレイヤーがいなくなった")

    OnInteractedReward(Agent : agent):void=
        var GetSummonIndex : int = 0
        if(not IsGotReward[Agent]) :
            logic{set IsGotReward[Agent] = false}
        if(IsGotReward[Agent] = false) :
            logic{set IsGotReward[Agent] = true}
            #報酬与える処理#
            if(Player := player[Agent]):

                # 強化報酬報酬
                RewardPowerupStone := PlayerData.big_number{Value := 10.0, Exponent := 0}

                # 召喚石報酬
                RandomReward := GetRandomFloat(0.0, 100.0)
                SummonReward := PlayerData.big_number{Value := 33.0}

                # 通常リソース報酬
                GetDamage :=MiningDamageDevice.GetNormalDamageAmount(Player).Amount
                RewardResource := PlayerData.BigNumber_MultiplyByBigNumber(GetDamage, PlayerData.big_number{Value := 250.0})

                # 報酬付与
                SaveDataManager.AddResourcePowerupStone(Player, RewardPowerupStone)
                SaveDataManager.AddResourceStone(Player, RewardResource)

                Print("RewardPowerupStone : {PlayerData.BigNumber_FormatDisplay(RewardPowerupStone)}")
                Print("RewardResource : {PlayerData.BigNumber_FormatDisplay(RewardResource)}")
                
                if(RandomReward < 76.5):
                    SaveDataManager.AddResourceSummonStoneCommon(Player,SummonReward)
                    set GetSummonIndex = 0
                    Print("Reward is SummonStoneCommon!!")
                if(76.5 <= RandomReward < 94.5):
                    SaveDataManager.AddResourceSummonStoneUncommon(Player,SummonReward)
                    set GetSummonIndex = 1
                    Print("Reward is SummonStoneUncommon!!")
                if(94.5 <= RandomReward < 98.8):
                    SaveDataManager.AddResourceSummonStoneRare(Player,SummonReward)
                    set GetSummonIndex = 2
                    Print("Reward is SummonStoneRare!!")
                if(98.8 <= RandomReward < 99.8):
                    SaveDataManager.AddResourceSummonStoneEpic(Player,SummonReward)
                    set GetSummonIndex = 3
                    Print("Reward is SummonStoneEpic!!")
                if(99.8 <= RandomReward <= 100.0):
                    SaveDataManager.AddResourceSummonStoneLegendary(Player,SummonReward)
                    set GetSummonIndex = 4
                    Print("Reward is SummonStoneLegendary!!")

            spawn{BackToDefault(Agent)}
            spawn{RewardNiagara(Agent, GetSummonIndex)}
            Print("報酬付与！")
        else :
            UI_GotReward.Show(Agent)
            Print("報酬付与済み")

    BackToDefault(Agent : agent)<suspends>:void=
        Sleep(5.0)
        DefaultTeleporter.Teleport(Agent)
        logic{set IsGotReward[Agent] = false}
        Print("TeleportTo Default")

    Motion01()<suspends> : void =
        race :
            block :
                EndTrigger.TriggeredEvent.Await()
                Print("!!race : Finished!!")
            block :
                set Count = 0
                set Count02 = 0
                PlayerList : []player = GetPlayspace().GetPlayers()
                Print("Motion01 moving")
                loop :
                    set Count = GetRandomInt(0 , PlayerList.Length - 1)
                    if(PlayerCheck.IsCounted[PlayerList[Count]]) :
                        Print("抽選成功！")
                        break
                    else :
                        Print("再抽選")
                    Sleep(0.1)

                Sleep(0.5)

                if(DPSChecker = "Triggered" , ThePlayer := PlayerList[Count] , Char : fort_character = ThePlayer.GetFortCharacter[]) :
                    Print("追跡中 , DPS後")
                    PlayerPos := Char.GetTransform().Translation
                    KrepperPos := Krepper_Rotator.GetTransform().Translation

                    if(LookDirection := (PlayerPos - KrepperPos).MakeUnitVector[]) :

                        set Yaw = RadiansToDegrees(ArcTan(LookDirection.Y , LookDirection.X))
                
                        NewRotation := MakeRotationFromYawPitchRollDegrees(Yaw , Pitch , Roll)
                        Krepper_Rotator.MoveTo(KrepperPos , NewRotation , 1.0)

                    set Yaw = 90.0
                    Sleep(5.0)
                    OriginTransform := Krepper_Rotator.GetTransform().Translation
                    OriginRotation := MakeRotationFromYawPitchRollDegrees(Yaw , Pitch , Roll)
                    Print("{OriginRotation}")
                    Krepper_Rotator.MoveTo(OriginTransform , OriginRotation , 1.5)

                    if(Selected_DPSed_Motion := DPSed_MotionSeq_Boss[0]) :
                        Selected_DPSed_Motion.StoppedEvent.Await()

                else if(ThePlayer := PlayerList[Count] , Char : fort_character = ThePlayer.GetFortCharacter[]) :
                    loop :
                        if(Count02 <= 7) :
                            Print("追跡中 , DPS前")
                            PlayerPos := Char.GetTransform().Translation
                            KrepperPos := Krepper_Rotator.GetTransform().Translation

                            if(LookDirection := (PlayerPos - KrepperPos).MakeUnitVector[]) :

                                set Yaw = RadiansToDegrees(ArcTan(LookDirection.Y , LookDirection.X))
                        
                                NewRotation := MakeRotationFromYawPitchRollDegrees(Yaw , Pitch , Roll)
                                Krepper_Rotator.MoveTo(KrepperPos , NewRotation , 0.5)
                            set Count02 += 1
                        else :
                            break
                        Sleep(0.1)

                    Sleep(1.3)
                    set Yaw = 90.0
                    OriginTransform := Krepper_Rotator.GetTransform().Translation
                    OriginRotation := MakeRotationFromYawPitchRollDegrees(Yaw , Pitch , Roll)
                    Print("{Yaw}{OriginRotation}")
                    Krepper_Rotator.MoveTo(OriginTransform , OriginRotation , 1.5)
                    if(TheSeq := MotionsSeq_Boss[1]):
                        TheSeq.StoppedEvent.Await()

                Print("Position Reset")
                set Count = 0
                set Count02 = 0

    Motion03()<suspends> : void =
        race :
            block :
                EndTrigger.TriggeredEvent.Await()
                Print("!!race : Finished!!")
            block :
                #使う変数たちを初期化、DPSチェックが終わってるかどうか判別して使うDVを配列にいれる
                set Count = 0
                set Count02 = 0
                var PlayersInVolume : []player = array{}
                set PropsArray = array{}
                set UseDVArray = array{}
                if(DPSChecker <> "Triggered") :
                    Print("UnTrigger")
                    loop :
                        if(Count <= 5 , UseDV := DV_SquirtUp[Count]) :
                            set UseDVArray += array{UseDV}
                            set Count += 1
                        else :
                            break
                        Sleep(0.1)
                else :
                    Print("Triggered")
                    loop :
                        if(Count <= 5 , UseDV := DV_SquirtUp_DPSed[Count]) :
                            set UseDVArray += array{UseDV}
                            set Count += 1
                        else:
                            break
                        Sleep(0.1)

                set Count = 0
                #モーション開始からAoE出現までの時間
                Sleep(4.86)
                #マップ内の全プレイヤーを取得→各プレイヤーに対してボリューム内にいるか確認
                AllPlayers := GetPlayspace().GetPlayers()
                for(Player : AllPlayers ):
                    if(PlayerCheck.IsCounted[Player] , ThePlayer := Player.GetFortCharacter[]) :
                        #trueなら配列にいれとく。そのプレイヤーの座標を元にエフェクト付きBPをSpawnさせる
                        set PlayersInVolume += array{Player}
                        PlayerPos : vector3 = ThePlayer.GetTransform().Translation
                        SpawnPos : vector3 = vector3{X := PlayerPos.X , Y := PlayerPos.Y , Z := 2463.0}
                        #Count番目のダメージボリュームをSpawnPosにテレポート
                        if(TheDV := UseDVArray[Count]) :
                            if(TheDV.TeleportTo[SpawnPos , rotation{}]) :
                                Print("{Count}DamageVolume was Tereported !")
                        #SpawnProp()してそのpropを変数として定義
                        if(DPSChecker <> "Triggered") :
                            var SpawnedProp : ?creative_prop = SpawnProp(VFX_SquirtUp , SpawnPos , rotation{})(0)
                            #SpawnedPropがcreative_propか確認して配列に入れる
                            if(TheProp : creative_prop = SpawnedProp?) :
                                set PropsArray += array{TheProp}
                                Print("PropSpawned!{SpawnPos}")
                            set Count += 1
                        else :
                            var SpawnedProp : ?creative_prop = SpawnProp(VFX_SquirtUp_DPSed , SpawnPos , rotation{})(0)
                            #SpawnedPropがcreative_propか確認して配列に入れる
                            if(TheProp : creative_prop = SpawnedProp?) :
                                set PropsArray += array{TheProp}
                                Print("PropSpawned!{SpawnPos}")
                            set Count += 1

                #AoEが出現してからダメージが入るまでの時間
                Sleep(1.5)
                for(Player : PlayersInVolume) :
                    if(TheDV := UseDVArray[Count02]) :
                        TheDV.Enable()
                        set Count02 += 1

                set Count02 = 0
                #ダメージが有効な時間
                Sleep(0.7)
                for(Player : PlayersInVolume) :
                    if(TheDV := UseDVArray[Count02]) :
                        TheDV.Disable()
                        set Count02 += 1

                set Count02 = 0
                if(TheSeq := MotionsSeq_Boss[4]) :
                    for(Player : PlayersInVolume) :
                        if(Prop := PropsArray[Count02]) :
                            Prop.Dispose()
                            Print("Dispose Prop{Count02}")
                        if(TheDV := UseDVArray[Count02]) :
                            if(TheDV.TeleportTo[OriginPos , rotation{}]) :
                                Print("Dispose DamageVolume{Count02}")
                        set Count02 += 1

    Motion04()<suspends> : void =
        race :
            block :
                EndTrigger.TriggeredEvent.Await()
                Print("!!race : Finished!!")
            block :
                Print("Motion04 spawned !")
                var InitialCount : int = 0
                #時系列 , Sleep()の時間関連はボス戦ドキュメント参照！
                if :
                    CPX0 := AreaSize[0].GetTransform().Translation.X
                    CPX1 := AreaSize[1].GetTransform().Translation.X
                    CPX2 := AreaSize[2].GetTransform().Translation.X
                    CPX3 := AreaSize[3].GetTransform().Translation.X
                    CPY0 := AreaSize[0].GetTransform().Translation.Y
                    CPY1 := AreaSize[1].GetTransform().Translation.Y
                    CPY2 := AreaSize[2].GetTransform().Translation.Y
                    CPY3 := AreaSize[3].GetTransform().Translation.Y

                then :
                    MinX01 := Min(CPX0 , CPX1)
                    MinX23 := Min(CPX2 , CPX3)
                    set MinX = Min(MinX01 , MinX23)

                    MaxX01 := Max(CPX0 , CPX1)
                    MaxX23 := Max(CPX2 , CPX3)
                    set MaxX = Max(MaxX01 , MaxX23)

                    MinY01 := Min(CPY0 , CPY1)
                    MinY23 := Min(CPY2 , CPY3)
                    set MinY = Min(MinY01 , MinY23)

                    MaxY01 := Max(CPY0 , CPY1)
                    MaxY23 := Max(CPY2 , CPY3)
                    set MaxY = Max(MaxY01 , MaxY23)
                Sleep(1.1)
                loop :
                    if(InitialCount < 6) :
                        set RandomPosX = GetRandomFloat(MinX , MaxX)
                        set RandomPosY = GetRandomFloat(MinY , MaxY)
                        SpawnPos : vector3 = vector3{X := RandomPosX , Y := RandomPosY , Z := 2463.0}
                        if(DPSChecker <> "Triggered") :
                            TheDVs : []damage_volume_device = DV_SquirtUp
                            TheVFX : creative_prop_asset = VFX_Fall
                            spawn{Motion04spawnProp(InitialCount , TheDVs , TheVFX , SpawnPos)}
                        else :
                            TheDVs : []damage_volume_device = DV_SquirtUp_DPSed
                            TheVFX : creative_prop_asset = VFX_Fall_DPSed
                            spawn{Motion04spawnProp(InitialCount , TheDVs , TheVFX , SpawnPos)}
                        Print("Spawn Fall")
                        set InitialCount += 1
                    else :
                        break
                    Sleep(0.2)

    Motion04spawnProp(InitialCount : int , TheDVs : []damage_volume_device , TheVFX : creative_prop_asset , SpawnPos : vector3)<suspends> : void =
                if :
                    TheDV := TheDVs[InitialCount]
                then :
                    SpawnedProp : ?creative_prop = SpawnProp(TheVFX , SpawnPos , rotation{})(0)
                    if :
                        TheProp : creative_prop = SpawnedProp?
                        TheDV.TeleportTo[SpawnPos , rotation{}]
                    then :
                        race :
                            block :
                                EndTrigger.TriggeredEvent.Await()
                                TheDV.Disable()
                                TheProp.Dispose()
                                Print("!!race : Finished!!")
                            block :
                                Sleep(4.23)
                                TheDV.Enable()
                                Print("Enable{InitialCount}")
                                Sleep(0.6)
                                TheDV.Disable()
                                Sleep(0.1)
                                TheProp.Dispose()
        
    MotionDPS()<suspends> : void =
        race :
            block :
                EndTrigger.TriggeredEvent.Await()
                DPSCheckSeq_Boss.Stop()
                Print("!!race : Finished!!")
            block :
                #初期化
                Seq_WaterPrison.Play()
                if(PlayerCheck.GetCount() = 1):
                    set DPSHealth = DPSDefaultHealth * 1
                if(PlayerCheck.GetCount() = 2):
                    set DPSHealth = DPSDefaultHealth * 1.5
                if(PlayerCheck.GetCount() = 3):
                    set DPSHealth = DPSDefaultHealth * 2
                if(PlayerCheck.GetCount() = 4):
                    set DPSHealth = DPSDefaultHealth * 2.5
                Print("WaterPrison Health{DPSHealth}")
                set WaterPrisonState = "Unbroken"
                set DPSTeleportPos = OriginPos
                <#set ConvertDPSHealthToFrameMultiplier = 40.0 / DPSHealth
                Print("ConvertDPSHealthToFrameMultiplier ={ConvertDPSHealthToFrameMultiplier}")
                DPSHealthSeq_Boss.Play()
                UI_DPSHealth_Boss.Show()#>

                #VerseUI呼び出し
                set DPS_MaxHealth = DPSHealth
                DPS_Start()
                
                set UseDVArray = array{}
                if(TheDV := DV_SquirtUp_DPSed[0]) :
                    set UseDVArray += array{TheDV}
                #モーション開始から水球に閉じ込めるまでの時間
                Sleep(1.4)
                #DPSHealthSeq_Boss.Pause()
                PlayerList : []player = GetPlayspace().GetPlayers()
                loop :
                    set Count = GetRandomInt(0 , PlayerList.Length - 1)
                    if(PlayerCheck.IsCounted[PlayerList[Count]]) :
                        Print("抽選成功！")
                        break
                    else :
                        Print("再抽選")

                    Sleep(0.1)

                if(ThePlayer := PlayerList[Count] , Char : fort_character = ThePlayer.GetFortCharacter[]) :
                    set DPSTeleportPos = Char.GetTransform().Translation
                    if(DPS01.TeleportTo[DPSTeleportPos , rotation{}]) :
                        Print("閉じ込めたよ")
                    KrepperPos := Krepper_Rotator.GetTransform().Translation
                    if(LookDirection := (DPSTeleportPos - KrepperPos).MakeUnitVector[]) :
                        
                        set Yaw = RadiansToDegrees(ArcTan(LookDirection.Y , LookDirection.X))

                        NewRotation := MakeRotationFromYawPitchRollDegrees(Yaw , Pitch , Roll)
                        Krepper_Rotator.MoveTo(KrepperPos , NewRotation , 2.0)
                        Print("クレッパーこっち向いてるｗ")

                Sleep(7.1)
                DPS_HideBBWidget()
                if(WaterPrisonState <> "Broken") :
                    set WaterPrisonState = "CouldntBreak"
                    #DVを有効化＆キャラの場所にテレポート（45ダメージが3回与えられる）
                    if(TheDV := UseDVArray[0]) :
                        TheDV.SetDamage(100)
                        TheDV.Enable()
                        if(TheDV.TeleportTo[DPSTeleportPos , rotation{}]) :
                            Print("DVTeleport{DPSTeleportPos}")

                Sleep(2.1)
                if(TheDV := UseDVArray[0]) :
                    TheDV.SetDamage(20)
                    TheDV.Disable()
                    if(TheDV.TeleportTo[OriginPos , rotation{}]) :
                #水球とDVを元に戻し、エフェクトをスポーンさせる（水球が壊されていない場合）
                if(WaterPrisonState <> "Broken") :
                    if(DPS01.TeleportTo[OriginPos , rotation{}]) :
                    SpawnedProp : ?creative_prop = SpawnProp(VFX_WaterPrisonExplode , DPSTeleportPos , rotation{})(0)
                    Sleep(2.0)
                    if(TheProp : creative_prop = SpawnedProp?) :
                        TheProp.Dispose()
                        Print("ExplodeVFX Disposed !")
                else :
                    Sleep(2.0)

                set Yaw = 90.0
                OriginTransform := Krepper_Rotator.GetTransform().Translation
                OriginRotation := MakeRotationFromYawPitchRollDegrees(Yaw , Pitch , Roll)
                Krepper_Rotator.MoveTo(OriginTransform , OriginRotation , 1.0)
                Print("元に戻った")

                DPSCheckSeq_Boss.StoppedEvent.Await()

    DPSWeaponCheck(Agent : agent) : void =
        if :
            Player := player[Agent]
            var WeaponValue : float = Weapon_Damage.CalculateDamage[Player]
        then :
            if(SaveDataManager.GetStrBuff(Player)?):
                set WeaponValue = WeaponValue * 1.2
            set DamageValue = WeaponValue

        set DPSHealth -= DamageValue
        Print("NowDPSHealth{DPSHealth}")
        if(DPSHealth > 0.0) :
            DPS_ChangeHPBar(DPSHealth)

        #WaterPrisonのHPが0になった時に1度だけエフェクトを呼び出す
        if(DPSHealth <= 0.0 , WaterPrisonState = "Unbroken") :
            DPS_ChangeHPBar(0.0)
            spawn{OnWaterPrisonBroke()}

    OnWaterPrisonBroke()<suspends>: void =
        race :
            block :
                EndTrigger.TriggeredEvent.Await()
                Print("!!race : Finished!!")
            block :
                set WaterPrisonState = "Broken"
                Seq_WaterPrison.SetPlaybackFrame(339)
                if(DPS01.TeleportTo[OriginPos , rotation{}]) :
                SpawnedProp : ?creative_prop = SpawnProp(VFX_WaterPrisonExplode , DPSTeleportPos , rotation{})(0)
                Print("ExplodeVFX Broken !")
                Sleep(2.0)
                if(TheProp : creative_prop = SpawnedProp?) :
                    TheProp.Dispose()
                    Print("ExplodeVFX Broken Disposed !")

    OnBegin<override>()<suspends>:void=
        Print("Start")
        for(SpawnPad : SpawnPads) :
            SpawnPad.SpawnedEvent.Subscribe(Default)
        BossStartTrigger.TriggeredEvent.Subscribe(StartBossFight)
        DamageTrigger.DamagedEvent.Subscribe(WeaponCheck)
        DPSDamageTrigger.DamagedEvent.Subscribe(DPSWeaponCheck)
        set DefaultHealth = Health
        set DPSDefaultHealth = DPSHealth
        ZoneVolume.AgentExitsEvent.Subscribe(OnExitFromVolume)
        RewardButton.InteractedWithEvent.Subscribe(OnInteractedReward)

    Default(Agent : agent) : void=
        if(Player := player[Agent]):
            if (SaveDataManager.GetWeaponLevel(Player) = 0):
                TeamChanger_Default.ChangeTeamAndClass(Agent)
                Print("チームリセット一般人")
            else:
                TeamChanger_Blader.ChangeTeamAndClass(Agent)
                Print("チームリセット刀")

    #ボスUI
    @editable var IncreaceTime : float = 2.8
    @editable var DivisionCount : int = 100
    var MaxHealth : float = 100.0

    @editable var DPS_IncreaceTime : float = 1.0
    @editable var DPS_DivisionCount : int = 33
    var DPS_MaxHealth : float = 100.0

    @editable var Motion_DivisionCount : int = 6

    var BBWidgetPer : [player]ui_bossbar1 = map{}

    StrToMes<localizes>(String : string) : message = "{String}"

    CheckBBWidget(Player : player):void=
        if(Widget := BBWidgetPer[Player]) :
            Print("ボスWidget作成済み")
            return #作成済みなら何もしない
        if(PlayerUI : player_ui = GetPlayerUI[Player]) :
            NewUI := ui_bossbar1{}
            PlayerUI.AddWidget(NewUI.CreateBBCanvas())
            logic{set BBWidgetPer[Player] = NewUI}
            if(Widget := BBWidgetPer[Player]) :
                Widget.Image_BG_BossBar.SetVisibility(widget_visibility.Collapsed)
                Widget.Image_HP_BossBar.SetVisibility(widget_visibility.Collapsed)
                Widget.Image_BossIcon.SetVisibility(widget_visibility.Collapsed)
                Widget.CB_BG_DPS.SetVisibility(widget_visibility.Collapsed)
                Widget.Image_DPSHP.SetVisibility(widget_visibility.Collapsed)
            Print("ボスWidget作成!")

    #ウィジェットを表示する関数
    ShowBBWidget()<suspends>:void=
        Sleep(0.1)
        for(Player : GetPlayspace().GetPlayers()) :
            if :
                TeamCollection := GetPlayspace().GetTeamCollection()
                BossTeam := TeamCollection.GetTeams()[1]
                TeamCollection.IsOnTeam[Player, BossTeam]
            then :
                CheckBBWidget(Player)
                if(Widget := BBWidgetPer[Player]) :
                    Widget.Image_BG_BossBar.SetVisibility(widget_visibility.Visible)
                    Widget.Image_HP_BossBar.SetVisibility(widget_visibility.Visible)
                    Widget.Image_BossIcon.SetVisibility(widget_visibility.Visible)
                    Widget.TB_Motion.SetText(StrToMes(""))
                    Widget.TB_Motion.SetShadowOpacity(0.0)
                    Widget.TB_Motion.SetVisibility(widget_visibility.Visible)
                spawn{StartIncreace(Player)}
                Print("ウィジェット表示", ?Color := NamedColors.Red)
            else :
                Print("ウィジェット表示失敗", ?Color := NamedColors.Red)

    HideBBWidget():void=
        for(Player : GetPlayspace().GetPlayers()) :
            if(Widget := BBWidgetPer[Player]) :
                Widget.Image_BG_BossBar.SetVisibility(widget_visibility.Collapsed)
                Widget.Image_HP_BossBar.SetVisibility(widget_visibility.Collapsed)
                Widget.Image_BossIcon.SetVisibility(widget_visibility.Collapsed)
                Widget.TB_Motion.SetText(StrToMes(""))
                Widget.TB_Motion.SetShadowOpacity(0.0)
                Widget.TB_Motion.SetVisibility(widget_visibility.Collapsed)

    #ボス戦開始時のHP満たされてくアニメーション関数
    StartIncreace(Player : player)<suspends>:void=
        var SizeX : float = 0.0

        if(Widget := BBWidgetPer[Player]) :
            Widget.Image_HP_BossBar.SetDesiredSize(vector2{X := 0.0, Y := 43.0})
            #シーケンスの再生時間を任意の数に分割してHPバーのサイズが合うように増やしていく
            for(X := 1..DivisionCount) :
                Sleep(IncreaceTime / (DivisionCount*1.0))
                set SizeX = Widget.Image_HP_BossBar.GetDesiredSize().X + 817.0 / (DivisionCount*1.0)
                Widget.Image_HP_BossBar.SetDesiredSize(vector2{X := SizeX, Y := 43.0})
            Print("Increaceアニメーション終了")

    #Healthを引数としてImage_HP_BossBarを対応するサイズに変更する関数
    ChangeHPBar(HealthValue : float):void=
        SizeX := 817.0 * (HealthValue / MaxHealth)
        TeamCollection := GetPlayspace().GetTeamCollection()

        for(Player : GetPlayspace().GetPlayers()) :
            if :
                BossTeam := TeamCollection.GetTeams()[1]
                TeamCollection.IsOnTeam[Player, BossTeam]
            then :
                if(Widget := BBWidgetPer[Player]) :
                    Widget.Image_HP_BossBar.SetDesiredSize(vector2{X := SizeX, Y := 43.0})
                    
    DPS_HideBBWidget():void=
        for(Player : GetPlayspace().GetPlayers()) :
            if(Widget := BBWidgetPer[Player]) :
                Widget.CB_BG_DPS.SetVisibility(widget_visibility.Collapsed)
                Widget.Image_DPSHP.SetVisibility(widget_visibility.Collapsed)

    #DPS開始時のHP満たされてくアニメーション関数
    DPS_Start():void=
        TeamCollection := GetPlayspace().GetTeamCollection()

        for(Player : GetPlayspace().GetPlayers()) :
            if :
                BossTeam := TeamCollection.GetTeams()[1]
                TeamCollection.IsOnTeam[Player, BossTeam]
                Widget := BBWidgetPer[Player]
            then :
                Widget.CB_BG_DPS.SetVisibility(widget_visibility.Visible)
                Widget.Image_DPSHP.SetVisibility(widget_visibility.Visible)
                spawn{DPS_StartIncreace(Player)}

    SetMotionText(Index : int, Frame : int)<suspends>:void=
        var Opacity : float = 0.0
        sync :
            block :
                for(X := 1..Motion_DivisionCount) :
                    Sleep(0.3 / (Motion_DivisionCount*1.0))
                    for(Player : GetPlayspace().GetPlayers()) :
                        if :
                            TeamCollection := GetPlayspace().GetTeamCollection()
                            BossTeam := TeamCollection.GetTeams()[1]
                            TeamCollection.IsOnTeam[Player, BossTeam]
                            Widget := BBWidgetPer[Player]
                        then :
                            if(Index = 0) :
                                Widget.TB_Motion.SetText(StrToMes(""))
                            if(Index = 1) :
                                Widget.TB_Motion.SetText(StrToMes("Aqua Blast"))
                            if(Index = 2) :
                                Widget.TB_Motion.SetText(StrToMes("Land Slide"))
                            if(Index = 3) :
                                Widget.TB_Motion.SetText(StrToMes("Squirt Up"))
                            if(Index = 4) :
                                Widget.TB_Motion.SetText(StrToMes("Fall"))
                            if(NewOpacity := type{_X:float where 0.000000 <= _X, _X <= 1.000000}[Opacity + (1.0  /(Motion_DivisionCount*1.0))]) :
                                Widget.TB_Motion.SetTextOpacity(NewOpacity)
                                Widget.TB_Motion.SetShadowOpacity(NewOpacity)
                                Print("{NewOpacity}", ?Color := NamedColors.Red)
                                set Opacity = NewOpacity
            block :
                race :
                    block :
                        EndTrigger.TriggeredEvent.Await()
                        Print("!!race : Finished!!")
                    block :
                        Sleep((Frame*1.0) / 30.0)
                        set Opacity = 1.0
                        for(X := 1..Motion_DivisionCount) :
                            Sleep(0.3 / (Motion_DivisionCount*1.0))
                            for(Player : GetPlayspace().GetPlayers()) :
                                if :
                                    TeamCollection := GetPlayspace().GetTeamCollection()
                                    BossTeam := TeamCollection.GetTeams()[1]
                                    TeamCollection.IsOnTeam[Player, BossTeam]
                                    Widget := BBWidgetPer[Player]
                                then :
                                    if(NewOpacity := type{_X:float where 0.000000 <= _X, _X <= 1.000000}[Opacity - (1.0  /(Motion_DivisionCount*1.0))]) :
                                        Widget.TB_Motion.SetTextOpacity(NewOpacity)
                                        Widget.TB_Motion.SetShadowOpacity(NewOpacity)
                                        set Opacity = NewOpacity

    DPS_StartIncreace(Player : player)<suspends>:void=
        var SizeX : float = 0.0

        if(Widget := BBWidgetPer[Player]) :
            Widget.Image_DPSHP.SetDesiredSize(vector2{X := 0.0, Y := 11.943165})
            #シーケンスの再生時間を任意の数に分割してHPバーのサイズが合うように増やしていく
            for(X := 1..DPS_DivisionCount) :
                Sleep(DPS_IncreaceTime / (DPS_DivisionCount*1.0))
                set SizeX = Widget.Image_DPSHP.GetDesiredSize().X + 595.5 / (DPS_DivisionCount*1.0)
                Widget.Image_DPSHP.SetDesiredSize(vector2{X := SizeX, Y := 11.943165})
            Print("Increaceアニメーション終了")

    #DPSHealthを引数としてImage_DPSHPを対応するサイズに変更する関数
    DPS_ChangeHPBar(HealthValue : float):void=
        SizeX := 595.5 * (HealthValue / DPS_MaxHealth)
        TeamCollection := GetPlayspace().GetTeamCollection()

        for(Player : GetPlayspace().GetPlayers()) :
            if :
                BossTeam := TeamCollection.GetTeams()[1]
                TeamCollection.IsOnTeam[Player, BossTeam]
            then :
                if(Widget := BBWidgetPer[Player]) :
                    Widget.Image_DPSHP.SetDesiredSize(vector2{X := SizeX, Y := 11.943165})
ui_bossbar1:= class:
    #通常のHPバー
    Image_BG_BossBar : texture_block = texture_block:
        DefaultImage := Objects.UI_Healthbar_outside
    Image_HP_BossBar : texture_block = texture_block:
        DefaultImage := Objects.UI_Healthbar_inside
        DefaultDesiredSize := vector2{X := 0.0, Y := 43.0}#最大 X := 817.0
    Image_BossIcon : texture_block = texture_block:
        DefaultImage := Texture.boss_name01
    #DPSのHPバー
    CB_BG_DPS : color_block = color_block:
        DefaultColor := color{R:=0.036889,G:=0.036889,B:=0.036889}
        DefaultOpacity := 1.0
    Image_DPSHP : texture_block = texture_block:
        DefaultImage := Texture.T_DPSHP
        DefaultTint := color{R:=0.0,G:=0.342978,B:=1.0}
        DefaultDesiredSize := vector2{X := 0.0, Y := 11.943165}#最大 X := 595.5
    #モーション名テキスト
    TB_Motion : text_block = text_block:
        DefaultTextColor := NamedColors.White
        DefaultShadowOffset := option{vector2{X:=1.614798,Y:=1.960874}}
        DefaultShadowOpacity := 0.0

    CreateBBCanvas<public>(): canvas =
        var CanvasSlots : []canvas_slot = array{}

        set CanvasSlots += array:
            #通常のHPバー
            canvas_slot:
                Anchors := anchors:
                    Minimum := vector2{X := 0.5, Y := 0.08}
                    Maximum := vector2{X := 0.5, Y := 0.08}
                Offsets := margin{Left := 0.0, Top := 0.0, Right := 880.0, Bottom := 43.0}
                SizeToContent := false
                Alignment := vector2{X := 0.5, Y := 0.5}
                Widget := Image_BG_BossBar
            canvas_slot:
                Anchors := anchors:
                    Minimum := vector2{X := 0.5, Y := 0.08}
                    Maximum := vector2{X := 0.5, Y := 0.08}
                Offsets := margin{Left := -409.000000, Top := 0.0, Right := 0.0, Bottom := 0.0}
                SizeToContent := true
                Alignment := vector2{X := 0.0, Y := 0.5}
                Widget := Image_HP_BossBar
            canvas_slot:
                Anchors := anchors:
                    Minimum := vector2{X := 0.5, Y := 0.08}
                    Maximum := vector2{X := 0.5, Y := 0.08}
                Offsets := margin{Left := 27.934517, Top := -15.321068, Right := 409.589447, Bottom := 70.322205}
                SizeToContent := false
                Alignment := vector2{X := 0.0, Y := 1.0}
                Widget := Image_BossIcon
            #DPSのHPバー
            canvas_slot:
                Anchors := anchors:
                    Minimum := vector2{X := 0.5, Y := 0.08}
                    Maximum := vector2{X := 0.5, Y := 0.08}
                Offsets := margin{Left := 0.0, Top := 68.778671, Right := 598.355835, Bottom := 15.257175}
                SizeToContent := false
                Alignment := vector2{X := 0.5, Y := 0.5}
                Widget := CB_BG_DPS
            canvas_slot:
                Anchors := anchors:
                    Minimum := vector2{X := 0.5, Y := 0.08}
                    Maximum := vector2{X := 0.5, Y := 0.08}
                Offsets := margin{Left := -297.941925, Top := 68.599998, Right := 0.0, Bottom := 0.0}
                SizeToContent := true
                Alignment := vector2{X := 0.0, Y := 0.5}
                Widget := Image_DPSHP
            #モーション名テキスト
            canvas_slot:
                Anchors := anchors:
                    Minimum := vector2{X := 0.5, Y := 0.08}
                    Maximum := vector2{X := 0.5, Y := 0.08}
                Offsets := margin{Left := 0.0, Top := 108.259071, Right := 0.0, Bottom := 0.0}
                SizeToContent := true
                Alignment := vector2{X := 0.5, Y := 0.5}
                Widget := TB_Motion

        Canvas : canvas = canvas:
            Slots := CanvasSlots
        return Canvas