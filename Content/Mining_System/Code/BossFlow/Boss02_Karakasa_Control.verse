using { /Fortnite.com/Devices }
using { /Verse.org/Simulation }
using { /UnrealEngine.com/Temporary/Diagnostics }
using { /Fortnite.com/Characters }
using { /UnrealEngine.com/Assets }
using { /UnrealEngine.com/Temporary/SpatialMath }
using { /Verse.org/Random }
using { /Verse.org/Assets }
using { /Verse.org/Colors }
using { /UnrealEngine.com/Temporary/UI }
using { /Fortnite.com/UI }
using { Pet }
using { Weapon }

Boss02_Karakasa_Control<public> := class(boss_device_base):
    BossImage<override> : texture = Boss_Textures.karakasaobake

    var Pet_Manager : pet_manager = pet_manager{}
    #ゲームの基本ロジック用シーケンス
    @editable
    StartSeq_Boss : cinematic_sequence_device = cinematic_sequence_device{}

    @editable
    EndSeq_Boss : cinematic_sequence_device = cinematic_sequence_device{}

    @editable
    TimeOverSeq_Boss : cinematic_sequence_device = cinematic_sequence_device{}

    @editable
    DamagedEffectSeq_Boss : cinematic_sequence_device = cinematic_sequence_device{}

    @editable
    MotionsSeq_Boss : []cinematic_sequence_device = array{}
    #DPS関連のシーケンス
    @editable
    DPSCheckSeq_Boss : cinematic_sequence_device = cinematic_sequence_device{}

    @editable
    DPSStunSeq_Bos : cinematic_sequence_device = cinematic_sequence_device{}

    @editable
    DPSedRingSeq_Boss : cinematic_sequence_device = cinematic_sequence_device{}
    #ゲームの基本ロジック用デバイス
    @editable
    PlayerCheck : player_counter_device = player_counter_device{}

    @editable
    ZoneVolume : volume_device = volume_device{}

    @editable
    Boss_Timer : timer_device = timer_device{}

    @editable
    StartTrigger : trigger_device = trigger_device{}
    @editable
    EndTrigger : trigger_device = trigger_device{}

    @editable
    EnterTeleporter : teleporter_device = teleporter_device{}
    @editable
    ExitTeleporter : teleporter_device = teleporter_device{}
    @editable
    DefaultTeleporter : teleporter_device = teleporter_device{}

    @editable
    AreaSize : []creative_prop = array{}

    @editable
    BGM : audio_player_device = audio_player_device{}

    @editable
    TeamChanger_Default : class_and_team_selector_device = class_and_team_selector_device{}
    @editable
    TeamChanger_Blader : class_and_team_selector_device = class_and_team_selector_device{}

    @editable
    RewardButton : button_device = button_device{}

    @editable
    TimeLimit_Second : float = 300.0

    #ペットの能力関連
    @editable
    VFX_Powerup : damage_amplifier_powerup_device = damage_amplifier_powerup_device{}
    @editable
    VFX_Lightning2 : creative_prop_asset = DefaultCreativePropAsset
    @editable
    SFX_OnHit : audio_player_device = audio_player_device{}
    @editable
    SFX_OnHeal : audio_player_device = audio_player_device{}
    @editable
    Zeus_Granter : item_granter_device = item_granter_device{}
    @editable
    Zeus_Remover : item_remover_device = item_remover_device{}

    var IsActivable_Powerup : [agent]int = map{}#0 = 起動不可, 1 = 15s待機中, 2 = 起動可能
    #攻撃時のダメージ判定
    var Weapon_Damage : weapon_damage = weapon_damage{}

    #ボスの体力
    @editable
    var Health : float = 100.0
    
    @editable
    var DPSHealth : float = 10.0
    #ダメージ判定用の小道具操作機
    @editable
    DamageTrigger  : prop_manipulator_device = prop_manipulator_device{}
    #プレイヤーにダメージを与えるデバイス
    @editable
    Explosives : []explosive_device = array{}

    @editable
    DVArray : []damage_volume_device = array{}

    @editable
    MovementModulator : movement_modulator_device = movement_modulator_device{}
    #ダメージ判定用の条件付きボタン配列
    @editable
    WeaponCheckers : []conditional_button_device = array{}
    #体力バーを表示するHUDメッセージデバイス
    @editable
    UI_GotReward : hud_message_device = hud_message_device{}
    #カラカサの回転を制御するprop
    @editable
    Kara_Rotator : creative_prop = creative_prop{}
    #エフェクト類
    @editable
    VFX_Rolling : creative_prop_asset = DefaultCreativePropAsset

    @editable
    VFX_WindAttack_Slash : creative_prop_asset = DefaultCreativePropAsset

    @editable
    VFX_WindAttack_Explode : creative_prop_asset = DefaultCreativePropAsset

    @editable
    SaveDataManager : PlayerData.save_data_manager = PlayerData.save_data_manager{}

    @editable
    MiningDamageDevice : MiningObject.mining_damage_device = MiningObject.mining_damage_device{}

    #ダメージチェックモーションが起動したか "UnTrigger" = 未起動 , "Trigger" = 起動中 , "Triggered" = 起動済み
    var DPSChecker : []char = "UnTrigger"
    #再生するモーション番号0~n , StartSeq = -1 , EndSeq = 999
    var MotionValue : int = -1

    var DefaultHealth : float = - 100.0
    var DPSDefaultHealth : float = - 100.0

    var DamageCheckHealth : float = 50.0
    #プレイヤーの攻撃力初期値
    var DamageValue : float = 0.0

    #"NotActive" = ゾーン内に誰もいない , "BeforeGame" = 誰かがゾーンに入ったけど最初のムービー中 , "InGame" = ボス戦中
    var BossState<public> : []char = "NotActive"

    var DPSChecking :logic = false

    var Count : int = 0

    var Count02 : int = 0

    var PropsArray : []creative_prop = array{}

    var UseDVArray : []damage_volume_device = array{}
    #"Unbroken" = クレッパーがDPS技を出すまでのstateで、水球が壊れていない状態
    #"CouldntBreak" = クレッパーがDPS技を出してからのstateで、水球が壊れていない状態（壊せなかった）
    #"Broken" = 水球が壊れた状態
    var ExplodeState : []char = "Unbroken"

    var OriginPos : vector3 = vector3{X := 0.0 , Y := 0.0 , Z := 0.0}
    var DPSTeleportPos : vector3 = vector3{X := 0.0 , Y := 0.0 , Z := 0.0}
    #ボスフィールドの座標最大値最小値
    var MinX : float = 0.0
    var MaxX : float = 0.0
    var MinY : float = 0.0
    var MaxY : float = 0.0

    var RandomPosX : float = 0.0
    var RandomPosY : float = 0.0
    var Yaw : float = 90.0
    var Pitch : float = 0.0
    var Roll : float = 0.0

    var Speed : float = 0.0
    var Angle : float = 0.0
    var Direction : vector3 = vector3{X := 0.0 , Y := 0.0 , Z := 0.0}
    var SpawnPos : vector3 = vector3{X := 0.0 , Y := 0.0 , Z := 0.0}
    var NewPos : vector3 = vector3{X := 0.0 , Y := 0.0 , Z := 0.0}

    var IsGotReward : [agent]logic = map{}

    ReturnBossState<public>():[]char=
        BossState

    StartBossFight<override>(Agent : ?agent) : void =
        #ボス戦のチームにいるプレイヤーをテレポートさせ、全員テレポートさせたらモーション開始
        if(Agent02 := Agent?) :
            for(ThePlayer : GetPlayspace().GetPlayers()) :
                if :
                    TeamCollection := GetPlayspace().GetTeamCollection()
                    BossTeam := TeamCollection.GetTeams()[2]
                    TeamCollection.IsOnTeam[ThePlayer , BossTeam]
                then :
                    EnterTeleporter.Teleport(ThePlayer)
                    Print("TeleportTo BossField")
            if(BossState = "NotActive") :
                set BossState = "BeforeGame"
                spawn{PlayMotions(Agent02)}
                spawn{TimerCheck(Agent02)}

    PlayMotions(Agent02 : agent)<suspends> : void = 
        loop :
            #プレイヤーカウンター内にプレイヤーがいるとき
            if(0 < PlayerCheck.GetCount()) :
                if(MotionValue = -1) :
                    Print("StartSeq.Play")
                    ShowBBWidget()
                    StartSeq_Boss.Play(Agent02)
                    set MotionValue += 1
                    BGM.Play()
                    #ボリューム内のプレイヤーを取得してタイマー取り付け
                    set Count = 0
                    PlayerList : []player = GetPlayspace().GetPlayers()
                    for(Players : PlayerList) :
                        if(PlayerCheck.IsCounted[PlayerList[Count]] , ThePlayer := PlayerList[Count]) :
                            Boss_Timer.Start(ThePlayer)
                            spawn{CheckAbility_01(ThePlayer)}
                            Print("タイマー取り付け！")
                        set Count += 1
                        Sleep(0.1)
                    set Count = 0
                    StartSeq_Boss.StoppedEvent.Await()
                    #各変数を初期化
                    if(PlayerCheck.GetCount() = 1):
                        set Health = DefaultHealth * 1
                    if(PlayerCheck.GetCount() = 2):
                        set Health = DefaultHealth * 1.5
                    if(PlayerCheck.GetCount() = 3):
                        set Health = DefaultHealth * 2
                    if(PlayerCheck.GetCount() = 4):
                        set Health = DefaultHealth * 2.5

                    set MaxHealth = Health
                    set DamageCheckHealth = Health * 0.7
                    set DPSChecking = false
                    set ExplodeState = "Unbroken"
                    set DPSChecker = "UnTrigger"
                    set BossState = "InGame"

                #DPSモーション
                else if(DPSChecker = "Trigger" , BossState <> "NotActive"):
                    Print("DPSCheck")
                    spawn{DPSTimer()}
                    DPSCheckSeq_Boss.Play()
                    race :
                        block :
                            EndTrigger.TriggeredEvent.Await()
                            set ExplodeState = "Broken"
                            DPSCheckSeq_Boss.Stop()
                            Print("DPS溜め直前にボス撃破")
                        block :
                            Sleep(1.5)
                            set DPSChecking = true
                            Print("set DPSChecking = true")
                    DPSCheckSeq_Boss.StoppedEvent.Await()
                    DPS_HideBBWidget()
                    if(ExplodeState = "Broken" , BossState <> "NotActive") :
                        race :
                            block :
                                EndTrigger.TriggeredEvent.Await()
                                DPSStunSeq_Bos.Stop()
                                Print("!!race : Finished!!")
                            block :
                                DPSStunSeq_Bos.Play()
                                DPSStunSeq_Bos.StoppedEvent.Await()
                                DPSedRingSeq_Boss.Play()
                    set DPSChecker = "Triggered"

                #(通常モーションの数) - 1 まで。カラカサはモーションが 00 ~ 05 なので 0 ~ 4 まで格納される
                else if(0 <= MotionValue <= MotionsSeq_Boss.Length -2) :
                    if(SelectedMotion := MotionsSeq_Boss[MotionValue] , DPSChecker = "UnTrigger" or "Triggered") :
                        #----------モーション名テキスト変更----------
                        var Frame : int = 0
                        if(MotionValue = 1):
                            set Frame = 174
                        if(MotionValue = 2):
                            set Frame = 240
                        if(MotionValue = 3):
                            set Frame = 152
                        if(MotionValue = 4):
                            set Frame = 345
                        spawn{SetMotionText(MotionValue, Frame)}
                        #-------------------------------------------
                        if(MotionValue = 0) :
                        if(MotionValue = 1) :
                            spawn{Motion01()}
                            if(DPSChecker = "Triggered") :
                                spawn{Motion01_DPSed()}
                        if(MotionValue = 3) :
                            spawn{Motion03()}

                        SelectedMotion.Play()
                        Print("{MotionValue}")
                        SelectedMotion.StoppedEvent.Await()

                        set MotionValue += 1

                #最後のモーションの時。カラカサは00 ~ 05 なので最後のモーションは 5
                else if(MotionValue = MotionsSeq_Boss.Length -1) :
                    if(SelectedMotion := MotionsSeq_Boss[MotionValue] , DPSChecker = "UnTrigger" or "Triggered") :
                        #----------モーション名テキスト変更----------
                        var Frame : int = 0
                            set Frame = 280
                        spawn{SetMotionText(MotionValue, Frame)}
                        #-------------------------------------------
                        Print("{MotionValue}")
                        SelectedMotion.Play()
                        Print("{MotionValue}DPS前")
                        set MotionValue = 0
                        SelectedMotion.StoppedEvent.Await()

                else if(MotionValue >= 999):
                    EndSeq_Boss.Play(Agent02)
                    set MotionValue = -1
                    EndSeq_Boss.StoppedEvent.Await()
                    #報酬部屋にテレポートさせる処理
                    ExitPos := ExitTeleporter.GetTransform().Translation
                    ExitRotation := ExitTeleporter.GetTransform().Rotation
                    for(ThePlayer : GetPlayspace().GetPlayers()) :
                        if :
                            TeamCollection := GetPlayspace().GetTeamCollection()
                            BossTeam := TeamCollection.GetTeams()[2]
                            TeamCollection.IsOnTeam[ThePlayer , BossTeam]
                            Char : fort_character = ThePlayer.GetFortCharacter[]
                        then :
                            ExitTeleporter.Teleport(ThePlayer)
                            XP_ElimedBoss.Award(ThePlayer)
                            if(Agent : agent = agent[ThePlayer]):
                                if (SaveDataManager.GetWeaponLevel(ThePlayer) = 0):
                                    TeamChanger_Default.ChangeTeamAndClass(Agent)
                                    Print("チームリセット一般人,ゾーン")
                                else:
                                    TeamChanger_Blader.ChangeTeamAndClass(Agent)
                                    Print("チームリセット刀,ゾーン")
                    #ここまで
                    BGM.Stop()
                    Print("break , End")
                    break

            #プレイヤーカウンター内にプレイヤーがいないとき
            else if(0 = PlayerCheck.GetCount()) :
                set BossState = "NotActive"
                HideBBWidget()
                DPS_HideBBWidget()
                set MotionValue = -1
                BGM.Stop()
                EndTrigger.Trigger()
                Print("break , No Players")
                break

            Sleep(0.1)

    EndMotions() : void =
        EndTrigger.Trigger()
        if(1 <= MotionValue <= MotionsSeq_Boss.Length -1 , SelectedMotion := MotionsSeq_Boss[MotionValue]) :
                SelectedMotion.Stop()

        else if(MotionValue = 0 , SelectedMotion := MotionsSeq_Boss[MotionsSeq_Boss.Length -1]) :
                SelectedMotion.Stop()

        set MotionValue = 999
        Print("Stopped Motions")

    WeaponCheck(Agent:agent) : void =
        spawn{DamageEffect()}
        Print("PropDamaged")
        if :
            BossState = "InGame"
            Player := player[Agent]
            var WeaponValue : float = Weapon_Damage.CalculateDamage[Player]
        then :
            if(SaveDataManager.GetStrBuff(Player)?):
                set WeaponValue = WeaponValue * 1.2
                Print("恐竜バフ", ?Color := NamedColors.Red, ?Duration := 3.0)
            set DamageValue = WeaponValue

        spawn{CheckAbility_02(Agent)}

        if(DPSChecking <> true) :
            set Health -= DamageValue
        else :
            spawn{DPSWeaponCheck()}

        Print("Health = {Health}")
        if(BossState = "InGame") :
            ChangeHPBar(Health)

        #体力が半分以下になったとき1度だけDPSCheckerをTriggerにset
        if(BossState = "InGame" , Health <= DamageCheckHealth , DPSChecker = "UnTrigger") :
            Print("Trigger")
            set DPSChecker = "Trigger"

        #体力が0以下になったときにMotionValueを999にsetしてボス戦終了シーケンスを再生
        if(BossState = "InGame" , Health <= 0.0) :
            Print("Health <= 0")
            set BossState = "NotActive"
            HideBBWidget()
            DPS_HideBBWidget()
            EndMotions()
##ここからペット能力##
    CheckAbility_01(Agent : agent)<suspends>:void=
        Print("Spawned CheckAbility_01")
        #ボス戦開始時にクレッパーとエンエンガマの能力チェック
        if :
            Char : fort_character = Agent.GetFortCharacter[]
            Player : player = player[Agent]
            Index := Pet_Manager.GetPetIndexInFirstSlot[Player]
            Upgrade := Pet_Manager.GetPetUpgradeInFirstSlot[Player]
            PetSup := SaveDataManager.GetSkillLevel[Player , 14]
            PetSup = 1
        then :
            if(Upgrade >= 1) :
                Print("Having Ability")
                if(Index = 1) :
                    if(Upgrade = 1) :
                        Char.SetShield(50.0)
                        Print("Kure 30~69lv")
                    if(Upgrade = 2) :
                        Char.SetShield(100.0)
                        Print("Kure 70~??lv")
                if(Index = 5) :
                    if(not IsActivable_Powerup[Agent]) :
                        logic{set IsActivable_Powerup[Agent] = 0}
                    race :
                        block :
                            EndTrigger.TriggeredEvent.Await()
                            logic{set IsActivable_Powerup[Agent] = 0}
                            Print("!!race : Finished!!")
                        block :
                            Sleep(15.0)
                            VFX_Powerup.SetDuration(-1.0)
                            VFX_Powerup.Pickup(Agent)
                            logic{set IsActivable_Powerup[Agent] = 2}

                if(Index = 2) :
                    Zeus_Granter.GrantItem(Agent)

    CheckAbility_02(Agent : agent)<suspends>:void=
        #ボスにダメージを与えた時にカラカサ,オオバンモメン,タチガラスの能力をチェック
        if :
            Char : fort_character = Agent.GetFortCharacter[]
            Player : player = player[Agent]
            Index := Pet_Manager.GetPetIndexInFirstSlot[Player]
            Upgrade := Pet_Manager.GetPetUpgradeInFirstSlot[Player]
            PetSup := SaveDataManager.GetSkillLevel[Player , 14]
            PetSup = 1
        then :
            if(Upgrade >= 1) :
                if(Index = 4) :
                    var Threshold : int = 10
                    if(Upgrade = 1 ) :
                        set Threshold = 10
                    if(Upgrade = 2 ) :
                        set Threshold = 15
                    if(BossState = "InGame" , GetRandomInt(0 , 100) <= Threshold) :
                        SpawnedProp : ?creative_prop = SpawnProp(VFX_Lightning2 , Kara_Rotator.GetTransform().Translation , rotation{})(0)
                        if(PlayerCheck.GetCount() = 1):
                            set DamageValue = DamageValue + DefaultHealth * 1 * 0.03
                        if(PlayerCheck.GetCount() = 2):
                            set DamageValue = DamageValue + DefaultHealth * 1.5 * 0.03
                        if(PlayerCheck.GetCount() = 3):
                            set DamageValue = DamageValue + DefaultHealth * 2 * 0.03
                        if(PlayerCheck.GetCount() = 4):
                            set DamageValue = DamageValue + DefaultHealth * 2.5 * 0.03
                        Print("Kara {Threshold}%")
                        Sleep(0.1)
                        SFX_OnHit.Play(Agent)
                        Sleep(1.0)
                        if(TheProp : creative_prop = SpawnedProp?) :
                            TheProp.Dispose()

                if(Index = 5) :
                    if(IsActivable_Powerup[Agent] = 2) :
                        SFX_OnHit.Play(Agent)
                        if(Upgrade = 1) :
                            set DamageValue *= 3.0
                            Print("Enen 30~69lv")
                        if(Upgrade = 2) :
                            set DamageValue *= 5.0
                            Print("Enen 70~??lv")
                        VFX_Powerup.SetDuration(0.1)
                        VFX_Powerup.Pickup(Agent)
                        logic{set IsActivable_Powerup[Agent] = 0}
                        spawn{CheckAbility_01(Agent)}

                if(Index = 3) :
                    var NewHealth : float = 0.0
                    if(Upgrade = 1 , Char.GetHealth() <= 99.0) :
                        set NewHealth = Char.GetHealth() + 1.0
                        Char.SetHealth(NewHealth)
                        SFX_OnHeal.Play(Agent)
                        Print("Ooban 0~69lv")
                    if(Upgrade = 2 , Char.GetHealth() <= 98.0) :
                        set NewHealth = Char.GetHealth() + 2.0
                        Char.SetHealth(NewHealth)
                        SFX_OnHeal.Play(Agent)
                        Print("Ooban 70~??lv")

                if(Index = 2) :
                    if :
                        BossState = "InGame"
                        Zeus := WeaponCheckers[0]
                        Zeus.IsHoldingItem[Agent , 1]
                        var WeaponValue : float = Weapon_Damage.CalculateDamage[Player]
                    then :
                        if(SaveDataManager.GetStrBuff(Player)?):
                            set WeaponValue = WeaponValue * 1.2
                        var Damage_Multiplier : float = 2.0
                        if(Upgrade = 1) :
                            set Damage_Multiplier = 1.0
                            Print("Tachi 30~69lv")
                        if(Upgrade = 2) :
                            set Damage_Multiplier = 1.5
                            Print("Tachi 70~??lv")
                        set DamageValue = WeaponValue * Damage_Multiplier

    ##ここまでペット能力##

    DamageEffect()<suspends> : void =
        if(BossState = "InGame") :
            DamagedEffectSeq_Boss.Play()
            Print("Spawned Damage Effect!")

    TimerCheck(Agent : agent)<suspends>: void =
        race :
            block :
                EndTrigger.TriggeredEvent.Await()
                Print("Timer Stop")
            block :
                var IsTimerChecked : logic = false
                Sleep(TimeLimit_Second)
                for(ThePlayer : GetPlayspace().GetPlayers()) :
                    if(PlayerCheck.IsCounted[ThePlayer] , Char : fort_character = ThePlayer.GetFortCharacter[]) :
                        Char.SetVulnerability(false)
                        Char.PutInStasis(stasis_args{AllowTurning := false , AllowFalling := false , AllowEmotes := false})
                        Print("Freeze & Valner")
                        if(IsTimerChecked = false) :
                            set IsTimerChecked = true
                            TimeOverSeq_Boss.Play(ThePlayer)
                            Print("Play")
                TimeOverSeq_Boss.StoppedEvent.Await()
                for(ThePlayer : GetPlayspace().GetPlayers()) :
                    if(PlayerCheck.IsCounted[ThePlayer] , Char : fort_character = ThePlayer.GetFortCharacter[]) :
                        Char.ReleaseFromStasis()
                        DefaultTeleporter.Teleport(ThePlayer)
                        Char.SetVulnerability(true)
                        Print("Player TP & Release")


    OnExitFromVolume(Agent : agent) : void =
        Boss_Timer.Reset(Agent)
        if(PlayerCheck.GetCount() = 0) :
            HideBBWidget()
            DPS_HideBBWidget()
        Zeus_Remover.Remove(Agent)
        Print("Timer Completed !")

    OnInteractedReward(Agent : agent):void=
        var GetSummonIndex : int = 0
        if(not IsGotReward[Agent]) :
            logic{set IsGotReward[Agent] = false}
        if(IsGotReward[Agent] = false) :
            logic{set IsGotReward[Agent] = true}
            #報酬与える処理#
            if(Player := player[Agent]):

                # 強化報酬報酬
                RewardPowerupStone := PlayerData.big_number{Value := 600.0, Exponent := 0}

                # 召喚石報酬
                RandomReward := GetRandomFloat(0.0, 100.0)
                SummonReward := PlayerData.big_number{Value := 2250.0}

                # 通常リソース報酬
                GetDamage :=MiningDamageDevice.GetNormalDamageAmount(Player).Amount
                RewardResource := PlayerData.BigNumber_MultiplyByBigNumber(GetDamage, PlayerData.big_number{Value := 1125.0})

                # 報酬付与
                SaveDataManager.AddResourcePowerupStone(Player, RewardPowerupStone)
                SaveDataManager.AddResourceStone(Player, RewardResource)

                Print("RewardPowerupStone : {PlayerData.BigNumber_FormatDisplay(RewardPowerupStone)}")
                Print("RewardResource : {PlayerData.BigNumber_FormatDisplay(RewardResource)}")
                Print("RandomReward : {RandomReward}")
                
                if(RandomReward < 76.5):
                    SaveDataManager.AddResourceSummonStoneCommon(Player,SummonReward)
                    set GetSummonIndex = 0
                    Print("Reward is SummonStoneCommon!!")
                if(76.5 <= RandomReward < 94.5):
                    SaveDataManager.AddResourceSummonStoneUncommon(Player,SummonReward)
                    set GetSummonIndex = 1
                    Print("Reward is SummonStoneUncommon!!")
                if(94.5 <= RandomReward < 98.8):
                    SaveDataManager.AddResourceSummonStoneRare(Player,SummonReward)
                    set GetSummonIndex = 2
                    Print("Reward is SummonStoneRare!!")
                if(98.8 <= RandomReward < 99.8):
                    SaveDataManager.AddResourceSummonStoneEpic(Player,SummonReward)
                    set GetSummonIndex = 3
                    Print("Reward is SummonStoneEpic!!")
                if(99.8 <= RandomReward <= 100.0):
                    SaveDataManager.AddResourceSummonStoneLegendary(Player,SummonReward)
                    set GetSummonIndex = 4
                    Print("Reward is SummonStoneLegendary!!")

            spawn{BackToDefault(Agent)}
            spawn{RewardNiagara(Agent, GetSummonIndex)}
            Print("報酬付与！")
        else :
            UI_GotReward.Show(Agent)
            Print("報酬付与済み")
            
    BackToDefault(Agent : agent)<suspends>:void=
        Sleep(5.0)
        DefaultTeleporter.Teleport(Agent)
        logic{set IsGotReward[Agent] = false}
        Print("TeleportTo Default")
            
    DPSWeaponCheck()<suspends> : void =
        set DPSHealth -= DamageValue
        Print("NowDPSHealth{DPSHealth}")
        if(DPSHealth > 0.0) :
            DPS_ChangeHPBar(DPSHealth)

        #DPSのHPが0になった時に1度だけエフェクトを呼び出す
        if(DPSHealth <= 0.0 , ExplodeState = "Unbroken") :
            set DPSChecking = false
            set ExplodeState = "Broken"
            Print("Broken")
            DPS_ChangeHPBar(0.0)
            DPSCheckSeq_Boss.Stop()

    DPSTimer()<suspends> : void =
        if(PlayerCheck.GetCount() = 1):
            set DPSHealth = DPSDefaultHealth * 1
        if(PlayerCheck.GetCount() = 2):
            set DPSHealth = DPSDefaultHealth * 1.5
        if(PlayerCheck.GetCount() = 3):
            set DPSHealth = DPSDefaultHealth * 2
        if(PlayerCheck.GetCount() = 4):
            set DPSHealth = DPSDefaultHealth * 2.5
            
        set DPS_MaxHealth = DPSHealth
        DPS_Start()
        Sleep(10.0)
        Print("DPSHealthCheck")
        if(ExplodeState <> "Broken") :
            set DPSChecking = false

    DPSRing_Looper(): void =
        if(DPSChecker = "Triggered" , BossState <> "NotActive") :
            Print("PlayRing")
            DPSedRingSeq_Boss.Play()

    Motion01()<suspends> : void =
        race :
            block :
                EndTrigger.TriggeredEvent.Await()
                Print("!!race : Finished!!")
            block :
                #初期化
                set Count = 0
                #ランダムでゾーン内の1人を抽選する
                PlayerList : []player = GetPlayspace().GetPlayers()
                loop :
                    set Count = GetRandomInt(0 , PlayerList.Length - 1)

                    if(PlayerCheck.GetCount() = 0):
                        break
                    else if(PlayerCheck.IsCounted[PlayerList[Count]]) :
                        Print("抽選成功！")
                        break
                    else:
                        Print("再抽選")
                    Sleep(0.1)

                Sleep(1.0)

                if(ThePlayer := PlayerList[Count] , Char : fort_character = ThePlayer.GetFortCharacter[]) :
                    TargetPos := Char.GetTransform().Translation
                    KaraPos := Kara_Rotator.GetTransform().Translation
                    if(LookDirection := (TargetPos - KaraPos).MakeUnitVector[]) :

                        set Yaw = RadiansToDegrees(ArcTan(LookDirection.Y , LookDirection.X))
                        var NewRotation : rotation = MakeRotationFromYawPitchRollDegrees(Yaw , Pitch , Roll)
                        Print("{NewRotation}")
                        Kara_Rotator.MoveTo(KaraPos , NewRotation , 1.0)
                        Print("Moved Kara_Rotator")
                        Sleep(2.8)
                        set Yaw = 90.0
                        set NewRotation = MakeRotationFromYawPitchRollDegrees(Yaw , Pitch , Roll)
                        Print("{NewRotation}")
                        Kara_Rotator.MoveTo(KaraPos , NewRotation , 1.0)
                        Print("Reseted Kara_Rotator")

    Motion01_DPSed()<suspends>: void =
        race :
            block :
                EndTrigger.TriggeredEvent.Await()
                Print("!!race : Finished!!")
            block :
                if :
                    CPX0 := AreaSize[0].GetTransform().Translation.X
                    CPX1 := AreaSize[1].GetTransform().Translation.X
                    CPX2 := AreaSize[2].GetTransform().Translation.X
                    CPX3 := AreaSize[3].GetTransform().Translation.X
                    CPY0 := AreaSize[0].GetTransform().Translation.Y
                    CPY1 := AreaSize[1].GetTransform().Translation.Y
                    CPY2 := AreaSize[2].GetTransform().Translation.Y
                    CPY3 := AreaSize[3].GetTransform().Translation.Y

                then :
                    MinX01 := Min(CPX0 , CPX1)
                    MinX23 := Min(CPX2 , CPX3)
                    set MinX = Min(MinX01 , MinX23)

                    MaxX01 := Max(CPX0 , CPX1)
                    MaxX23 := Max(CPX2 , CPX3)
                    set MaxX = Max(MaxX01 , MaxX23)

                    MinY01 := Min(CPY0 , CPY1)
                    MinY23 := Min(CPY2 , CPY3)
                    set MinY = Min(MinY01 , MinY23)

                    MaxY01 := Max(CPY0 , CPY1)
                    MaxY23 := Max(CPY2 , CPY3)
                    set MaxY = Max(MaxY01 , MaxY23)

                    set SpawnPos = vector3{X := -13138.0 , Y :=20020.0 , Z := 2721.0}
                    set RandomPosX = GetRandomFloat(MinX , MaxX)
                    set RandomPosY = GetRandomFloat(MinY , MaxY)
                    set NewPos = vector3{X := RandomPosX , Y := RandomPosY , Z := 2456.0}

                Sleep(2.1)
                spawn{Motion01Slash()}

    Motion01Slash()<suspends> : void =
        SpawnedProp : ?creative_prop = SpawnProp(VFX_WindAttack_Slash , SpawnPos ,rotation{})(0)
        Print("Prop Spawned !")
        if(TheProp : creative_prop = SpawnedProp?) :
            race :
                block :
                    EndTrigger.TriggeredEvent.Await()
                    TheProp.Dispose()
                    Print("!!race : Finished!!")
                block :
                    if(LookDirection := (NewPos - SpawnPos).MakeUnitVector[]) :
                        set Yaw = RadiansToDegrees(ArcTan(LookDirection.Y , LookDirection.X))
                        set Pitch = RadiansToDegrees(ArcTan(LookDirection.Z ,Sqrt((LookDirection.X * LookDirection.X) + (LookDirection.Y - LookDirection.Y))))
                        set Roll = 0.0
                        NewRotation := MakeRotationFromYawPitchRollDegrees(Yaw , Pitch ,Roll)
                        TheProp.MoveTo(SpawnPos , NewRotation , 1.0)
                        spawn{Motion01Explode()}
                        Print("Slash Looked , Spawned Explode")
                        TheProp.MoveTo(NewPos , rotation{} , 1.1)
                        TheProp.Dispose()
                        Print("Disposed Prop")
                        set Yaw = 0.0
                        set Pitch = 0.0
                        set Roll = 0.0

    Motion01Explode()<suspends> : void =
        SpawnedProp : ?creative_prop = SpawnProp(VFX_WindAttack_Explode , NewPos , rotation{})(0)
        if(TheProp : creative_prop = SpawnedProp? , TheExplosive := Explosives[0]) :
            race :
                block :
                    EndTrigger.TriggeredEvent.Await()
                    if(TheExplosive.TeleportTo[OriginPos , rotation{}]) :
                    TheProp.Dispose()
                    Print("!!race : Finished!!")
                block :
                    Sleep(1.1)
                    if(TheExplosive.TeleportTo[NewPos , rotation{}]) :
                        Sleep(0.6)
                        if(TheExplosive.TeleportTo[OriginPos , rotation{}]) :
                        Sleep(1.2)
                            TheProp.Dispose()

    Motion03()<suspends>: void =
        var Time : float = 0.0
        if(DPSChecker <> "Triggered") :
            set Count = 3
            set Time = 0.1
            Print("DPS前{Count}")
        else :
            set Count = 0
            set Time = 0.0
            Print("DPS後{Count}")
        Print("{Time}")
        race :
            block :
                EndTrigger.TriggeredEvent.Await()
                Print("!!race : Finished!!")
            block :
                Sleep(2.5)
                loop :
                    if(Count <= 6) :
                        set Speed = 1500.0
                        set Angle = GetRandomFloat(-3.14 , 3.14)
                        set Direction = vector3{X := Cos(Angle) , Y := Sin(Angle) , Z := 0.0}
                        set SpawnPos = vector3{X := Kara_Rotator.GetTransform().Translation.X , Y := Kara_Rotator.GetTransform().Translation.Y , Z := 2550.0}
                        set NewPos = SpawnPos + Direction * Speed
                        Print("Count{Count}")
                        Print("Speed{Speed}")
                        Print("Angle{Angle}")
                        Print("Direction{Direction}")
                        Print("SpawnPos{SpawnPos}")
                        Print("NewPos{NewPos}")

                        spawn{ExplosiveOperator()}
                        spawn{RollingOperator()}
                        Sleep(Time)
                    else :
                        break
                    Sleep(0.1)

    ExplosiveOperator()<suspends> : void =
        if(TheExplosive := Explosives[Count]) :
            race :
                block :
                    EndTrigger.TriggeredEvent.Await()
                    Print("!!race : Finished!!")
                block :
                    if(TheExplosive.TeleportTo[SpawnPos , rotation{}]) :
                        set Count += 1
                        TheExplosive.MoveTo(NewPos , rotation{} , 1.0)
            if(TheExplosive.TeleportTo[OriginPos , rotation{}]):

    RollingOperator()<suspends> : void =
        SpawnedProp : ?creative_prop = SpawnProp(VFX_Rolling , SpawnPos , rotation{})(0)
        if(TheProp : creative_prop = SpawnedProp?) :
            race :
                block :
                    EndTrigger.TriggeredEvent.Await()
                    Print("!!race : Finished!!")
                block :
                    TheProp.MoveTo(NewPos , rotation{} , 1.0)
            TheProp.Dispose()

    GiveSlow(Agent : agent) : void =
        if(BossState = true , DPSChecker = "Triggered") :
            MovementModulator.Activate(Agent)
            Print("Agent Slowed !")

    OnBegin<override>()<suspends>:void=
        Print("Start")
        StartTrigger.TriggeredEvent.Subscribe(StartBossFight)
        DamageTrigger.DamagedEvent.Subscribe(WeaponCheck)
        set DefaultHealth = Health
        set DPSDefaultHealth = DPSHealth
        ZoneVolume.AgentExitsEvent.Subscribe(OnExitFromVolume)
        for(TheDV : DVArray) :
            TheDV.AgentEntersEvent.Subscribe(GiveSlow)
        DPSedRingSeq_Boss.StoppedEvent.Subscribe(DPSRing_Looper)
        RewardButton.InteractedWithEvent.Subscribe(OnInteractedReward)


    #ボスUI
    @editable var IncreaceTime : float = 2.8
    @editable var DivisionCount : int = 100
    var MaxHealth : float = 100.0

    @editable var DPS_IncreaceTime : float = 1.0
    @editable var DPS_DivisionCount : int = 33
    var DPS_MaxHealth : float = 100.0

    @editable var Motion_DivisionCount : int = 6

    var BBWidgetPer : [player]ui_bossbar2 = map{}

    StrToMes<localizes>(String : string) : message = "{String}"

    CheckBBWidget(Player : player):void=
        if(Widget := BBWidgetPer[Player]) :
            Print("ボスWidget作成済み")
            return #作成済みなら何もしない
        if(PlayerUI : player_ui = GetPlayerUI[Player]) :
            NewUI := ui_bossbar2{}
            PlayerUI.AddWidget(NewUI.CreateBBCanvas())
            logic{set BBWidgetPer[Player] = NewUI}
            if(Widget := BBWidgetPer[Player]) :
                Widget.Image_BG_BossBar.SetVisibility(widget_visibility.Collapsed)
                Widget.Image_HP_BossBar.SetVisibility(widget_visibility.Collapsed)
                Widget.Image_BossIcon.SetVisibility(widget_visibility.Collapsed)
                Widget.CB_BG_DPS.SetVisibility(widget_visibility.Collapsed)
                Widget.Image_DPSHP.SetVisibility(widget_visibility.Collapsed)
            Print("ボスWidget作成!")

    #ウィジェットを表示する関数
    ShowBBWidget()<suspends>:void=
        Sleep(0.1)
        for(Player : GetPlayspace().GetPlayers()) :
            if :
                TeamCollection := GetPlayspace().GetTeamCollection()
                BossTeam := TeamCollection.GetTeams()[2]
                TeamCollection.IsOnTeam[Player, BossTeam]
            then :
                CheckBBWidget(Player)
                if(Widget := BBWidgetPer[Player]) :
                    Widget.Image_BG_BossBar.SetVisibility(widget_visibility.Visible)
                    Widget.Image_HP_BossBar.SetVisibility(widget_visibility.Visible)
                    Widget.Image_BossIcon.SetVisibility(widget_visibility.Visible)
                    Widget.TB_Motion.SetText(StrToMes(""))
                    Widget.TB_Motion.SetShadowOpacity(0.0)
                    Widget.TB_Motion.SetVisibility(widget_visibility.Visible)
                spawn{StartIncreace(Player)}
                Print("ウィジェット表示", ?Color := NamedColors.Red)
            else :
                Print("ウィジェット表示失敗", ?Color := NamedColors.Red)

    HideBBWidget():void=
        for(Player : GetPlayspace().GetPlayers()) :
            if(Widget := BBWidgetPer[Player]) :
                Widget.Image_BG_BossBar.SetVisibility(widget_visibility.Collapsed)
                Widget.Image_HP_BossBar.SetVisibility(widget_visibility.Collapsed)
                Widget.Image_BossIcon.SetVisibility(widget_visibility.Collapsed)
                Widget.TB_Motion.SetText(StrToMes(""))
                Widget.TB_Motion.SetShadowOpacity(0.0)
                Widget.TB_Motion.SetVisibility(widget_visibility.Collapsed)

    #ボス戦開始時のHP満たされてくアニメーション関数
    StartIncreace(Player : player)<suspends>:void=
        var SizeX : float = 0.0

        if(Widget := BBWidgetPer[Player]) :
            Widget.Image_HP_BossBar.SetDesiredSize(vector2{X := 0.0, Y := 43.0})
            #シーケンスの再生時間を任意の数に分割してHPバーのサイズが合うように増やしていく
            for(X := 1..DivisionCount) :
                Sleep(IncreaceTime / (DivisionCount*1.0))
                set SizeX = Widget.Image_HP_BossBar.GetDesiredSize().X + 817.0 / (DivisionCount*1.0)
                Widget.Image_HP_BossBar.SetDesiredSize(vector2{X := SizeX, Y := 43.0})
            Print("Increaceアニメーション終了")

    #Healthを引数としてImage_HP_BossBarを対応するサイズに変更する関数
    ChangeHPBar(HealthValue : float):void=
        SizeX := 817.0 * (HealthValue / MaxHealth)
        TeamCollection := GetPlayspace().GetTeamCollection()

        for(Player : GetPlayspace().GetPlayers()) :
            if :
                BossTeam := TeamCollection.GetTeams()[2]
                TeamCollection.IsOnTeam[Player, BossTeam]
            then :
                if(Widget := BBWidgetPer[Player]) :
                    Widget.Image_HP_BossBar.SetDesiredSize(vector2{X := SizeX, Y := 43.0})
                    
    DPS_HideBBWidget():void=
        for(Player : GetPlayspace().GetPlayers()) :
            if(Widget := BBWidgetPer[Player]) :
                Widget.CB_BG_DPS.SetVisibility(widget_visibility.Collapsed)
                Widget.Image_DPSHP.SetVisibility(widget_visibility.Collapsed)

    #DPS開始時のHP満たされてくアニメーション関数
    DPS_Start():void=
        TeamCollection := GetPlayspace().GetTeamCollection()

        for(Player : GetPlayspace().GetPlayers()) :
            if :
                BossTeam := TeamCollection.GetTeams()[2]
                TeamCollection.IsOnTeam[Player, BossTeam]
                Widget := BBWidgetPer[Player]
            then :
                Widget.CB_BG_DPS.SetVisibility(widget_visibility.Visible)
                Widget.Image_DPSHP.SetVisibility(widget_visibility.Visible)
                spawn{DPS_StartIncreace(Player)}

    SetMotionText(Index : int, Frame : int)<suspends>:void=
        var Opacity : float = 0.0
        sync :
            block :
                for(X := 1..Motion_DivisionCount) :
                    Sleep(0.3 / (Motion_DivisionCount*1.0))
                    for(Player : GetPlayspace().GetPlayers()) :
                        if :
                            TeamCollection := GetPlayspace().GetTeamCollection()
                            BossTeam := TeamCollection.GetTeams()[2]
                            TeamCollection.IsOnTeam[Player, BossTeam]
                            Widget := BBWidgetPer[Player]
                        then :
                            if(Index = 0) :
                                Widget.TB_Motion.SetText(StrToMes(""))
                            if(Index = 1) :
                                Widget.TB_Motion.SetText(StrToMes("Wind Attack"))
                            if(Index = 2) :
                                Widget.TB_Motion.SetText(StrToMes("Lightning"))
                            if(Index = 3) :
                                Widget.TB_Motion.SetText(StrToMes("Rolling"))
                            if(Index = 4) :
                                Widget.TB_Motion.SetText(StrToMes("Lightning"))
                            if(Index = 5) :
                                Widget.TB_Motion.SetText(StrToMes("Hurricane"))
                            if(NewOpacity := type{_X:float where 0.000000 <= _X, _X <= 1.000000}[Opacity + (1.0  /(Motion_DivisionCount*1.0))]) :
                                Widget.TB_Motion.SetTextOpacity(NewOpacity)
                                Widget.TB_Motion.SetShadowOpacity(NewOpacity)
                                Print("{NewOpacity}", ?Color := NamedColors.Red)
                                set Opacity = NewOpacity
            block :
                race :
                    block :
                        EndTrigger.TriggeredEvent.Await()
                        Print("!!race : Finished!!")
                    block :
                        Sleep((Frame*1.0) / 30.0)
                        set Opacity = 1.0
                        for(X := 1..Motion_DivisionCount) :
                            Sleep(0.3 / (Motion_DivisionCount*1.0))
                            for(Player : GetPlayspace().GetPlayers()) :
                                if :
                                    TeamCollection := GetPlayspace().GetTeamCollection()
                                    BossTeam := TeamCollection.GetTeams()[2]
                                    TeamCollection.IsOnTeam[Player, BossTeam]
                                    Widget := BBWidgetPer[Player]
                                then :
                                    if(NewOpacity := type{_X:float where 0.000000 <= _X, _X <= 1.000000}[Opacity - (1.0  /(Motion_DivisionCount*1.0))]) :
                                        Widget.TB_Motion.SetTextOpacity(NewOpacity)
                                        Widget.TB_Motion.SetShadowOpacity(NewOpacity)
                                        set Opacity = NewOpacity

    DPS_StartIncreace(Player : player)<suspends>:void=
        var SizeX : float = 0.0

        if(Widget := BBWidgetPer[Player]) :
            Widget.Image_DPSHP.SetDesiredSize(vector2{X := 0.0, Y := 11.943165})
            #シーケンスの再生時間を任意の数に分割してHPバーのサイズが合うように増やしていく
            for(X := 1..DPS_DivisionCount) :
                Sleep(DPS_IncreaceTime / (DPS_DivisionCount*1.0))
                set SizeX = Widget.Image_DPSHP.GetDesiredSize().X + 595.5 / (DPS_DivisionCount*1.0)
                Widget.Image_DPSHP.SetDesiredSize(vector2{X := SizeX, Y := 11.943165})
            Print("Increaceアニメーション終了")

    #DPSHealthを引数としてImage_DPSHPを対応するサイズに変更する関数
    DPS_ChangeHPBar(HealthValue : float):void=
        SizeX := 595.5 * (HealthValue / DPS_MaxHealth)
        TeamCollection := GetPlayspace().GetTeamCollection()

        for(Player : GetPlayspace().GetPlayers()) :
            if :
                BossTeam := TeamCollection.GetTeams()[2]
                TeamCollection.IsOnTeam[Player, BossTeam]
            then :
                if(Widget := BBWidgetPer[Player]) :
                    Widget.Image_DPSHP.SetDesiredSize(vector2{X := SizeX, Y := 11.943165})
ui_bossbar2:= class:
    #通常のHPバー
    Image_BG_BossBar : texture_block = texture_block:
        DefaultImage := Objects.UI_Healthbar_outside
    Image_HP_BossBar : texture_block = texture_block:
        DefaultImage := Objects.UI_Healthbar_inside
        DefaultDesiredSize := vector2{X := 0.0, Y := 43.0}#最大 X := 817.0
    Image_BossIcon : texture_block = texture_block:
        DefaultImage := Texture.boss_name02
    #DPSのHPバー
    CB_BG_DPS : color_block = color_block:
        DefaultColor := color{R:=0.036889,G:=0.036889,B:=0.036889}
        DefaultOpacity := 1.0
    Image_DPSHP : texture_block = texture_block:
        DefaultImage := Texture.T_DPSHP
        DefaultTint := color{R:=0.991999,G:=1.0,B:=0.860246}
        DefaultDesiredSize := vector2{X := 0.0, Y := 11.943165}#最大 X := 595.5
    #モーション名テキスト
    TB_Motion : text_block = text_block:
        DefaultTextColor := NamedColors.White
        DefaultShadowOffset := option{vector2{X:=1.614798,Y:=1.960874}}
        DefaultShadowOpacity := 0.0

    CreateBBCanvas<public>(): canvas =
        var CanvasSlots : []canvas_slot = array{}

        set CanvasSlots += array:
            #通常のHPバー
            canvas_slot:
                Anchors := anchors:
                    Minimum := vector2{X := 0.5, Y := 0.08}
                    Maximum := vector2{X := 0.5, Y := 0.08}
                Offsets := margin{Left := 0.0, Top := 0.0, Right := 880.0, Bottom := 43.0}
                SizeToContent := false
                Alignment := vector2{X := 0.5, Y := 0.5}
                Widget := Image_BG_BossBar
            canvas_slot:
                Anchors := anchors:
                    Minimum := vector2{X := 0.5, Y := 0.08}
                    Maximum := vector2{X := 0.5, Y := 0.08}
                Offsets := margin{Left := -409.000000, Top := 0.0, Right := 0.0, Bottom := 0.0}
                SizeToContent := true
                Alignment := vector2{X := 0.0, Y := 0.5}
                Widget := Image_HP_BossBar
            canvas_slot:
                Anchors := anchors:
                    Minimum := vector2{X := 0.5, Y := 0.08}
                    Maximum := vector2{X := 0.5, Y := 0.08}
                Offsets := margin{Left := 27.934517, Top := -15.321068, Right := 409.589447, Bottom := 108.112976}
                SizeToContent := false
                Alignment := vector2{X := 0.0, Y := 1.0}
                Widget := Image_BossIcon
            #DPSのHPバー
            canvas_slot:
                Anchors := anchors:
                    Minimum := vector2{X := 0.5, Y := 0.08}
                    Maximum := vector2{X := 0.5, Y := 0.08}
                Offsets := margin{Left := 0.0, Top := 68.778671, Right := 598.355835, Bottom := 15.257175}
                SizeToContent := false
                Alignment := vector2{X := 0.5, Y := 0.5}
                Widget := CB_BG_DPS
            canvas_slot:
                Anchors := anchors:
                    Minimum := vector2{X := 0.5, Y := 0.08}
                    Maximum := vector2{X := 0.5, Y := 0.08}
                Offsets := margin{Left := -297.941925, Top := 68.599998, Right := 0.0, Bottom := 0.0}
                SizeToContent := true
                Alignment := vector2{X := 0.0, Y := 0.5}
                Widget := Image_DPSHP
            #モーション名テキスト
            canvas_slot:
                Anchors := anchors:
                    Minimum := vector2{X := 0.5, Y := 0.08}
                    Maximum := vector2{X := 0.5, Y := 0.08}
                Offsets := margin{Left := 0.0, Top := 108.259071, Right := 0.0, Bottom := 0.0}
                SizeToContent := true
                Alignment := vector2{X := 0.5, Y := 0.5}
                Widget := TB_Motion

        Canvas : canvas = canvas:
            Slots := CanvasSlots
        return Canvas