game_level<public> := type{ _X: int where 1 <= _X, _X <= 8 }

# (Input: game_level).ToInt(): int =
#   return Input