input_command_type<public> := enum:
  UP,
  DOWN,
  LEFT,
  RIGHT

input_command<public> := class:
  Type<public>: input_command_type

(Input: input_command).ToString<public>(): string =
  case(Input.Type):
    input_command_type.UP =>
      return "UP"
    input_command_type.DOWN =>
      return "DOWN"
    input_command_type.LEFT =>
      return "LEFT"
    input_command_type.RIGHT =>
      return "RIGHT"