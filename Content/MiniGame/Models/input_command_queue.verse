using { Models }

input_queque<public> := class:
  var CommandQueue: []input_command = array {}

  Copy<public>(Queue: []input_command): input_queque =
    var NewQueue: input_queque = input_queque{}
    for(Command : CommandQueue):
      NewQueue.Push(Command)
    return NewQueue

  Slice<public>(Count: int): input_queque =
    var NewQueue: input_queque = input_queque{}
    for(Index -> Command : CommandQueue):
      if(Count > Index):
        NewQueue.Push(Command)
    return NewQueue

  Push<public>(Command: input_command): void =
    set CommandQueue += array { Command }

  Pop<public>(): ?input_command =
    if(CommandQueue.Length = 0):
      return false

    if(command := CommandQueue[0]):
      if(NewQueue := CommandQueue.RemoveElement[0]):
        set CommandQueue = NewQueue
      return option { command }
    else:
      return false

  GetCommandQueue<public>(): []input_command =
    return CommandQueue


  

