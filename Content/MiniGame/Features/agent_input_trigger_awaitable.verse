using { /Fortnite.com/Devices }
using { /Verse.org/Simulation }

agent_input_event<public> := class(cancelable):
    Agent<public>: agent
    InputTrigger<public> :input_trigger_device

    var IsCancel: logic = false

    Event<public>:event(agent) = event(agent){}

    Cancel<override>()<transacts>:void=
      set IsCancel = true

    Emit<public>():void=
      Event.Signal(Agent)
    
    EventLoop()<suspends>: void =
      loop: 
        if(IsCancel = true):
          break

        TriggerEvent := InputTrigger.ReleasedEvent.Await()
        if(TriggerEvent(0) = Agent):
          Emit()
          break
          
    Await<public>()<suspends>:void=
      spawn{ EventLoop() }
      Event.Await()
