using { /Verse.org/Simulation }
using { /Verse.org }
using { Models }
using { Canvas }

# TODO ランダムに生成する処理をいれる
generate_input_command_queue<public>(): input_queque =
  NewQueue: input_queque = input_queque{}
  NewQueue.Push(generate_input_command())
  NewQueue.Push(generate_input_command())
  NewQueue.Push(generate_input_command())
  NewQueue.Push(generate_input_command())
  NewQueue.Push(generate_input_command())
  NewQueue.Push(generate_input_command())
  NewQueue.Push(generate_input_command())
  NewQueue.Push(generate_input_command())
  NewQueue.Push(generate_input_command())
  NewQueue.Push(generate_input_command())
  return NewQueue

generate_input_command<public>(): input_command =
  Params :int = Random.GetRandomInt(1, 4)
  case(Params):
    1 => input_command{Type := input_command_type.UP}
    2 => input_command{Type := input_command_type.DOWN}
    3 => input_command{Type := input_command_type.LEFT}
    4 => input_command{Type := input_command_type.RIGHT}
    _ => input_command{Type := input_command_type.RIGHT}
  

# input_queueからinput_command_widgetを生成する
(Input: input_queque).CreateWidget<public>(Agent: agent): []input_command_widget =
  var WidgetList: []input_command_widget = array {}
  CommandQueue := Input.GetCommandQueue()
  for(Index -> Command : CommandQueue):
    Widget: input_command_widget = input_command_widget{ Index := Index, Command := Command, Agent := Agent }
    set WidgetList += array { Widget }
  return WidgetList
