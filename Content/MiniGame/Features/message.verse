StringToMessage<public><localizes>(value:string)<computes> : message = "{value}"