using { /Verse.org/Simulation }
using { /Fortnite.com/Devices }
using { /Fortnite.com/Characters }

game_controller_device<public> := class(creative_device):

  @editable
  PlayerGameController: player_game_controller_device = player_game_controller_device{}

  @editable
  SaveDataManager : Mining_System.Code.PlayerData.save_data_manager = Mining_System.Code.PlayerData.save_data_manager{}

  @editable
  SkillManager : Mining_System.Code.Skill.skill_manager = Mining_System.Code.Skill.skill_manager{}

  @editable
  QuestTrigger : trigger_device = trigger_device{}

  @editable
  StartTrigger : trigger_device = trigger_device{}

  @editable
  DebugButton : button_device = button_device{}

  @editable
  DebugButton2 : button_device = button_device{}

  @editable
  DebugButton3 : button_device = button_device{}

  @editable
  DebugButton4 : button_device = button_device{}

  @editable
  DebugButton5 : button_device = button_device{}

  @editable
  DebugButton6 : button_device = button_device{}

  OnBegin<override>(): void =
    for(Player : GetPlayspace().GetPlayers()):
      SpawnTicketTimer(Player)
    GetPlayspace().PlayerAddedEvent().Subscribe(SpawnTicketTimer)
    DebugButton.InteractedWithEvent.Subscribe(DebugGetMaxTickets)
    DebugButton2.InteractedWithEvent.Subscribe(DebugGetNowResourceTicket)
    DebugButton3.InteractedWithEvent.Subscribe(DebugAddMaxResourceTicket)
    DebugButton4.InteractedWithEvent.Subscribe(DebugAddResourceTicket)
    StartTrigger.TriggeredEvent.Subscribe(OnStartGame)

  DebugGetMaxTickets(Agent:agent):void=
    if(Player := player[Agent]):
      MaxTickets := SaveDataManager.GetMaxResourceTicket(Player)
      Print("MaxTicket : {MaxTickets}")

  DebugGetNowResourceTicket(Agent:agent):void=
    if(Player := player[Agent]):
      NowTicket := SaveDataManager.GetNowResourceTicket(Player)
      Print("NowTicket : {NowTicket}")

  DebugAddMaxResourceTicket(Agent:agent):void=
    if(Player := player[Agent]):
      SaveDataManager.AddMaxResourceTicket(Player)
      Print("AddMax")
  
  DebugAddResourceTicket(Agent:agent):void=
    if(Player := player[Agent]):
      SaveDataManager.AddResourceTicket(Player)
      Print("Add Ticket")

  DebugSubResourceTicket(Agent:agent):void=
    if(Player := player[Agent]):
      SaveDataManager.SubResourceTicket(Player)

  SpawnTicketTimer(Agent:agent):void=
    if(Player := player[Agent]):
      SaveDataManager.CheckRecoveryResourceTicket(Player)

  OnStartGame<public>(Agent: ?agent): void =
    if(Player := player[Agent?]):
      NowTicketAmount := SaveDataManager.GetNowResourceTicket(Player)

      if(NowTicketAmount >= 1):
        # Ticketの消費
        SaveDataManager.SubResourceTicket(Player)
        Print("チケットを消費してゲームを始めました！")

        PlayerGameController.Start(Player)
        # 追記
        if (FortCharacter := Player.GetFortCharacter[]):
          FortCharacter.PutInStasis(stasis_args{})
      else:
        Print("チケットがありません！")