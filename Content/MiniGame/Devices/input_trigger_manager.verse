using { /Fortnite.com/Devices }
using { /Verse.org/Simulation }
using { Models }
using { Features }
using { Canvas }

input_trigger_manager := class(creative_device):
  @editable
  UpInputTrigger: input_trigger_device = input_trigger_device{}

  @editable
  DownInputTrigger: input_trigger_device = input_trigger_device{}

  @editable
  LeftInputTrigger: input_trigger_device = input_trigger_device{}

  @editable
  RightInputTrigger: input_trigger_device = input_trigger_device{}

  @editable
  BadAudioPlayer: audio_player_device = audio_player_device{}

  GetInputCommandTrigger(InputCommand: input_command_type): tuple(input_trigger_device, input_trigger_device, input_trigger_device, input_trigger_device) =
    case(InputCommand):
      input_command_type.UP =>
        return (UpInputTrigger,  DownInputTrigger, LeftInputTrigger, RightInputTrigger)
      input_command_type.DOWN =>
        return (DownInputTrigger, UpInputTrigger, LeftInputTrigger, RightInputTrigger )
      input_command_type.LEFT =>
        return (LeftInputTrigger,  UpInputTrigger, DownInputTrigger, RightInputTrigger )
      input_command_type.RIGHT =>
        return (RightInputTrigger, UpInputTrigger, DownInputTrigger, LeftInputTrigger)

  WaitTriggerEvent<public>(Agent: agent, InputCommand: input_command_type, CommandWidget: []input_command_widget, Index: int)<suspends>: void =
    GetInputCommandTriggerResult := GetInputCommandTrigger(InputCommand)
    TargetTrigger := GetInputCommandTriggerResult(0)
    OtherTrigger1 := GetInputCommandTriggerResult(1)
    OtherTrigger2 := GetInputCommandTriggerResult(2)
    OtherTrigger3 := GetInputCommandTriggerResult(3)

    Result := race:
      GoodEvent(Agent, TargetTrigger)
      BadEvent(Agent, OtherTrigger1, CommandWidget, Index)
      BadEvent(Agent, OtherTrigger2, CommandWidget, Index)
      BadEvent(Agent, OtherTrigger3, CommandWidget, Index)

    if(Result = false):
      Sleep(2.0)
      WaitTriggerEvent(Agent, InputCommand, CommandWidget, Index)
    else:
      return


  GoodEvent(Agent: agent, Trigger: input_trigger_device)<suspends>: logic =
    agent_input_event{Agent := Agent, InputTrigger := Trigger}.Await()
    return true

  BadEvent(Agent: agent, Trigger: input_trigger_device, CommandWidget: []input_command_widget, Index: int)<suspends>: logic =
    agent_input_event{Agent := Agent, InputTrigger := Trigger}.Await()
    spawn { CommandWidget.BadEffect(Index) }
    BadAudioPlayer.Play(Agent)
    return false
  