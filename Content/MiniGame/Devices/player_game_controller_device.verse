using { /Verse.org/Simulation }
using { /Fortnite.com/Devices }
using { /UnrealEngine.com/Temporary/UI }
using { /UnrealEngine.com/Temporary/SpatialMath }
using { /Fortnite.com/Characters }
using { Devices }
using { Models }
using { Features }
using { Canvas }
using { VerseDevice }

player_game_controller_device :=  class(creative_device):
  #XP付与
  @editable XP_MiniBuff : accolades_device = accolades_device{}
  @editable
  GameTimer: timer_device = timer_device{}

  @editable
  PlayerGameUIManager: player_game_ui_manager_device = player_game_ui_manager_device{}

  @editable
  InputTriggerManager: input_trigger_manager = input_trigger_manager{}

  @editable
  GoodAudioPlayer: audio_player_device = audio_player_device{}

  @editable
  CountDownAudioPlayer: audio_player_device = audio_player_device{}

  @editable
  VFX_BuffAura : visual_effect_powerup_device = visual_effect_powerup_device{}

  # cube追記
  @editable
  var OPSequences : []cinematic_sequence_device = array{}

  @editable
  var WinSequences : []cinematic_sequence_device = array{}

  @editable
  var LoseSequences : []cinematic_sequence_device = array{}

  @editable
  QuestTrigger : trigger_device = trigger_device{}

  Skill_Manager : Mining_System.Code.Skill.skill_manager = Mining_System.Code.Skill.skill_manager{}

  @editable
  SaveDataManager : Mining_System.Code.PlayerData.save_data_manager = Mining_System.Code.PlayerData.save_data_manager{}

  @editable
  DebugButton : button_device = button_device{}
  @editable
  UI_Hud_Count : VerseDevice.ui_hud_count = VerseDevice.ui_hud_count{}

  var PlayerBuffing : [agent]logic = map{}
  var NowBuffTime : [agent]int = map{}

  var PlayerCommandWidgets: [agent]?[]input_command_widget = map{}
  # var PlayerTimerWidgets: [agent]?waiting_timer_widget = map{}
  var PlayerPlaying: [agent]logic = map{}
  var PlayerLevel: [agent]?type{ _X: int where 1 <= _X, _X <= 8 } = map{}

  OnBegin<override>()<suspends>:void=
    DebugButton.InteractedWithEvent.Subscribe(Debug2)
    #for(Player : GetPlayspace().GetPlayers()):
      #PlayerBuff(Player)

    #GetPlayspace().PlayerAddedEvent().Subscribe(PlayerBuff)

  Debug2(Agent:agent):void=    if(Player := player[Agent]):
    var CheckFailed : logic = false
    var Initial_BuffTime : int = 90
    Index := 5

    if(SkillStatus := SaveDataManager.GetSkillStatus[Player, Index]):
      var Skill : Mining_System.Code.Skill.skill_status = SkillStatus
      var Level : int = 7


      Print("Level is {Level}", ?Duration := 5.0)

      if(Level <= 4):
        set CheckFailed = true
        Print("Failed", ?Duration := 5.0)

      # 1段階目バフ
      if(Level = 1 or Level = 2):
        set Initial_BuffTime = Initial_BuffTime + 10

      # 2段階目バフ
      if(Level = 3 or Level = 4):
        set Initial_BuffTime = Initial_BuffTime + 20

      # 3段階目バフ
      if(Level = 5 or Level = 6):
        set Initial_BuffTime = Initial_BuffTime + 30

      # 4段階目バフ
      if(Level = 7 or Level = 8):
        set Initial_BuffTime = Initial_BuffTime + 40


      # EffectDurationだけ更新
      set Skill = Mining_System.Code.Skill.skill_status{
        Index := SkillStatus.Index,
        IsActive := true,
        CooldownTime := SkillStatus.CooldownTime,
        MaxCooldownTime := SkillStatus.MaxCooldownTime,
        Charges := SkillStatus.Charges,
        MaxCharges := SkillStatus.MaxCharges,
        LastUsedTime := SkillStatus.LastUsedTime,
        EffectDuration := Initial_BuffTime,
        Multiply := SkillStatus.Multiply
        }
        
        SaveDataManager.SetSkillStatus(Player, Index, Skill)

        PlayerBuff(Player)

        if(Data := SaveDataManager.GetSkillStatus[Player, Index]):
          Print("Index : {Data.Index}, CooldownTime : {Data.CooldownTime}, MaxCooldownTime : {Data.MaxCooldownTime}, Charges : {Data.Charges}, MaxCharges : {Data.MaxCharges}, LastUsedTime : {Data.LastUsedTime}, EffectDuration : {Data.EffectDuration}", ?Duration := 5.0)


  Debug(Agent:agent):void=
    Print("A")
    if(Player := player[Agent]):
      spawn{SaveDataManager.BuffTimeStatus(Player, 5)}

      SetSkillStatusIndex := 5
      SkillIndex := 5
      SkillCooldownTime := 0
      SkillMaxCooldownTime := 0
      SkillCharges := 0
      SkillMaxCharges := 0
      SkillLastUsedTime := 0
      SkillMultiply := 2

      Skill : Mining_System.Code.Skill.skill_status = Mining_System.Code.Skill.skill_status{
        Index := SkillIndex,
        IsActive := false,
        CooldownTime := SkillMaxCooldownTime,
        MaxCooldownTime := SkillMaxCooldownTime,
        Charges := SkillCharges,
        MaxCharges := SkillMaxCharges,
        LastUsedTime := SkillLastUsedTime,
        EffectDuration := 90,
        Multiply := SkillMultiply
        }

        SaveDataManager.SetSkillStatus(Player, SetSkillStatusIndex, Skill)
  
        if(Data := SaveDataManager.GetSkillStatus[Player, SkillIndex]):
          Print("Index : {Data.Index}, CooldownTime : {Data.CooldownTime}, MaxCooldownTime : {Data.MaxCooldownTime}, Charges : {Data.Charges}, MaxCharges : {Data.MaxCharges}, LastUsedTime : {Data.LastUsedTime}, EffectDuration : {Data.EffectDuration}", ?Duration := 5.0)

          
  PlayerBuff(Agent:agent):void=
    if(PlayerBuffing[Agent] = false):
      spawn{PlayerBuffTime(Agent)}
    if(not PlayerBuffing[Agent]):
      spawn{PlayerBuffTime(Agent)}

  PlayerBuffTime(Agent:agent)<suspends>:void=
    logic{ set PlayerBuffing[Agent] = true }

    VFX_BuffAura.SetDuration(-1.0)
    VFX_BuffAura.Pickup(Agent)
    XP_MiniBuff.Award(Agent)

    loop:
        Sleep(1.0)
        
        # AgentをPlayerに変換し、無効な場合はループを終了
        if(Player := player[Agent]):
          if (SaveDataManager.HasSaveData(Player)?):
            if:
              SkillStatus := SaveDataManager.GetSkillStatus[Player, 5]
              SkillStatus.IsActive = true
            then:
              var Skill : Mining_System.Code.Skill.skill_status = SkillStatus
              EffectDuration := SkillStatus.EffectDuration
              Print("EffectDuration : {EffectDuration}", ?Duration := 3.0)

              # EffectDurationが0以下の場合、バフを終了
              if (EffectDuration <= 0):
                  set Skill = Mining_System.Code.Skill.skill_status{
                      Index := SkillStatus.Index,
                      IsActive := false,
                      CooldownTime := SkillStatus.CooldownTime,
                      MaxCooldownTime := SkillStatus.MaxCooldownTime,
                      Charges := SkillStatus.Charges,
                      MaxCharges := SkillStatus.MaxCharges,
                      LastUsedTime := SkillStatus.LastUsedTime,
                      EffectDuration := 0,
                      Multiply := SkillStatus.Multiply
                  }
                  SaveDataManager.SetSkillStatus(Player, 5, Skill)
                  logic{ set PlayerBuffing[Agent] = false }
                  Print("End Buff!!", ?Duration := 5.0)
                  VFX_BuffAura.SetDuration(0.1)
                  VFX_BuffAura.Pickup(Agent)
                  if (IsBuffing := PlayerBuffing[Agent]):
                      spawn{UI_Hud_Count.Update_BuffUI(Agent, EffectDuration, IsBuffing)}
                  break

              # EffectDurationが残っている場合、減少させてバフを更新
              else:
                  set Skill = Mining_System.Code.Skill.skill_status{
                      Index := SkillStatus.Index,
                      IsActive := SkillStatus.IsActive,
                      CooldownTime := SkillStatus.CooldownTime,
                      MaxCooldownTime := SkillStatus.MaxCooldownTime,
                      Charges := SkillStatus.Charges,
                      MaxCharges := SkillStatus.MaxCharges,
                      LastUsedTime := SkillStatus.LastUsedTime,
                      EffectDuration := SkillStatus.EffectDuration - 1,
                      Multiply := SkillStatus.Multiply
                  }
                  SaveDataManager.SetSkillStatus(Player, 5, Skill)
                  if (IsBuffing := PlayerBuffing[Agent]):
                      spawn{UI_Hud_Count.Update_BuffUI(Agent, EffectDuration, IsBuffing)}
          else:
            Print("Error: PlayerBuffTime")
            logic { set PlayerBuffing[Agent] = false }
            break
        else:
            Print("Invalid Player from Agent, ending buff.")
            logic{ set PlayerBuffing[Agent] = false }
            break

  Start<public>(Agent: agent): void =
    if(IsPlaying := PlayerPlaying[Agent]?):
      return

    Print("player_game_controller_device.Start()")
    spawn { GameSequence(Agent) }

  (Agent:agent).OPReturnSequence()<transacts><decides>:cinematic_sequence_device=
      SelectedSequenceDevice := OPSequences[0]
      set OPSequences = for (Index -> SequenceDevice : OPSequences, Index <> 0):
          SequenceDevice

      SelectedSequenceDevice

  (Agent:agent).WinReturnSequence()<transacts><decides>:cinematic_sequence_device=
      SelectedSequenceDevice := WinSequences[0]
      set WinSequences = for (Index -> SequenceDevice : WinSequences, Index <> 0):
          SequenceDevice

      SelectedSequenceDevice

  (Agent:agent).LoseReturnSequence()<transacts><decides>:cinematic_sequence_device=
      SelectedSequenceDevice := LoseSequences[0]
      set LoseSequences = for (Index -> SequenceDevice : LoseSequences, Index <> 0):
          SequenceDevice

      SelectedSequenceDevice

  GameSequence(Agent: agent)<suspends>: void =

    var CheckFailed : logic = false

    var Initial_BuffTime : int = 90


    # cube追記
    # OPSequencesからSequenceをひとつ取り出して再生
    if (TheSequence := Agent.OPReturnSequence[], FortCharacter := Agent.GetFortCharacter[]):
      TheSequence.Play(Agent)
      FortCharacter.PutInStasis(stasis_args{})
      TheSequence.StoppedEvent.Await()
      set OPSequences += array{TheSequence}
    
    logic { set PlayerPlaying[Agent] = true }
    PlayerGameUIManager.Show(Agent)

    GameTimer.Start(Agent)

    race:
      GameTimer.SuccessEvent.Await()
      GameLoop(Agent)

    # if(Widget := PlayerTimerWidgets[Agent]?):
    #   Widget.RemoveWidget()

    if(Widget := PlayerCommandWidgets[Agent]?):
      Widget.HideAll()
      Widget.RemoveWidget()

    # 各プレイヤーのスキル情報を取得し、バフ時間を設定
    if (Player := player[Agent]):
      if (SkillInfo := SaveDataManager.GetSkill[Player, 7]):
          # スキルレベルに応じたバフ時間を設定
          set Initial_BuffTime = SkillInfo.EffectPower

    # Cube追記
    Index := 5


    if(Player := player[Agent]):

      if(SkillStatus := SaveDataManager.GetSkillStatus[Player, Index]):
        var Skill : Mining_System.Code.Skill.skill_status = SkillStatus

        if(Level := PlayerLevel[Agent]?):

          Print("Level is {Level}", ?Duration := 5.0)

          if(Level <= 4):
            set CheckFailed = true
            Print("Failed", ?Duration := 5.0)

          # 1段階目バフ
          if(Level = 1 or Level = 2):
            set Initial_BuffTime = Initial_BuffTime + 10

          # 2段階目バフ
          if(Level = 3 or Level = 4):
            set Initial_BuffTime = Initial_BuffTime + 20

          # 3段階目バフ
          if(Level = 5 or Level = 6):
            set Initial_BuffTime = Initial_BuffTime + 30

          # 4段階目バフ
          if(Level = 7 or Level = 8):
            set Initial_BuffTime = Initial_BuffTime + 40

        else:
          # 1回もクリアできなかったとき
          set CheckFailed = true

        # EffectDurationだけ更新
        set Skill = Mining_System.Code.Skill.skill_status{
          Index := SkillStatus.Index,
          IsActive := true,
          CooldownTime := SkillStatus.CooldownTime,
          MaxCooldownTime := SkillStatus.MaxCooldownTime,
          Charges := SkillStatus.Charges,
          MaxCharges := SkillStatus.MaxCharges,
          LastUsedTime := SkillStatus.LastUsedTime,
          EffectDuration := Initial_BuffTime,
          Multiply := SkillStatus.Multiply
          }
          
          SaveDataManager.SetSkillStatus(Player, Index, Skill)

          PlayerBuff(Player)

          if(Data := SaveDataManager.GetSkillStatus[Player, Index]):
            Print("Index : {Data.Index}, CooldownTime : {Data.CooldownTime}, MaxCooldownTime : {Data.MaxCooldownTime}, Charges : {Data.Charges}, MaxCharges : {Data.MaxCharges}, LastUsedTime : {Data.LastUsedTime}, EffectDuration : {Data.EffectDuration}", ?Duration := 5.0)

    GameEnd(Agent)

    # Cube追記
    if(CheckFailed = true):
        Print("CheckFailed is True")
        if (TheSequenceL := Agent.LoseReturnSequence[], FortCharacter := Agent.GetFortCharacter[]):
          TheSequenceL.Play(Agent)
          FortCharacter.PutInStasis(stasis_args{})
          TheSequenceL.StoppedEvent.Await()
          set LoseSequences += array{TheSequenceL}
    else:
        Print("CheckFailed is False")
        if (TheSequenceW := Agent.WinReturnSequence[], FortCharacter := Agent.GetFortCharacter[]):
          TheSequenceW.Play(Agent)
          FortCharacter.PutInStasis(stasis_args{})
          TheSequenceW.StoppedEvent.Await()
          set WinSequences += array{TheSequenceW}

    if (FortCharacter := Agent.GetFortCharacter[]):
      FortCharacter.ReleaseFromStasis()

    if(Player := player[Agent]):

      # Quest関係
      # 岩を壊すクエストの時に実行
      NowQuest := SaveDataManager.GetNowQuest(Player)
      #if(NowQuest = 13):
          #QuestTrigger.Trigger(Player)
      QuestTrigger.Trigger(Player)

  GameLoop(Agent: agent)<suspends>: void =
    InputCommandQueue := generate_input_command_queue()
    StartGameLevel(Agent, InputCommandQueue, 1)
    logic { set PlayerLevel[Agent] = option { 1 } }
    StartGameLevel(Agent, InputCommandQueue, 2)
    logic { set PlayerLevel[Agent] = option { 2 } }
    StartGameLevel(Agent, InputCommandQueue, 3)
    logic { set PlayerLevel[Agent] = option { 3 } }
    StartGameLevel(Agent, InputCommandQueue, 4)
    logic { set PlayerLevel[Agent] = option { 4 } }
    StartGameLevel(Agent, InputCommandQueue, 5)
    logic { set PlayerLevel[Agent] = option { 5 } }
    StartGameLevel(Agent, InputCommandQueue, 6)
    logic { set PlayerLevel[Agent] = option { 6 } }
    StartGameLevel(Agent, InputCommandQueue, 7)
    logic { set PlayerLevel[Agent] = option { 7 } }
    StartGameLevel(Agent, InputCommandQueue, 8)
    logic { set PlayerLevel[Agent] = option { 8 } }
    return

  StartGameLevel(Agent: agent, GameCommands: input_queque, Level: type{ _X: int where 1 <= _X, _X <= 8 })<suspends>: void =
    CommandWidget := GameCommands.CreateWidget(Agent)
    logic { set PlayerCommandWidgets[Agent] = option { CommandWidget } }
    CommandWidget.AddWidget()

    InputCommandQueue := GameCommands.Copy().Slice(Level + 2)

    CommandWidget.ShowAll(Level + 2)
    CommandWidget.ShowWaku(Level + 2)
    TimerWidget := waiting_timer_widget{ Agent := Agent }
    # logic { set PlayerTimerWidgets[Agent] = option { TimerWidget } }
    CountDownAudioPlayer.Stop(Agent)
    Sleep(0.1)
    CountDownAudioPlayer.Play(Agent)
    PlayerGameUIManager.StartCountDown(Agent)
    CommandWidget.HideAll()
    CommandWidget.ShowWaku(Level + 2)
    
    var Index: int = 0
    loop:
      # このレベルは終了
      Commands := InputCommandQueue.GetCommandQueue()

      # コマンドを入力し終わった
      if(Commands.Length = 0):
        break

      Command := InputCommandQueue.Pop()

      if(_Command := Command?):
        Print(_Command.ToString())
        InputTriggerManager.WaitTriggerEvent(Agent, _Command.Type, CommandWidget, Index)
        CommandWidget.Show(Index)
        set Index += 1
        Sleep(0.1)

    CommandWidget.HideAll()
    CommandWidget.RemoveWidget()
    CommandWidget.HideWaku()
    GoodAudioPlayer.Play(Agent)
    Sleep(0.5)

  GameEnd(Agent: agent): void =
    Print("ゲーム終了")
    GameTimer.Pause(Agent)
    GameTimer.Reset(Agent)
    PlayerGameUIManager.Hide(Agent)
    logic { set PlayerPlaying[Agent] = false}
    # 追記
    if (FortCharacter := Agent.GetFortCharacter[]):
      FortCharacter.ReleaseFromStasis()

    # TODO バフをかける
    if(Level := PlayerLevel[Agent]?):
      # Print("Level: {Level}")
    else:
      # Print("Level: None")
    logic { set PlayerLevel[Agent] = false }
  
