using { /Fortnite.com/Devices }
using { /Verse.org/Simulation }
using { /Fortnite.com/Playspaces }
using { /UnrealEngine.com/Temporary/UI }
using { Canvas }

player_game_ui_manager_device := class(creative_device): 
  var PlayerCommandsCanvas: [agent]?input_commands_canvas = map{}

  StartCountDown<public>(Agent: agent)<suspends>: void =
    if(TargerCanvas := PlayerCommandsCanvas[Agent]?):
      TargerCanvas.SetTimer(3)

  Show<public>(Agent: agent): void =
    if (InPlayer := player[Agent], PlayerUI := GetPlayerUI[InPlayer]):
      _Canvas := input_commands_canvas{}
      _Canvas.Create()
      logic { set PlayerCommandsCanvas[Agent] = option { _Canvas } }
      PlayerUI.AddWidget(_Canvas.GetCanvas())

  Hide<public>(Agent: agent): void =
    if (InPlayer := player[Agent], PlayerUI := GetPlayerUI[InPlayer]):
      if(_Canvas := PlayerCommandsCanvas[Agent]?):
        PlayerUI.RemoveWidget(_Canvas.GetCanvas())
        logic { set PlayerCommandsCanvas[Agent] = false }