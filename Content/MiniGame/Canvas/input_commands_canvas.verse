using { /UnrealEngine.com/Temporary/UI }
using { /UnrealEngine.com/Temporary/SpatialMath }
using { /Verse.org/Colors }
using { /Fortnite.com/UI }
using { /Verse.org/Simulation }
using { Features }

input_commands_canvas<public> := class():
  var TimerText : text_block = text_block{ DefaultTextColor := NamedColors.Black }
  var InputCommandsCanvas: canvas = canvas {}

  Create<public>(): void = 
    set InputCommandsCanvas = canvas:
      Slots := array:
        canvas_slot:
            Anchors := anchors{Minimum := vector2{X := 0.5, Y := 0.5}, Maximum := vector2{X := 0.5, Y := 0.5}}
            Offsets := margin{Top := 0.0, Left := 0.0, Right := 0.0, Bottom := 0.0}
            Alignment := vector2{X := 0.5, Y := 0.5}
            SizeToContent := true
            Widget := overlay:
              Slots := array:
                  overlay_slot:
                      Widget := MiniBGImage : texture_block = texture_block{DefaultImage := UI.Textures.Toei_minigame_BG, DefaultDesiredSize := vector2{X := 1620.0, Y := 1075.0}}
                  overlay_slot:
                    Widget := canvas:
                      Slots := array:
                        canvas_slot:
                          Anchors := anchors{Minimum := vector2{X := 0.5, Y := 0.5}, Maximum := vector2{X := 0.5, Y := 0.5}}
                          Offsets := margin{Top := -255.0}
                          Alignment := vector2{X := 0.5, Y := 0.5}
                          SizeToContent := true
                          Widget := TimerText

  GetCanvas<public>(): canvas =
    Print("input_commands_canvas.GetCanvas()")
    return InputCommandsCanvas

  SetTimer<public>(Time: int)<suspends>: void =
    var _Time: int = Time
    TimerText.SetText(StringToMessage("aaaaa"))
    loop:
      if(_Time <= 0):
        break
      TimerText.SetText(StringToMessage("{_Time}"))
      set _Time -= 1
      Sleep(1.0)
    TimerText.SetText(StringToMessage(""))
