# Deprecated

using { /UnrealEngine.com/Temporary/UI }
using { /UnrealEngine.com/Temporary/SpatialMath }
using {/Fortnite.com/UI}
using {/Verse.org/Assets}
using { UI }
using { Features }
using { Models }

MemoryMessage<localizes>:message = "覚えろ！"
ActionMessage<localizes>:message = "コマンドを入力しろ！"

input_commands_widget<public> := class():
  var Widget: stack_box = stack_box{ Orientation := orientation.Vertical, Slots := array{} }

  Command1: input_command_texture = input_command_texture{}
  Command2: input_command_texture = input_command_texture{}
  Command3: input_command_texture = input_command_texture{}
  Command4: input_command_texture = input_command_texture{}
  Command5: input_command_texture = input_command_texture{}
  Command6: input_command_texture = input_command_texture{}
  Command7: input_command_texture = input_command_texture{}
  Command8: input_command_texture = input_command_texture{}
  Command9: input_command_texture = input_command_texture{}
  Command10: input_command_texture = input_command_texture{}

  Create<public>(): void =
    set Widget = stack_box:
      Orientation := orientation.Vertical
      Slots := array:
        stack_box_slot:
          Widget := text_block{ DefaultText := MemoryMessage }
        stack_box_slot:
          Widget := CreateCommandRow(array { Command1, Command2, Command3, Command4, Command5 })
        stack_box_slot:
          Widget := CreateCommandRow(array { Command6, Command7, Command8, Command9, Command10 })

  ShowAll<public>(CommandList: []input_command): void =
    Command1.ChangeCommand(input_command{Type := input_command_type.UP})
    Command2.ChangeCommand(input_command{Type := input_command_type.DOWN})
    Command3.ChangeCommand(input_command{Type := input_command_type.LEFT})
    Command4.ChangeCommand(input_command{Type := input_command_type.RIGHT})
    Command5.ChangeCommand(input_command{Type := input_command_type.UP})
    Command6.ChangeCommand(input_command{Type := input_command_type.DOWN})
    Command7.ChangeCommand(input_command{Type := input_command_type.LEFT})
    Command8.ChangeCommand(input_command{Type := input_command_type.RIGHT})
    Command9.ChangeCommand(input_command{Type := input_command_type.UP})
    Command10.ChangeCommand(input_command{Type := input_command_type.DOWN})

  CreateCommandRow(Commands: []input_command_texture): stack_box =
    var Slots: []stack_box_slot = array{}
    for(Command : Commands):
      set Slots += array { stack_box_slot{ Widget := Command.GetWidget() } }

    stack_box:
      Orientation := orientation.Horizontal
      Slots := Slots

  GetWidget(): widget =
    return Widget
