using { /UnrealEngine.com/Temporary/UI }
using { /UnrealEngine.com/Temporary/SpatialMath }
using { /Verse.org/Assets }
using { /Verse.org/Colors }
using { UI }
using { Models }

ImageSize: float = 200.0

input_command_texture := class():
  var Widget: widget = color_block{ DefaultColor := NamedColors.White, DefaultOpacity := 1.0, DefaultDesiredSize := vector2{X := ImageSize, Y := ImageSize} }

  ChangeCommand(command: input_command): void =
    case(command.Type):
      input_command_type.UP =>
        set Widget = CreateTexture(Textures.mark_arrow_up)
      input_command_type.DOWN =>
        set Widget = CreateTexture(Textures.mark_arrow_down)
      input_command_type.LEFT =>
        set Widget = CreateTexture(Textures.mark_arrow_left)
      input_command_type.RIGHT =>
        set Widget = CreateTexture(Textures.mark_arrow_right)

  Hide(): void =
    Reset()

  CreateTexture(Texture1: texture): texture_block =
    return texture_block{DefaultImage := Texture1, DefaultDesiredSize := vector2{X := ImageSize, Y := ImageSize} }

  GetWidget(): widget =
    return Widget

  Reset(): widget =
    set Widget = color_block{ DefaultColor := NamedColors.White, DefaultOpacity := 1.0, DefaultDesiredSize := vector2{X := ImageSize, Y := ImageSize} }