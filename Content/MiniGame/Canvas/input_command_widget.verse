using { /UnrealEngine.com/Temporary/UI }
using { /UnrealEngine.com/Temporary/SpatialMath }
using { /Verse.org/Colors }
using { /Fortnite.com/UI }
using { /Verse.org/Simulation }
using { Features }
using { Models }
using { /Verse.org/Assets }
using { UI }

GetRowIndexAnchor(Index: int): float =
  case(Index):
    0 => 3.0
    1 => 1.5
    2 => 0.0
    3 => -1.5
    4 => -3.0
    5 => 3.0
    6 => 1.5
    7 => 0.0
    8 => -1.5
    9 => -3.0
    _ => 0.0

GetColumnIndexAnchor(Index: int): float =
  case(Index):
    0 => 0.5
    1 => 0.5
    2 => 0.5
    3 => 0.5
    4 => 0.5
    5 => -0.8
    6 => -0.8
    7 => -0.8
    8 => -0.8
    9 => -0.8
    _ => 0.0

input_command_widget<public> := class():
  Index<public>: int
  Agent<public>: agent
  Command<public>: input_command

  var Parent: widget = canvas{}
  CommandWidget: texture_block = texture_block{DefaultImage := Textures.transparent, DefaultDesiredSize := vector2{X := ImageSize, Y := ImageSize} }

  BadWidget: texture_block = texture_block{DefaultImage := Textures.transparent, DefaultDesiredSize := vector2{X := ImageSize, Y := ImageSize} }
  WakuWidget: texture_block = texture_block{DefaultImage := Textures.transparent, DefaultDesiredSize := vector2{X := ImageSize, Y := ImageSize} }

  AddWidget<public>(): void = 
    Row := GetRowIndexAnchor(Index)
    Column := GetColumnIndexAnchor(Index)
    if (InPlayer := player[Agent], PlayerUI := GetPlayerUI[InPlayer]):
      set Parent = canvas:
        Slots := array:
          canvas_slot:
            Anchors := anchors{Minimum := vector2{X := 0.5, Y := 0.5}, Maximum := vector2{X := 0.5, Y := 0.5}}
            Offsets := margin{Top := -100.0, Left := -100.0}
            Alignment := vector2{X := Row, Y := Column}
            SizeToContent := true
            Widget := overlay:
              Slots := array:
                overlay_slot:
                  Widget := WakuWidget
                overlay_slot:
                  Widget := BadWidget
                overlay_slot:
                  Widget := CommandWidget
      PlayerUI.AddWidget(Parent, player_ui_slot{ ZOrder := 5 })

  RemoveWidget<public>(): void =
    if (InPlayer := player[Agent], PlayerUI := GetPlayerUI[InPlayer]):
      PlayerUI.RemoveWidget(Parent)

  GetCommandTexture<public>(): texture =
    case(Command.Type):
      input_command_type.UP =>
        return Textures.mark_arrow_up
      input_command_type.DOWN =>
        return Textures.mark_arrow_down
      input_command_type.LEFT =>
        return Textures.mark_arrow_left
      input_command_type.RIGHT =>
        return Textures.mark_arrow_right

  GetCommandOpacityTexture<public>(): texture =
    case(Command.Type):
      input_command_type.UP =>
        return Textures.mark_arrow_up05
      input_command_type.DOWN =>
        return Textures.mark_arrow_down05
      input_command_type.LEFT =>
        return Textures.mark_arrow_left05
      input_command_type.RIGHT =>
        return Textures.mark_arrow_right05

  Show<public>(): void =
    Print("input_command_widget.Show() {Index}")
    CommandWidget.SetImage(GetCommandTexture())
    
  Hide<public>(): void =
    CommandWidget.SetImage(Textures.transparent)

  ShowWaku<public>(): void =
    WakuWidget.SetImage(Textures.waku)

  HideWaku<public>(): void =
    WakuWidget.SetImage(Textures.transparent)

  BadEffect<public>()<suspends>: void =
    BadWidget.SetImage(Textures.Bad)
    CommandWidget.SetImage(GetCommandOpacityTexture())
    Sleep(1.0)
    Hide()
    BadWidget.SetImage(Textures.transparent)

  HideBadEffect<public>(): void =
    BadWidget.SetImage(Textures.transparent)

(Input: []input_command_widget).AddWidget<public>(): void =
  for(Widget : Input):
    Widget.AddWidget()

(Input: []input_command_widget).RemoveWidget<public>(): void =
  for(Widget : Input):
    Widget.RemoveWidget()

(Input: []input_command_widget).Show<public>(Index: int): void =
  if(Target := Input[Index]):
    Target.Show()

(Input: []input_command_widget).BadEffect<public>(Index: int)<suspends>: void =
  if(Target := Input[Index]):
    Target.BadEffect()

(Input: []input_command_widget).ShowWaku<public>(Count: int): void =
  for(Index -> Widget : Input):
    if(Count - 1 >=  Index):
      Widget.ShowWaku()

(Input: []input_command_widget).HideWaku<public>(): void =
  for(Widget : Input):
    Widget.HideWaku()

(Input: []input_command_widget).ShowAll<public>(Count: int): void =
  for(Index -> Widget : Input):
    if(Count - 1 >=  Index):
      Widget.Show()
      Widget.AddWidget()
    
(Input: []input_command_widget).HideAll<public>(): void =
  for(Widget : Input):
    Widget.Hide()
    Widget.HideWaku()
    Widget.HideBadEffect()
