using { /UnrealEngine.com/Temporary/UI }
using { /UnrealEngine.com/Temporary/SpatialMath }
using { /Verse.org/Colors }
using { /Fortnite.com/UI }
using { /Verse.org/Simulation }
using { Features }

waiting_timer_widget<public> := class():
  Agent<public>: agent

  var TimerText : text_block = text_block{ DefaultTextColor := NamedColors.Black }
  var Widget: canvas = canvas{}

  AddWidget<public>(): void = 
    if (InPlayer := player[Agent], PlayerUI := GetPlayerUI[InPlayer]):
      set Widget = canvas:
        Slots := array:
          canvas_slot:
              Anchors := anchors{Minimum := vector2{X := 0.5, Y := 0.5}, Maximum := vector2{X := 0.5, Y := 0.5}}
              Offsets := margin{Top := 0.0, Left := 0.0, Right := 0.0, Bottom := 0.0}
              Alignment := vector2{X := -0.5, Y := 2.0}
              SizeToContent := true
              Widget := TimerText
      PlayerUI.AddWidget(Widget, player_ui_slot{ ZOrder := 2 })

  RemoveWidget<public>(): void =
    if (InPlayer := player[Agent], PlayerUI := GetPlayerUI[InPlayer]):
      PlayerUI.RemoveWidget(Widget)

  SetTimer<public>(Time: int)<suspends>: void =
    var _Time: int = Time
    AddWidget()
    loop:
      if(_Time <= 0):
        break
      TimerText.SetText(StringToMessage("{_Time}"))
      set _Time -= 1
      Sleep(1.0)
    TimerText.SetText(StringToMessage(""))
    RemoveWidget()