
using { /Verse.org/Assets }
using { /Verse.org/Simulation }
using { /Fortnite.com/UI }
using { /Fortnite.com/Devices }
using { /Fortnite.com/Characters }
using { /UnrealEngine.com/Temporary/UI }
using { /UnrealEngine.com/Temporary/SpatialMath }
using { /Verse.org/Simulation }
using { /UnrealEngine.com/Temporary/Diagnostics }
using { /Verse.org/Colors }
using { /Verse.org/Colors/NamedColors }


# See https://dev.epicgames.com/documentation/en-us/uefn/create-your-own-device-in-verse for how to create a verse device.

# A Verse-authored creative device that can be placed in a level
nakamuraza_ui_device := class(creative_device):
    @editable
    JoinNpc: character_device = character_device{}
    #SFXs
    @editable
    SFX_Open : audio_player_device = audio_player_device{}
    @editable
    SFX_Close : audio_player_device = audio_player_device{}
    @editable
    SFX_LevelUp : audio_player_device = audio_player_device{}
    @editable
    SFX_Failed : audio_player_device = audio_player_device{}

    @editable
    PlayerGameTrigger: trigger_device = trigger_device{}

    var NakamuraCanvasPerAgent : [agent]canvas = map{}

    StringToMessage<public><localizes>(Text:string):message="{Text}"

    CloseEvent<public> : event(player) = event(player){}

    OnBegin<override>()<suspends>:void=
        JoinNpc.InteractedWithEvent.Subscribe(OnNakamuraSignalEvented)

    OnPlayerSpawned(Agent : agent):void=
        if (NakamuraCanvas := NakamuraCanvasPerAgent[Agent]):
            return

        if (Player := player[Agent], PlayerUI := GetPlayerUI[Player]):
            NakamuraCanvas : canvas = CreateNakamuraCanvas(Agent)
            if (set NakamuraCanvasPerAgent[Agent] = NakamuraCanvas){}

    OnNakamuraSignalEvented(Agent : agent):void=
        SFX_Open.Play(Agent)
        OnPlayerSpawned(Agent)
        Print("OnNakamuraSignalEvented")
        if (Player := player[Agent], PlayerUI := GetPlayerUI[Player], NakamuraCanvas := NakamuraCanvasPerAgent[Agent]):
            PlayerUI.RemoveWidget(NakamuraCanvas)
            NewCanvas := CreateNakamuraCanvas(Agent)
            PlayerUI.AddWidget(NewCanvas, player_ui_slot{InputMode := ui_input_mode.All})
            if(set NakamuraCanvasPerAgent[Agent] = NewCanvas ){}
            

    OnCloseButtonClicked(WidgetMessage : widget_message):void=
        if(Agent : agent = agent[WidgetMessage.Player]) :
            SFX_Close.Play(Agent)
        if (PlayerUI := GetPlayerUI[WidgetMessage.Player], NakamuraCanvas := NakamuraCanvasPerAgent[WidgetMessage.Player]):
            PlayerUI.RemoveWidget(NakamuraCanvas)
            CloseEvent.Signal(WidgetMessage.Player)

    # ボタン処理
    OnStartButtonClicked(WidgetMessage : widget_message):void=
        if (PlayerUI := GetPlayerUI[WidgetMessage.Player], NakamuraCanvas := NakamuraCanvasPerAgent[WidgetMessage.Player]):
            PlayerGameTrigger.Trigger(WidgetMessage.Player)
            PlayerUI.RemoveWidget(NakamuraCanvas)
            CloseEvent.Signal(WidgetMessage.Player)



    # キャンバス
    CreateNakamuraCanvas(Agent : agent):canvas=
        
        var CanvasSlots : []canvas_slot = array{}
        CenterVector : vector2 = vector2{X:=0.5, Y:=0.5}
        CenterAnchors : anchors = anchors{Minimum := CenterVector, Maximum := CenterVector}
        # Image Stock
        BGImage : texture_block = texture_block{DefaultImage := MAKIMONO2, DefaultDesiredSize := vector2{X := 1214.0, Y := 793.0}}
        
        #========================================== BG ==========================================

        set CanvasSlots += array
        {
            canvas_slot:
                Anchors := CenterAnchors
                Offsets := margin{Top := 0.0,Left := 0.0, Right := 1214.0, Bottom := 793.0}
                Alignment := CenterVector
                SizeToContent := false
                Widget := BGImage
        }

        #========================================== Button ==========================================

        StartButton : button_loud = button_loud{DefaultText := StringToMessage("Start")}
        StartButton.OnClick().Subscribe(OnStartButtonClicked)
        set CanvasSlots += array
        {
            canvas_slot:
                Anchors := CenterAnchors
                Offsets := margin{Top:= 0.0, Left:= -300.0, Right := 100.0, Bottom := 40.0}
                Alignment := CenterVector
                SizeToContent := true
                Widget := StartButton
        }

        #========================================== Close ==========================================

        CloseButton : button_loud = button_loud{DefaultText := StringToMessage("CLOSE")}
        CloseButton.OnClick().Subscribe(OnCloseButtonClicked)
        set CanvasSlots += array
        {
            canvas_slot:
                Anchors := CenterAnchors
                Offsets := margin{Top:= 0.0, Left:= 300.0, Right := 100.0, Bottom := 40.0}
                Alignment := CenterVector
                SizeToContent := true
                Widget := CloseButton
        }

        Canvas : canvas = canvas:
            Slots := CanvasSlots

        return Canvas 